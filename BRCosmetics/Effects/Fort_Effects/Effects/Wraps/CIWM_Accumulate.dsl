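; Kill-reactive weapon wrap controller that progressively switches on seven HeadC material stages as the player's kills accumulate.
; When the pawn is set, it watches the AthenaKills stat, stores the count and sets stage scalars on wrapped meshes.
; It also passes game time into the materials, presumably to time transitions.
; New weapon and weapon-type events reapply stages from the stored count after a short delay.
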(event EventOnPawnSet
  (bind _returnvalue (Utilities|Time|GetGameTimeinSeconds))
  (bind _returnvalue_2 (CustomItemWrap|GetWrappedMeshComponents))
  (bind _asfort_player_pawn (Utilities|Casting|CastToFortPlayerPawn (CustomCharacter|GetAssociatedPlayerPawn)))
  (bind (_statname _statvalue) (Cosmetics|WatchCosmeticStat _asfort_player_pawn "(TagName=\"StatManager.AthenaKills\")")
    (:OnStatChanged
      (Variables|Default|SetKill_Stat _statvalue)
      (if (>= _statvalue 7)
        (Rendering|Material|SetScalarParameterValueonMaterials _returnvalue_2 "HeadC_Game_TimeBPB" _returnvalue)
        (Rendering|Material|SetScalarParameterValueonMaterials _returnvalue_2 "HeadC_HeadF" 1.0)
        (else
          (switch int (Variables|Default|GetKill_Stat)
            (:1
              (Rendering|Material|SetScalarParameterValueonMaterials _returnvalue_2 "HeadC_HeadA" 1.0)
              (Rendering|Material|SetScalarParameterValueonMaterials _returnvalue_2 "HeadC_Game_TimeBPA" _returnvalue))
            (:2
              (Rendering|Material|SetScalarParameterValueonMaterials _returnvalue_2 "HeadC_HeadB" 1.0)
              (Rendering|Material|SetScalarParameterValueonMaterials _returnvalue_2 "HeadC_Game_TimeBPA" _returnvalue))
            (:3
              (Rendering|Material|SetScalarParameterValueonMaterials _returnvalue_2 "HeadC_HeadC" 1.0)
              (Rendering|Material|SetScalarParameterValueonMaterials _returnvalue_2 "HeadC_Game_TimeBPA" _returnvalue))
            (:4
              (Rendering|Material|SetScalarParameterValueonMaterials _returnvalue_2 "HeadC_HeadD" 1.0)
              (Rendering|Material|SetScalarParameterValueonMaterials _returnvalue_2 "HeadC_Game_TimeBPA" _returnvalue))
            (:5
              (Rendering|Material|SetScalarParameterValueonMaterials _returnvalue_2 "HeadC_HeadE" 1.0)
              (Rendering|Material|SetScalarParameterValueonMaterials _returnvalue_2 "HeadC_Game_TimeBPA" _returnvalue))
            (:6
              (Rendering|Material|SetScalarParameterValueonMaterials _returnvalue_2 "HeadC_HeadF" 1.0)
              (Rendering|Material|SetScalarParameterValueonMaterials _returnvalue_2 "HeadC_Game_TimeBPA" _returnvalue)
              (Utilities|FlowControl|Delay 0.3)))))
      (Rendering|Material|SetScalarParameterValueonMaterials _returnvalue_2 "HeadC_HeadG" 1.0))))

(event EventBeginPlay
  (|CallParentFunction))

(event EventOnNewWeapons
  (params (WeaponsAsItems :type Array%20of%20Fort%20Item%20Object%20References))
  (bind _returnvalue (CustomItemWrap|GetWrappedMeshComponents))
  (bind _kill_stat (Variables|Default|GetKill_Stat))
  ;[1 1 1] Updating kill conter on new weapon
  (Utilities|FlowControl|Delay 0.2)
  (if (>= _kill_stat 7)
    (Rendering|Material|SetScalarParameterValueonMaterials _returnvalue "HeadC_HeadG" 1.0)
    (Rendering|Material|SetScalarParameterValueonMaterials _returnvalue "HeadC_HeadF" 1.0)
    (Rendering|Material|SetScalarParameterValueonMaterials _returnvalue "HeadC_HeadE" 1.0)
    (Rendering|Material|SetScalarParameterValueonMaterials _returnvalue "HeadC_HeadD" 1.0)
    (Rendering|Material|SetScalarParameterValueonMaterials _returnvalue "HeadC_HeadC" 1.0)
    (Rendering|Material|SetScalarParameterValueonMaterials _returnvalue "HeadC_HeadB" 1.0)
    (else
      (switch int _kill_stat
        (:1
          _)
        (:2
          _)
        (:3
          _)
        (:4
          _)
        (:5
          _)
        (:6
          _)
        (:7
          _))))
  (Rendering|Material|SetScalarParameterValueonMaterials _returnvalue "HeadC_HeadA" 1.0))

(event EventOnNewPlayerWeaponType
  (params (NewType :type EFortItemType%20Enum :default "WorldItem") (OldType :type EFortItemType%20Enum :default "WorldItem"))
  (bind _returnvalue (CustomItemWrap|GetWrappedMeshComponents))
  (bind _kill_stat (Variables|Default|GetKill_Stat))
  ;[1 1 1] Updating kill conter on new weapon
  (Utilities|FlowControl|Delay 0.2)
  (if (>= _kill_stat 7)
    (Rendering|Material|SetScalarParameterValueonMaterials _returnvalue "HeadC_HeadG" 1.0)
    (Rendering|Material|SetScalarParameterValueonMaterials _returnvalue "HeadC_HeadF" 1.0)
    (Rendering|Material|SetScalarParameterValueonMaterials _returnvalue "HeadC_HeadE" 1.0)
    (Rendering|Material|SetScalarParameterValueonMaterials _returnvalue "HeadC_HeadD" 1.0)
    (Rendering|Material|SetScalarParameterValueonMaterials _returnvalue "HeadC_HeadC" 1.0)
    (Rendering|Material|SetScalarParameterValueonMaterials _returnvalue "HeadC_HeadB" 1.0)
    (else
      (switch int _kill_stat
        (:1
          _)
        (:2
          _)
        (:3
          _)
        (:4
          _)
        (:5
          _)
        (:6
          _)
        (:7
          _))))
  (Rendering|Material|SetScalarParameterValueonMaterials _returnvalue "HeadC_HeadA" 1.0))

(fn ConstructionScript)
