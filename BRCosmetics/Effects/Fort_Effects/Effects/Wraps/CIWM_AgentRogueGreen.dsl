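; Reactive item wrap Blueprint (Agent Rogue Green) that changes its emissive look with the player's kill count.
; Watches the pawn's AthenaKills cosmetic stat, stores the count and last kill time, and pushes both to wrapped mesh materials.
; Kill count shifts an emissive gradient offset, while kill time restarts time-based brightness animations.
; Weapon changes re-push parameters after a short delay, with kill time reset far into the past.
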(event EventBeginPlay
  (Utilities|FlowControl|Delay 0.2))

(event EventOnPawnSet
  ;[1 1 1] geting kill count informamation
  (bind (_statname _statvalue) (Cosmetics|WatchCosmeticStat (CustomCharacter|GetAssociatedPlayerPawn) "(TagName=\"StatManager.AthenaKills\")")
    (:OnStatChanged
      (Variables|Default|SetKill_Count _statvalue)
      (Variables|Default|SetLastKillTime (Utilities|Time|GetGameTimeinSeconds))
      (CallFunction|PushMaterialParameters))))

(event Custom|Push%20Material%20Parameters
  (bind _returnvalue (CustomItemWrap|GetWrappedMeshComponents))
  (bind _lastkilltime (Variables|Default|GetLastKillTime))
  ;[1 1 1] offset for gradent
  (Rendering|Material|SetScalarParameterValueonMaterials _returnvalue "FDecal_Emissive_Gradent_Offset" (+ (/ (Math|Conversions|ToFloat(Integer) (Variables|Default|GetKill_Count)) 6.0) -1.0))
  ;[1 1 1] geting game time information to reset the timer for the animation
  (Rendering|Material|SetScalarParameterValueonMaterials _returnvalue "FDecal_Emissive_Brightness_Time_BP" _lastkilltime)
  (Rendering|Material|SetScalarParameterValueonMaterials _returnvalue "TopS Emissive_01-Brightness-TBP" _lastkilltime))

(event EventOnNewWeapons
  (params (WeaponsAsItems :type Array%20of%20Fort%20Item%20Object%20References))
  ;[1 1 1] Updated kill count on new weapon
  (Utilities|FlowControl|Delay 0.2)
  (Variables|Default|SetLastKillTime -999999999.0)
  (CallFunction|PushMaterialParameters))

(event EventOnNewPlayerWeaponType
  (params (NewType :type EFortItemType%20Enum :default "WorldItem") (OldType :type EFortItemType%20Enum :default "WorldItem"))
  ;[1 1 1] Updated kill count on new weapon
  (Utilities|FlowControl|Delay 0.2)
  (Variables|Default|SetLastKillTime -999999999.0)
  (CallFunction|PushMaterialParameters))

(fn ConstructionScript)
