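; Custom item wrap Blueprint for the AirKringle wrap that makes the wrapped weapon's materials react when it fires.
; On each shot, it switches a decal flipbook frame and raises a top-layer blend opacity on the wrapped mesh components through scalar parameters.
; Two chained retriggerable delays restore the frame shortly after firing and zero the blend opacities later, with repeat shots restarting the pending timer.
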(event EventOnWeaponFired
  (bind _returnvalue (CustomItemWrap|GetWrappedMeshComponents))
  (Rendering|Material|SetScalarParameterValueonMaterials _returnvalue "FDecalV2_FlipBook_Frame #" 1.0)
  (Rendering|Material|SetScalarParameterValueonMaterials _returnvalue "TopS_01_Blend_Opacity---End" 3.0)
  (Utilities|FlowControl|RetriggerableDelay 0.5)
  (Rendering|Material|SetScalarParameterValueonMaterials _returnvalue "FDecalV2_FlipBook_Frame #" 0.0)
  (Utilities|FlowControl|RetriggerableDelay 2.5)
  (Rendering|Material|SetScalarParameterValueonMaterials _returnvalue "TopS_01_Blend_Opacity-TBP" 0)
  (Rendering|Material|SetScalarParameterValueonMaterials _returnvalue "TopS_01_Blend_Opacity---End" 0.0))

(fn ConstructionScript)
