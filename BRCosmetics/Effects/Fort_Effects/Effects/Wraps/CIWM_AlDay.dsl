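; AlDay wrap cosmetic blueprint that reacts when the wrapped weapon fires.
; On the weapon-fired event, it picks a random float between 0 and 1 and writes it to the Ald_HueShift scalar on the wrapped mesh components' materials.
; Each shot therefore sets a new random hue shift value; the construction script is empty.
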(event EventOnWeaponFired
  ;[1 1 1] Turns on Emissive
  (Rendering|Material|SetScalarParameterValueonMaterials (CustomItemWrap|GetWrappedMeshComponents) "Ald_HueShift" (Math|Random|RandomFloatinRange 0.0 1.0)))

(fn ConstructionScript)
