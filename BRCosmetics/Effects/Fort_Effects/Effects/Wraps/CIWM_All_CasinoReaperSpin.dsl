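; Item wrap Blueprint for the CasinoReaperSpin cosmetic that spins its decal each time the weapon fires.
; Each shot raises a clamped speed multiplier, picks one of three speed curves at random and restarts SpinTimeline.
; Timeline updates integrate curve speed over game time into a wrapping rotation value, written to per-category UV rotation parameters on wrapped meshes.
; The multiplier resets when the timeline finishes, so rapid fire spins faster.
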(event EventOnWeaponFired
  (bind _returnvalue (CustomItemWrap|GetWrappedMeshComponents))
  (bind _float_value (Variables|Default|GetFloat_Value))
  (Variables|Default|SetMaxSpeed (Math|Float|Clamp(Float) (+ (Variables|Default|GetMaxSpeed) 0.25) 0.5 2.0))
  (Variables|Default|SetVar (Math|Random|RandomInteger 3))
  (Variables|Default|SetStartTime (Utilities|Time|GetGameTimeinSeconds))
  (bind (_direction _speedvar1 _speedvar2 _speedvar3) (Timeline|SpinTimeline PlayFromStart)
    (:Update
      (Variables|Default|SetFloat_Value (Math|Float|SelectFloat (+ (- (Variables|Default|GetFloat_Value) (* (* (Utilities|Select :Index (Variables|Default|GetVar) :Option%200 _speedvar1 :Option%201 _speedvar2 :Option%202 _speedvar3) (Variables|Default|GetMaxSpeed)) (- (Utilities|Time|GetGameTimeinSeconds) (Variables|Default|GetStartTime)))) 1.0) (- (Variables|Default|GetFloat_Value) (* (* (Utilities|Select :Index (Variables|Default|GetVar) :Option%200 _speedvar1 :Option%201 _speedvar2 :Option%202 _speedvar3) (Variables|Default|GetMaxSpeed)) (- (Utilities|Time|GetGameTimeinSeconds) (Variables|Default|GetStartTime)))) (< (- (Variables|Default|GetFloat_Value) (* (* (Utilities|Select :Index (Variables|Default|GetVar) :Option%200 _speedvar1 :Option%201 _speedvar2 :Option%202 _speedvar3) (Variables|Default|GetMaxSpeed)) (- (Utilities|Time|GetGameTimeinSeconds) (Variables|Default|GetStartTime)))) 1.0)))
      (Variables|Default|SetStartTime (Utilities|Time|GetGameTimeinSeconds))
      (Rendering|Material|SetScalarParameterValueonMaterials _returnvalue "FDecalV2_UV_Rotation_Weapon" _float_value)
      (Rendering|Material|SetScalarParameterValueonMaterials _returnvalue "FDecalV2_UV_Rotation_VehicleSide" _float_value)
      (Rendering|Material|SetScalarParameterValueonMaterials _returnvalue "FDecalV2_UV_Rotation_Vehicle" _float_value)
      (Rendering|Material|SetScalarParameterValueonMaterials _returnvalue "FDecalV2_UV_Rotation_Pickaxe" _float_value)
      (Rendering|Material|SetScalarParameterValueonMaterials _returnvalue "FDecalV2_UV_Rotation_Mech" _float_value)
      (Rendering|Material|SetScalarParameterValueonMaterials _returnvalue "FDecalV2_UV_Rotation_Character" _float_value)
      (Rendering|Material|SetScalarParameterValueonMaterials _returnvalue "FDecalV2_UV_Rotation_Backpack" _float_value))
    (:Finished
      (Variables|Default|SetMaxSpeed 0.8))))

(fn ConstructionScript)
