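; Weapon wrap Blueprint for the DecaSphere cosmetic that makes the wrap material react when the weapon fires.
; On fire, it writes the current game time into the top-side emissive brightness and offset scalar parameters on the wrapped mesh components.
; Passing a timestamp leaves any animation to the material.
; A Ready flag and 0.4-second delay throttle retriggering during rapid fire.
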(event EventOnWeaponFired
  (bind _returnvalue (Utilities|Time|GetGameTimeinSeconds))
  (bind _returnvalue_2 (CustomItemWrap|GetWrappedMeshComponents))
  (if (Variables|Default|Getready)
    (Variables|Default|Setready false)
    (Utilities|FlowControl|Sequence
      (:then_0
        ;[1 1 1] TopsSide V2
        (Rendering|Material|SetScalarParameterValueonMaterials _returnvalue_2 "TopS Emissive_01-Brightness-TBP" _returnvalue)
        (Rendering|Material|SetScalarParameterValueonMaterials _returnvalue_2 "TopS Emissive_01_Offset----TBP" _returnvalue))
      (:then_1
        (Utilities|FlowControl|Delay 0.4)
        (Variables|Default|Setready true)))))

(fn ConstructionScript)
