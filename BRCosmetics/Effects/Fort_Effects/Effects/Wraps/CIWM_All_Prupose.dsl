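; An all-purpose cosmetic wrap Blueprint, named CIWM_All_Prupose, in the Fortnite wraps effects folder.
; It declares an EventOnWeaponFired event and a construction script, both with no logic attached.
; The file shows no other behavior, variables or components.
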(event EventOnWeaponFired)

(fn ConstructionScript)
