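; Reactive weapon wrap Blueprint that animates material parameters on the wrapped meshes when the weapon fires.
; Each shot sets decal emissive timing, then picks one of four elements, avoiding an immediate repeat, for a timeline to drive brightness and warp.
; GradientAnim and SineAnimOFF start timer-driven warp and panning ramps, but nothing here calls them.
; In frontend display, a looping LockerCycle timer only resets a counter.
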(event EventBeginPlay
  (if (Reactive|Frontend|IsCosmeticObjectDisplayedinFrontendContext 0)
    (Utilities|Time|SetTimerbyFunctionName 0 "LockerCycle" 3.0 true)))

(event Custom|LockerCycle
  (if (>= (Variables|Default|GetLockerCounter) 3)
    (Variables|Default|SetLockerCounter 0)))

(event EventOnWeaponFired
  (bind _returnvalue (CustomItemWrap|GetWrappedMeshComponents))
  (bind _returnvalue_2 (CustomItemWrap|GetWrappedMeshComponents))
  (bind _returnvalue_3 (Utilities|Time|GetGameTimeinSeconds))
  (bind _abstrcelement (Variables|Default|GetAbstrcElement))
  (Utilities|FlowControl|Sequence
    (:then_0
      (Rendering|Material|SetScalarParameterValueonMaterials _returnvalue_2 "FDecalV2_Emissive_Brightness_Time_BP" _returnvalue_3)
      (Rendering|Material|SetScalarParameterValueonMaterials _returnvalue_2 "FDecalV2_Emissive_Brightness_Time__End" _returnvalue_3)
      (if (not (Variables|Default|GetKilled))
        (Rendering|Material|SetScalarParameterValueonMaterials _returnvalue_2 "FDecalV2_Emissive_Brightness_Time__End" 0.0)))
    (:then_1
      (Variables|Default|SetAbstrcElement (Variables|Default|GetLastElement))
      (while (== (Variables|Default|GetAbstrcElement) (Variables|Default|GetLastElement))
        (Variables|Default|SetAbstrcElement (Math|Random|RandomIntegerinRange 0 3)))
      (Variables|Default|SetLastElement (Variables|Default|GetAbstrcElement))
      (bind (_direction _alpha _warp) (Timeline|Timeline PlayFromStart)
        (:Update
          (Rendering|Material|SetScalarParameterValueonMaterials _returnvalue "EDRec_Emissive_A_Brightness_Time__End" (Utilities|Select :Index _abstrcelement :Option%200 _alpha :Option%201 0.0 :Option%202 0.0 :Option%203 0.0))
          (Rendering|Material|SetScalarParameterValueonMaterials _returnvalue "EDRec_WarpA_Strength" (Utilities|Select :Index _abstrcelement :Option%200 _warp :Option%201 0.0 :Option%202 0.0 :Option%203 0.0))
          (Rendering|Material|SetScalarParameterValueonMaterials _returnvalue "EDRec_Emissive_B_Brightness_Time__End" (Utilities|Select :Index _abstrcelement :Option%200 0.0 :Option%201 _alpha :Option%202 0.0 :Option%203 0.0))
          (Rendering|Material|SetScalarParameterValueonMaterials _returnvalue "EDRec_WarpB_Strength" (Utilities|Select :Index _abstrcelement :Option%200 0.0 :Option%201 _warp :Option%202 0.0 :Option%203 0.0))
          (Rendering|Material|SetScalarParameterValueonMaterials _returnvalue "EDRec_Emissive_C_Brightness_Time__End" (Utilities|Select :Index _abstrcelement :Option%200 0.0 :Option%201 0.0 :Option%202 _alpha :Option%203 0.0))
          (Rendering|Material|SetScalarParameterValueonMaterials _returnvalue "EDRec_WarpC_Strength" (Utilities|Select :Index _abstrcelement :Option%200 0.0 :Option%201 0.0 :Option%202 _warp :Option%203 0.0))
          (Rendering|Material|SetScalarParameterValueonMaterials _returnvalue "EDRec_Emissive_D_Brightness_Time__End" (Utilities|Select :Index _abstrcelement :Option%200 0.0 :Option%201 0.0 :Option%202 0.0 :Option%203 _alpha))
          (Rendering|Material|SetScalarParameterValueonMaterials _returnvalue "EDRec_WarpD_Strength" (Utilities|Select :Index _abstrcelement :Option%200 0.0 :Option%201 0.0 :Option%202 0.0 :Option%203 _warp)))))))

(event Custom|GradientAnim
  ; Layer Offset
  (Variables|Default|SetReverse false)
  (Variables|Default|SetStartTime (Utilities|Time|GetGameTimeinSeconds))
  (bind _returnvalue (Utilities|Time|SetTimerbyEvent (delegate Custom|SineAnimUpdate) 0.02 true))
  (Variables|Default|SetTimerHandle _returnvalue))

(event Custom|SineAnimOFF
  ; Layer Offset
  (Variables|Default|SetReverse true)
  (Variables|Default|SetStartTime (Utilities|Time|GetGameTimeinSeconds))
  (bind _returnvalue (Utilities|Time|SetTimerbyEvent (delegate Custom|SineAnimUpdate) 0.02 true))
  (Variables|Default|SetTimerHandle _returnvalue))

(event Custom|SineAnimUpdate
  (bind _returnvalue (Utilities|Time|GetGameTimeinSeconds))
  (bind _returnvalue_2 (CustomItemWrap|GetWrappedMeshComponents))
  (bind _starttime (Variables|Default|GetStartTime))
  (bind _returnvalue_3 (- _returnvalue _starttime))
  ; Layer Offset
  (if (>= _returnvalue_3 5.0)
    (Utilities|Time|ClearandInvalidateTimerbyHandle (Variables|Default|GetTimerHandle))
    (if (Variables|Default|GetReverse)
      ;; Start
      (Rendering|Material|SetScalarParameterValueonMaterials _returnvalue_2 "TopS_Em01_Warp_Amount" 0.05)
      (Rendering|Material|SetColorParameterValueonMaterials _returnvalue_2 "TopS_01_Panning_Speed------------" (0.035 0.0 0.0 1.0))
      (else
        ;; Arrive
        (Rendering|Material|SetScalarParameterValueonMaterials _returnvalue_2 "TopS_Em01_Warp_Amount" 0.07)
        (Rendering|Material|SetColorParameterValueonMaterials _returnvalue_2 "TopS_01_Panning_Speed------------" (0.07 0.0 0.0 1.0))))
    (elif (Variables|Default|GetReverse)
      ;; FadeOut
      (Rendering|Material|SetScalarParameterValueonMaterials _returnvalue_2 "TopS_Em01_Warp_Amount" (Math|Float|MapRangeClamped _returnvalue_3 0.0 0.25 0.07 0.05))
      (Rendering|Material|SetColorParameterValueonMaterials _returnvalue_2 "TopS_01_Panning_Speed------------" (Math|Color|MakeColor (Math|Float|MapRangeClamped _returnvalue_3 0.0 0.25 0.07 0.035)))
      (else
        ;; FadeIn
        (Rendering|Material|SetScalarParameterValueonMaterials _returnvalue_2 "FDecalV2_UV_Warp-Strength" (Math|Float|MapRangeClamped _returnvalue_3 0.0 0.25 0.05 0.07))
        (Rendering|Material|SetColorParameterValueonMaterials _returnvalue_2 "TopS_01_Panning_Speed------------" (Math|Color|MakeColor (Math|Float|MapRangeClamped _returnvalue_3 0.0 0.25 0.035 0.07)))))))

(fn SineAnimation)

(fn ConstructionScript)
