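; Fortnite weapon wrap cosmetic modifier that displays the current magazine ammo count on wrap decal materials.
; Pawn, weapon-change and fire events refresh the count, and a reload binding re-reads it after the reload time.
; AmmoUpdate pushes the number and game-time stamps into counter and emissive brightness parameters on wrapped meshes.
; The front end simulates an eight-round clip and plays a day-lighting timeline, while kill-count and offset helpers go uncalled.
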(event EventBeginPlay
  (bind _returnvalue (CustomItemWrap|GetWrappedMeshComponents))
  (bind _returnvalue_2 (CustomCharacter|GetAssociatedPlayerPawn))
  (bind _currentweapon (Variables|/Script/FortniteGame.FortPawn|GetCurrentWeapon _returnvalue_2))
  (if (CustomCosmeticModifier|IsinFrontEnd)
    (Utilities|FlowControl|Sequence
      (:then_0
        (bind (_direction _tod_value) (Timeline|Timeline Play)
          (:Update
            (Rendering|Material|SetScalarParameterValueonMaterials (CustomItemWrap|GetWrappedMeshComponents) "EDRec_Decal_TOD_Day" _tod_value))))
      (:then_1
        _)))
  (Utilities|FlowControl|Delay 0.2)
  (EventDispatchers|BindEventtoOnReloadStarted (Variables|/Script/FortniteGame.FortPawn|GetCurrentWeapon (CustomCharacter|GetAssociatedPlayerPawn)) (delegate Custom|ReloadStarted))
  (if (CustomCosmeticModifier|IsinFrontEnd)
    (Variables|Default|Setper%20Clip 8)
    (bind _output_get (Variables|Default|SetAmmo%20Count 8))
    (Rendering|Material|SetScalarParameterValueonMaterials _returnvalue "FDecalV2_Counter__Number" (Math|Conversions|ToFloat(Integer) _output_get))
    (Rendering|Material|SetScalarParameterValueonMaterials _returnvalue "EDRec_Counter__Number" (Math|Conversions|ToFloat(Integer) _output_get))
    (else
      (Variables|Default|Setper%20Clip (FortWeapon|GetBulletsPerClip _currentweapon))
      (bind _output_get_2 (Variables|Default|SetAmmo%20Count (FortWeapon|GetMagazineAmmoCount _currentweapon)))
      (CallFunction|AmmoUpdate (Math|Conversions|ToFloat(Integer) _output_get_2)))))

(event EventOnWeaponFired
  (bind _returnvalue (CustomCharacter|GetAssociatedPlayerPawn))
  (bind _currentweapon (Variables|/Script/FortniteGame.FortPawn|GetCurrentWeapon _returnvalue))
  (bind _returnvalue_2 (FortWeapon|GetMagazineAmmoCount _currentweapon))
  (if (CustomCosmeticModifier|IsinFrontEnd)
    (bind _output_get (Variables|Default|SetAmmo%20Count (Variables|Default|GetAmmo%20Count)))
    (Utilities|FlowControl|Delay 0.001)
    (bind _output_get_2 (Variables|Default|SetAmmo%20Count (- _output_get 1)))
    (if (== _output_get_2 0)
      (bind _output_get_3 (Variables|Default|SetAmmo%20Count 8))
      (CallFunction|AmmoUpdate (Math|Conversions|ToFloat(Integer) _output_get_3))
      (else
        _))
    (elif (== _returnvalue_2 0)
      _
      (else
        (Variables|Default|SetFireing true)
        (bind _output_get_4 (Variables|Default|SetAmmo%20Count _returnvalue_2))
        (Utilities|FlowControl|Delay 0.001)
        (Variables|Default|SetAmmo%20Count (- _output_get_4 1)))))
  (CallFunction|AmmoUpdate (Math|Conversions|ToFloat(Integer) (Variables|Default|GetAmmo%20Count))))

(event EventOnWeaponChanged
  (params (NewWeapon :type Fort%20Weapon%20Object%20Reference) (OldWeapon :type Fort%20Weapon%20Object%20Reference))
  (bind _returnvalue (CustomItemWrap|GetWrappedMeshComponents))
  (bind _returnvalue_2 (CustomCharacter|GetAssociatedPlayerPawn))
  (bind _currentweapon (Variables|/Script/FortniteGame.FortPawn|GetCurrentWeapon _returnvalue_2))
  (Utilities|FlowControl|Delay 0.2)
  (EventDispatchers|BindEventtoOnReloadStarted (Variables|/Script/FortniteGame.FortPawn|GetCurrentWeapon (CustomCharacter|GetAssociatedPlayerPawn)) (delegate Custom|ReloadStarted))
  (if (CustomCosmeticModifier|IsinFrontEnd)
    (Variables|Default|Setper%20Clip 8)
    (bind _output_get (Variables|Default|SetAmmo%20Count 8))
    (Rendering|Material|SetScalarParameterValueonMaterials _returnvalue "FDecalV2_Counter__Number" (Math|Conversions|ToFloat(Integer) _output_get))
    (Rendering|Material|SetScalarParameterValueonMaterials _returnvalue "EDRec_Counter__Number" (Math|Conversions|ToFloat(Integer) _output_get))
    (else
      (Variables|Default|Setper%20Clip (FortWeapon|GetBulletsPerClip _currentweapon))
      (bind _output_get_2 (Variables|Default|SetAmmo%20Count (FortWeapon|GetMagazineAmmoCount _currentweapon)))
      (CallFunction|AmmoUpdate (Math|Conversions|ToFloat(Integer) _output_get_2)))))

(event EventOnPawnSet
  (bind _returnvalue (CustomItemWrap|GetWrappedMeshComponents))
  (bind _returnvalue_2 (CustomCharacter|GetAssociatedPlayerPawn))
  (bind _currentweapon (Variables|/Script/FortniteGame.FortPawn|GetCurrentWeapon _returnvalue_2))
  (bind _returnvalue_3 (FortWeapon|GetMagazineAmmoCount _currentweapon))
  (bind _returnvalue_4 (CustomCharacter|GetAssociatedPlayerPawn))
  (Utilities|FlowControl|Sequence
    (:then_0
      (Utilities|FlowControl|Delay 0.2)
      (EventDispatchers|BindEventtoOnReloadStarted (Variables|/Script/FortniteGame.FortPawn|GetCurrentWeapon _returnvalue_4) (delegate Custom|ReloadStarted))
      (if (CustomCosmeticModifier|IsinFrontEnd)
        (Variables|Default|Setper%20Clip 8)
        (bind _output_get (Variables|Default|SetAmmo%20Count 8))
        (Rendering|Material|SetScalarParameterValueonMaterials _returnvalue "FDecalV2_Counter__Number" (Math|Conversions|ToFloat(Integer) _output_get))
        (Rendering|Material|SetScalarParameterValueonMaterials _returnvalue "EDRec_Counter__Number" (Math|Conversions|ToFloat(Integer) _output_get))
        (else
          (Variables|Default|Setper%20Clip (FortWeapon|GetBulletsPerClip _currentweapon))
          (bind _output_get_2 (Variables|Default|SetAmmo%20Count _returnvalue_3))
          (CallFunction|AmmoUpdate (Math|Conversions|ToFloat(Integer) _output_get_2)))))
    (:then_1
      (Cosmetics|WatchCosmeticStat _returnvalue_4 "(TagName=\"StatManager.GainedHealthTimes\")" 0 false
        (:OnStatChanged
          _)))
    (:then_2
      (Cosmetics|WatchCosmeticStat (CustomCharacter|GetAssociatedPlayerPawn) "(TagName=\"StatManager.GainedShieldTimes\")" 0 false
        (:OnStatChanged
          _))))
  (Variables|Default|SetFireing true)
  (bind _output_get_3 (Variables|Default|SetAmmo%20Count _returnvalue_3))
  (Utilities|FlowControl|Delay 0.001)
  (Variables|Default|SetAmmo%20Count (- _output_get_3 1))
  (CallFunction|AmmoUpdate (Math|Conversions|ToFloat(Integer) (Variables|Default|GetAmmo%20Count))))

(event EventOnNewWeapons
  (params (WeaponsAsItems :type Array%20of%20Fort%20Item%20Object%20References))
  (bind _returnvalue (CustomItemWrap|GetWrappedMeshComponents))
  (bind _returnvalue_2 (CustomCharacter|GetAssociatedPlayerPawn))
  (bind _currentweapon (Variables|/Script/FortniteGame.FortPawn|GetCurrentWeapon _returnvalue_2))
  (Utilities|FlowControl|Delay 0.2)
  (EventDispatchers|BindEventtoOnReloadStarted (Variables|/Script/FortniteGame.FortPawn|GetCurrentWeapon (CustomCharacter|GetAssociatedPlayerPawn)) (delegate Custom|ReloadStarted))
  (if (CustomCosmeticModifier|IsinFrontEnd)
    (Variables|Default|Setper%20Clip 8)
    (bind _output_get (Variables|Default|SetAmmo%20Count 8))
    (Rendering|Material|SetScalarParameterValueonMaterials _returnvalue "FDecalV2_Counter__Number" (Math|Conversions|ToFloat(Integer) _output_get))
    (Rendering|Material|SetScalarParameterValueonMaterials _returnvalue "EDRec_Counter__Number" (Math|Conversions|ToFloat(Integer) _output_get))
    (else
      (Variables|Default|Setper%20Clip (FortWeapon|GetBulletsPerClip _currentweapon))
      (bind _output_get_2 (Variables|Default|SetAmmo%20Count (FortWeapon|GetMagazineAmmoCount _currentweapon)))
      (CallFunction|AmmoUpdate (Math|Conversions|ToFloat(Integer) _output_get_2)))))

(event Custom|ReloadStarted
  (params (ReloadTime :type Float%20%28single-precision%29) (ReloadType :type EFortWeaponReloadType%20Enum :default "ReloadWholeClip"))
  (CallFunction|AmmoUpdate (Math|Conversions|ToFloat(Integer) (Variables|Default|GetAmmo%20Count)))
  (Utilities|FlowControl|Delay ReloadTime)
  (if (not (CustomCosmeticModifier|IsinFrontEnd))
    (bind _output_get (Variables|Default|SetAmmo%20Count (FortWeapon|GetMagazineAmmoCount (Variables|/Script/FortniteGame.FortPawn|GetCurrentWeapon (CustomCharacter|GetAssociatedPlayerPawn)))))
    (CallFunction|AmmoUpdate (Math|Conversions|ToFloat(Integer) _output_get))))

(event EventOnNewPlayerWeaponType
  (params (NewType :type EFortItemType%20Enum :default "WorldItem") (OldType :type EFortItemType%20Enum :default "WorldItem"))
  (bind _returnvalue (CustomItemWrap|GetWrappedMeshComponents))
  (bind _returnvalue_2 (CustomCharacter|GetAssociatedPlayerPawn))
  (bind _currentweapon (Variables|/Script/FortniteGame.FortPawn|GetCurrentWeapon _returnvalue_2))
  (Utilities|FlowControl|Delay 0.2)
  (EventDispatchers|BindEventtoOnReloadStarted (Variables|/Script/FortniteGame.FortPawn|GetCurrentWeapon (CustomCharacter|GetAssociatedPlayerPawn)) (delegate Custom|ReloadStarted))
  (if (CustomCosmeticModifier|IsinFrontEnd)
    (Variables|Default|Setper%20Clip 8)
    (bind _output_get (Variables|Default|SetAmmo%20Count 8))
    (Rendering|Material|SetScalarParameterValueonMaterials _returnvalue "FDecalV2_Counter__Number" (Math|Conversions|ToFloat(Integer) _output_get))
    (Rendering|Material|SetScalarParameterValueonMaterials _returnvalue "EDRec_Counter__Number" (Math|Conversions|ToFloat(Integer) _output_get))
    (else
      (Variables|Default|Setper%20Clip (FortWeapon|GetBulletsPerClip _currentweapon))
      (bind _output_get_2 (Variables|Default|SetAmmo%20Count (FortWeapon|GetMagazineAmmoCount _currentweapon)))
      (CallFunction|AmmoUpdate (Math|Conversions|ToFloat(Integer) _output_get_2)))))

(fn AmmoUpdate
  (params (Ammo :type Float%20%28single-precision%29) (SetColor :type Linear%20Color%20Structure :default "(R=1.250000,G=0.834369,B=0.000000,A=1.000000)"))
  (bind _returnvalue (CustomItemWrap|GetWrappedMeshComponents))
  (Rendering|Material|SetScalarParameterValueonMaterials _returnvalue "FDecalV2_Counter__Number" Ammo)
  (Rendering|Material|SetScalarParameterValueonMaterials _returnvalue "EDRec_Counter__Number" Ammo)
  (Rendering|Material|SetScalarParameterValueonMaterials _returnvalue "FDecalV2_Counter_Brightness_Time_BP" (Utilities|Time|GetGameTimeinSeconds))
  (Rendering|Material|SetScalarParameterValueonMaterials _returnvalue "EDRec_Emissive_A_Brightness_Time_BP" (Utilities|Time|GetGameTimeinSeconds))
  (Rendering|Material|SetScalarParameterValueonMaterials _returnvalue "EDRec_Emissive_B_Brightness_Time_BP" (Utilities|Time|GetGameTimeinSeconds))
  (Rendering|Material|SetScalarParameterValueonMaterials _returnvalue "EDRec_Emissive_C_Brightness_Time_BP" (Utilities|Time|GetGameTimeinSeconds)))

(fn Kill_Count_Update
  (params (Kill%20Count :type Float%20%28single-precision%29) (InLinearColor :type Linear%20Color%20Structure) (DelaySettings :type Submix%20Effect%20Delay%20Settings%20Structure%20%28by%20ref%29))
  (bind _returnvalue (CustomItemWrap|GetWrappedMeshComponents))
  (bind _returnvalue_2 (Math|Conversions|ToVector(LinearColor) InLinearColor))
  (Rendering|Material|SetVectorParameterValueonMaterials _returnvalue "EDRec_Emissive_A_Color-" _returnvalue_2)
  (Rendering|Material|SetScalarParameterValueonMaterials _returnvalue "EDRec_Counter__Number" Kill%20Count)
  (Rendering|Material|SetScalarParameterValueonMaterials _returnvalue "EDRec_Emissive_A_Brightness_Time_BP" (Utilities|Time|GetGameTimeinSeconds))
  (Rendering|Material|SetVectorParameterValueonMaterials _returnvalue "EDRec_Emissive_B_Color-" _returnvalue_2)
  (for _index (range (Variables|Default|GetKill%20Loop) 5)
    (Rendering|Material|SetScalarParameterValueonMaterials _returnvalue "EDRec_Emissive_A_Brightness_Time_BP" (Utilities|Time|GetGameTimeinSeconds))
    (Audio|Effects|Delay|SetDelayLength DelaySettings 0.5)
    (Variables|Default|SetKill%20Loop (+ _index 1))))

(fn ConstructionScript)

(fn Weapon_Offsets
  (params (Offset_End_Name :type Name) (Offset_End_Value :type Float%20%28double-precision%29) (Offset_Start_Name :type Name) (Offset_State_Value :type Float%20%28double-precision%29))
  (bind _returnvalue (CustomItemWrap|GetWrappedMeshComponents))
  (Rendering|Material|SetVectorParameterValueonMaterials _returnvalue Offset_End_Name (Math|Vector|MakeVector 0.0 0.0 Offset_End_Value))
  (Rendering|Material|SetVectorParameterValueonMaterials _returnvalue Offset_Start_Name (Math|Vector|MakeVector 0.0 0.0 Offset_State_Value)))
