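; Item wrap Blueprint that drives an ammo dial and damage-dealt decals on wrapped meshes through material parameters.
; Weapon fire, swap and reload events feed magazine fill into AmmoUpdate, which lerps dial rotations, animated by a timeline on reload.
; DamageDealt stat changes randomize decal offset, rotation and shuffled flipbook frame per item type, throttled by a pause flag.
; In the front end, firing simulates an eight-round clip with automatic refill and periodic damage reactions.
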(event EventBeginPlay
  (bind _returnvalue (CustomCharacter|GetAssociatedPlayerPawn))
  (bind _currentweapon (Variables|/Script/FortniteGame.FortPawn|GetCurrentWeapon _returnvalue))
  (Utilities|Array|Shuffle (Variables|Default|GetNumber%20of%20Frames_random))
  (if (CustomCosmeticModifier|IsinFrontEnd)
    (Utilities|FlowControl|Sequence
      (:then_0
        (bind (_direction _tod_value) (Timeline|Timeline PlayFromStart)
          (:Update
            (Rendering|Material|SetScalarParameterValueonMaterials (CustomItemWrap|GetWrappedMeshComponents) "EDRec_Decal_TOD_Day" _tod_value))))
      (:then_1
        _)))
  (Utilities|FlowControl|Delay 0.2)
  (EventDispatchers|BindEventtoOnReloadStarted (Variables|/Script/FortniteGame.FortPawn|GetCurrentWeapon (CustomCharacter|GetAssociatedPlayerPawn)) (delegate Custom|ReloadStarted))
  (if (CustomCosmeticModifier|IsinFrontEnd)
    (Variables|EnviromentVarb|Setper%20Clip 8)
    (Variables|EnviromentVarb|SetAmmo%20Count 8)
    (else
      (bind _output_get (Variables|EnviromentVarb|Setper%20Clip (FortWeapon|GetBulletsPerClip _currentweapon)))
      (bind _output_get_2 (Variables|EnviromentVarb|SetAmmo%20Count (FortWeapon|GetMagazineAmmoCount _currentweapon)))
      (CallFunction|AmmoUpdate (/ (Math|Conversions|ToFloat(Integer) _output_get_2) _output_get)))))

(event EventOnWeaponFired
  (bind _returnvalue (CustomCharacter|GetAssociatedPlayerPawn))
  (bind _currentweapon (Variables|/Script/FortniteGame.FortPawn|GetCurrentWeapon _returnvalue))
  (bind _returnvalue_2 (FortWeapon|GetMagazineAmmoCount _currentweapon))
  (bind _returnvalue_3 (CustomItemWrap|GetWrappedMeshComponents))
  (bind _returnvalue_4 (Utilities|Time|GetGameTimeinSeconds))
  (Utilities|FlowControl|Sequence
    (:then_0
      (if (CustomCosmeticModifier|IsinFrontEnd)
        (if (not (Variables|EnviromentVarb|GetPause))
          (Utilities|FlowControl|Sequence
            (:then_0
              (if (CustomCosmeticModifier|IsinFrontEnd)
                (bind _output_get (Variables|EnviromentVarb|SetIterationCount (+ (Variables|EnviromentVarb|GetIterationCount) 1)))
                (if (>= _output_get 5)
                  (if (== (Variables|EnviromentVarb|GetIterationCount) 6)
                    (Variables|EnviromentVarb|SetIterationCount 0))
                  (CallFunction|DamageDelt))))
            (:then_1
              (bind _output_get_2 (Variables|EnviromentVarb|SetAmmo%20Count (Variables|EnviromentVarb|GetAmmo%20Count)))
              (Utilities|FlowControl|Delay 0.001)
              (bind _output_get_3 (Variables|EnviromentVarb|SetAmmo%20Count (- _output_get_2 1)))
              (if (== _output_get_3 0)
                (bind (_direction _tod_value) (Timeline|Timeline_1 PlayFromStart)
                  (:Update
                    (bind _output_get_4 (Variables|EnviromentVarb|SetAmmo%20Count 8))
                    (CallFunction|AmmoUpdate (Math|Float|Lerp 0.0 (/ (Math|Conversions|ToFloat(Integer) _output_get_4) 8.0) _tod_value))
                    (CallFunction|DecalReactive))
                  (:Finished
                    (Variables|EnviromentVarb|SetPause false))
                  (:reload
                    (Variables|EnviromentVarb|SetPause true)
                    (if (Variables|Default|GetReload%20Rotation%20Activation)
                      (Rendering|Material|SetScalarParameterValueonMaterials _returnvalue_3 "EDRec_TimeFrom_BP_DecalA" _returnvalue_4)
                      (Rendering|Material|SetScalarParameterValueonMaterials _returnvalue_3 "EDRec_TimeFrom_BP_DecalB" _returnvalue_4)
                      (Rendering|Material|SetScalarParameterValueonMaterials _returnvalue_3 "EDRec_TimeFrom_BP_DecalC" _returnvalue_4)
                      (Rendering|Material|SetScalarParameterValueonMaterials _returnvalue_3 "EDRec_TimeFrom_BP_DecalD" _returnvalue_4))))
                (else
                  (CallFunction|AmmoUpdate (/ (Math|Conversions|ToFloat(Integer) (Variables|EnviromentVarb|GetAmmo%20Count)) 8.0)))))))
        (elif (== _returnvalue_2 0)
          _
          (else
            (Variables|EnviromentVarb|SetFireing true)
            (bind _output_get_5 (Variables|EnviromentVarb|SetAmmo%20Count _returnvalue_2))
            (Utilities|FlowControl|Delay 0.001)
            (bind _output_get_6 (Variables|EnviromentVarb|SetAmmo%20Count (- _output_get_5 1)))
            (Variables|EnviromentVarb|SetLast%20Round _output_get_6)
            (CallFunction|AmmoUpdate (/ (Variables|EnviromentVarb|GetLast%20Round) (Math|Conversions|ToFloat(Integer) (Variables|EnviromentVarb|Getper%20Clip))))))))))

(event EventOnWeaponChanged
  (params (NewWeapon :type Fort%20Weapon%20Object%20Reference) (OldWeapon :type Fort%20Weapon%20Object%20Reference))
  (bind _returnvalue (CustomCharacter|GetAssociatedPlayerPawn))
  (bind _currentweapon (Variables|/Script/FortniteGame.FortPawn|GetCurrentWeapon _returnvalue))
  (Utilities|FlowControl|Delay 0.2)
  (EventDispatchers|BindEventtoOnReloadStarted (Variables|/Script/FortniteGame.FortPawn|GetCurrentWeapon (CustomCharacter|GetAssociatedPlayerPawn)) (delegate Custom|ReloadStarted))
  (if (CustomCosmeticModifier|IsinFrontEnd)
    (Variables|EnviromentVarb|Setper%20Clip 8)
    (Variables|EnviromentVarb|SetAmmo%20Count 8)
    (else
      (bind _output_get (Variables|EnviromentVarb|Setper%20Clip (FortWeapon|GetBulletsPerClip _currentweapon)))
      (bind _output_get_2 (Variables|EnviromentVarb|SetAmmo%20Count (FortWeapon|GetMagazineAmmoCount _currentweapon)))
      (CallFunction|AmmoUpdate (/ (Math|Conversions|ToFloat(Integer) _output_get_2) _output_get)))))

(event EventOnPawnSet
  (bind _returnvalue (CustomCharacter|GetAssociatedPlayerPawn))
  (bind _currentweapon (Variables|/Script/FortniteGame.FortPawn|GetCurrentWeapon _returnvalue))
  (bind _returnvalue_2 (FortWeapon|GetMagazineAmmoCount _currentweapon))
  (bind _returnvalue_3 (CustomCharacter|GetAssociatedPlayerPawn))
  (Utilities|FlowControl|Sequence
    (:then_0
      (Utilities|FlowControl|Delay 0.2)
      (EventDispatchers|BindEventtoOnReloadStarted (Variables|/Script/FortniteGame.FortPawn|GetCurrentWeapon _returnvalue_3) (delegate Custom|ReloadStarted))
      (if (CustomCosmeticModifier|IsinFrontEnd)
        (Variables|EnviromentVarb|Setper%20Clip 8)
        (Variables|EnviromentVarb|SetAmmo%20Count 8)
        (else
          (bind _output_get (Variables|EnviromentVarb|Setper%20Clip (FortWeapon|GetBulletsPerClip _currentweapon)))
          (bind _output_get_2 (Variables|EnviromentVarb|SetAmmo%20Count _returnvalue_2))
          (CallFunction|AmmoUpdate (/ (Math|Conversions|ToFloat(Integer) _output_get_2) _output_get)))))
    (:then_1
      (Cosmetics|WatchCosmeticStat _returnvalue_3 "(TagName=\"StatManager.GainedHealthTimes\")"
        (:OnStatChanged
          _)))
    (:then_2
      (Cosmetics|WatchCosmeticStat (CustomCharacter|GetAssociatedPlayerPawn) "(TagName=\"StatManager.GainedShieldTimes\")"
        (:OnStatChanged
          _)))
    (:then_3
      (Cosmetics|WatchCosmeticStat (CustomCharacter|GetAssociatedPlayerPawn) "(TagName=\"StatManager.DamageDealt\")" 0 false
        (:OnStatChanged
          (if (not (Variables|EnviromentVarb|GetPause))
            (Variables|EnviromentVarb|SetPause true)
            (CallFunction|RandomSet "FDecalV2_FlipBook_Frame #" "FDecalaV2_Opacity_Time_BP_Z" "FDecalV2_UV_Offset_Weapon" "FDecalV2_UV_Rotation_Weapon" -0.38 0.2 -0.01 0.01 0.1 -0.15)
            (CallFunction|RandomSet "0" "0" "FDecalV2_UV_Offset_Pickaxe" "FDecalV2_UV_Rotation_Pickaxe" -0.3 0.15 0.2 -0.2 0.1 -0.15)
            (CallFunction|RandomSet "0" "0" "FDecalV2_UV_Offset_Backpack" "FDecalV2_UV_Rotation_Backpack" -0.15 0.15 -0.25 0.25 0.1 -0.15)
            (CallFunction|RandomSet "0" "0" "FDecalV2_UV_Offset_Character" "FDecalV2_UV_Rotation_Character" -0.15 0.15 -0.2 0.28 0.1 -0.15)
            (CallFunction|RandomSet "0" "0" "FDecalV2_UV_Offset_Vehicle" "FDecalV2_UV_Rotation_Vehicle" 0.15 0.4 -0.15 0.15 0.1 -0.15)
            (CallFunction|RandomSet "0" "0" "FDecalV2_UV_Offset_VehicleSide" "FDecalV2_UV_Rotation_VehicleSide" 0.3 0.42 -0.1 -0.13 0.1 -0.15)
            (CallFunction|RandomSet "0" "0" "FDecalV2_UV_Offset_Mech" "FDecalV2_UV_Rotation_Right_Mech" -0.1 0.1 0.32 0.2 0.1 -0.15)
            (Utilities|FlowControl|Delay 0.1)
            (CallFunction|DamageDeltTrack)
            (Variables|EnviromentVarb|SetPause false))))))
  (Variables|EnviromentVarb|SetFireing true)
  (bind _output_get_3 (Variables|EnviromentVarb|SetAmmo%20Count _returnvalue_2))
  (Utilities|FlowControl|Delay 0.001)
  (bind _output_get_4 (Variables|EnviromentVarb|SetAmmo%20Count (- _output_get_3 1)))
  (Variables|EnviromentVarb|SetLast%20Round _output_get_4)
  (CallFunction|AmmoUpdate (/ (Variables|EnviromentVarb|GetLast%20Round) (Math|Conversions|ToFloat(Integer) (Variables|EnviromentVarb|Getper%20Clip)))))

(event EventOnNewWeapons
  (params (WeaponsAsItems :type Array%20of%20Fort%20Item%20Object%20References))
  (bind _returnvalue (CustomCharacter|GetAssociatedPlayerPawn))
  (bind _currentweapon (Variables|/Script/FortniteGame.FortPawn|GetCurrentWeapon _returnvalue))
  (Utilities|FlowControl|Delay 0.2)
  (EventDispatchers|BindEventtoOnReloadStarted (Variables|/Script/FortniteGame.FortPawn|GetCurrentWeapon (CustomCharacter|GetAssociatedPlayerPawn)) (delegate Custom|ReloadStarted))
  (if (CustomCosmeticModifier|IsinFrontEnd)
    (Variables|EnviromentVarb|Setper%20Clip 8)
    (Variables|EnviromentVarb|SetAmmo%20Count 8)
    (else
      (bind _output_get (Variables|EnviromentVarb|Setper%20Clip (FortWeapon|GetBulletsPerClip _currentweapon)))
      (bind _output_get_2 (Variables|EnviromentVarb|SetAmmo%20Count (FortWeapon|GetMagazineAmmoCount _currentweapon)))
      (CallFunction|AmmoUpdate (/ (Math|Conversions|ToFloat(Integer) _output_get_2) _output_get)))))

(event Custom|ReloadStarted
  (params (ReloadTime :type Float%20%28single-precision%29) (ReloadType :type EFortWeaponReloadType%20Enum :default "ReloadWholeClip"))
  (bind _returnvalue (CustomItemWrap|GetWrappedMeshComponents))
  (bind _returnvalue_2 (Utilities|Time|GetGameTimeinSeconds))
  (CallFunction|AmmoUpdate (/ (Variables|EnviromentVarb|GetLast%20Round) (Math|Conversions|ToFloat(Integer) (Variables|EnviromentVarb|Getper%20Clip))))
  (Utilities|FlowControl|Delay ReloadTime)
  (if (not (CustomCosmeticModifier|IsinFrontEnd))
    (bind (_direction _tod_value) (Timeline|Timeline_0 PlayFromStart)
      (:Update
        (bind _output_get (Variables|EnviromentVarb|SetAmmo%20Count (FortWeapon|GetMagazineAmmoCount (Variables|/Script/FortniteGame.FortPawn|GetCurrentWeapon (CustomCharacter|GetAssociatedPlayerPawn)))))
        (CallFunction|AmmoUpdate (Math|Float|Lerp (/ (Variables|EnviromentVarb|GetLast%20Round) (Math|Conversions|ToFloat(Integer) (Variables|EnviromentVarb|Getper%20Clip))) (/ (Math|Conversions|ToFloat(Integer) _output_get) (Math|Conversions|ToFloat(Integer) (Variables|EnviromentVarb|Getper%20Clip))) _tod_value)))
      (:reload
        (if (Variables|Default|GetReload%20Rotation%20Activation)
          (Rendering|Material|SetScalarParameterValueonMaterials _returnvalue "EDRec_TimeFrom_BP_DecalA" _returnvalue_2)
          (Rendering|Material|SetScalarParameterValueonMaterials _returnvalue "EDRec_TimeFrom_BP_DecalB" _returnvalue_2)
          (Rendering|Material|SetScalarParameterValueonMaterials _returnvalue "EDRec_TimeFrom_BP_DecalC" _returnvalue_2)
          (Rendering|Material|SetScalarParameterValueonMaterials _returnvalue "EDRec_TimeFrom_BP_DecalD" _returnvalue_2))))))

(event EventOnNewPlayerWeaponType
  (params (NewType :type EFortItemType%20Enum :default "WorldItem") (OldType :type EFortItemType%20Enum :default "WorldItem"))
  (bind _returnvalue (CustomCharacter|GetAssociatedPlayerPawn))
  (bind _currentweapon (Variables|/Script/FortniteGame.FortPawn|GetCurrentWeapon _returnvalue))
  (Utilities|FlowControl|Delay 0.2)
  (EventDispatchers|BindEventtoOnReloadStarted (Variables|/Script/FortniteGame.FortPawn|GetCurrentWeapon (CustomCharacter|GetAssociatedPlayerPawn)) (delegate Custom|ReloadStarted))
  (if (CustomCosmeticModifier|IsinFrontEnd)
    (Variables|EnviromentVarb|Setper%20Clip 8)
    (Variables|EnviromentVarb|SetAmmo%20Count 8)
    (else
      (bind _output_get (Variables|EnviromentVarb|Setper%20Clip (FortWeapon|GetBulletsPerClip _currentweapon)))
      (bind _output_get_2 (Variables|EnviromentVarb|SetAmmo%20Count (FortWeapon|GetMagazineAmmoCount _currentweapon)))
      (CallFunction|AmmoUpdate (/ (Math|Conversions|ToFloat(Integer) _output_get_2) _output_get)))))

(event Custom|Damage%20Delt
  (if (not (Variables|EnviromentVarb|GetPause))
    (Variables|EnviromentVarb|SetPause true)
    (CallFunction|RandomSet "FDecalV2_FlipBook_Frame #" "FDecalaV2_Opacity_Time_BP_Z" "FDecalV2_UV_Offset_Weapon" "FDecalV2_UV_Rotation_Weapon" -0.38 0.2 -0.01 0.01 0.1 -0.15)
    (CallFunction|RandomSet "0" "0" "FDecalV2_UV_Offset_Pickaxe" "FDecalV2_UV_Rotation_Pickaxe" -0.3 0.15 0.2 -0.2 0.1 -0.15)
    (CallFunction|RandomSet "0" "0" "FDecalV2_UV_Offset_Backpack" "FDecalV2_UV_Rotation_Backpack" -0.15 0.15 -0.25 0.25 0.1 -0.15)
    (CallFunction|RandomSet "0" "0" "FDecalV2_UV_Offset_Character" "FDecalV2_UV_Rotation_Character" -0.15 0.15 -0.2 0.28 0.1 -0.15)
    (CallFunction|RandomSet "0" "0" "FDecalV2_UV_Offset_Vehicle" "FDecalV2_UV_Rotation_Vehicle" 0.15 0.4 -0.15 0.15 0.1 -0.15)
    (CallFunction|RandomSet "0" "0" "FDecalV2_UV_Offset_VehicleSide" "FDecalV2_UV_Rotation_VehicleSide" 0.3 0.42 -0.1 -0.13 0.1 -0.15)
    (CallFunction|RandomSet "0" "0" "FDecalV2_UV_Offset_Mech" "FDecalV2_UV_Rotation_Right_Mech" -0.1 0.1 0.32 0.2 0.1 -0.15)
    (Utilities|FlowControl|Delay 0.1)
    (CallFunction|DamageDeltTrack)
    (Variables|EnviromentVarb|SetPause false)))

(event Custom|Damage_DeltTrack
  (bind _output_get (Variables|EnviromentVarb|SetDamage (+ (Variables|EnviromentVarb|GetDamage) 1)))
  (if (== _output_get (Variables|Default|GetNumber%20of%20Frames))
    (Utilities|Array|Shuffle (Variables|Default|GetNumber%20of%20Frames_random))
    (Variables|EnviromentVarb|SetDamage -1)))

(event Custom|Decal%20reactive)

(fn AmmoUpdate
  (params (Ammo :type Float%20%28single-precision%29) (SetColor :type Linear%20Color%20Structure :default "(R=1.250000,G=0.834369,B=0.000000,A=1.000000)"))
  (bind _returnvalue (CustomItemWrap|GetWrappedMeshComponents))
  (bind _rotation_name (Variables|Default|GetRotation-Name))
  (bind _rotation_start (Variables|Default|GetRotation-Start-%20%20))
  (bind _rotation_end (Variables|Default|GetRotation-End-))
  (Rendering|Material|SetScalarParameterValueonMaterials _returnvalue (Utilities|Array|Get(acopy) _rotation_name) (Math|Float|Lerp (Utilities|Array|Get(acopy) _rotation_end) (Utilities|Array|Get(acopy) _rotation_start) Ammo))
  (Rendering|Material|SetScalarParameterValueonMaterials _returnvalue (Utilities|Array|Get(acopy) _rotation_name 1) (Math|Float|Lerp (Utilities|Array|Get(acopy) _rotation_end 1) (Utilities|Array|Get(acopy) _rotation_start 1) Ammo))
  (Rendering|Material|SetScalarParameterValueonMaterials _returnvalue (Utilities|Array|Get(acopy) _rotation_name 2) (Math|Float|Lerp (Utilities|Array|Get(acopy) _rotation_end 2) (Utilities|Array|Get(acopy) _rotation_start 2) Ammo))
  (Rendering|Material|SetScalarParameterValueonMaterials _returnvalue (Utilities|Array|Get(acopy) _rotation_name 3) (Math|Float|Lerp (Utilities|Array|Get(acopy) _rotation_end 3) (Utilities|Array|Get(acopy) _rotation_start 3) Ammo))
  (Rendering|Material|SetScalarParameterValueonMaterials _returnvalue (Utilities|Array|Get(acopy) _rotation_name 4) (Math|Float|Lerp (Utilities|Array|Get(acopy) _rotation_end 4) (Utilities|Array|Get(acopy) _rotation_start 4) Ammo))
  (Rendering|Material|SetScalarParameterValueonMaterials _returnvalue (Utilities|Array|Get(acopy) _rotation_name 5) (Math|Float|Lerp (Utilities|Array|Get(acopy) _rotation_end 5) (Utilities|Array|Get(acopy) _rotation_start 5) Ammo))
  (Rendering|Material|SetScalarParameterValueonMaterials _returnvalue (Utilities|Array|Get(acopy) _rotation_name 6) (Math|Float|Lerp (Utilities|Array|Get(acopy) _rotation_end 6) (Utilities|Array|Get(acopy) _rotation_start 6) Ammo))
  (Rendering|Material|SetScalarParameterValueonMaterials _returnvalue (Variables|Default|GetRotation%20Brightness%20Reactive%20Names%20) (Utilities|Time|GetGameTimeinSeconds)))

(fn RandomSet
  (params (FlipBook%20Frame :type Name) (TimeBP :type Name) (Offset :type Name) (Rotation :type Name) (X=min :type Float%20%28double-precision%29) (X=Max :type Float%20%28double-precision%29) (Y=Min :type Float%20%28double-precision%29) (Y=Max :type Float%20%28double-precision%29) (Rot=Min :type Float%20%28double-precision%29) (Rot=Max :type Float%20%28double-precision%29))
  (bind _returnvalue (CustomItemWrap|GetWrappedMeshComponents))
  (bind _stream (Variables|EnviromentVarb|GetStream))
  (bind _returnvalue_2 (Math|Random|RandomFloatinRangefromStream _stream X=min X=Max))
  (if (Variables|Default|GetHas%20Frames)
    (Rendering|Material|SetScalarParameterValueonMaterials _returnvalue FlipBook%20Frame (/ (Math|Conversions|ToFloat(Integer) (Utilities|Array|Get(aref) (Variables|Default|GetNumber%20of%20Frames_random) (Variables|EnviromentVarb|GetDamage))) (Variables|Default|GetNumber%20of%20Frames))))
  (Rendering|Material|SetScalarParameterValueonMaterials _returnvalue TimeBP (Utilities|Time|GetGameTimeinSeconds))
  (Rendering|Material|SetVectorParameterValueonMaterials _returnvalue Offset (Math|Vector|MakeVector _returnvalue_2 (Math|Random|RandomFloatinRangefromStream _stream Y=Min Y=Max)))
  (Rendering|Material|SetScalarParameterValueonMaterials _returnvalue Rotation (Math|Random|RandomFloatinRange Rot=Min Rot=Max))
  (Variables|EnviromentVarb|SetPosition _returnvalue_2))

(fn ConstructionScript)
