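; Custom item wrap logic (AnglePatch) that makes a weapon wrap's emissive effect react to firing.
; Each shot raises a step count capped at MaxSteps, which fades toward zero when the gap since the previous shot outlasts a 5.5-second hold.
; The step indexes brightness arrays, whose values are written with game-time stamps to two emissive layers' material parameters on wrapped meshes.
; A debug text print of the step remains.
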(event EventOnWeaponFired
  (bind _returnvalue (Utilities|Time|GetGameTimeinSeconds))
  (bind _returnvalue_2 (CustomItemWrap|GetWrappedMeshComponents))
  (bind _wait_timer (Variables|Default|Getwait_timer))
  (bind _returnvalue_3 (- _returnvalue _wait_timer))
  (bind _returnvalue_4 (/ _returnvalue_3 1.0))
  (bind _returnvalue_5 (Math|Float|Clamp(Float) _returnvalue_4))
  (bind _stepcount (Variables|Default|GetStepCount))
  (bind _returnvalue_6 (Math|Conversions|ToFloat(Integer) _stepcount))
  (bind _returnvalue_7 (Math|Float|Lerp _returnvalue_6 0.0 _returnvalue_5))
  (bind _returnvalue_8 (CustomItemWrap|GetWrappedMeshComponents))
  (bind _returnvalue_9 (Math|Float|Truncate _returnvalue_7))
  (Development|PrintText (Utilities|Text|ToText(Integer) _returnvalue_9))
  (bind _output_get (Variables|Default|SetStepCount (Math|Integer|Min(Integer) (+ _returnvalue_9 1) (Variables|Default|GetMaxSteps))))
  (Rendering|Material|SetScalarParameterValueonMaterials _returnvalue_8 "TopS Emissive_01-Brightness-TBP" (Utilities|Time|GetGameTimeinSeconds))
  (Rendering|Material|SetScalarParameterValueonMaterials _returnvalue_8 "TopS Emissive_01_Brightness---Start" (Utilities|Array|Get(aref) (Variables|Default|GetBrightness%20value) _output_get))
  (Rendering|Material|SetScalarParameterValueonMaterials _returnvalue_8 "TopS Emissive_01_Brightness_Darken_Base Start" (Utilities|Array|Get(aref) (Variables|Default|GetBrightness%20Flames) _output_get))
  (Variables|Default|Setwait_timer (+ (Utilities|Time|GetGameTimeinSeconds) 5.5))
  (Rendering|Material|SetScalarParameterValueonMaterials _returnvalue_2 "TopS Emissive_02-Brightness-TBP" (Utilities|Time|GetGameTimeinSeconds))
  (Rendering|Material|SetScalarParameterValueonMaterials _returnvalue_2 "TopS Emissive_02_Brightness_Darken_Base Start" (Utilities|Array|Get(aref) (Variables|Default|GetBrightness%20Flames) _output_get)))

(fn ConstructionScript)
