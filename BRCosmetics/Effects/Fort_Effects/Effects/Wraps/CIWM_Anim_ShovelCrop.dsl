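; Cosmetic wrap modifier (ShovelCrop) that plays a material animation on wrapped weapon meshes when the player's Athena kill stat changes.
; Kill watchers register on pawn set and weapon change; in the front end, a shot counter triggers it instead, resetting at seven.
; One timeline's color tracks drive blend opacity, base color, emissive brightness, mask channels and four decal scales.
; A Pause flag blocks retriggering until playback finishes.
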(event EventOnWeaponFired
  (bind _returnvalue (CustomItemWrap|GetWrappedMeshComponents))
  (if (CustomCosmeticModifier|IsinFrontEnd)
    (Utilities|FlowControl|Delay 0.001)
    (bind _output_get (Variables|Default|SetShotsFired (+ (Variables|Default|GetShotsFired) 1)))
    (if (>= _output_get 5)
      (if (== (Variables|Default|GetShotsFired) 7)
        (Variables|Default|SetShotsFired 0)
        (if (not (Variables|Default|GetPause))
          (Variables|Default|SetPause true)
          (Components|Timeline|SetPlayRate (Variables|CIWM_Anim_ShovelCrop|GetTimeline) 1.0)
          (bind (_direction _newtrack_0 _newtrack_1) (Timeline|Timeline PlayFromStart)
            (:Update
              (Rendering|Material|SetScalarParameterValueonMaterials _returnvalue "TopS_01_Blend_Opacity---End" (Utilities|Struct|BreakLinearColor _newtrack_0))
              (Rendering|Material|SetColorParameterValueonMaterials _returnvalue "TopS_02_Base Color_Simple" (Math|Color|Lerp(LinearColor) "(R=0.019382,G=0.022174,B=0.025187,A=0.500000)" "(R=0.880208,G=0.880208,B=0.880208,A=0.500000)" (Math|Float|Clamp(Float) (.G (Utilities|Struct|BreakLinearColor _newtrack_0)))))
              (Rendering|Material|SetScalarParameterValueonMaterials _returnvalue "TopS Emissive_02_Brightness---End" (* (Math|Float|Lerp 1.0 0.0 (Math|Float|Clamp(Float) (.G (Utilities|Struct|BreakLinearColor _newtrack_0)))) (Math|Float|Clamp(Float) (.B (Utilities|Struct|BreakLinearColor _newtrack_0)))))
              (Rendering|Material|SetScalarParameterValueonMaterials _returnvalue "TopS Emissive_03_Brightness---End" (* (Math|Float|Lerp 1.0 0.0 (Math|Float|Clamp(Float) (.G (Utilities|Struct|BreakLinearColor _newtrack_0)))) (Math|Float|Clamp(Float) (.B (Utilities|Struct|BreakLinearColor _newtrack_0)))))
              (Rendering|Material|SetColorParameterValueonMaterials _returnvalue "EDRec_Emissive_A_Gradent_MSk_Channel" (Math|Color|Lerp(LinearColor) "(R=0.000000,G=0.000000,B=0.000000,A=1.000000)" "(R=0.000000,G=1.000000,B=0.000000,A=0.000000)" (Math|Float|Clamp(Float) (.G (Utilities|Struct|BreakLinearColor _newtrack_0)))))
              (Rendering|Material|SetColorParameterValueonMaterials _returnvalue "EDRec_Emissive_B_Gradent_MSk_Channel" (Math|Color|Lerp(LinearColor) "(R=0.000000,G=0.000000,B=0.000000,A=1.000000)" "(R=0.000000,G=1.000000,B=0.000000,A=0.000000)" (Math|Float|Clamp(Float) (.G (Utilities|Struct|BreakLinearColor _newtrack_0)))))
              (Rendering|Material|SetColorParameterValueonMaterials _returnvalue "EDRec_Emissive_C_Gradent_MSk_Channel" (Math|Color|Lerp(LinearColor) "(R=0.000000,G=0.000000,B=0.000000,A=1.000000)" "(R=0.000000,G=1.000000,B=0.000000,A=0.000000)" (Math|Float|Clamp(Float) (.G (Utilities|Struct|BreakLinearColor _newtrack_0)))))
              (Rendering|Material|SetColorParameterValueonMaterials _returnvalue "EDRec_Emissive_D_Gradent_MSk_Channel" (Math|Color|Lerp(LinearColor) "(R=0.000000,G=0.000000,B=0.000000,A=1.000000)" "(R=0.000000,G=1.000000,B=0.000000,A=0.000000)" (Math|Float|Clamp(Float) (.G (Utilities|Struct|BreakLinearColor _newtrack_0)))))
              (Rendering|Material|SetColorParameterValueonMaterials _returnvalue "EDRec_Emissive_A_Mask_Channel" (Math|Color|Lerp(LinearColor) "(R=1.000000,G=0.000000,B=0.000000,A=0.000000)" "(R=0.000000,G=0.000000,B=0.000000,A=1.000000)" (Math|Float|Clamp(Float) (.G (Utilities|Struct|BreakLinearColor _newtrack_0)))))
              (Rendering|Material|SetColorParameterValueonMaterials _returnvalue "EDRec_Emissive_B_Mask_Channel" (Math|Color|Lerp(LinearColor) "(R=1.000000,G=0.000000,B=0.000000,A=0.000000)" "(R=0.000000,G=0.000000,B=0.000000,A=1.000000)" (Math|Float|Clamp(Float) (.G (Utilities|Struct|BreakLinearColor _newtrack_0)))))
              (Rendering|Material|SetColorParameterValueonMaterials _returnvalue "EDRec_Emissive_C_Mask_Channel" (Math|Color|Lerp(LinearColor) "(R=1.000000,G=0.000000,B=0.000000,A=0.000000)" "(R=0.000000,G=0.000000,B=0.000000,A=1.000000)" (Math|Float|Clamp(Float) (.G (Utilities|Struct|BreakLinearColor _newtrack_0)))))
              (Rendering|Material|SetColorParameterValueonMaterials _returnvalue "EDRec_Emissive_D_Mask_Channel" (Math|Color|Lerp(LinearColor) "(R=1.000000,G=0.000000,B=0.000000,A=0.000000)" "(R=0.000000,G=0.000000,B=0.000000,A=1.000000)" (Math|Float|Clamp(Float) (.G (Utilities|Struct|BreakLinearColor _newtrack_0)))))
              (Rendering|Material|SetColorParameterValueonMaterials _returnvalue "EDRec_DecalA_Scale" (Utilities|Operators|Add :A (Math|Color|MakeColor (Utilities|Struct|BreakLinearColor _newtrack_1) (Utilities|Struct|BreakLinearColor _newtrack_1) 1.0)))
              (Rendering|Material|SetScalarParameterValueonMaterials _returnvalue "EDRec_Emissive_A_Brightness_Time__End" (Math|Float|Lerp 1.0 0.0 (Math|Float|Clamp(Float) (.G (Utilities|Struct|BreakLinearColor _newtrack_0)))))
              (Rendering|Material|SetColorParameterValueonMaterials _returnvalue "EDRec_DecalB_Scale" (Utilities|Operators|Add :A (Math|Color|MakeColor (.G (Utilities|Struct|BreakLinearColor _newtrack_1)) (.G (Utilities|Struct|BreakLinearColor _newtrack_1)) 1.0)))
              (Rendering|Material|SetScalarParameterValueonMaterials _returnvalue "EDRec_Emissive_B_Brightness_Time__End" (Math|Float|Lerp 1.0 0.0 (Math|Float|Clamp(Float) (.G (Utilities|Struct|BreakLinearColor _newtrack_0)))))
              (Rendering|Material|SetColorParameterValueonMaterials _returnvalue "EDRec_DecalC_Scale" (Utilities|Operators|Add :A (Math|Color|MakeColor (.B (Utilities|Struct|BreakLinearColor _newtrack_1)) (.B (Utilities|Struct|BreakLinearColor _newtrack_1)) 1.0)))
              (Rendering|Material|SetScalarParameterValueonMaterials _returnvalue "EDRec_Emissive_C_Brightness_Time__End" (Math|Float|Lerp 1.0 0.0 (Math|Float|Clamp(Float) (.G (Utilities|Struct|BreakLinearColor _newtrack_0)))))
              (Rendering|Material|SetColorParameterValueonMaterials _returnvalue "EDRec_DecalD_Scale" (Utilities|Operators|Add :A (Math|Color|MakeColor (.A (Utilities|Struct|BreakLinearColor _newtrack_1)) (.A (Utilities|Struct|BreakLinearColor _newtrack_1)) 1.0)))
              (Rendering|Material|SetScalarParameterValueonMaterials _returnvalue "EDRec_Emissive_D_Brightness_Time__End" (Math|Float|Lerp 1.0 0.0 (Math|Float|Clamp(Float) (.G (Utilities|Struct|BreakLinearColor _newtrack_0))))))
            (:Finished
              (Variables|Default|SetPause false))))))))

(event EventOnPawnSet
  (bind _returnvalue (CustomItemWrap|GetWrappedMeshComponents))
  (Cosmetics|WatchCosmeticStat (CustomCharacter|GetAssociatedPlayerPawn) "(TagName=\"StatManager.AthenaKills\")" 0 false
    (:OnStatChanged
      (if (not (Variables|Default|GetPause))
        (Variables|Default|SetPause true)
        (Components|Timeline|SetPlayRate (Variables|CIWM_Anim_ShovelCrop|GetTimeline) 1.0)
        (Timeline|Timeline PlayFromStart)))))

(event EventOnWeaponChanged
  (params (NewWeapon :type Fort%20Weapon%20Object%20Reference) (OldWeapon :type Fort%20Weapon%20Object%20Reference))
  (bind _returnvalue (CustomItemWrap|GetWrappedMeshComponents))
  (Cosmetics|WatchCosmeticStat (CustomCharacter|GetAssociatedPlayerPawn) "(TagName=\"StatManager.AthenaKills\")" 0 false
    (:OnStatChanged
      (if (not (Variables|Default|GetPause))
        (Variables|Default|SetPause true)
        (Components|Timeline|SetPlayRate (Variables|CIWM_Anim_ShovelCrop|GetTimeline) 1.0)
        (Timeline|Timeline PlayFromStart)))))

(event Custom|update)

(fn ConstructionScript)
