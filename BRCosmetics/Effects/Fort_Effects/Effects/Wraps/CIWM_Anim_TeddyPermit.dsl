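; CIWM_Anim_TeddyPermit is an item wrap modifier that plays a reactive material animation on wrapped weapon meshes.
; A watched kill stat, bound on pawn set and weapon change, triggers it, as does a front-end shot counter from five shots until its reset at seven.
; A sped-up timeline drives blend opacity, decal scales, emissive gradient scale and lerped mask channels.
; A Pause flag blocks retriggering until the timeline finishes.
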(event EventOnWeaponFired
  (bind _returnvalue (CustomItemWrap|GetWrappedMeshComponents))
  (if (CustomCosmeticModifier|IsinFrontEnd)
    (Utilities|FlowControl|Delay 0.001)
    (bind _output_get (Variables|Default|SetShotsFired (+ (Variables|Default|GetShotsFired) 1)))
    (if (>= _output_get 5)
      (if (== (Variables|Default|GetShotsFired) 7)
        (Variables|Default|SetShotsFired 0)
        (if (not (Variables|Default|GetPause))
          (Variables|Default|SetPause true)
          (Components|Timeline|SetPlayRate (Variables|CIWM_Anim_TeddyPermit|GetTimeline) 1.5)
          (bind (_direction _scale _newtrack_2) (Timeline|Timeline PlayFromStart)
            (:Update
              (Rendering|Material|SetScalarParameterValueonMaterials _returnvalue "TopS_01_Blend_Opacity---End" (.A (Math|Color|BreakColor _scale)))
              (Rendering|Material|SetColorParameterValueonMaterials _returnvalue "EDRec_DecalA_Scale" (Utilities|Operators|Add :A (Math|Color|MakeColor (Math|Color|BreakColor _scale) (.G (Math|Color|BreakColor _scale)))))
              (Rendering|Material|SetColorParameterValueonMaterials _returnvalue "EDRec_DecalB_Scale" (Utilities|Operators|Add :A (Math|Color|MakeColor (.B (Math|Color|BreakColor _scale)) (.B (Math|Color|BreakColor _scale)))))
              (Rendering|Material|SetColorParameterValueonMaterials _returnvalue "EDRec_Emissive_A_Gradent_Scale" (Utilities|Operators|Add :A (Math|Color|MakeColor (Math|Color|BreakColor _newtrack_2) 1.0 1.0 (.A (Math|Color|BreakColor _newtrack_2)))))
              (Rendering|Material|SetColorParameterValueonMaterials _returnvalue "EDRec_Emissive_A_Gradent_MSk_Channel" (Math|Color|Lerp(LinearColor) "(R=0.000000,G=1.000000,B=0.000000,A=0.000000)" "(R=1.000000,G=0.000000,B=0.000000,A=0.000000)" (.B (Math|Color|BreakColor _newtrack_2))))
              (Rendering|Material|SetColorParameterValueonMaterials _returnvalue "EDRec_Emissive_A_Mask_Channel" (Math|Color|Lerp(LinearColor) "(R=0.000000,G=0.000000,B=0.000000,A=1.000000)" "(R=1.000000,G=0.000000,B=0.000000,A=0.000000)" (.G (Math|Color|BreakColor _newtrack_2)))))
            (:Finished
              (Variables|Default|SetPause false))))))))

(event EventOnPawnSet
  (bind _returnvalue (CustomItemWrap|GetWrappedMeshComponents))
  (Cosmetics|WatchCosmeticStat (CustomCharacter|GetAssociatedPlayerPawn) "(TagName=\"StatManager.AthenaKills\")" 0 false
    (:OnStatChanged
      (if (not (Variables|Default|GetPause))
        (Variables|Default|SetPause true)
        (Components|Timeline|SetPlayRate (Variables|CIWM_Anim_TeddyPermit|GetTimeline) 1.5)
        (Timeline|Timeline PlayFromStart)))))

(event EventOnWeaponChanged
  (params (NewWeapon :type Fort%20Weapon%20Object%20Reference) (OldWeapon :type Fort%20Weapon%20Object%20Reference))
  (bind _returnvalue (CustomItemWrap|GetWrappedMeshComponents))
  (Cosmetics|WatchCosmeticStat (CustomCharacter|GetAssociatedPlayerPawn) "(TagName=\"StatManager.AthenaKills\")" 0 false
    (:OnStatChanged
      (if (not (Variables|Default|GetPause))
        (Variables|Default|SetPause true)
        (Components|Timeline|SetPlayRate (Variables|CIWM_Anim_TeddyPermit|GetTimeline) 1.5)
        (Timeline|Timeline PlayFromStart)))))

(event Custom|update)

(fn ConstructionScript)
