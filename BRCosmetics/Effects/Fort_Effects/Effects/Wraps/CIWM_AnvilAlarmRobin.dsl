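; Weapon wrap Blueprint that animates wrapped mesh materials when the weapon fires.
; Each shot writes game time into emissive, blend and glitch parameters, while speed timestamps are throttled by a pause delay.
; A retriggerable delay detects sustained fire, starting a fade-in of decal warp and brightness, then a fade-out once firing stops.
; Despite the Sine names, looping timers drive linear clamped fades, and new timers overwrite the handle without clearing it.
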(event EventOnWeaponFired
  (bind _returnvalue (CustomItemWrap|GetWrappedMeshComponents))
  (bind _returnvalue_2 (Utilities|Time|GetGameTimeinSeconds))
  (Utilities|FlowControl|Sequence
    (:then_0
      ;[1 1 1] TopsSide V2
      (Rendering|Material|SetScalarParameterValueonMaterials _returnvalue "TopS Emissive_02-Brightness-TBP" _returnvalue_2)
      (Rendering|Material|SetScalarParameterValueonMaterials _returnvalue "TopS Emissive_02_Offset----TBP" _returnvalue_2)
      (Rendering|Material|SetScalarParameterValueonMaterials _returnvalue "TopS_01_Blend_Opacity-TBP" _returnvalue_2)
      (Rendering|Material|SetScalarParameterValueonMaterials _returnvalue "TopS_01_Blend_Offset-TBP" _returnvalue_2)
      (Rendering|Material|SetScalarParameterValueonMaterials _returnvalue "TimeFrom-BP----------------" _returnvalue_2)
      (Rendering|Material|SetScalarParameterValueonMaterials _returnvalue "Holo_Glitch_TimeFrom-BP------" _returnvalue_2))
    (:then_1
      (Rendering|Material|SetScalarParameterValueonMaterials _returnvalue "TopS Emissive_01-Brightness-TBP" _returnvalue_2)
      (Rendering|Material|SetScalarParameterValueonMaterials _returnvalue "TopS Emissive_01_Offset----TBP" _returnvalue_2)
      (Rendering|Material|SetScalarParameterValueonMaterials _returnvalue "TopS Emissive_03-Brightness-TBP" _returnvalue_2)
      (Rendering|Material|SetScalarParameterValueonMaterials _returnvalue "TopS Emissive_03_Offset----TBP" _returnvalue_2)
      (if (not (Variables|EnviromentVarable|GetPause))
        (Variables|EnviromentVarable|SetPause true)
        (Rendering|Material|SetScalarParameterValueonMaterials _returnvalue "TopS Emissive_01_Speed-TBP" _returnvalue_2)
        (Rendering|Material|SetScalarParameterValueonMaterials _returnvalue "TopS Emissive_02_Speed-TBP" _returnvalue_2)
        (Rendering|Material|SetScalarParameterValueonMaterials _returnvalue "TopS Emissive_03_Speed-TBP" _returnvalue_2)
        (Rendering|Material|SetScalarParameterValueonMaterials _returnvalue "TopS_01_Speed-TBP" _returnvalue_2)
        (Rendering|Material|SetScalarParameterValueonMaterials _returnvalue "TopS_02_Speed-TBP" _returnvalue_2)
        (Rendering|Material|SetScalarParameterValueonMaterials _returnvalue "TopS_03_Speed-TBP" _returnvalue_2)
        (Rendering|Material|SetScalarParameterValueonMaterials _returnvalue "TopS_04_Speed-TBP" _returnvalue_2)
        (Utilities|FlowControl|Delay (Variables|Default|GetSpeed%20Pause%20Duration%20))
        (Variables|EnviromentVarable|SetPause false)))
    (:then_2
      (if (not (Variables|Default|GetFullAutoWeapons))
        (Variables|Default|SetFullAutoWeapons true)
        (CallFunction|SineAnimation))
      (Utilities|FlowControl|RetriggerableDelay 0.25)
      (CallFunction|SineAnimOFF)
      (Variables|Default|SetFullAutoWeapons false))))

(event Custom|SineAnimation
  ; Layer Offset
  (Variables|Default|SetReverse false)
  (Variables|Default|SetStartTime (Utilities|Time|GetGameTimeinSeconds))
  (bind _returnvalue (Utilities|Time|SetTimerbyEvent (delegate Custom|SineAnimUpdate) 0.02 true))
  (Variables|Default|SetTimerHandle _returnvalue))

(event Custom|SineAnimOFF
  ; Layer Offset
  (Variables|Default|SetReverse true)
  (Variables|Default|SetStartTime (Utilities|Time|GetGameTimeinSeconds))
  (bind _returnvalue (Utilities|Time|SetTimerbyEvent (delegate Custom|SineAnimUpdate) 0.02 true))
  (Variables|Default|SetTimerHandle _returnvalue))

(event Custom|SineAnimUpdate
  (bind _returnvalue (Utilities|Time|GetGameTimeinSeconds))
  (bind _returnvalue_2 (CustomItemWrap|GetWrappedMeshComponents))
  (bind _starttime (Variables|Default|GetStartTime))
  (bind _returnvalue_3 (- _returnvalue _starttime))
  ; Layer Offset
  (if (>= _returnvalue_3 5.0)
    (Utilities|Time|ClearandInvalidateTimerbyHandle (Variables|Default|GetTimerHandle))
    (if (Variables|Default|GetReverse)
      ;; Start
      (Rendering|Material|SetScalarParameterValueonMaterials _returnvalue_2 "FDecalV2_UV_Warp-Strength" 0.04)
      (Rendering|Material|SetScalarParameterValueonMaterials _returnvalue_2 "FDecalV2_Emissive_Brightness_Time__End" 1.0)
      (else
        ;; Arrive
        (Rendering|Material|SetScalarParameterValueonMaterials _returnvalue_2 "FDecalV2_UV_Warp-Strength" 0.3)
        (Rendering|Material|SetScalarParameterValueonMaterials _returnvalue_2 "FDecalV2_Emissive_Brightness_Time__End" 1e+01)))
    (elif (Variables|Default|GetReverse)
      ;; FadeOut
      (Rendering|Material|SetScalarParameterValueonMaterials _returnvalue_2 "FDecalV2_UV_Warp-Strength" (Math|Float|MapRangeClamped _returnvalue_3 0.0 0.25 0.3 0.04))
      (Rendering|Material|SetScalarParameterValueonMaterials _returnvalue_2 "FDecalV2_Emissive_Brightness_Time__End" (Math|Float|MapRangeClamped _returnvalue_3 0.0 0.25 1e+01 1.0))
      (else
        ;; FadeIn
        (Rendering|Material|SetScalarParameterValueonMaterials _returnvalue_2 "FDecalV2_UV_Warp-Strength" (Math|Float|MapRangeClamped _returnvalue_3 0.0 0.25 0.04 0.3))
        (Rendering|Material|SetScalarParameterValueonMaterials _returnvalue_2 "FDecalV2_Emissive_Brightness_Time__End" (Math|Float|MapRangeClamped _returnvalue_3 0.0 0.25 1.0 1e+01))))))

(fn ConstructionScript)
