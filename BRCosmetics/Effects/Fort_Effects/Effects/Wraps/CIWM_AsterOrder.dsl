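; Weapon wrap modifier for the AsterOrder cosmetic that animates the wrap's UV warp whenever the weapon fires.
; Each shot restarts ShootTimeline, whose updates write scalar and vector parameters to the materials on the wrapped mesh components.
; One track sets the warp amount with opposite signs on the two TopS layers.
; Another adds into a persistent Offset that drives both layers' warp position, so it accumulates across shots instead of resetting.
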(event EventOnWeaponFired
  (bind _returnvalue (CustomItemWrap|GetWrappedMeshComponents))
  (bind _offset (Variables|Default|GetOffset))
  (bind (_direction _newtrack_0 _warpamount _newtrack_1) (Timeline|ShootTimeline PlayFromStart)
    (:Update
      (Variables|Default|SetOffset (+ _newtrack_0 (Variables|Default|GetOffset)))
      (Rendering|Material|SetScalarParameterValueonMaterials _returnvalue "TopS_02_Warp_UV_Amount" (* _newtrack_1 0.1))
      (Rendering|Material|SetScalarParameterValueonMaterials _returnvalue "TopS_01_Warp_UV_Amount" (* _newtrack_1 -0.1))
      (Rendering|Material|SetVectorParameterValueonMaterials _returnvalue "TopS_01_Warp_UV_Offset_Pos" (Math|Vector|MakeVector _offset))
      (Rendering|Material|SetVectorParameterValueonMaterials _returnvalue "TopS_02_Warp_UV_Offset_Pos" (Math|Vector|MakeVector _offset)))))

(fn ConstructionScript)
