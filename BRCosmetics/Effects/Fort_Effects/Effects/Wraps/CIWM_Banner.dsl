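; Custom item wrap Blueprint that applies the player's banner color and icon from their cosmetic loadout to wrapped meshes.
; When the wrap target is set, it stores both banner IDs and pushes the primary color as a vector parameter to the wrapped components' materials.
; It async-loads the large "PhysicalBanner" icon, sets it on each dynamic material instance, then calls OnBannerFullyApplied, which has no logic here.
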(event EventOnSetWrapTarget
  (Variables|Default|SetBanner%20Icon%20Id (Utilities|String|StringToName (CosmeticLoadout|GetBannerIconIdfromLoadout (CosmeticLoadout|GetCosmeticLoadout))))
  (Variables|Default|SetBanner%20Color%20Id (Utilities|String|StringToName (CosmeticLoadout|GetBannerColorIdfromLoadout (CosmeticLoadout|GetCosmeticLoadout))))
  (Utilities|FlowControl|Sequence
    (:then_0
      (bind _returnvalue (Banner|GetBannerColorPrimaryColorfromRowName (Variables|Default|GetBanner%20Color%20Id)))
      (Rendering|Material|SetVectorParameterValueonMaterials (CustomItemWrap|GetWrappedMeshComponents) (Variables|Default|GetBannerColorParamName) (Math|Conversions|ToVector(LinearColor) _returnvalue)))
    (:then_1
      (bind _returnvalue_2 (Banner|GetBannerLargeIconAssetfromRowName (Variables|Default|GetBanner%20Icon%20Id) "PhysicalBanner"))
      (bind _object (Utilities|AsyncLoadAsset _returnvalue_2)
        (:Completed
          (bind _astexture_2d (Utilities|Casting|CastToTexture2D _object))
          (Variables|Default|SetBanner%20Icon%20Texture _astexture_2d)
          (for _array_element (CustomItemWrap|GetWrappedMeshComponents)
            (for _array_element_2 (Rendering|Material|GetMaterials _array_element)
              (bind _asmaterial_instance_dynamic (Utilities|Casting|CastToMaterialInstanceDynamic _array_element_2))
              (Rendering|Material|SetTextureParameterValue _asmaterial_instance_dynamic (Variables|Default|GetBannerTextureParamName) (Variables|Default|GetBanner%20Icon%20Texture))))
          (CallFunction|OnBannerFullyApplied (Variables|Default|GetBanner%20Icon%20Texture)))))))

(event Custom|OnBannerFullyApplied
  (params (Icon :type Texture%202D%20Object%20Reference) (PrimaryColor :type Linear%20Color%20Structure)))

(fn ConstructionScript)
