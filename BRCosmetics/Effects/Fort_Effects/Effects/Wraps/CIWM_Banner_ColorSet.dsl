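; Custom item wrap Blueprint that applies the player's loadout banner icon and color to wrapped item meshes, with duplicate logic on pawn set and wrap target set.
; The banner's primary color drives two material vector parameters, with a fixed "set color" variable substituted when that color is zero.
; The banner icon loads asynchronously, is written to each dynamic material's texture parameter, then is passed to the empty OnBannerFullyApplied hook.
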(event EventOnPawnSet
  (bind _returnvalue (CustomItemWrap|GetWrappedMeshComponents))
  (bind _set_color (Variables|Default|Getset%20color))
  (bind _returnvalue_2 (Math|Conversions|ToVector(LinearColor) _set_color))
  (bind _returnvalue_3 (CustomItemWrap|GetWrappedMeshComponents))
  (Variables|Default|SetBanner%20Icon%20Id (Utilities|String|StringToName (CosmeticLoadout|GetBannerIconIdfromLoadout (CosmeticLoadout|GetCosmeticLoadout))))
  (Variables|Default|SetBanner%20Color%20Id (Utilities|String|StringToName (CosmeticLoadout|GetBannerColorIdfromLoadout (CosmeticLoadout|GetCosmeticLoadout))))
  (Utilities|FlowControl|Sequence
    (:then_0
      (bind _returnvalue_4 (Banner|GetBannerColorPrimaryColorfromRowName (Variables|Default|GetBanner%20Color%20Id)))
      (if (Utilities|Operators|Equal(==) :A (Math|Conversions|ToVector(LinearColor) _returnvalue_4))
        (Rendering|Material|SetVectorParameterValueonMaterials _returnvalue (Variables|Default|GetBannerColorParamName) _returnvalue_2)
        (Rendering|Material|SetVectorParameterValueonMaterials _returnvalue "TopS_01_Base Color_Simple" _returnvalue_2)
        (else
          (Rendering|Material|SetVectorParameterValueonMaterials _returnvalue_3 (Variables|Default|GetBannerColorParamName) (Math|Conversions|ToVector(LinearColor) _returnvalue_4))
          (Rendering|Material|SetVectorParameterValueonMaterials _returnvalue_3 "TopS_01_Base Color_Simple" (Math|Conversions|ToVector(LinearColor) _returnvalue_4)))))
    (:then_1
      (bind _returnvalue_5 (Banner|GetBannerLargeIconAssetfromRowName (Variables|Default|GetBanner%20Icon%20Id) "PhysicalBanner"))
      (bind _object (Utilities|AsyncLoadAsset _returnvalue_5)
        (:Completed
          (bind _astexture_2d (Utilities|Casting|CastToTexture2D _object))
          (Variables|Default|SetBanner%20Icon%20Texture _astexture_2d)
          (for _array_element (CustomItemWrap|GetWrappedMeshComponents)
            (for _array_element_2 (Rendering|Material|GetMaterials _array_element)
              (bind _asmaterial_instance_dynamic (Utilities|Casting|CastToMaterialInstanceDynamic _array_element_2))
              (Rendering|Material|SetTextureParameterValue _asmaterial_instance_dynamic (Variables|Default|GetBannerTextureParamName) (Variables|Default|GetBanner%20Icon%20Texture))))
          (CallFunction|OnBannerFullyApplied (Variables|Default|GetBanner%20Icon%20Texture)))))))

(event EventOnSetWrapTarget
  (bind _returnvalue (CustomItemWrap|GetWrappedMeshComponents))
  (bind _set_color (Variables|Default|Getset%20color))
  (bind _returnvalue_2 (Math|Conversions|ToVector(LinearColor) _set_color))
  (bind _returnvalue_3 (CustomItemWrap|GetWrappedMeshComponents))
  (Variables|Default|SetBanner%20Icon%20Id (Utilities|String|StringToName (CosmeticLoadout|GetBannerIconIdfromLoadout (CosmeticLoadout|GetCosmeticLoadout))))
  (Variables|Default|SetBanner%20Color%20Id (Utilities|String|StringToName (CosmeticLoadout|GetBannerColorIdfromLoadout (CosmeticLoadout|GetCosmeticLoadout))))
  (Utilities|FlowControl|Sequence
    (:then_0
      (bind _returnvalue_4 (Banner|GetBannerColorPrimaryColorfromRowName (Variables|Default|GetBanner%20Color%20Id)))
      (if (Utilities|Operators|Equal(==) :A (Math|Conversions|ToVector(LinearColor) _returnvalue_4))
        (Rendering|Material|SetVectorParameterValueonMaterials _returnvalue (Variables|Default|GetBannerColorParamName) _returnvalue_2)
        (Rendering|Material|SetVectorParameterValueonMaterials _returnvalue "TopS_01_Base Color_Simple" _returnvalue_2)
        (else
          (Rendering|Material|SetVectorParameterValueonMaterials _returnvalue_3 (Variables|Default|GetBannerColorParamName) (Math|Conversions|ToVector(LinearColor) _returnvalue_4))
          (Rendering|Material|SetVectorParameterValueonMaterials _returnvalue_3 "TopS_01_Base Color_Simple" (Math|Conversions|ToVector(LinearColor) _returnvalue_4)))))
    (:then_1
      (bind _returnvalue_5 (Banner|GetBannerLargeIconAssetfromRowName (Variables|Default|GetBanner%20Icon%20Id) "PhysicalBanner"))
      (bind _object (Utilities|AsyncLoadAsset _returnvalue_5)
        (:Completed
          (bind _astexture_2d (Utilities|Casting|CastToTexture2D _object))
          (Variables|Default|SetBanner%20Icon%20Texture _astexture_2d)
          (for _array_element (CustomItemWrap|GetWrappedMeshComponents)
            (for _array_element_2 (Rendering|Material|GetMaterials _array_element)
              (bind _asmaterial_instance_dynamic (Utilities|Casting|CastToMaterialInstanceDynamic _array_element_2))
              (Rendering|Material|SetTextureParameterValue _asmaterial_instance_dynamic (Variables|Default|GetBannerTextureParamName) (Variables|Default|GetBanner%20Icon%20Texture))))
          (CallFunction|OnBannerFullyApplied (Variables|Default|GetBanner%20Icon%20Texture)))))))

(event Custom|OnBannerFullyApplied
  (params (Icon :type Texture%202D%20Object%20Reference) (PrimaryColor :type Linear%20Color%20Structure)))

(fn ConstructionScript)
