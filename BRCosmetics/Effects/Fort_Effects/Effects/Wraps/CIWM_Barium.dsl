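; CIWM_Barium is a custom item wrap Blueprint that drives the Barium weapon wrap's material reaction to firing.
; On each weapon-fired event, a Sequence starts two short Delays and then writes game time to scalar parameters on the wrapped mesh components' materials.
; One branch stamps Barium_Time_BPD, while the slightly later branch rotates through BPA, BPB and BPC using an integer counter, giving three staggered timestamps.
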(event EventOnWeaponFired
  (bind _returnvalue (CustomItemWrap|GetWrappedMeshComponents))
  (bind _returnvalue_2 (Utilities|Time|GetGameTimeinSeconds))
  (Utilities|FlowControl|Sequence
    (:then_0
      (Utilities|FlowControl|Delay 0.1)
      (Rendering|Material|SetScalarParameterValueonMaterials (CustomItemWrap|GetWrappedMeshComponents) "Barium_Time_BPD" (Utilities|Time|GetGameTimeinSeconds)))
    (:then_1
      (Utilities|FlowControl|Delay 0.15)
      (switch int (Variables|Default|GetSet_Varable)
        (:0
          (Rendering|Material|SetScalarParameterValueonMaterials _returnvalue "Barium_Time_BPA" _returnvalue_2)
          (Variables|Default|SetSet_Varable 1))
        (:1
          (Rendering|Material|SetScalarParameterValueonMaterials _returnvalue "Barium_Time_BPB" _returnvalue_2)
          (Variables|Default|SetSet_Varable 2))
        (:2
          (Rendering|Material|SetScalarParameterValueonMaterials _returnvalue "Barium_Time_BPC" _returnvalue_2)
          (Variables|Default|SetSet_Varable 0))))))

(fn ConstructionScript)
