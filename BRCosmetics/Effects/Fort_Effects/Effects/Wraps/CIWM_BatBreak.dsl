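; CIWM_BatBreak is an item wrap Blueprint that animates a "break" effect on wrapped weapon materials in response to firing.
; Each shot raises a capped iteration count and works with a sped-up timeline that drives warp amount and mask-channel parameters.
; A retriggerable delay reverses the timeline after firing pauses.
; Separate custom events, not called here, decrement the count or fade the gradient mask through a looping timer; Unbreak is empty.
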(event EventBeginPlay
  (Rendering|Material|SetVectorParameterValueonMaterials (CustomItemWrap|GetWrappedMeshComponents) "TopS_Blend_Offset__Master" (0.25 0.0 0.0))
  (Components|Timeline|SetPlayRate (Variables|CIWM_BatBreak|GetTimeline) 4.4))

(event Custom|FadeStep
  (Variables|Default|SetIterationCount (Math|Integer|Max(Integer) (- (Variables|Default|GetIterationCount) 1))))

(event EventOnWeaponFired
  (bind _returnvalue (CustomItemWrap|GetWrappedMeshComponents))
  (Utilities|FlowControl|Sequence
    (:then_0
      ; Fire-seq
      (Utilities|FlowControl|Sequence
        (:then_0
          (if (Variables|Default|GetIsWaiting?)
            _
            (elif (>= (Variables|Default|GetIterationCount) 7)
              (Variables|Default|SetIsWaiting? true)
              (else
                (Math|Integer|IncrementInt (Variables|Default|GetIterationCount))))))
        (:then_1
          (bind (_direction _warp_amount _mask_channel _newtrack_2) (Timeline|Timeline Stop :NewTime (Math|Float|Min(Float) (+ (Components|Timeline|GetPlaybackPosition (Variables|CIWM_BatBreak|GetTimeline)) 1.0) 7.0))
            (:Update
              _)
            (:Finished
              (Variables|Default|SetIsWaiting? false))))
        (:then_2
          (Timeline|Timeline SetNewTime))
        (:then_3
          _))
      (Rendering|Material|SetScalarParameterValueonMaterials _returnvalue "TopS_01_Blend_Warp_Amount" _warp_amount)
      (Rendering|Material|SetColorParameterValueonMaterials _returnvalue "TopS Blend-TS 03_Mask_Channel" (Math|Color|MakeColor _mask_channel 0.0 0.0 0.0))
      (Rendering|Material|SetColorParameterValueonMaterials _returnvalue "TopS Blend_Grad-Tex_Mask_Channel" (Math|Color|MakeColor 0.0 0.0 0.0 _newtrack_2)))
    (:then_1
      (Utilities|FlowControl|RetriggerableDelay 6.75)
      (Timeline|Timeline Reverse))))

(event Custom|Unbreak)

(event Custom|SineAnimation
  ; Layer Offset
  (Variables|Default|SetReverse false)
  (Variables|Default|SetStartTime (Utilities|Time|GetGameTimeinSeconds))
  (bind _returnvalue (Utilities|Time|SetTimerbyEvent (delegate Custom|SineAnimUpdate) 0.02 true))
  (Variables|Default|SetTimerHandle _returnvalue))

(event Custom|SineAnimOFF
  ; Layer Offset
  (Variables|Default|SetReverse true)
  (Variables|Default|SetStartTime (Utilities|Time|GetGameTimeinSeconds))
  (bind _returnvalue (Utilities|Time|SetTimerbyEvent (delegate Custom|SineAnimUpdate) 0.02 true))
  (Variables|Default|SetTimerHandle _returnvalue))

(event Custom|SineAnimUpdate
  (bind _returnvalue (Utilities|Time|GetGameTimeinSeconds))
  (bind _returnvalue_2 (CustomItemWrap|GetWrappedMeshComponents))
  (bind _starttime (Variables|Default|GetStartTime))
  (bind _returnvalue_3 (- _returnvalue _starttime))
  ; Layer Offset
  (if (>= _returnvalue_3 7.0)
    (Utilities|Time|ClearandInvalidateTimerbyHandle (Variables|Default|GetTimerHandle))
    (if (Variables|Default|GetReverse)
      ;; Start
      (Rendering|Material|SetColorParameterValueonMaterials _returnvalue_2 "TopS Blend_Grad-Tex_Mask_Channel" (0.0 0.0 0.0 0.0))
      (else
        ;; Arrive
        (Rendering|Material|SetColorParameterValueonMaterials _returnvalue_2 "TopS Blend_Grad-Tex_Mask_Channel" (0.0 0.0 0.0 1.0))))
    (elif (Variables|Default|GetReverse)
      ;; OUT
      (Rendering|Material|SetColorParameterValueonMaterials _returnvalue_2 "TopS Blend_Grad-Tex_Mask_Channel" (Math|Color|MakeColor 0.0 0.0 0.0 (Math|Float|MapRangeClamped _returnvalue_3 0.0 1.0 1.0)))
      (else
        ;; IN
        (Rendering|Material|SetColorParameterValueonMaterials _returnvalue_2 "TopS Blend_Grad-Tex_Mask_Channel" (Math|Color|MakeColor 0.0 0.0 0.0 (Math|Float|MapRangeClamped _returnvalue_3 0.0 0.25 0.0 1.0)))))))

(fn ConstructionScript)
