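; Weapon wrap modifier for BattleHoundFierce that drives reactive emissive effects on wrapped meshes by writing game-time stamps to material parameters.
; Firing stamps channel A, while DamageDealt changes trigger the first free of channels B, C and D, each locked for 0.5 seconds by a Pause flag.
; In the front end, a timeline on BeginPlay fires B, C and D as a preview.
; Weapon-change events reset LastKillTime after a short delay.
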(event EventOnWeaponFired
  (Rendering|Material|SetScalarParameterValueonMaterials (CustomItemWrap|GetWrappedMeshComponents) "EDRec_Emissive_A_Brightness_Time_BP" (Utilities|Time|GetGameTimeinSeconds)))

(event EventOnPawnSet
  (bind _returnvalue (Utilities|Time|GetGameTimeinSeconds))
  (bind _returnvalue_2 (CustomItemWrap|GetWrappedMeshComponents))
  (bind _returnvalue_3 (CustomItemWrap|GetWrappedMeshComponents))
  (bind _returnvalue_4 (Utilities|Time|GetGameTimeinSeconds))
  (bind _returnvalue_5 (Utilities|Time|GetGameTimeinSeconds))
  (bind _returnvalue_6 (CustomItemWrap|GetWrappedMeshComponents))
  (Cosmetics|WatchCosmeticStat (CustomCharacter|GetAssociatedPlayerPawn) "(TagName=\"StatManager.DamageDealt\")" 0 false
    (:OnStatChanged
      (if (Variables|Default|GetPause%20A)
        (if (Variables|Default|GetPause%20B)
          (if (not (Variables|Default|GetPause%20C))
            (Variables|Default|SetPause%20C true)
            (Rendering|Material|SetScalarParameterValueonMaterials _returnvalue_3 "EDRec_Emissive_D_Brightness_Time_BP" _returnvalue_4)
            (Rendering|Material|SetScalarParameterValueonMaterials _returnvalue_3 "EDRec_Emissive_D_Gradent_Time_BP" _returnvalue_4)
            (Utilities|FlowControl|Delay 0.5)
            (Variables|Default|SetPause%20C false))
          (else
            (Variables|Default|SetPause%20B true)
            (Rendering|Material|SetScalarParameterValueonMaterials _returnvalue_2 "EDRec_Emissive_C_Brightness_Time_BP" _returnvalue)
            (Rendering|Material|SetScalarParameterValueonMaterials _returnvalue_2 "EDRec_Emissive_C_Gradent_Time_BP" _returnvalue)
            (Utilities|FlowControl|Delay 0.5)
            (Variables|Default|SetPause%20B false)))
        (else
          (Variables|Default|SetPause%20A true)
          (Rendering|Material|SetScalarParameterValueonMaterials _returnvalue_6 "EDRec_Emissive_B_Brightness_Time_BP" _returnvalue_5)
          (Rendering|Material|SetScalarParameterValueonMaterials _returnvalue_6 "EDRec_Emissive_B_Gradent_Time_BP" _returnvalue_5)
          (Utilities|FlowControl|Delay 0.5)
          (Variables|Default|SetPause%20A false))))))

(event EventBeginPlay
  (bind _returnvalue (Utilities|Time|GetGameTimeinSeconds))
  (bind _returnvalue_2 (CustomItemWrap|GetWrappedMeshComponents))
  (bind _returnvalue_3 (CustomItemWrap|GetWrappedMeshComponents))
  (bind _returnvalue_4 (Utilities|Time|GetGameTimeinSeconds))
  (bind _returnvalue_5 (Utilities|Time|GetGameTimeinSeconds))
  (bind _returnvalue_6 (CustomItemWrap|GetWrappedMeshComponents))
  (if (CustomCosmeticModifier|IsinFrontEnd)
    (Timeline|Timeline PlayFromStart
      (:A
        (Rendering|Material|SetScalarParameterValueonMaterials _returnvalue_6 "EDRec_Emissive_B_Brightness_Time_BP" _returnvalue_5)
        (Rendering|Material|SetScalarParameterValueonMaterials _returnvalue_6 "EDRec_Emissive_B_Gradent_Time_BP" _returnvalue_5)
        (Utilities|FlowControl|Delay 0.5)
        (Variables|Default|SetPause%20A false))
      (:B
        (Rendering|Material|SetScalarParameterValueonMaterials _returnvalue_2 "EDRec_Emissive_C_Brightness_Time_BP" _returnvalue)
        (Rendering|Material|SetScalarParameterValueonMaterials _returnvalue_2 "EDRec_Emissive_C_Gradent_Time_BP" _returnvalue)
        (Utilities|FlowControl|Delay 0.5)
        (Variables|Default|SetPause%20B false))
      (:C
        (Rendering|Material|SetScalarParameterValueonMaterials _returnvalue_3 "EDRec_Emissive_D_Brightness_Time_BP" _returnvalue_4)
        (Rendering|Material|SetScalarParameterValueonMaterials _returnvalue_3 "EDRec_Emissive_D_Gradent_Time_BP" _returnvalue_4)
        (Utilities|FlowControl|Delay 0.5)
        (Variables|Default|SetPause%20C false)))))

(event EventOnPostResIn)

(event EventOnNewWeapons
  (params (WeaponsAsItems :type Array%20of%20Fort%20Item%20Object%20References))
  (Utilities|FlowControl|Delay 0.05)
  ;[1 1 1] Updated kill count on new weapon
  (Variables|Default|SetLastKillTime -999999999999.0))

(event EventOnNewPlayerWeaponType
  (params (NewType :type EFortItemType%20Enum :default "WorldItem") (OldType :type EFortItemType%20Enum :default "WorldItem"))
  (Utilities|FlowControl|Delay 0.05)
  ;[1 1 1] Updated kill count on new weapon
  (Variables|Default|SetLastKillTime -999999999999.0))

(event EventOnWeaponChanged
  (params (NewWeapon :type Fort%20Weapon%20Object%20Reference) (OldWeapon :type Fort%20Weapon%20Object%20Reference))
  (Utilities|FlowControl|Delay 0.05)
  ;[1 1 1] Updated kill count on new weapon
  (Variables|Default|SetLastKillTime -999999999999.0))

(event EventApplyMaterialParameters)

(fn ConstructionScript)
