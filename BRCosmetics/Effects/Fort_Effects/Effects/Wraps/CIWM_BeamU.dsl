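; Custom item wrap Blueprint for the BeamU wrap that makes its materials react to the associated player's kills and to weapon fire.
; When the pawn is set, it watches the AthenaKills cosmetic stat and stamps game time into the Beam material parameter on each change.
; Firing the weapon stamps game time into the Lights parameter.
; Both write to all wrapped mesh components, leaving any time-based animation to the material.
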(event EventOnPawnSet
  (bind _asfort_player_pawn (Utilities|Casting|CastToFortPlayerPawn (CustomCharacter|GetAssociatedPlayerPawn)))
  (Cosmetics|WatchCosmeticStat _asfort_player_pawn "(TagName=\"StatManager.AthenaKills\")"
    (:OnStatChanged
      (Rendering|Material|SetScalarParameterValueonMaterials (CustomItemWrap|GetWrappedMeshComponents) "BU_TimeFrom-BP_Beam" (Utilities|Time|GetGameTimeinSeconds)))))

(event EventBeginPlay
  (|CallParentFunction))

(event EventOnWeaponFired
  (Rendering|Material|SetScalarParameterValueonMaterials (CustomItemWrap|GetWrappedMeshComponents) "BU_TimeFrom-BP_Lights" (Utilities|Time|GetGameTimeinSeconds)))

(fn ConstructionScript)
