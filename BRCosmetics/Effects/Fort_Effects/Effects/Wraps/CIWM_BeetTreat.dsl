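; BeetTreat weapon wrap modifier that shifts an emissive material offset on wrapped meshes as the magazine is fired.
; Pawn and weapon events bind reload, zero ShootStep and derive a per-shot increment from magazine ammo count.
; Shots advance the clamped step and update the End offset; reloads write Start, End and a game-time stamp, apparently for shader-side interpolation.
; In the front end, a delayed ReloadStarted call repeats itself.
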(event EventBeginPlay
  (bind _returnvalue (CustomCharacter|GetAssociatedPlayerPawn))
  (Utilities|IsValid _returnvalue
    (:Is%20Valid
      (bind _currentweapon (Variables|/Script/FortniteGame.FortPawn|GetCurrentWeapon _returnvalue)
        (:then
          (EventDispatchers|BindEventtoOnReloadStarted _currentweapon (delegate Custom|ReloadStarted))
          (Variables|Default|SetShotIncrement (/ 1.0 (Math|Conversions|ToFloat(Integer) (FortWeapon|GetMagazineAmmoCount _currentweapon))))
          (Variables|Default|SetShootStep 0.0)
          (Rendering|Material|SetScalarParameterValueonMaterials (CustomItemWrap|GetWrappedMeshComponents) "TopS Emissive_01_Offset----TBP" (Utilities|Time|GetGameTimeinSeconds))
          (Rendering|Material|SetScalarParameterValueonMaterials (CustomItemWrap|GetWrappedMeshComponents) "TopS Emissive_01_Offset---End" (Utilities|ShootSteptoValue))
          (if (CustomCosmeticModifier|IsinFrontEnd)
            (Utilities|FlowControl|Delay 3.0)
            (CallFunction|ReloadStarted 0.0)))
        (:else)))))

(event EventOnPawnSet
  (bind _returnvalue (CustomCharacter|GetAssociatedPlayerPawn))
  (Utilities|IsValid _returnvalue
    (:Is%20Valid
      (bind _currentweapon (Variables|/Script/FortniteGame.FortPawn|GetCurrentWeapon _returnvalue)
        (:then
          (EventDispatchers|BindEventtoOnReloadStarted _currentweapon (delegate Custom|ReloadStarted))
          (Variables|Default|SetShotIncrement (/ 1.0 (Math|Conversions|ToFloat(Integer) (FortWeapon|GetMagazineAmmoCount _currentweapon))))
          (Variables|Default|SetShootStep 0.0)
          (Rendering|Material|SetScalarParameterValueonMaterials (CustomItemWrap|GetWrappedMeshComponents) "TopS Emissive_01_Offset----TBP" (Utilities|Time|GetGameTimeinSeconds))
          (Rendering|Material|SetScalarParameterValueonMaterials (CustomItemWrap|GetWrappedMeshComponents) "TopS Emissive_01_Offset---End" (Utilities|ShootSteptoValue))
          (if (CustomCosmeticModifier|IsinFrontEnd)
            (Utilities|FlowControl|Delay 3.0)
            (CallFunction|ReloadStarted 0.0)))
        (:else)))))

(event EventOnNewWeapons
  (params (WeaponsAsItems :type Array%20of%20Fort%20Item%20Object%20References))
  (bind _returnvalue (CustomCharacter|GetAssociatedPlayerPawn))
  (Utilities|IsValid _returnvalue
    (:Is%20Valid
      (bind _currentweapon (Variables|/Script/FortniteGame.FortPawn|GetCurrentWeapon _returnvalue)
        (:then
          (EventDispatchers|BindEventtoOnReloadStarted _currentweapon (delegate Custom|ReloadStarted))
          (Variables|Default|SetShotIncrement (/ 1.0 (Math|Conversions|ToFloat(Integer) (FortWeapon|GetMagazineAmmoCount _currentweapon))))
          (Variables|Default|SetShootStep 0.0)
          (Rendering|Material|SetScalarParameterValueonMaterials (CustomItemWrap|GetWrappedMeshComponents) "TopS Emissive_01_Offset----TBP" (Utilities|Time|GetGameTimeinSeconds))
          (Rendering|Material|SetScalarParameterValueonMaterials (CustomItemWrap|GetWrappedMeshComponents) "TopS Emissive_01_Offset---End" (Utilities|ShootSteptoValue))
          (if (CustomCosmeticModifier|IsinFrontEnd)
            (Utilities|FlowControl|Delay 3.0)
            (CallFunction|ReloadStarted 0.0)))
        (:else)))))

(event EventOnWeaponChanged
  (params (NewWeapon :type Fort%20Weapon%20Object%20Reference) (OldWeapon :type Fort%20Weapon%20Object%20Reference))
  (bind _returnvalue (CustomCharacter|GetAssociatedPlayerPawn))
  (Utilities|IsValid _returnvalue
    (:Is%20Valid
      (bind _currentweapon (Variables|/Script/FortniteGame.FortPawn|GetCurrentWeapon _returnvalue)
        (:then
          (EventDispatchers|BindEventtoOnReloadStarted _currentweapon (delegate Custom|ReloadStarted))
          (Variables|Default|SetShotIncrement (/ 1.0 (Math|Conversions|ToFloat(Integer) (FortWeapon|GetMagazineAmmoCount _currentweapon))))
          (Variables|Default|SetShootStep 0.0)
          (Rendering|Material|SetScalarParameterValueonMaterials (CustomItemWrap|GetWrappedMeshComponents) "TopS Emissive_01_Offset----TBP" (Utilities|Time|GetGameTimeinSeconds))
          (Rendering|Material|SetScalarParameterValueonMaterials (CustomItemWrap|GetWrappedMeshComponents) "TopS Emissive_01_Offset---End" (Utilities|ShootSteptoValue))
          (if (CustomCosmeticModifier|IsinFrontEnd)
            (Utilities|FlowControl|Delay 3.0)
            (CallFunction|ReloadStarted 0.0)))
        (:else)))))

(event EventOnNewPlayerWeaponType
  (params (NewType :type EFortItemType%20Enum :default "WorldItem") (OldType :type EFortItemType%20Enum :default "WorldItem"))
  (bind _returnvalue (CustomCharacter|GetAssociatedPlayerPawn))
  (Utilities|IsValid _returnvalue
    (:Is%20Valid
      (bind _currentweapon (Variables|/Script/FortniteGame.FortPawn|GetCurrentWeapon _returnvalue)
        (:then
          (EventDispatchers|BindEventtoOnReloadStarted _currentweapon (delegate Custom|ReloadStarted))
          (Variables|Default|SetShotIncrement (/ 1.0 (Math|Conversions|ToFloat(Integer) (FortWeapon|GetMagazineAmmoCount _currentweapon))))
          (Variables|Default|SetShootStep 0.0)
          (Rendering|Material|SetScalarParameterValueonMaterials (CustomItemWrap|GetWrappedMeshComponents) "TopS Emissive_01_Offset----TBP" (Utilities|Time|GetGameTimeinSeconds))
          (Rendering|Material|SetScalarParameterValueonMaterials (CustomItemWrap|GetWrappedMeshComponents) "TopS Emissive_01_Offset---End" (Utilities|ShootSteptoValue))
          (if (CustomCosmeticModifier|IsinFrontEnd)
            (Utilities|FlowControl|Delay 3.0)
            (CallFunction|ReloadStarted 0.0)))
        (:else)))))

(event Custom|ReloadStarted
  (params (ReloadTime :type Float%20%28single-precision%29) (ReloadType :type EFortWeaponReloadType%20Enum :default "ReloadWholeClip"))
  (bind _returnvalue (CustomItemWrap|GetWrappedMeshComponents))
  (Rendering|Material|SetScalarParameterValueonMaterials _returnvalue "TopS Emissive_01_Offset---Start" (Utilities|ShootSteptoValue))
  (Variables|Default|SetShootStep 0.0)
  (Rendering|Material|SetScalarParameterValueonMaterials _returnvalue "TopS Emissive_01_Offset----TBP" (Utilities|Time|GetGameTimeinSeconds))
  (Rendering|Material|SetScalarParameterValueonMaterials (CustomItemWrap|GetWrappedMeshComponents) "TopS Emissive_01_Offset---End" (Utilities|ShootSteptoValue))
  (if (CustomCosmeticModifier|IsinFrontEnd)
    (Utilities|FlowControl|Delay 3.0)
    (CallFunction|ReloadStarted 0.0)))

(event EventOnWeaponFired
  (Variables|Default|SetShootStep (Math|Float|Clamp(Float) (+ (Variables|Default|GetShootStep) (Variables|Default|GetShotIncrement))))
  (Utilities|FlowControl|Delay 0.1)
  (Rendering|Material|SetScalarParameterValueonMaterials (CustomItemWrap|GetWrappedMeshComponents) "TopS Emissive_01_Offset---End" (Utilities|ShootSteptoValue))
  (if (CustomCosmeticModifier|IsinFrontEnd)
    (Utilities|FlowControl|Delay 3.0)
    (CallFunction|ReloadStarted 0.0)))

(fn ConstructionScript)
