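; CIWM_BenBaton is a weapon wrap Blueprint that animates the wrap's materials when the weapon fires, apparently as a veins effect.
; On fire, it gets the wrapped mesh components and plays the VeinsProgress timeline from the start.
; The timeline's two tracks drive a warp UV offset on Y and an emissive brightness parameter on those materials.
; A Playing flag ignores further shots until the timeline finishes.
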(event EventBeginPlay)

(event EventOnWeaponFired
  (bind _returnvalue (CustomItemWrap|GetWrappedMeshComponents))
  (if (not (Variables|Default|GetPlaying))
    (Variables|Default|SetPlaying true)
    (bind (_direction _newtrack_0 _newtrack_1) (Timeline|VeinsProgress PlayFromStart)
      (:Update
        (Rendering|Material|SetVectorParameterValueonMaterials _returnvalue "TopS_03_Warp_UV_Offset_Pos" (Math|Vector|MakeVector 0.0 _newtrack_0))
        (Rendering|Material|SetScalarParameterValueonMaterials _returnvalue "TopS Emissive_02_Brightness---End" _newtrack_1))
      (:Finished
        (Variables|Default|SetPlaying false)))))

(fn ConstructionScript)
