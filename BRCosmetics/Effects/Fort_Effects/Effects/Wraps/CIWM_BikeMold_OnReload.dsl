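; BikeMold reactive weapon wrap that makes wrapped meshes glow on reload and react to firing.
; Begin play, pawn and weapon-change events bind TriggerReactivity to the current weapon's OnReloadStarted.
; TriggerReactivity starts a bright, faster-moving emissive fade lasting slightly beyond reload time, then a Delay restores the dim idle glow.
; Game-time stamps written to material parameters drive the animation, including a separate emissive layer retriggered on each shot.
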(event EventBeginPlay
  (bind _returnvalue (CustomCharacter|GetAssociatedPlayerPawn))
  (Utilities|IsValid _returnvalue
    (:Is%20Valid
      (bind _currentweapon (Variables|/Script/FortniteGame.FortPawn|GetCurrentWeapon _returnvalue)
        (:then
          (EventDispatchers|BindEventtoOnReloadStarted _currentweapon (delegate Custom|TriggerReactivity)))
        (:else)))))

(event EventOnPawnSet
  (bind _returnvalue (CustomCharacter|GetAssociatedPlayerPawn))
  (Utilities|IsValid _returnvalue
    (:Is%20Valid
      (bind _currentweapon (Variables|/Script/FortniteGame.FortPawn|GetCurrentWeapon _returnvalue)
        (:then
          (EventDispatchers|BindEventtoOnReloadStarted _currentweapon (delegate Custom|TriggerReactivity)))
        (:else)))))

(event EventOnNewWeapons
  (params (WeaponsAsItems :type Array%20of%20Fort%20Item%20Object%20References))
  (bind _returnvalue (CustomCharacter|GetAssociatedPlayerPawn))
  (Utilities|IsValid _returnvalue
    (:Is%20Valid
      (bind _currentweapon (Variables|/Script/FortniteGame.FortPawn|GetCurrentWeapon _returnvalue)
        (:then
          (EventDispatchers|BindEventtoOnReloadStarted _currentweapon (delegate Custom|TriggerReactivity)))
        (:else)))))

(event EventOnWeaponChanged
  (params (NewWeapon :type Fort%20Weapon%20Object%20Reference) (OldWeapon :type Fort%20Weapon%20Object%20Reference))
  (bind _returnvalue (CustomCharacter|GetAssociatedPlayerPawn))
  (Utilities|IsValid _returnvalue
    (:Is%20Valid
      (bind _currentweapon (Variables|/Script/FortniteGame.FortPawn|GetCurrentWeapon _returnvalue)
        (:then
          (EventDispatchers|BindEventtoOnReloadStarted _currentweapon (delegate Custom|TriggerReactivity)))
        (:else)))))

(event EventOnNewPlayerWeaponType
  (params (NewType :type EFortItemType%20Enum :default "WorldItem") (OldType :type EFortItemType%20Enum :default "WorldItem"))
  (bind _returnvalue (CustomCharacter|GetAssociatedPlayerPawn))
  (Utilities|IsValid _returnvalue
    (:Is%20Valid
      (bind _currentweapon (Variables|/Script/FortniteGame.FortPawn|GetCurrentWeapon _returnvalue)
        (:then
          (EventDispatchers|BindEventtoOnReloadStarted _currentweapon (delegate Custom|TriggerReactivity)))
        (:else)))))

(event Custom|TriggerReactivity
  (params (ReloadTime :type Float%20%28single-precision%29) (ReloadType :type EFortWeaponReloadType%20Enum :default "ReloadWholeClip"))
  (bind _returnvalue (CustomItemWrap|GetWrappedMeshComponents))
  (bind _returnvalue_2 (+ ReloadTime 2.0))
  ;[0.01042 0.01042 0.01042 0.8] Reload Glow and fade
  (Rendering|Material|SetScalarParameterValueonMaterials _returnvalue "TopS Emissive_01_Brightness---Duration" _returnvalue_2)
  (Rendering|Material|SetScalarParameterValueonMaterials _returnvalue "TopS Emissive_01_Brightness---Start" 3.0)
  (Rendering|Material|SetScalarParameterValueonMaterials _returnvalue "TopS Emissive_01_Brightness---End" 0.0)
  (Rendering|Material|SetScalarParameterValueonMaterials _returnvalue "TopS Emissive_01-Brightness-TBP" (Utilities|Time|GetGameTimeinSeconds))
  (Rendering|Material|SetVectorParameterValueonMaterials _returnvalue "TopS_Em01_Speed__Master" "1.000000,0.000000,0.000000")
  (Utilities|FlowControl|Delay _returnvalue_2)
  ;[0.005208 0.005208 0.005208 0.8] Glow back to idle loop
  (Rendering|Material|SetVectorParameterValueonMaterials _returnvalue "TopS_Em01_Speed__Master" "0.400000,0.000000,0.000000")
  (Rendering|Material|SetScalarParameterValueonMaterials _returnvalue "TopS Emissive_01_Brightness---Duration" 1.0)
  (Rendering|Material|SetScalarParameterValueonMaterials _returnvalue "TopS Emissive_01_Brightness---Start" 0.0)
  (Rendering|Material|SetScalarParameterValueonMaterials _returnvalue "TopS Emissive_01-Brightness-TBP" (Utilities|Time|GetGameTimeinSeconds))
  (Rendering|Material|SetScalarParameterValueonMaterials _returnvalue "TopS Emissive_01_Brightness---End" 0.08))

(event EventOnWeaponFired
  (Rendering|Material|SetScalarParameterValueonMaterials (CustomItemWrap|GetWrappedMeshComponents) "TopS Emissive_03-Brightness-TBP" (Utilities|Time|GetGameTimeinSeconds)))

(fn ConstructionScript)
