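; Reactive Bionic Kitty item wrap that changes a flipbook decal frame and emissive glow on wrapped meshes in response to combat.
; On BeginPlay and pawn set it watches damage and kill stats, or just plays the glow timeline in frontend display.
; Damage and kills set frames behind a shared IsKilling lock, which only the damage path clears by timer.
; Firing, unless locked, pulses glow and briefly swaps frame; weapon changes set emissive levels.
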(event Custom|Revert
  (Utilities|Time|SetTimerbyEvent (delegate Custom|SetBaseFrame) 5.0))

(event EventOnPawnSet
  (bind _returnvalue (CustomItemWrap|GetWrappedMeshComponents))
  (bind _returnvalue_2 (CustomCharacter|GetAssociatedPlayerPawn))
  (if (not (Reactive|Frontend|IsCosmeticObjectDisplayedinFrontendContext 0))
    (Utilities|FlowControl|Sequence
      (:then_0
        (Cosmetics|WatchCosmeticStat _returnvalue_2 "(TagName=\"StatManager.DamageReceived\")" 0 false
          (:OnStatChanged
            (if (not (Variables|Default|GetIsKilling?))
              (Utilities|FlowControl|Sequence
                (:then_0
                  (Variables|Default|SetIsKilling? true)
                  (Rendering|Material|SetScalarParameterValueonMaterials _returnvalue "FDecalV2_FlipBook_Frame #" 0.7)
                  (CallFunction|Revert))
                (:then_1
                  _))))))
      (:then_1
        (Cosmetics|WatchCosmeticStat _returnvalue_2 "(TagName=\"StatManager.AthenaKills\")" 0 false
          (:OnStatChanged
            (if (not (Variables|Default|GetIsKilling?))
              (Utilities|FlowControl|Sequence
                (:then_0
                  (Variables|Default|SetIsKilling? true)
                  (Rendering|Material|SetScalarParameterValueonMaterials _returnvalue "FDecalV2_FlipBook_Frame #" 0.17))
                (:then_1
                  _)))))))
    (CallFunction|Glow)))

(event EventBeginPlay
  (bind _returnvalue (CustomItemWrap|GetWrappedMeshComponents))
  (bind _returnvalue_2 (CustomCharacter|GetAssociatedPlayerPawn))
  (if (not (Reactive|Frontend|IsCosmeticObjectDisplayedinFrontendContext 0))
    (Utilities|FlowControl|Sequence
      (:then_0
        (Cosmetics|WatchCosmeticStat _returnvalue_2 "(TagName=\"StatManager.DamageReceived\")" 0 false
          (:OnStatChanged
            (if (not (Variables|Default|GetIsKilling?))
              (Utilities|FlowControl|Sequence
                (:then_0
                  (Variables|Default|SetIsKilling? true)
                  (Rendering|Material|SetScalarParameterValueonMaterials _returnvalue "FDecalV2_FlipBook_Frame #" 0.7)
                  (CallFunction|Revert))
                (:then_1
                  _))))))
      (:then_1
        (Cosmetics|WatchCosmeticStat _returnvalue_2 "(TagName=\"StatManager.AthenaKills\")" 0 false
          (:OnStatChanged
            (if (not (Variables|Default|GetIsKilling?))
              (Utilities|FlowControl|Sequence
                (:then_0
                  (Variables|Default|SetIsKilling? true)
                  (Rendering|Material|SetScalarParameterValueonMaterials _returnvalue "FDecalV2_FlipBook_Frame #" 0.17))
                (:then_1
                  _)))))))
    (CallFunction|Glow)))

(event Custom|SetBaseFrame
  (Rendering|Material|SetScalarParameterValueonMaterials (CustomItemWrap|GetWrappedMeshComponents) "FDecalV2_FlipBook_Frame #" 0.5)
  (Variables|Default|SetIsKilling? false))

(event Custom|Glow
  (Utilities|FlowControl|Sequence
    (:then_0
      (bind (_direction _newtrack_0) (Timeline|Timeline Stop)
        (:Update
          (Rendering|Material|SetScalarParameterValueonMaterials (CustomItemWrap|GetWrappedMeshComponents) "TopS Emissive_02_ColorCuve-Brightness" (Math|Float|Lerp 0.0 1.6 _newtrack_0)))))
    (:then_1
      (Timeline|Timeline PlayFromStart))))

(event EventOnWeaponChanged
  (params (NewWeapon :type Fort%20Weapon%20Object%20Reference) (OldWeapon :type Fort%20Weapon%20Object%20Reference))
  (bind _returnvalue (CustomItemWrap|GetWrappedMeshComponents))
  (Rendering|Material|SetScalarParameterValueonMaterials _returnvalue "TopS_02 Emissive Brightness" 0.0)
  (Rendering|Material|SetScalarParameterValueonMaterials _returnvalue "TopS Emissive_01_ColorCuve-Brightness" 1.5))

(event EventOnWeaponFired
  (bind _returnvalue (CustomItemWrap|GetWrappedMeshComponents))
  (if (not (Variables|Default|GetIsKilling?))
    (Utilities|FlowControl|Sequence
      (:then_0
        (CallFunction|Glow))
      (:then_1
        (Rendering|Material|SetScalarParameterValueonMaterials _returnvalue "FDecalV2_FlipBook_Frame #" 0.34)
        (Utilities|FlowControl|RetriggerableDelay 1.0)
        (if (not (Variables|Default|GetIsKilling?))
          (Rendering|Material|SetScalarParameterValueonMaterials _returnvalue "FDecalV2_FlipBook_Frame #" 0.5))))))

(fn ConstructionScript)
