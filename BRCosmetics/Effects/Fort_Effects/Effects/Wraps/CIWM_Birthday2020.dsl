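; Weapon wrap logic for the Birthday 2020 cosmetic that triggers a confetti material effect each time the weapon fires.
; On each shot it writes the current game time to the wrapped meshes' material parameters, with the opacity time offset slightly later.
; A Switch boolean alternates between confetti sets A and B on successive shots, so each shot restarts only one of the two time-driven effects.
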(event EventOnWeaponFired
  (bind _returnvalue (CustomItemWrap|GetWrappedMeshComponents))
  (bind _returnvalue_2 (Utilities|Time|GetGameTimeinSeconds))
  (bind _returnvalue_3 (+ _returnvalue_2 0.01))
  (if (Variables|Default|GetSwitch)
    (Rendering|Material|SetScalarParameterValueonMaterials _returnvalue "BD_ConfettiA_Opacity_Time" _returnvalue_3)
    (Rendering|Material|SetScalarParameterValueonMaterials _returnvalue "BD_ConfettiA_Main_Time" _returnvalue_2)
    (Variables|Default|SetSwitch false)
    (else
      (Rendering|Material|SetScalarParameterValueonMaterials _returnvalue "BD_ConfettiB_Opacity_Time" _returnvalue_3)
      (Rendering|Material|SetScalarParameterValueonMaterials _returnvalue "BD_ConfettiB_Main_Time" _returnvalue_2)
      (Variables|Default|SetSwitch true))))

(fn ConstructionScript)
