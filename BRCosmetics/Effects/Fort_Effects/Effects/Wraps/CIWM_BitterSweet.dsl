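; BitterSweet reactive weapon wrap that progressively breaks as the player takes damage and restores when they gain health.
; Each damage stat change sets one scalar material parameter from a shuffled list to 1 and records it as fired.
; Health gain, or a short delay after the list empties, clears fired parameters in a staggered loop, then Reset reshuffles.
; In frontend display, a looping LockerBreak timer simulates damage; damage callbacks also leave a debug print.
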(event EventOnWeaponChanged
  (params (NewWeapon :type Fort%20Weapon%20Object%20Reference) (OldWeapon :type Fort%20Weapon%20Object%20Reference)))

(event EventOnNewWeapons
  (params (WeaponsAsItems :type Array%20of%20Fort%20Item%20Object%20References))
  (for _array_element (Variables|Default|GetBaseProperties)
    (Rendering|Material|SetScalarParameterValueonMaterials (CustomItemWrap|GetWrappedMeshComponents) _array_element 0.0))
  (bind _output_get (Variables|Default|SetCurrentProperties (Variables|Default|GetBaseProperties)))
  (Utilities|Array|Shuffle _output_get)
  (Utilities|Array|Clear (Variables|Default|GetFiredProperties))
  (Variables|Default|SetFullyRestored true)
  (Rendering|Material|SetVectorParameterValueonMaterials (CustomItemWrap|GetWrappedMeshComponents) "TopS_Blend_Offset__Master" "2.000000,0.0,0.0"))

(event Custom|Reset
  (for _array_element (Variables|Default|GetBaseProperties)
    (Rendering|Material|SetScalarParameterValueonMaterials (CustomItemWrap|GetWrappedMeshComponents) _array_element 0.0))
  (bind _output_get (Variables|Default|SetCurrentProperties (Variables|Default|GetBaseProperties)))
  (Utilities|Array|Shuffle _output_get)
  (Utilities|Array|Clear (Variables|Default|GetFiredProperties))
  (Variables|Default|SetFullyRestored true)
  (Rendering|Material|SetVectorParameterValueonMaterials (CustomItemWrap|GetWrappedMeshComponents) "TopS_Blend_Offset__Master" "2.000000,0.0,0.0"))

(event EventOnNewPlayerWeaponType
  (params (NewType :type EFortItemType%20Enum :default "WorldItem") (OldType :type EFortItemType%20Enum :default "WorldItem")))

(event Custom|Heal
  (bind _output_get (Variables|Default|SetCurrentProperties (Variables|Default|GetBaseProperties)))
  (Utilities|Array|Shuffle _output_get)
  (Utilities|Array|Clear (Variables|Default|GetFiredProperties))
  (Variables|Default|SetFullyRestored true)
  (Rendering|Material|SetVectorParameterValueonMaterials (CustomItemWrap|GetWrappedMeshComponents) "TopS_Blend_Offset__Master" "2.000000,0.0,0.0"))

(event EventBeginPlay
  (bind _currentproperties (Variables|Default|GetCurrentProperties))
  (bind _output (Utilities|Array|Get(acopy) _currentproperties))
  (if (Reactive|Frontend|IsCosmeticObjectDisplayedinFrontendContext 0)
    (Utilities|Time|SetTimerbyFunctionName 0 "LockerBreak" 1.0 true)
    (else
      (Utilities|FlowControl|Sequence
        (:then_0
          (bind (_statname _statvalue) (Cosmetics|WatchCosmeticStat (CustomCharacter|GetAssociatedPlayerPawn) "(TagName=\"StatManager.DamageReceived\")" 0 false)
            (:OnStatChanged
              (Development|PrintString (Utilities|String|ToString(Integer) _statvalue) true true "(R=0.426087,G=0.042071,B=0.000000,A=1.000000)")
              (if (Variables|Default|GetFullyRestored)
                (if (> (Utilities|Array|Length (Variables|Default|GetCurrentProperties)) 0)
                  (Rendering|Material|SetScalarParameterValueonMaterials (CustomItemWrap|GetWrappedMeshComponents) _output 1.0)
                  (Utilities|FlowControl|Delay 0.2)
                  (Utilities|Array|Add (Variables|Default|GetFiredProperties) _output)
                  (Utilities|Array|RemoveItem (Variables|Default|GetCurrentProperties) _output)
                  (else
                    (Variables|Default|SetFullyRestored false)
                    (Utilities|FlowControl|Delay 2.0)
                    (CallFunction|Unbreak)))))))
        (:then_1
          (Cosmetics|WatchCosmeticStat (CustomCharacter|GetAssociatedPlayerPawn) "(TagName=\"StatManager.GainedHealthTimes\")" 0 false
            (:OnStatChanged
              (Variables|Default|SetFullyRestored false)
              (bind _arrayelement (Utilities|ForEachLoopwithDelay (Variables|Default|GetFiredProperties) 0.2)
                (:LoopBody
                  (Rendering|Material|SetScalarParameterValueonMaterials (CustomItemWrap|GetWrappedMeshComponents) _arrayelement 0.0))
                (:Completed
                  (CallFunction|Reset))))))))))

(event Custom|LockerBreak
  (bind _currentproperties (Variables|Default|GetCurrentProperties))
  (bind _output (Utilities|Array|Get(acopy) _currentproperties))
  (if (Variables|Default|GetFullyRestored)
    (if (> (Utilities|Array|Length (Variables|Default|GetCurrentProperties)) 0)
      (Rendering|Material|SetScalarParameterValueonMaterials (CustomItemWrap|GetWrappedMeshComponents) _output 1.0)
      (Utilities|FlowControl|Delay 0.2)
      (Utilities|Array|Add (Variables|Default|GetFiredProperties) _output)
      (Utilities|Array|RemoveItem (Variables|Default|GetCurrentProperties) _output)
      (else
        (Variables|Default|SetFullyRestored false)
        (Utilities|FlowControl|Delay 2.0)
        (CallFunction|Unbreak)))))

(event EventOnPawnSet
  (bind _currentproperties (Variables|Default|GetCurrentProperties))
  (bind _output (Utilities|Array|Get(acopy) _currentproperties))
  (Utilities|FlowControl|Sequence
    (:then_0
      (bind (_statname _statvalue) (Cosmetics|WatchCosmeticStat (CustomCharacter|GetAssociatedPlayerPawn) "(TagName=\"StatManager.DamageReceived\")" 0 false)
        (:OnStatChanged
          (Development|PrintString (Utilities|String|ToString(Integer) _statvalue) true true "(R=0.426087,G=0.042071,B=0.000000,A=1.000000)")
          (if (Variables|Default|GetFullyRestored)
            (if (> (Utilities|Array|Length (Variables|Default|GetCurrentProperties)) 0)
              (Rendering|Material|SetScalarParameterValueonMaterials (CustomItemWrap|GetWrappedMeshComponents) _output 1.0)
              (Utilities|FlowControl|Delay 0.2)
              (Utilities|Array|Add (Variables|Default|GetFiredProperties) _output)
              (Utilities|Array|RemoveItem (Variables|Default|GetCurrentProperties) _output)
              (else
                (Variables|Default|SetFullyRestored false)
                (Utilities|FlowControl|Delay 2.0)
                (CallFunction|Unbreak)))))))
    (:then_1
      (Cosmetics|WatchCosmeticStat (CustomCharacter|GetAssociatedPlayerPawn) "(TagName=\"StatManager.GainedHealthTimes\")" 0 false
        (:OnStatChanged
          (Variables|Default|SetFullyRestored false)
          (bind _arrayelement (Utilities|ForEachLoopwithDelay (Variables|Default|GetFiredProperties) 0.2)
            (:LoopBody
              (Rendering|Material|SetScalarParameterValueonMaterials (CustomItemWrap|GetWrappedMeshComponents) _arrayelement 0.0))
            (:Completed
              (CallFunction|Reset))))))))

(event Custom|Unbreak
  (if (not (Reactive|Frontend|IsCosmeticObjectDisplayedinFrontendContext 0))
    (Cosmetics|WatchCosmeticStat (CustomCharacter|GetAssociatedPlayerPawn) "(TagName=\"StatManager.GainedHealthTimes\")" 0 false
      (:OnStatChanged
        _)))
  (Variables|Default|SetFullyRestored false)
  (bind _arrayelement (Utilities|ForEachLoopwithDelay (Variables|Default|GetFiredProperties) 0.2)
    (:LoopBody
      (Rendering|Material|SetScalarParameterValueonMaterials (CustomItemWrap|GetWrappedMeshComponents) _arrayelement 0.0))
    (:Completed
      (CallFunction|Reset))))

(fn ConstructionScript)
