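; Reactive cosmetic wrap Blueprint (BoldTouch) that switches a decal flipbook frame on wrapped meshes in response to combat.
; Damage-stat watchers, bound on BeginPlay and pawn set, and weapon fire each show their own frame, then restore the default after a retriggerable delay unless Killed is set.
; Firing also stamps game time into material parameters for timed effects.
; In frontend display, a looping timer runs LockerCycle to preview the frames.
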(event EventBeginPlay
  (bind _returnvalue (CustomItemWrap|GetWrappedMeshComponents))
  (if (Reactive|Frontend|IsCosmeticObjectDisplayedinFrontendContext 0)
    (Utilities|Time|SetTimerbyFunctionName 0 "LockerCycle" 2.5 true)
    (else
      (Utilities|FlowControl|Sequence
        (:then_0
          (Cosmetics|WatchCosmeticStat (CustomCharacter|GetAssociatedPlayerPawn) "(TagName=\"StatManager.DamageReceived\")" 0 false
            (:OnStatChanged
              (Rendering|Material|SetScalarParameterValueonMaterials _returnvalue "FDecalV2_FlipBook_Frame #" 0.66)
              (Utilities|FlowControl|RetriggerableDelay 0.5)
              (if (not (Variables|Default|GetKilled))
                (Rendering|Material|SetScalarParameterValueonMaterials _returnvalue "FDecalV2_FlipBook_Frame #" 0.0)))))))))

(event EventOnPawnSet
  (bind _returnvalue (CustomItemWrap|GetWrappedMeshComponents))
  (Utilities|FlowControl|Sequence
    (:then_0
      (Cosmetics|WatchCosmeticStat (CustomCharacter|GetAssociatedPlayerPawn) "(TagName=\"StatManager.DamageReceived\")" 0 false
        (:OnStatChanged
          (Rendering|Material|SetScalarParameterValueonMaterials _returnvalue "FDecalV2_FlipBook_Frame #" 0.66)
          (Utilities|FlowControl|RetriggerableDelay 0.5)
          (if (not (Variables|Default|GetKilled))
            (Rendering|Material|SetScalarParameterValueonMaterials _returnvalue "FDecalV2_FlipBook_Frame #" 0.0)))))))

(event EventOnWeaponFired
  (bind _returnvalue (Utilities|Time|GetGameTimeinSeconds))
  (bind _returnvalue_2 (CustomItemWrap|GetWrappedMeshComponents))
  (bind _returnvalue_3 (CustomItemWrap|GetWrappedMeshComponents))
  (Rendering|Material|SetScalarParameterValueonMaterials _returnvalue_3 "TimeFrom-BP----------------" _returnvalue)
  (Rendering|Material|SetScalarParameterValueonMaterials _returnvalue_3 "EDRec_Emissive_A_Brightness_Time_BP" _returnvalue)
  (if (not (Variables|Default|GetKilled))
    (Rendering|Material|SetScalarParameterValueonMaterials _returnvalue_2 "FDecalV2_FlipBook_Frame #" 0.33)
    (Utilities|FlowControl|RetriggerableDelay 0.5)
    (if (not (Variables|Default|GetKilled))
      (Rendering|Material|SetScalarParameterValueonMaterials _returnvalue_2 "FDecalV2_FlipBook_Frame #" 0.0))))

(event Custom|LockerCycle
  (if (>= (Variables|Default|GetLockerCounter) 3)
    (Variables|Default|SetLockerCounter 0)
    (else
      (Math|Integer|IncrementInt (Variables|Default|GetLockerCounter)
        (:%20%20
          (Rendering|Material|SetScalarParameterValueonMaterials (CustomItemWrap|GetWrappedMeshComponents) "FDecalV2_FlipBook_Frame #" (Utilities|Select :Index (Variables|Default|GetLockerCounter) :Option%200 0.0 :Option%201 0.0 :Option%202 0.33 :Option%203 0.66)))))))

(fn ConstructionScript)
