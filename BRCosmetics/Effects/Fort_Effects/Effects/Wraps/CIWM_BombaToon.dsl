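; Blueprint logic for the BombaToon item wrap, animating material parameters on the wrapped mesh components.
; Setting the wrap target starts a self-calling loop that toggles a top-layer blend opacity between 0 and 1 at one-second intervals.
; Firing the weapon switches a decal flipbook frame, and a short timer reverts it.
; LoopFlickerIdle, RandomSet and the construction script have empty bodies.
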(event EventOnWeaponFired
  (Rendering|Material|SetScalarParameterValueonMaterials (CustomItemWrap|GetWrappedMeshComponents) "FDecalV2_FlipBook_Frame #" 1.0)
  ; Revert
  (Utilities|Time|SetTimerbyEvent (delegate Custom|Revert) 0.2))

(event Custom|Revert
  ; Revert
  (Rendering|Material|SetScalarParameterValueonMaterials (CustomItemWrap|GetWrappedMeshComponents) "FDecalV2_FlipBook_Frame #" 0.0))

(event Custom|LoopAnim
  (bind _returnvalue (CustomItemWrap|GetWrappedMeshComponents))
  ; Base Layer Swap
  (Rendering|Material|SetScalarParameterValueonMaterials _returnvalue "TopS_01_Blend_Opacity---End" 0.0)
  (Utilities|FlowControl|Delay 1.0)
  (Rendering|Material|SetScalarParameterValueonMaterials _returnvalue "TopS_01_Blend_Opacity---End" 1.0)
  (Utilities|FlowControl|Delay 1.0)
  (CallFunction|LoopAnim))

(event EventOnSetWrapTarget
  ; Base Layer Swap
  (CallFunction|LoopAnim))

(fn LoopFlickerIdle)

(fn RandomSet
  (params (TimeBP_Flipbook%20Triger :type Name) (Offset :type Name) (Rotation :type Name) (X=min :type Float%20%28double-precision%29) (X=Max :type Float%20%28double-precision%29) (Y=Min :type Float%20%28double-precision%29) (Y=Max :type Float%20%28double-precision%29) (Scale=Min :type Float%20%28double-precision%29) (Scale=Max :type Float%20%28double-precision%29) (opacity :type Name)))

(fn ConstructionScript)
