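; Item wrap Blueprint for BounceRain that triggers reactive decal, emissive and gradient material effects from weapon fire and kills.
; A shot counter outside the front end and an AthenaKills cosmetic stat watch both write the current game time into time-based scalar parameters on wrapped meshes.
; A self-restarting front-end timer chain counts iterations but sets no material parameters, and RestartFE is empty.
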(event Custom|RestartAnim
  (Utilities|Time|SetTimerbyEvent (delegate Custom|FrontendKO) 5.0))

(event Custom|FrontendKO
  (bind _returnvalue (Utilities|Time|SetTimerbyEvent (delegate Custom|FEExtended_KO) 0.5 true))
  (Variables|Default|SetCurrentTimer _returnvalue))

(event Custom|FEExtended_KO
  (bind _result (Math|Integer|IncrementInt (Variables|Default|GetIterationCount))
    (:%20%20
      (if (>= _result 10)
        (Utilities|Time|ClearandInvalidateTimerbyHandle (Variables|Default|GetCurrentTimer))
        (Variables|Default|SetIterationCount 0)
        (CallFunction|RestartAnim)))))

(event EventOnWeaponFired
  (bind _returnvalue (CustomItemWrap|GetWrappedMeshComponents))
  (bind _returnvalue_2 (Utilities|Time|GetGameTimeinSeconds))
  (bind _returnvalue_3 (CustomItemWrap|GetWrappedMeshComponents))
  (bind _returnvalue_4 (Utilities|Time|GetGameTimeinSeconds))
  (bind _returnvalue_5 (CustomItemWrap|GetWrappedMeshComponents))
  (bind _returnvalue_6 (Utilities|Time|GetGameTimeinSeconds))
  (if (CustomCosmeticModifier|IsinFrontEnd)
    (Utilities|FlowControl|Delay 0.001)
    (bind _output_get (Variables|Default|SetShotsFired (+ (Variables|Default|GetShotsFired) 1)))
    (if (>= _output_get 5)
      (if (== (Variables|Default|GetShotsFired) 7)
        (Variables|Default|SetShotsFired 0)
        ;[1 1 1 0.5] Decal Reactive
        (Rendering|Material|SetScalarParameterValueonMaterials _returnvalue "EDRec_TimeFrom_BP_DecalA" _returnvalue_2)
        (Rendering|Material|SetScalarParameterValueonMaterials _returnvalue "EDRec_TimeFrom_BP_DecalB" _returnvalue_2)
        (Rendering|Material|SetScalarParameterValueonMaterials _returnvalue "EDRec_TimeFrom_BP_DecalC" _returnvalue_2)
        (Rendering|Material|SetScalarParameterValueonMaterials _returnvalue "EDRec_TimeFrom_BP_DecalD" _returnvalue_2)
        (Rendering|Material|SetScalarParameterValueonMaterials _returnvalue_5 "EDRec_Emissive_A_Brightness_Time_BP" _returnvalue_6)
        (Rendering|Material|SetScalarParameterValueonMaterials _returnvalue_5 "EDRec_Emissive_B_Brightness_Time_BP" _returnvalue_6)
        (Rendering|Material|SetScalarParameterValueonMaterials _returnvalue_5 "EDRec_Emissive_C_Brightness_Time_BP" _returnvalue_6)
        (Rendering|Material|SetScalarParameterValueonMaterials _returnvalue_5 "EDRec_Emissive_D_Brightness_Time_BP" _returnvalue_6)
        (Rendering|Material|SetScalarParameterValueonMaterials _returnvalue_3 "EDRec_Emissive_A_Gradent_Time_BP" _returnvalue_4)
        (Rendering|Material|SetScalarParameterValueonMaterials _returnvalue_3 "EDRec_Emissive_B_Gradent_Time_BP" _returnvalue_4)
        (Rendering|Material|SetScalarParameterValueonMaterials _returnvalue_3 "EDRec_Emissive_C_Gradent_Time_BP" _returnvalue_4)
        (Rendering|Material|SetScalarParameterValueonMaterials _returnvalue_3 "EDRec_Emissive_D_Gradent_Time_BP" _returnvalue_4)
        (Rendering|Material|SetScalarParameterValueonMaterials _returnvalue_3 "TopS Emissive_01-Brightness-TBP" _returnvalue_4)))))

(event EventOnPawnSet
  (bind _returnvalue (CustomItemWrap|GetWrappedMeshComponents))
  (bind _returnvalue_2 (Utilities|Time|GetGameTimeinSeconds))
  (bind _returnvalue_3 (CustomItemWrap|GetWrappedMeshComponents))
  (bind _returnvalue_4 (Utilities|Time|GetGameTimeinSeconds))
  (bind _returnvalue_5 (CustomItemWrap|GetWrappedMeshComponents))
  (bind _returnvalue_6 (Utilities|Time|GetGameTimeinSeconds))
  (Cosmetics|WatchCosmeticStat (CustomCharacter|GetAssociatedPlayerPawn) "(TagName=\"StatManager.AthenaKills\")"
    (:OnStatChanged
      ;[1 1 1 0.5] Decal Reactive
      (Rendering|Material|SetScalarParameterValueonMaterials _returnvalue "EDRec_TimeFrom_BP_DecalA" _returnvalue_2)
      (Rendering|Material|SetScalarParameterValueonMaterials _returnvalue "EDRec_TimeFrom_BP_DecalB" _returnvalue_2)
      (Rendering|Material|SetScalarParameterValueonMaterials _returnvalue "EDRec_TimeFrom_BP_DecalC" _returnvalue_2)
      (Rendering|Material|SetScalarParameterValueonMaterials _returnvalue "EDRec_TimeFrom_BP_DecalD" _returnvalue_2)
      (Rendering|Material|SetScalarParameterValueonMaterials _returnvalue_5 "EDRec_Emissive_A_Brightness_Time_BP" _returnvalue_6)
      (Rendering|Material|SetScalarParameterValueonMaterials _returnvalue_5 "EDRec_Emissive_B_Brightness_Time_BP" _returnvalue_6)
      (Rendering|Material|SetScalarParameterValueonMaterials _returnvalue_5 "EDRec_Emissive_C_Brightness_Time_BP" _returnvalue_6)
      (Rendering|Material|SetScalarParameterValueonMaterials _returnvalue_5 "EDRec_Emissive_D_Brightness_Time_BP" _returnvalue_6)
      (Rendering|Material|SetScalarParameterValueonMaterials _returnvalue_3 "EDRec_Emissive_A_Gradent_Time_BP" _returnvalue_4)
      (Rendering|Material|SetScalarParameterValueonMaterials _returnvalue_3 "EDRec_Emissive_B_Gradent_Time_BP" _returnvalue_4)
      (Rendering|Material|SetScalarParameterValueonMaterials _returnvalue_3 "EDRec_Emissive_C_Gradent_Time_BP" _returnvalue_4)
      (Rendering|Material|SetScalarParameterValueonMaterials _returnvalue_3 "EDRec_Emissive_D_Gradent_Time_BP" _returnvalue_4)
      (Rendering|Material|SetScalarParameterValueonMaterials _returnvalue_3 "TopS Emissive_01-Brightness-TBP" _returnvalue_4))))

(fn RestartFE)

(fn ConstructionScript)
