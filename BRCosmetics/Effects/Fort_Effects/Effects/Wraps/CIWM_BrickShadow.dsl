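; Item wrap modifier for the BrickShadow wrap, animating material parameters on the wrapped meshes in the front end and on weapon fire.
; After a short delay on begin play, a front-end check enables a time-of-day parameter and plays a timeline driving scalar and color overrides.
; Each shot starts one of two flipbook timelines, each driving a frame pair.
; Busy flags let a second shot overlap the first and ignore shots while both play.
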(event EventBeginPlay
  (bind _returnvalue (CustomItemWrap|GetWrappedMeshComponents))
  (Utilities|FlowControl|Delay 0.2)
  (if (CustomCosmeticModifier|IsinFrontEnd)
    (Rendering|Material|SetScalarParameterValueonMaterials (CustomItemWrap|GetWrappedMeshComponents) "00-FE-TOD--" 1.0)
    (bind (_direction _tod_value _newtrack_1 _newtrack_0) (Timeline|Timeline PlayFromStart)
      (:Update
        (Rendering|Material|SetScalarParameterValueonMaterials _returnvalue "00-FE-TODCC-Override" _tod_value)
        (Rendering|Material|SetColorParameterValueonMaterials _returnvalue "00-FE-TODV4-Override" _newtrack_0)))))

(event EventOnWeaponFired
  (bind _returnvalue (CustomItemWrap|GetWrappedMeshComponents))
  (bind _returnvalue_2 (CustomItemWrap|GetWrappedMeshComponents))
  (if (Variables|Default|GetFire1)
    (Variables|Default|SetFire1 false)
    (bind (_direction _newtrack_0) (Timeline|Timeline_0 PlayFromStart)
      (:Update
        (Rendering|Material|SetScalarParameterValueonMaterials _returnvalue "EDRec_FlipBook_FrameA" _newtrack_0)
        (Rendering|Material|SetScalarParameterValueonMaterials _returnvalue "EDRec_FlipBook_FrameC" _newtrack_0))
      (:Finished
        (Variables|Default|SetFire1 true)))
    (elif (Variables|Default|GetFire2)
      (Variables|Default|SetFire2 false)
      (bind (_direction_2 _newtrack_0_2) (Timeline|Timeline_1 PlayFromStart)
        (:Update
          (Rendering|Material|SetScalarParameterValueonMaterials _returnvalue_2 "EDRec_FlipBook_FrameD" _newtrack_0_2)
          (Rendering|Material|SetScalarParameterValueonMaterials _returnvalue_2 "EDRec_FlipBook_FrameB" _newtrack_0_2))
        (:Finished
          (Variables|Default|SetFire2 true))))))

(fn ConstructionScript)
