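; Reactive weapon wrap Blueprint for BurnLeer that animates emissive and warp material parameters on the wrapped meshes when the weapon fires.
; Each shot stamps game time into brightness and opacity parameters, overriding their End values unless Killed is set.
; Firing also starts a timer-driven ramp of UV warp, scale and emissive multiplier, reversed after a retriggerable delay without shots.
; A frontend LockerCycle timer only resets a counter.
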(event EventBeginPlay
  (if (Reactive|Frontend|IsCosmeticObjectDisplayedinFrontendContext 0)
    (Utilities|Time|SetTimerbyFunctionName 0 "LockerCycle" 3.0 true)))

(event EventOnWeaponFired
  (bind _returnvalue (CustomItemWrap|GetWrappedMeshComponents))
  (bind _returnvalue_2 (Utilities|Time|GetGameTimeinSeconds))
  (Utilities|FlowControl|Sequence
    (:then_0
      (Rendering|Material|SetScalarParameterValueonMaterials _returnvalue "TopS Emissive_03-Brightness-TBP" _returnvalue_2)
      (Rendering|Material|SetScalarParameterValueonMaterials _returnvalue "EDRec_Emissive_C_Brightness_Time_BP" _returnvalue_2)
      (Rendering|Material|SetScalarParameterValueonMaterials _returnvalue "EDRec_Emissive_A_Brightness_Time_BP" _returnvalue_2)
      (Rendering|Material|SetScalarParameterValueonMaterials _returnvalue "TopS Emissive_01-Brightness-TBP" _returnvalue_2)
      (Rendering|Material|SetScalarParameterValueonMaterials _returnvalue "TopS_01_Blend_Opacity-TBP" _returnvalue_2)
      (Rendering|Material|SetScalarParameterValueonMaterials _returnvalue "TopS Emissive_02-Brightness-TBP" _returnvalue_2)
      (Rendering|Material|SetScalarParameterValueonMaterials _returnvalue "TopS_01_Blend_Opacity---End" _returnvalue_2)
      (Rendering|Material|SetScalarParameterValueonMaterials _returnvalue "TopS Emissive_02_Brightness---End" _returnvalue_2)
      (Rendering|Material|SetScalarParameterValueonMaterials _returnvalue "TopS Emissive_01_Brightness---End" _returnvalue_2)
      (Rendering|Material|SetScalarParameterValueonMaterials _returnvalue "EDRec_Emissive_A_Brightness_Time__End" _returnvalue_2)
      (Rendering|Material|SetScalarParameterValueonMaterials _returnvalue "EDRec_Emissive_C_Brightness_Time__End" _returnvalue_2)
      (Rendering|Material|SetScalarParameterValueonMaterials _returnvalue "TopS Emissive_03_Brightness---End" _returnvalue_2)
      (if (not (Variables|Default|GetKilled))
        (Rendering|Material|SetScalarParameterValueonMaterials _returnvalue "TopS_01_Blend_Opacity---End" 0.0)
        (Rendering|Material|SetScalarParameterValueonMaterials _returnvalue "TopS Emissive_02_Brightness---End" 0.085)
        (Rendering|Material|SetScalarParameterValueonMaterials _returnvalue "TopS Emissive_01_Brightness---End" 0.085)
        (Rendering|Material|SetScalarParameterValueonMaterials _returnvalue "EDRec_Emissive_A_Brightness_Time__End" 0.0)
        (Rendering|Material|SetScalarParameterValueonMaterials _returnvalue "EDRec_Emissive_C_Brightness_Time__End" 0.0)
        (Rendering|Material|SetScalarParameterValueonMaterials _returnvalue "TopS Emissive_03_Brightness---End" 0.1)))
    (:then_1
      (if (not (Variables|Default|GetFullAutoWeapons))
        (Variables|Default|SetFullAutoWeapons true)
        (CallFunction|SineAnimation))
      (Utilities|FlowControl|RetriggerableDelay 3.0)
      (CallFunction|SineAnimOFF)
      (Variables|Default|SetFullAutoWeapons false))))

(event Custom|LockerCycle
  (if (>= (Variables|Default|GetLockerCounter) 3)
    (Variables|Default|SetLockerCounter 0)))

(event Custom|SineAnimation
  ; Layer Offset
  (Variables|Default|SetReverse false)
  (Variables|Default|SetStartTime (Utilities|Time|GetGameTimeinSeconds))
  (bind _returnvalue (Utilities|Time|SetTimerbyEvent (delegate Custom|SineAnimUpdate) 0.02 true))
  (Variables|Default|SetTimerHandle _returnvalue))

(event Custom|SineAnimOFF
  ; Layer Offset
  (Variables|Default|SetReverse true)
  (Variables|Default|SetStartTime (Utilities|Time|GetGameTimeinSeconds))
  (bind _returnvalue (Utilities|Time|SetTimerbyEvent (delegate Custom|SineAnimUpdate) 0.02 true))
  (Variables|Default|SetTimerHandle _returnvalue))

(event Custom|SineAnimUpdate
  (bind _returnvalue (Utilities|Time|GetGameTimeinSeconds))
  (bind _returnvalue_2 (CustomItemWrap|GetWrappedMeshComponents))
  (bind _starttime (Variables|Default|GetStartTime))
  (bind _returnvalue_3 (- _returnvalue _starttime))
  ; Layer Offset
  (if (>= _returnvalue_3 5.0)
    (Utilities|Time|ClearandInvalidateTimerbyHandle (Variables|Default|GetTimerHandle))
    (if (Variables|Default|GetReverse)
      ;; Start
      (Rendering|Material|SetScalarParameterValueonMaterials _returnvalue_2 "TopS_04_Warp_UV_Amount" 0.015)
      (Rendering|Material|SetScalarParameterValueonMaterials _returnvalue_2 "TopS_04_Warp_UV_Offset" 0.0)
      (Rendering|Material|SetScalarParameterValueonMaterials _returnvalue_2 "TopS_03_Warp_UV_Amount" 0.01)
      (Rendering|Material|SetColorParameterValueonMaterials _returnvalue_2 "TopS_Em03_Scale__Master" (0.4 2.0 1.0 1.0))
      (Rendering|Material|SetScalarParameterValueonMaterials _returnvalue_2 "EDRec_Emissive Multplyer" 1.0)
      (else
        ;; Arrive
        (Rendering|Material|SetScalarParameterValueonMaterials _returnvalue_2 "TopS_04_Warp_UV_Amount" 0.045)
        (Rendering|Material|SetScalarParameterValueonMaterials _returnvalue_2 "TopS_04_Warp_UV_Offset" -0.35)
        (Rendering|Material|SetScalarParameterValueonMaterials _returnvalue_2 "TopS_03_Warp_UV_Amount" 0.02)
        (Rendering|Material|SetColorParameterValueonMaterials _returnvalue_2 "TopS_Em03_Scale__Master" (0.4 2.0 1.0 0.8))
        (Rendering|Material|SetScalarParameterValueonMaterials _returnvalue_2 "EDRec_Emissive Multplyer" 3.0)))
    (elif (Variables|Default|GetReverse)
      ;; FadeIn
      (Rendering|Material|SetScalarParameterValueonMaterials _returnvalue_2 "TopS_04_Warp_UV_Amount" (Math|Float|MapRangeClamped _returnvalue_3 0.0 0.75 0.045 0.015))
      (Rendering|Material|SetScalarParameterValueonMaterials _returnvalue_2 "TopS_04_Warp_UV_Offset" (Math|Float|MapRangeClamped _returnvalue_3 0.0 0.75 -0.35))
      (Rendering|Material|SetScalarParameterValueonMaterials _returnvalue_2 "TopS_03_Warp_UV_Amount" (Math|Float|MapRangeClamped _returnvalue_3 0.0 0.75 0.02 0.01))
      (Rendering|Material|SetColorParameterValueonMaterials _returnvalue_2 "TopS_Em03_Scale__Master" (Math|Color|MakeColor 0.4 2.0 1.0 (Math|Float|MapRangeClamped _returnvalue_3 0.0 0.025 0.8 1.0)))
      (Rendering|Material|SetScalarParameterValueonMaterials _returnvalue_2 "EDRec_Emissive Multplyer" (Math|Float|MapRangeClamped _returnvalue_3 0.0 0.25 3.0 1.0))
      (else
        ;; FadeOut
        (Rendering|Material|SetScalarParameterValueonMaterials _returnvalue_2 "TopS_04_Warp_UV_Amount" (Math|Float|MapRangeClamped _returnvalue_3 0.0 0.25 0.015 0.045))
        (Rendering|Material|SetScalarParameterValueonMaterials _returnvalue_2 "TopS_04_Warp_UV_Offset" (Math|Float|MapRangeClamped _returnvalue_3 0.0 0.25 0.0 -0.35))
        (Rendering|Material|SetScalarParameterValueonMaterials _returnvalue_2 "TopS_03_Warp_UV_Amount" (Math|Float|MapRangeClamped _returnvalue_3 0.0 0.25 0.01 0.02))
        (Rendering|Material|SetColorParameterValueonMaterials _returnvalue_2 "TopS_Em03_Scale__Master" (Math|Color|MakeColor 0.4 2.0 1.0 (Math|Float|MapRangeClamped _returnvalue_3 0.0 0.025 1.0 0.8)))
        (Rendering|Material|SetScalarParameterValueonMaterials _returnvalue_2 "EDRec_Emissive Multplyer" (Math|Float|MapRangeClamped _returnvalue_3 0.0 0.25 1.0 3.0))))))

(fn ConstructionScript)
