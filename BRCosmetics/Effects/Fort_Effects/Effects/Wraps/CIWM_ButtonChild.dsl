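; Weapon wrap Blueprint that steps a material flipbook through four frames (0 to 3) as the wrapped weapon fires.
; Begin play, pawn and weapon-change events bind reload-start, derive a per-shot increment from magazine ammo count and reset the step.
; Each shot advances the clamped step, writes its truncated frame after a short delay and stamps game time into another material parameter.
; Reloading resets the step, updating the frame after a longer delay.
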(event EventBeginPlay
  (bind _returnvalue (CustomCharacter|GetAssociatedPlayerPawn))
  (Utilities|IsValid _returnvalue
    (:Is%20Valid
      (bind _currentweapon (Variables|/Script/FortniteGame.FortPawn|GetCurrentWeapon _returnvalue)
        (:then
          (EventDispatchers|BindEventtoOnLocalReloadStarted _currentweapon (delegate Custom|ReloadStarted))
          (Variables|Default|SetShotIncrement (/ 3.0 (Math|Conversions|ToFloat(Integer) (FortWeapon|GetMagazineAmmoCount _currentweapon))))
          (Variables|Default|SetShootStep 0.0)
          (Utilities|FlowControl|Delay 0.9)
          (Rendering|Material|SetScalarParameterValueonMaterials (CustomItemWrap|GetWrappedMeshComponents) "EDRec_FlipBook_FrameA" (Math|Conversions|ToFloat(Integer) (Math|Float|Truncate (Variables|Default|GetShootStep)))))
        (:else)))))

(event EventOnPawnSet
  (bind _returnvalue (CustomCharacter|GetAssociatedPlayerPawn))
  (Utilities|IsValid _returnvalue
    (:Is%20Valid
      (bind _currentweapon (Variables|/Script/FortniteGame.FortPawn|GetCurrentWeapon _returnvalue)
        (:then
          (EventDispatchers|BindEventtoOnLocalReloadStarted _currentweapon (delegate Custom|ReloadStarted))
          (Variables|Default|SetShotIncrement (/ 3.0 (Math|Conversions|ToFloat(Integer) (FortWeapon|GetMagazineAmmoCount _currentweapon))))
          (Variables|Default|SetShootStep 0.0)
          (Utilities|FlowControl|Delay 0.9)
          (Rendering|Material|SetScalarParameterValueonMaterials (CustomItemWrap|GetWrappedMeshComponents) "EDRec_FlipBook_FrameA" (Math|Conversions|ToFloat(Integer) (Math|Float|Truncate (Variables|Default|GetShootStep)))))
        (:else)))))

(event EventOnNewWeapons
  (params (WeaponsAsItems :type Array%20of%20Fort%20Item%20Object%20References))
  (bind _returnvalue (CustomCharacter|GetAssociatedPlayerPawn))
  (Utilities|IsValid _returnvalue
    (:Is%20Valid
      (bind _currentweapon (Variables|/Script/FortniteGame.FortPawn|GetCurrentWeapon _returnvalue)
        (:then
          (EventDispatchers|BindEventtoOnLocalReloadStarted _currentweapon (delegate Custom|ReloadStarted))
          (Variables|Default|SetShotIncrement (/ 3.0 (Math|Conversions|ToFloat(Integer) (FortWeapon|GetMagazineAmmoCount _currentweapon))))
          (Variables|Default|SetShootStep 0.0)
          (Utilities|FlowControl|Delay 0.9)
          (Rendering|Material|SetScalarParameterValueonMaterials (CustomItemWrap|GetWrappedMeshComponents) "EDRec_FlipBook_FrameA" (Math|Conversions|ToFloat(Integer) (Math|Float|Truncate (Variables|Default|GetShootStep)))))
        (:else)))))

(event EventOnWeaponChanged
  (params (NewWeapon :type Fort%20Weapon%20Object%20Reference) (OldWeapon :type Fort%20Weapon%20Object%20Reference))
  (bind _returnvalue (CustomCharacter|GetAssociatedPlayerPawn))
  (Utilities|IsValid _returnvalue
    (:Is%20Valid
      (bind _currentweapon (Variables|/Script/FortniteGame.FortPawn|GetCurrentWeapon _returnvalue)
        (:then
          (EventDispatchers|BindEventtoOnLocalReloadStarted _currentweapon (delegate Custom|ReloadStarted))
          (Variables|Default|SetShotIncrement (/ 3.0 (Math|Conversions|ToFloat(Integer) (FortWeapon|GetMagazineAmmoCount _currentweapon))))
          (Variables|Default|SetShootStep 0.0)
          (Utilities|FlowControl|Delay 0.9)
          (Rendering|Material|SetScalarParameterValueonMaterials (CustomItemWrap|GetWrappedMeshComponents) "EDRec_FlipBook_FrameA" (Math|Conversions|ToFloat(Integer) (Math|Float|Truncate (Variables|Default|GetShootStep)))))
        (:else)))))

(event EventOnNewPlayerWeaponType
  (params (NewType :type EFortItemType%20Enum :default "WorldItem") (OldType :type EFortItemType%20Enum :default "WorldItem"))
  (bind _returnvalue (CustomCharacter|GetAssociatedPlayerPawn))
  (Utilities|IsValid _returnvalue
    (:Is%20Valid
      (bind _currentweapon (Variables|/Script/FortniteGame.FortPawn|GetCurrentWeapon _returnvalue)
        (:then
          (EventDispatchers|BindEventtoOnLocalReloadStarted _currentweapon (delegate Custom|ReloadStarted))
          (Variables|Default|SetShotIncrement (/ 3.0 (Math|Conversions|ToFloat(Integer) (FortWeapon|GetMagazineAmmoCount _currentweapon))))
          (Variables|Default|SetShootStep 0.0)
          (Utilities|FlowControl|Delay 0.9)
          (Rendering|Material|SetScalarParameterValueonMaterials (CustomItemWrap|GetWrappedMeshComponents) "EDRec_FlipBook_FrameA" (Math|Conversions|ToFloat(Integer) (Math|Float|Truncate (Variables|Default|GetShootStep)))))
        (:else)))))

(event Custom|ReloadStarted
  (params (ReloadTime :type Float%20%28single-precision%29) (ReloadType :type EFortWeaponReloadType%20Enum :default "ReloadWholeClip"))
  (Variables|Default|SetShootStep 0.0)
  (Utilities|FlowControl|Delay 0.9)
  (Rendering|Material|SetScalarParameterValueonMaterials (CustomItemWrap|GetWrappedMeshComponents) "EDRec_FlipBook_FrameA" (Math|Conversions|ToFloat(Integer) (Math|Float|Truncate (Variables|Default|GetShootStep)))))

(event EventOnWeaponFired
  (bind _returnvalue (CustomItemWrap|GetWrappedMeshComponents))
  (Utilities|FlowControl|Sequence
    (:then_0
      (Variables|Default|SetShootStep (Math|Float|Clamp(Float) (+ (Variables|Default|GetShootStep) (Variables|Default|GetShotIncrement)) 0.0 3.0))
      (Utilities|FlowControl|Delay 0.1)
      (Rendering|Material|SetScalarParameterValueonMaterials _returnvalue "EDRec_FlipBook_FrameA" (Math|Conversions|ToFloat(Integer) (Math|Float|Truncate (Variables|Default|GetShootStep)))))
    (:then_1
      (Rendering|Material|SetScalarParameterValueonMaterials _returnvalue "TopS_01_Blend_Opacity-TBP" (Utilities|Time|GetGameTimeinSeconds)))))

(fn ConstructionScript)
