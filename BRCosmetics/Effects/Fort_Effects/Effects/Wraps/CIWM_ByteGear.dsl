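; Reactive weapon-wrap cosmetic (ByteGear) that animates wrapped mesh materials when the player deals damage or gets kills.
; Watched damage and kill stats timestamp material parameters, triggering emissive pulses and holo glitch, brighter on kills.
; Kills also fade three emissive channels in while timelines drive warp strength, with a delayed second timeline fading them out.
; In the front end, a shot counter previews both reactions, resetting after the kill effect.
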(event EventOnWeaponFired
  ; Front end reactive
  (if (CustomCosmeticModifier|IsinFrontEnd)
    (bind _output_get (Variables|Default|Setshots%20fired (+ (Variables|Default|Getshots%20fired) 1)))
    (if (>= _output_get 5)
      (if (== (Variables|Default|Getshots%20fired) 7)
        (CallFunction|Kill)
        (Variables|Default|Setshots%20fired 0)
        (else
          (CallFunction|DamageDelt))))))

(event EventOnPawnSet
  (bind _returnvalue (Utilities|Time|GetGameTimeinSeconds))
  (bind _returnvalue_2 (CustomItemWrap|GetWrappedMeshComponents))
  (Utilities|FlowControl|Sequence
    (:then_0
      (Cosmetics|WatchCosmeticStat (CustomCharacter|GetAssociatedPlayerPawn) "(TagName=\"StatManager.AthenaKills\")" 0 false
        (:OnStatChanged
          (CallFunction|Kill))))
    (:then_1
      (Cosmetics|WatchCosmeticStat (CustomCharacter|GetAssociatedPlayerPawn) "(TagName=\"StatManager.DamageDealt\")" 0 false
        (:OnStatChanged
          (Rendering|Material|SetScalarParameterValueonMaterials _returnvalue_2 "TopS Emissive_01-Brightness-TBP" _returnvalue)
          (Rendering|Material|SetScalarParameterValueonMaterials _returnvalue_2 "TopS Emissive_01_Brightness---Start" 2.0)
          (Rendering|Material|SetScalarParameterValueonMaterials _returnvalue_2 "TopS Emissive_02-Brightness-TBP" _returnvalue)
          (Rendering|Material|SetScalarParameterValueonMaterials _returnvalue_2 "TopS Emissive_02_Brightness---Start" 2.0)
          (Rendering|Material|SetScalarParameterValueonMaterials _returnvalue_2 "TopS Emissive_03-Brightness-TBP" _returnvalue)
          (Rendering|Material|SetScalarParameterValueonMaterials _returnvalue_2 "TopS Emissive_03_Brightness---Start" 2.0)
          (Rendering|Material|SetScalarParameterValueonMaterials _returnvalue_2 "Holo_Glitch_TimeFrom-BP------" _returnvalue))))))

(event Custom|DamageDelt
  (bind _returnvalue (Utilities|Time|GetGameTimeinSeconds))
  (bind _returnvalue_2 (CustomItemWrap|GetWrappedMeshComponents))
  (Rendering|Material|SetScalarParameterValueonMaterials _returnvalue_2 "TopS Emissive_01-Brightness-TBP" _returnvalue)
  (Rendering|Material|SetScalarParameterValueonMaterials _returnvalue_2 "TopS Emissive_01_Brightness---Start" 2.0)
  (Rendering|Material|SetScalarParameterValueonMaterials _returnvalue_2 "TopS Emissive_02-Brightness-TBP" _returnvalue)
  (Rendering|Material|SetScalarParameterValueonMaterials _returnvalue_2 "TopS Emissive_02_Brightness---Start" 2.0)
  (Rendering|Material|SetScalarParameterValueonMaterials _returnvalue_2 "TopS Emissive_03-Brightness-TBP" _returnvalue)
  (Rendering|Material|SetScalarParameterValueonMaterials _returnvalue_2 "TopS Emissive_03_Brightness---Start" 2.0)
  (Rendering|Material|SetScalarParameterValueonMaterials _returnvalue_2 "Holo_Glitch_TimeFrom-BP------" _returnvalue))

(event Custom|Kill
  (bind _returnvalue (CustomItemWrap|GetWrappedMeshComponents))
  (bind _returnvalue_2 (CustomItemWrap|GetWrappedMeshComponents))
  (bind _returnvalue_3 (CustomItemWrap|GetWrappedMeshComponents))
  (bind _returnvalue_4 (CustomItemWrap|GetWrappedMeshComponents))
  (bind _returnvalue_5 (CustomItemWrap|GetWrappedMeshComponents))
  (bind _returnvalue_6 (CustomItemWrap|GetWrappedMeshComponents))
  (bind _returnvalue_7 (CustomItemWrap|GetWrappedMeshComponents))
  (bind _returnvalue_8 (Utilities|Time|GetGameTimeinSeconds))
  (bind _returnvalue_9 (Utilities|Time|GetGameTimeinSeconds))
  (bind _returnvalue_10 (Utilities|Time|GetGameTimeinSeconds))
  (Rendering|Material|SetScalarParameterValueonMaterials _returnvalue_2 "TopS Emissive_01-Brightness-TBP" _returnvalue_10)
  (Rendering|Material|SetScalarParameterValueonMaterials _returnvalue_2 "TopS Emissive_01_Brightness---Start" 3.0)
  (Rendering|Material|SetScalarParameterValueonMaterials _returnvalue_2 "TopS Emissive_02-Brightness-TBP" _returnvalue_10)
  (Rendering|Material|SetScalarParameterValueonMaterials _returnvalue_2 "TopS Emissive_02_Brightness---Start" 3.0)
  (Rendering|Material|SetScalarParameterValueonMaterials _returnvalue_2 "TopS Emissive_03-Brightness-TBP" _returnvalue_10)
  (Rendering|Material|SetScalarParameterValueonMaterials _returnvalue_2 "TopS Emissive_03_Brightness---Start" 3.0)
  (Rendering|Material|SetScalarParameterValueonMaterials _returnvalue_2 "Holo_Glitch_TimeFrom-BP------" _returnvalue_10)
  (Utilities|FlowControl|Sequence
    (:then_0
      (Rendering|Material|SetScalarParameterValueonMaterials _returnvalue_3 "EDRec_Emissive_A_Brightness_Time_BP" _returnvalue_8)
      (Rendering|Material|SetScalarParameterValueonMaterials _returnvalue_3 "EDRec_Emissive_A_Brightness_Time__Start" 0.0)
      (Rendering|Material|SetScalarParameterValueonMaterials _returnvalue_3 "EDRec_Emissive_A_Brightness_Time__End" 2.0)
      (Rendering|Material|SetScalarParameterValueonMaterials _returnvalue_3 "EDRec_Emissive_A_Brightness_Duration" 1.0)
      (Utilities|FlowControl|Delay 3.0)
      (bind (_direction _newtrack_0) (Timeline|Timeline_1 PlayFromStart)
        (:Update
          (Rendering|Material|SetScalarParameterValueonMaterials (CustomItemWrap|GetWrappedMeshComponents) "EDRec_WarpB_Strength" _newtrack_0))
        (:NewTrack_1
          (Utilities|FlowControl|Sequence
            (:then_0
              (Rendering|Material|SetScalarParameterValueonMaterials _returnvalue_6 "EDRec_Emissive_A_Brightness_Time_BP" _returnvalue_9)
              (Rendering|Material|SetScalarParameterValueonMaterials _returnvalue_6 "EDRec_Emissive_A_Brightness_Time__Start" 2.0)
              (Rendering|Material|SetScalarParameterValueonMaterials _returnvalue_6 "EDRec_Emissive_A_Brightness_Time__End" 0.0)
              (Rendering|Material|SetScalarParameterValueonMaterials _returnvalue_6 "EDRec_Emissive_A_Brightness_Duration" 1.0))
            (:then_1
              (Rendering|Material|SetScalarParameterValueonMaterials _returnvalue_7 "EDRec_Emissive_B_Brightness_Time_BP" _returnvalue_9)
              (Rendering|Material|SetScalarParameterValueonMaterials _returnvalue_7 "EDRec_Emissive_B_Brightness_Time__Start" 2.0)
              (Rendering|Material|SetScalarParameterValueonMaterials _returnvalue_7 "EDRec_Emissive_B_Brightness_Time__End" 0.0)
              (Rendering|Material|SetScalarParameterValueonMaterials _returnvalue_7 "EDRec_Emissive_B_Brightness_Duration" 1.0))
            (:then_2
              (Rendering|Material|SetScalarParameterValueonMaterials _returnvalue "EDRec_Emissive_C_Brightness_Time_BP" _returnvalue_9)
              (Rendering|Material|SetScalarParameterValueonMaterials _returnvalue "EDRec_Emissive_C_Brightness_Time__Start" 2.0)
              (Rendering|Material|SetScalarParameterValueonMaterials _returnvalue "EDRec_Emissive_C_Brightness_Time__End" 0.0)
              (Rendering|Material|SetScalarParameterValueonMaterials _returnvalue "EDRec_Emissive_C_Brightness_Duration" 1.0))))))
    (:then_1
      (Rendering|Material|SetScalarParameterValueonMaterials _returnvalue_4 "EDRec_Emissive_B_Brightness_Time_BP" _returnvalue_8)
      (Rendering|Material|SetScalarParameterValueonMaterials _returnvalue_4 "EDRec_Emissive_B_Brightness_Time__Start" 0.0)
      (Rendering|Material|SetScalarParameterValueonMaterials _returnvalue_4 "EDRec_Emissive_B_Brightness_Time__End" 2.0)
      (Rendering|Material|SetScalarParameterValueonMaterials _returnvalue_4 "EDRec_Emissive_B_Brightness_Duration" 1.0))
    (:then_2
      (Rendering|Material|SetScalarParameterValueonMaterials _returnvalue_5 "EDRec_Emissive_C_Brightness_Time_BP" _returnvalue_8)
      (Rendering|Material|SetScalarParameterValueonMaterials _returnvalue_5 "EDRec_Emissive_C_Brightness_Time__Start" 0.0)
      (Rendering|Material|SetScalarParameterValueonMaterials _returnvalue_5 "EDRec_Emissive_C_Brightness_Time__End" 2.0)
      (Rendering|Material|SetScalarParameterValueonMaterials _returnvalue_5 "EDRec_Emissive_C_Brightness_Duration" 1.0))
    (:then_3
      (bind (_direction_2 _newtrack_0_2) (Timeline|Timeline_0 PlayFromStart)
        (:Update
          (Rendering|Material|SetScalarParameterValueonMaterials (CustomItemWrap|GetWrappedMeshComponents) "EDRec_WarpB_Strength" _newtrack_0_2))))))

(fn CountPerctange
  (params (End%20count :type Float%20%28double-precision%29) (StartCount :type Float%20%28double-precision%29) (Character%20offset :type Float%20%28single-precision%29) (BackPack%20offset :type Float%20%28double-precision%29)))

(fn ConstructionScript)
