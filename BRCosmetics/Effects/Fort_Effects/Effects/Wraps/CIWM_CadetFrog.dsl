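; CadetFrog custom item wrap Blueprint that animates the wrap's materials when the weapon fires.
; Each shot writes the game time to two scalar parameters and restarts a three-second Recheck timer.
; A Pause flag allows one looping Move timer, which scrolls two emissive layer offsets in differing directions and speeds from an accumulating Offset.
; Recheck clears that loop once firing lapses, leaving Offset to resume from its last value.
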(event EventOnWeaponFired
  (bind _returnvalue (Utilities|Time|GetGameTimeinSeconds))
  (bind _returnvalue_2 (CustomItemWrap|GetWrappedMeshComponents))
  (Utilities|Time|SetTimerbyFunctionName 0 "Recheck" 3.0)
  (Rendering|Material|SetScalarParameterValueonMaterials _returnvalue_2 "TimeFrom-BP----------------" _returnvalue)
  (Rendering|Material|SetScalarParameterValueonMaterials _returnvalue_2 "TopS Emissive_03-Brightness-TBP" _returnvalue)
  (if (not (Variables|Default|GetPause))
    (Variables|Default|SetPause true)
    (bind _returnvalue_3 (Utilities|Time|SetTimerbyFunctionName 0 "Move" 0.01 true))
    (Variables|Default|SetLoopingTimer _returnvalue_3)))

(event Custom|Move
  (bind _returnvalue (CustomItemWrap|GetWrappedMeshComponents))
  (bind _offset (Variables|Default|GetOffset))
  (Variables|Default|SetOffset (+ (Variables|Default|GetOffset) 0.00125))
  (Rendering|Material|SetVectorParameterValueonMaterials _returnvalue "TopS_Em01_Offset__Master" (Math|Vector|MakeVector (* _offset -1.0)))
  (Rendering|Material|SetVectorParameterValueonMaterials _returnvalue "TopS_Em02_Offset__Master" (Math|Vector|MakeVector (/ _offset 2.0) (/ (/ _offset 2.0) 2.0))))

(event Custom|Recheck
  (Variables|Default|SetPause false)
  (Utilities|Time|ClearandInvalidateTimerbyHandle (Variables|Default|GetLoopingTimer)))

(fn ConstructionScript)
