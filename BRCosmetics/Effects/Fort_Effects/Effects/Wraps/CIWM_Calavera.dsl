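; Calavera weapon wrap logic that drives emissive material effects when the owning player's Athena kill count changes.
; When the pawn is set, it watches the StatManager.AthenaKills cosmetic stat on the associated player pawn.
; Each change writes game-time stamps to scalar parameters on the wrapped meshes, offset ahead for two brightness parameters and unshifted for the hue-shift one, leaving animation to the materials.
; The weapon-fired event and construction script are empty.
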(event EventOnWeaponFired)

(event EventOnPawnSet
  (bind _returnvalue (CustomItemWrap|GetWrappedMeshComponents))
  (bind _returnvalue_2 (Utilities|Time|GetGameTimeinSeconds))
  (bind _returnvalue_3 (+ _returnvalue_2 2.0))
  (bind _asfort_player_pawn (Utilities|Casting|CastToFortPlayerPawn (CustomCharacter|GetAssociatedPlayerPawn)))
  (Cosmetics|WatchCosmeticStat _asfort_player_pawn "(TagName=\"StatManager.AthenaKills\")"
    (:OnStatChanged
      (Utilities|FlowControl|Sequence
        (:then_0
          (Rendering|Material|SetScalarParameterValueonMaterials _returnvalue "TopS Emissive_01-Brightness-TBP" _returnvalue_3))
        (:then_1
          (Rendering|Material|SetScalarParameterValueonMaterials _returnvalue "TS_Emissive01_Time_BP_Brightness" _returnvalue_3)
          (Rendering|Material|SetScalarParameterValueonMaterials _returnvalue "HueShift_EmissiveColor_Time_BP_Brightness" (+ (Utilities|Time|GetGameTimeinSeconds) 0)))))))

(fn ConstructionScript)
