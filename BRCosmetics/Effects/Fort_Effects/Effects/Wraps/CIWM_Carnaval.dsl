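; Blueprint logic for the Carnaval weapon wrap that triggers animated material effects on the wrapped meshes when the weapon fires.
; Each shot writes game time to a horn parameter, then after a short delay writes star main and slightly later opacity times.
; A boolean switch alternates the A and B star parameter sets between triggers.
; The Blueprint only stamps start times, leaving the animation to the material.
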(event EventOnWeaponFired
  (bind _returnvalue (CustomItemWrap|GetWrappedMeshComponents))
  (bind _returnvalue_2 (Utilities|Time|GetGameTimeinSeconds))
  (bind _returnvalue_3 (+ _returnvalue_2 0.15))
  (Rendering|Material|SetScalarParameterValueonMaterials _returnvalue "CarnB_HornA_Time" (Utilities|Time|GetGameTimeinSeconds))
  (Utilities|FlowControl|Delay 0.1)
  (if (Variables|Default|GetSwitch)
    (Rendering|Material|SetScalarParameterValueonMaterials _returnvalue "CanB_StarsA_Opacity_Time" _returnvalue_3)
    (Rendering|Material|SetScalarParameterValueonMaterials _returnvalue "CanB_StarsA_Main_Time" _returnvalue_2)
    (Variables|Default|SetSwitch false)
    (else
      (Rendering|Material|SetScalarParameterValueonMaterials _returnvalue "CanB_StarsB_Opacity_Time" _returnvalue_3)
      (Rendering|Material|SetScalarParameterValueonMaterials _returnvalue "CanB_StarsB_Main_Time" _returnvalue_2)
      (Variables|Default|SetSwitch true))))

(fn ConstructionScript)
