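; Reactive item wrap (CattleJar) that changes its wrapped mesh materials to reflect the player pawn's health.
; Cosmetic stat watchers on damage received and health gained store a red or green color.
; After a short delay, health thresholds pick one of four decal flipbook frames, and the stored color drives an emissive color parameter.
; The emissive brightness parameter receives current game time rather than an intensity; the weapon-changed event is empty.
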(event EventOnWeaponChanged
  (params (NewWeapon :type Fort%20Weapon%20Object%20Reference) (OldWeapon :type Fort%20Weapon%20Object%20Reference)))

(event EventOnPawnSet
  (bind _returnvalue (CustomCharacter|GetAssociatedPlayerPawn))
  (bind _returnvalue_2 (CustomCharacter|GetAssociatedPlayerPawn))
  (bind _returnvalue_3 (FortPawn|GetHealth _returnvalue_2))
  (bind _returnvalue_4 (CustomItemWrap|GetWrappedMeshComponents))
  (Utilities|FlowControl|Sequence
    (:then_0
      (Cosmetics|WatchCosmeticStat _returnvalue "(TagName=\"StatManager.DamageReceived\")" 0 false
        (:OnStatChanged
          (Variables|Default|SetColor%20Value (Math|Conversions|ToVector(LinearColor) "(R=1.000000,G=0.000000,B=0.000000,A=1.000000)")))))
    (:then_1
      (Cosmetics|WatchCosmeticStat _returnvalue "(TagName=\"StatManager.GainedHealthTimes\")" 0 false
        (:OnStatChanged
          (Variables|Default|SetColor%20Value (Math|Conversions|ToVector(LinearColor) "(R=0.000000,G=1.000000,B=0.064321,A=1.000000)"))))))
  (Utilities|FlowControl|Delay 0.05)
  (if (< _returnvalue_3 25.0)
    (Rendering|Material|SetScalarParameterValueonMaterials _returnvalue_4 "FDecalV2_FlipBook_Frame #" (/ 3.0 4.0))
    (elif (< _returnvalue_3 5e+01)
      (Rendering|Material|SetScalarParameterValueonMaterials _returnvalue_4 "FDecalV2_FlipBook_Frame #" (/ 2.0 4.0))
      (elif (< _returnvalue_3 75.0)
        (Rendering|Material|SetScalarParameterValueonMaterials _returnvalue_4 "FDecalV2_FlipBook_Frame #" (/ 1.0 4.0))
        (else
          (Rendering|Material|SetScalarParameterValueonMaterials _returnvalue_4 "FDecalV2_FlipBook_Frame #" (/ 0 4.0))))))
  (Rendering|Material|SetScalarParameterValueonMaterials _returnvalue_4 "TopS Emissive_01-Brightness-TBP" (Utilities|Time|GetGameTimeinSeconds))
  (Rendering|Material|SetVectorParameterValueonMaterials _returnvalue_4 "TopS Emissive_01_Color-" (Variables|Default|GetColor%20Value)))

(fn ConstructionScript)
