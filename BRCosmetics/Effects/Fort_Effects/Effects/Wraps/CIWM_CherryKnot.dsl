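; Custom item wrap Blueprint for CherryKnot that drives the wrap's reactive material effects when the weapon fires.
; On each shot it writes current game time into top/side emissive brightness, offset and blend parameters on the wrapped mesh components, so the material can animate from that timestamp.
; Speed parameters receive the timestamp only when the Pause flag is clear; the flag then holds for Speed Pause Duration, throttling those updates during sustained fire.
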(event EventOnWeaponFired
  (bind _returnvalue (Utilities|Time|GetGameTimeinSeconds))
  (bind _returnvalue_2 (CustomItemWrap|GetWrappedMeshComponents))
  (Utilities|FlowControl|Sequence
    (:then_0
      ;[1 1 1] TopsSide V2
      (Rendering|Material|SetScalarParameterValueonMaterials _returnvalue_2 "TopS Emissive_02-Brightness-TBP" _returnvalue)
      (Rendering|Material|SetScalarParameterValueonMaterials _returnvalue_2 "TopS Emissive_02_Offset----TBP" _returnvalue)
      (Rendering|Material|SetScalarParameterValueonMaterials _returnvalue_2 "TopS_01_Blend_Opacity-TBP" _returnvalue)
      (Rendering|Material|SetScalarParameterValueonMaterials _returnvalue_2 "TopS_01_Blend_Offset-TBP" _returnvalue)
      (Rendering|Material|SetScalarParameterValueonMaterials _returnvalue_2 "TimeFrom-BP----------------" _returnvalue))
    (:then_1
      (Rendering|Material|SetScalarParameterValueonMaterials _returnvalue_2 "TopS Emissive_01-Brightness-TBP" _returnvalue)
      (Rendering|Material|SetScalarParameterValueonMaterials _returnvalue_2 "TopS Emissive_01_Offset----TBP" _returnvalue)
      (Rendering|Material|SetScalarParameterValueonMaterials _returnvalue_2 "TopS Emissive_03-Brightness-TBP" _returnvalue)
      (Rendering|Material|SetScalarParameterValueonMaterials _returnvalue_2 "TopS Emissive_03_Offset----TBP" _returnvalue)
      (if (not (Variables|EnviromentVarable|GetPause))
        (Variables|EnviromentVarable|SetPause true)
        (Rendering|Material|SetScalarParameterValueonMaterials _returnvalue_2 "TopS Emissive_01_Speed-TBP" _returnvalue)
        (Rendering|Material|SetScalarParameterValueonMaterials _returnvalue_2 "TopS Emissive_02_Speed-TBP" _returnvalue)
        (Rendering|Material|SetScalarParameterValueonMaterials _returnvalue_2 "TopS Emissive_03_Speed-TBP" _returnvalue)
        (Rendering|Material|SetScalarParameterValueonMaterials _returnvalue_2 "TopS_01_Speed-TBP" _returnvalue)
        (Rendering|Material|SetScalarParameterValueonMaterials _returnvalue_2 "TopS_02_Speed-TBP" _returnvalue)
        (Rendering|Material|SetScalarParameterValueonMaterials _returnvalue_2 "TopS_03_Speed-TBP" _returnvalue)
        (Rendering|Material|SetScalarParameterValueonMaterials _returnvalue_2 "TopS_04_Speed-TBP" _returnvalue)
        (Utilities|FlowControl|Delay (Variables|Default|GetSpeed%20Pause%20Duration%20))
        (Variables|EnviromentVarable|SetPause false)))))

(fn ConstructionScript)
