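; Chicken Warrior item wrap Blueprint that drives a crack effect on wrapped meshes as the weapon fires.
; Each shot raises a capped step count that indexes a multiplier array, with high steps enabling a transition override.
; Elapsed time since the stored reset deadline lowers the count at the next shot; firing before it adds extra steps.
; Material scalars receive these values and the new deadline; weapon-change events clear the cracks.
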(event EventBeginPlay
  (|CallParentFunction))

(event EventOnWeaponChanged
  (params (NewWeapon :type Fort%20Weapon%20Object%20Reference) (OldWeapon :type Fort%20Weapon%20Object%20Reference))
  (bind _returnvalue (CustomItemWrap|GetWrappedMeshComponents))
  (bind _returnvalue_2 (Utilities|Time|GetGameTimeinSeconds))
  (bind _resetdelay (Variables|Default|GetResetDelay))
  (bind _returnvalue_3 (+ _returnvalue_2 _resetdelay))
  (Variables|Default|SetStepCount 0)
  (Rendering|Material|SetScalarParameterValueonMaterials _returnvalue "CW_CrackTransitionMultiplier" 0.0)
  (Rendering|Material|SetScalarParameterValueonMaterials _returnvalue "CW_CrackTransitionOverride" 0.0)
  (Rendering|Material|SetScalarParameterValueonMaterials _returnvalue "CW_TimeFromBP" _returnvalue_3)
  (Variables|Default|SetWaitTime _returnvalue_3))

(event EventOnNewWeapons
  (params (WeaponsAsItems :type Array%20of%20Fort%20Item%20Object%20References))
  (bind _returnvalue (CustomItemWrap|GetWrappedMeshComponents))
  (bind _returnvalue_2 (Utilities|Time|GetGameTimeinSeconds))
  (bind _resetdelay (Variables|Default|GetResetDelay))
  (bind _returnvalue_3 (+ _returnvalue_2 _resetdelay))
  (Utilities|FlowControl|Delay 0.2)
  (Variables|Default|SetStepCount 0)
  (Rendering|Material|SetScalarParameterValueonMaterials _returnvalue "CW_CrackTransitionMultiplier" 0.0)
  (Rendering|Material|SetScalarParameterValueonMaterials _returnvalue "CW_CrackTransitionOverride" 0.0)
  (Rendering|Material|SetScalarParameterValueonMaterials _returnvalue "CW_TimeFromBP" _returnvalue_3)
  (Variables|Default|SetWaitTime _returnvalue_3))

(event EventOnNewPlayerWeaponType
  (params (NewType :type EFortItemType%20Enum :default "WorldItem") (OldType :type EFortItemType%20Enum :default "WorldItem"))
  (bind _returnvalue (CustomItemWrap|GetWrappedMeshComponents))
  (bind _returnvalue_2 (Utilities|Time|GetGameTimeinSeconds))
  (bind _resetdelay (Variables|Default|GetResetDelay))
  (bind _returnvalue_3 (+ _returnvalue_2 _resetdelay))
  (Utilities|FlowControl|Delay 0.2)
  (Variables|Default|SetStepCount 0)
  (Rendering|Material|SetScalarParameterValueonMaterials _returnvalue "CW_CrackTransitionMultiplier" 0.0)
  (Rendering|Material|SetScalarParameterValueonMaterials _returnvalue "CW_CrackTransitionOverride" 0.0)
  (Rendering|Material|SetScalarParameterValueonMaterials _returnvalue "CW_TimeFromBP" _returnvalue_3)
  (Variables|Default|SetWaitTime _returnvalue_3))

(event EventOnWeaponFired
  (bind _returnvalue (CustomItemWrap|GetWrappedMeshComponents))
  (bind _returnvalue_2 (Utilities|Time|GetGameTimeinSeconds))
  (bind _resetdelay (Variables|Default|GetResetDelay))
  (bind _returnvalue_3 (+ _returnvalue_2 _resetdelay))
  (Utilities|FlowControl|Delay 0.1)
  (bind _output_get (Variables|Default|SetStepCount (Math|Integer|Min(Integer) (+ (- (Variables|Default|GetStepCount) (Math|Integer|Min(Integer) (Variables|Default|GetStepCount) (Math|Float|Floor (* (- (Utilities|Time|GetGameTimeinSeconds) (Variables|Default|GetWaitTime)) 3.0)))) 1) (Variables|Default|GetMaxSteps))))
  (Rendering|Material|SetScalarParameterValueonMaterials _returnvalue "CW_CrackTransitionMultiplier" (Utilities|Array|Get(aref) (Variables|Default|GetMultiplierAmount) _output_get))
  (if (>= _output_get 9)
    (Rendering|Material|SetScalarParameterValueonMaterials _returnvalue "CW_CrackTransitionOverride" 1.0)
    (else
      (Rendering|Material|SetScalarParameterValueonMaterials _returnvalue "CW_CrackTransitionOverride" 0.0)))
  (Rendering|Material|SetScalarParameterValueonMaterials _returnvalue "CW_TimeFromBP" _returnvalue_3)
  (Variables|Default|SetWaitTime _returnvalue_3))

(fn ConstructionScript)
