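; CIWM_ChillyTime is a custom item wrap Blueprint that feeds weapon-fire timing to the wrap's reactive materials.
; On weapon fired, a DoOnceMultiInput gate writes current game time into scalar parameters on the wrapped mesh components.
; These timestamps cover four channels (A–D) each of decal, emissive brightness and emissive gradient, leaving any animation to the materials.
; The chain ends in a short Delay, with no gate reset shown.
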(event EventOnWeaponFired
  (bind _returnvalue (Utilities|Time|GetGameTimeinSeconds))
  (bind _returnvalue_2 (CustomItemWrap|GetWrappedMeshComponents))
  (bind _returnvalue_3 (CustomItemWrap|GetWrappedMeshComponents))
  (bind _returnvalue_4 (Utilities|Time|GetGameTimeinSeconds))
  (bind _returnvalue_5 (CustomItemWrap|GetWrappedMeshComponents))
  (bind _returnvalue_6 (Utilities|Time|GetGameTimeinSeconds))
  (Utilities|FlowControl|DoOnceMultiInput
    (:A%20Out
      ;[1 1 1 0.5] Decal Reactive
      (Rendering|Material|SetScalarParameterValueonMaterials _returnvalue_3 "EDRec_TimeFrom_BP_DecalA" _returnvalue_4)
      (Rendering|Material|SetScalarParameterValueonMaterials _returnvalue_3 "EDRec_TimeFrom_BP_DecalB" _returnvalue_4)
      (Rendering|Material|SetScalarParameterValueonMaterials _returnvalue_3 "EDRec_TimeFrom_BP_DecalC" _returnvalue_4)
      (Rendering|Material|SetScalarParameterValueonMaterials _returnvalue_3 "EDRec_TimeFrom_BP_DecalD" _returnvalue_4)
      (Rendering|Material|SetScalarParameterValueonMaterials _returnvalue_5 "EDRec_Emissive_A_Brightness_Time_BP" _returnvalue_6)
      (Rendering|Material|SetScalarParameterValueonMaterials _returnvalue_5 "EDRec_Emissive_B_Brightness_Time_BP" _returnvalue_6)
      (Rendering|Material|SetScalarParameterValueonMaterials _returnvalue_5 "EDRec_Emissive_C_Brightness_Time_BP" _returnvalue_6)
      (Rendering|Material|SetScalarParameterValueonMaterials _returnvalue_5 "EDRec_Emissive_D_Brightness_Time_BP" _returnvalue_6)
      (Rendering|Material|SetScalarParameterValueonMaterials _returnvalue_2 "EDRec_Emissive_A_Gradent_Time_BP" _returnvalue)
      (Rendering|Material|SetScalarParameterValueonMaterials _returnvalue_2 "EDRec_Emissive_B_Gradent_Time_BP" _returnvalue)
      (Rendering|Material|SetScalarParameterValueonMaterials _returnvalue_2 "EDRec_Emissive_C_Gradent_Time_BP" _returnvalue)
      (Rendering|Material|SetScalarParameterValueonMaterials _returnvalue_2 "EDRec_Emissive_D_Gradent_Time_BP" _returnvalue)
      (Utilities|FlowControl|Delay 1.5))))

(fn ConstructionScript)
