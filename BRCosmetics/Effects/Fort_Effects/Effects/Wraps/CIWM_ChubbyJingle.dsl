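; Weapon wrap cosmetic modifier for ChubbyJingle that changes the wrapped weapon's decal when fired.
; Firing switches the flipbook frame and emissive brightness parameters on the wrapped mesh materials, and a three-second timer reverts both.
; It also registers a watch on the AthenaKills stat at BeginPlay and pawn set, repeated on a looping timer in the front end.
; No handler here reacts to that stat.
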(event EventBeginPlay
  (if (CustomCosmeticModifier|IsinFrontEnd)
    (Utilities|Time|SetTimerbyEvent (delegate Custom|LockerElim) 3.0 true)
    (else
      (Cosmetics|WatchCosmeticStat (CustomCharacter|GetAssociatedPlayerPawn) "(TagName=\"StatManager.AthenaKills\")" 0 false))))

(event Custom|LockerElim
  (Cosmetics|WatchCosmeticStat (CustomCharacter|GetAssociatedPlayerPawn) "(TagName=\"StatManager.AthenaKills\")" 0 false))

(event EventOnWeaponFired
  (bind _returnvalue (CustomItemWrap|GetWrappedMeshComponents))
  (Rendering|Material|SetScalarParameterValueonMaterials _returnvalue "FDecalV2_FlipBook_Frame #" 1.0)
  (Rendering|Material|SetScalarParameterValueonMaterials _returnvalue "FDecalV2_Emissive_Brightness_Time__End" 1.0)
  (Utilities|Time|SetTimerbyEvent (delegate Custom|Revert) 3.0))

(event EventOnPawnSet
  (Cosmetics|WatchCosmeticStat (CustomCharacter|GetAssociatedPlayerPawn) "(TagName=\"StatManager.AthenaKills\")" 0 false))

(event Custom|Revert
  (bind _returnvalue (CustomItemWrap|GetWrappedMeshComponents))
  (Rendering|Material|SetScalarParameterValueonMaterials _returnvalue "FDecalV2_FlipBook_Frame #" 0.0)
  (Rendering|Material|SetScalarParameterValueonMaterials _returnvalue "FDecalV2_Emissive_Brightness_Time__End" 0.0))

(fn ConstructionScript)
