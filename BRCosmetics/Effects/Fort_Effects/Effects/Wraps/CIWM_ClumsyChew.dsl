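; Weapon wrap Blueprint that steps an emissive offset on the wrapped meshes as the weapon fires, then depletes it.
; An accepted shot runs the Progress timeline, lerping the material parameter between adjacent Value array entries.
; Pause acts as a ready flag, ignoring shots until each step finishes and advances CurrentIndex.
; After the last step, a delayed Deplete timeline lerps back to the first value and resets the cycle.
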(event EventOnWeaponFired
  (if (Variables|Default|GetPause)
    (Variables|Default|SetPause false)
    (bind (_direction _newtrack_0) (Timeline|Progress PlayFromStart)
      (:Update
        (Rendering|Material|SetScalarParameterValueonMaterials (CustomItemWrap|GetWrappedMeshComponents) "TopS Emissive_01_Offset---End" (Math|Float|Lerp (Utilities|Array|Get(acopy) (Variables|Default|GetValue) (- (Variables|Default|GetCurrentIndex) 1)) (Utilities|Array|Get(acopy) (Variables|Default|GetValue) (Variables|Default|GetCurrentIndex)) _newtrack_0)))
      (:Finished
        (if (>= (Variables|Default|GetCurrentIndex) 6)
          (Utilities|Time|SetTimerbyFunctionName 0 "Deplete" 5.0)
          (else
            (Math|Integer|IncrementInt (Variables|Default|GetCurrentIndex)
              (:%20%20
                (Variables|Default|SetPause true)))))))))

(event Custom|Deplete
  (bind (_direction _newtrack_0) (Timeline|Depletion PlayFromStart)
    (:Update
      (Rendering|Material|SetScalarParameterValueonMaterials (CustomItemWrap|GetWrappedMeshComponents) "TopS Emissive_01_Offset---End" (Math|Float|Lerp (Utilities|Array|Get(acopy) (Variables|Default|GetValue) 6) (Utilities|Array|Get(acopy) (Variables|Default|GetValue)) _newtrack_0)))
    (:Finished
      (Variables|Default|SetCurrentIndex 1)
      (Variables|Default|SetPause true))))

(fn ConstructionScript)
