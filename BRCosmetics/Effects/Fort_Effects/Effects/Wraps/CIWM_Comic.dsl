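; Comic item wrap Blueprint that re-randomizes the wrap's right-side decal each time the weapon fires.
; After a tiny delay, RandomSet writes UV offset and rotation parameters on the wrapped meshes' materials, with separate ranges per item type.
; Only the weapon call names real parameters for the random flipbook frame and game-time stamp.
; Offsets draw from a random stream, frame and rotation do not, and each call prints the frame value.
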(event EventOnWeaponFired
  (Utilities|FlowControl|Delay 0.001)
  ;[1 1 1] Left Right Decals MF
  (CallFunction|RandomSet "FDecal_FlipBook_Frame #" "FDecal_Opacity_Time_BP_Z" "FDecal_UV_Offset_Right_Weapon" "FDecal_UV_Rotation_Right_Weapon" -0.38 0.2 -0.01 0.01 0.1 -0.15)
  (CallFunction|RandomSet "0" "0" "FDecal_UV_Offset_Right_Pickaxe" "FDecal_UV_Rotation_Right_Pickaxe" -0.3 0.15 0.2 -0.2 0.1 -0.15)
  (CallFunction|RandomSet "0" "0" "FDecal_UV_Offset_Right_Backpack" "FDecal_UV_Rotation_Right_Backpack" -0.15 0.15 -0.25 0.25 0.1 -0.15)
  (CallFunction|RandomSet "0" "0" "FDecal_UV_Offset_Right_Character" "FDecal_UV_Rotation_Right_Character" -0.15 0.15 -0.2 0.28 0.1 -0.15)
  (CallFunction|RandomSet "0" "0" "FDecal_UV_Offset_Right_Vehicle" "FDecal_UV_Rotation_Right_Vehicle" 0.15 0.4 -0.15 0.15 0.1 -0.15)
  (CallFunction|RandomSet "0" "0" "FDecal_UV_Offset_Right_VehicleSide" "FDecal_UV_Rotation_Right_VehicleSide" 0.3 0.42 -0.1 -0.13 0.1 -0.15)
  (CallFunction|RandomSet "0" "0" "FDecal_UV_Offset_Right_Mech" "FDecal_UV_Rotation_Right_Mech" -0.1 0.1 0.32 0.2 0.1 -0.15))

(fn RandomSet
  (params (FlipBook%20Frame :type Name) (TimeBP :type Name) (Offset :type Name) (Rotation :type Name) (X=min :type Float%20%28double-precision%29) (X=Max :type Float%20%28double-precision%29) (Y=Min :type Float%20%28double-precision%29) (Y=Max :type Float%20%28double-precision%29) (Rot=Min :type Float%20%28double-precision%29) (Rot=Max :type Float%20%28double-precision%29))
  (bind _returnvalue (CustomItemWrap|GetWrappedMeshComponents))
  (bind _stream (Variables|Default|GetStream))
  (bind _returnvalue_2 (Math|Random|RandomFloatinRangefromStream _stream X=min X=Max))
  (bind _returnvalue_3 (Math|Random|RandomIntegerinRange 0 8))
  (bind _returnvalue_4 (Math|Conversions|ToFloat(Integer) _returnvalue_3))
  (bind _returnvalue_5 (/ _returnvalue_4 8.0))
  (Rendering|Material|SetScalarParameterValueonMaterials _returnvalue FlipBook%20Frame _returnvalue_5)
  (Rendering|Material|SetScalarParameterValueonMaterials _returnvalue TimeBP (Utilities|Time|GetGameTimeinSeconds))
  (Rendering|Material|SetVectorParameterValueonMaterials _returnvalue Offset (Math|Vector|MakeVector _returnvalue_2 (Math|Random|RandomFloatinRangefromStream _stream Y=Min Y=Max)))
  (Rendering|Material|SetScalarParameterValueonMaterials _returnvalue Rotation (Math|Random|RandomFloatinRange Rot=Min Rot=Max))
  (Variables|Default|SetPosition _returnvalue_2)
  (Development|PrintString (Utilities|String|ToString(Float) _returnvalue_5) true true "(R=0.060001,G=1.000000,B=0.000000,A=1.000000)" 3.0))

(fn ConstructionScript)
