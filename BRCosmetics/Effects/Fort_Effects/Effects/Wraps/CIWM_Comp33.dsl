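; Kill-reactive weapon wrap that fills an emissive gradient on wrapped meshes as the player gets kills.
; A watched AthenaKills stat increments Kill_Count, mapped over five kills to decal and TopS material offsets; weapon changes reapply the decal gradient.
; At six or more kills, a timeline animates emissive brightness, then resets the gradient and count.
; In the front end, a separate timeline drives the count as a preview.
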(event EventBeginPlay
  (Utilities|FlowControl|Delay 0.2)
  (Variables|Default|SetKill_Count 0)
  (if (CustomCosmeticModifier|IsinFrontEnd)
    (bind (_direction _newtrack_0) (Timeline|Timeline_0 PlayFromStart)
      (:Stage1
        (Variables|Default|SetKill_Count (Math|Float|Truncate _newtrack_0))
        (CallFunction|UpdatKillCount)))))

(event EventOnPawnSet
  (bind _asfort_player_pawn (Utilities|Casting|CastToFortPlayerPawn (CustomCharacter|GetAssociatedPlayerPawn)))
  (Cosmetics|WatchCosmeticStat _asfort_player_pawn "(TagName=\"StatManager.AthenaKills\")" 0 false
    (:OnStatChanged
      (Variables|Default|SetKill_Count (+ (Variables|Default|GetKill_Count) 1))
      (CallFunction|UpdatKillCount))))

(event Custom|OnBannerFullyApplied
  (params (Icon :type Texture%202D%20Object%20Reference) (PrimaryColor :type Linear%20Color%20Structure)))

(event Custom|Updat_Kill_Count
  (bind _kill_count (Variables|Default|GetKill_Count))
  (bind _returnvalue (Math|Conversions|ToFloat(Integer) _kill_count))
  (bind _returnvalue_2 (CustomItemWrap|GetWrappedMeshComponents))
  (bind _returnvalue_3 (/ _returnvalue 5.0))
  (bind _returnvalue_4 (Math|Float|Lerp -0.6 0.6 _returnvalue_3))
  (Rendering|Material|SetScalarParameterValueonMaterials _returnvalue_2 "FDecalV2_Emissive_Gradent_Time_End" _returnvalue_4)
  (Rendering|Material|SetScalarParameterValueonMaterials _returnvalue_2 "TopS Emissive_01_Offset---End" _returnvalue_4)
  (Rendering|Material|SetScalarParameterValueonMaterials _returnvalue_2 "TopS Emissive_01-Brightness-TBP" (Utilities|Time|GetGameTimeinSeconds))
  (CallFunction|ResetCounter))

(event Custom|Reset_Counter
  (bind _returnvalue (CustomItemWrap|GetWrappedMeshComponents))
  (if (>= (Variables|Default|GetKill_Count) 6)
    (Utilities|FlowControl|Delay 0.2)
    (bind (_direction _newtrack_0) (Timeline|Timeline PlayFromStart)
      (:Update
        (Utilities|FlowControl|Sequence
          (:then_0
            (Rendering|Material|SetScalarParameterValueonMaterials _returnvalue "FDecalV2_Emissive_Brightness_Time__End" _newtrack_0))
          (:then_1
            (Rendering|Material|SetScalarParameterValueonMaterials _returnvalue "TopS Emissive_01_Brightness---End" _newtrack_0))))
      (:Finished
        (Utilities|FlowControl|Sequence
          (:then_0
            (Rendering|Material|SetScalarParameterValueonMaterials _returnvalue "FDecalV2_Emissive_Gradent_Time_End" -0.6)
            (Rendering|Material|SetScalarParameterValueonMaterials _returnvalue "FDecalV2_Emissive_Gradent_Time_Start" 0.6)
            (Rendering|Material|SetScalarParameterValueonMaterials _returnvalue "FDecalV2_Emissive_Gradent_Time_BP" (Utilities|Time|GetGameTimeinSeconds))
            (Variables|Default|SetKill_Count 0))
          (:then_1
            (Rendering|Material|SetScalarParameterValueonMaterials _returnvalue "TopS Emissive_01_Offset---End" -0.6)
            (Rendering|Material|SetScalarParameterValueonMaterials _returnvalue "TopS Emissive_01_Offset---Start" 0.6)
            (Rendering|Material|SetScalarParameterValueonMaterials _returnvalue "TopS Emissive_01_Offset----TBP" (Utilities|Time|GetGameTimeinSeconds))))))))

(event EventOnWeaponChanged
  (params (NewWeapon :type Fort%20Weapon%20Object%20Reference) (OldWeapon :type Fort%20Weapon%20Object%20Reference))
  ;[1 1 1] Updated kill count on new weapon
  (Utilities|FlowControl|Delay 0.2)
  (Rendering|Material|SetScalarParameterValueonMaterials (CustomItemWrap|GetWrappedMeshComponents) "FDecalV2_Emissive_Gradent_Time_End" (Math|Float|Lerp -0.6 0.6 (/ (Math|Conversions|ToFloat(Integer) (Variables|Default|GetKill_Count)) 5.0))))

(event EventOnNewWeapons
  (params (WeaponsAsItems :type Array%20of%20Fort%20Item%20Object%20References))
  ;[1 1 1] Updated kill count on new weapon
  (Utilities|FlowControl|Delay 0.2)
  (Rendering|Material|SetScalarParameterValueonMaterials (CustomItemWrap|GetWrappedMeshComponents) "FDecalV2_Emissive_Gradent_Time_End" (Math|Float|Lerp -0.6 0.6 (/ (Math|Conversions|ToFloat(Integer) (Variables|Default|GetKill_Count)) 5.0))))

(event EventOnNewPlayerWeaponType
  (params (NewType :type EFortItemType%20Enum :default "WorldItem") (OldType :type EFortItemType%20Enum :default "WorldItem"))
  ;[1 1 1] Updated kill count on new weapon
  (Utilities|FlowControl|Delay 0.2)
  (Rendering|Material|SetScalarParameterValueonMaterials (CustomItemWrap|GetWrappedMeshComponents) "FDecalV2_Emissive_Gradent_Time_End" (Math|Float|Lerp -0.6 0.6 (/ (Math|Conversions|ToFloat(Integer) (Variables|Default|GetKill_Count)) 5.0))))

(fn Blink
  (params (DelaySettings :type Submix%20Effect%20Delay%20Settings%20Structure%20%28by%20ref%29))
  (bind _returnvalue (CustomItemWrap|GetWrappedMeshComponents))
  (Rendering|Material|SetScalarParameterValueonMaterials _returnvalue "FDecalV2_Emissive_Brightness_Time__End" 0.0)
  (Audio|Effects|Delay|SetDelayLength DelaySettings 0.2)
  (Rendering|Material|SetScalarParameterValueonMaterials _returnvalue "FDecalV2_Emissive_Brightness_Time__End" 1.0)
  (Audio|Effects|Delay|SetDelayLength DelaySettings 0.2))

(fn ConstructionScript)
