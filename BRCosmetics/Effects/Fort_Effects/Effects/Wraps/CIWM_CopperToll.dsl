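; Custom item wrap Blueprint (CopperToll) that makes a wrap's emissive material react to chest opens and weapon fire.
; When the pawn is set, it watches the player's OpenChests cosmetic stat and writes game time to both emissive brightness parameters on each change.
; Firing the weapon writes game time to the second parameter only, after a very short delay.
; The parameters receive timestamps, so any animation is left to the material.
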(event EventOnPawnSet
  (bind _returnvalue (CustomItemWrap|GetWrappedMeshComponents))
  (bind _asfort_player_pawn (Utilities|Casting|CastToFortPlayerPawn (CustomCharacter|GetAssociatedPlayerPawn)))
  (Cosmetics|WatchCosmeticStat _asfort_player_pawn "(TagName=\"StatManager.OpenChests\")"
    (:OnStatChanged
      (Rendering|Material|SetScalarParameterValueonMaterials _returnvalue "TopS Emissive_01-Brightness-TBP" (Utilities|Time|GetGameTimeinSeconds))
      (Rendering|Material|SetScalarParameterValueonMaterials _returnvalue "TopS Emissive_02-Brightness-TBP" (Utilities|Time|GetGameTimeinSeconds)))))

(event EventOnWeaponFired
  (Utilities|FlowControl|Delay 0.01)
  (Rendering|Material|SetScalarParameterValueonMaterials (CustomItemWrap|GetWrappedMeshComponents) "TopS Emissive_02-Brightness-TBP" (Utilities|Time|GetGameTimeinSeconds)))

(fn ConstructionScript)
