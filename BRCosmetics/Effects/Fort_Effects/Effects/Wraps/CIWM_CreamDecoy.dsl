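; CIWM_CreamDecoy is a kill-reactive weapon wrap modifier that reveals up to three decals as Athena kills accumulate.
; A kill-stat watcher, or a demo timeline in the front end, increments a counter and plays the Anim timeline, whose counter-indexed Selects drive decal opacity and emissive glow.
; Outside the front end, a timer calls Restart to save and reset the count through a fifth state.
; Weapon fire also writes emissive timing parameters.
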(event EventBeginPlay
  (bind _returnvalue (CustomItemWrap|GetWrappedMeshComponents))
  (bind _kill_count (Variables|Default|GetKill_Count))
  (Utilities|FlowControl|Delay 0.2)
  (if (CustomCosmeticModifier|IsinFrontEnd)
    (Timeline|Timeline PlayFromStart
      (:NewTrack_0
        (if (Variables|Default|GetPause)
          _
          (elif (<= (Variables|Default|GetKill_Count) 3)
            (if (not (>= (Variables|Default|GetKill_Count) 3))
              (Math|Integer|IncrementInt (Variables|Default|GetKill_Count)
                (:%20%20
                  (Components|Timeline|SetPlayRate (Variables|CIWM_CreamDecoy|GetAnim) 0.4)
                  (Variables|Default|SetPause true)
                  (bind (_direction _newtrack_0 _glow_alpha) (Timeline|Anim PlayFromStart)
                    (:Update
                      (Rendering|Material|SetScalarParameterValueonMaterials _returnvalue "EDRec_DecalA_Opacity_End" (Math|Float|Lerp (Utilities|Select :Index _kill_count :Option%200 0.0 :Option%201 0.0 :Option%202 1.0 :Option%203 1.0 :Option%204 0.0) (Utilities|Select :Index _kill_count :Option%200 0.0 :Option%201 1.0 :Option%202 1.0 :Option%203 1.0 :Option%204 1.0) _newtrack_0))
                      (Rendering|Material|SetScalarParameterValueonMaterials _returnvalue "EDRec_Emissive_A_Brightness_Time__End" (Utilities|Select :Index _kill_count :Option%200 0.0 :Option%201 _glow_alpha :Option%202 0.0 :Option%203 0.0 :Option%204 0.0))
                      (Rendering|Material|SetScalarParameterValueonMaterials _returnvalue "EDRec_DecalB_Opacity_End" (Math|Float|Lerp (Utilities|Select :Index _kill_count :Option%200 0.0 :Option%201 0.0 :Option%202 0.0 :Option%203 1.0 :Option%204 0.0) (Utilities|Select :Index _kill_count :Option%200 0.0 :Option%201 0.0 :Option%202 1.0 :Option%203 1.0 :Option%204 (select (>= (Variables|Default|GetSaveKills) 2) 1.0 0.0)) _newtrack_0))
                      (Rendering|Material|SetScalarParameterValueonMaterials _returnvalue "EDRec_Emissive_B_Brightness_Time__End" (Utilities|Select :Index _kill_count :Option%200 0.0 :Option%201 0.0 :Option%202 _glow_alpha :Option%203 0.0 :Option%204 0.0))
                      (Rendering|Material|SetScalarParameterValueonMaterials _returnvalue "EDRec_DecalC_Opacity_End" (Math|Float|Lerp (Utilities|Select :Index _kill_count :Option%200 0.0 :Option%201 0.0 :Option%202 0.0 :Option%203 0.0 :Option%204 0.0) (Utilities|Select :Index _kill_count :Option%200 0.0 :Option%201 0.0 :Option%202 0.0 :Option%203 1.0 :Option%204 (select (>= (Variables|Default|GetSaveKills) 3) 1.0 0.0)) _newtrack_0))
                      (Rendering|Material|SetScalarParameterValueonMaterials _returnvalue "EDRec_Emissive_C_Brightness_Time__End" (Utilities|Select :Index _kill_count :Option%200 0.0 :Option%201 0.0 :Option%202 0.0 :Option%203 _glow_alpha :Option%204 0.0)))
                    (:Finished
                      (Variables|Default|SetPause false)
                      (if (not (CustomCosmeticModifier|IsinFrontEnd))
                        (Utilities|Time|SetTimerbyFunctionName 0 "Restart" 5.0)))))))))))))

(event EventOnPawnSet
  (bind _returnvalue (CustomItemWrap|GetWrappedMeshComponents))
  (bind _kill_count (Variables|Default|GetKill_Count))
  (Cosmetics|WatchCosmeticStat (CustomCharacter|GetAssociatedPlayerPawn) "(TagName=\"StatManager.AthenaKills\")" 0 false
    (:OnStatChanged
      (if (Variables|Default|GetPause)
        _
        (elif (<= (Variables|Default|GetKill_Count) 3)
          (if (not (>= (Variables|Default|GetKill_Count) 3))
            (Math|Integer|IncrementInt (Variables|Default|GetKill_Count)
              (:%20%20
                (Components|Timeline|SetPlayRate (Variables|CIWM_CreamDecoy|GetAnim) 0.4)
                (Variables|Default|SetPause true)
                (Timeline|Anim PlayFromStart)))))))))

(event EventOnWeaponFired
  (bind _returnvalue (Utilities|Time|GetGameTimeinSeconds))
  (bind _returnvalue_2 (CustomItemWrap|GetWrappedMeshComponents))
  (Rendering|Material|SetScalarParameterValueonMaterials _returnvalue_2 "FDecalV2_Emissive_Brightness_Time_BP" _returnvalue)
  (Rendering|Material|SetScalarParameterValueonMaterials _returnvalue_2 "FDecalV2_Emissive_Brightness_Time__End" _returnvalue)
  (if (not (Variables|Default|GetKilled))
    (Rendering|Material|SetScalarParameterValueonMaterials _returnvalue_2 "FDecalV2_Emissive_Brightness_Time__End" 0.0)))

(event Custom|Restart
  (if (Variables|Default|GetPause)
    _
    (elif (>= (Variables|Default|GetKill_Count) 0)
      (Variables|Default|SetPause true)
      (Variables|Default|SetSaveKills (Variables|Default|GetKill_Count))
      (Variables|Default|SetKill_Count 4)
      (Utilities|FlowControl|Sequence
        (:then_0
          (Components|Timeline|SetPlayRate (Variables|CIWM_CreamDecoy|GetAnim) 0.5))
        (:then_1
          (Utilities|FlowControl|RetriggerableDelay 3.0)
          (Variables|Default|SetKill_Count 0)
          (Variables|Default|SetPause false))))))

(event Custom|Reverse'
  (if (CustomCosmeticModifier|IsinFrontEnd)
    (if (Variables|Default|GetPause)
      _
      (elif (>= (Variables|Default|GetKill_Count) 0)
        (Variables|Default|SetPause true)
        (Variables|Default|SetKill_Count 4)
        (Utilities|FlowControl|Sequence
          (:then_0
            (Components|Timeline|SetPlayRate (Variables|CIWM_CreamDecoy|GetAnim) 0.5))
          (:then_1
            (Utilities|FlowControl|RetriggerableDelay 3.0)
            (Variables|Default|SetKill_Count 0)
            (Variables|Default|SetPause false)))))))

(fn ConstructionScript)
