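; PixieParts wrap Blueprint that derives wrap material settings from a specific character's selected cosmetic variants.
; Pawn, weapon, wrap-target and customization events check the character name, cache the Body material, and map six metal, plastic and energy variant channels to indices for MaterialSetuup.
; Metal channels are cross-wired: Tri feeds Metal-Prim, Prim Metal-Sec, Sec Metal-Tret.
; ColorSet also reruns on a half-second timer; Color Set builds named color and MRS palettes.
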(event EventOnNewWeapons
  (params (WeaponsAsItems :type Array%20of%20Fort%20Item%20Object%20References))
  (bind _returnvalue (CustomCharacter|GetAssociatedPlayerPawn))
  (bind _cosmeticloadout (Variables|/Script/FortniteGame.FortPlayerPawn|GetCosmeticLoadout _returnvalue))
  (Utilities|FlowControl|Delay 0.2)
  (if (Utilities|String|Equal,CaseInsensitive(String) (Item|GetPersistentName (CosmeticLoadout|GetItemDefinitionfromLoadout (CosmeticLoadout|GetCosmeticLoadout) "(TagName=\"LoadoutSlot.Character\")")) (Variables|Default|GetCharacter-Name))
    (Utilities|FlowControl|Sequence
      (:then_0
        (bind _asmaterial_instance_dynamic (Utilities|Casting|CastToMaterialInstanceDynamic (Rendering|Material|GetMaterial (Customization|GetSkeletalMeshforPartType (CustomCharacter|GetAssociatedPlayerPawn) "Body" :!class /Script/FortniteGame.FortPlayerPawn) 0)))
        (Variables|Default|SetCharacter%20Material _asmaterial_instance_dynamic)
        (bind _asmaterial_instance_dynamic_2 (Utilities|Casting|CastToMaterialInstanceDynamic (Rendering|Material|GetMaterial (Customization|GetSkeletalMeshforPartType (CustomCharacter|GetAssociatedPlayerPawn) "Body" :!class /Script/FortniteGame.FortPlayerPawn) 0)))
        (Variables|Default|SetSigelMat _asmaterial_instance_dynamic_2)
        (Utilities|FlowControl|Sequence
          (:then_2
            (bind _index (Class|FortCosmeticFunctionLibrary|%28Deprecated%29GetVariantChannelActiveIndexForSlot "(TagName=\"LoadoutSlot.Character\")" _cosmeticloadout (Variables|Default|GetMetal%20Prim) (Utilities|Array|MakeArray "(TagName=\"Cosmetics.Variant.Property.Mat0\")" "(TagName=\"Cosmetics.Variant.Property.Mat1\")" "(TagName=\"Cosmetics.Variant.Property.Mat2\")" "(TagName=\"Cosmetics.Variant.Property.Mat3\")" "(TagName=\"Cosmetics.Variant.Property.Mat4\")" "(TagName=\"Cosmetics.Variant.Property.Mat5\")" "(TagName=\"Cosmetics.Variant.Property.Mat6\")" "(TagName=\"Cosmetics.Variant.Property.Mat7\")" "(TagName=\"Cosmetics.Variant.Property.Mat8\")" "(TagName=\"Cosmetics.Variant.Property.Mat9\")") false false))
            (Variables|Default|SetMetal_prim _index))
          (:then_3
            (bind _index_2 (Class|FortCosmeticFunctionLibrary|%28Deprecated%29GetVariantChannelActiveIndexForSlot "(TagName=\"LoadoutSlot.Character\")" _cosmeticloadout (Variables|Default|GetMetal%20sec) (Utilities|Array|MakeArray "(TagName=\"Cosmetics.Variant.Property.Mat10\")" "(TagName=\"Cosmetics.Variant.Property.Mat11\")" "(TagName=\"Cosmetics.Variant.Property.Mat12\")" "(TagName=\"Cosmetics.Variant.Property.Mat13\")" "(TagName=\"Cosmetics.Variant.Property.Mat14\")" "(TagName=\"Cosmetics.Variant.Property.Mat15\")" "(TagName=\"Cosmetics.Variant.Property.Mat16\")" "(TagName=\"Cosmetics.Variant.Property.Mat17\")" "(TagName=\"Cosmetics.Variant.Property.Mat18\")" "(TagName=\"Cosmetics.Variant.Property.Mat19\")") false false)))
          (:then_4
            (bind _index_3 (Class|FortCosmeticFunctionLibrary|%28Deprecated%29GetVariantChannelActiveIndexForSlot "(TagName=\"LoadoutSlot.Character\")" _cosmeticloadout (Variables|Default|GetMetal%20Tri) (Utilities|Array|MakeArray "(TagName=\"Cosmetics.Variant.Property.Mat20\")" "(TagName=\"Cosmetics.Variant.Property.Mat21\")" "(TagName=\"Cosmetics.Variant.Property.Mat22\")" "(TagName=\"Cosmetics.Variant.Property.Mat23\")" "(TagName=\"Cosmetics.Variant.Property.Mat24\")" "(TagName=\"Cosmetics.Variant.Property.Mat25\")" "(TagName=\"Cosmetics.Variant.Property.Mat26\")" "(TagName=\"Cosmetics.Variant.Property.Mat27\")" "(TagName=\"Cosmetics.Variant.Property.Mat28\")" "(TagName=\"Cosmetics.Variant.Property.Mat29\")") false false)))
          (:then_5
            (bind _index_4 (Class|FortCosmeticFunctionLibrary|%28Deprecated%29GetVariantChannelActiveIndexForSlot "(TagName=\"LoadoutSlot.Character\")" _cosmeticloadout (Variables|Default|GetPlastic%20Prim) (Utilities|Array|MakeArray "(TagName=\"Cosmetics.Variant.Property.Mat30\")" "(TagName=\"Cosmetics.Variant.Property.Mat31\")" "(TagName=\"Cosmetics.Variant.Property.Mat32\")" "(TagName=\"Cosmetics.Variant.Property.Mat33\")" "(TagName=\"Cosmetics.Variant.Property.Mat34\")" "(TagName=\"Cosmetics.Variant.Property.Mat35\")" "(TagName=\"Cosmetics.Variant.Property.Mat36\")" "(TagName=\"Cosmetics.Variant.Property.Mat37\")" "(TagName=\"Cosmetics.Variant.Property.Mat38\")" "(TagName=\"Cosmetics.Variant.Property.Mat39\")") false false)))
          (:then_6
            (bind _index_5 (Class|FortCosmeticFunctionLibrary|%28Deprecated%29GetVariantChannelActiveIndexForSlot "(TagName=\"LoadoutSlot.Character\")" _cosmeticloadout (Variables|Default|Getplastic%20sec) (Utilities|Array|MakeArray "(TagName=\"Cosmetics.Variant.Property.Mat40\")" "(TagName=\"Cosmetics.Variant.Property.Mat41\")" "(TagName=\"Cosmetics.Variant.Property.Mat42\")" "(TagName=\"Cosmetics.Variant.Property.Mat43\")" "(TagName=\"Cosmetics.Variant.Property.Mat44\")" "(TagName=\"Cosmetics.Variant.Property.Mat45\")" "(TagName=\"Cosmetics.Variant.Property.Mat46\")" "(TagName=\"Cosmetics.Variant.Property.Mat47\")" "(TagName=\"Cosmetics.Variant.Property.Mat48\")" "(TagName=\"Cosmetics.Variant.Property.Mat49\")") false false)))
          (:then_7
            (bind _index_6 (Class|FortCosmeticFunctionLibrary|%28Deprecated%29GetVariantChannelActiveIndexForSlot "(TagName=\"LoadoutSlot.Character\")" _cosmeticloadout (Variables|Default|Getenergy) (Utilities|Array|MakeArray "(TagName=\"Cosmetics.Variant.Property.Mat50\")" "(TagName=\"Cosmetics.Variant.Property.Mat51\")" "(TagName=\"Cosmetics.Variant.Property.Mat52\")" "(TagName=\"Cosmetics.Variant.Property.Mat53\")" "(TagName=\"Cosmetics.Variant.Property.Mat54\")" "(TagName=\"Cosmetics.Variant.Property.Mat55\")") false false))))
        (CallFunction|MaterialSetuup (Utilities|Array|Get(aref) (Variables|Default|GetMetal-Prim) _index_3) (Utilities|Array|Get(aref) (Variables|Default|GetMetal-Prim-MRS) _index_3) (Utilities|Array|Get(aref) (Variables|Default|GetMetal-Sec) _index) (Utilities|Array|Get(aref) (Variables|Default|GetMetal-Sec%20MRS) _index) (Utilities|Array|Get(aref) (Variables|Default|GetMetal-Tret) _index_2) (Utilities|Array|Get(aref) (Variables|Default|GetMetal-Tret%20MRS) _index_2) (Utilities|Array|Get(aref) (Variables|Default|GetPlastic-Prim) _index_4) (Utilities|Array|Get(aref) (Variables|Default|GetPlastic-Prim%20MRS) _index_4) (Utilities|Array|Get(aref) (Variables|Default|GetPlastic-Sec) _index_5) (Utilities|Array|Get(aref) (Variables|Default|GetPlastic-Sec%20MRS) _index_5) (Utilities|Array|Get(aref) (Variables|Default|GetEngery) _index_6)))
      (:then_1
        (CallFunction|ColorSet)))))

(event EventOnNewPlayerWeaponType
  (params (NewType :type EFortItemType%20Enum :default "WorldItem") (OldType :type EFortItemType%20Enum :default "WorldItem"))
  (bind _returnvalue (CustomCharacter|GetAssociatedPlayerPawn))
  (bind _cosmeticloadout (Variables|/Script/FortniteGame.FortPlayerPawn|GetCosmeticLoadout _returnvalue))
  (Utilities|FlowControl|Delay 0.2)
  (if (Utilities|String|Equal,CaseInsensitive(String) (Item|GetPersistentName (CosmeticLoadout|GetItemDefinitionfromLoadout (CosmeticLoadout|GetCosmeticLoadout) "(TagName=\"LoadoutSlot.Character\")")) (Variables|Default|GetCharacter-Name))
    (Utilities|FlowControl|Sequence
      (:then_0
        (bind _asmaterial_instance_dynamic (Utilities|Casting|CastToMaterialInstanceDynamic (Rendering|Material|GetMaterial (Customization|GetSkeletalMeshforPartType (CustomCharacter|GetAssociatedPlayerPawn) "Body" :!class /Script/FortniteGame.FortPlayerPawn) 0)))
        (Variables|Default|SetCharacter%20Material _asmaterial_instance_dynamic)
        (bind _asmaterial_instance_dynamic_2 (Utilities|Casting|CastToMaterialInstanceDynamic (Rendering|Material|GetMaterial (Customization|GetSkeletalMeshforPartType (CustomCharacter|GetAssociatedPlayerPawn) "Body" :!class /Script/FortniteGame.FortPlayerPawn) 0)))
        (Variables|Default|SetSigelMat _asmaterial_instance_dynamic_2)
        (Utilities|FlowControl|Sequence
          (:then_2
            (bind _index (Class|FortCosmeticFunctionLibrary|%28Deprecated%29GetVariantChannelActiveIndexForSlot "(TagName=\"LoadoutSlot.Character\")" _cosmeticloadout (Variables|Default|GetMetal%20Prim) (Utilities|Array|MakeArray "(TagName=\"Cosmetics.Variant.Property.Mat0\")" "(TagName=\"Cosmetics.Variant.Property.Mat1\")" "(TagName=\"Cosmetics.Variant.Property.Mat2\")" "(TagName=\"Cosmetics.Variant.Property.Mat3\")" "(TagName=\"Cosmetics.Variant.Property.Mat4\")" "(TagName=\"Cosmetics.Variant.Property.Mat5\")" "(TagName=\"Cosmetics.Variant.Property.Mat6\")" "(TagName=\"Cosmetics.Variant.Property.Mat7\")" "(TagName=\"Cosmetics.Variant.Property.Mat8\")" "(TagName=\"Cosmetics.Variant.Property.Mat9\")") false false))
            (Variables|Default|SetMetal_prim _index))
          (:then_3
            (bind _index_2 (Class|FortCosmeticFunctionLibrary|%28Deprecated%29GetVariantChannelActiveIndexForSlot "(TagName=\"LoadoutSlot.Character\")" _cosmeticloadout (Variables|Default|GetMetal%20sec) (Utilities|Array|MakeArray "(TagName=\"Cosmetics.Variant.Property.Mat10\")" "(TagName=\"Cosmetics.Variant.Property.Mat11\")" "(TagName=\"Cosmetics.Variant.Property.Mat12\")" "(TagName=\"Cosmetics.Variant.Property.Mat13\")" "(TagName=\"Cosmetics.Variant.Property.Mat14\")" "(TagName=\"Cosmetics.Variant.Property.Mat15\")" "(TagName=\"Cosmetics.Variant.Property.Mat16\")" "(TagName=\"Cosmetics.Variant.Property.Mat17\")" "(TagName=\"Cosmetics.Variant.Property.Mat18\")" "(TagName=\"Cosmetics.Variant.Property.Mat19\")") false false)))
          (:then_4
            (bind _index_3 (Class|FortCosmeticFunctionLibrary|%28Deprecated%29GetVariantChannelActiveIndexForSlot "(TagName=\"LoadoutSlot.Character\")" _cosmeticloadout (Variables|Default|GetMetal%20Tri) (Utilities|Array|MakeArray "(TagName=\"Cosmetics.Variant.Property.Mat20\")" "(TagName=\"Cosmetics.Variant.Property.Mat21\")" "(TagName=\"Cosmetics.Variant.Property.Mat22\")" "(TagName=\"Cosmetics.Variant.Property.Mat23\")" "(TagName=\"Cosmetics.Variant.Property.Mat24\")" "(TagName=\"Cosmetics.Variant.Property.Mat25\")" "(TagName=\"Cosmetics.Variant.Property.Mat26\")" "(TagName=\"Cosmetics.Variant.Property.Mat27\")" "(TagName=\"Cosmetics.Variant.Property.Mat28\")" "(TagName=\"Cosmetics.Variant.Property.Mat29\")") false false)))
          (:then_5
            (bind _index_4 (Class|FortCosmeticFunctionLibrary|%28Deprecated%29GetVariantChannelActiveIndexForSlot "(TagName=\"LoadoutSlot.Character\")" _cosmeticloadout (Variables|Default|GetPlastic%20Prim) (Utilities|Array|MakeArray "(TagName=\"Cosmetics.Variant.Property.Mat30\")" "(TagName=\"Cosmetics.Variant.Property.Mat31\")" "(TagName=\"Cosmetics.Variant.Property.Mat32\")" "(TagName=\"Cosmetics.Variant.Property.Mat33\")" "(TagName=\"Cosmetics.Variant.Property.Mat34\")" "(TagName=\"Cosmetics.Variant.Property.Mat35\")" "(TagName=\"Cosmetics.Variant.Property.Mat36\")" "(TagName=\"Cosmetics.Variant.Property.Mat37\")" "(TagName=\"Cosmetics.Variant.Property.Mat38\")" "(TagName=\"Cosmetics.Variant.Property.Mat39\")") false false)))
          (:then_6
            (bind _index_5 (Class|FortCosmeticFunctionLibrary|%28Deprecated%29GetVariantChannelActiveIndexForSlot "(TagName=\"LoadoutSlot.Character\")" _cosmeticloadout (Variables|Default|Getplastic%20sec) (Utilities|Array|MakeArray "(TagName=\"Cosmetics.Variant.Property.Mat40\")" "(TagName=\"Cosmetics.Variant.Property.Mat41\")" "(TagName=\"Cosmetics.Variant.Property.Mat42\")" "(TagName=\"Cosmetics.Variant.Property.Mat43\")" "(TagName=\"Cosmetics.Variant.Property.Mat44\")" "(TagName=\"Cosmetics.Variant.Property.Mat45\")" "(TagName=\"Cosmetics.Variant.Property.Mat46\")" "(TagName=\"Cosmetics.Variant.Property.Mat47\")" "(TagName=\"Cosmetics.Variant.Property.Mat48\")" "(TagName=\"Cosmetics.Variant.Property.Mat49\")") false false)))
          (:then_7
            (bind _index_6 (Class|FortCosmeticFunctionLibrary|%28Deprecated%29GetVariantChannelActiveIndexForSlot "(TagName=\"LoadoutSlot.Character\")" _cosmeticloadout (Variables|Default|Getenergy) (Utilities|Array|MakeArray "(TagName=\"Cosmetics.Variant.Property.Mat50\")" "(TagName=\"Cosmetics.Variant.Property.Mat51\")" "(TagName=\"Cosmetics.Variant.Property.Mat52\")" "(TagName=\"Cosmetics.Variant.Property.Mat53\")" "(TagName=\"Cosmetics.Variant.Property.Mat54\")" "(TagName=\"Cosmetics.Variant.Property.Mat55\")") false false))))
        (CallFunction|MaterialSetuup (Utilities|Array|Get(aref) (Variables|Default|GetMetal-Prim) _index_3) (Utilities|Array|Get(aref) (Variables|Default|GetMetal-Prim-MRS) _index_3) (Utilities|Array|Get(aref) (Variables|Default|GetMetal-Sec) _index) (Utilities|Array|Get(aref) (Variables|Default|GetMetal-Sec%20MRS) _index) (Utilities|Array|Get(aref) (Variables|Default|GetMetal-Tret) _index_2) (Utilities|Array|Get(aref) (Variables|Default|GetMetal-Tret%20MRS) _index_2) (Utilities|Array|Get(aref) (Variables|Default|GetPlastic-Prim) _index_4) (Utilities|Array|Get(aref) (Variables|Default|GetPlastic-Prim%20MRS) _index_4) (Utilities|Array|Get(aref) (Variables|Default|GetPlastic-Sec) _index_5) (Utilities|Array|Get(aref) (Variables|Default|GetPlastic-Sec%20MRS) _index_5) (Utilities|Array|Get(aref) (Variables|Default|GetEngery) _index_6)))
      (:then_1
        (CallFunction|ColorSet)))))

(event EventOnPawnSet
  (bind _returnvalue (CustomCharacter|GetAssociatedPlayerPawn))
  (bind _cosmeticloadout (Variables|/Script/FortniteGame.FortPlayerPawn|GetCosmeticLoadout _returnvalue))
  (if (Utilities|String|Equal,CaseInsensitive(String) (Item|GetPersistentName (CosmeticLoadout|GetItemDefinitionfromLoadout (CosmeticLoadout|GetCosmeticLoadout) "(TagName=\"LoadoutSlot.Character\")")) (Variables|Default|GetCharacter-Name))
    (Utilities|FlowControl|Sequence
      (:then_0
        (bind _asmaterial_instance_dynamic (Utilities|Casting|CastToMaterialInstanceDynamic (Rendering|Material|GetMaterial (Customization|GetSkeletalMeshforPartType (CustomCharacter|GetAssociatedPlayerPawn) "Body" :!class /Script/FortniteGame.FortPlayerPawn) 0)))
        (Variables|Default|SetCharacter%20Material _asmaterial_instance_dynamic)
        (bind _asmaterial_instance_dynamic_2 (Utilities|Casting|CastToMaterialInstanceDynamic (Rendering|Material|GetMaterial (Customization|GetSkeletalMeshforPartType (CustomCharacter|GetAssociatedPlayerPawn) "Body" :!class /Script/FortniteGame.FortPlayerPawn) 0)))
        (Variables|Default|SetSigelMat _asmaterial_instance_dynamic_2)
        (Utilities|FlowControl|Sequence
          (:then_2
            (bind _index (Class|FortCosmeticFunctionLibrary|%28Deprecated%29GetVariantChannelActiveIndexForSlot "(TagName=\"LoadoutSlot.Character\")" _cosmeticloadout (Variables|Default|GetMetal%20Prim) (Utilities|Array|MakeArray "(TagName=\"Cosmetics.Variant.Property.Mat0\")" "(TagName=\"Cosmetics.Variant.Property.Mat1\")" "(TagName=\"Cosmetics.Variant.Property.Mat2\")" "(TagName=\"Cosmetics.Variant.Property.Mat3\")" "(TagName=\"Cosmetics.Variant.Property.Mat4\")" "(TagName=\"Cosmetics.Variant.Property.Mat5\")" "(TagName=\"Cosmetics.Variant.Property.Mat6\")" "(TagName=\"Cosmetics.Variant.Property.Mat7\")" "(TagName=\"Cosmetics.Variant.Property.Mat8\")" "(TagName=\"Cosmetics.Variant.Property.Mat9\")") false false))
            (Variables|Default|SetMetal_prim _index))
          (:then_3
            (bind _index_2 (Class|FortCosmeticFunctionLibrary|%28Deprecated%29GetVariantChannelActiveIndexForSlot "(TagName=\"LoadoutSlot.Character\")" _cosmeticloadout (Variables|Default|GetMetal%20sec) (Utilities|Array|MakeArray "(TagName=\"Cosmetics.Variant.Property.Mat10\")" "(TagName=\"Cosmetics.Variant.Property.Mat11\")" "(TagName=\"Cosmetics.Variant.Property.Mat12\")" "(TagName=\"Cosmetics.Variant.Property.Mat13\")" "(TagName=\"Cosmetics.Variant.Property.Mat14\")" "(TagName=\"Cosmetics.Variant.Property.Mat15\")" "(TagName=\"Cosmetics.Variant.Property.Mat16\")" "(TagName=\"Cosmetics.Variant.Property.Mat17\")" "(TagName=\"Cosmetics.Variant.Property.Mat18\")" "(TagName=\"Cosmetics.Variant.Property.Mat19\")") false false)))
          (:then_4
            (bind _index_3 (Class|FortCosmeticFunctionLibrary|%28Deprecated%29GetVariantChannelActiveIndexForSlot "(TagName=\"LoadoutSlot.Character\")" _cosmeticloadout (Variables|Default|GetMetal%20Tri) (Utilities|Array|MakeArray "(TagName=\"Cosmetics.Variant.Property.Mat20\")" "(TagName=\"Cosmetics.Variant.Property.Mat21\")" "(TagName=\"Cosmetics.Variant.Property.Mat22\")" "(TagName=\"Cosmetics.Variant.Property.Mat23\")" "(TagName=\"Cosmetics.Variant.Property.Mat24\")" "(TagName=\"Cosmetics.Variant.Property.Mat25\")" "(TagName=\"Cosmetics.Variant.Property.Mat26\")" "(TagName=\"Cosmetics.Variant.Property.Mat27\")" "(TagName=\"Cosmetics.Variant.Property.Mat28\")" "(TagName=\"Cosmetics.Variant.Property.Mat29\")") false false)))
          (:then_5
            (bind _index_4 (Class|FortCosmeticFunctionLibrary|%28Deprecated%29GetVariantChannelActiveIndexForSlot "(TagName=\"LoadoutSlot.Character\")" _cosmeticloadout (Variables|Default|GetPlastic%20Prim) (Utilities|Array|MakeArray "(TagName=\"Cosmetics.Variant.Property.Mat30\")" "(TagName=\"Cosmetics.Variant.Property.Mat31\")" "(TagName=\"Cosmetics.Variant.Property.Mat32\")" "(TagName=\"Cosmetics.Variant.Property.Mat33\")" "(TagName=\"Cosmetics.Variant.Property.Mat34\")" "(TagName=\"Cosmetics.Variant.Property.Mat35\")" "(TagName=\"Cosmetics.Variant.Property.Mat36\")" "(TagName=\"Cosmetics.Variant.Property.Mat37\")" "(TagName=\"Cosmetics.Variant.Property.Mat38\")" "(TagName=\"Cosmetics.Variant.Property.Mat39\")") false false)))
          (:then_6
            (bind _index_5 (Class|FortCosmeticFunctionLibrary|%28Deprecated%29GetVariantChannelActiveIndexForSlot "(TagName=\"LoadoutSlot.Character\")" _cosmeticloadout (Variables|Default|Getplastic%20sec) (Utilities|Array|MakeArray "(TagName=\"Cosmetics.Variant.Property.Mat40\")" "(TagName=\"Cosmetics.Variant.Property.Mat41\")" "(TagName=\"Cosmetics.Variant.Property.Mat42\")" "(TagName=\"Cosmetics.Variant.Property.Mat43\")" "(TagName=\"Cosmetics.Variant.Property.Mat44\")" "(TagName=\"Cosmetics.Variant.Property.Mat45\")" "(TagName=\"Cosmetics.Variant.Property.Mat46\")" "(TagName=\"Cosmetics.Variant.Property.Mat47\")" "(TagName=\"Cosmetics.Variant.Property.Mat48\")" "(TagName=\"Cosmetics.Variant.Property.Mat49\")") false false)))
          (:then_7
            (bind _index_6 (Class|FortCosmeticFunctionLibrary|%28Deprecated%29GetVariantChannelActiveIndexForSlot "(TagName=\"LoadoutSlot.Character\")" _cosmeticloadout (Variables|Default|Getenergy) (Utilities|Array|MakeArray "(TagName=\"Cosmetics.Variant.Property.Mat50\")" "(TagName=\"Cosmetics.Variant.Property.Mat51\")" "(TagName=\"Cosmetics.Variant.Property.Mat52\")" "(TagName=\"Cosmetics.Variant.Property.Mat53\")" "(TagName=\"Cosmetics.Variant.Property.Mat54\")" "(TagName=\"Cosmetics.Variant.Property.Mat55\")") false false))))
        (CallFunction|MaterialSetuup (Utilities|Array|Get(aref) (Variables|Default|GetMetal-Prim) _index_3) (Utilities|Array|Get(aref) (Variables|Default|GetMetal-Prim-MRS) _index_3) (Utilities|Array|Get(aref) (Variables|Default|GetMetal-Sec) _index) (Utilities|Array|Get(aref) (Variables|Default|GetMetal-Sec%20MRS) _index) (Utilities|Array|Get(aref) (Variables|Default|GetMetal-Tret) _index_2) (Utilities|Array|Get(aref) (Variables|Default|GetMetal-Tret%20MRS) _index_2) (Utilities|Array|Get(aref) (Variables|Default|GetPlastic-Prim) _index_4) (Utilities|Array|Get(aref) (Variables|Default|GetPlastic-Prim%20MRS) _index_4) (Utilities|Array|Get(aref) (Variables|Default|GetPlastic-Sec) _index_5) (Utilities|Array|Get(aref) (Variables|Default|GetPlastic-Sec%20MRS) _index_5) (Utilities|Array|Get(aref) (Variables|Default|GetEngery) _index_6)))
      (:then_1
        (CallFunction|ColorSet)))))

(event EventOnSetWrapTarget
  (bind _returnvalue (CustomCharacter|GetAssociatedPlayerPawn))
  (bind _cosmeticloadout (Variables|/Script/FortniteGame.FortPlayerPawn|GetCosmeticLoadout _returnvalue))
  (if (Utilities|String|Equal,CaseInsensitive(String) (Item|GetPersistentName (CosmeticLoadout|GetItemDefinitionfromLoadout (CosmeticLoadout|GetCosmeticLoadout) "(TagName=\"LoadoutSlot.Character\")")) (Variables|Default|GetCharacter-Name))
    (Utilities|FlowControl|Sequence
      (:then_0
        (bind _asmaterial_instance_dynamic (Utilities|Casting|CastToMaterialInstanceDynamic (Rendering|Material|GetMaterial (Customization|GetSkeletalMeshforPartType (CustomCharacter|GetAssociatedPlayerPawn) "Body" :!class /Script/FortniteGame.FortPlayerPawn) 0)))
        (Variables|Default|SetCharacter%20Material _asmaterial_instance_dynamic)
        (bind _asmaterial_instance_dynamic_2 (Utilities|Casting|CastToMaterialInstanceDynamic (Rendering|Material|GetMaterial (Customization|GetSkeletalMeshforPartType (CustomCharacter|GetAssociatedPlayerPawn) "Body" :!class /Script/FortniteGame.FortPlayerPawn) 0)))
        (Variables|Default|SetSigelMat _asmaterial_instance_dynamic_2)
        (Utilities|FlowControl|Sequence
          (:then_2
            (bind _index (Class|FortCosmeticFunctionLibrary|%28Deprecated%29GetVariantChannelActiveIndexForSlot "(TagName=\"LoadoutSlot.Character\")" _cosmeticloadout (Variables|Default|GetMetal%20Prim) (Utilities|Array|MakeArray "(TagName=\"Cosmetics.Variant.Property.Mat0\")" "(TagName=\"Cosmetics.Variant.Property.Mat1\")" "(TagName=\"Cosmetics.Variant.Property.Mat2\")" "(TagName=\"Cosmetics.Variant.Property.Mat3\")" "(TagName=\"Cosmetics.Variant.Property.Mat4\")" "(TagName=\"Cosmetics.Variant.Property.Mat5\")" "(TagName=\"Cosmetics.Variant.Property.Mat6\")" "(TagName=\"Cosmetics.Variant.Property.Mat7\")" "(TagName=\"Cosmetics.Variant.Property.Mat8\")" "(TagName=\"Cosmetics.Variant.Property.Mat9\")") false false))
            (Variables|Default|SetMetal_prim _index))
          (:then_3
            (bind _index_2 (Class|FortCosmeticFunctionLibrary|%28Deprecated%29GetVariantChannelActiveIndexForSlot "(TagName=\"LoadoutSlot.Character\")" _cosmeticloadout (Variables|Default|GetMetal%20sec) (Utilities|Array|MakeArray "(TagName=\"Cosmetics.Variant.Property.Mat10\")" "(TagName=\"Cosmetics.Variant.Property.Mat11\")" "(TagName=\"Cosmetics.Variant.Property.Mat12\")" "(TagName=\"Cosmetics.Variant.Property.Mat13\")" "(TagName=\"Cosmetics.Variant.Property.Mat14\")" "(TagName=\"Cosmetics.Variant.Property.Mat15\")" "(TagName=\"Cosmetics.Variant.Property.Mat16\")" "(TagName=\"Cosmetics.Variant.Property.Mat17\")" "(TagName=\"Cosmetics.Variant.Property.Mat18\")" "(TagName=\"Cosmetics.Variant.Property.Mat19\")") false false)))
          (:then_4
            (bind _index_3 (Class|FortCosmeticFunctionLibrary|%28Deprecated%29GetVariantChannelActiveIndexForSlot "(TagName=\"LoadoutSlot.Character\")" _cosmeticloadout (Variables|Default|GetMetal%20Tri) (Utilities|Array|MakeArray "(TagName=\"Cosmetics.Variant.Property.Mat20\")" "(TagName=\"Cosmetics.Variant.Property.Mat21\")" "(TagName=\"Cosmetics.Variant.Property.Mat22\")" "(TagName=\"Cosmetics.Variant.Property.Mat23\")" "(TagName=\"Cosmetics.Variant.Property.Mat24\")" "(TagName=\"Cosmetics.Variant.Property.Mat25\")" "(TagName=\"Cosmetics.Variant.Property.Mat26\")" "(TagName=\"Cosmetics.Variant.Property.Mat27\")" "(TagName=\"Cosmetics.Variant.Property.Mat28\")" "(TagName=\"Cosmetics.Variant.Property.Mat29\")") false false)))
          (:then_5
            (bind _index_4 (Class|FortCosmeticFunctionLibrary|%28Deprecated%29GetVariantChannelActiveIndexForSlot "(TagName=\"LoadoutSlot.Character\")" _cosmeticloadout (Variables|Default|GetPlastic%20Prim) (Utilities|Array|MakeArray "(TagName=\"Cosmetics.Variant.Property.Mat30\")" "(TagName=\"Cosmetics.Variant.Property.Mat31\")" "(TagName=\"Cosmetics.Variant.Property.Mat32\")" "(TagName=\"Cosmetics.Variant.Property.Mat33\")" "(TagName=\"Cosmetics.Variant.Property.Mat34\")" "(TagName=\"Cosmetics.Variant.Property.Mat35\")" "(TagName=\"Cosmetics.Variant.Property.Mat36\")" "(TagName=\"Cosmetics.Variant.Property.Mat37\")" "(TagName=\"Cosmetics.Variant.Property.Mat38\")" "(TagName=\"Cosmetics.Variant.Property.Mat39\")") false false)))
          (:then_6
            (bind _index_5 (Class|FortCosmeticFunctionLibrary|%28Deprecated%29GetVariantChannelActiveIndexForSlot "(TagName=\"LoadoutSlot.Character\")" _cosmeticloadout (Variables|Default|Getplastic%20sec) (Utilities|Array|MakeArray "(TagName=\"Cosmetics.Variant.Property.Mat40\")" "(TagName=\"Cosmetics.Variant.Property.Mat41\")" "(TagName=\"Cosmetics.Variant.Property.Mat42\")" "(TagName=\"Cosmetics.Variant.Property.Mat43\")" "(TagName=\"Cosmetics.Variant.Property.Mat44\")" "(TagName=\"Cosmetics.Variant.Property.Mat45\")" "(TagName=\"Cosmetics.Variant.Property.Mat46\")" "(TagName=\"Cosmetics.Variant.Property.Mat47\")" "(TagName=\"Cosmetics.Variant.Property.Mat48\")" "(TagName=\"Cosmetics.Variant.Property.Mat49\")") false false)))
          (:then_7
            (bind _index_6 (Class|FortCosmeticFunctionLibrary|%28Deprecated%29GetVariantChannelActiveIndexForSlot "(TagName=\"LoadoutSlot.Character\")" _cosmeticloadout (Variables|Default|Getenergy) (Utilities|Array|MakeArray "(TagName=\"Cosmetics.Variant.Property.Mat50\")" "(TagName=\"Cosmetics.Variant.Property.Mat51\")" "(TagName=\"Cosmetics.Variant.Property.Mat52\")" "(TagName=\"Cosmetics.Variant.Property.Mat53\")" "(TagName=\"Cosmetics.Variant.Property.Mat54\")" "(TagName=\"Cosmetics.Variant.Property.Mat55\")") false false))))
        (CallFunction|MaterialSetuup (Utilities|Array|Get(aref) (Variables|Default|GetMetal-Prim) _index_3) (Utilities|Array|Get(aref) (Variables|Default|GetMetal-Prim-MRS) _index_3) (Utilities|Array|Get(aref) (Variables|Default|GetMetal-Sec) _index) (Utilities|Array|Get(aref) (Variables|Default|GetMetal-Sec%20MRS) _index) (Utilities|Array|Get(aref) (Variables|Default|GetMetal-Tret) _index_2) (Utilities|Array|Get(aref) (Variables|Default|GetMetal-Tret%20MRS) _index_2) (Utilities|Array|Get(aref) (Variables|Default|GetPlastic-Prim) _index_4) (Utilities|Array|Get(aref) (Variables|Default|GetPlastic-Prim%20MRS) _index_4) (Utilities|Array|Get(aref) (Variables|Default|GetPlastic-Sec) _index_5) (Utilities|Array|Get(aref) (Variables|Default|GetPlastic-Sec%20MRS) _index_5) (Utilities|Array|Get(aref) (Variables|Default|GetEngery) _index_6)))
      (:then_1
        (CallFunction|ColorSet)))))

(event EventOnCharacterCustomizationFinished
  (params (CharacterPartInfo :type Array%20of%20Applied%20Character%20Part%20Info%20Structures))
  (bind _returnvalue (CustomCharacter|GetAssociatedPlayerPawn))
  (bind _cosmeticloadout (Variables|/Script/FortniteGame.FortPlayerPawn|GetCosmeticLoadout _returnvalue))
  (if (Utilities|String|Equal,CaseInsensitive(String) (Item|GetPersistentName (CosmeticLoadout|GetItemDefinitionfromLoadout (CosmeticLoadout|GetCosmeticLoadout) "(TagName=\"LoadoutSlot.Character\")")) (Variables|Default|GetCharacter-Name))
    (Utilities|FlowControl|Sequence
      (:then_0
        (bind _asmaterial_instance_dynamic (Utilities|Casting|CastToMaterialInstanceDynamic (Rendering|Material|GetMaterial (Customization|GetSkeletalMeshforPartType (CustomCharacter|GetAssociatedPlayerPawn) "Body" :!class /Script/FortniteGame.FortPlayerPawn) 0)))
        (Variables|Default|SetCharacter%20Material _asmaterial_instance_dynamic)
        (bind _asmaterial_instance_dynamic_2 (Utilities|Casting|CastToMaterialInstanceDynamic (Rendering|Material|GetMaterial (Customization|GetSkeletalMeshforPartType (CustomCharacter|GetAssociatedPlayerPawn) "Body" :!class /Script/FortniteGame.FortPlayerPawn) 0)))
        (Variables|Default|SetSigelMat _asmaterial_instance_dynamic_2)
        (Utilities|FlowControl|Sequence
          (:then_2
            (bind _index (Class|FortCosmeticFunctionLibrary|%28Deprecated%29GetVariantChannelActiveIndexForSlot "(TagName=\"LoadoutSlot.Character\")" _cosmeticloadout (Variables|Default|GetMetal%20Prim) (Utilities|Array|MakeArray "(TagName=\"Cosmetics.Variant.Property.Mat0\")" "(TagName=\"Cosmetics.Variant.Property.Mat1\")" "(TagName=\"Cosmetics.Variant.Property.Mat2\")" "(TagName=\"Cosmetics.Variant.Property.Mat3\")" "(TagName=\"Cosmetics.Variant.Property.Mat4\")" "(TagName=\"Cosmetics.Variant.Property.Mat5\")" "(TagName=\"Cosmetics.Variant.Property.Mat6\")" "(TagName=\"Cosmetics.Variant.Property.Mat7\")" "(TagName=\"Cosmetics.Variant.Property.Mat8\")" "(TagName=\"Cosmetics.Variant.Property.Mat9\")") false false))
            (Variables|Default|SetMetal_prim _index))
          (:then_3
            (bind _index_2 (Class|FortCosmeticFunctionLibrary|%28Deprecated%29GetVariantChannelActiveIndexForSlot "(TagName=\"LoadoutSlot.Character\")" _cosmeticloadout (Variables|Default|GetMetal%20sec) (Utilities|Array|MakeArray "(TagName=\"Cosmetics.Variant.Property.Mat10\")" "(TagName=\"Cosmetics.Variant.Property.Mat11\")" "(TagName=\"Cosmetics.Variant.Property.Mat12\")" "(TagName=\"Cosmetics.Variant.Property.Mat13\")" "(TagName=\"Cosmetics.Variant.Property.Mat14\")" "(TagName=\"Cosmetics.Variant.Property.Mat15\")" "(TagName=\"Cosmetics.Variant.Property.Mat16\")" "(TagName=\"Cosmetics.Variant.Property.Mat17\")" "(TagName=\"Cosmetics.Variant.Property.Mat18\")" "(TagName=\"Cosmetics.Variant.Property.Mat19\")") false false)))
          (:then_4
            (bind _index_3 (Class|FortCosmeticFunctionLibrary|%28Deprecated%29GetVariantChannelActiveIndexForSlot "(TagName=\"LoadoutSlot.Character\")" _cosmeticloadout (Variables|Default|GetMetal%20Tri) (Utilities|Array|MakeArray "(TagName=\"Cosmetics.Variant.Property.Mat20\")" "(TagName=\"Cosmetics.Variant.Property.Mat21\")" "(TagName=\"Cosmetics.Variant.Property.Mat22\")" "(TagName=\"Cosmetics.Variant.Property.Mat23\")" "(TagName=\"Cosmetics.Variant.Property.Mat24\")" "(TagName=\"Cosmetics.Variant.Property.Mat25\")" "(TagName=\"Cosmetics.Variant.Property.Mat26\")" "(TagName=\"Cosmetics.Variant.Property.Mat27\")" "(TagName=\"Cosmetics.Variant.Property.Mat28\")" "(TagName=\"Cosmetics.Variant.Property.Mat29\")") false false)))
          (:then_5
            (bind _index_4 (Class|FortCosmeticFunctionLibrary|%28Deprecated%29GetVariantChannelActiveIndexForSlot "(TagName=\"LoadoutSlot.Character\")" _cosmeticloadout (Variables|Default|GetPlastic%20Prim) (Utilities|Array|MakeArray "(TagName=\"Cosmetics.Variant.Property.Mat30\")" "(TagName=\"Cosmetics.Variant.Property.Mat31\")" "(TagName=\"Cosmetics.Variant.Property.Mat32\")" "(TagName=\"Cosmetics.Variant.Property.Mat33\")" "(TagName=\"Cosmetics.Variant.Property.Mat34\")" "(TagName=\"Cosmetics.Variant.Property.Mat35\")" "(TagName=\"Cosmetics.Variant.Property.Mat36\")" "(TagName=\"Cosmetics.Variant.Property.Mat37\")" "(TagName=\"Cosmetics.Variant.Property.Mat38\")" "(TagName=\"Cosmetics.Variant.Property.Mat39\")") false false)))
          (:then_6
            (bind _index_5 (Class|FortCosmeticFunctionLibrary|%28Deprecated%29GetVariantChannelActiveIndexForSlot "(TagName=\"LoadoutSlot.Character\")" _cosmeticloadout (Variables|Default|Getplastic%20sec) (Utilities|Array|MakeArray "(TagName=\"Cosmetics.Variant.Property.Mat40\")" "(TagName=\"Cosmetics.Variant.Property.Mat41\")" "(TagName=\"Cosmetics.Variant.Property.Mat42\")" "(TagName=\"Cosmetics.Variant.Property.Mat43\")" "(TagName=\"Cosmetics.Variant.Property.Mat44\")" "(TagName=\"Cosmetics.Variant.Property.Mat45\")" "(TagName=\"Cosmetics.Variant.Property.Mat46\")" "(TagName=\"Cosmetics.Variant.Property.Mat47\")" "(TagName=\"Cosmetics.Variant.Property.Mat48\")" "(TagName=\"Cosmetics.Variant.Property.Mat49\")") false false)))
          (:then_7
            (bind _index_6 (Class|FortCosmeticFunctionLibrary|%28Deprecated%29GetVariantChannelActiveIndexForSlot "(TagName=\"LoadoutSlot.Character\")" _cosmeticloadout (Variables|Default|Getenergy) (Utilities|Array|MakeArray "(TagName=\"Cosmetics.Variant.Property.Mat50\")" "(TagName=\"Cosmetics.Variant.Property.Mat51\")" "(TagName=\"Cosmetics.Variant.Property.Mat52\")" "(TagName=\"Cosmetics.Variant.Property.Mat53\")" "(TagName=\"Cosmetics.Variant.Property.Mat54\")" "(TagName=\"Cosmetics.Variant.Property.Mat55\")") false false))))
        (CallFunction|MaterialSetuup (Utilities|Array|Get(aref) (Variables|Default|GetMetal-Prim) _index_3) (Utilities|Array|Get(aref) (Variables|Default|GetMetal-Prim-MRS) _index_3) (Utilities|Array|Get(aref) (Variables|Default|GetMetal-Sec) _index) (Utilities|Array|Get(aref) (Variables|Default|GetMetal-Sec%20MRS) _index) (Utilities|Array|Get(aref) (Variables|Default|GetMetal-Tret) _index_2) (Utilities|Array|Get(aref) (Variables|Default|GetMetal-Tret%20MRS) _index_2) (Utilities|Array|Get(aref) (Variables|Default|GetPlastic-Prim) _index_4) (Utilities|Array|Get(aref) (Variables|Default|GetPlastic-Prim%20MRS) _index_4) (Utilities|Array|Get(aref) (Variables|Default|GetPlastic-Sec) _index_5) (Utilities|Array|Get(aref) (Variables|Default|GetPlastic-Sec%20MRS) _index_5) (Utilities|Array|Get(aref) (Variables|Default|GetEngery) _index_6)))
      (:then_1
        (CallFunction|ColorSet)))))

(event EventBeginPlay
  (Utilities|FlowControl|Delay 0.5)
  (Utilities|Time|SetTimerbyEvent (delegate Custom|CustomEvent) 0.5 true))

(event Custom|CustomEvent
  (CallFunction|ColorSet))

(event Custom|OnBannerFullyApplied
  (params (Icon :type Texture%202D%20Object%20Reference) (PrimaryColor :type Linear%20Color%20Structure)))

(event Custom|Color%20Set
  ; Bolt
  (bind (_alpha _alpha_m_r_s _blue _blue_mrs _green _green_mrs _red _red_mrs _accent _accent_mrs _glowswrap) (CallFunction|ColorSetups "(R=1.350000,G=1.111013,B=0.021734,A=1.000000)" "0.000000,0.450000,0.500000" "(R=0.020084,G=0.802084,B=1.200000,A=1.000000)" "0.000000,0.350000,0.500000" "(R=0.289603,G=0.313720,B=0.400000,A=1.000000)" "0.300000,0.750000,0.500000" "(R=0.100882,G=0.119508,B=0.140000,A=1.000000)" "0.000000,0.500000,0.500000" "(R=0.013963,G=0.033347,B=0.055000,A=1.000000)" "0.000000,0.250000,0.500000" "(R=0.250000,G=0.863862,B=2.500000,A=1.000000)"))
  (Variables|Default|SetBolt (Utilities|Array|MakeArray _alpha _blue _green _red _accent _glowswrap))
  (Variables|Default|SetBolt%20MRS (Utilities|Array|MakeArray _alpha_m_r_s _blue_mrs _green_mrs _red_mrs _accent_mrs))
  ; Lava
  (bind (_alpha_2 _alpha_m_r_s_2 _blue_2 _blue_mrs_2 _green_2 _green_mrs_2 _red_2 _red_mrs_2 _accent_2 _accent_mrs_2 _glowswrap_2) (CallFunction|ColorSetups "(R=0.430706,G=0.478077,B=0.600000,A=1.000000)" "0.660000,0.400000,0.500000" "(R=0.247205,G=0.580656,B=1.000000,A=1.000000)" "0.450000,0.450000,0.500000" "(R=0.605803,G=0.625498,B=0.666000,A=1.000000)" "0.600000,0.500000,0.500000" "(R=0.075000,G=0.075000,B=0.075000,A=1.000000)" "0.000000,0.650000,0.500000" "(R=0.012000,G=0.012000,B=0.012000,A=1.000000)" "0.000000,0.200000,0.500000" "(R=2.580000,G=0.258000,B=0.258039,A=1.000000)"))
  (Variables|Default|SetLava (Utilities|Array|MakeArray _alpha_2 _blue_2 _green_2 _red_2 _accent_2 _glowswrap_2))
  (Variables|Default|SetLava%20MRS (Utilities|Array|MakeArray _alpha_m_r_s_2 _blue_mrs_2 _green_mrs_2 _red_mrs_2 _accent_mrs_2))
  (bind (_alpha_3 _alpha_m_r_s_3 _blue_3 _blue_mrs_3 _green_3 _green_mrs_3 _red_3 _red_mrs_3 _accent_3 _accent_mrs_3 _glowswrap_3) (CallFunction|ColorSetups "(R=1.000000,G=0.053329,B=0.021039,A=1.000000)" "0.000000,0.400000,0.500000" "(R=1.500000,G=1.484463,B=1.438421,A=1.000000)" "0.000000,0.300000,0.500000" "(R=0.550000,G=0.526165,B=0.495340,A=1.000000)" "0.550000,0.850000,0.500000" "(R=0.335932,G=0.350000,B=0.335932,A=1.000000)" "0.000000,0.450000,0.500000" "(R=0.208637,G=0.205079,B=0.254152,A=1.000000)" "0.700000,0.330000,0.500000" "(R=3.500000,G=0.193591,B=0.000728,A=1.000000)"))
  (Variables|Default|SetDrill (Utilities|Array|MakeArray _alpha_3 _blue_3 _green_3 _red_3 _accent_3 _glowswrap_3))
  (Variables|Default|SetDrill%20MRS (Utilities|Array|MakeArray _alpha_m_r_s_3 _blue_mrs_3 _green_mrs_3 _red_mrs_3 _accent_mrs_3))
  (bind (_alpha_4 _alpha_m_r_s_4 _blue_4 _blue_mrs_4 _green_4 _green_mrs_4 _red_4 _red_mrs_4 _accent_4 _accent_mrs_4 _glowswrap_4) (CallFunction|ColorSetups "(R=0.350000,G=0.343475,B=0.299786,A=1.000000)" "0.650000,0.550000,0.500000" "(R=0.093784,G=0.119975,B=0.150000,A=1.000000)" "0.500000,0.750000,0.500000" "(R=0.018211,G=0.019294,B=0.024000,A=1.000000)" "0.000000,1.500000,0.500000" "(R=0.550000,G=0.501775,B=0.379500,A=1.000000)" "0.000000,0.550000,0.350000" "(R=0.400000,G=0.390125,B=0.375582,A=1.000000)" "0.850000,0.400000,0.350000" "(R=0.000000,G=1.500000,B=0.016750,A=1.000000)"))
  (Variables|Default|SetNail (Utilities|Array|MakeArray _alpha_4 _blue_4 _green_4 _red_4 _accent_4 _glowswrap_4))
  (Variables|Default|SetNail%20MRS (Utilities|Array|MakeArray _alpha_m_r_s_4 _blue_mrs_4 _green_mrs_4 _red_mrs_4 _accent_mrs_4))
  (bind (_alpha_5 _alpha_m_r_s_5 _blue_5 _blue_mrs_5 _green_5 _green_mrs_5 _red_5 _red_mrs_5 _accent_5 _accent_mrs_5 _glowswrap_5) (CallFunction|ColorSetups "(R=1.000000,G=0.058559,B=0.029781,A=1.000000)" "0.000000,0.500000,0.500000" "(R=0.450000,G=0.049521,B=0.010622,A=1.000000)" "0.000000,0.400000,0.500000" "(R=0.450000,G=0.426026,B=0.410469,A=1.000000)" "0.650000,0.650000,0.500000" "(R=1.206760,G=1.250000,B=1.206760,A=1.000000)" "0.000000,0.450000,0.500000" "(R=0.380000,G=0.341004,B=0.341004,A=1.000000)" "0.000000,0.300000,0.350000" "(R=0.006411,G=0.367315,B=3.500000,A=1.000000)"))
  (Variables|Default|SetScrew (Utilities|Array|MakeArray _alpha_5 _blue_5 _green_5 _red_5 _accent_5 _glowswrap_5))
  (Variables|Default|SetScrew%20MRS (Utilities|Array|MakeArray _alpha_m_r_s_5 _blue_mrs_5 _green_mrs_5 _red_mrs_5 _accent_mrs_5))
  (bind (_alpha_6 _alpha_m_r_s_6 _blue_6 _blue_mrs_6 _green_6 _green_mrs_6 _red_6 _red_mrs_6 _accent_6 _accent_mrs_6 _glowswrap_6) (CallFunction|ColorSetups "(R=0.327923,G=0.381000,B=0.374111,A=1.000000)" "0.000000,0.750000,0.300000" "(R=0.327923,G=0.381000,B=0.374111,A=1.000000)" "0.000000,0.650000,0.300000" "(R=0.327923,G=0.381000,B=0.374111,A=1.000000)" "0.000000,0.800000,0.300000" "(R=0.327923,G=0.381000,B=0.374111,A=1.000000)" "0.000000,0.500000,0.300000" "(R=0.132868,G=0.155926,B=0.155926,A=1.000000)" "0.000000,0.350000,0.300000" "(R=1.775276,G=0.000000,B=3.000000,A=1.000000)"))
  (Variables|Default|SetClay (Utilities|Array|MakeArray _alpha_6 _blue_6 _green_6 _red_6 _accent_6 _glowswrap_6))
  (Variables|Default|SetClay%20MRS (Utilities|Array|MakeArray _alpha_m_r_s_6 _blue_mrs_6 _green_mrs_6 _red_mrs_6 _accent_mrs_6))
  ; Cosmic
  (bind (_alpha_7 _alpha_m_r_s_7 _blue_7 _blue_mrs_7 _green_7 _green_mrs_7 _red_7 _red_mrs_7 _accent_7 _accent_mrs_7) (CallFunction|ColorSetups "(R=0.096984,G=0.000000,B=0.450000,A=1.000000)" "0.100000,0.500000,0.250000" "(R=0.096984,G=0.000000,B=0.450000,A=1.000000)" "0.100000,0.500000,0.250000" "(R=0.096984,G=0.000000,B=0.450000,A=1.000000)" "0.100000,0.750000,0.250000" "(R=0.096984,G=0.000000,B=0.450000,A=1.000000)" "0.100000,0.500000,0.250000" "(R=0.053880,G=0.000000,B=0.250000,A=1.000000)" "0.100000,0.400000,0.250000" "(R=0.184275,G=0.010022,B=0.375000,A=1.000000)"))
  ; Gold
  (bind (_alpha_8 _alpha_m_r_s_8 _blue_8 _blue_mrs_8 _green_8 _green_mrs_8 _red_8 _red_mrs_8 _accent_8 _accent_mrs_8) (CallFunction|ColorSetups "(R=1.750000,G=0.990328,B=0.237893,A=1.000000)" "1.000000,0.400000,0.500000" "(R=1.750000,G=0.990329,B=0.237893,A=1.000000)" "1.000000,0.400000,0.500000" "(R=1.750000,G=0.990329,B=0.237893,A=1.000000)" "1.000000,0.800000,0.500000" "(R=1.750000,G=0.990329,B=0.237893,A=1.000000)" "1.000000,0.400000,0.500000" "(R=0.750000,G=0.424427,B=0.101954,A=1.000000)" "1.000000,0.300000,0.500000" "(R=0.701102,G=0.396755,B=0.095307,A=1.000000)"))
  ; Crystal
  (bind (_alpha_9 _alpha_m_r_s_9 _blue_9 _blue_mrs_9 _green_9 _green_mrs_9 _red_9 _red_mrs_9 _accent_9 _accent_mrs_9) (CallFunction|ColorSetups "(R=1.500000,G=1.500000,B=1.480825,A=1.000000)" "0.800000,0.150000,0.500000" "(R=1.500000,G=1.500000,B=1.480825,A=1.000000)" "0.800000,0.150000,0.500000" "(R=1.500000,G=1.500000,B=1.480825,A=1.000000)" "0.800000,0.250000,0.500000" "(R=1.500000,G=1.500000,B=1.480825,A=1.000000)" "0.800000,0.150000,0.500000" "(R=0.750000,G=0.750000,B=0.740411,A=1.000000)" "0.800000,0.100000,0.500000" "(R=0.500000,G=0.500000,B=0.493608,A=1.000000)"))
  ; Pearl
  (bind (_alpha_10 _alpha_m_r_s_10 _blue_10 _blue_mrs_10 _green_10 _green_mrs_10 _red_10 _red_mrs_10 _accent_10 _accent_mrs_10) (CallFunction|ColorSetups "(R=1.500000,G=1.326280,B=1.244060,A=1.000000)" "0.800000,0.500000,0.500000" "(R=1.500000,G=1.326280,B=1.244060,A=1.000000)" "0.800000,0.500000,0.500000" "(R=1.500000,G=1.326280,B=1.244060,A=1.000000)" "0.800000,0.850000,0.500000" "(R=1.500000,G=1.326280,B=1.244060,A=1.000000)" "0.800000,0.500000,0.500000" "(R=0.550000,G=0.492187,B=0.449789,A=1.000000)" "0.800000,0.300000,0.500000" "(R=0.184275,G=0.010022,B=0.375000,A=1.000000)"))
  (Variables|Default|SetMetal-Prim (Utilities|Array|MakeArray _alpha_2 _alpha _alpha_3 _alpha_4 _alpha_5 _alpha_6 _alpha_7 _alpha_8 _alpha_9 _alpha_10))
  (Variables|Default|SetMetal-Prim-MRS (Utilities|Array|MakeArray _alpha_m_r_s_2 _alpha_m_r_s _alpha_m_r_s_3 _alpha_m_r_s_4 _alpha_m_r_s_5 _alpha_m_r_s_6 _alpha_m_r_s_7 _alpha_m_r_s_8 _alpha_m_r_s_9 _alpha_m_r_s_10))
  (Variables|Default|SetMetal-Sec (Utilities|Array|MakeArray _blue_2 _blue _blue_3 _blue_4 _blue_5 _blue_6 _blue_7 _blue_8 _blue_9 _blue_10))
  (Variables|Default|SetMetal-Sec%20MRS (Utilities|Array|MakeArray _blue_mrs_2 _blue_mrs _blue_mrs_3 _blue_mrs_4 _blue_mrs_5 _blue_mrs_6 _blue_mrs_7 _blue_mrs_8 _blue_mrs_9 _blue_mrs_10))
  (Variables|Default|SetMetal-Tret (Utilities|Array|MakeArray _green_2 _green _green_3 _green_4 _green_5 _green_6 _green_7 _green_8 _green_9 _green_10))
  (Variables|Default|SetMetal-Tret%20MRS (Utilities|Array|MakeArray _green_mrs_2 _green_mrs _green_mrs_3 _green_mrs_4 _green_mrs_5 _green_mrs_6 _green_mrs_7 _green_mrs_8 _green_mrs_9 _green_mrs_10))
  (Variables|Default|SetPlastic-Prim (Utilities|Array|MakeArray _red_2 _red _red_3 _red_4 _red_5 _red_6 _red_7 _red_8 _red_9 _red_10))
  (Variables|Default|SetPlastic-Prim%20MRS (Utilities|Array|MakeArray _red_mrs_2 _red_mrs _red_mrs_3 _red_mrs_4 _red_mrs_5 _red_mrs_6 _red_mrs_7 _red_mrs_8 _red_mrs_9 _red_mrs_10))
  (Variables|Default|SetPlastic-Sec (Utilities|Array|MakeArray _accent_2 _accent _accent_3 _accent_4 _accent_5 _accent_6 _accent_7 _accent_8 _accent_9 _accent_10))
  (Variables|Default|SetPlastic-Sec%20MRS (Utilities|Array|MakeArray _accent_mrs_2 _accent_mrs _accent_mrs_3 _accent_mrs_4 _accent_mrs_5 _accent_mrs_6 _accent_mrs_7 _accent_mrs_8 _accent_mrs_9 _accent_mrs_10))
  (Variables|Default|SetEngery (Utilities|Array|MakeArray :%5B0%5D _glowswrap_2 :%5B1%5D _glowswrap :%5B2%5D _glowswrap_3 :%5B3%5D _glowswrap_4 :%5B4%5D _glowswrap_5 :%5B5%5D _glowswrap_6)))

(fn Colorsetups
  (params (Mat1%20Albedo :type Linear%20Color%20Structure) (Mat%201%20Remap :type Vector) (Mat2 :type Linear%20Color%20Structure) (Mat2%20MRS :type Vector) (Mat3 :type Linear%20Color%20Structure) (Mat3%20MRS :type Vector) (Mat4 :type Linear%20Color%20Structure) (Mat4%20MRS :type Vector) (Mat5 :type Linear%20Color%20Structure) (Mat5%20MRS :type Vector) (Glows :type Linear%20Color%20Structure))
  (return Mat3 Mat3%20MRS Mat1%20Albedo Mat%201%20Remap Mat2 Mat2%20MRS Mat4 Mat4%20MRS Mat5 Mat5%20MRS Glows))

(fn Materialsetuup
  (params (Alpha%20Color :type Linear%20Color%20Structure) (Alpha%20Metalic_X :type Float%20%28double-precision%29) (Alpha%20Metalic_Y :type Float%20%28double-precision%29) (Alpha%20Metalic_Z :type Float%20%28double-precision%29) (Blue%20Color :type Linear%20Color%20Structure) (Blue%20Metalic_X :type Float%20%28double-precision%29) (Blue%20Metalic_Y :type Float%20%28double-precision%29) (Blue%20Metalic_Z :type Float%20%28double-precision%29) (Green%20Color :type Linear%20Color%20Structure) (Green%20metalic_X :type Float%20%28double-precision%29) (Green%20metalic_Y :type Float%20%28double-precision%29) (Green%20metalic_Z :type Float%20%28double-precision%29) (Red%20Color :type Linear%20Color%20Structure) (Red%20Metalic_X :type Float%20%28double-precision%29) (Red%20Metalic_Y :type Float%20%28double-precision%29) (Red%20Metalic_Z :type Float%20%28double-precision%29) (Accent%20Color :type Linear%20Color%20Structure) (Accent%20Metalic_X :type Float%20%28double-precision%29) (Accent%20Metalic_Y :type Float%20%28double-precision%29) (Accent%20Metalic_Z :type Float%20%28double-precision%29) (Emissive%20color :type Linear%20Color%20Structure))
  (Utilities|FlowControl|Sequence
    (:then_0
      (for _array_element (CustomItemWrap|GetWrappedMeshComponents)
        (for _array_element_2 (Rendering|Material|GetMaterials _array_element)
          (bind _asmaterial_instance_dynamic (Utilities|Casting|CastToMaterialInstanceDynamic _array_element_2))
          (Rendering|Material|SetVectorParameterValue _asmaterial_instance_dynamic "FDecalV2_Base_RGBA_Color=Red" Alpha%20Color :!class /Script/Engine.MaterialInstanceDynamic)
          (Rendering|Material|SetScalarParameterValue _asmaterial_instance_dynamic "FDecalV2_Metallic" Alpha%20Metalic_X :!class /Script/Engine.MaterialInstanceDynamic)
          (Rendering|Material|SetScalarParameterValue _asmaterial_instance_dynamic "FDecalV2_Rougness" (* Alpha%20Metalic_Y 0.25) :!class /Script/Engine.MaterialInstanceDynamic)
          (Rendering|Material|SetScalarParameterValue _asmaterial_instance_dynamic "FDecalV2_Spec" Alpha%20Metalic_Z :!class /Script/Engine.MaterialInstanceDynamic))))
    (:then_1
      (for _array_element_3 (CustomItemWrap|GetWrappedMeshComponents)
        (for _array_element_4 (Rendering|Material|GetMaterials _array_element_3)
          (bind _asmaterial_instance_dynamic_2 (Utilities|Casting|CastToMaterialInstanceDynamic _array_element_4))
          (Rendering|Material|SetVectorParameterValue _asmaterial_instance_dynamic_2 "TopS_01_ColorTint_CH_Alpha   A=0=off" Alpha%20Color :!class /Script/Engine.MaterialInstanceDynamic)
          (Rendering|Material|SetScalarParameterValue _asmaterial_instance_dynamic_2 "TopS_01_Metalic_CH_Alpha" Alpha%20Metalic_X :!class /Script/Engine.MaterialInstanceDynamic)
          (Rendering|Material|SetScalarParameterValue _asmaterial_instance_dynamic_2 "TopS_01_Roughtness_CH_Alpha" Alpha%20Metalic_Y :!class /Script/Engine.MaterialInstanceDynamic)
          (Rendering|Material|SetScalarParameterValue _asmaterial_instance_dynamic_2 "TopS_01_Spec_CH_Alpha" Alpha%20Metalic_Z :!class /Script/Engine.MaterialInstanceDynamic))))
    (:then_2
      (for _array_element_5 (CustomItemWrap|GetWrappedMeshComponents)
        (for _array_element_6 (Rendering|Material|GetMaterials _array_element_5)
          (bind _asmaterial_instance_dynamic_3 (Utilities|Casting|CastToMaterialInstanceDynamic _array_element_6))
          (Rendering|Material|SetVectorParameterValue _asmaterial_instance_dynamic_3 "TopS_01_ColorTint_CH_Blue  A=0=off" Blue%20Color :!class /Script/Engine.MaterialInstanceDynamic)
          (Rendering|Material|SetScalarParameterValue _asmaterial_instance_dynamic_3 "TopS_01_Metalic_CH_Blue" Blue%20Metalic_X :!class /Script/Engine.MaterialInstanceDynamic)
          (Rendering|Material|SetScalarParameterValue _asmaterial_instance_dynamic_3 "TopS_01_Roughtness_CH_Blue" Blue%20Metalic_Y :!class /Script/Engine.MaterialInstanceDynamic)
          (Rendering|Material|SetScalarParameterValue _asmaterial_instance_dynamic_3 "TopS_01_Spec_CH_Blue" Blue%20Metalic_Z :!class /Script/Engine.MaterialInstanceDynamic))))
    (:then_3
      (for _array_element_7 (CustomItemWrap|GetWrappedMeshComponents)
        (for _array_element_8 (Rendering|Material|GetMaterials _array_element_7)
          (bind _asmaterial_instance_dynamic_4 (Utilities|Casting|CastToMaterialInstanceDynamic _array_element_8))
          (Rendering|Material|SetVectorParameterValue _asmaterial_instance_dynamic_4 "TopS_01_ColorTint_CH_Green A=0=off" Green%20Color :!class /Script/Engine.MaterialInstanceDynamic)
          (Rendering|Material|SetScalarParameterValue _asmaterial_instance_dynamic_4 "TopS_01_Metalic_CH_Green" Green%20metalic_X :!class /Script/Engine.MaterialInstanceDynamic)
          (Rendering|Material|SetScalarParameterValue _asmaterial_instance_dynamic_4 "TopS_01_Roughtness_CH_Green" Green%20metalic_Y :!class /Script/Engine.MaterialInstanceDynamic)
          (Rendering|Material|SetScalarParameterValue _asmaterial_instance_dynamic_4 "TopS_01_Spec_CH_Green" Green%20metalic_Z :!class /Script/Engine.MaterialInstanceDynamic))))
    (:then_4
      (for _array_element_9 (CustomItemWrap|GetWrappedMeshComponents)
        (for _array_element_10 (Rendering|Material|GetMaterials _array_element_9)
          (bind _asmaterial_instance_dynamic_5 (Utilities|Casting|CastToMaterialInstanceDynamic _array_element_10))
          (Rendering|Material|SetVectorParameterValue _asmaterial_instance_dynamic_5 "TopS_01_ColorTint_CH_Red  A=0=off" Red%20Color :!class /Script/Engine.MaterialInstanceDynamic)
          (Rendering|Material|SetScalarParameterValue _asmaterial_instance_dynamic_5 "TopS_01_Metalic_CH_Red" Red%20Metalic_X :!class /Script/Engine.MaterialInstanceDynamic)
          (Rendering|Material|SetScalarParameterValue _asmaterial_instance_dynamic_5 "TopS_01_Roughtness_CH_Red" Red%20Metalic_Y :!class /Script/Engine.MaterialInstanceDynamic)
          (Rendering|Material|SetScalarParameterValue _asmaterial_instance_dynamic_5 "TopS_01_Spec_CH_Red" Red%20Metalic_Z :!class /Script/Engine.MaterialInstanceDynamic))))
    (:then_5
      (for _array_element_11 (CustomItemWrap|GetWrappedMeshComponents)
        (for _array_element_12 (Rendering|Material|GetMaterials _array_element_11)
          (bind _asmaterial_instance_dynamic_6 (Utilities|Casting|CastToMaterialInstanceDynamic _array_element_12))
          (Rendering|Material|SetVectorParameterValue _asmaterial_instance_dynamic_6 "AccentV2_BaseColor_" Accent%20Color :!class /Script/Engine.MaterialInstanceDynamic)
          (Rendering|Material|SetScalarParameterValue _asmaterial_instance_dynamic_6 "AccentV2_Metalness" Accent%20Metalic_X :!class /Script/Engine.MaterialInstanceDynamic)
          (Rendering|Material|SetScalarParameterValue _asmaterial_instance_dynamic_6 "AccentV2_Roughness" Accent%20Metalic_Y :!class /Script/Engine.MaterialInstanceDynamic)
          (Rendering|Material|SetScalarParameterValue _asmaterial_instance_dynamic_6 "AccentV2_Specular" Accent%20Metalic_Z :!class /Script/Engine.MaterialInstanceDynamic))))
    (:then_6
      (for _array_element_13 (CustomItemWrap|GetWrappedMeshComponents)
        (for _array_element_14 (Rendering|Material|GetMaterials _array_element_13)
          (bind _asmaterial_instance_dynamic_7 (Utilities|Casting|CastToMaterialInstanceDynamic _array_element_14))
          (Rendering|Material|SetVectorParameterValue _asmaterial_instance_dynamic_7 "TopS Emissive_01_Color-" Emissive%20color :!class /Script/Engine.MaterialInstanceDynamic)
          (Rendering|Material|SetVectorParameterValue _asmaterial_instance_dynamic_7 "FDecalV2-Emissive-Color--Alpha=Multplyer" Emissive%20color :!class /Script/Engine.MaterialInstanceDynamic))))))

(fn ConstructionScript)
