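; CIWM_CyberFNCS is the logic Blueprint for the CyberFNCS weapon wrap, driving emissive decal and TopS material effects when the wrapped weapon fires.
; Each shot stamps current game time into the brightness time and end parameters on all wrapped mesh materials.
; Unless Killed is set, it immediately overwrites the end parameters with fixed constants.
; Any animation from these timestamps is left to the materials, as the graph has no tick or timer.
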(event EventOnWeaponFired
  (bind _returnvalue (CustomItemWrap|GetWrappedMeshComponents))
  (bind _returnvalue_2 (Utilities|Time|GetGameTimeinSeconds))
  (Rendering|Material|SetScalarParameterValueonMaterials _returnvalue "FDecalV2_Emissive_Brightness_Time_BP" _returnvalue_2)
  (Rendering|Material|SetScalarParameterValueonMaterials _returnvalue "TopS Emissive_02-Brightness-TBP" _returnvalue_2)
  (Rendering|Material|SetScalarParameterValueonMaterials _returnvalue "TopS Emissive_01-Brightness-TBP" _returnvalue_2)
  (Rendering|Material|SetScalarParameterValueonMaterials _returnvalue "TopS Emissive_01_Brightness---End" _returnvalue_2)
  (Rendering|Material|SetScalarParameterValueonMaterials _returnvalue "TopS Emissive_02_Brightness---End" _returnvalue_2)
  (Rendering|Material|SetScalarParameterValueonMaterials _returnvalue "FDecalV2_Emissive_Brightness_Time__End" _returnvalue_2)
  (if (not (Variables|Default|GetKilled))
    (Rendering|Material|SetScalarParameterValueonMaterials _returnvalue "FDecalV2_Emissive_Brightness_Time__End" 0.0)
    (Rendering|Material|SetScalarParameterValueonMaterials _returnvalue "TopS Emissive_02_Brightness---End" 0.0)
    (Rendering|Material|SetScalarParameterValueonMaterials _returnvalue "TopS Emissive_01_Brightness---End" 0.2)))

(fn ConstructionScript)
