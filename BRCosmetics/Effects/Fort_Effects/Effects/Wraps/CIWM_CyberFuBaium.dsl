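; Weapon wrap Blueprint for CyberFuBaium that refreshes one of four material decals (A to D) whenever the weapon fires.
; After a tiny delay, it picks a decal slot at random and calls RandomSet for weapon, character, backpack, pickaxe, mech and vehicle layouts.
; RandomSet stamps game time for decal and emissive reactions on wrapped-mesh materials, randomizing flipbook frame, rotation and jittered offset.
; Most slot B calls write DecalA's time parameter.
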(event EventOnWeaponFired
  (Utilities|FlowControl|Delay 0.001)
  (switch int (Math|Random|RandomInteger 4)
    (:0
      (CallFunction|RandomSet 0 "EDRec_TimeFrom_BP_DecalA" "EDRec_DecalA_Offset_Weapon" "EDRec_DecalA_Rotation_Weapon" "EDRec_FlipBook_FrameA" (-0.4 0.0 0.0 0.1) "EDRec_Emissive_A_Brightness_Time_BP" 0.025)
      (CallFunction|RandomSet 0 "EDRec_TimeFrom_BP_DecalA" "EDRec_DecalA_Offset_Character" "EDRec_DecalA_Rotation_Character" "EDRec_FlipBook_FrameA" (0.13 0.25 0.0 0.12) "EDRec_Emissive_A_Brightness_Time_BP" 0.1)
      (CallFunction|RandomSet 0 "EDRec_TimeFrom_BP_DecalA" "EDRec_DecalA_Offset_Backpack----A=Scale" "EDRec_DecalA_Rotation_Backpack" "EDRec_FlipBook_FrameA" (0.2 0.3 0.0 0.15) "EDRec_Emissive_A_Brightness_Time_BP" 0.1)
      (CallFunction|RandomSet 0 "EDRec_TimeFrom_BP_DecalA" "EDRec_DecalA_Offset_Pickaxe" "EDRec_DecalA_Rotation_Pickaxe" "EDRec_FlipBook_FrameA" (-0.35 0.12 0.0 0.12) "EDRec_Emissive_A_Brightness_Time_BP" 0.1)
      (CallFunction|RandomSet 0 "EDRec_TimeFrom_BP_DecalA" "EDRec_DecalA_Offset_Mech" "EDRec_DecalA_Rotation_Mech" "EDRec_FlipBook_FrameA" (-0.05 0.22 0.0 0.05) "EDRec_Emissive_A_Brightness_Time_BP" 0.05)
      (CallFunction|RandomSet 0 "EDRec_TimeFrom_BP_DecalA" "EDRec_DecalA_Offset_Vehicle" "EDRec_DecalA_Rotation_Vehicle" "EDRec_FlipBook_FrameA" (-0.2 -0.12 0.0 0.1) "EDRec_Emissive_A_Brightness_Time_BP" 0.1)
      (CallFunction|RandomSet 0 "EDRec_TimeFrom_BP_DecalA" "EDRec_DecalA_Offset_VehicleSide" "EDRec_DecalA_Rotation_VehicleSide" "EDRec_FlipBook_FrameA" (-0.3 -0.1 0.0 0.06) "EDRec_Emissive_A_Brightness_Time_BP" 0.025))
    (:1
      (CallFunction|RandomSet 0 "EDRec_TimeFrom_BP_DecalB" "EDRec_DecalB_Offset_Weapon" "EDRec_DecalB_Rotation_Weapon" "EDRec_FlipBook_FrameB" (-0.25 0.0 0.0 0.1) "EDRec_Emissive_B_Brightness_Time_BP" 0.025)
      (CallFunction|RandomSet 0 "EDRec_TimeFrom_BP_DecalA" "EDRec_DecalB_Offset_Character" "EDRec_DecalB_Rotation_Character" "EDRec_FlipBook_FrameB" (-0.11 0.2 0.0 0.12) "EDRec_Emissive_B_Brightness_Time_BP" 0.1)
      (CallFunction|RandomSet 0 "EDRec_TimeFrom_BP_DecalA" "EDRec_DecalB_Offset_Backpack----A=Scale" "EDRec_DecalB_Rotation_Backpack" "EDRec_FlipBook_FrameB" (-0.22 0.2 0.0 0.15) "EDRec_Emissive_B_Brightness_Time_BP" 0.04)
      (CallFunction|RandomSet 0 "EDRec_TimeFrom_BP_DecalA" "EDRec_DecalB_Offset_Pickaxe" "EDRec_DecalB_Rotation_Pickaxe" "EDRec_FlipBook_FrameB" (-0.3 -0.13 0.0 0.12) "EDRec_Emissive_B_Brightness_Time_BP" 0.1)
      (CallFunction|RandomSet 0 "EDRec_TimeFrom_BP_DecalA" "EDRec_DecalB_Offset_Mech" "EDRec_DecalB_Rotation_Mech" "EDRec_FlipBook_FrameB" (-0.02 0.25 0.0 0.05) "EDRec_Emissive_B_Brightness_Time_BP" 0.05)
      (CallFunction|RandomSet 0 "EDRec_TimeFrom_BP_DecalA" "EDRec_DecalB_Offset_Vehicle" "EDRec_DecalB_Rotation_Vehicle" "EDRec_FlipBook_FrameB" (-0.3 -0.12 0.0 0.08) "EDRec_Emissive_B_Brightness_Time_BP" 0.1)
      (CallFunction|RandomSet 0 "EDRec_TimeFrom_BP_DecalA" "EDRec_DecalB_Offset_VehicleSide" "EDRec_DecalB_Rotation_VehicleSide" "EDRec_FlipBook_FrameB" (-0.36 -0.13 0.0 0.06) "EDRec_Emissive_B_Brightness_Time_BP" 0.025))
    (:2
      (CallFunction|RandomSet 0 "EDRec_TimeFrom_BP_DecalC" "EDRec_DecalC_Offset_Weapon" "EDRec_DecalC_Rotation_Weapon" "EDRec_FlipBook_FrameC" (-0.1 0.0 0.0 0.1) "EDRec_Emissive_C_Brightness_Time_BP" 0.025)
      (CallFunction|RandomSet 0 "EDRec_TimeFrom_BP_DecalC" "EDRec_DecalC_Offset_Character" "EDRec_DecalC_Rotation_Character" "EDRec_FlipBook_FrameC" (-0.1 -0.2 0.0 0.12) "EDRec_Emissive_C_Brightness_Time_BP" 0.1)
      (CallFunction|RandomSet 0 "EDRec_TimeFrom_BP_DecalC" "EDRec_DecalC_Offset_Backpack----A=Scale" "EDRec_DecalC_Rotation_Backpack" "EDRec_FlipBook_FrameC" (0.18 -0.158 0.0 0.15) "EDRec_Emissive_C_Brightness_Time_BP" 0.1)
      (CallFunction|RandomSet 0 "EDRec_TimeFrom_BP_DecalC" "EDRec_DecalC_Offset_Pickaxe" "EDRec_DecalC_Rotation_Pickaxe" "EDRec_FlipBook_FrameC" (-0.19 0.2 0.0 0.12) "EDRec_Emissive_C_Brightness_Time_BP" 0.1)
      (CallFunction|RandomSet 0 "EDRec_TimeFrom_BP_DecalC" "EDRec_DecalC_Offset_Mech" "EDRec_DecalC_Rotation_Mech" "EDRec_FlipBook_FrameC" (0.1 0.3 0.0 0.05) "EDRec_Emissive_C_Brightness_Time_BP" 0.05)
      (CallFunction|RandomSet 0 "EDRec_TimeFrom_BP_DecalC" "EDRec_DecalC_Offset_Vehicle" "EDRec_DecalC_Rotation_Vehicle" "EDRec_FlipBook_FrameC" (-0.12 0.05 0.0 0.1) "EDRec_Emissive_C_Brightness_Time_BP" 0.1)
      (CallFunction|RandomSet 0 "EDRec_TimeFrom_BP_DecalC" "EDRec_DecalC_Offset_VehicleSide" "EDRec_DecalC_Rotation_VehicleSide" "EDRec_FlipBook_FrameC" (-0.02 -0.12 0.0 0.06) "EDRec_Emissive_C_Brightness_Time_BP" 0.025))
    (:3
      (CallFunction|RandomSet 0 "EDRec_TimeFrom_BP_DecalD" "EDRec_DecalD_Offset_Weapon" "EDRec_DecalD_Rotation_Weapon" "EDRec_FlipBook_FrameD" (0.35 0.0 0.0 0.1) "EDRec_Emissive_D_Brightness_Time_BP" 0.025)
      (CallFunction|RandomSet 0 "EDRec_TimeFrom_BP_DecalD" "EDRec_DecalD_Offset_Character" "EDRec_DecalD_Rotation_Character" "EDRec_FlipBook_FrameD" (0.11 -0.28 0.0 0.12) "EDRec_Emissive_D_Brightness_Time_BP" 0.1)
      (CallFunction|RandomSet 0 "EDRec_TimeFrom_BP_DecalD" "EDRec_DecalD_Offset_Backpack----A=Scale" "EDRec_DecalD_Rotation_Backpack" "EDRec_FlipBook_FrameD" (-0.2 -0.35 0.0 0.15) "EDRec_Emissive_D_Brightness_Time_BP" 0.1)
      (CallFunction|RandomSet 0 "EDRec_TimeFrom_BP_DecalD" "EDRec_DecalD_Offset_Pickaxe" "EDRec_DecalD_Rotation_Pickaxe" "EDRec_FlipBook_FrameD" (0.05 0.05 0.0 0.12) "EDRec_Emissive_D_Brightness_Time_BP" 0.1)
      (CallFunction|RandomSet 0 "EDRec_TimeFrom_BP_DecalD" "EDRec_DecalD_Offset_Mech" "EDRec_DecalD_Rotation_Mech" "EDRec_FlipBook_FrameD" (0.1 0.35 0.0 0.05) "EDRec_Emissive_D_Brightness_Time_BP" 0.05)
      (CallFunction|RandomSet 0 "EDRec_TimeFrom_BP_DecalD" "EDRec_DecalD_Offset_Vehicle" "EDRec_DecalD_Rotation_Vehicle" "EDRec_FlipBook_FrameD" (0.05 -0.075 0.0 0.1) "EDRec_Emissive_D_Brightness_Time_BP" 0.1)
      (CallFunction|RandomSet 0 "EDRec_TimeFrom_BP_DecalD" "EDRec_DecalD_Offset_VehicleSide" "EDRec_DecalD_Rotation_VehicleSide" "EDRec_FlipBook_FrameD" (-0.02 -0.06 0.0 0.05) "EDRec_Emissive_D_Brightness_Time_BP" 0.025))))

(fn RandomSet
  (params (TimeBP :type Name) (Offset :type Name) (Rotation :type Name) (Frame_Name :type Name) (Position :type Linear%20Color%20Structure) (EmissiveReacrive :type Name) (random%20Varation%20Range :type Float%20%28double-precision%29 :default 0.025))
  (bind _returnvalue (CustomItemWrap|GetWrappedMeshComponents))
  (bind _returnvalue_2 (Utilities|Time|GetGameTimeinSeconds))
  (bind (_r _g _b _a) (Math|Color|BreakColor Position))
  (Rendering|Material|SetScalarParameterValueonMaterials _returnvalue Frame_Name (Math|Conversions|ToFloat(Integer) (Math|Random|RandomIntegerinRange 0 3)))
  (Rendering|Material|SetScalarParameterValueonMaterials _returnvalue TimeBP _returnvalue_2)
  (Rendering|Material|SetScalarParameterValueonMaterials _returnvalue EmissiveReacrive _returnvalue_2)
  (Rendering|Material|SetColorParameterValueonMaterials _returnvalue Offset (Math|Color|MakeColor (+ _r (Math|Random|RandomFloatinRange (* random%20Varation%20Range -1.0) random%20Varation%20Range)) (+ _g (Math|Random|RandomFloatinRange (* random%20Varation%20Range -1.0) random%20Varation%20Range)) 0.0 _a))
  (Rendering|Material|SetScalarParameterValueonMaterials _returnvalue Rotation (Math|Random|RandomFloat)))

(fn ConstructionScript)
