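; Cosmetic item wrap Blueprint (CIWM_CyberMitt) that makes the wrap's material react when the wrapped weapon fires.
; On each weapon-fired event, it randomly picks one of three decal time parameters (A, B or C).
; It writes the current game time into that scalar parameter on all wrapped mesh materials, letting the shader time that decal's effect from the timestamp.
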(event EventOnWeaponFired
  (Rendering|Material|SetScalarParameterValueonMaterials (CustomItemWrap|GetWrappedMeshComponents) (Utilities|Select :Index (Math|Random|RandomIntegerinRange 0 2) :Option%200 "EDRec_TimeFrom_BP_DecalA" :Option%201 "EDRec_TimeFrom_BP_DecalB" :Option%202 "EDRec_TimeFrom_BP_DecalC") (Utilities|Time|GetGameTimeinSeconds)))

(fn ConstructionScript)
