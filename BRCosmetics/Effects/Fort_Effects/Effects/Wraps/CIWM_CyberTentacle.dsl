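; CyberTentacle weapon wrap Blueprint that triggers animated emissive effects on the wrapped meshes when the weapon fires.
; A half-second time gate limits retriggering, and each accepted shot decrements an iteration counter to set start and end values for two emissive offsets.
; It then stamps current game time into brightness, offset and blend material parameters, presumably so the material animates the transition from that moment.
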(event EventOnWeaponFired
  (bind _returnvalue (CustomItemWrap|GetWrappedMeshComponents))
  (bind _returnvalue_2 (Utilities|Time|GetGameTimeinSeconds))
  (bind _returnvalue_3 (CustomItemWrap|GetWrappedMeshComponents))
  (bind _iterationcount (Variables|Default|GetIterationCount))
  (bind _returnvalue_4 (Math|Conversions|ToFloat(Integer) _iterationcount))
  (if (>= (Utilities|Time|GetGameTimeinSeconds) (Variables|Default|GetFloat_Value))
    (Rendering|Material|SetScalarParameterValueonMaterials _returnvalue_3 "TopS Emissive_01_Offset---Start" _returnvalue_4)
    (Rendering|Material|SetScalarParameterValueonMaterials _returnvalue_3 "TopS Emissive_02_Offset---Start" _returnvalue_4)
    (bind _output_get (Variables|Default|SetIterationCount (- _iterationcount 3)))
    (Rendering|Material|SetScalarParameterValueonMaterials _returnvalue_3 "TopS Emissive_01_Offset---End" (Math|Conversions|ToFloat(Integer) _output_get))
    (Rendering|Material|SetScalarParameterValueonMaterials _returnvalue_3 "TopS Emissive_02_Offset---End" (Math|Conversions|ToFloat(Integer) _output_get))
    (Utilities|FlowControl|Sequence
      (:then_0
        ;[1 1 1] TopsSide V2
        (Rendering|Material|SetScalarParameterValueonMaterials _returnvalue "TopS Emissive_02-Brightness-TBP" _returnvalue_2)
        (Rendering|Material|SetScalarParameterValueonMaterials _returnvalue "TopS Emissive_02_Offset----TBP" _returnvalue_2)
        (Rendering|Material|SetScalarParameterValueonMaterials _returnvalue "TopS_01_Blend_Opacity-TBP" _returnvalue_2)
        (Rendering|Material|SetScalarParameterValueonMaterials _returnvalue "TopS_01_Blend_Offset-TBP" _returnvalue_2)
        (Rendering|Material|SetScalarParameterValueonMaterials _returnvalue "TimeFrom-BP----------------" _returnvalue_2))
      (:then_1
        (Rendering|Material|SetScalarParameterValueonMaterials _returnvalue "TopS Emissive_01-Brightness-TBP" _returnvalue_2)
        (Rendering|Material|SetScalarParameterValueonMaterials _returnvalue "TopS Emissive_01_Offset----TBP" _returnvalue_2)
        (Rendering|Material|SetScalarParameterValueonMaterials _returnvalue "TopS Emissive_03-Brightness-TBP" _returnvalue_2)
        (Rendering|Material|SetScalarParameterValueonMaterials _returnvalue "TopS Emissive_03_Offset----TBP" _returnvalue_2)
        (Variables|Default|SetFloat_Value (+ (Utilities|Time|GetGameTimeinSeconds) 0.5))))))

(fn ConstructionScript)
