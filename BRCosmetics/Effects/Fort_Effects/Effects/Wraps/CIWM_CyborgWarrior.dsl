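; Reactive item wrap modifier for the Cyborg Warrior cosmetic that shifts wrapped-mesh material parameters as the weapon fires.
; Each accepted shot steps a target value down, and a timeline lerps to it, writing vector parameters while a flag blocks overlapping shots.
; Once the value passes its lower limit, a shot starts a DoOnce-gated delayed Retract that replays the timeline slower toward the upper value.
; The frontend display check and BeginPlay lead to no actions.
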(event EventOnPawnSet
  (if (Reactive|Frontend|IsCosmeticObjectDisplayedinFrontendContext 0)
    _))

(event EventBeginPlay)

(event EventOnWeaponFired
  (if (Variables|Default|GetIsKilling?)
    _
    (elif (>= (Variables|Default|GetCurrentValue) -0.55)
      (Variables|Default|SetNewValue (- (Variables|Default|GetCurrentValue) (/ 1.0 5.0)))
      (Variables|Default|SetIsKilling? true)
      (Components|Timeline|SetPlayRate (Variables|CIWM_CyborgWarrior|GetTimeline) 2.0)
      (bind (_direction _newtrack_0) (Timeline|Timeline PlayFromStart)
        (:Update
          (for _array_element (Variables|Default|GetParameter%20Name)
            (Rendering|Material|SetVectorParameterValueonMaterials (CustomItemWrap|GetWrappedMeshComponents) _array_element (Math|Vector|MakeVector (Math|Float|Lerp (Variables|Default|GetCurrentValue) (Variables|Default|GetNewValue) _newtrack_0)))))
        (:Finished
          (Variables|Default|SetCurrentValue (Variables|Default|GetNewValue))
          (Variables|Default|SetIsKilling? false)))
      (else
        (Utilities|FlowControl|DoOnce false
          (:Completed
            (Utilities|Time|SetTimerbyEvent (delegate Custom|Retract) 8.0)))))))

(event Custom|ResetCounter
  (Utilities|FlowControl|DoOnce false
    (:Completed
      (Utilities|Time|SetTimerbyEvent (delegate Custom|Retract) 8.0))))

(event Custom|Retract
  (Components|Timeline|SetPlayRate (Variables|CIWM_CyborgWarrior|GetTimeline) 0.3)
  (Variables|Default|SetNewValue 0.55)
  (Utilities|FlowControl|Sequence
    (:then_0
      (Timeline|Timeline PlayFromStart))
    (:then_1
      (CallFunction|ResetCounter))))

(fn ConstructionScript)
