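; Reactive weapon wrap modifier that animates wrap materials as the player deals damage and scores kills.
; Pawn setup watches both stats: damage events step a ten-level counter that lerps emissive gradient endpoints, while kills play a longer full sweep.
; Game-time stamps drive material animations, with Kill and Pause flags gating retriggers and a reset clearing counters and retiming decals.
; In the front end, a weapon-fire counter simulates damage and kill reactions.
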(event EventOnWeaponFired
  (bind _returnvalue (CustomItemWrap|GetWrappedMeshComponents))
  (bind _returnvalue_2 (Utilities|Time|GetGameTimeinSeconds))
  (bind _returnvalue_3 (CustomItemWrap|GetWrappedMeshComponents))
  (bind _returnvalue_4 (CustomItemWrap|GetWrappedMeshComponents))
  (bind _returnvalue_5 (CustomItemWrap|GetWrappedMeshComponents))
  (bind _returnvalue_6 (CustomItemWrap|GetWrappedMeshComponents))
  (bind _returnvalue_7 (Utilities|Time|GetGameTimeinSeconds))
  ; Frontnd reactive
  (if (CustomCosmeticModifier|IsinFrontEnd)
    (bind _output_get (Variables|Default|Setweapon%20fire (+ (Variables|Default|Getweapon%20fire) 1)))
    (if (>= _output_get 5)
      (if (== (Variables|Default|Getweapon%20fire) 7)
        (Variables|Default|Setweapon%20fire 0)
        (if (not (Variables|Default|GetKill))
          (Utilities|FlowControl|Sequence
            (:then_0
              (CallFunction|Glows))
            (:then_1
              (Variables|Default|SetKill true)
              (Rendering|Material|SetScalarParameterValueonMaterials _returnvalue_3 "EDRec_Emissive_A_Gradent_Duration" 1.5)
              (Rendering|Material|SetScalarParameterValueonMaterials _returnvalue_3 "EDRec_Emissive_B_Gradent_Duration" 1.5)
              (Rendering|Material|SetScalarParameterValueonMaterials _returnvalue_3 "EDRec_Emissive_C_Gradent_Duration" 1.5)
              (Rendering|Material|SetScalarParameterValueonMaterials _returnvalue_4 "EDRec_Emissive_B_Gradent_Time_End" -0.4)
              (Rendering|Material|SetScalarParameterValueonMaterials _returnvalue_4 "EDRec_Emissive_B_Gradent_Time_Start" (Math|Float|Lerp -0.05 -0.4 (/ (Math|Conversions|ToFloat(Integer) (Variables|Default|GetDamage_Last)) 1e+01)))
              (Rendering|Material|SetScalarParameterValueonMaterials _returnvalue_4 "EDRec_Emissive_A_Gradent_Delay" 0.0)
              (Rendering|Material|SetScalarParameterValueonMaterials _returnvalue_4 "EDRec_Emissive_B_Gradent_Time_BP" (Utilities|Time|GetGameTimeinSeconds))
              (Rendering|Material|SetScalarParameterValueonMaterials _returnvalue_4 "EDRec_Emissive_C_Gradent_Time_End" 0.4)
              (Rendering|Material|SetScalarParameterValueonMaterials _returnvalue_4 "EDRec_Emissive_C_Gradent_Time_Start" (Math|Float|Lerp 0.76 0.4 (/ (Math|Conversions|ToFloat(Integer) (Variables|Default|GetDamage_Last)) 1e+01)))
              (Rendering|Material|SetScalarParameterValueonMaterials _returnvalue_4 "EDRec_Emissive_C_Gradent_Time_BP" (Utilities|Time|GetGameTimeinSeconds))
              (Rendering|Material|SetScalarParameterValueonMaterials _returnvalue_4 "EDRec_Emissive_A_Gradent_Delay" 0.1)
              (Rendering|Material|SetScalarParameterValueonMaterials _returnvalue_4 "EDRec_Emissive_A_Gradent_Time_End" -0.81)
              (Rendering|Material|SetScalarParameterValueonMaterials _returnvalue_4 "EDRec_Emissive_A_Gradent_Time_Start" 0.41)
              (Rendering|Material|SetScalarParameterValueonMaterials _returnvalue_4 "EDRec_Emissive_A_Gradent_Time_BP" (Utilities|Time|GetGameTimeinSeconds))
              (Utilities|FlowControl|Delay 1.5))))
        (elif (Variables|Default|GetKill)
          _
          (else
            (Utilities|FlowControl|Delay 0.05)
            (Utilities|FlowControl|Sequence
              (:then_0
                (CallFunction|Glows))
              (:then_1
                (if (not (Variables|Default|GetPause))
                  (Variables|Default|SetDamage (Math|Integer|Clamp(Integer) (+ (Variables|Default|GetDamage) 1) 0 10))
                  (Variables|Default|SetPause true)
                  (Rendering|Material|SetScalarParameterValueonMaterials _returnvalue_5 "EDRec_Emissive_A_Gradent_Duration" 0.5)
                  (Rendering|Material|SetScalarParameterValueonMaterials _returnvalue_5 "EDRec_Emissive_B_Gradent_Duration" 0.5)
                  (Rendering|Material|SetScalarParameterValueonMaterials _returnvalue_5 "EDRec_Emissive_C_Gradent_Duration" 0.5)
                  (Rendering|Material|SetScalarParameterValueonMaterials _returnvalue_6 "EDRec_Emissive_B_Gradent_Time_End" (Math|Float|Lerp -0.05 -0.4 (/ (Math|Conversions|ToFloat(Integer) (Variables|Default|GetDamage)) 1e+01)))
                  (Rendering|Material|SetScalarParameterValueonMaterials _returnvalue_6 "EDRec_Emissive_B_Gradent_Time_Start" (Math|Float|Lerp -0.05 -0.4 (/ (Math|Conversions|ToFloat(Integer) (Variables|Default|GetDamage_Last)) 1e+01)))
                  (Rendering|Material|SetScalarParameterValueonMaterials _returnvalue_6 "EDRec_Emissive_A_Gradent_Delay" 0.0)
                  (Rendering|Material|SetScalarParameterValueonMaterials _returnvalue_6 "EDRec_Emissive_B_Gradent_Time_BP" (Utilities|Time|GetGameTimeinSeconds))
                  (Rendering|Material|SetScalarParameterValueonMaterials _returnvalue_6 "EDRec_Emissive_C_Gradent_Time_End" (Math|Float|Lerp 0.76 0.4 (/ (Math|Conversions|ToFloat(Integer) (Variables|Default|GetDamage)) 1e+01)))
                  (Rendering|Material|SetScalarParameterValueonMaterials _returnvalue_6 "EDRec_Emissive_C_Gradent_Time_Start" (Math|Float|Lerp 0.76 0.4 (/ (Math|Conversions|ToFloat(Integer) (Variables|Default|GetDamage_Last)) 1e+01)))
                  (Rendering|Material|SetScalarParameterValueonMaterials _returnvalue_6 "EDRec_Emissive_C_Gradent_Time_BP" (Utilities|Time|GetGameTimeinSeconds))
                  (Rendering|Material|SetScalarParameterValueonMaterials _returnvalue_6 "EDRec_Emissive_A_Gradent_Delay" 0.1)
                  (Rendering|Material|SetScalarParameterValueonMaterials _returnvalue_6 "EDRec_Emissive_A_Gradent_Time_End" -0.81)
                  (Rendering|Material|SetScalarParameterValueonMaterials _returnvalue_6 "EDRec_Emissive_A_Gradent_Time_Start" 0.41)
                  (Rendering|Material|SetScalarParameterValueonMaterials _returnvalue_6 "EDRec_Emissive_A_Gradent_Time_BP" (Utilities|Time|GetGameTimeinSeconds))
                  (Variables|Default|SetDamage_Last (+ (Variables|Default|GetDamage_Last) 1))
                  (Utilities|FlowControl|Delay 0.5)
                  (if (not (== (Variables|Default|GetDamage) 10))
                    (Variables|Default|SetPause false))))))))
      (Variables|Default|SetPause true)
      (Utilities|FlowControl|Delay 0.2)
      (CallFunction|Glows)
      (Variables|Default|SetDamage 0)
      (Variables|Default|SetDamage_Last 0)
      (Rendering|Material|SetScalarParameterValueonMaterials _returnvalue "EDRec_Emissive_B_Gradent_Time_End" 0.4)
      (Rendering|Material|SetScalarParameterValueonMaterials _returnvalue "EDRec_Emissive_C_Gradent_Time_End" -0.4)
      (Rendering|Material|SetScalarParameterValueonMaterials _returnvalue "EDRec_TimeFrom_BP_DecalA" _returnvalue_7)
      (Rendering|Material|SetScalarParameterValueonMaterials _returnvalue "EDRec_TimeFrom_BP_DecalB" _returnvalue_7)
      (Rendering|Material|SetScalarParameterValueonMaterials _returnvalue "EDRec_TimeFrom_BP_DecalC" _returnvalue_7)
      (Rendering|Material|SetScalarParameterValueonMaterials _returnvalue "EDRec_Emissive_D_Brightness_Time_BP" _returnvalue_2)
      (Rendering|Material|SetScalarParameterValueonMaterials _returnvalue "EDRec_TimeFrom_BP_DecalD" _returnvalue_2)
      (Utilities|FlowControl|Delay 1.45)
      (Variables|Default|SetPause false)
      (Variables|Default|SetKill false))))

(event Custom|Glows
  (bind _returnvalue (Utilities|Time|GetGameTimeinSeconds))
  (bind _returnvalue_2 (CustomItemWrap|GetWrappedMeshComponents))
  (Rendering|Material|SetScalarParameterValueonMaterials _returnvalue_2 "EDRec_Emissive_B_Brightness_Time_BP" _returnvalue)
  (Rendering|Material|SetScalarParameterValueonMaterials _returnvalue_2 "EDRec_Emissive_C_Brightness_Time_BP" _returnvalue))

(event EventOnPawnSet
  (bind _returnvalue (CustomItemWrap|GetWrappedMeshComponents))
  (bind _returnvalue_2 (Utilities|Time|GetGameTimeinSeconds))
  (bind _returnvalue_3 (CustomItemWrap|GetWrappedMeshComponents))
  (bind _returnvalue_4 (CustomItemWrap|GetWrappedMeshComponents))
  (bind _returnvalue_5 (CustomItemWrap|GetWrappedMeshComponents))
  (bind _returnvalue_6 (CustomItemWrap|GetWrappedMeshComponents))
  (bind _returnvalue_7 (Utilities|Time|GetGameTimeinSeconds))
  (Utilities|FlowControl|Sequence
    (:then_0
      (Cosmetics|WatchCosmeticStat (CustomCharacter|GetAssociatedPlayerPawn) "(TagName=\"StatManager.AthenaKills\")" 0 false
        (:OnStatChanged
          (if (not (Variables|Default|GetKill))
            (Utilities|FlowControl|Sequence
              (:then_0
                (CallFunction|Glows))
              (:then_1
                (Variables|Default|SetKill true)
                (Rendering|Material|SetScalarParameterValueonMaterials _returnvalue_3 "EDRec_Emissive_A_Gradent_Duration" 1.5)
                (Rendering|Material|SetScalarParameterValueonMaterials _returnvalue_3 "EDRec_Emissive_B_Gradent_Duration" 1.5)
                (Rendering|Material|SetScalarParameterValueonMaterials _returnvalue_3 "EDRec_Emissive_C_Gradent_Duration" 1.5)
                (Rendering|Material|SetScalarParameterValueonMaterials _returnvalue_4 "EDRec_Emissive_B_Gradent_Time_End" -0.4)
                (Rendering|Material|SetScalarParameterValueonMaterials _returnvalue_4 "EDRec_Emissive_B_Gradent_Time_Start" (Math|Float|Lerp -0.05 -0.4 (/ (Math|Conversions|ToFloat(Integer) (Variables|Default|GetDamage_Last)) 1e+01)))
                (Rendering|Material|SetScalarParameterValueonMaterials _returnvalue_4 "EDRec_Emissive_A_Gradent_Delay" 0.0)
                (Rendering|Material|SetScalarParameterValueonMaterials _returnvalue_4 "EDRec_Emissive_B_Gradent_Time_BP" (Utilities|Time|GetGameTimeinSeconds))
                (Rendering|Material|SetScalarParameterValueonMaterials _returnvalue_4 "EDRec_Emissive_C_Gradent_Time_End" 0.4)
                (Rendering|Material|SetScalarParameterValueonMaterials _returnvalue_4 "EDRec_Emissive_C_Gradent_Time_Start" (Math|Float|Lerp 0.76 0.4 (/ (Math|Conversions|ToFloat(Integer) (Variables|Default|GetDamage_Last)) 1e+01)))
                (Rendering|Material|SetScalarParameterValueonMaterials _returnvalue_4 "EDRec_Emissive_C_Gradent_Time_BP" (Utilities|Time|GetGameTimeinSeconds))
                (Rendering|Material|SetScalarParameterValueonMaterials _returnvalue_4 "EDRec_Emissive_A_Gradent_Delay" 0.1)
                (Rendering|Material|SetScalarParameterValueonMaterials _returnvalue_4 "EDRec_Emissive_A_Gradent_Time_End" -0.81)
                (Rendering|Material|SetScalarParameterValueonMaterials _returnvalue_4 "EDRec_Emissive_A_Gradent_Time_Start" 0.41)
                (Rendering|Material|SetScalarParameterValueonMaterials _returnvalue_4 "EDRec_Emissive_A_Gradent_Time_BP" (Utilities|Time|GetGameTimeinSeconds))
                (Utilities|FlowControl|Delay 1.5)))))))
    (:then_1
      (Cosmetics|WatchCosmeticStat (CustomCharacter|GetAssociatedPlayerPawn) "(TagName=\"StatManager.DamageDealt\")" 0 false
        (:OnStatChanged
          (if (not (Variables|Default|GetKill))
            (Utilities|FlowControl|Delay 0.05)
            (Utilities|FlowControl|Sequence
              (:then_0
                (CallFunction|Glows))
              (:then_1
                (if (not (Variables|Default|GetPause))
                  (Variables|Default|SetDamage (Math|Integer|Clamp(Integer) (+ (Variables|Default|GetDamage) 1) 0 10))
                  (Variables|Default|SetPause true)
                  (Rendering|Material|SetScalarParameterValueonMaterials _returnvalue_5 "EDRec_Emissive_A_Gradent_Duration" 0.5)
                  (Rendering|Material|SetScalarParameterValueonMaterials _returnvalue_5 "EDRec_Emissive_B_Gradent_Duration" 0.5)
                  (Rendering|Material|SetScalarParameterValueonMaterials _returnvalue_5 "EDRec_Emissive_C_Gradent_Duration" 0.5)
                  (Rendering|Material|SetScalarParameterValueonMaterials _returnvalue_6 "EDRec_Emissive_B_Gradent_Time_End" (Math|Float|Lerp -0.05 -0.4 (/ (Math|Conversions|ToFloat(Integer) (Variables|Default|GetDamage)) 1e+01)))
                  (Rendering|Material|SetScalarParameterValueonMaterials _returnvalue_6 "EDRec_Emissive_B_Gradent_Time_Start" (Math|Float|Lerp -0.05 -0.4 (/ (Math|Conversions|ToFloat(Integer) (Variables|Default|GetDamage_Last)) 1e+01)))
                  (Rendering|Material|SetScalarParameterValueonMaterials _returnvalue_6 "EDRec_Emissive_A_Gradent_Delay" 0.0)
                  (Rendering|Material|SetScalarParameterValueonMaterials _returnvalue_6 "EDRec_Emissive_B_Gradent_Time_BP" (Utilities|Time|GetGameTimeinSeconds))
                  (Rendering|Material|SetScalarParameterValueonMaterials _returnvalue_6 "EDRec_Emissive_C_Gradent_Time_End" (Math|Float|Lerp 0.76 0.4 (/ (Math|Conversions|ToFloat(Integer) (Variables|Default|GetDamage)) 1e+01)))
                  (Rendering|Material|SetScalarParameterValueonMaterials _returnvalue_6 "EDRec_Emissive_C_Gradent_Time_Start" (Math|Float|Lerp 0.76 0.4 (/ (Math|Conversions|ToFloat(Integer) (Variables|Default|GetDamage_Last)) 1e+01)))
                  (Rendering|Material|SetScalarParameterValueonMaterials _returnvalue_6 "EDRec_Emissive_C_Gradent_Time_BP" (Utilities|Time|GetGameTimeinSeconds))
                  (Rendering|Material|SetScalarParameterValueonMaterials _returnvalue_6 "EDRec_Emissive_A_Gradent_Delay" 0.1)
                  (Rendering|Material|SetScalarParameterValueonMaterials _returnvalue_6 "EDRec_Emissive_A_Gradent_Time_End" -0.81)
                  (Rendering|Material|SetScalarParameterValueonMaterials _returnvalue_6 "EDRec_Emissive_A_Gradent_Time_Start" 0.41)
                  (Rendering|Material|SetScalarParameterValueonMaterials _returnvalue_6 "EDRec_Emissive_A_Gradent_Time_BP" (Utilities|Time|GetGameTimeinSeconds))
                  (Variables|Default|SetDamage_Last (+ (Variables|Default|GetDamage_Last) 1))
                  (Utilities|FlowControl|Delay 0.5)
                  (if (not (== (Variables|Default|GetDamage) 10))
                    (Variables|Default|SetPause false))))))))))
  (Variables|Default|SetPause true)
  (Utilities|FlowControl|Delay 0.2)
  (CallFunction|Glows)
  (Variables|Default|SetDamage 0)
  (Variables|Default|SetDamage_Last 0)
  (Rendering|Material|SetScalarParameterValueonMaterials _returnvalue "EDRec_Emissive_B_Gradent_Time_End" 0.4)
  (Rendering|Material|SetScalarParameterValueonMaterials _returnvalue "EDRec_Emissive_C_Gradent_Time_End" -0.4)
  (Rendering|Material|SetScalarParameterValueonMaterials _returnvalue "EDRec_TimeFrom_BP_DecalA" _returnvalue_7)
  (Rendering|Material|SetScalarParameterValueonMaterials _returnvalue "EDRec_TimeFrom_BP_DecalB" _returnvalue_7)
  (Rendering|Material|SetScalarParameterValueonMaterials _returnvalue "EDRec_TimeFrom_BP_DecalC" _returnvalue_7)
  (Rendering|Material|SetScalarParameterValueonMaterials _returnvalue "EDRec_Emissive_D_Brightness_Time_BP" _returnvalue_2)
  (Rendering|Material|SetScalarParameterValueonMaterials _returnvalue "EDRec_TimeFrom_BP_DecalD" _returnvalue_2)
  (Utilities|FlowControl|Delay 1.45)
  (Variables|Default|SetPause false)
  (Variables|Default|SetKill false))

(fn ConstructionScript)
