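; Custom item wrap modifier that makes wrapped weapon materials react to the player's DamageDealt stat.
; On pawn set it watches that stat; each change stamps game time into emissive brightness parameters and applies a color-curve offset, then advances that offset by Step Amount.
; In the front end, where weapon fire substitutes for damage, a shot counter runs the same update on shots five and six and resets at seven.
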(event EventOnWeaponFired
  (bind _returnvalue (Utilities|Time|GetGameTimeinSeconds))
  (bind _returnvalue_2 (CustomItemWrap|GetWrappedMeshComponents))
  (bind _float_value (Variables|Default|GetFloat_Value))
  (if (CustomCosmeticModifier|IsinFrontEnd)
    (bind _output_get (Variables|Envirometn_Varable|SetIterationCount (+ (Variables|Envirometn_Varable|GetIterationCount) 1)))
    (if (>= _output_get 5)
      (if (== (Variables|Envirometn_Varable|GetIterationCount) 7)
        (Variables|Envirometn_Varable|SetIterationCount 0)
        (else
          (Utilities|FlowControl|Sequence
            (:then_0
              (Utilities|FlowControl|Sequence
                (:then_0
                  (Rendering|Material|SetScalarParameterValueonMaterials _returnvalue_2 "TopS Emissive_01-Brightness-TBP" _returnvalue)
                  ;[1 1 1] TopsSide V2
                  (Rendering|Material|SetScalarParameterValueonMaterials _returnvalue_2 "EDRec_Emissive_A_Brightness_Time_BP" _returnvalue)
                  (Rendering|Material|SetScalarParameterValueonMaterials _returnvalue_2 "EDRec_Emissive_B_Brightness_Time_BP" _returnvalue)
                  (Rendering|Material|SetScalarParameterValueonMaterials _returnvalue_2 "EDRec_Emissive_B_CC_Offset" _float_value)
                  (Rendering|Material|SetScalarParameterValueonMaterials _returnvalue_2 "Eff_Galaxy_CC_Offset" _float_value)
                  (Rendering|Material|SetScalarParameterValueonMaterials _returnvalue_2 "TopS Emissive_01_ColorCuve_Offset" _float_value))))
            (:then_1
              (Variables|Default|SetFloat_Value (+ (Variables|Default|GetFloat_Value) (Variables|Default|GetStep%20Amount))))))))))

(event EventOnPawnSet
  (bind _returnvalue (Utilities|Time|GetGameTimeinSeconds))
  (bind _returnvalue_2 (CustomItemWrap|GetWrappedMeshComponents))
  (bind _float_value (Variables|Default|GetFloat_Value))
  (Cosmetics|WatchCosmeticStat (CustomCharacter|GetAssociatedPlayerPawn) "(TagName=\"StatManager.DamageDealt\")"
    (:OnStatChanged
      (Utilities|FlowControl|Sequence
        (:then_0
          (Utilities|FlowControl|Sequence
            (:then_0
              (Rendering|Material|SetScalarParameterValueonMaterials _returnvalue_2 "TopS Emissive_01-Brightness-TBP" _returnvalue)
              ;[1 1 1] TopsSide V2
              (Rendering|Material|SetScalarParameterValueonMaterials _returnvalue_2 "EDRec_Emissive_A_Brightness_Time_BP" _returnvalue)
              (Rendering|Material|SetScalarParameterValueonMaterials _returnvalue_2 "EDRec_Emissive_B_Brightness_Time_BP" _returnvalue)
              (Rendering|Material|SetScalarParameterValueonMaterials _returnvalue_2 "EDRec_Emissive_B_CC_Offset" _float_value)
              (Rendering|Material|SetScalarParameterValueonMaterials _returnvalue_2 "Eff_Galaxy_CC_Offset" _float_value)
              (Rendering|Material|SetScalarParameterValueonMaterials _returnvalue_2 "TopS Emissive_01_ColorCuve_Offset" _float_value))))
        (:then_1
          (Variables|Default|SetFloat_Value (+ (Variables|Default|GetFloat_Value) (Variables|Default|GetStep%20Amount))))))))

(fn ConstructionScript)
