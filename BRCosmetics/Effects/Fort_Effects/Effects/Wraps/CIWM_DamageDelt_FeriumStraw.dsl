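; FeriumStraw reactive wrap modifier that adds colored animated decals to wrapped meshes as the player deals damage.
; A DamageDealt stat watch, or weapon firing in the front end, cycles Set C, B and A, each drawing from a shuffled color list.
; Each layer randomizes decal placement per item type, then timelines animate the decal and blend its color into base and emissive tints.
; A matching character skin also tints the wrap from its body material's MutedColor.
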(event EventOnSetWrapTarget
  (if (Utilities|String|Equal,CaseInsensitive(String) (ItemDefinition|GetTemplateId (CosmeticLoadout|GetItemDefinitionfromLoadout (CosmeticLoadout|GetCosmeticLoadout) "(TagName=\"LoadoutSlot.Character\")")) (Variables|Default|GetCharacter-Name))
    (bind _asmaterial_instance_dynamic (Utilities|Casting|CastToMaterialInstanceDynamic (Rendering|Material|GetMaterial (Customization|GetSkeletalMeshforPartType (CustomCharacter|GetAssociatedPlayerPawn) "Body" :!class /Script/FortniteGame.FortPlayerPawn) 0)))
    (bind _output_get (Variables|Default|SetCharacter%20Material _asmaterial_instance_dynamic))
    (bind _returnvalue (Rendering|Material|GetVectorParameterValue _output_get "MutedColor" :!class /Script/Engine.MaterialInstanceDynamic))
    (for _array_element (CustomItemWrap|GetWrappedMeshComponents)
      (for _array_element_2 (Rendering|Material|GetMaterials _array_element)
        (bind _asmaterial_instance_dynamic_2 (Utilities|Casting|CastToMaterialInstanceDynamic _array_element_2))
        (Rendering|Material|SetVectorParameterValue _asmaterial_instance_dynamic_2 "TopS_01_ColorTint_CH_Red  A=0=off" (* "(R=0.350000,G=0.350000,B=0.350000,A=1.000000)" _returnvalue) :!class /Script/Engine.MaterialInstanceDynamic)))))

(event EventOnCharacterCustomizationFinished
  (params (CharacterPartInfo :type Array%20of%20Applied%20Character%20Part%20Info%20Structures))
  (if (Utilities|String|Equal,CaseInsensitive(String) (ItemDefinition|GetTemplateId (CosmeticLoadout|GetItemDefinitionfromLoadout (CosmeticLoadout|GetCosmeticLoadout) "(TagName=\"LoadoutSlot.Character\")")) (Variables|Default|GetCharacter-Name))
    (bind _asmaterial_instance_dynamic (Utilities|Casting|CastToMaterialInstanceDynamic (Rendering|Material|GetMaterial (Customization|GetSkeletalMeshforPartType (CustomCharacter|GetAssociatedPlayerPawn) "Body" :!class /Script/FortniteGame.FortPlayerPawn) 0)))
    (bind _output_get (Variables|Default|SetCharacter%20Material _asmaterial_instance_dynamic))
    (bind _returnvalue (Rendering|Material|GetVectorParameterValue _output_get "MutedColor" :!class /Script/Engine.MaterialInstanceDynamic))
    (for _array_element (CustomItemWrap|GetWrappedMeshComponents)
      (for _array_element_2 (Rendering|Material|GetMaterials _array_element)
        (bind _asmaterial_instance_dynamic_2 (Utilities|Casting|CastToMaterialInstanceDynamic _array_element_2))
        (Rendering|Material|SetVectorParameterValue _asmaterial_instance_dynamic_2 "TopS_01_ColorTint_CH_Red  A=0=off" (* "(R=0.350000,G=0.350000,B=0.350000,A=1.000000)" _returnvalue) :!class /Script/Engine.MaterialInstanceDynamic)))))

(event Custom|ColorSet)

(event EventOnWeaponFired
  (if (CustomCosmeticModifier|IsinFrontEnd)
    (bind _output_get (Variables|Default|SetIterationCount (+ (Variables|Default|GetIterationCount) 1)))
    (if (>= _output_get 5)
      (if (== (Variables|Default|GetIterationCount) 7)
        (Variables|Default|SetIterationCount 0)
        (else
          (CallFunction|Activate))))))

(event Custom|ResetColors
  (if (>= (Variables|Default|GetColor%20seq) (Variables|Default|GetColor_Count))
    (Utilities|FlowControl|Delay 0.2)
    (Utilities|Array|Shuffle (Variables|Default|GetColor%20Range))
    (Variables|Default|SetColor%20seq 0)
    (Development|PrintText "NSLOCTEXT(\"[0C1245E77EA0116B716A6466FC60E58C]\", \"93BBB98D4FD8802A19080B9B3E494498\", \"reset\")" true true "(R=1.000000,G=0.057285,B=0.000000,A=1.000000)")))

(event Custom|Set%20A
  (bind _returnvalue (CustomItemWrap|GetWrappedMeshComponents))
  (bind _returnvalue_2 (CustomItemWrap|GetWrappedMeshComponents))
  (bind _returnvalue_3 (CustomItemWrap|GetWrappedMeshComponents))
  (bind _returnvalue_4 (CustomItemWrap|GetWrappedMeshComponents))
  (bind _color_a (Variables|Default|GetColor_A))
  (bind _returnvalue_5 (* _color_a "(R=1.500000,G=1.500000,B=1.500000,A=1.000000)"))
  (bind _color (Variables|Default|GetColor))
  (bind _color_a_2 (Variables|Default|GetColor_A))
  (Variables|Default|SetFire%20seq 0)
  (if (not (Variables|Default|GetPause_A))
    (CallFunction|ColorSet)
    (Variables|Default|SetPause_A true)
    (Rendering|Material|SetColorParameterValueonMaterials _returnvalue_2 "EDRec_Emissive_A_CC Tint" _color)
    (Variables|Default|SetColor_A _color)
    (Rendering|Material|SetScalarParameterValueonMaterials _returnvalue_2 "EDRec_Emissive_A_Brightness_DarkenBase_Time_End" 1.0)
    (Rendering|Material|SetScalarParameterValueonMaterials _returnvalue_2 "EDRec_Emissive_A_Brightness_Time__End" 1.0)
    (Rendering|Material|SetColorParameterValueonMaterials _returnvalue_2 "EDRec_Emissive_A_Gradent_Offset" (Utilities|Operators|Add :A (Math|Color|MakeColor (Math|Random|RandomFloatinRange 0.0 1.0) (Math|Random|RandomFloatinRange 0.0 1.0))))
    (Rendering|Material|SetColorParameterValueonMaterials _returnvalue_2 "EDRec_Emissive_A_Gradent_Scale" (Utilities|Operators|Add :A (Math|Color|MakeColor 1.0 1.0 1.0 (Math|Random|RandomFloatinRange 0.75 1.5))))
    (Utilities|FlowControl|Sequence
      (:then_0
        (bind (_direction _scale _splat _newtrack_3) (Timeline|Timeline PlayFromStart)
          (:Update
            (bind (_newtrack_3_x _newtrack_3_y) (Math|Vector|BreakVector _newtrack_3))
            (Rendering|Material|SetColorParameterValueonMaterials _returnvalue_3 "EDRec_Emissive_A_Gradent_MSk_Channel" (Math|Color|Lerp(LinearColor) "(R=0.000000,G=0.000000,B=0.000000,A=1.000000)" "(R=1.000000,G=0.000000,B=0.000000,A=0.000000)" _newtrack_3_y))
            (Rendering|Material|SetColorParameterValueonMaterials _returnvalue_3 "EDRec_DecalA_Scale" (Math|Color|Lerp(LinearColor) :B "(R=1.000000,G=1.000000,B=0.000000,A=1.000000)" :Alpha _newtrack_3_x)))
          (:Finished
            (Rendering|Material|SetScalarParameterValueonMaterials _returnvalue "EDRec_Emissive_A_Brightness_DarkenBase_Time_End" 0.0)
            (Rendering|Material|SetScalarParameterValueonMaterials _returnvalue "EDRec_Emissive_A_Brightness_Time__End" 0.0)
            (Rendering|Material|SetScalarParameterValueonMaterials _returnvalue "EDRec_Emissive_A_Brightness_Time_BP" (Utilities|Time|GetGameTimeinSeconds))
            (Utilities|FlowControl|Sequence
              (:then_0
                (bind (_direction_2 _scale_2 _splat_2 _newtrack_3_2) (Timeline|Timeline_2 PlayFromStart)
                  (:Update
                    (bind (_newtrack_3_x_2 _newtrack_3_y_2 _newtrack_3_z) (Math|Vector|BreakVector _newtrack_3_2))
                    (Rendering|Material|SetColorParameterValueonMaterials _returnvalue_4 "TopS_01_ColorTint_CH_Blue  A=0=off" (Math|Color|Lerp(LinearColor) (Variables|Default|GetColor_A_Last) _returnvalue_5 _newtrack_3_z))
                    (Rendering|Material|SetColorParameterValueonMaterials _returnvalue_4 "TopS Emissive_01_ColorCuve Tint" (* (Math|Color|Lerp(LinearColor) (Variables|Default|GetColor_A_Last_0) _color_a_2 _newtrack_3_z) "(R=0.450000,G=0.450000,B=0.450000,A=1.000000)")))
                  (:Finished
                    (Variables|Default|SetColor_A_Last _returnvalue_5)
                    (Variables|Default|SetColor_A_Last_0 _color_a_2))
                  (:NewTrack_0
                    (Rendering|Material|SetScalarParameterValueonMaterials _returnvalue_4 "TopS Emissive_01-Darken_Base    1=On" 1.0))))
              (:then_1
                (Utilities|FlowControl|Delay 0.15)
                (Variables|Default|SetPause_A false))))))
      (:then_1
        ;[1 1 1] multi A
        (CallFunction|RandomSet "EDRec_Decal_A_Animation_Time_BP" "EDRec_DecalA_Offset_Weapon" "EDRec_DecalA_Rotation_Weapon" -0.38 0.2 -0.01 0.01 0.5 0.8)
        (CallFunction|RandomSet "0" "EDRec_DecalA_Offset_Pickaxe" "EDRec_DecalA_Rotation_Pickaxe" -0.3 0.15 0.2 -0.2 0.4 0.7)
        (CallFunction|RandomSet "0" "EDRec_DecalA_Offset_Backpack----A=Scale" "EDRec_DecalA_Rotation_Backpack" -0.15 0.15 -0.25 0.25 0.5 1.0)
        (CallFunction|RandomSet "0" "EDRec_DecalA_Offset_Character" "EDRec_DecalA_Rotation_Character" -0.15 0.15 -0.2 0.28 0.5 1.0)
        (CallFunction|RandomSet "0" "EDRec_DecalA_Offset_Vehicle" "EDRec_DecalA_Rotation_Vehicle" 0.15 0.4 -0.15 0.15 0.5 0.6)
        (CallFunction|RandomSet "0" "EDRec_DecalA_Offset_VehicleSide" "EDRec_DecalA_Rotation_Vehicle_Side" 0.3 0.42 -0.1 -0.13 0.3 0.4)
        (CallFunction|RandomSet "0" "EDRec_DecalA_Offset_Mech" "EDRec_DecalA_Rotation_Mech" -0.1 0.1 0.32 0.2 0.3 0.4)))))

(event EventBeginPlay)

(event EventOnPawnSet
  (Utilities|FlowControl|Sequence
    (:then_0
      (Utilities|FlowControl|Delay 0.2)
      (Utilities|Array|Shuffle (Variables|Default|GetAray))
      (Utilities|Array|Shuffle (Variables|Default|GetColor%20Range)))
    (:then_1
      (Cosmetics|WatchCosmeticStat (CustomCharacter|GetAssociatedPlayerPawn) "(TagName=\"StatManager.DamageDealt\")" 0 false
        (:OnStatChanged
          (switch int (Variables|Default|GetFire%20seq)
            (:0
              (CallFunction|SetC))
            (:1
              (CallFunction|SetB))
            (:2
              (CallFunction|SetA)))
          (CallFunction|ResetColors))))
    (:then_2
      (CallFunction|ColorSet))))

(event Custom|Color%20set
  (bind _output_get (Variables|Default|SetColor%20seq (+ (Variables|Default|GetColor%20seq) 1)))
  (bind _output_get_2 (Variables|Default|SetColor (Utilities|Array|Get(acopy) (Variables|Default|GetColor%20Range) _output_get)))
  (Development|PrintText (Utilities|Text|ToText(Integer) (Variables|Default|GetColor%20seq)) true true _output_get_2 5.0))

(event Custom|Activate
  (switch int (Variables|Default|GetFire%20seq)
    (:0
      (CallFunction|SetC))
    (:1
      (CallFunction|SetB))
    (:2
      (CallFunction|SetA)))
  (CallFunction|ResetColors))

(event Custom|Set%20B
  (bind _returnvalue (CustomItemWrap|GetWrappedMeshComponents))
  (bind _returnvalue_2 (CustomItemWrap|GetWrappedMeshComponents))
  (bind _returnvalue_3 (CustomItemWrap|GetWrappedMeshComponents))
  (bind _returnvalue_4 (CustomItemWrap|GetWrappedMeshComponents))
  (bind _color_b_last (Variables|Default|GetColor_B_Last))
  (bind _color (Variables|Default|GetColor))
  (bind _color_b (Variables|Default|GetColor_B))
  (bind _color_b_2 (Variables|Default|GetColor_B))
  (Variables|Default|SetFire%20seq 2)
  (if (not (Variables|Default|GetPause_B))
    (CallFunction|ColorSet)
    (Variables|Default|SetPause_B true)
    (Rendering|Material|SetColorParameterValueonMaterials _returnvalue "EDRec_Emissive_B_CC Tint" _color)
    (Variables|Default|SetColor_B _color)
    (Rendering|Material|SetScalarParameterValueonMaterials _returnvalue "EDRec_Emissive_B_Brightness_DarkenBase_Time_End" 1.0)
    (Rendering|Material|SetScalarParameterValueonMaterials _returnvalue "EDRec_Emissive_B_Brightness_Time__End" 1.0)
    (Rendering|Material|SetColorParameterValueonMaterials _returnvalue "EDRec_Emissive_B_Gradent_Offset" (Utilities|Operators|Add :A (Math|Color|MakeColor (Math|Random|RandomFloatinRange 0.0 1.0))))
    (Rendering|Material|SetColorParameterValueonMaterials _returnvalue "EDRec_Emissive_B_Gradent_Scale" (Utilities|Operators|Add :A (Math|Color|MakeColor 1.0 1.0 1.0 (Math|Random|RandomFloatinRange 0.75 1.5))))
    (Utilities|FlowControl|Sequence
      (:then_0
        (bind (_direction _scale _splat _newtrack_3) (Timeline|Timeline_0 PlayFromStart)
          (:Update
            (bind (_newtrack_3_x _newtrack_3_y) (Math|Vector|BreakVector _newtrack_3))
            (Rendering|Material|SetColorParameterValueonMaterials _returnvalue_2 "EDRec_Emissive_B_Gradent_MSk_Channel" (Math|Color|Lerp(LinearColor) "(R=0.000000,G=0.000000,B=0.000000,A=1.000000)" "(R=1.000000,G=0.000000,B=0.000000,A=0.000000)" _newtrack_3_y))
            (Rendering|Material|SetColorParameterValueonMaterials _returnvalue_2 "EDRec_DecalB_Scale" (Math|Color|Lerp(LinearColor) :B "(R=1.000000,G=1.000000,B=0.000000,A=1.000000)" :Alpha _newtrack_3_x)))
          (:Finished
            (Rendering|Material|SetScalarParameterValueonMaterials _returnvalue_3 "EDRec_Emissive_B_Brightness_DarkenBase_Time_End" 0.0)
            (Rendering|Material|SetScalarParameterValueonMaterials _returnvalue_3 "EDRec_Emissive_B_Brightness_Time__End" 0.0)
            (Rendering|Material|SetScalarParameterValueonMaterials _returnvalue_3 "EDRec_Emissive_B_Brightness_Time_BP" (Utilities|Time|GetGameTimeinSeconds))
            (Utilities|FlowControl|Sequence
              (:then_0
                (bind (_direction_2 _scale_2 _splat_2 _newtrack_3_2) (Timeline|Timeline_3 PlayFromStart)
                  (:Update
                    (bind (_newtrack_3_x_2 _newtrack_3_y_2 _newtrack_3_z) (Math|Vector|BreakVector _newtrack_3_2))
                    (Rendering|Material|SetColorParameterValueonMaterials _returnvalue_4 "TopS_01_ColorTint_CH_Green A=0=off" (Math|Color|Lerp(LinearColor) _color_b_last _color_b _newtrack_3_z))
                    (Rendering|Material|SetColorParameterValueonMaterials _returnvalue_4 "AccentV2_BaseColor_" (Math|Color|Lerp(LinearColor) _color_b_last _color_b _newtrack_3_z))
                    (Rendering|Material|SetColorParameterValueonMaterials _returnvalue_4 "TopS Emissive_02_ColorCuve Tint" (* (Math|Color|Lerp(LinearColor) (Variables|Default|GetColor_B_Last_0) _color_b_2 _newtrack_3_z) "(R=0.450000,G=0.450000,B=0.450000,A=1.000000)")))
                  (:Finished
                    (Variables|Default|SetColor_B_Last _color_b)
                    (Variables|Default|SetColor_B_Last_0 _color_b_2))
                  (:NewTrack_0
                    (Rendering|Material|SetScalarParameterValueonMaterials _returnvalue_4 "TopS Emissive_02-Darken_Base    1=On" 1.0))))
              (:then_1
                (Utilities|FlowControl|Delay 0.15)
                (Variables|Default|SetPause_B false))))))
      (:then_1
        ;[1 1 1] multi B
        (CallFunction|RandomSet "EDRec_Decal_B_Animation_Time_BP" "EDRec_DecalB_Offset_Weapon" "EDRec_DecalB_Rotation_Weapon" -0.38 0.2 -0.01 0.01 0.5 0.8)
        (CallFunction|RandomSet "0" "EDRec_DecalB_Offset_Pickaxe" "EDRec_DecalB_Rotation_Pickaxe" -0.3 0.15 0.2 -0.2 0.4 0.7)
        (CallFunction|RandomSet "0" "EDRec_DecalB_Offset_Backpack----A=Scale" "EDRec_DecalB_Rotation_Backpack" -0.15 0.15 -0.25 0.25 0.5 1.0)
        (CallFunction|RandomSet "0" "EDRec_DecalB_Offset_Character" "EDRec_DecalB_Rotation_Character" -0.15 0.15 -0.2 0.28 0.5 1.0)
        (CallFunction|RandomSet "0" "EDRec_DecalB_Offset_Vehicle" "EDRec_DecalB_Rotation_Vehicle" 0.15 0.4 -0.15 0.15 0.5 0.6)
        (CallFunction|RandomSet "0" "EDRec_DecalB_Offset_VehicleSide" "EDRec_DecalB_Rotation_Vehicle_Side" 0.3 0.42 -0.1 -0.13 0.3 0.4)
        (CallFunction|RandomSet "0" "EDRec_DecalB_Offset_Mech" "EDRec_DecalB_Rotation_Mech" -0.1 0.1 0.32 0.2 0.3 0.4)))))

(event Custom|Set%20C
  (bind _returnvalue (CustomItemWrap|GetWrappedMeshComponents))
  (bind _returnvalue_2 (CustomItemWrap|GetWrappedMeshComponents))
  (bind _returnvalue_3 (CustomItemWrap|GetWrappedMeshComponents))
  (bind _returnvalue_4 (CustomItemWrap|GetWrappedMeshComponents))
  (bind _color (Variables|Default|GetColor))
  (bind _color_c (Variables|Default|GetColor_C))
  (bind _color_c_2 (Variables|Default|GetColor_C))
  (Variables|Default|SetFire%20seq 1)
  (if (not (Variables|Default|GetPause_C))
    (CallFunction|ColorSet)
    (Variables|Default|SetPause_C true)
    (Rendering|Material|SetColorParameterValueonMaterials _returnvalue_4 "EDRec_Emissive_C_CC Tint" _color)
    (Variables|Default|SetColor_C _color)
    (Rendering|Material|SetScalarParameterValueonMaterials _returnvalue_4 "EDRec_Emissive_C_Brightness_DarkenBase_Time_End" 1.0)
    (Rendering|Material|SetScalarParameterValueonMaterials _returnvalue_4 "EDRec_Emissive_C_Brightness_Time__End" 1.0)
    (Rendering|Material|SetColorParameterValueonMaterials _returnvalue_4 "EDRec_Emissive_C_Gradent_Offset" (Utilities|Operators|Add :A (Math|Color|MakeColor (Math|Random|RandomFloatinRange 0.0 1.0))))
    (Rendering|Material|SetColorParameterValueonMaterials _returnvalue_4 "EDRec_Emissive_C_Gradent_Scale" (Utilities|Operators|Add :A (Math|Color|MakeColor 1.0 1.0 1.0 (Math|Random|RandomFloatinRange 0.75 1.5))))
    (Utilities|FlowControl|Sequence
      (:then_0
        (bind (_direction _scale _splat _newtrack_3) (Timeline|Timeline_1 PlayFromStart)
          (:Update
            (bind (_newtrack_3_x _newtrack_3_y) (Math|Vector|BreakVector _newtrack_3))
            (Rendering|Material|SetColorParameterValueonMaterials _returnvalue_3 "EDRec_Emissive_C_Gradent_MSk_Channel" (Math|Color|Lerp(LinearColor) "(R=0.000000,G=0.000000,B=0.000000,A=1.000000)" "(R=1.000000,G=0.000000,B=0.000000,A=0.000000)" _newtrack_3_y))
            (Rendering|Material|SetColorParameterValueonMaterials _returnvalue_3 "EDRec_DecalC_Scale" (Math|Color|Lerp(LinearColor) :B "(R=1.000000,G=1.000000,B=0.000000,A=1.000000)" :Alpha _newtrack_3_x)))
          (:Finished
            (Rendering|Material|SetScalarParameterValueonMaterials _returnvalue "EDRec_Emissive_C_Brightness_DarkenBase_Time_End" 0.0)
            (Rendering|Material|SetScalarParameterValueonMaterials _returnvalue "EDRec_Emissive_C_Brightness_Time__End" 0.0)
            (Rendering|Material|SetScalarParameterValueonMaterials _returnvalue "EDRec_Emissive_C_Brightness_Time_BP" (Utilities|Time|GetGameTimeinSeconds))
            (Utilities|FlowControl|Sequence
              (:then_0
                (bind (_direction_2 _scale_2 _splat_2 _newtrack_3_2) (Timeline|Timeline_4 PlayFromStart)
                  (:Update
                    (bind (_newtrack_3_x_2 _newtrack_3_y_2 _newtrack_3_z) (Math|Vector|BreakVector _newtrack_3_2))
                    (Rendering|Material|SetColorParameterValueonMaterials _returnvalue_2 "EDRec_Emissive_D_Color-" (Math|Color|Lerp(LinearColor) (Variables|Default|GetColor_C_Last) _color_c _newtrack_3_z))
                    (Rendering|Material|SetColorParameterValueonMaterials _returnvalue_2 "TopS Emissive_03_ColorCuve Tint" (* (Math|Color|Lerp(LinearColor) (Variables|Default|GetColor_C_Last_0) _color_c_2 _newtrack_3_z) "(R=0.450000,G=0.450000,B=0.450000,A=1.000000)")))
                  (:Finished
                    (Variables|Default|SetColor_C_Last _color_c)
                    (Variables|Default|SetColor_C_Last_0 _color_c_2))
                  (:NewTrack_0
                    (Rendering|Material|SetScalarParameterValueonMaterials _returnvalue_2 "TopS Emissive_03-Darken_Base    1=On" 1.0))))
              (:then_1
                (Utilities|FlowControl|Delay 0.15)
                (Variables|Default|SetPause_C false))))))
      (:then_1
        ; C
        (CallFunction|RandomSet "EDRec_Decal_C_Animation_Time_BP" "EDRec_DecalC_Offset_Weapon" "EDRec_DecalC_Rotation_Weapon" -0.38 0.2 -0.01 0.01 0.5 0.8)
        (CallFunction|RandomSet "0" "EDRec_DecalC_Offset_Pickaxe" "EDRec_DecalC_Rotation_Pickaxe" -0.3 0.15 0.2 -0.2 0.4 0.7)
        (CallFunction|RandomSet "0" "EDRec_DecalC_Offset_Backpack----A=Scale" "EDRec_DecalC_Rotation_Backpack" -0.15 0.15 -0.25 0.25 0.5 1.0)
        (CallFunction|RandomSet "0" "EDRec_DecalC_Offset_Character" "EDRec_DecalC_Rotation_Character" -0.15 0.15 -0.2 0.28 0.5 1.0)
        (CallFunction|RandomSet "0" "EDRec_DecalC_Offset_Vehicle" "EDRec_DecalC_Rotation_Vehicle" 0.15 0.4 -0.15 0.15 0.5 0.6)
        (CallFunction|RandomSet "0" "EDRec_DecalC_Offset_VehicleSide" "EDRec_DecalC_Rotation_Vehicle_Side" 0.3 0.42 -0.1 -0.13 0.3 0.4)
        (CallFunction|RandomSet "0" "EDRec_DecalC_Offset_Mech" "EDRec_DecalC_Rotation_Mech" -0.1 0.1 0.32 0.2 0.3 0.4)))))

(fn RandomSet
  (params (TimeBP_Flipbook%20Triger :type Name) (Offset :type Name) (Rotation :type Name) (X=min :type Float%20%28double-precision%29) (X=Max :type Float%20%28double-precision%29) (Y=Min :type Float%20%28double-precision%29) (Y=Max :type Float%20%28double-precision%29) (Scale=Min :type Float%20%28double-precision%29) (Scale=Max :type Float%20%28double-precision%29))
  (bind _returnvalue (CustomItemWrap|GetWrappedMeshComponents))
  (bind _stream (Variables|Default|GetStream))
  (bind _returnvalue_2 (Math|Random|RandomFloatinRangefromStream _stream X=min X=Max))
  (Rendering|Material|SetScalarParameterValueonMaterials _returnvalue TimeBP_Flipbook%20Triger (Utilities|Time|GetGameTimeinSeconds))
  (Rendering|Material|SetColorParameterValueonMaterials _returnvalue Offset (Math|Color|MakeColor _returnvalue_2 (Math|Random|RandomFloatinRangefromStream _stream Y=Min Y=Max) 0.0 (Math|Random|RandomFloatinRangefromStream _stream Scale=Min Scale=Max)))
  (Rendering|Material|SetScalarParameterValueonMaterials _returnvalue Rotation (Math|Random|RandomFloat))
  (Variables|Default|SetPosition _returnvalue_2))

(fn ConstructionScript)
