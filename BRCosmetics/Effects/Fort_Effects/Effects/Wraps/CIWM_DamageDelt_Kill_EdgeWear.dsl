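; Weapon wrap cosmetic modifier that flashes edge-wear emissive on wrapped meshes when the player deals damage or gets a kill.
; It watches the pawn's damage and kill stats, then stamps game time, start and duration into material scalar parameters, with kills flashing stronger and longer.
; A Pause flag keeps damage pulses from overriding an active kill pulse.
; In the front end, a weapon-fire counter imitates both reactions for preview.
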(event EventOnWeaponFired
  (bind _returnvalue (CustomItemWrap|GetWrappedMeshComponents))
  (bind _returnvalue_2 (Utilities|Time|GetGameTimeinSeconds))
  (bind _returnvalue_3 (CustomItemWrap|GetWrappedMeshComponents))
  ; Frontnd reactive
  (if (CustomCosmeticModifier|IsinFrontEnd)
    (bind _output_get (Variables|Default|Setweapon%20fire (+ (Variables|Default|Getweapon%20fire) 1)))
    (if (>= _output_get 5)
      (if (== (Variables|Default|Getweapon%20fire) 7)
        (Variables|Default|Setweapon%20fire 0)
        (bind _output_get_2 (Variables|Default|SetPause true))
        (Rendering|Material|SetScalarParameterValueonMaterials _returnvalue_3 "Eff_Edge\\Dirt_Edge_Emissive_Time_BP" _returnvalue_2)
        (Rendering|Material|SetScalarParameterValueonMaterials _returnvalue_3 "Eff_Edge\\Dirt_Edge_Emissive_Start" 1e+01)
        (Rendering|Material|SetScalarParameterValueonMaterials _returnvalue_3 "Eff_Edge\\Dirt_Edge_Emissive_Duration" 4.0)
        (Utilities|FlowControl|Delay 3.0)
        (Variables|Default|SetPause false)
        (else
          (Utilities|FlowControl|Delay 0.05)
          (if (not _output_get_2)
            (Rendering|Material|SetScalarParameterValueonMaterials _returnvalue "Eff_Edge\\Dirt_Edge_Emissive_Time_BP" _returnvalue_2)
            (Rendering|Material|SetScalarParameterValueonMaterials _returnvalue "Eff_Edge\\Dirt_Edge_Emissive_Start" 1.0)
            (Rendering|Material|SetScalarParameterValueonMaterials _returnvalue "Eff_Edge\\Dirt_Edge_Emissive_Duration" 1.0)))))))

(event EventOnPawnSet
  (bind _returnvalue (CustomItemWrap|GetWrappedMeshComponents))
  (bind _returnvalue_2 (Utilities|Time|GetGameTimeinSeconds))
  (bind _returnvalue_3 (CustomItemWrap|GetWrappedMeshComponents))
  (Utilities|FlowControl|Sequence
    (:then_0
      (Cosmetics|WatchCosmeticStat (CustomCharacter|GetAssociatedPlayerPawn) "(TagName=\"StatManager.AthenaKills\")" 0 false
        (:OnStatChanged
          (bind _output_get (Variables|Default|SetPause true))
          (Rendering|Material|SetScalarParameterValueonMaterials _returnvalue_3 "Eff_Edge\\Dirt_Edge_Emissive_Time_BP" _returnvalue_2)
          (Rendering|Material|SetScalarParameterValueonMaterials _returnvalue_3 "Eff_Edge\\Dirt_Edge_Emissive_Start" 1e+01)
          (Rendering|Material|SetScalarParameterValueonMaterials _returnvalue_3 "Eff_Edge\\Dirt_Edge_Emissive_Duration" 4.0)
          (Utilities|FlowControl|Delay 3.0)
          (Variables|Default|SetPause false))))
    (:then_1
      (Cosmetics|WatchCosmeticStat (CustomCharacter|GetAssociatedPlayerPawn) "(TagName=\"StatManager.DamageDealt\")" 0 false
        (:OnStatChanged
          (Utilities|FlowControl|Delay 0.05)
          (if (not _output_get)
            (Rendering|Material|SetScalarParameterValueonMaterials _returnvalue "Eff_Edge\\Dirt_Edge_Emissive_Time_BP" _returnvalue_2)
            (Rendering|Material|SetScalarParameterValueonMaterials _returnvalue "Eff_Edge\\Dirt_Edge_Emissive_Start" 1.0)
            (Rendering|Material|SetScalarParameterValueonMaterials _returnvalue "Eff_Edge\\Dirt_Edge_Emissive_Duration" 1.0)))))))

(fn ConstructionScript)
