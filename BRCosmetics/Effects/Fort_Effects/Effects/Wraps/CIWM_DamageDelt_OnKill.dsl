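; Item wrap cosmetic modifier that animates wrapped weapon materials when the player deals damage or gets a kill.
; Cosmetic stat watchers write game time into named material parameters, throttled by pause flags and delays; in the front end, a shot counter simulates both events.
; Optional flipbook effects draw frames from a shuffled array and randomize offset and rotation across seven parameter sets.
; Spawn and weapon-change events reshuffle the frame order.
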(event EventOnWeaponFired
  (if (CustomCosmeticModifier|IsinFrontEnd)
    (bind _output_get (Variables|Envirometn_Varable|SetIterationCount (+ (Variables|Envirometn_Varable|GetIterationCount) 1)))
    (if (>= _output_get 5)
      (if (== (Variables|Envirometn_Varable|GetIterationCount) 7)
        (Variables|Envirometn_Varable|SetIterationCount 0)
        (Utilities|FlowControl|Sequence
          (:then_0
            (if (Variables|Default|Onkill|GetOn%20Kill)
              (Utilities|FlowControl|Sequence
                (:then_0
                  (if (not (Variables|EnvirometnVarable|GetPause%20on%20kill))
                    (Variables|EnvirometnVarable|SetPause%20on%20kill true)
                    (Utilities|FlowControl|Sequence
                      (:then_0
                        ;[1 1 1] onKill
                        (for _array_element (Variables|Default|Onkill|GetOnKill%20Trigger%20Names%20BP)
                          (Rendering|Material|SetScalarParameterValueonMaterials (CustomItemWrap|GetWrappedMeshComponents) _array_element (Utilities|Time|GetGameTimeinSeconds))
                          (Utilities|FlowControl|Delay (Variables|Default|Onkill|GetOn%20Kill%20Pause))
                          (Variables|EnvirometnVarable|SetPause%20on%20kill false)))
                      (:then_1
                        ; On Kill Frame
                        (if (Variables|Default|Onkill|Flip|GetOnKill_FlipBook_Switch)
                          (Variables|EnvirometnVarable|SetKills (+ (Variables|EnvirometnVarable|GetKills) 1))
                          (for _array_element_2 (Variables|Default|Onkill|Flip|GetOnKill%20Trigger%20FrameChange_Name)
                            (Rendering|Material|SetScalarParameterValueonMaterials (CustomItemWrap|GetWrappedMeshComponents) _array_element_2 (Math|Conversions|ToFloat(Integer) (Utilities|Array|Get(aref) (Variables|Default|FlipBookRanPos|GetNumber%20of%20Frames_random) (Variables|EnvirometnVarable|GetKills)))))
                          (Utilities|FlowControl|Sequence
                            (:then_1
                              (if (== (Variables|EnvirometnVarable|GetKills) (Variables|Default|FlipBookRanPos|GetNumber%20of%20Frames))
                                (Variables|EnvirometnVarable|SetKills -1)
                                (Utilities|Array|Shuffle (Variables|Default|FlipBookRanPos|GetNumber%20of%20Frames_random))))))))))
                (:then_1
                  (if (Variables|Default|FlipBookRanPos|GetOnKill_Flip\Rand\Pos)
                    (Variables|EnvirometnVarable|SetFlip-Delay (Variables|Default|FlipBookRanPos|GetOnKill-Flip-%20Delay))
                    (CallFunction|FlipBookRandPlacement-Kill))))))
          (:then_1
            _)))
      (if (Variables|Default|DamageDelt|GetDamageDelt)
        (Utilities|FlowControl|Sequence
          (:then_0
            (if (not (Variables|EnvirometnVarable|GetPause%20Damage%20Delt))
              (Variables|EnvirometnVarable|SetPause%20Damage%20Delt true)
              ;[1 1 1] DamageDelt
              (for _array_element_3 (Variables|Default|DamageDelt|GetDamageDelt%20Trigger%20Names%20BP)
                (Rendering|Material|SetScalarParameterValueonMaterials (CustomItemWrap|GetWrappedMeshComponents) _array_element_3 (Utilities|Time|GetGameTimeinSeconds)))
              (Utilities|FlowControl|Delay (Variables|Default|DamageDelt|GetDamage_Delt%20Pause))
              (Variables|EnvirometnVarable|SetPause%20Damage%20Delt false)))
          (:then_1
            (if (Variables|Default|FlipBookRanPos|GetDamageDelt_Flip\Rand\Pos)
              (Variables|EnvirometnVarable|SetFlip-Delay (Variables|Default|FlipBookRanPos|GetDamageDelt-Flip-%20Delay))
              (CallFunction|FlipBookRandPlacement))))))))

(event EventOnPawnSet
  (bind _returnvalue (CustomCharacter|GetAssociatedPlayerPawn))
  (Utilities|FlowControl|Sequence
    (:then_0
      (if (Variables|Default|DamageDelt|GetDamageDelt)
        (Cosmetics|WatchCosmeticStat _returnvalue "(TagName=\"StatManager.DamageDealt\")"
          (:OnStatChanged
            (Utilities|FlowControl|Sequence
              (:then_0
                (if (not (Variables|EnvirometnVarable|GetPause%20Damage%20Delt))
                  (Variables|EnvirometnVarable|SetPause%20Damage%20Delt true)
                  ;[1 1 1] DamageDelt
                  (for _array_element (Variables|Default|DamageDelt|GetDamageDelt%20Trigger%20Names%20BP)
                    (Rendering|Material|SetScalarParameterValueonMaterials (CustomItemWrap|GetWrappedMeshComponents) _array_element (Utilities|Time|GetGameTimeinSeconds)))
                  (Utilities|FlowControl|Delay (Variables|Default|DamageDelt|GetDamage_Delt%20Pause))
                  (Variables|EnvirometnVarable|SetPause%20Damage%20Delt false)))
              (:then_1
                (if (Variables|Default|FlipBookRanPos|GetDamageDelt_Flip\Rand\Pos)
                  (Variables|EnvirometnVarable|SetFlip-Delay (Variables|Default|FlipBookRanPos|GetDamageDelt-Flip-%20Delay))
                  (CallFunction|FlipBookRandPlacement))))))))
    (:then_1
      (if (Variables|Default|Onkill|GetOn%20Kill)
        (Cosmetics|WatchCosmeticStat _returnvalue "(TagName=\"StatManager.AthenaKills\")"
          (:OnStatChanged
            (Utilities|FlowControl|Sequence
              (:then_0
                (if (not (Variables|EnvirometnVarable|GetPause%20on%20kill))
                  (Variables|EnvirometnVarable|SetPause%20on%20kill true)
                  (Utilities|FlowControl|Sequence
                    (:then_0
                      ;[1 1 1] onKill
                      (for _array_element_2 (Variables|Default|Onkill|GetOnKill%20Trigger%20Names%20BP)
                        (Rendering|Material|SetScalarParameterValueonMaterials (CustomItemWrap|GetWrappedMeshComponents) _array_element_2 (Utilities|Time|GetGameTimeinSeconds))
                        (Utilities|FlowControl|Delay (Variables|Default|Onkill|GetOn%20Kill%20Pause))
                        (Variables|EnvirometnVarable|SetPause%20on%20kill false)))
                    (:then_1
                      ; On Kill Frame
                      (if (Variables|Default|Onkill|Flip|GetOnKill_FlipBook_Switch)
                        (Variables|EnvirometnVarable|SetKills (+ (Variables|EnvirometnVarable|GetKills) 1))
                        (for _array_element_3 (Variables|Default|Onkill|Flip|GetOnKill%20Trigger%20FrameChange_Name)
                          (Rendering|Material|SetScalarParameterValueonMaterials (CustomItemWrap|GetWrappedMeshComponents) _array_element_3 (Math|Conversions|ToFloat(Integer) (Utilities|Array|Get(aref) (Variables|Default|FlipBookRanPos|GetNumber%20of%20Frames_random) (Variables|EnvirometnVarable|GetKills)))))
                        (Utilities|FlowControl|Sequence
                          (:then_1
                            (if (== (Variables|EnvirometnVarable|GetKills) (Variables|Default|FlipBookRanPos|GetNumber%20of%20Frames))
                              (Variables|EnvirometnVarable|SetKills -1)
                              (Utilities|Array|Shuffle (Variables|Default|FlipBookRanPos|GetNumber%20of%20Frames_random))))))))))
              (:then_1
                (if (Variables|Default|FlipBookRanPos|GetOnKill_Flip\Rand\Pos)
                  (Variables|EnvirometnVarable|SetFlip-Delay (Variables|Default|FlipBookRanPos|GetOnKill-Flip-%20Delay))
                  (CallFunction|FlipBookRandPlacement-Kill))))))))))

(event EventBeginPlay
  (Utilities|FlowControl|Delay 0.2)
  (if (Variables|Default|FlipBookRanPos|GetHas%20Frames)
    (Utilities|Array|Shuffle (Variables|Default|FlipBookRanPos|GetNumber%20of%20Frames_random))))

(event EventOnPostResIn
  ; Shuffle on load
  (Utilities|FlowControl|Delay 0.05)
  (if (Variables|Default|Onkill|Flip|GetOnKill_FlipBook_Switch)
    (Utilities|Array|Shuffle (Variables|Default|FlipBookRanPos|GetNumber%20of%20Frames_random))))

(event EventOnNewWeapons
  (params (WeaponsAsItems :type Array%20of%20Fort%20Item%20Object%20References))
  ; Shuffle on load
  (Utilities|FlowControl|Delay 0.05)
  (if (Variables|Default|Onkill|Flip|GetOnKill_FlipBook_Switch)
    (Utilities|Array|Shuffle (Variables|Default|FlipBookRanPos|GetNumber%20of%20Frames_random))))

(event EventOnNewPlayerWeaponType
  (params (NewType :type EFortItemType%20Enum :default "WorldItem") (OldType :type EFortItemType%20Enum :default "WorldItem"))
  ; Shuffle on load
  (Utilities|FlowControl|Delay 0.05)
  (if (Variables|Default|Onkill|Flip|GetOnKill_FlipBook_Switch)
    (Utilities|Array|Shuffle (Variables|Default|FlipBookRanPos|GetNumber%20of%20Frames_random))))

(event EventOnWeaponChanged
  (params (NewWeapon :type Fort%20Weapon%20Object%20Reference) (OldWeapon :type Fort%20Weapon%20Object%20Reference))
  ; Shuffle on load
  (Utilities|FlowControl|Delay 0.05)
  (if (Variables|Default|Onkill|Flip|GetOnKill_FlipBook_Switch)
    (Utilities|Array|Shuffle (Variables|Default|FlipBookRanPos|GetNumber%20of%20Frames_random))))

(event Custom|FlipBook%20rand%20Placement
  (bind _flipboook_offset_names (Variables|Default|FlipBookRanPos|GetFlipBoook%20offset%20names))
  (bind _flipboook_rotation_names (Variables|Default|FlipBookRanPos|GetFlipBoook%20Rotation%20names))
  (if (not (Variables|EnvirometnVarable|GetOnKill_Flip\Rand\Pos%20%20Pause))
    (Variables|EnvirometnVarable|SetOnKill_Flip\Rand\Pos%20%20Pause true)
    (Default|RandomSet (Variables|Default|FlipBookRanPos|GetFlip%20Book%20Frame%20Name) (Variables|Default|FlipBookRanPos|GetFlip%20Book%20Time%20BP%20Name) (Utilities|Array|Get(aref) _flipboook_offset_names) (Utilities|Array|Get(aref) _flipboook_rotation_names) -0.38 0.2 -0.01 0.01 0.1 -0.15)
    (Default|RandomSet "0" "0" (Utilities|Array|Get(aref) _flipboook_offset_names 1) (Utilities|Array|Get(aref) _flipboook_rotation_names 1) -0.3 0.15 0.2 -0.2 0.1 -0.15)
    (Default|RandomSet "0" "0" (Utilities|Array|Get(aref) _flipboook_offset_names 2) (Utilities|Array|Get(aref) _flipboook_rotation_names 2) -0.15 0.15 -0.25 0.25 0.1 -0.15)
    (Default|RandomSet "0" "0" (Utilities|Array|Get(aref) _flipboook_offset_names 3) (Utilities|Array|Get(aref) _flipboook_rotation_names 3) -0.15 0.15 -0.2 0.28 0.1 -0.15)
    (Default|RandomSet "0" "0" (Utilities|Array|Get(aref) _flipboook_offset_names 4) (Utilities|Array|Get(aref) _flipboook_rotation_names 4) 0.15 0.4 -0.15 0.15 0.1 -0.15)
    (Default|RandomSet "0" "0" (Utilities|Array|Get(aref) _flipboook_offset_names 5) (Utilities|Array|Get(aref) _flipboook_rotation_names 5) 0.3 0.42 -0.1 -0.13 0.1 -0.15)
    (Default|RandomSet "0" "0" (Utilities|Array|Get(aref) _flipboook_offset_names 6) (Utilities|Array|Get(aref) _flipboook_rotation_names 6) -0.1 0.1 0.32 0.2 0.1 -0.15)
    (Utilities|FlowControl|Delay (Variables|Default|FlipBookRanPos|GetFlip\Rand\Pos%20Pause))
    (CallFunction|DamageDeltTrack)
    (Variables|EnvirometnVarable|SetOnKill_Flip\Rand\Pos%20%20Pause false)))

(event Custom|FlipBook%20rand%20Placement-Kill
  (bind _flipboook_offset_names (Variables|Default|FlipBookRanPos|GetFlipBoook%20offset%20names))
  (bind _flipboook_rotation_names (Variables|Default|FlipBookRanPos|GetFlipBoook%20Rotation%20names))
  (Variables|EnvirometnVarable|SetOnKill_Flip\Rand\Pos%20%20Pause false)
  (if (not (Variables|EnvirometnVarable|GetOnKill_Flip\Rand\Pos%20%20Pause))
    (Variables|EnvirometnVarable|SetOnKill_Flip\Rand\Pos%20%20Pause true)
    (Default|RandomSet (Variables|Default|FlipBookRanPos|GetFlip%20Book%20Frame%20Name) (Variables|Default|FlipBookRanPos|GetFlip%20Book%20Time%20BP%20Name) (Utilities|Array|Get(aref) _flipboook_offset_names) (Utilities|Array|Get(aref) _flipboook_rotation_names) -0.38 0.2 -0.01 0.01 0.1 -0.15)
    (Default|RandomSet "0" "0" (Utilities|Array|Get(aref) _flipboook_offset_names 1) (Utilities|Array|Get(aref) _flipboook_rotation_names 1) -0.3 0.15 0.2 -0.2 0.1 -0.15)
    (Default|RandomSet "0" "0" (Utilities|Array|Get(aref) _flipboook_offset_names 2) (Utilities|Array|Get(aref) _flipboook_rotation_names 2) -0.15 0.15 -0.25 0.25 0.1 -0.15)
    (Default|RandomSet "0" "0" (Utilities|Array|Get(aref) _flipboook_offset_names 3) (Utilities|Array|Get(aref) _flipboook_rotation_names 3) -0.15 0.15 -0.2 0.28 0.1 -0.15)
    (Default|RandomSet "0" "0" (Utilities|Array|Get(aref) _flipboook_offset_names 4) (Utilities|Array|Get(aref) _flipboook_rotation_names 4) 0.15 0.4 -0.15 0.15 0.1 -0.15)
    (Default|RandomSet "0" "0" (Utilities|Array|Get(aref) _flipboook_offset_names 5) (Utilities|Array|Get(aref) _flipboook_rotation_names 5) 0.3 0.42 -0.1 -0.13 0.1 -0.15)
    (Default|RandomSet "0" "0" (Utilities|Array|Get(aref) _flipboook_offset_names 6) (Utilities|Array|Get(aref) _flipboook_rotation_names 6) -0.1 0.1 0.32 0.2 0.1 -0.15)
    (Utilities|FlowControl|Delay (Variables|Default|FlipBookRanPos|GetFlip\Rand\Pos%20Pause))
    (CallFunction|DamageDeltTrack)
    (Variables|EnvirometnVarable|SetOnKill_Flip\Rand\Pos%20%20Pause false)))

(event Custom|Damage_DeltTrack
  (bind _output_get (Variables|EnvirometnVarable|SetDamage (+ (Variables|EnvirometnVarable|GetDamage) 1)))
  (if (== _output_get (Variables|Default|FlipBookRanPos|GetNumber%20of%20Frames))
    (Utilities|Array|Shuffle (Variables|Default|FlipBookRanPos|GetNumber%20of%20Frames_random))
    (Variables|EnvirometnVarable|SetDamage -1)))

(fn RandomSet
  (params (FlipBook%20Frame :type Name) (TimeBP :type Name) (Offset :type Name) (Rotation :type Name) (X=min :type Float%20%28double-precision%29) (X=Max :type Float%20%28double-precision%29) (Y=Min :type Float%20%28double-precision%29) (Y=Max :type Float%20%28double-precision%29) (Rot=Min :type Float%20%28double-precision%29) (Rot=Max :type Float%20%28double-precision%29))
  (bind _returnvalue (CustomItemWrap|GetWrappedMeshComponents))
  (bind _stream (Variables|EnvirometnVarable|GetStream))
  (bind _returnvalue_2 (Math|Random|RandomFloatinRangefromStream _stream X=min X=Max))
  (if (Variables|Default|FlipBookRanPos|GetHas%20Frames)
    (Rendering|Material|SetScalarParameterValueonMaterials _returnvalue FlipBook%20Frame (/ (Math|Conversions|ToFloat(Integer) (Utilities|Array|Get(aref) (Variables|Default|FlipBookRanPos|GetNumber%20of%20Frames_random) (Variables|EnvirometnVarable|GetDamage))) (Variables|Default|FlipBookRanPos|GetNumber%20of%20Frames))))
  (Rendering|Material|SetScalarParameterValueonMaterials _returnvalue TimeBP (+ (Utilities|Time|GetGameTimeinSeconds) (Variables|EnvirometnVarable|GetFlip-Delay)))
  (Rendering|Material|SetVectorParameterValueonMaterials _returnvalue Offset (Math|Vector|MakeVector _returnvalue_2 (Math|Random|RandomFloatinRangefromStream _stream Y=Min Y=Max)))
  (Rendering|Material|SetScalarParameterValueonMaterials _returnvalue Rotation (Math|Random|RandomFloatinRange Rot=Min Rot=Max))
  (Variables|EnvirometnVarable|SetPosition _returnvalue_2))

(fn ConstructionScript)
