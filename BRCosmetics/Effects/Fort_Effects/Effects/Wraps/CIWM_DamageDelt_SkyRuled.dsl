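; Weapon wrap modifier for SkyRuled that animates wrapped mesh materials when the player deals damage.
; The DamageDealt stat watcher stamps game time into scalar parameters, zeroes emissive brightness and swaps mask textures by a shuffled index, using Onkilled as a re-entry guard.
; Short delays precede restoring brightness and advancing the sequence, which reshuffles on wrap-around.
; Front-end weapon fire replays the effect on certain shot counts; the AthenaKills watcher has no handler.
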(event EventOnWeaponFired
  (bind _array (Utilities|Array|MakeArray "TopS Emissive_01-Brightness-Reactive--->Gradent Add" "TopS Emissive_02-Brightness-Reactive--->Gradent  Add" "TopS Emissive_03-Brightness-Reactive--->Gradent  Add"))
  (if (CustomCosmeticModifier|IsinFrontEnd)
    (bind _output_get (Variables|Envirometn_Varable|SetIterationCount (+ (Variables|Envirometn_Varable|GetIterationCount) 1)))
    (if (>= _output_get 5)
      (if (== (Variables|Envirometn_Varable|GetIterationCount) 7)
        (Variables|Envirometn_Varable|SetIterationCount 0)
        (else
          (if (not (Variables|Default|GetOnkilled))
            (bind _output_get_2 (Variables|Default|SetFire%20seq (Variables|Default|GetFire%20seq)))
            (Utilities|FlowControl|Sequence
              (:then_0
                (for _array_element (Utilities|Array|MakeArray "EDRec_Emissive_A_Brightness_Time_BP" "EDRec_Emissive_A_Gradent_Time_BP" "Eff_Galaxy_Edge_TFBP" "Eff_Galaxy_Gmap_TFBP" "EDRec_Emissive_B_Brightness_Time_BP" "EDRec_Emissive_B_Gradent_Time_BP")
                  (Rendering|Material|SetScalarParameterValueonMaterials (CustomItemWrap|GetWrappedMeshComponents) _array_element (Utilities|Time|GetGameTimeinSeconds))))
              (:then_1
                (for _array_element_2 _array
                  (Rendering|Material|SetScalarParameterValueonMaterials (CustomItemWrap|GetWrappedMeshComponents) _array_element_2 0.0)
                  (switch int (Utilities|Array|Get(aref) (Variables|Default|GetAray) _output_get_2)
                    (:0
                      (Variables|Default|SetOnkilled true)
                      (for _array_element_3 (CustomItemWrap|GetWrappedMeshComponents)
                        (for _array_element_4 (Rendering|Material|GetMaterials _array_element_3)
                          (bind _asmaterial_instance_dynamic (Utilities|Casting|CastToMaterialInstanceDynamic _array_element_4))
                          (Rendering|Material|SetTextureParameterValue _asmaterial_instance_dynamic "TopS_01__Tex_Mask" "/BRCosmetics/Weapons/WeaponSkins/Wraps/Textures/skyRuled/T_Wrap_SkyRuled_DA_Msk.T_Wrap_SkyRuled_DA_Msk")
                          (Rendering|Material|SetTextureParameterValue _asmaterial_instance_dynamic "TopS_02__Tex_Mask" "/BRCosmetics/Weapons/WeaponSkins/Wraps/Textures/skyRuled/T_Wrap_SkyRuled_DB_Msk.T_Wrap_SkyRuled_DB_Msk"))
                        (Utilities|FlowControl|Sequence
                          (:then_0
                            (Utilities|FlowControl|Delay 0.1)
                            (for _array_element_5 _array
                              (Rendering|Material|SetScalarParameterValueonMaterials (CustomItemWrap|GetWrappedMeshComponents) _array_element_5 1.0)
                              (for _array_element_6 (Variables|Default|GetDamageDelt%20Trigger%20Names%20BP)
                                (Rendering|Material|SetScalarParameterValueonMaterials (CustomItemWrap|GetWrappedMeshComponents) _array_element_6 (Utilities|Time|GetGameTimeinSeconds)))))
                          (:then_1
                            (Utilities|FlowControl|Delay 0.1)
                            (Variables|Default|SetOnkilled false)
                            (bind _output_get_3 (Variables|Default|SetFire%20seq (+ (Variables|Default|GetFire%20seq) 1)))
                            (if (>= _output_get_3 3)
                              (Variables|Default|SetFire%20seq 0)
                              (Utilities|Array|Shuffle (Variables|Default|GetAray)))))))
                    (:1
                      (Variables|Default|SetOnkilled true)
                      (for _array_element_7 (CustomItemWrap|GetWrappedMeshComponents)
                        (for _array_element_8 (Rendering|Material|GetMaterials _array_element_7)
                          (bind _asmaterial_instance_dynamic_2 (Utilities|Casting|CastToMaterialInstanceDynamic _array_element_8))
                          (Rendering|Material|SetTextureParameterValue _asmaterial_instance_dynamic_2 "TopS_01__Tex_Mask" "/BRCosmetics/Weapons/WeaponSkins/Wraps/Textures/skyRuled/T_Wrap_SkyRuled_EA_Msk_1.T_Wrap_SkyRuled_EA_Msk_1")
                          (Rendering|Material|SetTextureParameterValue _asmaterial_instance_dynamic_2 "TopS_02__Tex_Mask" "/BRCosmetics/Weapons/WeaponSkins/Wraps/Textures/skyRuled/T_Wrap_SkyRuled_EB_Msk_1.T_Wrap_SkyRuled_EB_Msk_1"))))
                    (:2
                      (Variables|Default|SetOnkilled true)
                      (for _array_element_9 (CustomItemWrap|GetWrappedMeshComponents)
                        (for _array_element_10 (Rendering|Material|GetMaterials _array_element_9)
                          (bind _asmaterial_instance_dynamic_3 (Utilities|Casting|CastToMaterialInstanceDynamic _array_element_10))
                          (Rendering|Material|SetTextureParameterValue _asmaterial_instance_dynamic_3 "TopS_01__Tex_Mask" "/BRCosmetics/Weapons/WeaponSkins/Wraps/Textures/skyRuled/T_Wrap_SkyRuled_C_Msk.T_Wrap_SkyRuled_C_Msk")
                          (Rendering|Material|SetTextureParameterValue _asmaterial_instance_dynamic_3 "TopS_02__Tex_Mask" "/BRCosmetics/Weapons/WeaponSkins/Wraps/Textures/skyRuled/T_Wrap_SkyRuled_B_Msk.T_Wrap_SkyRuled_B_Msk"))))))))))))))

(event EventBeginPlay
  (Utilities|FlowControl|Delay 0.5)
  (Utilities|Array|Shuffle (Variables|Default|GetAray)))

(event EventOnPawnSet
  (bind _returnvalue (CustomCharacter|GetAssociatedPlayerPawn))
  (bind _array (Utilities|Array|MakeArray "TopS Emissive_01-Brightness-Reactive--->Gradent Add" "TopS Emissive_02-Brightness-Reactive--->Gradent  Add" "TopS Emissive_03-Brightness-Reactive--->Gradent  Add"))
  (Utilities|FlowControl|Sequence
    (:then_0
      (Cosmetics|WatchCosmeticStat _returnvalue "(TagName=\"StatManager.DamageDealt\")" 0 false
        (:OnStatChanged
          (if (not (Variables|Default|GetOnkilled))
            (bind _output_get (Variables|Default|SetFire%20seq (Variables|Default|GetFire%20seq)))
            (Utilities|FlowControl|Sequence
              (:then_0
                (for _array_element (Utilities|Array|MakeArray "EDRec_Emissive_A_Brightness_Time_BP" "EDRec_Emissive_A_Gradent_Time_BP" "Eff_Galaxy_Edge_TFBP" "Eff_Galaxy_Gmap_TFBP" "EDRec_Emissive_B_Brightness_Time_BP" "EDRec_Emissive_B_Gradent_Time_BP")
                  (Rendering|Material|SetScalarParameterValueonMaterials (CustomItemWrap|GetWrappedMeshComponents) _array_element (Utilities|Time|GetGameTimeinSeconds))))
              (:then_1
                (for _array_element_2 _array
                  (Rendering|Material|SetScalarParameterValueonMaterials (CustomItemWrap|GetWrappedMeshComponents) _array_element_2 0.0)
                  (switch int (Utilities|Array|Get(aref) (Variables|Default|GetAray) _output_get)
                    (:0
                      (Variables|Default|SetOnkilled true)
                      (for _array_element_3 (CustomItemWrap|GetWrappedMeshComponents)
                        (for _array_element_4 (Rendering|Material|GetMaterials _array_element_3)
                          (bind _asmaterial_instance_dynamic (Utilities|Casting|CastToMaterialInstanceDynamic _array_element_4))
                          (Rendering|Material|SetTextureParameterValue _asmaterial_instance_dynamic "TopS_01__Tex_Mask" "/BRCosmetics/Weapons/WeaponSkins/Wraps/Textures/skyRuled/T_Wrap_SkyRuled_DA_Msk.T_Wrap_SkyRuled_DA_Msk")
                          (Rendering|Material|SetTextureParameterValue _asmaterial_instance_dynamic "TopS_02__Tex_Mask" "/BRCosmetics/Weapons/WeaponSkins/Wraps/Textures/skyRuled/T_Wrap_SkyRuled_DB_Msk.T_Wrap_SkyRuled_DB_Msk"))
                        (Utilities|FlowControl|Sequence
                          (:then_0
                            (Utilities|FlowControl|Delay 0.1)
                            (for _array_element_5 _array
                              (Rendering|Material|SetScalarParameterValueonMaterials (CustomItemWrap|GetWrappedMeshComponents) _array_element_5 1.0)
                              (for _array_element_6 (Variables|Default|GetDamageDelt%20Trigger%20Names%20BP)
                                (Rendering|Material|SetScalarParameterValueonMaterials (CustomItemWrap|GetWrappedMeshComponents) _array_element_6 (Utilities|Time|GetGameTimeinSeconds)))))
                          (:then_1
                            (Utilities|FlowControl|Delay 0.1)
                            (Variables|Default|SetOnkilled false)
                            (bind _output_get_2 (Variables|Default|SetFire%20seq (+ (Variables|Default|GetFire%20seq) 1)))
                            (if (>= _output_get_2 3)
                              (Variables|Default|SetFire%20seq 0)
                              (Utilities|Array|Shuffle (Variables|Default|GetAray)))))))
                    (:1
                      (Variables|Default|SetOnkilled true)
                      (for _array_element_7 (CustomItemWrap|GetWrappedMeshComponents)
                        (for _array_element_8 (Rendering|Material|GetMaterials _array_element_7)
                          (bind _asmaterial_instance_dynamic_2 (Utilities|Casting|CastToMaterialInstanceDynamic _array_element_8))
                          (Rendering|Material|SetTextureParameterValue _asmaterial_instance_dynamic_2 "TopS_01__Tex_Mask" "/BRCosmetics/Weapons/WeaponSkins/Wraps/Textures/skyRuled/T_Wrap_SkyRuled_EA_Msk_1.T_Wrap_SkyRuled_EA_Msk_1")
                          (Rendering|Material|SetTextureParameterValue _asmaterial_instance_dynamic_2 "TopS_02__Tex_Mask" "/BRCosmetics/Weapons/WeaponSkins/Wraps/Textures/skyRuled/T_Wrap_SkyRuled_EB_Msk_1.T_Wrap_SkyRuled_EB_Msk_1"))))
                    (:2
                      (Variables|Default|SetOnkilled true)
                      (for _array_element_9 (CustomItemWrap|GetWrappedMeshComponents)
                        (for _array_element_10 (Rendering|Material|GetMaterials _array_element_9)
                          (bind _asmaterial_instance_dynamic_3 (Utilities|Casting|CastToMaterialInstanceDynamic _array_element_10))
                          (Rendering|Material|SetTextureParameterValue _asmaterial_instance_dynamic_3 "TopS_01__Tex_Mask" "/BRCosmetics/Weapons/WeaponSkins/Wraps/Textures/skyRuled/T_Wrap_SkyRuled_C_Msk.T_Wrap_SkyRuled_C_Msk")
                          (Rendering|Material|SetTextureParameterValue _asmaterial_instance_dynamic_3 "TopS_02__Tex_Mask" "/BRCosmetics/Weapons/WeaponSkins/Wraps/Textures/skyRuled/T_Wrap_SkyRuled_B_Msk.T_Wrap_SkyRuled_B_Msk"))))))))))))
    (:then_1
      (Cosmetics|WatchCosmeticStat _returnvalue "(TagName=\"StatManager.AthenaKills\")" 0 false))))

(fn ConstructionScript)
