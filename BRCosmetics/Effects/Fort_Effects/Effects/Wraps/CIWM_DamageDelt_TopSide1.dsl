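; Weapon wrap cosmetic modifier that drives reactive material effects when the player deals damage.
; On pawn set, it watches the DamageDealt stat and stamps game time into named scalar parameters on the wrapped meshes' materials.
; TopSide and MultiDecal toggles each enable their own parameter list.
; In the front end, weapon fire stands in for damage via a counter that triggers the same effect on shots five and six, resetting on seven.
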(event EventOnWeaponFired
  (if (CustomCosmeticModifier|IsinFrontEnd)
    (bind _output_get (Variables|Envirometn_Varable|SetIterationCount (+ (Variables|Envirometn_Varable|GetIterationCount) 1)))
    (if (>= _output_get 5)
      (if (== (Variables|Envirometn_Varable|GetIterationCount) 7)
        (Variables|Envirometn_Varable|SetIterationCount 0)
        (else
          (Utilities|FlowControl|Sequence
            (:then_0
              (if (Variables|Default|GetTopSide)
                ;[1 1 1] TopsSide V2
                (for _array_element (Variables|Default|GetTopSide-Calls)
                  (Rendering|Material|SetScalarParameterValueonMaterials (CustomItemWrap|GetWrappedMeshComponents) _array_element (Utilities|Time|GetGameTimeinSeconds)))))
            (:then_1
              (if (Variables|Default|GetMultiDecal%20)
                ;[1 1 1 0.5] Decal Reactive
                (for _array_element_2 (Variables|Default|GetmultiDecal%20Calls)
                  (Rendering|Material|SetScalarParameterValueonMaterials (CustomItemWrap|GetWrappedMeshComponents) _array_element_2 (Utilities|Time|GetGameTimeinSeconds)))))))))))

(event EventOnPawnSet
  (Cosmetics|WatchCosmeticStat (CustomCharacter|GetAssociatedPlayerPawn) "(TagName=\"StatManager.DamageDealt\")"
    (:OnStatChanged
      (Utilities|FlowControl|Sequence
        (:then_0
          (if (Variables|Default|GetTopSide)
            ;[1 1 1] TopsSide V2
            (for _array_element (Variables|Default|GetTopSide-Calls)
              (Rendering|Material|SetScalarParameterValueonMaterials (CustomItemWrap|GetWrappedMeshComponents) _array_element (Utilities|Time|GetGameTimeinSeconds)))))
        (:then_1
          (if (Variables|Default|GetMultiDecal%20)
            ;[1 1 1 0.5] Decal Reactive
            (for _array_element_2 (Variables|Default|GetmultiDecal%20Calls)
              (Rendering|Material|SetScalarParameterValueonMaterials (CustomItemWrap|GetWrappedMeshComponents) _array_element_2 (Utilities|Time|GetGameTimeinSeconds)))))))))

(fn ConstructionScript)
