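; Item wrap modifier that builds up a top-side material blend and emissive effect as the player deals damage.
; Each DamageDealt stat change adds a capped step, after decaying the count by time elapsed past a pause, then writes blend start and game-time stamps to the wrapped meshes.
; Weapon changes set step cap and pause from clip size.
; In the front end, counted weapon-fire events simulate damage hits.
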(event EventOnWeaponFired
  (bind _returnvalue (Utilities|Time|GetGameTimeinSeconds))
  (bind _returnvalue_2 (CustomItemWrap|GetWrappedMeshComponents))
  (if (CustomCosmeticModifier|IsinFrontEnd)
    (bind _output_get (Variables|Default|SetIterationCount (+ (Variables|Default|GetIterationCount) 1)))
    (if (>= _output_get 5)
      (if (== (Variables|Default|GetIterationCount) 6)
        (Variables|Default|SetIterationCount 0))
      (bind _output_get_2 (Variables|Default|SetStepCount (Math|Integer|Min(Integer) (+ (Math|Float|Truncate (Math|Float|Lerp (Math|Conversions|ToFloat(Integer) (Variables|Default|GetStepCount)) 0.0 (Math|Float|Clamp(Float) (/ (- (Utilities|Time|GetGameTimeinSeconds) (Variables|Default|Getwait_timer)) 5.0)))) 1) (Variables|Default|GetMaxSteps))))
      (Rendering|Material|SetScalarParameterValueonMaterials _returnvalue_2 "TopS_01_Blend_Opacity-TBP" _returnvalue)
      (Rendering|Material|SetScalarParameterValueonMaterials _returnvalue_2 "TopS Emissive_02-Brightness-TBP" _returnvalue)
      (Rendering|Material|SetScalarParameterValueonMaterials _returnvalue_2 "TopS_01_Blend_Opacity---Start" (Math|Float|Lerp 0.0 1.0 (/ (Math|Conversions|ToFloat(Integer) _output_get_2) (Variables|Default|GetMaxSteps))))
      (Variables|Default|Setwait_timer (+ (Utilities|Time|GetGameTimeinSeconds) (Variables|Default|GetPauseTime))))))

(event EventBeginPlay
  (Utilities|FlowControl|Delay 0.2)
  (if (CustomCosmeticModifier|IsinFrontEnd)
    (Variables|Default|SetMaxSteps 5)
    (Variables|Default|SetPauseTime 2.0)
    (else
      (Variables|Default|SetMaxSteps 10))))

(event EventOnWeaponChanged
  (params (NewWeapon :type Fort%20Weapon%20Object%20Reference) (OldWeapon :type Fort%20Weapon%20Object%20Reference))
  (Utilities|FlowControl|Delay 0.2)
  (if (> (FortWeapon|GetBulletsPerClip (Variables|/Script/FortniteGame.FortPawn|GetCurrentWeapon (CustomCharacter|GetAssociatedPlayerPawn))) 5)
    (Variables|Default|SetMaxSteps 10)
    (Variables|Default|SetPauseTime 1.0)
    (else
      (Variables|Default|SetMaxSteps 5)
      (Variables|Default|SetPauseTime 2.0))))

(event EventOnNewWeapons
  (params (WeaponsAsItems :type Array%20of%20Fort%20Item%20Object%20References))
  (Utilities|FlowControl|Delay 0.2)
  (if (> (FortWeapon|GetBulletsPerClip (Variables|/Script/FortniteGame.FortPawn|GetCurrentWeapon (CustomCharacter|GetAssociatedPlayerPawn))) 5)
    (Variables|Default|SetMaxSteps 10)
    (Variables|Default|SetPauseTime 1.0)
    (else
      (Variables|Default|SetMaxSteps 5)
      (Variables|Default|SetPauseTime 2.0))))

(event EventOnNewPlayerWeaponType
  (params (NewType :type EFortItemType%20Enum :default "WorldItem") (OldType :type EFortItemType%20Enum :default "WorldItem"))
  (Utilities|FlowControl|Delay 0.2)
  (if (> (FortWeapon|GetBulletsPerClip (Variables|/Script/FortniteGame.FortPawn|GetCurrentWeapon (CustomCharacter|GetAssociatedPlayerPawn))) 5)
    (Variables|Default|SetMaxSteps 10)
    (Variables|Default|SetPauseTime 1.0)
    (else
      (Variables|Default|SetMaxSteps 5)
      (Variables|Default|SetPauseTime 2.0))))

(event EventOnPawnSet
  (bind _returnvalue (Utilities|Time|GetGameTimeinSeconds))
  (bind _returnvalue_2 (CustomItemWrap|GetWrappedMeshComponents))
  (Cosmetics|WatchCosmeticStat (CustomCharacter|GetAssociatedPlayerPawn) "(TagName=\"StatManager.DamageDealt\")"
    (:OnStatChanged
      (bind _output_get (Variables|Default|SetStepCount (Math|Integer|Min(Integer) (+ (Math|Float|Truncate (Math|Float|Lerp (Math|Conversions|ToFloat(Integer) (Variables|Default|GetStepCount)) 0.0 (Math|Float|Clamp(Float) (/ (- (Utilities|Time|GetGameTimeinSeconds) (Variables|Default|Getwait_timer)) 5.0)))) 1) (Variables|Default|GetMaxSteps))))
      (Rendering|Material|SetScalarParameterValueonMaterials _returnvalue_2 "TopS_01_Blend_Opacity-TBP" _returnvalue)
      (Rendering|Material|SetScalarParameterValueonMaterials _returnvalue_2 "TopS Emissive_02-Brightness-TBP" _returnvalue)
      (Rendering|Material|SetScalarParameterValueonMaterials _returnvalue_2 "TopS_01_Blend_Opacity---Start" (Math|Float|Lerp 0.0 1.0 (/ (Math|Conversions|ToFloat(Integer) _output_get) (Variables|Default|GetMaxSteps))))
      (Variables|Default|Setwait_timer (+ (Utilities|Time|GetGameTimeinSeconds) (Variables|Default|GetPauseTime))))))

(fn ConstructionScript)
