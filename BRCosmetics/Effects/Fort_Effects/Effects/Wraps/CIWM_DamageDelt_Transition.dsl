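; Reactive weapon wrap modifier that advances a material transition as the player's DamageDealt stat changes.
; Each unpaused stat change steps a counter mapped to start/end blend and emissive offsets, stamped with game time for material-side animation, then pauses briefly.
; Every change also timestamps emissive brightness.
; At maximum count, the next step resets and flips invert masks, alternating transition direction; weapon changes reapply the offsets.
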(event EventOnWeaponFired
  ; Frontnd reactive
  (if (CustomCosmeticModifier|IsinFrontEnd)
    (bind _output_get (Variables|EviromentVarables|Setweapon%20fire (+ (Variables|EviromentVarables|Getweapon%20fire) 1)))
    (if (>= _output_get 5)
      (if (== (Variables|EviromentVarables|Getweapon%20fire) 7)
        (Variables|EviromentVarables|Setweapon%20fire 0)))))

(event EventOnPawnSet
  (Cosmetics|WatchCosmeticStat (CustomCharacter|GetAssociatedPlayerPawn) "(TagName=\"StatManager.DamageDealt\")" 0 false
    (:OnStatChanged
      (if (Variables|EviromentVarables|GetPause)
        (CallFunction|ReactiveBrightness)
        (else
          (CallFunction|ResetCount)
          (Variables|EviromentVarables|SetPause true)
          (Variables|EviromentVarables|SetDamage_Delt (+ (Variables|EviromentVarables|GetDamage_Delt) 1))
          (CallFunction|DamageUpdate)
          (CallFunction|ReactiveBrightness))))))

(event Custom|Reset_Count
  (bind _returnvalue (CustomItemWrap|GetWrappedMeshComponents))
  (if (== (Variables|EviromentVarables|GetDamage_Delt_Last) (Variables|Default|GetDamage_MaxNumber))
    (Variables|EviromentVarables|SetDamage_Delt 0)
    (Variables|EviromentVarables|SetDamage_Delt_Last 0)
    (if (Variables|EviromentVarables|Getinvert_State)
      (Variables|EviromentVarables|Setinvert_State false)
      (Variables|EviromentVarables|Setinvert%20state 0.0)
      (Variables|EviromentVarables|Setinvert%20state_Emiss 1.0)
      (else
        (Variables|EviromentVarables|Setinvert_State true)
        (Variables|EviromentVarables|Setinvert%20state 1.0)
        (Variables|EviromentVarables|Setinvert%20state_Emiss 0.0)))
    (Rendering|Material|SetScalarParameterValueonMaterials _returnvalue "TopS Blend-InvertMsk   1=invert" (Variables|EviromentVarables|Getinvert%20state))
    (Rendering|Material|SetScalarParameterValueonMaterials _returnvalue "TopS Emissive_01_Blend 01---> 02 Invert Msk  1=inverted" (Variables|EviromentVarables|Getinvert%20state_Emiss))
    (Rendering|Material|SetScalarParameterValueonMaterials _returnvalue "TopS Emissive_02_Blend 01---> 02 Invert Msk  1=inverted" (Variables|EviromentVarables|Getinvert%20state_Emiss))
    (Rendering|Material|SetScalarParameterValueonMaterials _returnvalue "TopS Emissive_03_Blend 01---> 02 Invert Msk  1=inverted" (Variables|EviromentVarables|Getinvert%20state_Emiss))))

(event Custom|Damage_update
  (CallFunction|UpdateMaterial)
  (Utilities|FlowControl|Delay 0.01)
  (Variables|EviromentVarables|SetDamage_Delt_Last (+ (Variables|EviromentVarables|GetDamage_Delt_Last) 1))
  (Utilities|FlowControl|Delay (Variables|Default|GetPause%20Duration))
  (Variables|EviromentVarables|SetPause false))

(event EventOnNewWeapons
  (params (WeaponsAsItems :type Array%20of%20Fort%20Item%20Object%20References))
  (Utilities|FlowControl|Delay 0.05)
  (CallFunction|UpdateMaterial))

(event EventOnNewPlayerWeaponType
  (params (NewType :type EFortItemType%20Enum :default "WorldItem") (OldType :type EFortItemType%20Enum :default "WorldItem"))
  (Utilities|FlowControl|Delay 0.05)
  (CallFunction|UpdateMaterial))

(event EventOnWeaponChanged
  (params (NewWeapon :type Fort%20Weapon%20Object%20Reference) (OldWeapon :type Fort%20Weapon%20Object%20Reference))
  (Utilities|FlowControl|Delay 0.05)
  (CallFunction|UpdateMaterial))

(fn ReactiveBrightness
  (bind _returnvalue (CustomItemWrap|GetWrappedMeshComponents))
  (bind _returnvalue_2 (Utilities|Time|GetGameTimeinSeconds))
  (Rendering|Material|SetScalarParameterValueonMaterials _returnvalue "TopS Emissive_01-Brightness-TBP" _returnvalue_2)
  (Rendering|Material|SetScalarParameterValueonMaterials _returnvalue "TopS Emissive_02-Brightness-TBP" _returnvalue_2)
  (Rendering|Material|SetScalarParameterValueonMaterials _returnvalue "TopS Emissive_03-Brightness-TBP" _returnvalue_2))

(fn UpdateMaterial
  (bind _damage_maxnumber (Variables|Default|GetDamage_MaxNumber))
  (bind _damage_delt_last (Variables|EviromentVarables|GetDamage_Delt_Last))
  (bind _returnvalue (Math|Conversions|ToFloat(Integer) _damage_delt_last))
  (bind _returnvalue_2 (/ _returnvalue _damage_maxnumber))
  (bind _offset_end_pos (Variables|Default|GetOffset%20End%20pos))
  (bind _offset_start_pos (Variables|Default|GetOffset%20Start%20pos))
  (bind _returnvalue_3 (Math|Float|Lerp _offset_start_pos _offset_end_pos _returnvalue_2))
  (bind _returnvalue_4 (CustomItemWrap|GetWrappedMeshComponents))
  (bind _damage_maxnumber_2 (Variables|Default|GetDamage_MaxNumber))
  (bind _damage_delt (Variables|EviromentVarables|GetDamage_Delt))
  (bind _returnvalue_5 (Math|Conversions|ToFloat(Integer) _damage_delt))
  (bind _returnvalue_6 (/ _returnvalue_5 _damage_maxnumber_2))
  (bind _offset_end_pos_2 (Variables|Default|GetOffset%20End%20pos))
  (bind _offset_start_pos_2 (Variables|Default|GetOffset%20Start%20pos))
  (bind _returnvalue_7 (Math|Float|Lerp _offset_start_pos_2 _offset_end_pos_2 _returnvalue_6))
  (bind _returnvalue_8 (Utilities|Time|GetGameTimeinSeconds))
  (Rendering|Material|SetScalarParameterValueonMaterials _returnvalue_4 "TopS_01_Blend_Offset---End" _returnvalue_7)
  (Rendering|Material|SetScalarParameterValueonMaterials _returnvalue_4 "TopS_01_Blend_Offset---Start" _returnvalue_3)
  (Rendering|Material|SetScalarParameterValueonMaterials _returnvalue_4 "TopS_01_Blend_Offset-TBP" _returnvalue_8)
  (Rendering|Material|SetScalarParameterValueonMaterials _returnvalue_4 "TopS Emissive_01_Offset---End" _returnvalue_7)
  (Rendering|Material|SetScalarParameterValueonMaterials _returnvalue_4 "TopS Emissive_01_Offset---Start" _returnvalue_3)
  (Rendering|Material|SetScalarParameterValueonMaterials _returnvalue_4 "TopS Emissive_01_Offset----TBP" _returnvalue_8)
  (Rendering|Material|SetScalarParameterValueonMaterials _returnvalue_4 "TopS Emissive_02_Offset---End" _returnvalue_7)
  (Rendering|Material|SetScalarParameterValueonMaterials _returnvalue_4 "TopS Emissive_02_Offset---Start" _returnvalue_3)
  (Rendering|Material|SetScalarParameterValueonMaterials _returnvalue_4 "TopS Emissive_02_Offset----TBP" _returnvalue_8)
  (Rendering|Material|SetScalarParameterValueonMaterials _returnvalue_4 "TopS Emissive_03_Offset---End" _returnvalue_7)
  (Rendering|Material|SetScalarParameterValueonMaterials _returnvalue_4 "TopS Emissive_03_Offset---Start" _returnvalue_3)
  (Rendering|Material|SetScalarParameterValueonMaterials _returnvalue_4 "TopS Emissive_03_Offset----TBP" _returnvalue_8))

(fn ConstructionScript)
