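; Cosmetic wrap modifier for the TopSide wrap that makes its materials react when the player takes damage.
; On pawn set, it watches the DamageReceived cosmetic stat and, on change, stamps game time into emissive, offset, blend opacity and time scalars on wrapped meshes.
; In the front end, a timeline event track sends the same stamps instead, with no damage needed.
; The weapon-fired event is empty.
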(event EventBeginPlay
  (bind _returnvalue (Utilities|Time|GetGameTimeinSeconds))
  (bind _returnvalue_2 (CustomItemWrap|GetWrappedMeshComponents))
  (if (CustomCosmeticModifier|IsinFrontEnd)
    (Timeline|Timeline PlayFromStart
      (:NewTrack_0
        (Utilities|FlowControl|Sequence
          (:then_0
            ;[1 1 1] TopsSide V2
            (Rendering|Material|SetScalarParameterValueonMaterials _returnvalue_2 "TopS Emissive_02-Brightness-TBP" _returnvalue)
            (Rendering|Material|SetScalarParameterValueonMaterials _returnvalue_2 "TopS Emissive_02_Offset----TBP" _returnvalue)
            (Rendering|Material|SetScalarParameterValueonMaterials _returnvalue_2 "TopS_01_Blend_Opacity-TBP" _returnvalue)
            (Rendering|Material|SetScalarParameterValueonMaterials _returnvalue_2 "TimeFrom-BP----------------" _returnvalue))
          (:then_1
            (Rendering|Material|SetScalarParameterValueonMaterials _returnvalue_2 "TopS Emissive_01-Brightness-TBP" _returnvalue)
            (Rendering|Material|SetScalarParameterValueonMaterials _returnvalue_2 "TopS Emissive_01_Offset----TBP" _returnvalue)
            (Rendering|Material|SetScalarParameterValueonMaterials _returnvalue_2 "TopS Emissive_03-Brightness-TBP" _returnvalue)
            (Rendering|Material|SetScalarParameterValueonMaterials _returnvalue_2 "TopS Emissive_03_Offset----TBP" _returnvalue)))))))

(event EventOnWeaponFired)

(event EventOnPawnSet
  (bind _returnvalue (Utilities|Time|GetGameTimeinSeconds))
  (bind _returnvalue_2 (CustomItemWrap|GetWrappedMeshComponents))
  (Cosmetics|WatchCosmeticStat (CustomCharacter|GetAssociatedPlayerPawn) "(TagName=\"StatManager.DamageReceived\")" 0 false
    (:OnStatChanged
      (Utilities|FlowControl|Sequence
        (:then_0
          ;[1 1 1] TopsSide V2
          (Rendering|Material|SetScalarParameterValueonMaterials _returnvalue_2 "TopS Emissive_02-Brightness-TBP" _returnvalue)
          (Rendering|Material|SetScalarParameterValueonMaterials _returnvalue_2 "TopS Emissive_02_Offset----TBP" _returnvalue)
          (Rendering|Material|SetScalarParameterValueonMaterials _returnvalue_2 "TopS_01_Blend_Opacity-TBP" _returnvalue)
          (Rendering|Material|SetScalarParameterValueonMaterials _returnvalue_2 "TimeFrom-BP----------------" _returnvalue))
        (:then_1
          (Rendering|Material|SetScalarParameterValueonMaterials _returnvalue_2 "TopS Emissive_01-Brightness-TBP" _returnvalue)
          (Rendering|Material|SetScalarParameterValueonMaterials _returnvalue_2 "TopS Emissive_01_Offset----TBP" _returnvalue)
          (Rendering|Material|SetScalarParameterValueonMaterials _returnvalue_2 "TopS Emissive_03-Brightness-TBP" _returnvalue)
          (Rendering|Material|SetScalarParameterValueonMaterials _returnvalue_2 "TopS Emissive_03_Offset----TBP" _returnvalue))))))

(fn ConstructionScript)
