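; Reactive weapon wrap modifier that animates TopSide and decal material effects on damage dealt and on kills.
; On pawn set it watches both stats, stamping game time into scalar parameters on wrapped meshes: damage drives emissive brightness and gradients, kills drive TopSide blend/speed and decal timers.
; In the front end, a shot counter simulates both reactions, firing damage on shots five and six, then kill on seven before resetting.
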(event EventOnWeaponFired
  (bind _returnvalue (Utilities|Time|GetGameTimeinSeconds))
  (bind _returnvalue_2 (Utilities|Time|GetGameTimeinSeconds))
  (bind _returnvalue_3 (CustomItemWrap|GetWrappedMeshComponents))
  (bind _returnvalue_4 (Utilities|Time|GetGameTimeinSeconds))
  (bind _returnvalue_5 (CustomItemWrap|GetWrappedMeshComponents))
  (bind _returnvalue_6 (CustomItemWrap|GetWrappedMeshComponents))
  (bind _returnvalue_7 (Utilities|Time|GetGameTimeinSeconds))
  (bind _returnvalue_8 (CustomItemWrap|GetWrappedMeshComponents))
  (bind _returnvalue_9 (Utilities|Time|GetGameTimeinSeconds))
  (bind _returnvalue_10 (CustomItemWrap|GetWrappedMeshComponents))
  (if (CustomCosmeticModifier|IsinFrontEnd)
    (Utilities|FlowControl|Delay 0.001)
    (bind _output_get (Variables|Default|SetShotsFired (+ (Variables|Default|GetShotsFired) 1)))
    (if (>= _output_get 5)
      (if (== (Variables|Default|GetShotsFired) 7)
        (Variables|Default|SetShotsFired 0)
        (Utilities|FlowControl|Sequence
          (:then_0
            ;[1 1 1] TopsSide V2
            (Rendering|Material|SetScalarParameterValueonMaterials _returnvalue_10 "TopS_01_Blend_Opacity-TBP" _returnvalue)
            (Rendering|Material|SetScalarParameterValueonMaterials _returnvalue_10 "TopS_01_Blend_Offset-TBP" _returnvalue)
            (Rendering|Material|SetScalarParameterValueonMaterials _returnvalue_10 "TopS_02_Speed-TBP" _returnvalue)
            (Rendering|Material|SetScalarParameterValueonMaterials _returnvalue_10 "TopS_03_Speed-TBP" _returnvalue)
            (Rendering|Material|SetScalarParameterValueonMaterials _returnvalue_10 "TopS_04_Speed-TBP" _returnvalue))
          (:then_1
            ;[1 1 1 0.5] Decal Reactive
            (Rendering|Material|SetScalarParameterValueonMaterials _returnvalue_6 "EDRec_TimeFrom_BP_DecalA" _returnvalue_7)
            (Rendering|Material|SetScalarParameterValueonMaterials _returnvalue_6 "EDRec_TimeFrom_BP_DecalB" _returnvalue_7)
            (Rendering|Material|SetScalarParameterValueonMaterials _returnvalue_6 "EDRec_TimeFrom_BP_DecalC" _returnvalue_7)
            (Rendering|Material|SetScalarParameterValueonMaterials _returnvalue_6 "EDRec_TimeFrom_BP_DecalD" _returnvalue_7)))
        (else
          (Rendering|Material|SetScalarParameterValueonMaterials _returnvalue_3 "TopS Emissive_02-Brightness-TBP" _returnvalue_2)
          (Rendering|Material|SetScalarParameterValueonMaterials _returnvalue_3 "TopS Emissive_02_Offset----TBP" _returnvalue_2)
          (Rendering|Material|SetScalarParameterValueonMaterials _returnvalue_3 "TopS Emissive_01-Brightness-TBP" _returnvalue_2)
          (Rendering|Material|SetScalarParameterValueonMaterials _returnvalue_3 "TopS Emissive_01_Offset----TBP" _returnvalue_2)
          (Rendering|Material|SetScalarParameterValueonMaterials _returnvalue_3 "TopS Emissive_03-Brightness-TBP" _returnvalue_2)
          (Rendering|Material|SetScalarParameterValueonMaterials _returnvalue_3 "TopS Emissive_03_Offset----TBP" _returnvalue_2)
          (Rendering|Material|SetScalarParameterValueonMaterials _returnvalue_8 "EDRec_Emissive_A_Brightness_Time_BP" _returnvalue_9)
          (Rendering|Material|SetScalarParameterValueonMaterials _returnvalue_8 "EDRec_Emissive_B_Brightness_Time_BP" _returnvalue_9)
          (Rendering|Material|SetScalarParameterValueonMaterials _returnvalue_8 "EDRec_Emissive_C_Brightness_Time_BP" _returnvalue_9)
          (Rendering|Material|SetScalarParameterValueonMaterials _returnvalue_8 "EDRec_Emissive_D_Brightness_Time_BP" _returnvalue_9)
          (Rendering|Material|SetScalarParameterValueonMaterials _returnvalue_5 "EDRec_Emissive_A_Gradent_Time_BP" _returnvalue_4)
          (Rendering|Material|SetScalarParameterValueonMaterials _returnvalue_5 "EDRec_Emissive_B_Gradent_Time_BP" _returnvalue_4)
          (Rendering|Material|SetScalarParameterValueonMaterials _returnvalue_5 "EDRec_Emissive_C_Gradent_Time_BP" _returnvalue_4)
          (Rendering|Material|SetScalarParameterValueonMaterials _returnvalue_5 "EDRec_Emissive_D_Gradent_Time_BP" _returnvalue_4))))))

(event EventOnPawnSet
  (bind _returnvalue (Utilities|Time|GetGameTimeinSeconds))
  (bind _returnvalue_2 (Utilities|Time|GetGameTimeinSeconds))
  (bind _returnvalue_3 (CustomItemWrap|GetWrappedMeshComponents))
  (bind _returnvalue_4 (Utilities|Time|GetGameTimeinSeconds))
  (bind _returnvalue_5 (CustomItemWrap|GetWrappedMeshComponents))
  (bind _returnvalue_6 (CustomItemWrap|GetWrappedMeshComponents))
  (bind _returnvalue_7 (Utilities|Time|GetGameTimeinSeconds))
  (bind _returnvalue_8 (CustomItemWrap|GetWrappedMeshComponents))
  (bind _returnvalue_9 (Utilities|Time|GetGameTimeinSeconds))
  (bind _returnvalue_10 (CustomItemWrap|GetWrappedMeshComponents))
  (Utilities|FlowControl|Sequence
    (:then_0
      (Cosmetics|WatchCosmeticStat (CustomCharacter|GetAssociatedPlayerPawn) "(TagName=\"StatManager.DamageDealt\")" 0 false
        (:OnStatChanged
          (Rendering|Material|SetScalarParameterValueonMaterials _returnvalue_3 "TopS Emissive_02-Brightness-TBP" _returnvalue_2)
          (Rendering|Material|SetScalarParameterValueonMaterials _returnvalue_3 "TopS Emissive_02_Offset----TBP" _returnvalue_2)
          (Rendering|Material|SetScalarParameterValueonMaterials _returnvalue_3 "TopS Emissive_01-Brightness-TBP" _returnvalue_2)
          (Rendering|Material|SetScalarParameterValueonMaterials _returnvalue_3 "TopS Emissive_01_Offset----TBP" _returnvalue_2)
          (Rendering|Material|SetScalarParameterValueonMaterials _returnvalue_3 "TopS Emissive_03-Brightness-TBP" _returnvalue_2)
          (Rendering|Material|SetScalarParameterValueonMaterials _returnvalue_3 "TopS Emissive_03_Offset----TBP" _returnvalue_2)
          (Rendering|Material|SetScalarParameterValueonMaterials _returnvalue_8 "EDRec_Emissive_A_Brightness_Time_BP" _returnvalue_9)
          (Rendering|Material|SetScalarParameterValueonMaterials _returnvalue_8 "EDRec_Emissive_B_Brightness_Time_BP" _returnvalue_9)
          (Rendering|Material|SetScalarParameterValueonMaterials _returnvalue_8 "EDRec_Emissive_C_Brightness_Time_BP" _returnvalue_9)
          (Rendering|Material|SetScalarParameterValueonMaterials _returnvalue_8 "EDRec_Emissive_D_Brightness_Time_BP" _returnvalue_9)
          (Rendering|Material|SetScalarParameterValueonMaterials _returnvalue_5 "EDRec_Emissive_A_Gradent_Time_BP" _returnvalue_4)
          (Rendering|Material|SetScalarParameterValueonMaterials _returnvalue_5 "EDRec_Emissive_B_Gradent_Time_BP" _returnvalue_4)
          (Rendering|Material|SetScalarParameterValueonMaterials _returnvalue_5 "EDRec_Emissive_C_Gradent_Time_BP" _returnvalue_4)
          (Rendering|Material|SetScalarParameterValueonMaterials _returnvalue_5 "EDRec_Emissive_D_Gradent_Time_BP" _returnvalue_4))))
    (:then_1
      (Cosmetics|WatchCosmeticStat (CustomCharacter|GetAssociatedPlayerPawn) "(TagName=\"StatManager.AthenaKills\")" 0 false
        (:OnStatChanged
          (Utilities|FlowControl|Sequence
            (:then_0
              ;[1 1 1] TopsSide V2
              (Rendering|Material|SetScalarParameterValueonMaterials _returnvalue_10 "TopS_01_Blend_Opacity-TBP" _returnvalue)
              (Rendering|Material|SetScalarParameterValueonMaterials _returnvalue_10 "TopS_01_Blend_Offset-TBP" _returnvalue)
              (Rendering|Material|SetScalarParameterValueonMaterials _returnvalue_10 "TopS_02_Speed-TBP" _returnvalue)
              (Rendering|Material|SetScalarParameterValueonMaterials _returnvalue_10 "TopS_03_Speed-TBP" _returnvalue)
              (Rendering|Material|SetScalarParameterValueonMaterials _returnvalue_10 "TopS_04_Speed-TBP" _returnvalue))
            (:then_1
              ;[1 1 1 0.5] Decal Reactive
              (Rendering|Material|SetScalarParameterValueonMaterials _returnvalue_6 "EDRec_TimeFrom_BP_DecalA" _returnvalue_7)
              (Rendering|Material|SetScalarParameterValueonMaterials _returnvalue_6 "EDRec_TimeFrom_BP_DecalB" _returnvalue_7)
              (Rendering|Material|SetScalarParameterValueonMaterials _returnvalue_6 "EDRec_TimeFrom_BP_DecalC" _returnvalue_7)
              (Rendering|Material|SetScalarParameterValueonMaterials _returnvalue_6 "EDRec_TimeFrom_BP_DecalD" _returnvalue_7))))))))

(fn ConstructionScript)
