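; Item wrap cosmetic Blueprint that animates four material decals (A-D) when the pawn's DamageReceived stat changes.
; Each trigger writes game time to one decal's time and bounce-rotation parameters on wrapped meshes, stepping through a shuffled index array that reshuffles after four.
; In the front end, a timeline event track drives the same sequence.
; RandomSet randomizes flipbook frame, offset and rotation but is not called here; OnWeaponFired only delays.
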(event EventBeginPlay
  (bind _returnvalue (Utilities|Time|GetGameTimeinSeconds))
  (bind _returnvalue_2 (CustomItemWrap|GetWrappedMeshComponents))
  (Utilities|FlowControl|Delay 0.2)
  (Utilities|Array|Shuffle (Variables|Default|GetAray))
  (if (CustomCosmeticModifier|IsinFrontEnd)
    (Timeline|Timeline PlayFromStart
      (:NewTrack_0
        (bind _output_get (Variables|Default|SetFire%20seq (Variables|Default|GetFire%20seq)))
        (switch int (Utilities|Array|Get(aref) (Variables|Default|GetAray) _output_get)
          (:0
            (Rendering|Material|SetScalarParameterValueonMaterials _returnvalue_2 "EDRec_TimeFrom_BP_DecalA" _returnvalue)
            (Rendering|Material|SetScalarParameterValueonMaterials _returnvalue_2 "EDRec_DecalA_BounceRot_TimeFrom_BP" _returnvalue))
          (:1
            (Rendering|Material|SetScalarParameterValueonMaterials _returnvalue_2 "EDRec_TimeFrom_BP_DecalB" _returnvalue)
            (Rendering|Material|SetScalarParameterValueonMaterials _returnvalue_2 "EDRec_DecalB_BounceRot_TimeFrom_BP" _returnvalue))
          (:2
            (Rendering|Material|SetScalarParameterValueonMaterials _returnvalue_2 "EDRec_TimeFrom_BP_DecalC" _returnvalue)
            (Rendering|Material|SetScalarParameterValueonMaterials _returnvalue_2 "EDRec_DecalC_BounceRot_TimeFrom_BP" _returnvalue))
          (:3
            (Rendering|Material|SetScalarParameterValueonMaterials _returnvalue_2 "EDRec_TimeFrom_BP_DecalD" _returnvalue)
            (Rendering|Material|SetScalarParameterValueonMaterials _returnvalue_2 "EDRec_DecalD_BounceRot_TimeFrom_BP" _returnvalue)))
        (bind _output_get_2 (Variables|Default|SetFire%20seq (+ (Variables|Default|GetFire%20seq) 1)))
        (if (== _output_get_2 4)
          (Utilities|Array|Shuffle (Variables|Default|GetAray))
          (Variables|Default|SetFire%20seq 0))))))

(event EventOnPawnSet
  (bind _returnvalue (Utilities|Time|GetGameTimeinSeconds))
  (bind _returnvalue_2 (CustomItemWrap|GetWrappedMeshComponents))
  (Cosmetics|WatchCosmeticStat (CustomCharacter|GetAssociatedPlayerPawn) "(TagName=\"StatManager.DamageReceived\")" 0 false
    (:OnStatChanged
      (bind _output_get (Variables|Default|SetFire%20seq (Variables|Default|GetFire%20seq)))
      (switch int (Utilities|Array|Get(aref) (Variables|Default|GetAray) _output_get)
        (:0
          (Rendering|Material|SetScalarParameterValueonMaterials _returnvalue_2 "EDRec_TimeFrom_BP_DecalA" _returnvalue)
          (Rendering|Material|SetScalarParameterValueonMaterials _returnvalue_2 "EDRec_DecalA_BounceRot_TimeFrom_BP" _returnvalue))
        (:1
          (Rendering|Material|SetScalarParameterValueonMaterials _returnvalue_2 "EDRec_TimeFrom_BP_DecalB" _returnvalue)
          (Rendering|Material|SetScalarParameterValueonMaterials _returnvalue_2 "EDRec_DecalB_BounceRot_TimeFrom_BP" _returnvalue))
        (:2
          (Rendering|Material|SetScalarParameterValueonMaterials _returnvalue_2 "EDRec_TimeFrom_BP_DecalC" _returnvalue)
          (Rendering|Material|SetScalarParameterValueonMaterials _returnvalue_2 "EDRec_DecalC_BounceRot_TimeFrom_BP" _returnvalue))
        (:3
          (Rendering|Material|SetScalarParameterValueonMaterials _returnvalue_2 "EDRec_TimeFrom_BP_DecalD" _returnvalue)
          (Rendering|Material|SetScalarParameterValueonMaterials _returnvalue_2 "EDRec_DecalD_BounceRot_TimeFrom_BP" _returnvalue)))
      (bind _output_get_2 (Variables|Default|SetFire%20seq (+ (Variables|Default|GetFire%20seq) 1)))
      (if (== _output_get_2 4)
        (Utilities|Array|Shuffle (Variables|Default|GetAray))
        (Variables|Default|SetFire%20seq 0)))))

(event EventOnWeaponFired
  (Utilities|FlowControl|Delay 0.01))

(fn RandomSet
  (params (FlipBook%20Frame :type Name) (TimeBP :type Name) (Offset :type Name) (Rotation :type Name) (X=min :type Float%20%28double-precision%29) (X=Max :type Float%20%28double-precision%29) (Y=Min :type Float%20%28double-precision%29) (Y=Max :type Float%20%28double-precision%29) (Rot=Min :type Float%20%28double-precision%29) (Rot=Max :type Float%20%28double-precision%29) (TimeBP-first :type Name))
  (bind _returnvalue (CustomItemWrap|GetWrappedMeshComponents))
  (bind _stream (Variables|Default|GetStream))
  (bind _returnvalue_2 (Math|Random|RandomFloatinRangefromStream _stream X=min X=Max))
  (bind _returnvalue_3 (Math|Random|RandomIntegerinRange 0 8))
  (bind _returnvalue_4 (Math|Conversions|ToFloat(Integer) _returnvalue_3))
  (bind _returnvalue_5 (/ _returnvalue_4 8.0))
  (Rendering|Material|SetScalarParameterValueonMaterials _returnvalue FlipBook%20Frame _returnvalue_5)
  (Rendering|Material|SetScalarParameterValueonMaterials _returnvalue TimeBP-first (Utilities|Time|GetGameTimeinSeconds))
  (Rendering|Material|SetVectorParameterValueonMaterials _returnvalue Offset (Math|Vector|MakeVector _returnvalue_2 (Math|Random|RandomFloatinRangefromStream _stream Y=Min Y=Max)))
  (Rendering|Material|SetScalarParameterValueonMaterials _returnvalue Rotation (Math|Random|RandomFloatinRange Rot=Min Rot=Max))
  (Rendering|Material|SetScalarParameterValueonMaterials _returnvalue TimeBP (Utilities|Time|GetGameTimeinSeconds))
  (Variables|Default|SetPosition _returnvalue_2)
  (Development|PrintString (Utilities|String|ToString(Float) _returnvalue_5) true true "(R=0.060001,G=1.000000,B=0.000000,A=1.000000)" 3.0))

(fn ConstructionScript)
