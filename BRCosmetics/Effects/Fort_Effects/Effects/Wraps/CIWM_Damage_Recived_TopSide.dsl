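; Cosmetic wrap modifier that triggers material effects on wrapped meshes when the player takes damage.
; Once the pawn is set, it watches the DamageReceived stat and writes current game time into material scalar parameters on each change.
; TopSide and LeftRight flags select emissive or flipbook-decal parameter sets, leaving animation to the materials.
; In the front end, after a short delay, a rate-adjusted timeline's event track fires the same writes as preview.
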(event EventBeginPlay
  (bind _returnvalue (Utilities|Time|GetGameTimeinSeconds))
  (bind _returnvalue_2 (CustomItemWrap|GetWrappedMeshComponents))
  (bind _returnvalue_3 (CustomItemWrap|GetWrappedMeshComponents))
  (bind _returnvalue_4 (Utilities|Time|GetGameTimeinSeconds))
  (Utilities|FlowControl|Delay 0.2)
  (if (CustomCosmeticModifier|IsinFrontEnd)
    (Components|Timeline|SetPlayRate (Variables|CIWM_Damage_Recived_TopSide|GetTimeline) (Variables|Default|GetFrontEnd%20PlayRate))
    (Timeline|Timeline PlayFromStart
      (:NewTrack_0
        (Utilities|FlowControl|Sequence
          (:then_0
            (if (Variables|Default|GetTopSide)
              (Utilities|FlowControl|Sequence
                (:then_0
                  ;[1 1 1] TopsSide V2
                  (Rendering|Material|SetScalarParameterValueonMaterials _returnvalue_2 "TopS Emissive_02-Brightness-TBP" _returnvalue)
                  (Rendering|Material|SetScalarParameterValueonMaterials _returnvalue_2 "TopS Emissive_02_Offset----TBP" _returnvalue)
                  (Rendering|Material|SetScalarParameterValueonMaterials _returnvalue_2 "TopS_01_Blend_Opacity-TBP" _returnvalue)
                  (Rendering|Material|SetScalarParameterValueonMaterials _returnvalue_2 "TimeFrom-BP----------------" _returnvalue))
                (:then_1
                  (Rendering|Material|SetScalarParameterValueonMaterials _returnvalue_2 "TopS Emissive_01-Brightness-TBP" _returnvalue)
                  (Rendering|Material|SetScalarParameterValueonMaterials _returnvalue_2 "TopS Emissive_01_Offset----TBP" _returnvalue)
                  (Rendering|Material|SetScalarParameterValueonMaterials _returnvalue_2 "TopS Emissive_03-Brightness-TBP" _returnvalue)
                  (Rendering|Material|SetScalarParameterValueonMaterials _returnvalue_2 "TopS Emissive_03_Offset----TBP" _returnvalue)))))
          (:then_1
            (if (Variables|Default|GetLeftRight)
              ; Left\Right V2
              (Rendering|Material|SetScalarParameterValueonMaterials _returnvalue_3 "FDecalV2_FlipBook_Time_FromBP" _returnvalue_4)
              (Rendering|Material|SetScalarParameterValueonMaterials _returnvalue_3 "FDecalaV2_Opacity_Time_BP_Z" _returnvalue_4)
              (Rendering|Material|SetScalarParameterValueonMaterials _returnvalue_3 "FDecalV2_Emissive_Brightness_Time_BP" _returnvalue_4)
              (Rendering|Material|SetScalarParameterValueonMaterials _returnvalue_3 "FDecalV2_Emissive_Gradent_Time_BP" _returnvalue_4)
              (Rendering|Material|SetScalarParameterValueonMaterials _returnvalue_3 "FDecalV2_Scale_Time_BP" _returnvalue_4))))))))

(event EventOnPawnSet
  (bind _returnvalue (Utilities|Time|GetGameTimeinSeconds))
  (bind _returnvalue_2 (CustomItemWrap|GetWrappedMeshComponents))
  (bind _returnvalue_3 (CustomItemWrap|GetWrappedMeshComponents))
  (bind _returnvalue_4 (Utilities|Time|GetGameTimeinSeconds))
  (Cosmetics|WatchCosmeticStat (CustomCharacter|GetAssociatedPlayerPawn) "(TagName=\"StatManager.DamageReceived\")" 0 false
    (:OnStatChanged
      (Utilities|FlowControl|Sequence
        (:then_0
          (if (Variables|Default|GetTopSide)
            (Utilities|FlowControl|Sequence
              (:then_0
                ;[1 1 1] TopsSide V2
                (Rendering|Material|SetScalarParameterValueonMaterials _returnvalue_2 "TopS Emissive_02-Brightness-TBP" _returnvalue)
                (Rendering|Material|SetScalarParameterValueonMaterials _returnvalue_2 "TopS Emissive_02_Offset----TBP" _returnvalue)
                (Rendering|Material|SetScalarParameterValueonMaterials _returnvalue_2 "TopS_01_Blend_Opacity-TBP" _returnvalue)
                (Rendering|Material|SetScalarParameterValueonMaterials _returnvalue_2 "TimeFrom-BP----------------" _returnvalue))
              (:then_1
                (Rendering|Material|SetScalarParameterValueonMaterials _returnvalue_2 "TopS Emissive_01-Brightness-TBP" _returnvalue)
                (Rendering|Material|SetScalarParameterValueonMaterials _returnvalue_2 "TopS Emissive_01_Offset----TBP" _returnvalue)
                (Rendering|Material|SetScalarParameterValueonMaterials _returnvalue_2 "TopS Emissive_03-Brightness-TBP" _returnvalue)
                (Rendering|Material|SetScalarParameterValueonMaterials _returnvalue_2 "TopS Emissive_03_Offset----TBP" _returnvalue)))))
        (:then_1
          (if (Variables|Default|GetLeftRight)
            ; Left\Right V2
            (Rendering|Material|SetScalarParameterValueonMaterials _returnvalue_3 "FDecalV2_FlipBook_Time_FromBP" _returnvalue_4)
            (Rendering|Material|SetScalarParameterValueonMaterials _returnvalue_3 "FDecalaV2_Opacity_Time_BP_Z" _returnvalue_4)
            (Rendering|Material|SetScalarParameterValueonMaterials _returnvalue_3 "FDecalV2_Emissive_Brightness_Time_BP" _returnvalue_4)
            (Rendering|Material|SetScalarParameterValueonMaterials _returnvalue_3 "FDecalV2_Emissive_Gradent_Time_BP" _returnvalue_4)
            (Rendering|Material|SetScalarParameterValueonMaterials _returnvalue_3 "FDecalV2_Scale_Time_BP" _returnvalue_4)))))))

(fn ConstructionScript)
