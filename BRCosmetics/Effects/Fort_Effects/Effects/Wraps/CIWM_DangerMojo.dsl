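; Weapon wrap cosmetic modifier for DangerMojo that makes the wrap's emissive layers react to the player's kill count.
; A watched kill stat, or a timeline preview in the front end, updates the count and last kill time.
; PushMaterialParameters maps up to six kills to an emissive offset and feeds kill time to three brightness parameters on wrapped meshes.
; Weapon changes re-push after a short delay with a far-past time, so they do not register as fresh kills.
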(event EventOnWeaponFired)

(event EventBeginPlay
  (if (CustomCosmeticModifier|IsinFrontEnd)
    (bind (_direction _newtrack_1) (Timeline|Timeline PlayFromStart)
      (:NewTrack_0
        (Variables|Default|SetKill_Count (Math|Float|Truncate _newtrack_1))
        (Variables|Default|SetLastKillTime (Utilities|Time|GetGameTimeinSeconds))
        (CallFunction|PushMaterialParameters)))))

(event EventOnPawnSet
  (bind (_statname _statvalue) (Cosmetics|WatchCosmeticStat (CustomCharacter|GetAssociatedPlayerPawn) "(TagName=\"StatManager.AthenaKills\")" 0 false)
    (:OnStatChanged
      (Variables|Default|SetKill_Count _statvalue)
      (Variables|Default|SetLastKillTime (Utilities|Time|GetGameTimeinSeconds))
      (CallFunction|PushMaterialParameters))))

(event Custom|Push%20Material%20Parameters
  (bind _returnvalue (CustomItemWrap|GetWrappedMeshComponents))
  (bind _lastkilltime (Variables|Default|GetLastKillTime))
  (Rendering|Material|SetScalarParameterValueonMaterials _returnvalue "TopS Emissive_02_Offset---End" (Math|Float|Lerp -0.06 1.0 (/ (Math|Conversions|ToFloat(Integer) (Math|Integer|Clamp(Integer) (Variables|Default|GetKill_Count) 0 6)) 6.0)))
  (Rendering|Material|SetScalarParameterValueonMaterials _returnvalue "TopS Emissive_02-Brightness-TBP" _lastkilltime)
  (Rendering|Material|SetScalarParameterValueonMaterials _returnvalue "TopS Emissive_01-Brightness-TBP" _lastkilltime)
  (Rendering|Material|SetScalarParameterValueonMaterials _returnvalue "TopS Emissive_03-Brightness-TBP" _lastkilltime))

(event EventOnPostResIn
  (CallFunction|PushMaterialParameters))

(event EventOnNewWeapons
  (params (WeaponsAsItems :type Array%20of%20Fort%20Item%20Object%20References))
  (Utilities|FlowControl|Delay 0.05)
  ;[1 1 1] Updated kill count on new weapon
  (Variables|Default|SetLastKillTime -999999999999.0)
  (CallFunction|PushMaterialParameters))

(event EventOnNewPlayerWeaponType
  (params (NewType :type EFortItemType%20Enum :default "WorldItem") (OldType :type EFortItemType%20Enum :default "WorldItem"))
  (Utilities|FlowControl|Delay 0.05)
  ;[1 1 1] Updated kill count on new weapon
  (Variables|Default|SetLastKillTime -999999999999.0)
  (CallFunction|PushMaterialParameters))

(event EventOnWeaponChanged
  (params (NewWeapon :type Fort%20Weapon%20Object%20Reference) (OldWeapon :type Fort%20Weapon%20Object%20Reference))
  (Utilities|FlowControl|Delay 0.05)
  ;[1 1 1] Updated kill count on new weapon
  (Variables|Default|SetLastKillTime -999999999999.0)
  (CallFunction|PushMaterialParameters))

(event EventApplyMaterialParameters
  (CallFunction|PushMaterialParameters))

(fn CountPerctange
  (params (End%20count :type Float%20%28double-precision%29) (StartCount :type Float%20%28double-precision%29) (Character%20offset :type Float%20%28single-precision%29) (BackPack%20offset :type Float%20%28double-precision%29))
  (Variables|Default|SetEnd (Math|Float|Lerp 0.0 1.0 (/ End%20count 6.0))))

(fn ConstructionScript)
