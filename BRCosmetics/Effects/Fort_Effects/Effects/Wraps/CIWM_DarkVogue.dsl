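; Dark Vogue weapon wrap Blueprint that changes the wrap's emissive look each time the weapon fires.
; On fire, it writes game time to an emissive brightness parameter on the wrapped meshes, waits briefly, then switches on MskNumber.
; MskTextureSwap sets the next gradient mask, scale and speed on each dynamic material instance and stores the next index, cycling through four looks.
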(event EventOnWeaponFired
  (Rendering|Material|SetScalarParameterValueonMaterials (CustomItemWrap|GetWrappedMeshComponents) "TopS Emissive_01-Brightness-TBP" (Utilities|Time|GetGameTimeinSeconds))
  (Utilities|FlowControl|Delay 0.2)
  (switch int (Variables|Default|GetMskNumber)
    (:0
      (CallFunction|MskTextureSwap "/BRCosmetics/Weapons/WeaponSkins/Wraps/Textures/DarkVogue/T_Wrap_DarkVogue_GradB.T_Wrap_DarkVogue_GradB" 1 0.25 0.3))
    (:1
      (CallFunction|MskTextureSwap "/BRCosmetics/Weapons/WeaponSkins/Wraps/Textures/DarkVogue/T_Wrap_DarkVogue_GradC.T_Wrap_DarkVogue_GradC" 2 0.25 0.3))
    (:2
      (CallFunction|MskTextureSwap "/BRCosmetics/Weapons/WeaponSkins/Wraps/Textures/DarkVogue/T_Wrap_DarkVogue_GradD.T_Wrap_DarkVogue_GradD" 3 0.25 0.3))
    (:3
      (CallFunction|MskTextureSwap "/BRCosmetics/Weapons/WeaponSkins/Wraps/Textures/DarkVogue/T_Wrap_DarkVogue_Grad.T_Wrap_DarkVogue_Grad" 0 0.5 -0.2))))

(fn MskTextureSwap
  (params (Msk_Texture :type Texture%20Object%20Reference) (MskNumber :type Integer) (Scale :type Float%20%28single-precision%29) (Speed :type Float%20%28single-precision%29))
  (for _array_element (CustomItemWrap|GetWrappedMeshComponents)
    (for _array_element_2 (Rendering|Material|GetMaterials _array_element)
      (bind _asmaterial_instance_dynamic (Utilities|Casting|CastToMaterialInstanceDynamic _array_element_2))
      (Rendering|Material|SetTextureParameterValue _asmaterial_instance_dynamic "TopS_00A_Sharerd_Tex_Msk" Msk_Texture)
      (Rendering|Material|SetVectorParameterValue _asmaterial_instance_dynamic "TopS_Em01_Speed__Master" (Math|Color|MakeColor Speed) :!class /Script/Engine.MaterialInstanceDynamic)
      (Rendering|Material|SetVectorParameterValue _asmaterial_instance_dynamic "TopS_Em01_Scale__Master" (Math|Color|MakeColor Scale) :!class /Script/Engine.MaterialInstanceDynamic)
      (Variables|Default|SetMskNumber MskNumber))))

(fn ConstructionScript)
