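; Reactive weapon wrap Blueprint that drives decal material effects on wrapped meshes whenever the weapon fires.
; After a very short delay, it writes game time plus a fixed offset to one of four scalar timestamp parameters, DecalA through DecalD, on the wrapped components' materials.
; An IterationCount switch advances the slot each shot and wraps around, so successive shots retime different decals in round-robin order.
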(event EventOnWeaponFired
  (bind _returnvalue (CustomItemWrap|GetWrappedMeshComponents))
  (bind _returnvalue_2 (Utilities|Time|GetGameTimeinSeconds))
  (bind _returnvalue_3 (+ _returnvalue_2 2.0))
  (Utilities|FlowControl|Delay 0.01)
  (switch int (Variables|Default|GetIterationCount)
    (:0
      (Rendering|Material|SetScalarParameterValueonMaterials _returnvalue "EDRec_TimeFrom_BP_DecalA" _returnvalue_3)
      (Variables|Default|SetIterationCount 1))
    (:1
      (Rendering|Material|SetScalarParameterValueonMaterials _returnvalue "EDRec_TimeFrom_BP_DecalB" _returnvalue_3)
      (Variables|Default|SetIterationCount 2))
    (:2
      (Rendering|Material|SetScalarParameterValueonMaterials _returnvalue "EDRec_TimeFrom_BP_DecalC" _returnvalue_3)
      (Variables|Default|SetIterationCount 3))
    (:3
      (Rendering|Material|SetScalarParameterValueonMaterials _returnvalue "EDRec_TimeFrom_BP_DecalD" _returnvalue_3)
      (Variables|Default|SetIterationCount 0))))

(fn ConstructionScript)
