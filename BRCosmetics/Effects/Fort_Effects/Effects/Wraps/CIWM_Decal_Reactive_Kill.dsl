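; Reactive weapon wrap modifier that reveals up to four decals (A–D) on wrapped meshes as the player's kill count rises.
; A watch on the AthenaKills stat stores count and kill time, then drives decal opacity and start-time material parameters, retriggering all four at five or more kills.
; Weapon changes reapply the state after a short delay.
; In the front end, a timeline steps through kill stages as a preview; otherwise decals start hidden.
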(event EventBeginPlay
  (bind _returnvalue (CustomItemWrap|GetWrappedMeshComponents))
  (Utilities|FlowControl|Delay 0.2)
  (if (CustomCosmeticModifier|IsinFrontEnd)
    (Timeline|Timeline PlayFromStart
      (:Stage1
        (Variables|Default|SetKill_Count 0))
      (:Stage2
        (Variables|Default|SetKill_Count 1))
      (:Stage3
        (Variables|Default|SetKill_Count 2))
      (:Stage4
        (Variables|Default|SetKill_Count 3))
      (:stage5
        (Variables|Default|SetKill_Count 4))
      (:stage6
        (Variables|Default|SetKill_Count 5)))
    (Variables|Default|SetLastKillTime (Utilities|Time|GetGameTimeinSeconds))
    (CallFunction|UpdatMAt)
    (else
      (Rendering|Material|SetScalarParameterValueonMaterials _returnvalue "EDRec_DecalA_Opacity_End" 0.0)
      (Rendering|Material|SetScalarParameterValueonMaterials _returnvalue "EDRec_DecalB_Opacity_End" 0.0)
      (Rendering|Material|SetScalarParameterValueonMaterials _returnvalue "EDRec_DecalC_Opacity_End" 0.0)
      (Rendering|Material|SetScalarParameterValueonMaterials _returnvalue "EDRec_DecalD_Opacity_End" 0.0))))

(event Custom|Kill%20material
  (bind _returnvalue (CustomItemWrap|GetWrappedMeshComponents))
  (bind _returnvalue_2 (CustomItemWrap|GetWrappedMeshComponents))
  (bind _returnvalue_3 (CustomItemWrap|GetWrappedMeshComponents))
  (bind _kill_count (Variables|Default|GetKill_Count))
  (bind _returnvalue_4 (>= _kill_count 5))
  (bind _lastkilltime (Variables|Default|GetLastKillTime))
  (bind _lastkilltime_2 (Variables|Default|GetLastKillTime))
  (if _returnvalue_4
    (Rendering|Material|SetScalarParameterValueonMaterials _returnvalue_3 "EDRec_DecalD_Opacity_End" 1.0)
    (Rendering|Material|SetScalarParameterValueonMaterials _returnvalue_3 "EDRec_DecalC_Opacity_End" 1.0)
    (Rendering|Material|SetScalarParameterValueonMaterials _returnvalue_3 "EDRec_DecalB_Opacity_End" 1.0)
    (else
      (switch int _kill_count
        (:0
          (Rendering|Material|SetScalarParameterValueonMaterials _returnvalue "EDRec_DecalA_Opacity_End" 0.0)
          (Rendering|Material|SetScalarParameterValueonMaterials _returnvalue "EDRec_DecalB_Opacity_End" 0.0)
          (Rendering|Material|SetScalarParameterValueonMaterials _returnvalue "EDRec_DecalC_Opacity_End" 0.0)
          (Rendering|Material|SetScalarParameterValueonMaterials _returnvalue "EDRec_DecalD_Opacity_End" 0.0))
        (:1
          (Rendering|Material|SetScalarParameterValueonMaterials _returnvalue_3 "EDRec_TimeFrom_BP_DecalA" _lastkilltime_2))
        (:2
          (Rendering|Material|SetScalarParameterValueonMaterials _returnvalue_3 "EDRec_TimeFrom_BP_DecalB" _lastkilltime_2))
        (:3
          (Rendering|Material|SetScalarParameterValueonMaterials _returnvalue_3 "EDRec_TimeFrom_BP_DecalC" _lastkilltime_2))
        (:4
          (Rendering|Material|SetScalarParameterValueonMaterials _returnvalue_3 "EDRec_TimeFrom_BP_DecalD" _lastkilltime_2)))))
  (Rendering|Material|SetScalarParameterValueonMaterials _returnvalue_3 "EDRec_DecalA_Opacity_End" 1.0)
  (if _returnvalue_4
    (Rendering|Material|SetScalarParameterValueonMaterials _returnvalue_2 "EDRec_TimeFrom_BP_DecalA" _lastkilltime)
    (Rendering|Material|SetScalarParameterValueonMaterials _returnvalue_2 "EDRec_TimeFrom_BP_DecalB" _lastkilltime)
    (Rendering|Material|SetScalarParameterValueonMaterials _returnvalue_2 "EDRec_TimeFrom_BP_DecalC" _lastkilltime)
    (Rendering|Material|SetScalarParameterValueonMaterials _returnvalue_2 "EDRec_TimeFrom_BP_DecalD" _lastkilltime)))

(event Custom|UpdatMAt
  (bind _returnvalue (CustomItemWrap|GetWrappedMeshComponents))
  (bind _returnvalue_2 (CustomItemWrap|GetWrappedMeshComponents))
  (bind _returnvalue_3 (CustomItemWrap|GetWrappedMeshComponents))
  (bind _lastkilltime (Variables|Default|GetLastKillTime))
  (bind _kill_count (Variables|Default|GetKill_Count))
  (bind _lastkilltime_2 (Variables|Default|GetLastKillTime))
  (switch int _kill_count
    (:0
      (Rendering|Material|SetScalarParameterValueonMaterials _returnvalue "EDRec_DecalA_Opacity_End" 0.0)
      (Rendering|Material|SetScalarParameterValueonMaterials _returnvalue "EDRec_DecalB_Opacity_End" 0.0)
      (Rendering|Material|SetScalarParameterValueonMaterials _returnvalue "EDRec_DecalC_Opacity_End" 0.0)
      (Rendering|Material|SetScalarParameterValueonMaterials _returnvalue "EDRec_DecalD_Opacity_End" 0.0))
    (:1
      (Rendering|Material|SetScalarParameterValueonMaterials _returnvalue_3 "EDRec_TimeFrom_BP_DecalA" _lastkilltime_2))
    (:2
      (Rendering|Material|SetScalarParameterValueonMaterials _returnvalue_3 "EDRec_TimeFrom_BP_DecalB" _lastkilltime_2)
      (Rendering|Material|SetScalarParameterValueonMaterials _returnvalue_3 "EDRec_DecalB_Opacity_End" 1.0))
    (:3
      (Rendering|Material|SetScalarParameterValueonMaterials _returnvalue_3 "EDRec_TimeFrom_BP_DecalC" _lastkilltime_2)
      (Rendering|Material|SetScalarParameterValueonMaterials _returnvalue_3 "EDRec_DecalC_Opacity_End" 1.0))
    (:4
      (Rendering|Material|SetScalarParameterValueonMaterials _returnvalue_3 "EDRec_TimeFrom_BP_DecalD" _lastkilltime_2)
      (Rendering|Material|SetScalarParameterValueonMaterials _returnvalue_3 "EDRec_DecalD_Opacity_End" 1.0)))
  (Rendering|Material|SetScalarParameterValueonMaterials _returnvalue_3 "EDRec_DecalA_Opacity_End" 1.0)
  (if (>= _kill_count 5)
    (Rendering|Material|SetScalarParameterValueonMaterials _returnvalue_2 "EDRec_TimeFrom_BP_DecalA" _lastkilltime)
    (Rendering|Material|SetScalarParameterValueonMaterials _returnvalue_2 "EDRec_TimeFrom_BP_DecalB" _lastkilltime)
    (Rendering|Material|SetScalarParameterValueonMaterials _returnvalue_2 "EDRec_TimeFrom_BP_DecalC" _lastkilltime)
    (Rendering|Material|SetScalarParameterValueonMaterials _returnvalue_2 "EDRec_TimeFrom_BP_DecalD" _lastkilltime)))

(event EventOnPawnSet
  (bind (_statname _statvalue) (Cosmetics|WatchCosmeticStat (CustomCharacter|GetAssociatedPlayerPawn) "(TagName=\"StatManager.AthenaKills\")")
    (:OnStatChanged
      (Variables|Default|SetKill_Count _statvalue)
      (Variables|Default|SetLastKillTime (Utilities|Time|GetGameTimeinSeconds))
      (CallFunction|KillMaterial))))

(event EventOnNewWeapons
  (params (WeaponsAsItems :type Array%20of%20Fort%20Item%20Object%20References))
  (Utilities|FlowControl|Delay 0.2)
  (CallFunction|KillMaterial))

(event EventOnNewPlayerWeaponType
  (params (NewType :type EFortItemType%20Enum :default "WorldItem") (OldType :type EFortItemType%20Enum :default "WorldItem"))
  (Utilities|FlowControl|Delay 0.2)
  (CallFunction|KillMaterial))

(fn ConstructionScript)
