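; Weapon wrap Blueprint for the DinoHunter cosmetic that passes shot timing to the wrap's materials.
; On each weapon fire event it gets the wrapped mesh components and writes the current game time to one of three scalar parameters, DinoH_TimeFrom_BP_A, B or C.
; An integer variable cycles through the three slots in round-robin order, so the materials retain the times of the three most recent shots.
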(event EventOnWeaponFired
  (bind _returnvalue (CustomItemWrap|GetWrappedMeshComponents))
  (bind _returnvalue_2 (Utilities|Time|GetGameTimeinSeconds))
  (switch int (Variables|Default|GetSet_Varable)
    (:0
      (Rendering|Material|SetScalarParameterValueonMaterials _returnvalue "DinoH_TimeFrom_BP_A" _returnvalue_2)
      (Variables|Default|SetSet_Varable 1))
    (:1
      (Rendering|Material|SetScalarParameterValueonMaterials _returnvalue "DinoH_TimeFrom_BP_B" _returnvalue_2)
      (Variables|Default|SetSet_Varable 2))
    (:2
      (Rendering|Material|SetScalarParameterValueonMaterials _returnvalue "DinoH_TimeFrom_BP_C" _returnvalue_2)
      (Variables|Default|SetSet_Varable 0))))

(fn ConstructionScript)
