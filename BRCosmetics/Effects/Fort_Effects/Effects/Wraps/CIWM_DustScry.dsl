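; Custom item wrap Blueprint for DustScry that triggers time-driven emissive, flipbook and decal animations on the wrapped meshes' materials.
; In the front end a timeline event fires them after a short delay, while on a pawn they fire when the DamageReceived cosmetic stat changes.
; Both paths stamp game time into scalar parameters, then stamp a shuffled parameter-name list with staggered delays.
; RandomSet randomizes decal frame, offset and rotation but is not called here.
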(event EventBeginPlay
  (bind _returnvalue (CustomItemWrap|GetWrappedMeshComponents))
  (bind _returnvalue_2 (CustomItemWrap|GetWrappedMeshComponents))
  (bind _returnvalue_3 (Utilities|Time|GetGameTimeinSeconds))
  (bind _returnvalue_4 (Utilities|Time|GetGameTimeinSeconds))
  (Utilities|FlowControl|Delay 0.2)
  (if (CustomCosmeticModifier|IsinFrontEnd)
    (Timeline|Timeline PlayFromStart
      (:NewTrack_0
        (Utilities|FlowControl|Sequence
          (:then_0
            ;[1 1 1] TopsSide V2
            (Rendering|Material|SetScalarParameterValueonMaterials _returnvalue "TopS Emissive_02-Brightness-TBP" _returnvalue_4)
            (Rendering|Material|SetScalarParameterValueonMaterials _returnvalue "TopS Emissive_02_Offset----TBP" _returnvalue_4)
            (Rendering|Material|SetScalarParameterValueonMaterials _returnvalue "TopS_01_Blend_Opacity-TBP" _returnvalue_4)
            (Rendering|Material|SetScalarParameterValueonMaterials _returnvalue "TimeFrom-BP----------------" _returnvalue_4))
          (:then_1
            (Rendering|Material|SetScalarParameterValueonMaterials _returnvalue "TopS Emissive_01-Brightness-TBP" _returnvalue_4)
            (Rendering|Material|SetScalarParameterValueonMaterials _returnvalue "TopS Emissive_01_Offset----TBP" _returnvalue_4)
            (Rendering|Material|SetScalarParameterValueonMaterials _returnvalue "TopS Emissive_03-Brightness-TBP" _returnvalue_4)
            (Rendering|Material|SetScalarParameterValueonMaterials _returnvalue "TopS Emissive_03_Offset----TBP" _returnvalue_4)
            ; Left\Right V2
            (Rendering|Material|SetScalarParameterValueonMaterials _returnvalue_2 "FDecalV2_FlipBook_Time_FromBP" _returnvalue_3)
            (Rendering|Material|SetScalarParameterValueonMaterials _returnvalue_2 "FDecalaV2_Opacity_Time_BP_Z" _returnvalue_3)
            (Rendering|Material|SetScalarParameterValueonMaterials _returnvalue_2 "FDecalV2_Emissive_Brightness_Time_BP" _returnvalue_3)
            (Rendering|Material|SetScalarParameterValueonMaterials _returnvalue_2 "FDecalV2_Emissive_Gradent_Time_BP" _returnvalue_3))
          (:then_2
            (Utilities|Array|Shuffle (Variables|Default|GetParameter%20Name))
            (bind _item (Utilities|ForEachLoopwithDelay (Variables|Default|GetParameter%20Name) 0.2)
              (:LoopBody
                (Rendering|Material|SetScalarParameterValueonMaterials (CustomItemWrap|GetWrappedMeshComponents) _item (Utilities|Time|GetGameTimeinSeconds))))))))))

(event EventOnPawnSet
  (bind _returnvalue (CustomItemWrap|GetWrappedMeshComponents))
  (bind _returnvalue_2 (CustomItemWrap|GetWrappedMeshComponents))
  (bind _returnvalue_3 (Utilities|Time|GetGameTimeinSeconds))
  (bind _returnvalue_4 (Utilities|Time|GetGameTimeinSeconds))
  (Cosmetics|WatchCosmeticStat (CustomCharacter|GetAssociatedPlayerPawn) "(TagName=\"StatManager.DamageReceived\")" 0 false
    (:OnStatChanged
      (Utilities|FlowControl|Sequence
        (:then_0
          ;[1 1 1] TopsSide V2
          (Rendering|Material|SetScalarParameterValueonMaterials _returnvalue "TopS Emissive_02-Brightness-TBP" _returnvalue_4)
          (Rendering|Material|SetScalarParameterValueonMaterials _returnvalue "TopS Emissive_02_Offset----TBP" _returnvalue_4)
          (Rendering|Material|SetScalarParameterValueonMaterials _returnvalue "TopS_01_Blend_Opacity-TBP" _returnvalue_4)
          (Rendering|Material|SetScalarParameterValueonMaterials _returnvalue "TimeFrom-BP----------------" _returnvalue_4))
        (:then_1
          (Rendering|Material|SetScalarParameterValueonMaterials _returnvalue "TopS Emissive_01-Brightness-TBP" _returnvalue_4)
          (Rendering|Material|SetScalarParameterValueonMaterials _returnvalue "TopS Emissive_01_Offset----TBP" _returnvalue_4)
          (Rendering|Material|SetScalarParameterValueonMaterials _returnvalue "TopS Emissive_03-Brightness-TBP" _returnvalue_4)
          (Rendering|Material|SetScalarParameterValueonMaterials _returnvalue "TopS Emissive_03_Offset----TBP" _returnvalue_4)
          ; Left\Right V2
          (Rendering|Material|SetScalarParameterValueonMaterials _returnvalue_2 "FDecalV2_FlipBook_Time_FromBP" _returnvalue_3)
          (Rendering|Material|SetScalarParameterValueonMaterials _returnvalue_2 "FDecalaV2_Opacity_Time_BP_Z" _returnvalue_3)
          (Rendering|Material|SetScalarParameterValueonMaterials _returnvalue_2 "FDecalV2_Emissive_Brightness_Time_BP" _returnvalue_3)
          (Rendering|Material|SetScalarParameterValueonMaterials _returnvalue_2 "FDecalV2_Emissive_Gradent_Time_BP" _returnvalue_3))
        (:then_2
          (Utilities|Array|Shuffle (Variables|Default|GetParameter%20Name))
          (bind _item (Utilities|ForEachLoopwithDelay (Variables|Default|GetParameter%20Name) 0.2)
            (:LoopBody
              (Rendering|Material|SetScalarParameterValueonMaterials (CustomItemWrap|GetWrappedMeshComponents) _item (Utilities|Time|GetGameTimeinSeconds)))))))))

(fn RandomSet
  (params (FlipBook%20Frame :type Name) (TimeBP :type Name) (Offset :type Name) (Rotation :type Name) (X=min :type Float%20%28double-precision%29) (X=Max :type Float%20%28double-precision%29) (Y=Min :type Float%20%28double-precision%29) (Y=Max :type Float%20%28double-precision%29) (Rot=Min :type Float%20%28double-precision%29) (Rot=Max :type Float%20%28double-precision%29) (TimeBP-first :type Name))
  (bind _returnvalue (CustomItemWrap|GetWrappedMeshComponents))
  (bind _stream (Variables|Default|GetStream))
  (bind _returnvalue_2 (Math|Random|RandomFloatinRangefromStream _stream X=min X=Max))
  (bind _returnvalue_3 (Math|Random|RandomIntegerinRange 0 8))
  (bind _returnvalue_4 (Math|Conversions|ToFloat(Integer) _returnvalue_3))
  (bind _returnvalue_5 (/ _returnvalue_4 8.0))
  (Rendering|Material|SetScalarParameterValueonMaterials _returnvalue FlipBook%20Frame _returnvalue_5)
  (Rendering|Material|SetScalarParameterValueonMaterials _returnvalue TimeBP-first (Utilities|Time|GetGameTimeinSeconds))
  (Rendering|Material|SetVectorParameterValueonMaterials _returnvalue Offset (Math|Vector|MakeVector _returnvalue_2 (Math|Random|RandomFloatinRangefromStream _stream Y=Min Y=Max)))
  (Rendering|Material|SetScalarParameterValueonMaterials _returnvalue Rotation (Math|Random|RandomFloatinRange Rot=Min Rot=Max))
  (Rendering|Material|SetScalarParameterValueonMaterials _returnvalue TimeBP (Utilities|Time|GetGameTimeinSeconds))
  (Variables|Default|SetPosition _returnvalue_2)
  (Development|PrintString (Utilities|String|ToString(Float) _returnvalue_5) true true "(R=0.060001,G=1.000000,B=0.000000,A=1.000000)" 3.0))

(fn ConstructionScript)
