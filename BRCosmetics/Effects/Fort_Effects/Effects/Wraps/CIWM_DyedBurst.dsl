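; CIWM_DyedBurst is an item wrap Blueprint that alternates two colors on wrapped weapon meshes.
; Weapon fire, new weapons and weapon type changes each wait briefly, then call Kill material, which despite its name only recolors.
; That event writes ColorA and ColorB to the two "TopS" base color vector parameters on the wrapped mesh components' materials, then toggles a flag so the next call swaps them.
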(event EventOnWeaponFired
  (Utilities|FlowControl|Delay 0.2)
  (CallFunction|KillMaterial))

(event EventOnPawnSet)

(event EventOnNewWeapons
  (params (WeaponsAsItems :type Array%20of%20Fort%20Item%20Object%20References))
  (Utilities|FlowControl|Delay 0.2)
  (CallFunction|KillMaterial))

(event EventOnNewPlayerWeaponType
  (params (NewType :type EFortItemType%20Enum :default "WorldItem") (OldType :type EFortItemType%20Enum :default "WorldItem"))
  (Utilities|FlowControl|Delay 0.2)
  (CallFunction|KillMaterial))

(event Custom|Kill%20material
  (bind _returnvalue (CustomItemWrap|GetWrappedMeshComponents))
  (bind _colora (Variables|Default|GetColorA))
  (bind _returnvalue_2 (Math|Conversions|ToVector(LinearColor) _colora))
  (bind _colorb (Variables|Default|GetColorB))
  (bind _returnvalue_3 (Math|Conversions|ToVector(LinearColor) _colorb))
  (bind _returnvalue_4 (CustomItemWrap|GetWrappedMeshComponents))
  (bind _flag (Variables|Default|Getflag))
  (if _flag
    (Rendering|Material|SetVectorParameterValueonMaterials _returnvalue "TopS_02_Base Color_Simple" _returnvalue_2)
    (Rendering|Material|SetVectorParameterValueonMaterials _returnvalue "TopS_01_Base Color_Simple" _returnvalue_3)
    (else
      (Rendering|Material|SetVectorParameterValueonMaterials _returnvalue_4 "TopS_02_Base Color_Simple" _returnvalue_3)
      (Rendering|Material|SetVectorParameterValueonMaterials _returnvalue_4 "TopS_01_Base Color_Simple" _returnvalue_2)))
  (Variables|Default|Setflag (not _flag)))

(fn ConstructionScript)
