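; Item wrap Blueprint for Elastic Champion that matches wrapped item materials to the look of the player's character.
; On pawn set, wrap target set or customization finished, identical logic name-checks the character, then reads its Body slot 0 dynamic material.
; It copies colors, sticker and mask textures, unpacking material vectors into metallic, specular and roughness, and prints debug text.
; The banner event is empty; setcolors, a rotated-color emissive tint, is never called here.
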(event EventOnPawnSet
  (if (Utilities|NameCheck (Item|GetPersistentName (CosmeticLoadout|GetItemDefinitionfromLoadout (CosmeticLoadout|GetCosmeticLoadout) "(TagName=\"LoadoutSlot.Character\")")) (Variables|Default|GetNames%20to%20check%20))
    (bind _asmaterial_instance_dynamic (Utilities|Casting|CastToMaterialInstanceDynamic (Rendering|Material|GetMaterial (Customization|GetSkeletalMeshforPartType (CustomCharacter|GetAssociatedPlayerPawn) "Body" :!class /Script/FortniteGame.FortPlayerPawn) 0)))
    (Utilities|FlowControl|Sequence
      (:then_0
        (bind _returnvalue (Rendering|Material|GetVectorParameterValue _asmaterial_instance_dynamic "SecondaryColor" :!class /Script/Engine.MaterialInstanceDynamic))
        (bind _returnvalue_2 (Rendering|Material|GetVectorParameterValue _asmaterial_instance_dynamic "SecondaryMaterial" :!class /Script/Engine.MaterialInstanceDynamic))
        (Development|PrintText "INVTEXT(\"Color\")" true true _returnvalue 5.0)
        (for _array_element (CustomItemWrap|GetWrappedMeshComponents)
          (for _array_element_2 (Rendering|Material|GetMaterials _array_element)
            (bind _asmaterial_instance_dynamic_2 (Utilities|Casting|CastToMaterialInstanceDynamic _array_element_2))
            (Rendering|Material|SetVectorParameterValue _asmaterial_instance_dynamic_2 "TopS_01_ColorTint_CH_Green A=0=off" _returnvalue :!class /Script/Engine.MaterialInstanceDynamic)
            (Rendering|Material|SetScalarParameterValue _asmaterial_instance_dynamic_2 "TopS_01_Roughtness_CH_Green" (.B (Math|Color|BreakColor _returnvalue_2)) :!class /Script/Engine.MaterialInstanceDynamic)
            (Rendering|Material|SetScalarParameterValue _asmaterial_instance_dynamic_2 "TopS_01_Metalic_CH_Green" (Math|Color|BreakColor _returnvalue_2) :!class /Script/Engine.MaterialInstanceDynamic)
            (Rendering|Material|SetScalarParameterValue _asmaterial_instance_dynamic_2 "TopS_01_Spec_CH_Green" (.G (Math|Color|BreakColor _returnvalue_2)) :!class /Script/Engine.MaterialInstanceDynamic))))
      (:then_1
        (bind _returnvalue_3 (Rendering|Material|GetVectorParameterValue _asmaterial_instance_dynamic "PrimaryColor" :!class /Script/Engine.MaterialInstanceDynamic))
        (bind _returnvalue_4 (Rendering|Material|GetVectorParameterValue _asmaterial_instance_dynamic "PrimaryMaterial" :!class /Script/Engine.MaterialInstanceDynamic))
        (for _array_element_3 (CustomItemWrap|GetWrappedMeshComponents)
          (for _array_element_4 (Rendering|Material|GetMaterials _array_element_3)
            (bind _asmaterial_instance_dynamic_3 (Utilities|Casting|CastToMaterialInstanceDynamic _array_element_4))
            (Rendering|Material|SetVectorParameterValue _asmaterial_instance_dynamic_3 "TopS_01_ColorTint_CH_Red  A=0=off" _returnvalue_3 :!class /Script/Engine.MaterialInstanceDynamic)
            (Rendering|Material|SetScalarParameterValue _asmaterial_instance_dynamic_3 "TopS_01_Roughtness_CH_Red" (.B (Math|Color|BreakColor _returnvalue_4)) :!class /Script/Engine.MaterialInstanceDynamic)
            (Rendering|Material|SetScalarParameterValue _asmaterial_instance_dynamic_3 "TopS_01_Metalic_CH_Red" (Math|Color|BreakColor _returnvalue_4) :!class /Script/Engine.MaterialInstanceDynamic)
            (Rendering|Material|SetScalarParameterValue _asmaterial_instance_dynamic_3 "TopS_01_Spec_CH_Red" (.G (Math|Color|BreakColor _returnvalue_4)) :!class /Script/Engine.MaterialInstanceDynamic))))
      (:then_2
        (bind _returnvalue_5 (Rendering|Material|GetTextureParameterValue _asmaterial_instance_dynamic "Sticker" :!class /Script/Engine.MaterialInstanceDynamic))
        (for _array_element_5 (CustomItemWrap|GetWrappedMeshComponents)
          (for _array_element_6 (Rendering|Material|GetMaterials _array_element_5)
            (bind _asmaterial_instance_dynamic_4 (Utilities|Casting|CastToMaterialInstanceDynamic _array_element_6))
            (Rendering|Material|SetTextureParameterValue _asmaterial_instance_dynamic_4 "FDecal_BaseColor_Tex" _returnvalue_5))))
      (:then_3
        (bind _returnvalue_6 (Rendering|Material|GetVectorParameterValue _asmaterial_instance_dynamic "AccessoryColor" :!class /Script/Engine.MaterialInstanceDynamic))
        (bind _returnvalue_7 (Rendering|Material|GetVectorParameterValue _asmaterial_instance_dynamic "AccessoryMaterial" :!class /Script/Engine.MaterialInstanceDynamic))
        (for _array_element_7 (CustomItemWrap|GetWrappedMeshComponents)
          (for _array_element_8 (Rendering|Material|GetMaterials _array_element_7)
            (bind _asmaterial_instance_dynamic_5 (Utilities|Casting|CastToMaterialInstanceDynamic _array_element_8))
            (Rendering|Material|SetVectorParameterValue _asmaterial_instance_dynamic_5 "Simple_Color" _returnvalue_6 :!class /Script/Engine.MaterialInstanceDynamic)
            (Rendering|Material|SetScalarParameterValue _asmaterial_instance_dynamic_5 "Simple_Roughness" (.B (Math|Color|BreakColor _returnvalue_7)) :!class /Script/Engine.MaterialInstanceDynamic)
            (Rendering|Material|SetScalarParameterValue _asmaterial_instance_dynamic_5 "Simple_Metalness" (Math|Color|BreakColor _returnvalue_7) :!class /Script/Engine.MaterialInstanceDynamic)
            (Rendering|Material|SetScalarParameterValue _asmaterial_instance_dynamic_5 "Simple_Specular" (.G (Math|Color|BreakColor _returnvalue_7)) :!class /Script/Engine.MaterialInstanceDynamic))))
      (:then_4
        (bind _returnvalue_8 (Rendering|Material|GetTextureParameterValue _asmaterial_instance_dynamic "SpecularMasks" :!class /Script/Engine.MaterialInstanceDynamic))
        (for _array_element_9 (CustomItemWrap|GetWrappedMeshComponents)
          (for _array_element_10 (Rendering|Material|GetMaterials _array_element_9)
            (bind _asmaterial_instance_dynamic_6 (Utilities|Casting|CastToMaterialInstanceDynamic _array_element_10))
            (Rendering|Material|SetTextureParameterValue _asmaterial_instance_dynamic_6 "TopS_01_ColorTint_CH_Text --DoNotTouch" _returnvalue_8)))))))

(event EventOnSetWrapTarget
  (if (Utilities|NameCheck (Item|GetPersistentName (CosmeticLoadout|GetItemDefinitionfromLoadout (CosmeticLoadout|GetCosmeticLoadout) "(TagName=\"LoadoutSlot.Character\")")) (Variables|Default|GetNames%20to%20check%20))
    (bind _asmaterial_instance_dynamic (Utilities|Casting|CastToMaterialInstanceDynamic (Rendering|Material|GetMaterial (Customization|GetSkeletalMeshforPartType (CustomCharacter|GetAssociatedPlayerPawn) "Body" :!class /Script/FortniteGame.FortPlayerPawn) 0)))
    (Utilities|FlowControl|Sequence
      (:then_0
        (bind _returnvalue (Rendering|Material|GetVectorParameterValue _asmaterial_instance_dynamic "SecondaryColor" :!class /Script/Engine.MaterialInstanceDynamic))
        (bind _returnvalue_2 (Rendering|Material|GetVectorParameterValue _asmaterial_instance_dynamic "SecondaryMaterial" :!class /Script/Engine.MaterialInstanceDynamic))
        (Development|PrintText "INVTEXT(\"Color\")" true true _returnvalue 5.0)
        (for _array_element (CustomItemWrap|GetWrappedMeshComponents)
          (for _array_element_2 (Rendering|Material|GetMaterials _array_element)
            (bind _asmaterial_instance_dynamic_2 (Utilities|Casting|CastToMaterialInstanceDynamic _array_element_2))
            (Rendering|Material|SetVectorParameterValue _asmaterial_instance_dynamic_2 "TopS_01_ColorTint_CH_Green A=0=off" _returnvalue :!class /Script/Engine.MaterialInstanceDynamic)
            (Rendering|Material|SetScalarParameterValue _asmaterial_instance_dynamic_2 "TopS_01_Roughtness_CH_Green" (.B (Math|Color|BreakColor _returnvalue_2)) :!class /Script/Engine.MaterialInstanceDynamic)
            (Rendering|Material|SetScalarParameterValue _asmaterial_instance_dynamic_2 "TopS_01_Metalic_CH_Green" (Math|Color|BreakColor _returnvalue_2) :!class /Script/Engine.MaterialInstanceDynamic)
            (Rendering|Material|SetScalarParameterValue _asmaterial_instance_dynamic_2 "TopS_01_Spec_CH_Green" (.G (Math|Color|BreakColor _returnvalue_2)) :!class /Script/Engine.MaterialInstanceDynamic))))
      (:then_1
        (bind _returnvalue_3 (Rendering|Material|GetVectorParameterValue _asmaterial_instance_dynamic "PrimaryColor" :!class /Script/Engine.MaterialInstanceDynamic))
        (bind _returnvalue_4 (Rendering|Material|GetVectorParameterValue _asmaterial_instance_dynamic "PrimaryMaterial" :!class /Script/Engine.MaterialInstanceDynamic))
        (for _array_element_3 (CustomItemWrap|GetWrappedMeshComponents)
          (for _array_element_4 (Rendering|Material|GetMaterials _array_element_3)
            (bind _asmaterial_instance_dynamic_3 (Utilities|Casting|CastToMaterialInstanceDynamic _array_element_4))
            (Rendering|Material|SetVectorParameterValue _asmaterial_instance_dynamic_3 "TopS_01_ColorTint_CH_Red  A=0=off" _returnvalue_3 :!class /Script/Engine.MaterialInstanceDynamic)
            (Rendering|Material|SetScalarParameterValue _asmaterial_instance_dynamic_3 "TopS_01_Roughtness_CH_Red" (.B (Math|Color|BreakColor _returnvalue_4)) :!class /Script/Engine.MaterialInstanceDynamic)
            (Rendering|Material|SetScalarParameterValue _asmaterial_instance_dynamic_3 "TopS_01_Metalic_CH_Red" (Math|Color|BreakColor _returnvalue_4) :!class /Script/Engine.MaterialInstanceDynamic)
            (Rendering|Material|SetScalarParameterValue _asmaterial_instance_dynamic_3 "TopS_01_Spec_CH_Red" (.G (Math|Color|BreakColor _returnvalue_4)) :!class /Script/Engine.MaterialInstanceDynamic))))
      (:then_2
        (bind _returnvalue_5 (Rendering|Material|GetTextureParameterValue _asmaterial_instance_dynamic "Sticker" :!class /Script/Engine.MaterialInstanceDynamic))
        (for _array_element_5 (CustomItemWrap|GetWrappedMeshComponents)
          (for _array_element_6 (Rendering|Material|GetMaterials _array_element_5)
            (bind _asmaterial_instance_dynamic_4 (Utilities|Casting|CastToMaterialInstanceDynamic _array_element_6))
            (Rendering|Material|SetTextureParameterValue _asmaterial_instance_dynamic_4 "FDecal_BaseColor_Tex" _returnvalue_5))))
      (:then_3
        (bind _returnvalue_6 (Rendering|Material|GetVectorParameterValue _asmaterial_instance_dynamic "AccessoryColor" :!class /Script/Engine.MaterialInstanceDynamic))
        (bind _returnvalue_7 (Rendering|Material|GetVectorParameterValue _asmaterial_instance_dynamic "AccessoryMaterial" :!class /Script/Engine.MaterialInstanceDynamic))
        (for _array_element_7 (CustomItemWrap|GetWrappedMeshComponents)
          (for _array_element_8 (Rendering|Material|GetMaterials _array_element_7)
            (bind _asmaterial_instance_dynamic_5 (Utilities|Casting|CastToMaterialInstanceDynamic _array_element_8))
            (Rendering|Material|SetVectorParameterValue _asmaterial_instance_dynamic_5 "Simple_Color" _returnvalue_6 :!class /Script/Engine.MaterialInstanceDynamic)
            (Rendering|Material|SetScalarParameterValue _asmaterial_instance_dynamic_5 "Simple_Roughness" (.B (Math|Color|BreakColor _returnvalue_7)) :!class /Script/Engine.MaterialInstanceDynamic)
            (Rendering|Material|SetScalarParameterValue _asmaterial_instance_dynamic_5 "Simple_Metalness" (Math|Color|BreakColor _returnvalue_7) :!class /Script/Engine.MaterialInstanceDynamic)
            (Rendering|Material|SetScalarParameterValue _asmaterial_instance_dynamic_5 "Simple_Specular" (.G (Math|Color|BreakColor _returnvalue_7)) :!class /Script/Engine.MaterialInstanceDynamic))))
      (:then_4
        (bind _returnvalue_8 (Rendering|Material|GetTextureParameterValue _asmaterial_instance_dynamic "SpecularMasks" :!class /Script/Engine.MaterialInstanceDynamic))
        (for _array_element_9 (CustomItemWrap|GetWrappedMeshComponents)
          (for _array_element_10 (Rendering|Material|GetMaterials _array_element_9)
            (bind _asmaterial_instance_dynamic_6 (Utilities|Casting|CastToMaterialInstanceDynamic _array_element_10))
            (Rendering|Material|SetTextureParameterValue _asmaterial_instance_dynamic_6 "TopS_01_ColorTint_CH_Text --DoNotTouch" _returnvalue_8)))))))

(event EventOnCharacterCustomizationFinished
  (params (CharacterPartInfo :type Array%20of%20Applied%20Character%20Part%20Info%20Structures))
  (if (Utilities|NameCheck (Item|GetPersistentName (CosmeticLoadout|GetItemDefinitionfromLoadout (CosmeticLoadout|GetCosmeticLoadout) "(TagName=\"LoadoutSlot.Character\")")) (Variables|Default|GetNames%20to%20check%20))
    (bind _asmaterial_instance_dynamic (Utilities|Casting|CastToMaterialInstanceDynamic (Rendering|Material|GetMaterial (Customization|GetSkeletalMeshforPartType (CustomCharacter|GetAssociatedPlayerPawn) "Body" :!class /Script/FortniteGame.FortPlayerPawn) 0)))
    (Utilities|FlowControl|Sequence
      (:then_0
        (bind _returnvalue (Rendering|Material|GetVectorParameterValue _asmaterial_instance_dynamic "SecondaryColor" :!class /Script/Engine.MaterialInstanceDynamic))
        (bind _returnvalue_2 (Rendering|Material|GetVectorParameterValue _asmaterial_instance_dynamic "SecondaryMaterial" :!class /Script/Engine.MaterialInstanceDynamic))
        (Development|PrintText "INVTEXT(\"Color\")" true true _returnvalue 5.0)
        (for _array_element (CustomItemWrap|GetWrappedMeshComponents)
          (for _array_element_2 (Rendering|Material|GetMaterials _array_element)
            (bind _asmaterial_instance_dynamic_2 (Utilities|Casting|CastToMaterialInstanceDynamic _array_element_2))
            (Rendering|Material|SetVectorParameterValue _asmaterial_instance_dynamic_2 "TopS_01_ColorTint_CH_Green A=0=off" _returnvalue :!class /Script/Engine.MaterialInstanceDynamic)
            (Rendering|Material|SetScalarParameterValue _asmaterial_instance_dynamic_2 "TopS_01_Roughtness_CH_Green" (.B (Math|Color|BreakColor _returnvalue_2)) :!class /Script/Engine.MaterialInstanceDynamic)
            (Rendering|Material|SetScalarParameterValue _asmaterial_instance_dynamic_2 "TopS_01_Metalic_CH_Green" (Math|Color|BreakColor _returnvalue_2) :!class /Script/Engine.MaterialInstanceDynamic)
            (Rendering|Material|SetScalarParameterValue _asmaterial_instance_dynamic_2 "TopS_01_Spec_CH_Green" (.G (Math|Color|BreakColor _returnvalue_2)) :!class /Script/Engine.MaterialInstanceDynamic))))
      (:then_1
        (bind _returnvalue_3 (Rendering|Material|GetVectorParameterValue _asmaterial_instance_dynamic "PrimaryColor" :!class /Script/Engine.MaterialInstanceDynamic))
        (bind _returnvalue_4 (Rendering|Material|GetVectorParameterValue _asmaterial_instance_dynamic "PrimaryMaterial" :!class /Script/Engine.MaterialInstanceDynamic))
        (for _array_element_3 (CustomItemWrap|GetWrappedMeshComponents)
          (for _array_element_4 (Rendering|Material|GetMaterials _array_element_3)
            (bind _asmaterial_instance_dynamic_3 (Utilities|Casting|CastToMaterialInstanceDynamic _array_element_4))
            (Rendering|Material|SetVectorParameterValue _asmaterial_instance_dynamic_3 "TopS_01_ColorTint_CH_Red  A=0=off" _returnvalue_3 :!class /Script/Engine.MaterialInstanceDynamic)
            (Rendering|Material|SetScalarParameterValue _asmaterial_instance_dynamic_3 "TopS_01_Roughtness_CH_Red" (.B (Math|Color|BreakColor _returnvalue_4)) :!class /Script/Engine.MaterialInstanceDynamic)
            (Rendering|Material|SetScalarParameterValue _asmaterial_instance_dynamic_3 "TopS_01_Metalic_CH_Red" (Math|Color|BreakColor _returnvalue_4) :!class /Script/Engine.MaterialInstanceDynamic)
            (Rendering|Material|SetScalarParameterValue _asmaterial_instance_dynamic_3 "TopS_01_Spec_CH_Red" (.G (Math|Color|BreakColor _returnvalue_4)) :!class /Script/Engine.MaterialInstanceDynamic))))
      (:then_2
        (bind _returnvalue_5 (Rendering|Material|GetTextureParameterValue _asmaterial_instance_dynamic "Sticker" :!class /Script/Engine.MaterialInstanceDynamic))
        (for _array_element_5 (CustomItemWrap|GetWrappedMeshComponents)
          (for _array_element_6 (Rendering|Material|GetMaterials _array_element_5)
            (bind _asmaterial_instance_dynamic_4 (Utilities|Casting|CastToMaterialInstanceDynamic _array_element_6))
            (Rendering|Material|SetTextureParameterValue _asmaterial_instance_dynamic_4 "FDecal_BaseColor_Tex" _returnvalue_5))))
      (:then_3
        (bind _returnvalue_6 (Rendering|Material|GetVectorParameterValue _asmaterial_instance_dynamic "AccessoryColor" :!class /Script/Engine.MaterialInstanceDynamic))
        (bind _returnvalue_7 (Rendering|Material|GetVectorParameterValue _asmaterial_instance_dynamic "AccessoryMaterial" :!class /Script/Engine.MaterialInstanceDynamic))
        (for _array_element_7 (CustomItemWrap|GetWrappedMeshComponents)
          (for _array_element_8 (Rendering|Material|GetMaterials _array_element_7)
            (bind _asmaterial_instance_dynamic_5 (Utilities|Casting|CastToMaterialInstanceDynamic _array_element_8))
            (Rendering|Material|SetVectorParameterValue _asmaterial_instance_dynamic_5 "Simple_Color" _returnvalue_6 :!class /Script/Engine.MaterialInstanceDynamic)
            (Rendering|Material|SetScalarParameterValue _asmaterial_instance_dynamic_5 "Simple_Roughness" (.B (Math|Color|BreakColor _returnvalue_7)) :!class /Script/Engine.MaterialInstanceDynamic)
            (Rendering|Material|SetScalarParameterValue _asmaterial_instance_dynamic_5 "Simple_Metalness" (Math|Color|BreakColor _returnvalue_7) :!class /Script/Engine.MaterialInstanceDynamic)
            (Rendering|Material|SetScalarParameterValue _asmaterial_instance_dynamic_5 "Simple_Specular" (.G (Math|Color|BreakColor _returnvalue_7)) :!class /Script/Engine.MaterialInstanceDynamic))))
      (:then_4
        (bind _returnvalue_8 (Rendering|Material|GetTextureParameterValue _asmaterial_instance_dynamic "SpecularMasks" :!class /Script/Engine.MaterialInstanceDynamic))
        (for _array_element_9 (CustomItemWrap|GetWrappedMeshComponents)
          (for _array_element_10 (Rendering|Material|GetMaterials _array_element_9)
            (bind _asmaterial_instance_dynamic_6 (Utilities|Casting|CastToMaterialInstanceDynamic _array_element_10))
            (Rendering|Material|SetTextureParameterValue _asmaterial_instance_dynamic_6 "TopS_01_ColorTint_CH_Text --DoNotTouch" _returnvalue_8)))))))

(event Custom|OnBannerFullyApplied
  (params (Icon :type Texture%202D%20Object%20Reference) (PrimaryColor :type Linear%20Color%20Structure)))

(fn setcolors
  (params (Color :type Linear%20Color%20Structure))
  (bind _returnvalue (Math|Conversions|ToVector(LinearColor) Color))
  (bind _returnvalue_2 (Math|Vector|RotateVector _returnvalue "25.000000,25.000000,25.000000"))
  (bind _returnvalue_3 (Math|Conversions|ToLinearColor(Vector) _returnvalue_2))
  (for _array_element (CustomItemWrap|GetWrappedMeshComponents)
    (for _array_element_2 (Rendering|Material|GetMaterials _array_element)
      (bind _asmaterial_instance_dynamic (Utilities|Casting|CastToMaterialInstanceDynamic _array_element_2))
      (Rendering|Material|SetVectorParameterValue _asmaterial_instance_dynamic "Emissive_A_ColorCW" Color :!class /Script/Engine.MaterialInstanceDynamic)
      (Rendering|Material|SetVectorParameterValue _asmaterial_instance_dynamic "Emissive_C_ColorCW" Color :!class /Script/Engine.MaterialInstanceDynamic)
      (Rendering|Material|SetVectorParameterValue _asmaterial_instance_dynamic "FDecal_BaseColor_Simple" Color :!class /Script/Engine.MaterialInstanceDynamic)
      (Rendering|Material|SetVectorParameterValue _asmaterial_instance_dynamic "Emissive_B_ColorCW" _returnvalue_3 :!class /Script/Engine.MaterialInstanceDynamic)
      (Rendering|Material|SetVectorParameterValue _asmaterial_instance_dynamic "Emissive_D_ColorCW" _returnvalue_3 :!class /Script/Engine.MaterialInstanceDynamic))))

(fn ConstructionScript)
