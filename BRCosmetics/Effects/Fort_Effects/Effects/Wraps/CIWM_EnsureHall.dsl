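; Weapon wrap modifier (EnsureHall) that builds up emissive glow on wrapped meshes as the weapon fires.
; Each shot stamps game time into three TopS emissive layers and raises a capped step count, scaling layer 03 brightness and darken; the cap is lower in the front end.
; The count decays lazily at the next shot after a hold period.
; Pawn, target, customization and weapon changes reset state after a short delay and reapply parameters.
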(event EventOnPawnSet
  ; Gets brightness and daken base values from material from topside emissive 03
  (Utilities|FlowControl|Delay 0.2)
  (Variables|Default|SetStepCount 0)
  (Variables|Default|Setwait_timer 0.0)
  (Variables|Default|SetBrightness 1.0)
  (Variables|Default|SetDarken 0.0)
  (CallFunction|Update))

(event EventOnSetWrapTarget
  ; Gets brightness and daken base values from material from topside emissive 03
  (Utilities|FlowControl|Delay 0.2)
  (Variables|Default|SetStepCount 0)
  (Variables|Default|Setwait_timer 0.0)
  (Variables|Default|SetBrightness 1.0)
  (Variables|Default|SetDarken 0.0)
  (CallFunction|Update))

(event EventOnCharacterCustomizationFinished
  (params (CharacterPartInfo :type Array%20of%20Applied%20Character%20Part%20Info%20Structures))
  ; Gets brightness and daken base values from material from topside emissive 03
  (Utilities|FlowControl|Delay 0.2)
  (Variables|Default|SetStepCount 0)
  (Variables|Default|Setwait_timer 0.0)
  (Variables|Default|SetBrightness 1.0)
  (Variables|Default|SetDarken 0.0)
  (CallFunction|Update))

(event EventOnWeaponChanged
  (params (NewWeapon :type Fort%20Weapon%20Object%20Reference) (OldWeapon :type Fort%20Weapon%20Object%20Reference))
  ; Gets brightness and daken base values from material from topside emissive 03
  (Utilities|FlowControl|Delay 0.2)
  (Variables|Default|SetStepCount 0)
  (Variables|Default|Setwait_timer 0.0)
  (Variables|Default|SetBrightness 1.0)
  (Variables|Default|SetDarken 0.0)
  (CallFunction|Update))

(event EventOnNewWeapons
  (params (WeaponsAsItems :type Array%20of%20Fort%20Item%20Object%20References))
  ; Gets brightness and daken base values from material from topside emissive 03
  (Utilities|FlowControl|Delay 0.2)
  (Variables|Default|SetStepCount 0)
  (Variables|Default|Setwait_timer 0.0)
  (Variables|Default|SetBrightness 1.0)
  (Variables|Default|SetDarken 0.0)
  (CallFunction|Update))

(event EventBeginPlay
  ; TopSide Emissive 03  build up
  (Utilities|FlowControl|Delay 0.2)
  (if (CustomCosmeticModifier|IsinFrontEnd)
    (Variables|Default|SetMaxSteps 5)
    (else
      (Variables|Default|SetMaxSteps 10))))

(event Custom|update
  (bind _returnvalue (CustomItemWrap|GetWrappedMeshComponents))
  (bind _maxsteps (Variables|Default|GetMaxSteps))
  (Rendering|Material|SetScalarParameterValueonMaterials _returnvalue "TopS Emissive_03-Brightness-TBP" (Utilities|Time|GetGameTimeinSeconds))
  (Rendering|Material|SetScalarParameterValueonMaterials _returnvalue "TopS Emissive_03_Brightness---Start" (Math|Float|Lerp 0.0 (Variables|Default|GetBrightness) (/ (Math|Conversions|ToFloat(Integer) (stale EventOnWeaponFired _output_get)) _maxsteps)))
  (Rendering|Material|SetScalarParameterValueonMaterials _returnvalue "TopS Emissive_03_Brightness_Darken_Base Start" (Math|Float|Lerp 0.0 (Variables|Default|GetDarken) (/ (Math|Conversions|ToFloat(Integer) (stale EventOnWeaponFired _output_get)) _maxsteps))))

(event EventOnWeaponFired
  (bind _returnvalue (CustomItemWrap|GetWrappedMeshComponents))
  (bind _returnvalue_2 (Utilities|Time|GetGameTimeinSeconds))
  (bind _returnvalue_3 (CustomItemWrap|GetWrappedMeshComponents))
  (bind _maxsteps (Variables|Default|GetMaxSteps))
  (bind _output_get (Variables|Default|SetStepCount (Math|Integer|Min(Integer) (+ (Math|Float|Truncate (Math|Float|Lerp (Math|Conversions|ToFloat(Integer) (Variables|Default|GetStepCount)) 0.0 (Math|Float|Clamp(Float) (/ (- (Utilities|Time|GetGameTimeinSeconds) (Variables|Default|Getwait_timer)) 1.0)))) 1) (Variables|Default|GetMaxSteps))))
  (Rendering|Material|SetScalarParameterValueonMaterials _returnvalue_3 "TopS Emissive_01-Brightness-TBP" _returnvalue_2)
  (Rendering|Material|SetScalarParameterValueonMaterials _returnvalue_3 "TopS Emissive_02-Brightness-TBP" _returnvalue_2)
  (Variables|Default|Setwait_timer (+ (Utilities|Time|GetGameTimeinSeconds) 5.5))
  (Rendering|Material|SetScalarParameterValueonMaterials _returnvalue "TopS Emissive_03-Brightness-TBP" (Utilities|Time|GetGameTimeinSeconds))
  (Rendering|Material|SetScalarParameterValueonMaterials _returnvalue "TopS Emissive_03_Brightness---Start" (Math|Float|Lerp 0.0 (Variables|Default|GetBrightness) (/ (Math|Conversions|ToFloat(Integer) _output_get) _maxsteps)))
  (Rendering|Material|SetScalarParameterValueonMaterials _returnvalue "TopS Emissive_03_Brightness_Darken_Base Start" (Math|Float|Lerp 0.0 (Variables|Default|GetDarken) (/ (Math|Conversions|ToFloat(Integer) _output_get) _maxsteps))))

(fn ConstructionScript)
