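; Fashion Feline weapon wrap Blueprint that drives reactive decal, emissive and flipbook animation on the wrapped meshes' materials.
; On begin play outside the front end, and on pawn set, it watches the AthenaKills stat and stamps game time into decal and emissive parameters.
; Firing raises flipbook speed, and a timer later zeroes it.
; In the front end, looping timers fake repeated kill pulses in bursts, separated by pauses.
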(event Custom|StopPunches
  (bind _output_get (Variables|Default|SetSpeed 0.0))
  (Rendering|Material|SetScalarParameterValueonMaterials (CustomItemWrap|GetWrappedMeshComponents) "FDecalV2_FlipBook_Speed" _output_get))

(event EventOnWeaponFired
  (if (not (>= (Variables|Default|GetSpeed) 1.2))
    (Variables|Default|SetSpeed 1.2)
    (Rendering|Material|SetScalarParameterValueonMaterials (CustomItemWrap|GetWrappedMeshComponents) "FDecalV2_FlipBook_Speed" (Variables|Default|GetSpeed)))
  (Utilities|Time|SetTimerbyFunctionName 0 "StopPunches" 2.5))

(event Custom|RestartFE
  (Utilities|Time|SetTimerbyEvent (delegate Custom|FrontendKO) 5.0))

(event EventBeginPlay
  (bind _returnvalue (Utilities|Time|GetGameTimeinSeconds))
  (bind _returnvalue_2 (Utilities|Time|GetGameTimeinSeconds))
  (bind _returnvalue_3 (CustomItemWrap|GetWrappedMeshComponents))
  (bind _returnvalue_4 (CustomItemWrap|GetWrappedMeshComponents))
  (bind _returnvalue_5 (Utilities|Time|GetGameTimeinSeconds))
  (bind _returnvalue_6 (CustomItemWrap|GetWrappedMeshComponents))
  (if (CustomCosmeticModifier|IsinFrontEnd)
    (Utilities|Time|SetTimerbyEvent (delegate Custom|FrontendKO) 5.0)
    (else
      (Cosmetics|WatchCosmeticStat (CustomCharacter|GetAssociatedPlayerPawn) "(TagName=\"StatManager.AthenaKills\")"
        (:OnStatChanged
          ;[1 1 1 0.5] Decal Reactive
          (Rendering|Material|SetScalarParameterValueonMaterials _returnvalue_6 "EDRec_TimeFrom_BP_DecalA" _returnvalue)
          (Rendering|Material|SetScalarParameterValueonMaterials _returnvalue_6 "EDRec_TimeFrom_BP_DecalB" _returnvalue)
          (Rendering|Material|SetScalarParameterValueonMaterials _returnvalue_6 "EDRec_TimeFrom_BP_DecalC" _returnvalue)
          (Rendering|Material|SetScalarParameterValueonMaterials _returnvalue_6 "EDRec_TimeFrom_BP_DecalD" _returnvalue)
          (Rendering|Material|SetScalarParameterValueonMaterials _returnvalue_4 "EDRec_Emissive_A_Brightness_Time_BP" _returnvalue_5)
          (Rendering|Material|SetScalarParameterValueonMaterials _returnvalue_4 "EDRec_Emissive_B_Brightness_Time_BP" _returnvalue_5)
          (Rendering|Material|SetScalarParameterValueonMaterials _returnvalue_4 "EDRec_Emissive_C_Brightness_Time_BP" _returnvalue_5)
          (Rendering|Material|SetScalarParameterValueonMaterials _returnvalue_4 "EDRec_Emissive_D_Brightness_Time_BP" _returnvalue_5)
          (Rendering|Material|SetScalarParameterValueonMaterials _returnvalue_3 "EDRec_Emissive_A_Gradent_Time_BP" _returnvalue_2)
          (Rendering|Material|SetScalarParameterValueonMaterials _returnvalue_3 "EDRec_Emissive_B_Gradent_Time_BP" _returnvalue_2)
          (Rendering|Material|SetScalarParameterValueonMaterials _returnvalue_3 "EDRec_Emissive_C_Gradent_Time_BP" _returnvalue_2)
          (Rendering|Material|SetScalarParameterValueonMaterials _returnvalue_3 "EDRec_Emissive_D_Gradent_Time_BP" _returnvalue_2))))))

(event Custom|FrontendKO
  (bind _returnvalue (Utilities|Time|SetTimerbyEvent (delegate Custom|FEExtended_KO) 0.5 true))
  (Variables|Default|SetCurrentTimer _returnvalue))

(event Custom|FEExtended_KO
  (bind _returnvalue (Utilities|Time|GetGameTimeinSeconds))
  (bind _returnvalue_2 (Utilities|Time|GetGameTimeinSeconds))
  (bind _returnvalue_3 (CustomItemWrap|GetWrappedMeshComponents))
  (bind _returnvalue_4 (CustomItemWrap|GetWrappedMeshComponents))
  (bind _returnvalue_5 (Utilities|Time|GetGameTimeinSeconds))
  (bind _returnvalue_6 (CustomItemWrap|GetWrappedMeshComponents))
  (bind _result (Math|Integer|IncrementInt (Variables|Default|GetIterationCount))
    (:%20%20
      (if (>= _result 10)
        (Utilities|Time|ClearandInvalidateTimerbyHandle (Variables|Default|GetCurrentTimer))
        (Variables|Default|SetIterationCount 0)
        (CallFunction|RestartFE)
        (else
          ;[1 1 1 0.5] Decal Reactive
          (Rendering|Material|SetScalarParameterValueonMaterials _returnvalue_6 "EDRec_TimeFrom_BP_DecalA" _returnvalue)
          (Rendering|Material|SetScalarParameterValueonMaterials _returnvalue_6 "EDRec_TimeFrom_BP_DecalB" _returnvalue)
          (Rendering|Material|SetScalarParameterValueonMaterials _returnvalue_6 "EDRec_TimeFrom_BP_DecalC" _returnvalue)
          (Rendering|Material|SetScalarParameterValueonMaterials _returnvalue_6 "EDRec_TimeFrom_BP_DecalD" _returnvalue)
          (Rendering|Material|SetScalarParameterValueonMaterials _returnvalue_4 "EDRec_Emissive_A_Brightness_Time_BP" _returnvalue_5)
          (Rendering|Material|SetScalarParameterValueonMaterials _returnvalue_4 "EDRec_Emissive_B_Brightness_Time_BP" _returnvalue_5)
          (Rendering|Material|SetScalarParameterValueonMaterials _returnvalue_4 "EDRec_Emissive_C_Brightness_Time_BP" _returnvalue_5)
          (Rendering|Material|SetScalarParameterValueonMaterials _returnvalue_4 "EDRec_Emissive_D_Brightness_Time_BP" _returnvalue_5)
          (Rendering|Material|SetScalarParameterValueonMaterials _returnvalue_3 "EDRec_Emissive_A_Gradent_Time_BP" _returnvalue_2)
          (Rendering|Material|SetScalarParameterValueonMaterials _returnvalue_3 "EDRec_Emissive_B_Gradent_Time_BP" _returnvalue_2)
          (Rendering|Material|SetScalarParameterValueonMaterials _returnvalue_3 "EDRec_Emissive_C_Gradent_Time_BP" _returnvalue_2)
          (Rendering|Material|SetScalarParameterValueonMaterials _returnvalue_3 "EDRec_Emissive_D_Gradent_Time_BP" _returnvalue_2))))))

(event EventOnPawnSet
  (bind _returnvalue (Utilities|Time|GetGameTimeinSeconds))
  (bind _returnvalue_2 (Utilities|Time|GetGameTimeinSeconds))
  (bind _returnvalue_3 (CustomItemWrap|GetWrappedMeshComponents))
  (bind _returnvalue_4 (CustomItemWrap|GetWrappedMeshComponents))
  (bind _returnvalue_5 (Utilities|Time|GetGameTimeinSeconds))
  (bind _returnvalue_6 (CustomItemWrap|GetWrappedMeshComponents))
  (Cosmetics|WatchCosmeticStat (CustomCharacter|GetAssociatedPlayerPawn) "(TagName=\"StatManager.AthenaKills\")"
    (:OnStatChanged
      ;[1 1 1 0.5] Decal Reactive
      (Rendering|Material|SetScalarParameterValueonMaterials _returnvalue_6 "EDRec_TimeFrom_BP_DecalA" _returnvalue)
      (Rendering|Material|SetScalarParameterValueonMaterials _returnvalue_6 "EDRec_TimeFrom_BP_DecalB" _returnvalue)
      (Rendering|Material|SetScalarParameterValueonMaterials _returnvalue_6 "EDRec_TimeFrom_BP_DecalC" _returnvalue)
      (Rendering|Material|SetScalarParameterValueonMaterials _returnvalue_6 "EDRec_TimeFrom_BP_DecalD" _returnvalue)
      (Rendering|Material|SetScalarParameterValueonMaterials _returnvalue_4 "EDRec_Emissive_A_Brightness_Time_BP" _returnvalue_5)
      (Rendering|Material|SetScalarParameterValueonMaterials _returnvalue_4 "EDRec_Emissive_B_Brightness_Time_BP" _returnvalue_5)
      (Rendering|Material|SetScalarParameterValueonMaterials _returnvalue_4 "EDRec_Emissive_C_Brightness_Time_BP" _returnvalue_5)
      (Rendering|Material|SetScalarParameterValueonMaterials _returnvalue_4 "EDRec_Emissive_D_Brightness_Time_BP" _returnvalue_5)
      (Rendering|Material|SetScalarParameterValueonMaterials _returnvalue_3 "EDRec_Emissive_A_Gradent_Time_BP" _returnvalue_2)
      (Rendering|Material|SetScalarParameterValueonMaterials _returnvalue_3 "EDRec_Emissive_B_Gradent_Time_BP" _returnvalue_2)
      (Rendering|Material|SetScalarParameterValueonMaterials _returnvalue_3 "EDRec_Emissive_C_Gradent_Time_BP" _returnvalue_2)
      (Rendering|Material|SetScalarParameterValueonMaterials _returnvalue_3 "EDRec_Emissive_D_Gradent_Time_BP" _returnvalue_2))))

(fn ConstructionScript)
