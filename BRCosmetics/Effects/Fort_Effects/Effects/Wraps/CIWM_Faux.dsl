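; Custom item wrap Blueprint for the Faux weapon wrap, which reacts visually to the player's kills and weapon fire.
; When the pawn is set, it watches the AthenaKills cosmetic stat, toggling an Active flag and the zipper invert mask on each change.
; Game time is written to material parameters on the wrapped meshes to time the effects, including an emissive timestamp shortly after each shot.
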(event EventOnPawnSet
  (bind _returnvalue (CustomItemWrap|GetWrappedMeshComponents))
  (bind _asfort_player_pawn (Utilities|Casting|CastToFortPlayerPawn (CustomCharacter|GetAssociatedPlayerPawn)))
  (Cosmetics|WatchCosmeticStat _asfort_player_pawn "(TagName=\"StatManager.AthenaKills\")"
    (:OnStatChanged
      (if (Variables|Default|GetActive)
        (Rendering|Material|SetScalarParameterValueonMaterials _returnvalue "Faux_Zipper_Invert_Mask" 0.0)
        (Rendering|Material|SetScalarParameterValueonMaterials _returnvalue "Faux_TimeFrom_BP" (Utilities|Time|GetGameTimeinSeconds))
        (Variables|Default|SetActive false)
        (else
          (Rendering|Material|SetScalarParameterValueonMaterials _returnvalue "Faux_Zipper_Invert_Mask" 1.0)
          (Rendering|Material|SetScalarParameterValueonMaterials _returnvalue "Faux_TimeFrom_BP" (Utilities|Time|GetGameTimeinSeconds))
          (Variables|Default|SetActive true))))))

(event EventBeginPlay
  (|CallParentFunction))

(event EventOnWeaponFired
  (Utilities|FlowControl|Delay 0.01)
  (Rendering|Material|SetScalarParameterValueonMaterials (CustomItemWrap|GetWrappedMeshComponents) "Faux_TimeFrom_BP_Emis" (Utilities|Time|GetGameTimeinSeconds)))

(fn ConstructionScript)
