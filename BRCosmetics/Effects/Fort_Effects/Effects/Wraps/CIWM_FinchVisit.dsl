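; Item wrap cosmetic modifier that animates a flipbook decal on the wrapped meshes in three stages as the player gets kills.
; Each watched AthenaKills change increments a capped counter and plays the Anim timeline, lerping the material's frame parameter across that third.
; A Pause flag ignores triggers during playback; after the third stage, a timer rewinds the whole sequence at half speed.
; In the front end, a separate timeline drives the same progression without kills.
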(event EventBeginPlay
  (Utilities|FlowControl|Delay 0.2)
  (if (CustomCosmeticModifier|IsinFrontEnd)
    (Timeline|Timeline PlayFromStart
      (:NewTrack_0
        (if (not (Variables|Default|GetPause))
          (Variables|Default|SetPause true)
          (if (not (>= (Variables|Default|GetKill_Count) 3))
            (Math|Integer|IncrementInt (Variables|Default|GetKill_Count)
              (:%20%20
                (Components|Timeline|SetPlayRate (Variables|CIWM_FinchVisit|GetAnim) 1.0)
                (bind (_direction _newtrack_0) (Timeline|Anim PlayFromStart)
                  (:Update
                    (Rendering|Material|SetScalarParameterValueonMaterials (CustomItemWrap|GetWrappedMeshComponents) "FDecalV2_FlipBook_Frame #" (Math|Float|Lerp (Utilities|Select :Index (Variables|Default|GetKill_Count) :Option%200 0.0 :Option%201 0.0 :Option%202 0.33 :Option%203 0.66) (Utilities|Select :Index (Variables|Default|GetKill_Count) :Option%200 1.0 :Option%201 0.33 :Option%202 0.66 :Option%203 1.0) _newtrack_0)))
                  (:Finished
                    (Variables|Default|SetPause false)
                    (if (>= (Variables|Default|GetKill_Count) 3)
                      (Utilities|Time|SetTimerbyFunctionName 0 "Restart" 5.0))))))))))))

(event EventOnPawnSet
  (Cosmetics|WatchCosmeticStat (CustomCharacter|GetAssociatedPlayerPawn) "(TagName=\"StatManager.AthenaKills\")" 0 false
    (:OnStatChanged
      (if (not (Variables|Default|GetPause))
        (Variables|Default|SetPause true)
        (if (not (>= (Variables|Default|GetKill_Count) 3))
          (Math|Integer|IncrementInt (Variables|Default|GetKill_Count)
            (:%20%20
              (Components|Timeline|SetPlayRate (Variables|CIWM_FinchVisit|GetAnim) 1.0)
              (Timeline|Anim PlayFromStart))))))))

(event Custom|Restart
  (Variables|Default|SetPause true)
  (Variables|Default|SetKill_Count 0)
  (Components|Timeline|SetPlayRate (Variables|CIWM_FinchVisit|GetAnim) 0.5)
  (Timeline|Anim ReverseFromEnd))

(fn ConstructionScript)
