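; Custom item wrap modifier (FlakeSlide) that shifts the wrap material in stages as its pawn takes damage, then recovers.
; DamageReceived stat changes, or timeline events in the front end, increment a capped counter and play the Holes timeline at half rate, driving a blend-offset parameter on wrapped meshes.
; At the cap, a retriggerable delay reverses the timeline at full rate and resets the counter.
; RandomSet randomizes flipbook, offset and rotation parameters but is uncalled here.
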(event EventBeginPlay
  (Utilities|FlowControl|Delay 0.2)
  (if (CustomCosmeticModifier|IsinFrontEnd)
    (Timeline|Timeline PlayFromStart
      (:NewTrack_0
        (if (not (>= (Variables|Default|GetDamage_Count) 2))
          (Math|Integer|IncrementInt (Variables|Default|GetDamage_Count)
            (:%20%20
              (Variables|Default|GetRecover
                (:then)
                (:else
                  (Components|Timeline|SetPlayRate (Variables|CIWM_FlakeSlide|GetHoles) 0.5)
                  (bind (_direction _2 _4 _1 _1b _4b) (Timeline|Holes PlayFromStart)
                    (:Update
                      (Rendering|Material|SetVectorParameterValueonMaterials (CustomItemWrap|GetWrappedMeshComponents) "TopS_Blend_Offset__Master" (Math|Vector|MakeVector (Math|Float|SelectFloat (Utilities|Select :Index (Variables|Default|GetDamage_Count) :Option%200 _2 :Option%201 _4b :Option%202 _1b :Option%203 0.0) (Utilities|Select :Index (Variables|Default|GetDamage_Count) :Option%200 _2 :Option%201 _4 :Option%202 _1) (Variables|Default|GetRecover)))))
                    (:Finished
                      (Variables|Default|GetRecover
                        (:then
                          (Variables|Default|SetDamage_Count -1)
                          (Variables|Default|SetRecover false))
                        (:else
                          _)))))))))
        (Utilities|FlowControl|RetriggerableDelay 5.0)
        (Variables|Default|SetRecover true)
        (CallFunction|Revert)))))

(event EventOnPawnSet
  (Cosmetics|WatchCosmeticStat (CustomCharacter|GetAssociatedPlayerPawn) "(TagName=\"StatManager.DamageReceived\")" 0 false
    (:OnStatChanged
      (if (>= (Variables|Default|GetDamage_Count) 2)
        (Utilities|FlowControl|RetriggerableDelay 5.0)
        (Variables|Default|SetRecover true)
        (CallFunction|Revert)
        (else
          (Math|Integer|IncrementInt (Variables|Default|GetDamage_Count)
            (:%20%20
              (Variables|Default|GetRecover
                (:then)
                (:else
                  (Components|Timeline|SetPlayRate (Variables|CIWM_FlakeSlide|GetHoles) 0.5)
                  (Timeline|Holes PlayFromStart))))))))))

(event Custom|Revert
  (Components|Timeline|SetPlayRate (Variables|CIWM_FlakeSlide|GetHoles) 1.0)
  (Timeline|Holes ReverseFromEnd))

(fn RandomSet
  (params (FlipBook%20Frame :type Name) (TimeBP :type Name) (Offset :type Name) (Rotation :type Name) (X=min :type Float%20%28double-precision%29) (X=Max :type Float%20%28double-precision%29) (Y=Min :type Float%20%28double-precision%29) (Y=Max :type Float%20%28double-precision%29) (Rot=Min :type Float%20%28double-precision%29) (Rot=Max :type Float%20%28double-precision%29) (TimeBP-first :type Name))
  (bind _returnvalue (CustomItemWrap|GetWrappedMeshComponents))
  (bind _stream (Variables|Default|GetStream))
  (bind _returnvalue_2 (Math|Random|RandomFloatinRangefromStream _stream X=min X=Max))
  (bind _returnvalue_3 (Math|Random|RandomIntegerinRange 0 8))
  (bind _returnvalue_4 (Math|Conversions|ToFloat(Integer) _returnvalue_3))
  (bind _returnvalue_5 (/ _returnvalue_4 8.0))
  (Rendering|Material|SetScalarParameterValueonMaterials _returnvalue FlipBook%20Frame _returnvalue_5)
  (Rendering|Material|SetScalarParameterValueonMaterials _returnvalue TimeBP-first (Utilities|Time|GetGameTimeinSeconds))
  (Rendering|Material|SetVectorParameterValueonMaterials _returnvalue Offset (Math|Vector|MakeVector _returnvalue_2 (Math|Random|RandomFloatinRangefromStream _stream Y=Min Y=Max)))
  (Rendering|Material|SetScalarParameterValueonMaterials _returnvalue Rotation (Math|Random|RandomFloatinRange Rot=Min Rot=Max))
  (Rendering|Material|SetScalarParameterValueonMaterials _returnvalue TimeBP (Utilities|Time|GetGameTimeinSeconds))
  (Variables|Default|SetNewPosition _returnvalue_2)
  (Development|PrintString (Utilities|String|ToString(Float) _returnvalue_5) true true "(R=0.060001,G=1.000000,B=0.000000,A=1.000000)" 3.0))

(fn ConstructionScript)
