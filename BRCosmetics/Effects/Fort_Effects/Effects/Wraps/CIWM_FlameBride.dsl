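; FlameBride weapon wrap modifier that drives blend and emissive material offsets on the wrapped meshes from the player's health.
; Watchers on health-gained and damage-received stats accompany delayed updates that write start and end offsets plus game-time stamps, then store the new starting health.
; Weapon change events reapply the end offsets after a short delay.
; In the front end, a timeline substitutes simulated health; debug health printing remains in the update.
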(event EventOnPawnSet
  (bind _returnvalue (CustomCharacter|GetAssociatedPlayerPawn))
  (bind _returnvalue_2 (CustomItemWrap|GetWrappedMeshComponents))
  (bind _returnvalue_3 (CustomCharacter|GetAssociatedPlayerPawn))
  (bind _starting_health (Variables|Default|GetStarting_Health))
  (bind _returnvalue_4 (Math|Float|Lerp 0.4 -0.2 _starting_health))
  (bind _returnvalue_5 (FortPawn|GetHealth _returnvalue_3))
  (bind _returnvalue_6 (CustomItemWrap|GetWrappedMeshComponents))
  (bind _returnvalue_7 (Utilities|Time|GetGameTimeinSeconds))
  (bind _returnvalue_8 (CustomItemWrap|GetWrappedMeshComponents))
  (bind _returnvalue_9 (CustomItemWrap|GetWrappedMeshComponents))
  (bind _returnvalue_10 (Utilities|Time|GetGameTimeinSeconds))
  (bind _returnvalue_11 (* _returnvalue_5 0.01))
  (bind _frontend (Variables|Default|GetFrontEnd))
  (Utilities|FlowControl|Sequence
    (:then_0
      (Cosmetics|WatchCosmeticStat _returnvalue "(TagName=\"StatManager.GainedHealthTimes\")" 0 false
        (:OnStatChanged
          _)))
    (:then_1
      (Cosmetics|WatchCosmeticStat _returnvalue "(TagName=\"StatManager.DamageReceived\")" 0 false
        (:OnStatChanged
          _))))
  (Utilities|FlowControl|Delay 0.1)
  (Development|PrintText (Utilities|Text|ToText(Float) (* (FortPawn|GetHealth (CustomCharacter|GetAssociatedPlayerPawn)) 0.01)) true true "(R=1.000000,G=0.902830,B=0.002053,A=1.000000)" 5.0)
  (Rendering|Material|SetScalarParameterValueonMaterials _returnvalue_6 "TopS_01_Blend_Offset---Start" _returnvalue_4)
  (Rendering|Material|SetScalarParameterValueonMaterials _returnvalue_6 "TopS Emissive_01_Offset---Start" _returnvalue_4)
  (Rendering|Material|SetScalarParameterValueonMaterials _returnvalue_6 "TopS Emissive_03_Offset---Start" _returnvalue_4)
  (Rendering|Material|SetScalarParameterValueonMaterials _returnvalue_8 "TopS_01_Blend_Offset---End" (Math|Float|Lerp 0.4 -0.2 (Math|Float|Lerp _returnvalue_11 (stale EventBeginPlay _health) _frontend)))
  (Rendering|Material|SetScalarParameterValueonMaterials _returnvalue_8 "TopS Emissive_01_Offset---End" (Math|Float|Lerp 0.4 -0.2 (Math|Float|Lerp _returnvalue_11 (stale EventBeginPlay _health) _frontend)))
  (Rendering|Material|SetScalarParameterValueonMaterials _returnvalue_8 "TopS Emissive_03_Offset---End" (Math|Float|Lerp 0.4 -0.2 (Math|Float|Lerp _returnvalue_11 (stale EventBeginPlay _health) _frontend)))
  (Rendering|Material|SetScalarParameterValueonMaterials _returnvalue_9 "TopS_01_Blend_Offset-TBP" _returnvalue_10)
  (Rendering|Material|SetScalarParameterValueonMaterials _returnvalue_9 "TopS Emissive_01_Offset----TBP" _returnvalue_10)
  (Rendering|Material|SetScalarParameterValueonMaterials _returnvalue_9 "TopS Emissive_03_Offset----TBP" _returnvalue_10)
  (Rendering|Material|SetScalarParameterValueonMaterials _returnvalue_2 "TopS Emissive_02-Brightness-TBP" _returnvalue_7)
  (Rendering|Material|SetScalarParameterValueonMaterials _returnvalue_2 "TopS Emissive_03-Brightness-TBP" _returnvalue_7)
  (Utilities|FlowControl|Delay 0.2)
  (Variables|Default|SetPause false)
  (bind _output_get (Variables|Default|SetStarting_Health (Math|Float|Lerp _returnvalue_11 (stale EventBeginPlay _health) _frontend))))

(event EventOnNewWeapons
  (params (WeaponsAsItems :type Array%20of%20Fort%20Item%20Object%20References))
  (bind _returnvalue (CustomCharacter|GetAssociatedPlayerPawn))
  (bind _returnvalue_2 (FortPawn|GetHealth _returnvalue))
  (bind _returnvalue_3 (* _returnvalue_2 0.01))
  (bind _returnvalue_4 (Math|Float|Lerp 0.4 -0.2 _returnvalue_3))
  (bind _returnvalue_5 (CustomItemWrap|GetWrappedMeshComponents))
  (Utilities|FlowControl|Delay 0.05)
  (Variables|Default|SetStarting_Health _returnvalue_3)
  (Rendering|Material|SetScalarParameterValueonMaterials _returnvalue_5 "TopS_01_Blend_Offset---End" _returnvalue_4)
  (Rendering|Material|SetScalarParameterValueonMaterials _returnvalue_5 "TopS Emissive_01_Offset---End" _returnvalue_4)
  (Rendering|Material|SetScalarParameterValueonMaterials _returnvalue_5 "TopS Emissive_03_Offset---End" _returnvalue_4))

(event EventOnNewPlayerWeaponType
  (params (NewType :type EFortItemType%20Enum :default "WorldItem") (OldType :type EFortItemType%20Enum :default "WorldItem"))
  (bind _returnvalue (CustomCharacter|GetAssociatedPlayerPawn))
  (bind _returnvalue_2 (FortPawn|GetHealth _returnvalue))
  (bind _returnvalue_3 (* _returnvalue_2 0.01))
  (bind _returnvalue_4 (Math|Float|Lerp 0.4 -0.2 _returnvalue_3))
  (bind _returnvalue_5 (CustomItemWrap|GetWrappedMeshComponents))
  (Utilities|FlowControl|Delay 0.05)
  (Variables|Default|SetStarting_Health _returnvalue_3)
  (Rendering|Material|SetScalarParameterValueonMaterials _returnvalue_5 "TopS_01_Blend_Offset---End" _returnvalue_4)
  (Rendering|Material|SetScalarParameterValueonMaterials _returnvalue_5 "TopS Emissive_01_Offset---End" _returnvalue_4)
  (Rendering|Material|SetScalarParameterValueonMaterials _returnvalue_5 "TopS Emissive_03_Offset---End" _returnvalue_4))

(event EventOnWeaponChanged
  (params (NewWeapon :type Fort%20Weapon%20Object%20Reference) (OldWeapon :type Fort%20Weapon%20Object%20Reference))
  (bind _returnvalue (CustomCharacter|GetAssociatedPlayerPawn))
  (bind _returnvalue_2 (FortPawn|GetHealth _returnvalue))
  (bind _returnvalue_3 (* _returnvalue_2 0.01))
  (bind _returnvalue_4 (Math|Float|Lerp 0.4 -0.2 _returnvalue_3))
  (bind _returnvalue_5 (CustomItemWrap|GetWrappedMeshComponents))
  (Utilities|FlowControl|Delay 0.05)
  (Variables|Default|SetStarting_Health _returnvalue_3)
  (Rendering|Material|SetScalarParameterValueonMaterials _returnvalue_5 "TopS_01_Blend_Offset---End" _returnvalue_4)
  (Rendering|Material|SetScalarParameterValueonMaterials _returnvalue_5 "TopS Emissive_01_Offset---End" _returnvalue_4)
  (Rendering|Material|SetScalarParameterValueonMaterials _returnvalue_5 "TopS Emissive_03_Offset---End" _returnvalue_4))

(event EventBeginPlay
  (bind _returnvalue (CustomItemWrap|GetWrappedMeshComponents))
  (bind _returnvalue_2 (CustomCharacter|GetAssociatedPlayerPawn))
  (bind _starting_health (Variables|Default|GetStarting_Health))
  (bind _returnvalue_3 (Math|Float|Lerp 0.4 -0.2 _starting_health))
  (bind _returnvalue_4 (FortPawn|GetHealth _returnvalue_2))
  (bind _returnvalue_5 (CustomItemWrap|GetWrappedMeshComponents))
  (bind _returnvalue_6 (Utilities|Time|GetGameTimeinSeconds))
  (bind _returnvalue_7 (CustomItemWrap|GetWrappedMeshComponents))
  (bind _returnvalue_8 (CustomItemWrap|GetWrappedMeshComponents))
  (bind _returnvalue_9 (Utilities|Time|GetGameTimeinSeconds))
  (bind _returnvalue_10 (* _returnvalue_4 0.01))
  (bind _frontend (Variables|Default|GetFrontEnd))
  (Utilities|FlowControl|Delay 0.2)
  (if (CustomCosmeticModifier|IsinFrontEnd)
    (Utilities|FlowControl|Sequence
      (:then_0
        (Variables|Default|SetFrontEnd 1.0))
      (:then_1
        (Variables|Default|SetStarting_Health 1.0)
        (bind (_direction _health) (Timeline|Timeline Play)
          (:Events
            (Utilities|FlowControl|Delay 0.1)
            (Development|PrintText (Utilities|Text|ToText(Float) (* (FortPawn|GetHealth (CustomCharacter|GetAssociatedPlayerPawn)) 0.01)) true true "(R=1.000000,G=0.902830,B=0.002053,A=1.000000)" 5.0)
            (Rendering|Material|SetScalarParameterValueonMaterials _returnvalue_5 "TopS_01_Blend_Offset---Start" _returnvalue_3)
            (Rendering|Material|SetScalarParameterValueonMaterials _returnvalue_5 "TopS Emissive_01_Offset---Start" _returnvalue_3)
            (Rendering|Material|SetScalarParameterValueonMaterials _returnvalue_5 "TopS Emissive_03_Offset---Start" _returnvalue_3)
            (Rendering|Material|SetScalarParameterValueonMaterials _returnvalue_7 "TopS_01_Blend_Offset---End" (Math|Float|Lerp 0.4 -0.2 (Math|Float|Lerp _returnvalue_10 _health _frontend)))
            (Rendering|Material|SetScalarParameterValueonMaterials _returnvalue_7 "TopS Emissive_01_Offset---End" (Math|Float|Lerp 0.4 -0.2 (Math|Float|Lerp _returnvalue_10 _health _frontend)))
            (Rendering|Material|SetScalarParameterValueonMaterials _returnvalue_7 "TopS Emissive_03_Offset---End" (Math|Float|Lerp 0.4 -0.2 (Math|Float|Lerp _returnvalue_10 _health _frontend)))
            (Rendering|Material|SetScalarParameterValueonMaterials _returnvalue_8 "TopS_01_Blend_Offset-TBP" _returnvalue_9)
            (Rendering|Material|SetScalarParameterValueonMaterials _returnvalue_8 "TopS Emissive_01_Offset----TBP" _returnvalue_9)
            (Rendering|Material|SetScalarParameterValueonMaterials _returnvalue_8 "TopS Emissive_03_Offset----TBP" _returnvalue_9)
            (Rendering|Material|SetScalarParameterValueonMaterials _returnvalue "TopS Emissive_02-Brightness-TBP" _returnvalue_6)
            (Rendering|Material|SetScalarParameterValueonMaterials _returnvalue "TopS Emissive_03-Brightness-TBP" _returnvalue_6)
            (Utilities|FlowControl|Delay 0.2)
            (Variables|Default|SetPause false)
            (bind _output_get (Variables|Default|SetStarting_Health (Math|Float|Lerp _returnvalue_10 _health _frontend)))))))
    (else
      (Variables|Default|SetFrontEnd 0.0))))

(fn ConstructionScript)
