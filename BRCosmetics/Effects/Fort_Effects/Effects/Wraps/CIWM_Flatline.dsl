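; Cosmetic item wrap Blueprint for Flatline, driving a heartbeat-themed material that reacts to the wearer's stats.
; On pawn set, it watches kills, damage received, health gains and shield gains, each stamping game time, flash duration and color onto wrapped-mesh materials.
; UpdatePanSpeed, not called within this file, sets gradient pan speed: fixed in the front end, stepped during the transition window, zero afterward, with debug prints.
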(event EventBeginPlay
  (|CallParentFunction))

(event EventOnPawnSet
  (bind _returnvalue (CustomItemWrap|GetWrappedMeshComponents))
  (bind _returnvalue_2 (Utilities|Time|GetGameTimeinSeconds))
  (bind _returnvalue_3 (CustomCharacter|GetAssociatedPlayerPawn))
  (Utilities|FlowControl|Sequence
    (:then_0
      ;[1 1 1] Kills
      (bind _asfort_player_pawn (Utilities|Casting|CastToFortPlayerPawn _returnvalue_3))
      (Cosmetics|WatchCosmeticStat _asfort_player_pawn "(TagName=\"StatManager.AthenaKills\")"
        (:OnStatChanged
          (Rendering|Material|SetScalarParameterValueonMaterials _returnvalue "TimeFrom-BP----------------" _returnvalue_2)
          (Rendering|Material|SetScalarParameterValueonMaterials _returnvalue "FL_Time_Duration" 3.0)
          (Rendering|Material|SetVectorParameterValueonMaterials _returnvalue "FL_BaseColor_Reactive" "1.000000,1.000000,1.000000"))))
    (:then_1
      ;[1 1 1] Damage Recieved
      (bind _asfort_player_pawn_2 (Utilities|Casting|CastToFortPlayerPawn _returnvalue_3))
      (Cosmetics|WatchCosmeticStat _asfort_player_pawn_2 "(TagName=\"StatManager.DamageReceived\")"
        (:OnStatChanged
          (Rendering|Material|SetScalarParameterValueonMaterials _returnvalue "TimeFrom-BP----------------" _returnvalue_2)
          (Rendering|Material|SetScalarParameterValueonMaterials _returnvalue "FL_Time_Duration" 2.0)
          (Rendering|Material|SetVectorParameterValueonMaterials _returnvalue "FL_BaseColor_Reactive" "3.000000,0.100000,0.100000"))))
    (:then_2
      ;[1 1 1] Health Gained
      (bind _asfort_player_pawn_3 (Utilities|Casting|CastToFortPlayerPawn _returnvalue_3))
      (Cosmetics|WatchCosmeticStat _asfort_player_pawn_3 "(TagName=\"StatManager.GainedHealthTimes\")"
        (:OnStatChanged
          (Rendering|Material|SetScalarParameterValueonMaterials _returnvalue "TimeFrom-BP----------------" _returnvalue_2)
          (Rendering|Material|SetScalarParameterValueonMaterials _returnvalue "FL_Time_Duration" 3.0)
          (Rendering|Material|SetVectorParameterValueonMaterials _returnvalue "FL_BaseColor_Reactive" "0.100000,3.000000,0.100000"))))
    (:then_3
      ;[1 1 1] Shields Gained
      (bind _asfort_player_pawn_4 (Utilities|Casting|CastToFortPlayerPawn _returnvalue_3))
      (Cosmetics|WatchCosmeticStat _asfort_player_pawn_4 "(TagName=\"StatManager.GainedShieldTimes\")"
        (:OnStatChanged
          (Rendering|Material|SetScalarParameterValueonMaterials _returnvalue "TimeFrom-BP----------------" _returnvalue_2)
          (Rendering|Material|SetScalarParameterValueonMaterials _returnvalue "FL_Time_Duration" 3.0)
          (Rendering|Material|SetVectorParameterValueonMaterials _returnvalue "FL_BaseColor_Reactive" "0.100000,0.975000,3.000000"))))))

(fn UpdatePanSpeed
  (bind _returnvalue (CustomItemWrap|GetWrappedMeshComponents))
  (bind _amount_of_time_to_complete_speed_transition_in_secs (Variables|Default|GetAmount%20of%20time%20to%20complete%20speed%20transition%20in%20Secs))
  ;[1 1 1] Heart Rate speed over time
  (bind _asfort_game_state (Utilities|Casting|CastToFortGameState (Game|GetGameState))
    (:then
      (if (CustomCosmeticModifier|IsinFrontEnd)
        (Rendering|Material|SetScalarParameterValueonMaterials _returnvalue "FL_Gradient_PanSpeed_OverTime" -0.25)
        (else
          (Development|PrintString (Utilities|String|ToString(Float) (/ (GameState|GetMatchTime _asfort_game_state) _amount_of_time_to_complete_speed_transition_in_secs)))
          (if (> (/ (GameState|GetMatchTime _asfort_game_state) _amount_of_time_to_complete_speed_transition_in_secs) 1.0)
            (Rendering|Material|SetScalarParameterValueonMaterials _returnvalue "FL_Gradient_PanSpeed_OverTime" 0.0)
            (else
              (bind _output_get (Variables|Default|Setanim%20time (Math|Float|Clamp(Float) (+ (Variables|Default|Getanim%20time) -0.05) -1.0 0.0)))
              (Rendering|Material|SetScalarParameterValueonMaterials _returnvalue "FL_Gradient_PanSpeed_OverTime" _output_get))))))
    (:CastFailed
      (Development|PrintString "Cannot Find game state"))))

(fn ConstructionScript)
