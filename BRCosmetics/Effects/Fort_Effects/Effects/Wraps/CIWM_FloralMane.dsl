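; Floral Mane cosmetic wrap Blueprint that shifts two material layers on wrapped meshes whenever the weapon fires.
; Each shot restarts a timeline whose offset is subtracted every update from two running totals, the second layer moving 1.5 times as fast.
; The totals drive X of per-item-type vector parameters (weapon, backpack, pickaxe, vehicle, vehicle side, character), with fixed Y nudges on some.
; Because offsets accumulate, the pattern advances rather than resetting between shots.
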(event EventOnWeaponFired
  (bind _returnvalue (CustomItemWrap|GetWrappedMeshComponents))
  (bind _returnvalue_2 (CustomItemWrap|GetWrappedMeshComponents))
  (bind (_direction _offset) (Timeline|Timeline PlayFromStart :NewTime 0.5)
    (:Update
      (bind _output_get (Variables|Default|SetOffset_A (- (Variables|Default|GetOffset_A) _offset)))
      (Rendering|Material|SetVectorParameterValueonMaterials _returnvalue_2 "TopS_01_OffsetS_Weapon" (Math|Vector|MakeVector _output_get))
      (Rendering|Material|SetVectorParameterValueonMaterials _returnvalue_2 "TopS_01_OffsetS_Backpack" (Math|Vector|MakeVector _output_get))
      (Rendering|Material|SetVectorParameterValueonMaterials _returnvalue_2 "TopS_01_OffsetS_Pickaxe" (Math|Vector|MakeVector _output_get))
      (Rendering|Material|SetVectorParameterValueonMaterials _returnvalue_2 "TopS_01_OffsetS_Vehicle" (Math|Vector|MakeVector _output_get -0.265))
      (Rendering|Material|SetVectorParameterValueonMaterials _returnvalue_2 "TopS_01_OffsetS_VehicleSide" (Math|Vector|MakeVector _output_get -0.15))
      (Rendering|Material|SetVectorParameterValueonMaterials _returnvalue_2 "TopS_01_OffsetS_Character" (Math|Vector|MakeVector _output_get))
      (bind _output_get_2 (Variables|Default|SetOffset_B (- (Variables|Default|GetOffset_B) (* _offset 1.5))))
      (Rendering|Material|SetVectorParameterValueonMaterials _returnvalue "TopS_02_OffsetS_Backpack" (Math|Vector|MakeVector _output_get_2))
      (Rendering|Material|SetVectorParameterValueonMaterials _returnvalue "TopS_02_OffsetS_Character" (Math|Vector|MakeVector _output_get_2))
      (Rendering|Material|SetVectorParameterValueonMaterials _returnvalue "TopS_02_OffsetS_Pickaxe" (Math|Vector|MakeVector _output_get_2))
      (Rendering|Material|SetVectorParameterValueonMaterials _returnvalue "TopS_02_OffsetS_Vehicle" (Math|Vector|MakeVector _output_get_2 -0.1))
      (Rendering|Material|SetVectorParameterValueonMaterials _returnvalue "TopS_02_OffsetS_VehicleSide" (Math|Vector|MakeVector _output_get_2))
      (Rendering|Material|SetVectorParameterValueonMaterials _returnvalue "TopS_02_OffsetS_Weapon" (Math|Vector|MakeVector _output_get_2 0.017)))))

(fn ConstructionScript)
