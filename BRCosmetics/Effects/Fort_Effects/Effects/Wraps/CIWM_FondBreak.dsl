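; FondBreak item wrap Blueprint that animates emissive decal and warp effects on wrapped meshes when the player takes damage.
; It watches the DamageReceived cosmetic stat; in the front end, a timeline event track triggers the same reaction instead.
; Each trigger stamps game time into material parameters, then drives a shared timeline for flipbook frame and warp.
; A Pause flag makes repeat triggers jump that timeline's time rather than restart it.
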(event EventBeginPlay
  (bind _returnvalue (Utilities|Time|GetGameTimeinSeconds))
  (bind _returnvalue_2 (CustomItemWrap|GetWrappedMeshComponents))
  (bind _returnvalue_3 (CustomItemWrap|GetWrappedMeshComponents))
  (bind _returnvalue_4 (CustomItemWrap|GetWrappedMeshComponents))
  (bind _returnvalue_5 (Utilities|Time|GetGameTimeinSeconds))
  (Utilities|FlowControl|Delay 0.2)
  (if (CustomCosmeticModifier|IsinFrontEnd)
    (Timeline|Timeline_0 PlayFromStart
      (:NewTrack_0
        (Utilities|FlowControl|Sequence
          (:then_0
            ;[1 1 1] Left Right v1
            (Rendering|Material|SetScalarParameterValueonMaterials _returnvalue_3 "FDecal_Emissive_Gradent_Time_BP" _returnvalue)
            (Rendering|Material|SetScalarParameterValueonMaterials _returnvalue_3 "FDecal_Opacity_Time_BP_Z" _returnvalue)
            (Rendering|Material|SetScalarParameterValueonMaterials _returnvalue_3 "FDecal_Emissive_Brightness_Time_BP" _returnvalue)
            (Rendering|Material|SetScalarParameterValueonMaterials _returnvalue_3 "TopS_03_Speed-TBP" _returnvalue))
          (:then_1
            ; Left\Right V2
            (Rendering|Material|SetScalarParameterValueonMaterials _returnvalue_4 "FDecalV2_FlipBook_Time_FromBP" _returnvalue_5)
            (Rendering|Material|SetScalarParameterValueonMaterials _returnvalue_4 "FDecalaV2_Opacity_Time_BP_Z" _returnvalue_5)
            (Rendering|Material|SetScalarParameterValueonMaterials _returnvalue_4 "FDecalV2_Emissive_Brightness_Time_BP" _returnvalue_5)
            (Rendering|Material|SetScalarParameterValueonMaterials _returnvalue_4 "FDecalV2_Emissive_Gradent_Time_BP" _returnvalue_5)
            (Rendering|Material|SetScalarParameterValueonMaterials _returnvalue_4 "FDecalV2_Counter_Brightness_Time_BP" _returnvalue_5))
          (:then_2
            (Variables|Default|GetPause
              (:then
                (bind (_direction _newtrack_0) (Timeline|Timeline SetNewTime :NewTime 1.0)
                  (:Update
                    (Rendering|Material|SetScalarParameterValueonMaterials _returnvalue_2 "FDecalV2_FlipBook_Frame #" _newtrack_0)
                    (Rendering|Material|SetScalarParameterValueonMaterials _returnvalue_2 "TopS_Em01_Warp_Amount" (Math|Float|Lerp 0.05 0.2 _newtrack_0))
                    (Rendering|Material|SetScalarParameterValueonMaterials _returnvalue_2 "TopS_Em01_Warp_Offset" (Math|Float|Lerp -0.2 -0.4 _newtrack_0)))
                  (:Finished
                    (Variables|Default|SetPause false))))
              (:else
                (Variables|Default|SetPause true)
                (Timeline|Timeline PlayFromStart)))))))))

(event EventOnPawnSet
  (bind _returnvalue (Utilities|Time|GetGameTimeinSeconds))
  (bind _returnvalue_2 (CustomItemWrap|GetWrappedMeshComponents))
  (bind _returnvalue_3 (CustomItemWrap|GetWrappedMeshComponents))
  (bind _returnvalue_4 (CustomItemWrap|GetWrappedMeshComponents))
  (bind _returnvalue_5 (Utilities|Time|GetGameTimeinSeconds))
  (Cosmetics|WatchCosmeticStat (CustomCharacter|GetAssociatedPlayerPawn) "(TagName=\"StatManager.DamageReceived\")" 0 false
    (:OnStatChanged
      (Utilities|FlowControl|Sequence
        (:then_0
          ;[1 1 1] Left Right v1
          (Rendering|Material|SetScalarParameterValueonMaterials _returnvalue_3 "FDecal_Emissive_Gradent_Time_BP" _returnvalue)
          (Rendering|Material|SetScalarParameterValueonMaterials _returnvalue_3 "FDecal_Opacity_Time_BP_Z" _returnvalue)
          (Rendering|Material|SetScalarParameterValueonMaterials _returnvalue_3 "FDecal_Emissive_Brightness_Time_BP" _returnvalue)
          (Rendering|Material|SetScalarParameterValueonMaterials _returnvalue_3 "TopS_03_Speed-TBP" _returnvalue))
        (:then_1
          ; Left\Right V2
          (Rendering|Material|SetScalarParameterValueonMaterials _returnvalue_4 "FDecalV2_FlipBook_Time_FromBP" _returnvalue_5)
          (Rendering|Material|SetScalarParameterValueonMaterials _returnvalue_4 "FDecalaV2_Opacity_Time_BP_Z" _returnvalue_5)
          (Rendering|Material|SetScalarParameterValueonMaterials _returnvalue_4 "FDecalV2_Emissive_Brightness_Time_BP" _returnvalue_5)
          (Rendering|Material|SetScalarParameterValueonMaterials _returnvalue_4 "FDecalV2_Emissive_Gradent_Time_BP" _returnvalue_5)
          (Rendering|Material|SetScalarParameterValueonMaterials _returnvalue_4 "FDecalV2_Counter_Brightness_Time_BP" _returnvalue_5))
        (:then_2
          (Variables|Default|GetPause
            (:then
              (Timeline|Timeline SetNewTime))
            (:else
              (Variables|Default|SetPause true)
              (Timeline|Timeline PlayFromStart))))))))

(fn ConstructionScript)
