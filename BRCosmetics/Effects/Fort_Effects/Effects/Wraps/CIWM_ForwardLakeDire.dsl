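; Custom item wrap Blueprint that animates a saber-like emissive effect on wrapped weapon meshes when the weapon fires.
; A shot starts the Saber timeline, while shots fired during playback jump it to a fixed time.
; The timeline lerps two emissive offsets and randomly flickers the first layer's brightness, then zeroes both brightnesses and clears the Playing flag on finish.
; The other weapon and pawn events are empty.
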(event EventBeginPlay)

(event EventOnPawnSet)

(event EventOnNewWeapons
  (params (WeaponsAsItems :type Array%20of%20Fort%20Item%20Object%20References)))

(event EventOnWeaponChanged
  (params (NewWeapon :type Fort%20Weapon%20Object%20Reference) (OldWeapon :type Fort%20Weapon%20Object%20Reference)))

(event EventOnNewPlayerWeaponType
  (params (NewType :type EFortItemType%20Enum :default "WorldItem") (OldType :type EFortItemType%20Enum :default "WorldItem")))

(event EventOnWeaponFired
  (bind _returnvalue (CustomItemWrap|GetWrappedMeshComponents))
  (bind _returnvalue_2 (CustomItemWrap|GetWrappedMeshComponents))
  (if (Variables|Default|GetPlaying)
    (bind (_direction _newtrack_0 _newtrack_1) (Timeline|Saber SetNewTime :NewTime 1.5)
      (:Update
        (Rendering|Material|SetScalarParameterValueonMaterials _returnvalue_2 "TopS Emissive_02_Offset---End" (Math|Float|Lerp 0.55 0.05 _newtrack_0))
        (Rendering|Material|SetScalarParameterValueonMaterials _returnvalue_2 "TopS Emissive_01_Offset---End" (Math|Float|Lerp 0.55 0.05 _newtrack_0))
        (Rendering|Material|SetScalarParameterValueonMaterials _returnvalue_2 "TopS Emissive_01_Brightness---End" (Math|Float|Lerp 1.0 (Math|Random|RandomFloatinRange 0.1 0.5) _newtrack_1))
        (Rendering|Material|SetScalarParameterValueonMaterials _returnvalue_2 "TopS Emissive_02_Brightness---End" 1.0))
      (:Finished
        (Variables|Default|SetPlaying false)
        (Rendering|Material|SetScalarParameterValueonMaterials _returnvalue "TopS Emissive_01_Brightness---End" 0.0)
        (Rendering|Material|SetScalarParameterValueonMaterials _returnvalue "TopS Emissive_02_Brightness---End" 0.0)))
    (else
      (Variables|Default|SetPlaying true)
      (Timeline|Saber PlayFromStart))))

(fn ConstructionScript)
