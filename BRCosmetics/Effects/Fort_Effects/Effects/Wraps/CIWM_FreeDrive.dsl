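; FreeDrive item wrap modifier that animates the wrapped weapon's materials when the weapon fires.
; Each shot stamps game time into two emissive brightness parameters, then triggers one of four decal timers (A to D) selected in round-robin order.
; Each decal locks for the pause length, while selection advances after that length divided by a multiplier, staggering the decals.
; BeginPlay sets a lower multiplier in the front end than in game.
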(event EventBeginPlay
  (Utilities|FlowControl|Delay 0.05)
  (if (CustomCosmeticModifier|IsinFrontEnd)
    (Variables|Default|SetPause%20Mulplyer 2.0)
    (else
      (Variables|Default|SetPause%20Mulplyer 4.0))))

(event EventOnWeaponFired
  (bind _returnvalue (CustomItemWrap|GetWrappedMeshComponents))
  (bind _returnvalue_2 (Utilities|Time|GetGameTimeinSeconds))
  (bind _pause_mulplyer (Variables|Default|GetPause%20Mulplyer))
  (bind _pause_length (Variables|Default|GetPause%20length))
  (bind _pause_length_2 (Variables|Default|GetPause%20length))
  (bind _pause_length_3 (Variables|Default|GetPause%20length))
  (bind _pause_length_4 (Variables|Default|GetPause%20length))
  (bind _pause_mulplyer_2 (Variables|Default|GetPause%20Mulplyer))
  (Utilities|FlowControl|Delay 0.01)
  (Rendering|Material|SetScalarParameterValueonMaterials _returnvalue "TopS Emissive_01-Brightness-TBP" _returnvalue_2)
  (Rendering|Material|SetScalarParameterValueonMaterials _returnvalue "TopS Emissive_02-Brightness-TBP" _returnvalue_2)
  (Utilities|FlowControl|Delay 0.05)
  (switch int (Variables|Default|GetSelection)
    (:0
      (if (not (Variables|Default|GetPause))
        (Variables|Default|SetPause true)
        (Rendering|Material|SetScalarParameterValueonMaterials (CustomItemWrap|GetWrappedMeshComponents) "EDRec_TimeFrom_BP_DecalA" (Utilities|Time|GetGameTimeinSeconds))
        (Utilities|FlowControl|Sequence
          (:then_0
            (Utilities|FlowControl|Delay _pause_length_2)
            (Variables|Default|SetPause false))
          (:then_1
            (Utilities|FlowControl|Delay (/ _pause_length_2 _pause_mulplyer_2))
            (Variables|Default|SetSelection 1)))))
    (:1
      (if (not (Variables|Default|GetPause_B))
        (Variables|Default|SetPause_B true)
        (Rendering|Material|SetScalarParameterValueonMaterials (CustomItemWrap|GetWrappedMeshComponents) "EDRec_TimeFrom_BP_DecalB" (Utilities|Time|GetGameTimeinSeconds))
        (Utilities|FlowControl|Sequence
          (:then_0
            (Utilities|FlowControl|Delay _pause_length_3)
            (Variables|Default|SetPause_B false))
          (:then_1
            (Utilities|FlowControl|Delay (/ _pause_length_3 _pause_mulplyer_2))
            (Variables|Default|SetSelection 2)))))
    (:2
      (if (not (Variables|Default|GetPause_C))
        (Variables|Default|SetPause_C true)
        (Rendering|Material|SetScalarParameterValueonMaterials (CustomItemWrap|GetWrappedMeshComponents) "EDRec_TimeFrom_BP_DecalC" (Utilities|Time|GetGameTimeinSeconds))
        (Utilities|FlowControl|Sequence
          (:then_0
            (Utilities|FlowControl|Delay _pause_length_4)
            (Variables|Default|SetPause_C false))
          (:then_1
            (Utilities|FlowControl|Delay (/ _pause_length_4 _pause_mulplyer))
            (Variables|Default|SetSelection 3)))))
    (:3
      (if (not (Variables|Default|GetPause_D))
        (Variables|Default|SetPause_D true)
        (Rendering|Material|SetScalarParameterValueonMaterials (CustomItemWrap|GetWrappedMeshComponents) "EDRec_TimeFrom_BP_DecalD" (Utilities|Time|GetGameTimeinSeconds))
        (Utilities|FlowControl|Sequence
          (:then_0
            (Utilities|FlowControl|Delay _pause_length)
            (Variables|Default|SetPause_D false))
          (:then_1
            (Utilities|FlowControl|Delay (/ _pause_length _pause_mulplyer))
            (Variables|Default|SetSelection 0)))))))

(fn ConstructionScript)
