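; Cosmetic modifier Blueprint for the FruitCake weapon/item wrap that animates time-of-day override material parameters in the front end.
; On BeginPlay it gets the wrapped mesh components, waits briefly, then plays a timeline only if it is in the front end.
; Each timeline update writes a scalar and a color track to the TOD override parameters on those components' materials.
; The weapon-fired event and construction script are empty.
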(event EventOnWeaponFired)

(event EventBeginPlay
  (bind _returnvalue (CustomItemWrap|GetWrappedMeshComponents))
  (Utilities|FlowControl|Delay 0.2)
  (if (CustomCosmeticModifier|IsinFrontEnd)
    (bind (_direction _tod_value _newtrack_1) (Timeline|Timeline PlayFromStart)
      (:Update
        (Rendering|Material|SetScalarParameterValueonMaterials _returnvalue "00-FE-TODCC-Override" _tod_value)
        (Rendering|Material|SetColorParameterValueonMaterials _returnvalue "00-FE-TODV4-Override" _newtrack_1)))))

(fn ConstructionScript)
