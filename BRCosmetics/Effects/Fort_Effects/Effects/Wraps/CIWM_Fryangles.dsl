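; Item wrap blueprint for the Fryangles weapon wrap, which reacts to weapon fire.
; On each weapon fired event, it writes the current game time into the FA_TimeFrom_BP scalar parameter on the wrapped mesh components' materials.
; The material presumably uses that timestamp to time a shot-triggered effect; the construction script is empty.
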(event EventOnWeaponFired
  (Rendering|Material|SetScalarParameterValueonMaterials (CustomItemWrap|GetWrappedMeshComponents) "FA_TimeFrom_BP" (Utilities|Time|GetGameTimeinSeconds)))

(fn ConstructionScript)
