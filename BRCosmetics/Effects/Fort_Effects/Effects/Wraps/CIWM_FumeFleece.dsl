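; FumeFleece weapon wrap Blueprint that recolors wrapped meshes to match the equipped character skin.
; Wrap target, pawn set and customization finished events each repeat a persistent-name check for five characters, defaulting to red.
; The chosen branch calls Material, setting base texture, accent, emissive color, hue shift and brightness on every dynamic material.
; Firing writes game time into emissive parameters for shader-driven animation; WrapMAt, a two-stage material switcher, is never called here.
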(event EventOnWeaponFired
  (bind _returnvalue (Utilities|Time|GetGameTimeinSeconds))
  (bind _returnvalue_2 (CustomItemWrap|GetWrappedMeshComponents))
  (Utilities|FlowControl|Sequence
    (:then_0
      ;[1 1 1] TopsSide V2
      (Rendering|Material|SetScalarParameterValueonMaterials _returnvalue_2 "TopS Emissive_02-Brightness-TBP" _returnvalue)
      (Rendering|Material|SetScalarParameterValueonMaterials _returnvalue_2 "TopS Emissive_02_Offset----TBP" _returnvalue)
      (Rendering|Material|SetScalarParameterValueonMaterials _returnvalue_2 "TopS_01_Blend_Opacity-TBP" _returnvalue)
      (Rendering|Material|SetScalarParameterValueonMaterials _returnvalue_2 "TimeFrom-BP----------------" _returnvalue))
    (:then_1
      (Rendering|Material|SetScalarParameterValueonMaterials _returnvalue_2 "TopS Emissive_01-Brightness-TBP" _returnvalue)
      (Rendering|Material|SetScalarParameterValueonMaterials _returnvalue_2 "TopS Emissive_01_Offset----TBP" _returnvalue)
      (Rendering|Material|SetScalarParameterValueonMaterials _returnvalue_2 "TopS Emissive_03-Brightness-TBP" _returnvalue)
      (Rendering|Material|SetScalarParameterValueonMaterials _returnvalue_2 "TopS Emissive_03_Offset----TBP" _returnvalue))))

(event EventOnCharacterCustomizationFinished
  (params (CharacterPartInfo :type Array%20of%20Applied%20Character%20Part%20Info%20Structures))
  (bind _returnvalue (CosmeticLoadout|GetCosmeticLoadout))
  (bind _returnvalue_2 (CosmeticLoadout|GetItemDefinitionfromLoadout _returnvalue "(TagName=\"LoadoutSlot.Character\")"))
  (bind _returnvalue_3 (Item|GetPersistentName _returnvalue_2))
  ; Blue
  (if (Utilities|String|EqualExactly(String) _returnvalue_3 "AthenaCharacter:character_fumefleececlap")
    (CallFunction|Material "/BRCosmetics/Weapons/WeaponSkins/Wraps/Textures/FumeFleece/T_Wrap_FumeFleece_d5.T_Wrap_FumeFleece_d5" "(R=0.014444,G=0.141263,B=0.428691,A=1.000000)" 0.65 0.0 0.2 0.75 "(R=0.000000,G=0.291482,B=1.000000,A=1.000000)")
    (elif (Utilities|String|EqualExactly(String) _returnvalue_3 "AthenaCharacter:character_fumefleecejig")
      (CallFunction|Material "/BRCosmetics/Weapons/WeaponSkins/Wraps/Textures/FumeFleece/T_Wrap_FumeFleece_d4.T_Wrap_FumeFleece_d4" "(R=0.445201,G=0.082283,B=0.201556,A=1.000000)" 0.85 0.0 0.2 0.75 "(R=1.000000,G=0.002698,B=0.839742,A=1.000000)")
      (elif (Utilities|String|EqualExactly(String) _returnvalue_3 "AthenaCharacter:character_fumefleeceshack")
        (CallFunction|Material "/BRCosmetics/Weapons/WeaponSkins/Wraps/Textures/FumeFleece/T_Wrap_FumeFleece_d2.T_Wrap_FumeFleece_d2" "(R=0.025187,G=0.024158,B=0.024158,A=1.000000)" 0.0 1.0 0.0 0.75)
        (elif (Utilities|String|EqualExactly(String) _returnvalue_3 "AthenaCharacter:character_fumefleecewag")
          (CallFunction|Material "/BRCosmetics/Weapons/WeaponSkins/Wraps/Textures/FumeFleece/T_Wrap_FumeFleece_d3.T_Wrap_FumeFleece_d3" "(R=0.745404,G=0.401978,B=0.012286,A=1.000000)" 0.1 0.0 0.2 0.75 "(R=1.000000,G=0.914078,B=0.000000,A=1.000000)")
          (elif (Utilities|String|EqualExactly(String) _returnvalue_3 "AthenaCharacter:character_satirecane_ode")
            (CallFunction|Material "/BRCosmetics/Weapons/WeaponSkins/Wraps/Textures/FumeFleece/T_Wrap_FumeFleece_d6.T_Wrap_FumeFleece_d6" "(R=0.043735,G=0.177888,B=0.046665,A=1.000000)" 0.3 0.0 0.2 0.75 "(R=0.016262,G=1.000000,B=0.000000,A=1.000000)")
            (else
              (CallFunction|Material "/BRCosmetics/Weapons/WeaponSkins/Wraps/Textures/FumeFleece/T_Wrap_FumeFleece_d1.T_Wrap_FumeFleece_d1" "(R=0.623961,G=0.023153,B=0.020289,A=1.000000)" 0.0 0.0 0.2 0.75 "(R=1.000000,G=0.000000,B=0.000000,A=1.000000)"))))))))

(event EventOnSetWrapTarget
  (bind _returnvalue (CosmeticLoadout|GetCosmeticLoadout))
  (bind _returnvalue_2 (CosmeticLoadout|GetItemDefinitionfromLoadout _returnvalue "(TagName=\"LoadoutSlot.Character\")"))
  (bind _returnvalue_3 (Item|GetPersistentName _returnvalue_2))
  ; Blue
  (if (Utilities|String|EqualExactly(String) _returnvalue_3 "AthenaCharacter:character_fumefleececlap")
    (CallFunction|Material "/BRCosmetics/Weapons/WeaponSkins/Wraps/Textures/FumeFleece/T_Wrap_FumeFleece_d5.T_Wrap_FumeFleece_d5" "(R=0.014444,G=0.141263,B=0.428691,A=1.000000)" 0.65 0.0 0.2 0.75 "(R=0.000000,G=0.291482,B=1.000000,A=1.000000)")
    (elif (Utilities|String|EqualExactly(String) _returnvalue_3 "AthenaCharacter:character_fumefleecejig")
      (CallFunction|Material "/BRCosmetics/Weapons/WeaponSkins/Wraps/Textures/FumeFleece/T_Wrap_FumeFleece_d4.T_Wrap_FumeFleece_d4" "(R=0.445201,G=0.082283,B=0.201556,A=1.000000)" 0.85 0.0 0.2 0.75 "(R=1.000000,G=0.002698,B=0.839742,A=1.000000)")
      (elif (Utilities|String|EqualExactly(String) _returnvalue_3 "AthenaCharacter:character_fumefleeceshack")
        (CallFunction|Material "/BRCosmetics/Weapons/WeaponSkins/Wraps/Textures/FumeFleece/T_Wrap_FumeFleece_d2.T_Wrap_FumeFleece_d2" "(R=0.025187,G=0.024158,B=0.024158,A=1.000000)" 0.0 1.0 0.0 0.75)
        (elif (Utilities|String|EqualExactly(String) _returnvalue_3 "AthenaCharacter:character_fumefleecewag")
          (CallFunction|Material "/BRCosmetics/Weapons/WeaponSkins/Wraps/Textures/FumeFleece/T_Wrap_FumeFleece_d3.T_Wrap_FumeFleece_d3" "(R=0.745404,G=0.401978,B=0.012286,A=1.000000)" 0.1 0.0 0.2 0.75 "(R=1.000000,G=0.914078,B=0.000000,A=1.000000)")
          (elif (Utilities|String|EqualExactly(String) _returnvalue_3 "AthenaCharacter:character_satirecane_ode")
            (CallFunction|Material "/BRCosmetics/Weapons/WeaponSkins/Wraps/Textures/FumeFleece/T_Wrap_FumeFleece_d6.T_Wrap_FumeFleece_d6" "(R=0.043735,G=0.177888,B=0.046665,A=1.000000)" 0.3 0.0 0.2 0.75 "(R=0.016262,G=1.000000,B=0.000000,A=1.000000)")
            (else
              (CallFunction|Material "/BRCosmetics/Weapons/WeaponSkins/Wraps/Textures/FumeFleece/T_Wrap_FumeFleece_d1.T_Wrap_FumeFleece_d1" "(R=0.623961,G=0.023153,B=0.020289,A=1.000000)" 0.0 0.0 0.2 0.75 "(R=1.000000,G=0.000000,B=0.000000,A=1.000000)"))))))))

(event EventOnPawnSet
  (bind _returnvalue (CosmeticLoadout|GetCosmeticLoadout))
  (bind _returnvalue_2 (CosmeticLoadout|GetItemDefinitionfromLoadout _returnvalue "(TagName=\"LoadoutSlot.Character\")"))
  (bind _returnvalue_3 (Item|GetPersistentName _returnvalue_2))
  ; Blue
  (if (Utilities|String|EqualExactly(String) _returnvalue_3 "AthenaCharacter:character_fumefleececlap")
    (CallFunction|Material "/BRCosmetics/Weapons/WeaponSkins/Wraps/Textures/FumeFleece/T_Wrap_FumeFleece_d5.T_Wrap_FumeFleece_d5" "(R=0.014444,G=0.141263,B=0.428691,A=1.000000)" 0.65 0.0 0.2 0.75 "(R=0.000000,G=0.291482,B=1.000000,A=1.000000)")
    (elif (Utilities|String|EqualExactly(String) _returnvalue_3 "AthenaCharacter:character_fumefleecejig")
      (CallFunction|Material "/BRCosmetics/Weapons/WeaponSkins/Wraps/Textures/FumeFleece/T_Wrap_FumeFleece_d4.T_Wrap_FumeFleece_d4" "(R=0.445201,G=0.082283,B=0.201556,A=1.000000)" 0.85 0.0 0.2 0.75 "(R=1.000000,G=0.002698,B=0.839742,A=1.000000)")
      (elif (Utilities|String|EqualExactly(String) _returnvalue_3 "AthenaCharacter:character_fumefleeceshack")
        (CallFunction|Material "/BRCosmetics/Weapons/WeaponSkins/Wraps/Textures/FumeFleece/T_Wrap_FumeFleece_d2.T_Wrap_FumeFleece_d2" "(R=0.025187,G=0.024158,B=0.024158,A=1.000000)" 0.0 1.0 0.0 0.75)
        (elif (Utilities|String|EqualExactly(String) _returnvalue_3 "AthenaCharacter:character_fumefleecewag")
          (CallFunction|Material "/BRCosmetics/Weapons/WeaponSkins/Wraps/Textures/FumeFleece/T_Wrap_FumeFleece_d3.T_Wrap_FumeFleece_d3" "(R=0.745404,G=0.401978,B=0.012286,A=1.000000)" 0.1 0.0 0.2 0.75 "(R=1.000000,G=0.914078,B=0.000000,A=1.000000)")
          (elif (Utilities|String|EqualExactly(String) _returnvalue_3 "AthenaCharacter:character_satirecane_ode")
            (CallFunction|Material "/BRCosmetics/Weapons/WeaponSkins/Wraps/Textures/FumeFleece/T_Wrap_FumeFleece_d6.T_Wrap_FumeFleece_d6" "(R=0.043735,G=0.177888,B=0.046665,A=1.000000)" 0.3 0.0 0.2 0.75 "(R=0.016262,G=1.000000,B=0.000000,A=1.000000)")
            (else
              (CallFunction|Material "/BRCosmetics/Weapons/WeaponSkins/Wraps/Textures/FumeFleece/T_Wrap_FumeFleece_d1.T_Wrap_FumeFleece_d1" "(R=0.623961,G=0.023153,B=0.020289,A=1.000000)" 0.0 0.0 0.2 0.75 "(R=1.000000,G=0.000000,B=0.000000,A=1.000000)"))))))))

(fn Material
  (params (Base_Color :type Texture%20Object%20Reference) (Accent%20BC :type Linear%20Color%20Structure) (Hue%20Shift :type Float%20%28single-precision%29) (Saturation :type Float%20%28single-precision%29) (Brightness :type Float%20%28single-precision%29) (Darken :type Float%20%28single-precision%29) (Value :type Linear%20Color%20Structure))
  (for _array_element (CustomItemWrap|GetWrappedMeshComponents)
    (for _array_element_2 (Rendering|Material|GetMaterials _array_element)
      (bind _asmaterial_instance_dynamic (Utilities|Casting|CastToMaterialInstanceDynamic _array_element_2))
      (Rendering|Material|SetTextureParameterValue _asmaterial_instance_dynamic "TopS_01__Tex_BaseColor" Base_Color)
      (Rendering|Material|SetVectorParameterValue _asmaterial_instance_dynamic "AccentV2_BaseColor_" Accent%20BC :!class /Script/Engine.MaterialInstanceDynamic)
      (Rendering|Material|SetVectorParameterValue _asmaterial_instance_dynamic "TopS Emissive_01_Color-" Value :!class /Script/Engine.MaterialInstanceDynamic)
      (Rendering|Material|SetScalarParameterValue _asmaterial_instance_dynamic "HueShift_EmissiveColor" Hue%20Shift :!class /Script/Engine.MaterialInstanceDynamic)
      (Rendering|Material|SetScalarParameterValue _asmaterial_instance_dynamic "HueShift_EmissiveSaturation" Saturation :!class /Script/Engine.MaterialInstanceDynamic)
      (Rendering|Material|SetScalarParameterValue _asmaterial_instance_dynamic "TopS Emissive_01_Brightness---Start" Brightness :!class /Script/Engine.MaterialInstanceDynamic)
      (Rendering|Material|SetScalarParameterValue _asmaterial_instance_dynamic "TopS Emissive_01_Brightness_Darken_Base Start" Darken :!class /Script/Engine.MaterialInstanceDynamic))))

(fn ConstructionScript)

(fn WrapMAt
  (params (Stages :type Integer) (A_Set_BaseColor :type Texture%20Object%20Reference) (B_Set_BaseColor :type Texture%20Object%20Reference) (A_Set_Mask :type Texture%20Object%20Reference) (B_Set_Mask :type Texture%20Object%20Reference) (A_Set_Norm :type Texture%20Object%20Reference) (B_Set_Norm :type Texture%20Object%20Reference) (A_Acent_Color :type Linear%20Color%20Structure) (A_Acent_Spec :type Float%20%28double-precision%29) (A_Acent_Roughness :type Float%20%28double-precision%29) (A_Acent_Metal :type Float%20%28double-precision%29) (B_Acent_Color :type Linear%20Color%20Structure) (B_Acent_Spec :type Float%20%28double-precision%29) (B_Acent_Roughness :type Float%20%28double-precision%29) (B_Acent_Metal :type Float%20%28double-precision%29) (A_Lut_MSk_Acent :type Float%20%28single-precision%29) (B_Lut_MSk_Acent :type Float%20%28single-precision%29))
  (for _array_element (CustomItemWrap|GetWrappedMeshComponents)
    (for _array_element_2 (Rendering|Material|GetMaterials _array_element)
      (bind _asmaterial_instance_dynamic (Utilities|Casting|CastToMaterialInstanceDynamic _array_element_2))
      (Rendering|Material|SetTextureParameterValue _asmaterial_instance_dynamic "TopS_01__Tex_BaseColor" (Utilities|Array|Get(acopy) (Utilities|Array|MakeArray A_Set_BaseColor B_Set_BaseColor) Stages))
      (Rendering|Material|SetTextureParameterValue _asmaterial_instance_dynamic "TopS_01__Tex_Mask" (Utilities|Array|Get(acopy) (Utilities|Array|MakeArray A_Set_Mask B_Set_Mask) Stages))
      (Rendering|Material|SetTextureParameterValue _asmaterial_instance_dynamic "TopS_01__Tex_Normal" (Utilities|Array|Get(acopy) (Utilities|Array|MakeArray A_Set_Norm B_Set_Norm) Stages))
      (Rendering|Material|SetVectorParameterValue _asmaterial_instance_dynamic "Simple_Color" (Utilities|Array|Get(acopy) (Utilities|Array|MakeArray A_Acent_Color B_Acent_Color) Stages) :!class /Script/Engine.MaterialInstanceDynamic)
      (Rendering|Material|SetScalarParameterValue _asmaterial_instance_dynamic "Simple_Specular" (Utilities|Array|Get(acopy) (Utilities|Array|MakeArray A_Acent_Spec B_Acent_Spec) Stages) :!class /Script/Engine.MaterialInstanceDynamic)
      (Rendering|Material|SetScalarParameterValue _asmaterial_instance_dynamic "Simple_Roughness" (Utilities|Array|Get(acopy) (Utilities|Array|MakeArray A_Acent_Roughness B_Acent_Roughness) Stages) :!class /Script/Engine.MaterialInstanceDynamic)
      (Rendering|Material|SetScalarParameterValue _asmaterial_instance_dynamic "Simple_Metalness" (Utilities|Array|Get(acopy) (Utilities|Array|MakeArray A_Acent_Metal B_Acent_Metal) Stages) :!class /Script/Engine.MaterialInstanceDynamic)
      (Rendering|Material|SetScalarParameterValue _asmaterial_instance_dynamic "Metal_LUT_Msk_Accent   1=On  0=Off" (Utilities|Array|Get(acopy) (Utilities|Array|MakeArray A_Lut_MSk_Acent B_Lut_MSk_Acent) Stages) :!class /Script/Engine.MaterialInstanceDynamic))))
