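; Reactive item wrap cosmetic (Funky Feather) that animates emissive color, brightness and offsets on wrapped mesh materials.
; In game, weapon fire tints it blue and a DamageReceived stat watch tints it red through the AnimColor timeline, reversing after a delay.
; A looping timer drives eased brightness and offset fades from elapsed time.
; In the front end, LockerCycle steps through base, damage-dealt and damage-received previews, pausing while firing.
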(event EventBeginPlay
  (bind _targetcolor (Variables|Default|GetTargetColor))
  (bind _basecolor (Variables|Default|GetBaseColor))
  (if (CustomCosmeticModifier|IsinFrontEnd)
    (Utilities|Time|SetTimerbyFunctionName 0 "LockerCycle" 2.5 true)
    (else
      (bind _statname (Cosmetics|WatchCosmeticStat (CustomCharacter|GetAssociatedPlayerPawn) "(TagName=\"StatManager.DamageReceived\")" 0 false)
        (:OnStatChanged
          (Utilities|FlowControl|Sequence
            (:then_0
              (if (== _statname "DamageReceived")
                (Variables|Default|SetBeingDamaged true)
                (CallFunction|SineAnimation)
                (Utilities|FlowControl|RetriggerableDelay 2.0)
                (CallFunction|SineAnimOFF)))
            (:then_1
              (if (== _statname "DamageReceived")
                (Utilities|FlowControl|Sequence
                  (:then_0
                    (Variables|Default|SetBeingDamaged true)
                    (Variables|Default|SetTargetColor "(R=5.000000,G=0.000000,B=0.000000,A=50.000000)")
                    (Components|Timeline|SetPlayRate (Variables|CIWM_FunkyFeather_Cap|GetAnimColor) 1.0)
                    (bind (_direction _alpha) (Timeline|AnimColor PlayFromStart)
                      (:Update
                        (for _array_element (CustomItemWrap|GetWrappedMeshComponents)
                          (Rendering|Material|SetVectorParameterValueonMaterials _array_element "TopS Emissive_01_Color-" (Math|Conversions|ToVector(LinearColor) (Math|Color|Lerp(LinearColor) _basecolor _targetcolor _alpha)))
                          (Rendering|Material|SetVectorParameterValueonMaterials _array_element "EDRec_Emissive_B_Color-" (Math|Conversions|ToVector(LinearColor) (Math|Color|Lerp(LinearColor) _basecolor _targetcolor _alpha)))))))
                  (:then_1
                    (Utilities|FlowControl|RetriggerableDelay 2.0)
                    (Variables|Default|SetBeingDamaged false)
                    (Timeline|AnimColor Reverse)))))))))))

(event EventOnPawnSet
  (bind _targetcolor (Variables|Default|GetTargetColor))
  (bind _basecolor (Variables|Default|GetBaseColor))
  (bind _statname (Cosmetics|WatchCosmeticStat (CustomCharacter|GetAssociatedPlayerPawn) "(TagName=\"StatManager.DamageReceived\")" 0 false)
    (:OnStatChanged
      (Utilities|FlowControl|Sequence
        (:then_0
          (if (== _statname "DamageReceived")
            (Variables|Default|SetBeingDamaged true)
            (CallFunction|SineAnimation)
            (Utilities|FlowControl|RetriggerableDelay 2.0)
            (CallFunction|SineAnimOFF)))
        (:then_1
          (if (== _statname "DamageReceived")
            (Utilities|FlowControl|Sequence
              (:then_0
                (Variables|Default|SetBeingDamaged true)
                (Variables|Default|SetTargetColor "(R=5.000000,G=0.000000,B=0.000000,A=50.000000)")
                (Components|Timeline|SetPlayRate (Variables|CIWM_FunkyFeather_Cap|GetAnimColor) 1.0)
                (Timeline|AnimColor PlayFromStart))
              (:then_1
                (Utilities|FlowControl|RetriggerableDelay 2.0)
                (Variables|Default|SetBeingDamaged false)
                (Timeline|AnimColor Reverse)))))))))

(event EventOnWeaponFired
  (bind _returnvalue (Utilities|Time|GetGameTimeinSeconds))
  (bind _returnvalue_2 (CustomItemWrap|GetWrappedMeshComponents))
  (bind _basecolor_locker (Variables|Default|GetBaseColor%20Locker))
  (bind _dmgrecivedcolor_locker (Variables|Default|GetDMGrecivedColor%20Locker))
  (bind _dmgdealtcolor_locker (Variables|Default|GetDMGdealtColor%20Locker))
  (bind _returnvalue_3 (Utilities|Select :Index (Variables|Default|GetLockerCounter) :Option%200 _basecolor_locker :Option%201 _dmgdealtcolor_locker :Option%202 _dmgrecivedcolor_locker :Option%203 _basecolor_locker))
  (bind _returnvalue_4 (Math|Conversions|ToVector(LinearColor) _returnvalue_3))
  (bind _returnvalue_5 (CustomCosmeticModifier|IsinFrontEnd))
  (bind _returnvalue_6 (CustomItemWrap|GetWrappedMeshComponents))
  (bind _returnvalue_7 (Utilities|Time|GetGameTimeinSeconds))
  (bind _returnvalue_8 (CustomItemWrap|GetWrappedMeshComponents))
  (bind _targetcolor (Variables|Default|GetTargetColor))
  (bind _basecolor (Variables|Default|GetBaseColor))
  (bind _lockercounter (Variables|Default|GetLockerCounter))
  (if (CustomCosmeticModifier|IsinFrontEnd)
    (Rendering|Material|SetScalarParameterValueonMaterials _returnvalue_2 "EDRec_Emissive_A_Brightness_Time_BP" _returnvalue)
    (Rendering|Material|SetScalarParameterValueonMaterials _returnvalue_2 "EDRec_Emissive_B_Brightness_Time_BP" _returnvalue)
    (Rendering|Material|SetScalarParameterValueonMaterials _returnvalue_2 "EDRec_Emissive_A_Brightness_Time__End" _returnvalue)
    (Rendering|Material|SetScalarParameterValueonMaterials _returnvalue_2 "EDRec_Emissive_B_Brightness_Time__End" _returnvalue)
    (if (not (Variables|Default|GetKilled))
      (Rendering|Material|SetScalarParameterValueonMaterials _returnvalue_2 "EDRec_Emissive_A_Brightness_Time__End" 0.0)
      (Rendering|Material|SetScalarParameterValueonMaterials _returnvalue_2 "EDRec_Emissive_B_Brightness_Time__End" 0.0))
    (else
      (Utilities|FlowControl|Sequence
        (:then_0
          (if (not (Variables|Default|GetBeingDamaged))
            (Utilities|FlowControl|Sequence
              (:then_0
                (Rendering|Material|SetScalarParameterValueonMaterials _returnvalue_8 "TopS Emissive_01_Offset----TBP" _returnvalue_7)
                (Rendering|Material|SetScalarParameterValueonMaterials _returnvalue_8 "EDRec_Emissive_A_Brightness_Time_BP" _returnvalue_7)
                (Rendering|Material|SetScalarParameterValueonMaterials _returnvalue_8 "TopS Emissive_01_Offset---End" _returnvalue_7)
                (Rendering|Material|SetScalarParameterValueonMaterials _returnvalue_8 "EDRec_Emissive_A_Brightness_Time__End" _returnvalue_7)
                (if (not (Variables|Default|GetKilled))
                  (Rendering|Material|SetScalarParameterValueonMaterials _returnvalue_8 "TopS Emissive_01_Offset---End" -0.5)
                  (Rendering|Material|SetScalarParameterValueonMaterials _returnvalue_8 "EDRec_Emissive_A_Brightness_Time__End" 0.0)))
              (:then_1
                (if (not (Variables|Default|GetFullAutoWeapons))
                  (Variables|Default|SetFullAutoWeapons true)
                  (CallFunction|SineAnimation))
                (Utilities|FlowControl|RetriggerableDelay 2.0)
                (CallFunction|SineAnimOFF)
                (Variables|Default|SetFullAutoWeapons false))
              (:then_2
                (Utilities|FlowControl|RetriggerableDelay 2.0)
                (Timeline|AnimColor Reverse))
              (:then_3
                (Variables|Default|SetTargetColor "(R=0.260420,G=3.709189,B=10.000000,A=1.000000)")
                (Components|Timeline|SetPlayRate (Variables|CIWM_FunkyFeather_Cap|GetAnimColor) 1.0)
                (Timeline|AnimColor Play)))))
        (:then_1
          (if _returnvalue_5
            (Variables|Default|SetIsFiring true)
            (Utilities|FlowControl|RetriggerableDelay 2.35)
            (Variables|Default|SetIsFiring false)
            (Variables|Default|SetLockerCounter 0)))
        (:then_2
          (if _returnvalue_5
            (Variables|Default|SetIsFiring true)
            ; FE Preview
            (bind (_direction _newtrack_0) (Timeline|Timeline Stop)
              (:Update
                (Rendering|Material|SetScalarParameterValueonMaterials _returnvalue_6 "TopS Emissive_01_Brightness---End" (Utilities|Select :Index _lockercounter :Option%200 0.35 :Option%201 (Math|Float|Lerp 0.35 2.5 _newtrack_0) :Option%202 (Math|Float|Lerp 2.5 0.35 _newtrack_0) :Option%203 0.35))
                (Rendering|Material|SetScalarParameterValueonMaterials _returnvalue_6 "EDRec_Emissive_B_Brightness_Time__End" (Utilities|Select :Index _lockercounter :Option%200 0.0 :Option%201 (Math|Float|Lerp 0.0 1.0 _newtrack_0) :Option%202 1.0 :Option%203 (Math|Float|Lerp 1.0 0.0 _newtrack_0)))
                (Rendering|Material|SetVectorParameterValueonMaterials _returnvalue_6 "TopS Emissive_01_Color-" _returnvalue_4)
                (Rendering|Material|SetVectorParameterValueonMaterials _returnvalue_6 "EDRec_Emissive_B_Color-" _returnvalue_4)
                ;; DMG dealt
                (Rendering|Material|SetVectorParameterValueonMaterials _returnvalue_6 "TopS_Em03_Offset_Weapon" (Math|Vector|MakeVector 0.0 (Utilities|Select :Index _lockercounter :Option%200 0.04 :Option%201 (Math|Float|Lerp 0.04 0.02 _newtrack_0) :Option%202 (Math|Float|Lerp 0.02 0.04 _newtrack_0) :Option%203 0.04)))
                (Rendering|Material|SetVectorParameterValueonMaterials _returnvalue_6 "TopS_Em02_Offset_Weapon" (Math|Vector|MakeVector 0.0 (Utilities|Select :Index _lockercounter :Option%200 0.04 :Option%201 (Math|Float|Lerp 0.04 0.06 _newtrack_0) :Option%202 (Math|Float|Lerp 0.06 0.02 _newtrack_0) :Option%203 (Math|Float|Lerp 0.02 0.04 _newtrack_0))))
                (Rendering|Material|SetScalarParameterValueonMaterials _returnvalue_6 "TopS Emissive_01_Offset---End" (Utilities|Select :Index _lockercounter :Option%200 -0.5 :Option%201 (Math|Float|Lerp -0.5 0.0 _newtrack_0) :Option%202 (Math|Float|Lerp 0.0 -0.5 _newtrack_0) :Option%203 -0.5))))))))))

(event Custom|LockerCycle
  (bind _basecolor_locker (Variables|Default|GetBaseColor%20Locker))
  (bind _dmgrecivedcolor_locker (Variables|Default|GetDMGrecivedColor%20Locker))
  (bind _dmgdealtcolor_locker (Variables|Default|GetDMGdealtColor%20Locker))
  (bind _returnvalue (Utilities|Select :Index (Variables|Default|GetLockerCounter) :Option%200 _basecolor_locker :Option%201 _dmgdealtcolor_locker :Option%202 _dmgrecivedcolor_locker :Option%203 _basecolor_locker))
  (bind _returnvalue_2 (Math|Conversions|ToVector(LinearColor) _returnvalue))
  (bind _returnvalue_3 (CustomItemWrap|GetWrappedMeshComponents))
  (bind _lockercounter (Variables|Default|GetLockerCounter))
  (if (CustomCosmeticModifier|IsinFrontEnd)
    (if (Variables|Default|GetIsFiring)
      _
      (elif (>= (Variables|Default|GetLockerCounter) 3)
        (Variables|Default|SetLockerCounter 0)
        ; FE Preview
        (Timeline|Timeline PlayFromStart)
        (else
          (Math|Integer|IncrementInt (Variables|Default|GetLockerCounter)
            (:%20%20
              (Timeline|Timeline PlayFromStart))))))))

(event Custom|SineAnimation
  (Variables|Default|SetReverse false)
  (Variables|Default|SetStartTime (Utilities|Time|GetGameTimeinSeconds))
  (bind _returnvalue (Utilities|Time|SetTimerbyEvent (delegate Custom|SineAnimUpdate) 0.02 true))
  (Variables|Default|SetTimerHandle _returnvalue))

(event Custom|SineAnimOFF
  (Variables|Default|SetReverse true)
  (Variables|Default|SetStartTime (Utilities|Time|GetGameTimeinSeconds))
  (bind _returnvalue (Utilities|Time|SetTimerbyEvent (delegate Custom|SineAnimUpdate) 0.02 true))
  (Variables|Default|SetTimerHandle _returnvalue))

(event Custom|SineAnimUpdate
  (bind _returnvalue (Utilities|Time|GetGameTimeinSeconds))
  (bind _returnvalue_2 (CustomItemWrap|GetWrappedMeshComponents))
  (bind _starttime (Variables|Default|GetStartTime))
  (bind _returnvalue_3 (- _returnvalue _starttime))
  (if (>= _returnvalue_3 5.0)
    (Utilities|Time|ClearandInvalidateTimerbyHandle (Variables|Default|GetTimerHandle))
    (if (Variables|Default|GetReverse)
      ; Start
      (Rendering|Material|SetScalarParameterValueonMaterials _returnvalue_2 "TopS Emissive_01_Brightness---End" 0.35)
      (Rendering|Material|SetScalarParameterValueonMaterials _returnvalue_2 "EDRec_Emissive_B_Brightness_Time__End" 0.0)
      (if (Variables|Default|GetBeingDamaged)
        (Rendering|Material|SetVectorParameterValueonMaterials _returnvalue_2 "TopS_Em02_Offset_Weapon" (0.0 0.04 0.0))
        (else
          (Rendering|Material|SetVectorParameterValueonMaterials _returnvalue_2 "TopS_Em03_Offset_Weapon" (0.0 0.04 0.0))
          (Rendering|Material|SetVectorParameterValueonMaterials _returnvalue_2 "TopS_Em02_Offset_Weapon" (0.0 0.04 0.0))))
      (else
        ; Arrive
        (Rendering|Material|SetScalarParameterValueonMaterials _returnvalue_2 "TopS Emissive_01_Brightness---End" 2.5)
        (Rendering|Material|SetScalarParameterValueonMaterials _returnvalue_2 "EDRec_Emissive_B_Brightness_Time__End" 1.0)
        (if (Variables|Default|GetBeingDamaged)
          (Rendering|Material|SetVectorParameterValueonMaterials _returnvalue_2 "TopS_Em02_Offset_Weapon" (0.0 0.02 0.0))
          (else
            (Rendering|Material|SetVectorParameterValueonMaterials _returnvalue_2 "TopS_Em03_Offset_Weapon" (0.0 0.02 0.0))
            (Rendering|Material|SetVectorParameterValueonMaterials _returnvalue_2 "TopS_Em02_Offset_Weapon" (0.0 0.06 0.0))))))
    (elif (Variables|Default|GetReverse)
      ; FadeOut
      (Rendering|Material|SetScalarParameterValueonMaterials _returnvalue_2 "TopS Emissive_01_Brightness---End" (Math|Float|MapRangeClamped _returnvalue_3 0.0 0.25 2.5 0.35))
      (Rendering|Material|SetScalarParameterValueonMaterials _returnvalue_2 "EDRec_Emissive_B_Brightness_Time__End" (Math|Float|MapRangeClamped _returnvalue_3 0.0 0.25 1.0))
      (if (Variables|Default|GetBeingDamaged)
        (Rendering|Material|SetVectorParameterValueonMaterials _returnvalue_2 "TopS_Em02_Offset_Weapon" (Math|Vector|MakeVector 0.0 (Math|Float|MapRangeClamped _returnvalue_3 0.0 0.25 0.02 0.04)))
        (else
          (Rendering|Material|SetVectorParameterValueonMaterials _returnvalue_2 "TopS_Em03_Offset_Weapon" (Math|Vector|MakeVector 0.0 (Math|Float|MapRangeClamped _returnvalue_3 0.0 0.25 0.02 0.04)))
          (Rendering|Material|SetVectorParameterValueonMaterials _returnvalue_2 "TopS_Em02_Offset_Weapon" (Math|Vector|MakeVector 0.0 (Math|Float|MapRangeClamped _returnvalue_3 0.0 0.25 0.06 0.04)))))
      (else
        ; FadeIn
        (Rendering|Material|SetScalarParameterValueonMaterials _returnvalue_2 "TopS Emissive_01_Brightness---End" (Math|Float|MapRangeClamped _returnvalue_3 0.0 0.25 0.35 2.5))
        (Rendering|Material|SetScalarParameterValueonMaterials _returnvalue_2 "EDRec_Emissive_B_Brightness_Time__End" (Math|Float|MapRangeClamped _returnvalue_3 0.0 0.25 0.0 1.0))
        (if (Variables|Default|GetBeingDamaged)
          (Rendering|Material|SetVectorParameterValueonMaterials _returnvalue_2 "TopS_Em02_Offset_Weapon" (Math|Vector|MakeVector 0.0 (Math|Float|MapRangeClamped _returnvalue_3 0.0 0.25 0.04 0.02)))
          (else
            (Rendering|Material|SetVectorParameterValueonMaterials _returnvalue_2 "TopS_Em03_Offset_Weapon" (Math|Vector|MakeVector 0.0 (Math|Float|MapRangeClamped _returnvalue_3 0.0 0.25 0.04 0.02)))
            (Rendering|Material|SetVectorParameterValueonMaterials _returnvalue_2 "TopS_Em02_Offset_Weapon" (Math|Vector|MakeVector 0.0 (Math|Float|MapRangeClamped _returnvalue_3 0.0 0.25 0.04 0.06)))))))))

(fn ConstructionScript)
