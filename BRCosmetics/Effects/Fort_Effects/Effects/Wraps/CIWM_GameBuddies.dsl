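; Custom item wrap Blueprint for the GameBuddies wrap that makes its materials react each time the weapon fires.
; On each shot, after a tiny delay, it sets start, end and game-time scalar parameters on the wrapped mesh components' materials.
; An Active flag alternates two stored random values, so each reaction starts at the previous end value and targets a fresh random one.
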(event EventOnWeaponFired
  (bind _returnvalue (CustomItemWrap|GetWrappedMeshComponents))
  (Utilities|FlowControl|Delay 0.01)
  (if (Variables|Default|GetActive)
    (bind _output_get (Variables|Default|SetRand%20B (Math|Random|RandomFloat)))
    (Rendering|Material|SetScalarParameterValueonMaterials _returnvalue "GameBud_React___End" _output_get)
    (Rendering|Material|SetScalarParameterValueonMaterials _returnvalue "GameBud_React___Start" (Variables|Default|GetRand%20A))
    (Rendering|Material|SetScalarParameterValueonMaterials _returnvalue "GameBud_React_Time_BP" (Utilities|Time|GetGameTimeinSeconds))
    (Variables|Default|SetActive false)
    (else
      (bind _output_get_2 (Variables|Default|SetRand%20A (Math|Random|RandomFloat)))
      (Rendering|Material|SetScalarParameterValueonMaterials _returnvalue "GameBud_React___End" _output_get_2)
      (Rendering|Material|SetScalarParameterValueonMaterials _returnvalue "GameBud_React___Start" (Variables|Default|GetRand%20B))
      (Rendering|Material|SetScalarParameterValueonMaterials _returnvalue "GameBud_React_Time_BP" (Utilities|Time|GetGameTimeinSeconds))
      (Variables|Default|SetActive true))))

(fn ConstructionScript)
