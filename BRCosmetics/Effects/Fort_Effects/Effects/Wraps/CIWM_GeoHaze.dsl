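; GeoHaze weapon wrap modifier that makes emissive materials react to shots and player kills.
; It writes current game time into scalar parameters on the wrapped meshes, hitting two emissive layers per shot.
; An AthenaKills stat watcher also stamps the second layer and its darken base, then toggles a one-second Pause flag that never gates those writes.
; In the front end, a shot counter periodically fires that kill response.
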(event EventOnWeaponFired
  (bind _returnvalue (CustomItemWrap|GetWrappedMeshComponents))
  (bind _returnvalue_2 (Utilities|Time|GetGameTimeinSeconds))
  (bind _returnvalue_3 (CustomItemWrap|GetWrappedMeshComponents))
  (bind _returnvalue_4 (Utilities|Time|GetGameTimeinSeconds))
  (if (CustomCosmeticModifier|IsinFrontEnd)
    (Rendering|Material|SetScalarParameterValueonMaterials _returnvalue "TopS Emissive_01-Brightness-TBP" _returnvalue_4)
    (Rendering|Material|SetScalarParameterValueonMaterials _returnvalue "TopS Emissive_03-Brightness-TBP" _returnvalue_4)
    (Utilities|FlowControl|Delay 0.001)
    (bind _output_get (Variables|Default|SetShotsFired (+ (Variables|Default|GetShotsFired) 1)))
    (if (>= _output_get 5)
      (if (== (Variables|Default|GetShotsFired) 7)
        (Variables|Default|SetShotsFired 0)
        (Utilities|FlowControl|Sequence
          (:then_1
            (Rendering|Material|SetScalarParameterValueonMaterials _returnvalue_3 "TopS Emissive_01-Brightness-TBP" _returnvalue_2)
            (Rendering|Material|SetScalarParameterValueonMaterials _returnvalue_3 "TopS Emissive_03-Brightness-TBP" _returnvalue_2)
            (Rendering|Material|SetScalarParameterValueonMaterials _returnvalue_3 "TopS Emissive_02-Brightness-TBP" _returnvalue_2)
            (Rendering|Material|SetScalarParameterValueonMaterials _returnvalue_3 "TopS Emissive_02_Brightness_Darken_Base" _returnvalue_2)
            (if (not (Variables|Default|GetPause))
              (Variables|Default|SetPause true)
              (Utilities|FlowControl|Delay 1.0)
              (Variables|Default|SetPause false))))))
    (else
      (Rendering|Material|SetScalarParameterValueonMaterials _returnvalue_3 "TopS Emissive_01-Brightness-TBP" _returnvalue_2)
      (Rendering|Material|SetScalarParameterValueonMaterials _returnvalue_3 "TopS Emissive_03-Brightness-TBP" _returnvalue_2))))

(event EventOnPawnSet
  (bind _returnvalue (Utilities|Time|GetGameTimeinSeconds))
  (bind _returnvalue_2 (CustomItemWrap|GetWrappedMeshComponents))
  (Cosmetics|WatchCosmeticStat (CustomCharacter|GetAssociatedPlayerPawn) "(TagName=\"StatManager.AthenaKills\")" 0 false
    (:OnStatChanged
      (Utilities|FlowControl|Sequence
        (:then_1
          (Rendering|Material|SetScalarParameterValueonMaterials _returnvalue_2 "TopS Emissive_01-Brightness-TBP" _returnvalue)
          (Rendering|Material|SetScalarParameterValueonMaterials _returnvalue_2 "TopS Emissive_03-Brightness-TBP" _returnvalue)
          (Rendering|Material|SetScalarParameterValueonMaterials _returnvalue_2 "TopS Emissive_02-Brightness-TBP" _returnvalue)
          (Rendering|Material|SetScalarParameterValueonMaterials _returnvalue_2 "TopS Emissive_02_Brightness_Darken_Base" _returnvalue)
          (if (not (Variables|Default|GetPause))
            (Variables|Default|SetPause true)
            (Utilities|FlowControl|Delay 1.0)
            (Variables|Default|SetPause false)))))))

(fn ConstructionScript)
