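; Custom item wrap modifier that drives a numeric counter material parameter on wrapped meshes from the player's height in the world.
; In game, begin play, pawn set, wrap target and customization events start a looping timer that converts body mesh Z to rounded, clamped meters and writes it to EDRec_Counter__Number.
; In the front end, a timeline instead lerps the counter between random values, re-randomized by its event tracks.
; The timer handler also prints each value as debug text.
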(event EventBeginPlay
  (bind _returnvalue (Math|Random|RandomIntegerinRange 0 139))
  (if (CustomCosmeticModifier|IsinFrontEnd)
    (Variables|Default|SetA (Math|Conversions|ToFloat(Integer) (Math|Random|RandomIntegerinRange 50 100)))
    (Variables|Default|SetB (Math|Conversions|ToFloat(Integer) (Math|Random|RandomIntegerinRange 5 43)))
    (bind (_direction _newtrack_0) (Timeline|Timeline PlayFromStart)
      (:Update
        (Rendering|Material|SetScalarParameterValueonMaterials (CustomItemWrap|GetWrappedMeshComponents) "EDRec_Counter__Number" (Math|Float|Lerp (Variables|Default|GetA) (Variables|Default|GetB) _newtrack_0)))
      (:NewTrack_1
        (Variables|Default|SetA (Math|Conversions|ToFloat(Integer) _returnvalue)))
      (:NewTrack_2
        (Variables|Default|SetB (Math|Conversions|ToFloat(Integer) _returnvalue))))
    (else
      (Utilities|Time|SetTimerbyEvent (delegate Custom|Z_position) 0.2 true))))

(event EventOnPawnSet
  (if (not (CustomCosmeticModifier|IsinFrontEnd))
    (Utilities|Time|SetTimerbyEvent (delegate Custom|Z_position) 0.2 true)))

(event EventOnSetWrapTarget
  (if (not (CustomCosmeticModifier|IsinFrontEnd))
    (Utilities|Time|SetTimerbyEvent (delegate Custom|Z_position) 0.2 true)))

(event EventOnCharacterCustomizationFinished
  (params (CharacterPartInfo :type Array%20of%20Applied%20Character%20Part%20Info%20Structures))
  (if (not (CustomCosmeticModifier|IsinFrontEnd))
    (Utilities|Time|SetTimerbyEvent (delegate Custom|Z_position) 0.2 true)))

(event Custom|Z_position
  (bind _returnvalue (CustomCharacter|GetAssociatedPlayerPawn))
  (bind _returnvalue_2 (Customization|GetSkeletalMeshforPartType _returnvalue "Body" :!class /Script/FortniteGame.FortPlayerPawn))
  (bind (_returnvalue_x _returnvalue_y _returnvalue_z) (Math|Vector|BreakVector (Transformation|GetWorldLocation _returnvalue_2)))
  (bind _returnvalue_3 (/ _returnvalue_z 1e+02))
  (bind _returnvalue_4 (Math|Float|Round _returnvalue_3))
  (bind _returnvalue_5 (Math|Conversions|ToFloat(Integer) _returnvalue_4))
  (bind _returnvalue_6 (Math|Float|Clamp(Float) _returnvalue_5 -999.0 999.0))
  (Rendering|Material|SetScalarParameterValueonMaterials (CustomItemWrap|GetWrappedMeshComponents) "EDRec_Counter__Number" _returnvalue_6)
  (Development|PrintText (Utilities|Text|ToText(Float) _returnvalue_6) true true "(R=0.000000,G=0.660000,B=1.000000,A=1.000000)" 3.0))

(event Custom|OnBannerFullyApplied
  (params (Icon :type Texture%202D%20Object%20Reference) (PrimaryColor :type Linear%20Color%20Structure)))

(fn ConstructionScript)
