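; GlintRush reactive item wrap logic that shifts material blend and emissive offsets as the player's kill count rises.
; A stat watcher on Athena kills stores count and kill time, while a front-end timeline previews the stages.
; CountPerctange converts the clamped count (0–4) into start/end offsets, pushed with kill time to wrapped mesh materials.
; Weapon changes re-push after a short delay with kill time set far in the past.
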(event EventOnWeaponFired)

(event EventBeginPlay
  (if (CustomCosmeticModifier|IsinFrontEnd)
    (Timeline|Timeline PlayFromStart
      (:Stage1
        (Variables|Default|SetKill_Count 0))
      (:Stage2
        (Variables|Default|SetKill_Count 1))
      (:Stage3
        (Variables|Default|SetKill_Count 2))
      (:Stage4
        (Variables|Default|SetKill_Count 3))
      (:stage5
        (Variables|Default|SetKill_Count 4)))
    (Variables|Default|SetLastKillTime (Utilities|Time|GetGameTimeinSeconds))
    (CallFunction|PushMaterialParameters)))

(event EventOnPawnSet
  (Utilities|FlowControl|Sequence
    (:then_1
      (bind (_statname _statvalue) (Cosmetics|WatchCosmeticStat (CustomCharacter|GetAssociatedPlayerPawn) "(TagName=\"StatManager.AthenaKills\")")
        (:OnStatChanged
          (bind _output_get (Variables|Default|SetKill_Count _statvalue))
          (Variables|Default|SetLastKillTime (Utilities|Time|GetGameTimeinSeconds))
          (CallFunction|PushMaterialParameters))))))

(event EventOnPostResIn
  (CallFunction|PushMaterialParameters))

(event EventOnNewWeapons
  (params (WeaponsAsItems :type Array%20of%20Fort%20Item%20Object%20References))
  (Utilities|FlowControl|Delay 0.05)
  ;[1 1 1] Updated kill count on new weapon
  (Variables|Default|SetLastKillTime -999999999999.0)
  (CallFunction|PushMaterialParameters))

(event EventOnNewPlayerWeaponType
  (params (NewType :type EFortItemType%20Enum :default "WorldItem") (OldType :type EFortItemType%20Enum :default "WorldItem"))
  (Utilities|FlowControl|Delay 0.05)
  ;[1 1 1] Updated kill count on new weapon
  (Variables|Default|SetLastKillTime -999999999999.0)
  (CallFunction|PushMaterialParameters))

(event EventOnWeaponChanged
  (params (NewWeapon :type Fort%20Weapon%20Object%20Reference) (OldWeapon :type Fort%20Weapon%20Object%20Reference))
  (Utilities|FlowControl|Delay 0.05)
  ;[1 1 1] Updated kill count on new weapon
  (Variables|Default|SetLastKillTime -999999999999.0)
  (CallFunction|PushMaterialParameters))

(event EventApplyMaterialParameters
  (CallFunction|PushMaterialParameters))

(event Custom|Push%20Material%20Parameters
  (bind _returnvalue (CustomItemWrap|GetWrappedMeshComponents))
  (bind _backpack_offset (Variables|Default|GetBackPack%20offset))
  (bind _lastkilltime (Variables|Default|GetLastKillTime))
  (bind _end (Variables|Default|GetEnd))
  (bind _start (Variables|Default|GetStart))
  (switch int (Math|Integer|Clamp(Integer) (Variables|Default|GetKill_Count) 0 4)
    (:0
      (CallFunction|CountPerctange 0.0 0.0 1.0 0.2))
    (:1
      (CallFunction|CountPerctange 1.295))
    (:2
      (CallFunction|CountPerctange 2.5 1.295))
    (:3
      (CallFunction|CountPerctange 3.5 2.5))
    (:4
      (CallFunction|CountPerctange 3.5 3.5)))
  (Utilities|FlowControl|Sequence
    (:then_0
      (Rendering|Material|SetVectorParameterValueonMaterials _returnvalue "TopS_Blend_Offset_Backpack" (Math|Vector|MakeVector _backpack_offset))
      (Rendering|Material|SetVectorParameterValueonMaterials _returnvalue "TopS_Blend_Offset_Character" (Math|Vector|MakeVector (Variables|Default|GetCharacter%20offset)))
      (Rendering|Material|SetScalarParameterValueonMaterials _returnvalue "TopS_01_Blend_Offset---End" _end)
      (Rendering|Material|SetScalarParameterValueonMaterials _returnvalue "TopS_01_Blend_Offset---Start" _start)
      (Rendering|Material|SetScalarParameterValueonMaterials _returnvalue "TopS_01_Blend_Offset-TBP" (Variables|Default|GetLastKillTime)))
    (:then_1
      (Rendering|Material|SetVectorParameterValueonMaterials _returnvalue "TopS_Em01_Offset_Backpack" (Math|Vector|MakeVector _backpack_offset))
      (Rendering|Material|SetScalarParameterValueonMaterials _returnvalue "TopS Emissive_01_Offset---End" _end)
      (Rendering|Material|SetScalarParameterValueonMaterials _returnvalue "TopS Emissive_01_Offset---Start" _start)
      (Rendering|Material|SetScalarParameterValueonMaterials _returnvalue "TopS Emissive_01_Offset----TBP" _lastkilltime)
      (Rendering|Material|SetScalarParameterValueonMaterials _returnvalue "TopS Emissive_01-Brightness-TBP" _lastkilltime))))

(fn CountPerctange
  (params (End%20count :type Float%20%28double-precision%29) (StartCount :type Float%20%28double-precision%29) (Character%20offset :type Float%20%28single-precision%29) (BackPack%20offset :type Float%20%28double-precision%29))
  (Variables|Default|SetBackPack%20offset BackPack%20offset)
  (Variables|Default|SetCharacter%20offset Character%20offset)
  (Variables|Default|SetEnd (Math|Float|Lerp 0.65 -0.65 (/ End%20count 3.0)))
  (Variables|Default|SetStart (Math|Float|Lerp 0.65 -0.65 (/ StartCount 3.0))))

(fn ConstructionScript)
