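; Weapon wrap Blueprint for the GothDevil cosmetic that makes the wrap's materials react each time the weapon fires.
; On fire, it writes the current game time into decal, flipbook, opacity and emissive scalar parameters on the wrapped mesh components, giving the materials a start time for their effects.
; It also restarts a FadeOut timeline that scales StarColorVector to drive the small and medium galaxy star brightness parameters.
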(event EventOnWeaponFired
  (bind _returnvalue (CustomItemWrap|GetWrappedMeshComponents))
  (bind _returnvalue_2 (Utilities|Time|GetGameTimeinSeconds))
  (bind _returnvalue_3 (CustomItemWrap|GetWrappedMeshComponents))
  (bind _returnvalue_4 (Utilities|Time|GetGameTimeinSeconds))
  (bind _returnvalue_5 (CustomItemWrap|GetWrappedMeshComponents))
  (bind _returnvalue_6 (CustomItemWrap|GetWrappedMeshComponents))
  (bind _returnvalue_7 (Utilities|Time|GetGameTimeinSeconds))
  (bind _starcolorvector (Variables|Default|GetStarColorVector))
  (Utilities|FlowControl|Sequence
    (:then_0
      ;[1 1 1] General
      (Rendering|Material|SetScalarParameterValueonMaterials _returnvalue_3 "TimeFrom-BP----------------" _returnvalue_2)
      (Rendering|Material|SetScalarParameterValueonMaterials _returnvalue_3 "FDecal_Emissive_Brightness_Time_BP" _returnvalue_2)
      (Rendering|Material|SetScalarParameterValueonMaterials _returnvalue_3 "TopS Emissive_01-Brightness-TBP" _returnvalue_2))
    (:then_1
      ;[1 1 1] Left Right
      (Rendering|Material|SetScalarParameterValueonMaterials _returnvalue_6 "FDecal_Emissive_Gradent_Time_BP" _returnvalue_7)
      (Rendering|Material|SetScalarParameterValueonMaterials _returnvalue_6 "FDecal_Opacity_Time_BP_Z" _returnvalue_7))
    (:then_2
      (Rendering|Material|SetScalarParameterValueonMaterials _returnvalue_5 "FDecalV2_FlipBook_Time_FromBP" _returnvalue_4)
      (Rendering|Material|SetScalarParameterValueonMaterials _returnvalue_5 "FDecalaV2_Opacity_Time_BP_Z" _returnvalue_4)
      (Rendering|Material|SetScalarParameterValueonMaterials _returnvalue_5 "FDecalV2_Emissive_Brightness_Time_BP" _returnvalue_4)
      (Rendering|Material|SetScalarParameterValueonMaterials _returnvalue_5 "FDecalV2_Emissive_Gradent_Time_BP" _returnvalue_4)
      (Rendering|Material|SetScalarParameterValueonMaterials _returnvalue_5 "FDecalV2_Counter_Brightness_Time_BP" _returnvalue_4))
    (:then_3
      (bind (_direction _value) (Timeline|FadeOut PlayFromStart)
        (:Update
          (Rendering|Material|SetVectorParameterValueonMaterials _returnvalue (Utilities|String|StringToName "Eff_Galaxy_Star_Small_Brightness") (* _starcolorvector _value))
          (Rendering|Material|SetVectorParameterValueonMaterials _returnvalue (Utilities|String|StringToName "Eff_Galaxy_Star_Med_Brightness") (* _starcolorvector _value)))))))

(fn ConstructionScript)
