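; GrimTalon reactive weapon wrap Blueprint that animates the wrapped meshes' materials on weapon fire and kills.
; Both events stamp game time into decal and emissive parameters, letting the material animate from that moment.
; After a retriggerable delay, firing picks a random flipbook frame, while kills show frame 19 and briefly lock out changes.
; In the front end, a timeline drives time-of-day overrides and every seventh shot mimics the kill reaction.
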(event EventBeginPlay
  (bind _returnvalue (CustomItemWrap|GetWrappedMeshComponents))
  (Utilities|FlowControl|Delay 0.2)
  (if (CustomCosmeticModifier|IsinFrontEnd)
    (Rendering|Material|SetScalarParameterValueonMaterials (CustomItemWrap|GetWrappedMeshComponents) "00-FE-TOD--" 1.0)
    (bind (_direction _tod_value _newtrack_1 _newtrack_0) (Timeline|Timeline PlayFromStart)
      (:Update
        (Rendering|Material|SetScalarParameterValueonMaterials _returnvalue "00-FE-TODCC-Override" _tod_value)
        (Rendering|Material|SetColorParameterValueonMaterials _returnvalue "00-FE-TODV4-Override" _newtrack_0)))))

(event EventOnWeaponFired
  (bind _returnvalue (CustomItemWrap|GetWrappedMeshComponents))
  (bind _returnvalue_2 (Utilities|Time|GetGameTimeinSeconds))
  (bind _returnvalue_3 (CustomItemWrap|GetWrappedMeshComponents))
  (bind _returnvalue_4 (Utilities|Time|GetGameTimeinSeconds))
  (bind _returnvalue_5 (CustomItemWrap|GetWrappedMeshComponents))
  (bind _returnvalue_6 (Utilities|Time|GetGameTimeinSeconds))
  (if (CustomCosmeticModifier|IsinFrontEnd)
    (Variables|Default|SetFire? true)
    (if (Variables|Default|GetEnabled?)
      (Utilities|FlowControl|Delay 0.001)
      (Variables|Default|SetShotsFired (+ (Variables|Default|GetShotsFired) 1))
      (if (== (Variables|Default|GetShotsFired) 7)
        (Variables|Default|SetShotsFired 0)
        (Variables|Default|SetFire? false)
        (Variables|Default|SetEnabled? false)
        (else
          _)))
    (else
      (Variables|Default|SetFire? true)
      (if (Variables|Default|GetEnabled?)
        _)))
  (Utilities|FlowControl|Delay 0.05)
  ;[1 1 1 0.5] Decal Reactive
  (Rendering|Material|SetScalarParameterValueonMaterials _returnvalue_3 "EDRec_TimeFrom_BP_DecalA" _returnvalue_4)
  (Rendering|Material|SetScalarParameterValueonMaterials _returnvalue_3 "EDRec_TimeFrom_BP_DecalB" _returnvalue_4)
  (Rendering|Material|SetScalarParameterValueonMaterials _returnvalue_3 "EDRec_TimeFrom_BP_DecalC" _returnvalue_4)
  (Rendering|Material|SetScalarParameterValueonMaterials _returnvalue_3 "EDRec_TimeFrom_BP_DecalD" _returnvalue_4)
  (Rendering|Material|SetScalarParameterValueonMaterials _returnvalue_5 "EDRec_Emissive_A_Brightness_Time_BP" _returnvalue_6)
  (Rendering|Material|SetScalarParameterValueonMaterials _returnvalue_5 "EDRec_Emissive_B_Brightness_Time_BP" _returnvalue_6)
  (Rendering|Material|SetScalarParameterValueonMaterials _returnvalue_5 "EDRec_Emissive_C_Brightness_Time_BP" _returnvalue_6)
  (Rendering|Material|SetScalarParameterValueonMaterials _returnvalue_5 "EDRec_Emissive_D_Brightness_Time_BP" _returnvalue_6)
  (Rendering|Material|SetScalarParameterValueonMaterials _returnvalue "EDRec_Emissive_A_Gradent_Time_BP" _returnvalue_2)
  (Rendering|Material|SetScalarParameterValueonMaterials _returnvalue "EDRec_Emissive_B_Gradent_Time_BP" _returnvalue_2)
  (Rendering|Material|SetScalarParameterValueonMaterials _returnvalue "EDRec_Emissive_C_Gradent_Time_BP" _returnvalue_2)
  (Rendering|Material|SetScalarParameterValueonMaterials _returnvalue "EDRec_Emissive_D_Gradent_Time_BP" _returnvalue_2)
  (Rendering|Material|SetScalarParameterValueonMaterials _returnvalue "EDRec_DecalA_BounceRot_TimeFrom_BP" _returnvalue_2)
  (Rendering|Material|SetScalarParameterValueonMaterials _returnvalue "EDRec_DecalB_BounceRot_TimeFrom_BP" _returnvalue_2)
  (Rendering|Material|SetScalarParameterValueonMaterials _returnvalue "EDRec_DecalC_BounceRot_TimeFrom_BP" _returnvalue_2)
  (Rendering|Material|SetScalarParameterValueonMaterials _returnvalue "EDRec_DecalD_BounceRot_TimeFrom_BP" _returnvalue_2)
  (Rendering|Material|SetScalarParameterValueonMaterials _returnvalue "EDRec_DecalB__Sine_TimeFrom_BP" _returnvalue_2)
  (if (Variables|Default|GetFire?)
    (Utilities|FlowControl|RetriggerableDelay 0.8)
    (if (Variables|Default|GetEnabled?)
      (Rendering|Material|SetScalarParameterValueonMaterials _returnvalue "EDRec_FlipBook_FrameA" (Math|Conversions|ToFloat(Integer) (Math|Random|RandomIntegerinRange 0 18))))
    (else
      (Utilities|FlowControl|RetriggerableDelay 0.8)
      (Rendering|Material|SetScalarParameterValueonMaterials _returnvalue "EDRec_FlipBook_FrameA" 19.0)
      (if (CustomCosmeticModifier|IsinFrontEnd)
        (Utilities|FlowControl|Delay 2.0)
        (Variables|Default|SetEnabled? true)
        (else
          (Utilities|FlowControl|Delay 4.0)
          (Variables|Default|SetEnabled? true))))))

(event EventOnPawnSet
  (bind _returnvalue (CustomItemWrap|GetWrappedMeshComponents))
  (bind _returnvalue_2 (Utilities|Time|GetGameTimeinSeconds))
  (bind _returnvalue_3 (CustomItemWrap|GetWrappedMeshComponents))
  (bind _returnvalue_4 (Utilities|Time|GetGameTimeinSeconds))
  (bind _returnvalue_5 (CustomItemWrap|GetWrappedMeshComponents))
  (bind _returnvalue_6 (Utilities|Time|GetGameTimeinSeconds))
  (Rendering|Material|SetScalarParameterValueonMaterials (CustomItemWrap|GetWrappedMeshComponents) "EDRec_FlipBook_FrameA" (Math|Conversions|ToFloat(Integer) (Math|Random|RandomIntegerinRange 0 18)))
  (Cosmetics|WatchCosmeticStat (CustomCharacter|GetAssociatedPlayerPawn) "(TagName=\"StatManager.AthenaKills\")" 0 false
    (:OnStatChanged
      (Variables|Default|SetFire? false)
      (Variables|Default|SetEnabled? false)
      (Utilities|FlowControl|Delay 0.05)
      ;[1 1 1 0.5] Decal Reactive
      (Rendering|Material|SetScalarParameterValueonMaterials _returnvalue_3 "EDRec_TimeFrom_BP_DecalA" _returnvalue_4)
      (Rendering|Material|SetScalarParameterValueonMaterials _returnvalue_3 "EDRec_TimeFrom_BP_DecalB" _returnvalue_4)
      (Rendering|Material|SetScalarParameterValueonMaterials _returnvalue_3 "EDRec_TimeFrom_BP_DecalC" _returnvalue_4)
      (Rendering|Material|SetScalarParameterValueonMaterials _returnvalue_3 "EDRec_TimeFrom_BP_DecalD" _returnvalue_4)
      (Rendering|Material|SetScalarParameterValueonMaterials _returnvalue_5 "EDRec_Emissive_A_Brightness_Time_BP" _returnvalue_6)
      (Rendering|Material|SetScalarParameterValueonMaterials _returnvalue_5 "EDRec_Emissive_B_Brightness_Time_BP" _returnvalue_6)
      (Rendering|Material|SetScalarParameterValueonMaterials _returnvalue_5 "EDRec_Emissive_C_Brightness_Time_BP" _returnvalue_6)
      (Rendering|Material|SetScalarParameterValueonMaterials _returnvalue_5 "EDRec_Emissive_D_Brightness_Time_BP" _returnvalue_6)
      (Rendering|Material|SetScalarParameterValueonMaterials _returnvalue "EDRec_Emissive_A_Gradent_Time_BP" _returnvalue_2)
      (Rendering|Material|SetScalarParameterValueonMaterials _returnvalue "EDRec_Emissive_B_Gradent_Time_BP" _returnvalue_2)
      (Rendering|Material|SetScalarParameterValueonMaterials _returnvalue "EDRec_Emissive_C_Gradent_Time_BP" _returnvalue_2)
      (Rendering|Material|SetScalarParameterValueonMaterials _returnvalue "EDRec_Emissive_D_Gradent_Time_BP" _returnvalue_2)
      (Rendering|Material|SetScalarParameterValueonMaterials _returnvalue "EDRec_DecalA_BounceRot_TimeFrom_BP" _returnvalue_2)
      (Rendering|Material|SetScalarParameterValueonMaterials _returnvalue "EDRec_DecalB_BounceRot_TimeFrom_BP" _returnvalue_2)
      (Rendering|Material|SetScalarParameterValueonMaterials _returnvalue "EDRec_DecalC_BounceRot_TimeFrom_BP" _returnvalue_2)
      (Rendering|Material|SetScalarParameterValueonMaterials _returnvalue "EDRec_DecalD_BounceRot_TimeFrom_BP" _returnvalue_2)
      (Rendering|Material|SetScalarParameterValueonMaterials _returnvalue "EDRec_DecalB__Sine_TimeFrom_BP" _returnvalue_2)
      (if (Variables|Default|GetFire?)
        (Utilities|FlowControl|RetriggerableDelay 0.8)
        (if (Variables|Default|GetEnabled?)
          (Rendering|Material|SetScalarParameterValueonMaterials _returnvalue "EDRec_FlipBook_FrameA" (Math|Conversions|ToFloat(Integer) (Math|Random|RandomIntegerinRange 0 18))))
        (else
          (Utilities|FlowControl|RetriggerableDelay 0.8)
          (Rendering|Material|SetScalarParameterValueonMaterials _returnvalue "EDRec_FlipBook_FrameA" 19.0)
          (if (CustomCosmeticModifier|IsinFrontEnd)
            (Utilities|FlowControl|Delay 2.0)
            (Variables|Default|SetEnabled? true)
            (else
              (Utilities|FlowControl|Delay 4.0)
              (Variables|Default|SetEnabled? true))))))))

(fn ConstructionScript)
