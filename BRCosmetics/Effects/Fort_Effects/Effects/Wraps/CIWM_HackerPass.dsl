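; Weapon wrap Blueprint for HackerPass that reacts to firing by changing the wrap's emissive and decal on the wrapped meshes.
; Each shot writes game time into an emissive brightness material parameter, likely as a timestamp for a shader-driven pulse.
; DecalTextureSwap sets the decal texture on every dynamic material instance.
; A DoOnce gates the swap to the fired decal, and a three-second Delay restores the base decal.
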(event EventOnWeaponFired
  (Rendering|Material|SetScalarParameterValueonMaterials (CustomItemWrap|GetWrappedMeshComponents) "TopS Emissive_02-Brightness-TBP" (Utilities|Time|GetGameTimeinSeconds))
  (Utilities|FlowControl|Sequence
    (:then_0
      _)
    (:then_1
      (Utilities|FlowControl|Delay 3.0)
      (Default|DecalTextureSwap (Variables|Default|GetDecalBase))))
  (Utilities|FlowControl|DoOnce false
    (:Completed
      (Default|DecalTextureSwap (Variables|Default|GetDecalFired)))))

(fn DecalTextureSwap
  (params (Msk_Texture :type Texture%20Object%20Reference))
  (for _array_element (CustomItemWrap|GetWrappedMeshComponents)
    (for _array_element_2 (Rendering|Material|GetMaterials _array_element)
      (bind _asmaterial_instance_dynamic (Utilities|Casting|CastToMaterialInstanceDynamic _array_element_2))
      (Rendering|Material|SetTextureParameterValue _asmaterial_instance_dynamic "FDecalV2_BaseColor_Tex" Msk_Texture))))

(fn ConstructionScript)
