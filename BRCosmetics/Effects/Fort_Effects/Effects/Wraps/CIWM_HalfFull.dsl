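; Custom item wrap Blueprint for the HalfFull weapon wrap that reacts when the weapon fires.
; On each shot it writes the current game time to the "Liquid_TimeFrom-BP" scalar parameter on the materials of the wrapped mesh components.
; The Blueprint supplies only that timestamp, leaving any time-based liquid response to the material; its Tick and construction script are empty.
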(event EventOnWeaponFired
  (Rendering|Material|SetScalarParameterValueonMaterials (CustomItemWrap|GetWrappedMeshComponents) "Liquid_TimeFrom-BP" (Utilities|Time|GetGameTimeinSeconds)))

(event EventTick
  (params (DeltaSeconds :type Float%20%28single-precision%29)))

(fn ConstructionScript)
