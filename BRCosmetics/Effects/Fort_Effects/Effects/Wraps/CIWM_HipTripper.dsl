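; Cosmetic wrap modifier for HipTripper that animates wrap materials as the player deals damage, or on certain counted shots in the front end.
; Each trigger raises a capped step count that decays over time, then alternates between two timelines that sweep emissive layers and warp UVs with random offsets.
; A timeline event calls Wet Suirface, which sets blend coverage from the step count and writes timestamped parameters for a delayed material-driven fade.
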(event EventOnWeaponFired
  (bind _returnvalue (CustomItemWrap|GetWrappedMeshComponents))
  (bind _returnvalue_2 (CustomItemWrap|GetWrappedMeshComponents))
  (bind _returnvalue_3 (CustomItemWrap|GetWrappedMeshComponents))
  (bind _returnvalue_4 (CustomItemWrap|GetWrappedMeshComponents))
  (bind _returnvalue_5 (CustomItemWrap|GetWrappedMeshComponents))
  (bind _returnvalue_6 (CustomItemWrap|GetWrappedMeshComponents))
  ; Front End
  (if (CustomCosmeticModifier|IsinFrontEnd)
    (bind _output_get (Variables|Default|SetIterationCount (+ (Variables|Default|GetIterationCount) 1)))
    (if (>= _output_get 5)
      (if (== (Variables|Default|GetIterationCount) 7)
        (Variables|Default|SetIterationCount 0)
        (else
          (Variables|Default|SetStepCount (Math|Integer|Min(Integer) (+ (Math|Float|Truncate (Math|Float|Lerp (Math|Conversions|ToFloat(Integer) (Variables|Default|GetStepCount)) 0.0 (Math|Float|Clamp(Float) (/ (- (Utilities|Time|GetGameTimeinSeconds) (Variables|Default|Getwait_timer)) 1.0)))) 1) (Variables|Default|GetMaxSteps)))
          (Utilities|FlowControl|Delay 0.02)
          (switch int (Variables|Default|GetSelection)
            (:0
              (Variables|Default|SetSelection 1)
              (if (not (Variables|Default|GetPause))
                (Variables|Default|SetPause true)
                (Variables|Default|SetRandom_Float (Math|Random|RandomFloat))
                (Rendering|Material|SetScalarParameterValueonMaterials _returnvalue_3 "TopS Emissive_01_Offset---End" -1.0)
                (Rendering|Material|SetScalarParameterValueonMaterials _returnvalue_3 "TopS Emissive_01_Brightness_Darken_Base End" 0.9)
                (Rendering|Material|SetScalarParameterValueonMaterials _returnvalue_3 "TopS Emissive_01_Brightness---End" 1.0)
                (Components|Timeline|SetPlayRate (Variables|CIWM_HipTripper|GetA_Time) 2.0)
                (bind (_direction _newtrack_0 _noise _warp) (Timeline|A_Time PlayFromStart)
                  (:Update
                    (Rendering|Material|SetScalarParameterValueonMaterials _returnvalue "TopS Emissive_01_Offset---End" (Math|Float|Lerp -0.75 0.1 _newtrack_0))
                    (Rendering|Material|SetScalarParameterValueonMaterials _returnvalue "TopS_Em01_Warp_Amount" (Math|Float|Lerp 0.0 1.0 _warp))
                    (CallFunction|OffsetWarpUv03 (Math|Float|Lerp 0.0 0.75 _noise) (Variables|Default|GetRandom_Float)))
                  (:Finished
                    (Variables|Default|SetPause false))
                  (:NewTrack_3
                    (Rendering|Material|SetScalarParameterValueonMaterials _returnvalue_2 "TopS Emissive_01-Brightness-TBP" (Utilities|Time|GetGameTimeinSeconds))
                    (Rendering|Material|SetScalarParameterValueonMaterials _returnvalue_2 "TopS Emissive_01_Brightness_Darken_Base End" 0.0)
                    (Rendering|Material|SetScalarParameterValueonMaterials _returnvalue_2 "TopS Emissive_01_Brightness---End" 0.0)
                    (CallFunction|WetSuirface)))))
            (:1
              (Variables|Default|SetSelection 0)
              (if (not (Variables|Default|GetPause_0))
                (Variables|Default|SetPause_0 true)
                (Variables|Default|SetRandom_Float_B (Math|Random|RandomFloat))
                (Rendering|Material|SetScalarParameterValueonMaterials _returnvalue_5 "TopS Emissive_02_Offset---End" -1.0)
                (Rendering|Material|SetScalarParameterValueonMaterials _returnvalue_5 "TopS Emissive_02_Brightness_Darken_Base End" 0.9)
                (Rendering|Material|SetScalarParameterValueonMaterials _returnvalue_5 "TopS Emissive_02_Brightness---End" 1.0)
                (Components|Timeline|SetPlayRate (Variables|CIWM_HipTripper|GetB_Time) 2.0)
                (bind (_direction_2 _newtrack_0_2 _noise_2 _warp_2) (Timeline|B_Time PlayFromStart)
                  (:Update
                    (Rendering|Material|SetScalarParameterValueonMaterials _returnvalue_4 "TopS Emissive_02_Offset---End" (Math|Float|Lerp -0.75 0.1 _newtrack_0_2))
                    (Rendering|Material|SetScalarParameterValueonMaterials _returnvalue_4 "TopS_Em02_Warp_Amount" (Math|Float|Lerp 0.0 1.0 _warp_2))
                    (CallFunction|OffsetWarpUv4 (Math|Float|Lerp 0.0 0.75 _noise_2) (Variables|Default|GetRandom_Float_B)))
                  (:Finished
                    (Variables|Default|SetPause_0 false))
                  (:NewTrack_3
                    (Rendering|Material|SetScalarParameterValueonMaterials _returnvalue_6 "TopS Emissive_02-Brightness-TBP" (Utilities|Time|GetGameTimeinSeconds))
                    (Rendering|Material|SetScalarParameterValueonMaterials _returnvalue_6 "TopS Emissive_02_Brightness_Darken_Base End" 0.0)
                    (Rendering|Material|SetScalarParameterValueonMaterials _returnvalue_6 "TopS Emissive_02_Brightness---End" 0.0)
                    (CallFunction|WetSuirface)))))))))))

(event EventBeginPlay)

(event EventOnPawnSet
  (bind _returnvalue (CustomItemWrap|GetWrappedMeshComponents))
  (bind _returnvalue_2 (CustomItemWrap|GetWrappedMeshComponents))
  (bind _returnvalue_3 (CustomItemWrap|GetWrappedMeshComponents))
  (bind _returnvalue_4 (CustomItemWrap|GetWrappedMeshComponents))
  (bind _returnvalue_5 (CustomItemWrap|GetWrappedMeshComponents))
  (bind _returnvalue_6 (CustomItemWrap|GetWrappedMeshComponents))
  (Cosmetics|WatchCosmeticStat (CustomCharacter|GetAssociatedPlayerPawn) "(TagName=\"StatManager.DamageDealt\")" 0 false
    (:OnStatChanged
      (Variables|Default|SetStepCount (Math|Integer|Min(Integer) (+ (Math|Float|Truncate (Math|Float|Lerp (Math|Conversions|ToFloat(Integer) (Variables|Default|GetStepCount)) 0.0 (Math|Float|Clamp(Float) (/ (- (Utilities|Time|GetGameTimeinSeconds) (Variables|Default|Getwait_timer)) 1.0)))) 1) (Variables|Default|GetMaxSteps)))
      (Utilities|FlowControl|Delay 0.02)
      (switch int (Variables|Default|GetSelection)
        (:0
          (Variables|Default|SetSelection 1)
          (if (not (Variables|Default|GetPause))
            (Variables|Default|SetPause true)
            (Variables|Default|SetRandom_Float (Math|Random|RandomFloat))
            (Rendering|Material|SetScalarParameterValueonMaterials _returnvalue_3 "TopS Emissive_01_Offset---End" -1.0)
            (Rendering|Material|SetScalarParameterValueonMaterials _returnvalue_3 "TopS Emissive_01_Brightness_Darken_Base End" 0.9)
            (Rendering|Material|SetScalarParameterValueonMaterials _returnvalue_3 "TopS Emissive_01_Brightness---End" 1.0)
            (Components|Timeline|SetPlayRate (Variables|CIWM_HipTripper|GetA_Time) 2.0)
            (Timeline|A_Time PlayFromStart)))
        (:1
          (Variables|Default|SetSelection 0)
          (if (not (Variables|Default|GetPause_0))
            (Variables|Default|SetPause_0 true)
            (Variables|Default|SetRandom_Float_B (Math|Random|RandomFloat))
            (Rendering|Material|SetScalarParameterValueonMaterials _returnvalue_5 "TopS Emissive_02_Offset---End" -1.0)
            (Rendering|Material|SetScalarParameterValueonMaterials _returnvalue_5 "TopS Emissive_02_Brightness_Darken_Base End" 0.9)
            (Rendering|Material|SetScalarParameterValueonMaterials _returnvalue_5 "TopS Emissive_02_Brightness---End" 1.0)
            (Components|Timeline|SetPlayRate (Variables|CIWM_HipTripper|GetB_Time) 2.0)
            (Timeline|B_Time PlayFromStart)))))))

(event Custom|Wet%20Suirface
  (bind _returnvalue (CustomItemWrap|GetWrappedMeshComponents))
  (bind _returnvalue_2 (CustomItemWrap|GetWrappedMeshComponents))
  (Rendering|Material|SetScalarParameterValueonMaterials _returnvalue_2 "TopS_01_Blend_Opacity---End" 1.0)
  (bind _output_get (Variables|Default|SetLast%20Shot%20Value (Math|Float|Lerp 0.5 -1.0 (/ (Math|Conversions|ToFloat(Integer) (Variables|Default|GetStepCount)) (Variables|Default|GetMaxSteps)))))
  (Rendering|Material|SetScalarParameterValueonMaterials _returnvalue_2 "TopS_01_Blend_Offset---End" _output_get)
  (Variables|Default|Setwait_timer (+ (Utilities|Time|GetGameTimeinSeconds) 5.5))
  (Utilities|FlowControl|Delay 1.0)
  (Rendering|Material|SetScalarParameterValueonMaterials _returnvalue "TopS_01_Blend_Opacity-TBP" (+ (Utilities|Time|GetGameTimeinSeconds) 2.0))
  (Rendering|Material|SetScalarParameterValueonMaterials _returnvalue "TopS_01_Blend_Opacity---Start" 1.0)
  (Rendering|Material|SetScalarParameterValueonMaterials _returnvalue "TopS_01_Blend_Opacity---End" 0.0)
  (Rendering|Material|SetScalarParameterValueonMaterials _returnvalue "TopS_01_Blend_Opacity---Duration" 5.0))

(fn Offset_Warp_Uv03
  (params (offset :type Float%20%28double-precision%29) (Random :type Float%20%28double-precision%29))
  (bind _returnvalue (CustomItemWrap|GetWrappedMeshComponents))
  (Rendering|Material|SetVectorParameterValueonMaterials _returnvalue "TopS_03_OffsetS_Weapon" (Math|Vector|MakeVector Random offset))
  (Rendering|Material|SetVectorParameterValueonMaterials _returnvalue "TopS_03_OffsetS_Backpack" (Math|Vector|MakeVector Random offset))
  (Rendering|Material|SetVectorParameterValueonMaterials _returnvalue "TopS_03_OffsetS_Character" (Math|Vector|MakeVector Random offset))
  (Rendering|Material|SetVectorParameterValueonMaterials _returnvalue "TopS_03_OffsetS_Mech" (Math|Vector|MakeVector Random offset))
  (Rendering|Material|SetVectorParameterValueonMaterials _returnvalue "TopS_03_ScaleS_Pickaxe" (Math|Vector|MakeVector Random offset))
  (Rendering|Material|SetVectorParameterValueonMaterials _returnvalue "TopS_03_OffsetS_Vehicle" (Math|Vector|MakeVector Random offset))
  (Rendering|Material|SetVectorParameterValueonMaterials _returnvalue "TopS_03_OffsetS_VehicleSide" (Math|Vector|MakeVector Random)))

(fn Offset_Warp_uv4
  (params (offset :type Float%20%28double-precision%29) (Random :type Float%20%28double-precision%29))
  (bind _returnvalue (CustomItemWrap|GetWrappedMeshComponents))
  (Rendering|Material|SetVectorParameterValueonMaterials _returnvalue "TopS_04_Offset_Weapon" (Math|Vector|MakeVector Random offset))
  (Rendering|Material|SetVectorParameterValueonMaterials _returnvalue "TopS_04_Offset_Backpack" (Math|Vector|MakeVector Random offset))
  (Rendering|Material|SetVectorParameterValueonMaterials _returnvalue "TopS_04_Offset_Character" (Math|Vector|MakeVector Random offset))
  (Rendering|Material|SetVectorParameterValueonMaterials _returnvalue "TopS_04_Offset_Mech" (Math|Vector|MakeVector Random offset))
  (Rendering|Material|SetVectorParameterValueonMaterials _returnvalue "TopS_04_Offset_Pickaxe" (Math|Vector|MakeVector Random offset))
  (Rendering|Material|SetVectorParameterValueonMaterials _returnvalue "TopS_04_Offset_Vehicle" (Math|Vector|MakeVector Random offset))
  (Rendering|Material|SetVectorParameterValueonMaterials _returnvalue "TopS_04_Offset_VehicleSide" (Math|Vector|MakeVector Random offset)))

(fn TimeLine)

(fn ConstructionScript)
