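; Item wrap modifier for the Horned Judgment wrap that matches its glow to the body emissive color of a specific character.
; Pawn set, wrap target and customization finished events cache banner IDs and compare the equipped character's name to a stored name.
; On a match, SetColors writes the Body material's AdvancedEmissiveColor, boosted tenfold in RGB, to two emissive parameters on all wrapped mesh materials.
; The banner-applied event is empty.
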(event EventOnPawnSet
  (bind _returnvalue (CosmeticLoadout|GetCosmeticLoadout))
  (Variables|Default|SetBanner%20Icon%20Id (Utilities|String|StringToName (CosmeticLoadout|GetBannerIconIdfromLoadout _returnvalue)))
  (Variables|Default|SetBanner%20Color%20Id (Utilities|String|StringToName (CosmeticLoadout|GetBannerColorIdfromLoadout _returnvalue)))
  (if (Utilities|String|EqualExactly(String) (Item|GetPersistentName (CosmeticLoadout|GetItemDefinitionfromLoadout (CosmeticLoadout|GetCosmeticLoadout) "(TagName=\"LoadoutSlot.Character\")")) (Variables|Default|GetCharacter-Name))
    (bind _asmaterial_instance_dynamic (Utilities|Casting|CastToMaterialInstanceDynamic (Rendering|Material|GetMaterial (Customization|GetSkeletalMeshforPartType (CustomCharacter|GetAssociatedPlayerPawn) "Body" :!class /Script/FortniteGame.FortPlayerPawn) 0)))
    (bind _returnvalue_2 (Rendering|Material|GetVectorParameterValue _asmaterial_instance_dynamic "AdvancedEmissiveColor" :!class /Script/Engine.MaterialInstanceDynamic))
    (CallFunction|SetColors _returnvalue_2)))

(event EventOnSetWrapTarget
  (bind _returnvalue (CosmeticLoadout|GetCosmeticLoadout))
  (Variables|Default|SetBanner%20Icon%20Id (Utilities|String|StringToName (CosmeticLoadout|GetBannerIconIdfromLoadout _returnvalue)))
  (Variables|Default|SetBanner%20Color%20Id (Utilities|String|StringToName (CosmeticLoadout|GetBannerColorIdfromLoadout _returnvalue)))
  (if (Utilities|String|EqualExactly(String) (Item|GetPersistentName (CosmeticLoadout|GetItemDefinitionfromLoadout (CosmeticLoadout|GetCosmeticLoadout) "(TagName=\"LoadoutSlot.Character\")")) (Variables|Default|GetCharacter-Name))
    (bind _asmaterial_instance_dynamic (Utilities|Casting|CastToMaterialInstanceDynamic (Rendering|Material|GetMaterial (Customization|GetSkeletalMeshforPartType (CustomCharacter|GetAssociatedPlayerPawn) "Body" :!class /Script/FortniteGame.FortPlayerPawn) 0)))
    (bind _returnvalue_2 (Rendering|Material|GetVectorParameterValue _asmaterial_instance_dynamic "AdvancedEmissiveColor" :!class /Script/Engine.MaterialInstanceDynamic))
    (CallFunction|SetColors _returnvalue_2)))

(event EventOnCharacterCustomizationFinished
  (params (CharacterPartInfo :type Array%20of%20Applied%20Character%20Part%20Info%20Structures))
  (bind _returnvalue (CosmeticLoadout|GetCosmeticLoadout))
  (Variables|Default|SetBanner%20Icon%20Id (Utilities|String|StringToName (CosmeticLoadout|GetBannerIconIdfromLoadout _returnvalue)))
  (Variables|Default|SetBanner%20Color%20Id (Utilities|String|StringToName (CosmeticLoadout|GetBannerColorIdfromLoadout _returnvalue)))
  (if (Utilities|String|EqualExactly(String) (Item|GetPersistentName (CosmeticLoadout|GetItemDefinitionfromLoadout (CosmeticLoadout|GetCosmeticLoadout) "(TagName=\"LoadoutSlot.Character\")")) (Variables|Default|GetCharacter-Name))
    (bind _asmaterial_instance_dynamic (Utilities|Casting|CastToMaterialInstanceDynamic (Rendering|Material|GetMaterial (Customization|GetSkeletalMeshforPartType (CustomCharacter|GetAssociatedPlayerPawn) "Body" :!class /Script/FortniteGame.FortPlayerPawn) 0)))
    (bind _returnvalue_2 (Rendering|Material|GetVectorParameterValue _asmaterial_instance_dynamic "AdvancedEmissiveColor" :!class /Script/Engine.MaterialInstanceDynamic))
    (CallFunction|SetColors _returnvalue_2)))

(event Custom|OnBannerFullyApplied
  (params (Icon :type Texture%202D%20Object%20Reference) (PrimaryColor :type Linear%20Color%20Structure)))

(fn setcolors
  (params (Color :type Linear%20Color%20Structure))
  (for _array_element (CustomItemWrap|GetWrappedMeshComponents)
    (for _array_element_2 (Rendering|Material|GetMaterials _array_element)
      (bind _asmaterial_instance_dynamic (Utilities|Casting|CastToMaterialInstanceDynamic _array_element_2))
      (Rendering|Material|SetVectorParameterValue _asmaterial_instance_dynamic "TopS Emissive_02_Color-" (* Color "(R=10.000000,G=10.000000,B=10.000000,A=1.000000)") :!class /Script/Engine.MaterialInstanceDynamic)
      (Rendering|Material|SetVectorParameterValue _asmaterial_instance_dynamic "TopS Emissive_01_Color-" (* Color "(R=10.000000,G=10.000000,B=10.000000,A=1.000000)") :!class /Script/Engine.MaterialInstanceDynamic))))

(fn ConstructionScript)
