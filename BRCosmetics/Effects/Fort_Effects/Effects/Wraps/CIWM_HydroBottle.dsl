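; Weapon wrap modifier for the HydroBottle wrap that drives an ammo counter decal and timed emissive effect on wrapped meshes.
; Pawn and weapon events cache magazine counts after short delays and bind reload-start, which resyncs once reload time elapses.
; Firing decrements the count, and AmmoUpdate writes it plus game time to material parameters.
; The front end simulates a looping eight-round clip; health and shield stat watchers have empty handlers.
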(event EventOnWeaponFired
  (bind _returnvalue (CustomCharacter|GetAssociatedPlayerPawn))
  (bind _currentweapon (Variables|/Script/FortniteGame.FortPawn|GetCurrentWeapon _returnvalue))
  (bind _returnvalue_2 (FortWeapon|GetMagazineAmmoCount _currentweapon))
  (if (CustomCosmeticModifier|IsinFrontEnd)
    (bind _output_get (Variables|Default|SetAmmo%20Count (Variables|Default|GetAmmo%20Count)))
    (Utilities|FlowControl|Delay 0.001)
    (bind _output_get_2 (Variables|Default|SetAmmo%20Count (- _output_get 1)))
    (if (== _output_get_2 0)
      (bind _output_get_3 (Variables|Default|SetAmmo%20Count 8))
      (CallFunction|AmmoUpdate (Math|Conversions|ToFloat(Integer) _output_get_3))
      (else
        _))
    (elif (== _returnvalue_2 0)
      _
      (else
        (bind _output_get_4 (Variables|Default|SetAmmo%20Count _returnvalue_2))
        (Utilities|FlowControl|Delay 0.001)
        (Variables|Default|SetAmmo%20Count (- _output_get_4 1)))))
  (CallFunction|AmmoUpdate (Math|Conversions|ToFloat(Integer) (Variables|Default|GetAmmo%20Count))))

(event EventBeginPlay
  (bind _returnvalue (CustomItemWrap|GetWrappedMeshComponents))
  (bind _returnvalue_2 (CustomCharacter|GetAssociatedPlayerPawn))
  (bind _currentweapon (Variables|/Script/FortniteGame.FortPawn|GetCurrentWeapon _returnvalue_2))
  (if (CustomCosmeticModifier|IsinFrontEnd)
    (Utilities|FlowControl|Delay 0.2)
    (else
      (Utilities|FlowControl|Delay 0.2)
      (bind _currentweapon_2 (Variables|/Script/FortniteGame.FortPawn|GetCurrentWeapon (CustomCharacter|GetAssociatedPlayerPawn))
        (:then
          (EventDispatchers|BindEventtoOnReloadStarted _currentweapon_2 (delegate Custom|ReloadStarted))
          (if (not (CustomCosmeticModifier|IsinFrontEnd))
            (Variables|Default|Setper%20Clip (FortWeapon|GetBulletsPerClip _currentweapon))
            (bind _output_get (Variables|Default|SetAmmo%20Count (FortWeapon|GetMagazineAmmoCount _currentweapon)))
            (Rendering|Material|SetScalarParameterValueonMaterials _returnvalue "FDecalV2_Counter__Number" (Math|Conversions|ToFloat(Integer) _output_get))))
        (:else))))
  (Variables|Default|Setper%20Clip 8)
  (bind _output_get_2 (Variables|Default|SetAmmo%20Count 8))
  (Rendering|Material|SetScalarParameterValueonMaterials _returnvalue "FDecalV2_Counter__Number" (Math|Conversions|ToFloat(Integer) _output_get_2)))

(event EventOnWeaponChanged
  (params (NewWeapon :type Fort%20Weapon%20Object%20Reference) (OldWeapon :type Fort%20Weapon%20Object%20Reference))
  (bind _returnvalue (CustomItemWrap|GetWrappedMeshComponents))
  (bind _returnvalue_2 (CustomCharacter|GetAssociatedPlayerPawn))
  (bind _currentweapon (Variables|/Script/FortniteGame.FortPawn|GetCurrentWeapon _returnvalue_2))
  (Utilities|FlowControl|Delay 0.2)
  (bind _currentweapon_2 (Variables|/Script/FortniteGame.FortPawn|GetCurrentWeapon (CustomCharacter|GetAssociatedPlayerPawn))
    (:then
      (EventDispatchers|BindEventtoOnReloadStarted _currentweapon_2 (delegate Custom|ReloadStarted))
      (if (CustomCosmeticModifier|IsinFrontEnd)
        (Variables|Default|Setper%20Clip 8)
        (bind _output_get (Variables|Default|SetAmmo%20Count 8))
        (Rendering|Material|SetScalarParameterValueonMaterials _returnvalue "FDecalV2_Counter__Number" (Math|Conversions|ToFloat(Integer) _output_get))
        (else
          (Variables|Default|Setper%20Clip (FortWeapon|GetBulletsPerClip _currentweapon))
          (bind _output_get_2 (Variables|Default|SetAmmo%20Count (FortWeapon|GetMagazineAmmoCount _currentweapon)))
          (Rendering|Material|SetScalarParameterValueonMaterials _returnvalue "FDecalV2_Counter__Number" (Math|Conversions|ToFloat(Integer) _output_get_2)))))
    (:else)))

(event EventOnPawnSet
  (bind _returnvalue (CustomItemWrap|GetWrappedMeshComponents))
  (bind _returnvalue_2 (CustomCharacter|GetAssociatedPlayerPawn))
  (bind _returnvalue_3 (CustomCharacter|GetAssociatedPlayerPawn))
  (bind _currentweapon (Variables|/Script/FortniteGame.FortPawn|GetCurrentWeapon _returnvalue_2))
  (bind _returnvalue_4 (FortWeapon|GetMagazineAmmoCount _currentweapon))
  (Utilities|FlowControl|Sequence
    (:then_0
      (Utilities|FlowControl|Delay 0.2)
      (bind _currentweapon_2 (Variables|/Script/FortniteGame.FortPawn|GetCurrentWeapon _returnvalue_3)
        (:then
          (EventDispatchers|BindEventtoOnReloadStarted _currentweapon_2 (delegate Custom|ReloadStarted))
          (if (CustomCosmeticModifier|IsinFrontEnd)
            (Variables|Default|Setper%20Clip 8)
            (bind _output_get (Variables|Default|SetAmmo%20Count 8))
            (Rendering|Material|SetScalarParameterValueonMaterials _returnvalue "FDecalV2_Counter__Number" (Math|Conversions|ToFloat(Integer) _output_get))
            (else
              (Variables|Default|Setper%20Clip (FortWeapon|GetBulletsPerClip _currentweapon))
              (bind _output_get_2 (Variables|Default|SetAmmo%20Count _returnvalue_4))
              (Rendering|Material|SetScalarParameterValueonMaterials _returnvalue "FDecalV2_Counter__Number" (Math|Conversions|ToFloat(Integer) _output_get_2)))))
        (:else)))
    (:then_1
      (bind _asfort_player_pawn (Utilities|Casting|CastToFortPlayerPawn _returnvalue_3))
      (Cosmetics|WatchCosmeticStat _asfort_player_pawn "(TagName=\"StatManager.GainedHealthTimes\")" 0 false
        (:OnStatChanged
          _)))
    (:then_2
      (bind _asfort_player_pawn_2 (Utilities|Casting|CastToFortPlayerPawn (CustomCharacter|GetAssociatedPlayerPawn)))
      (Cosmetics|WatchCosmeticStat _asfort_player_pawn_2 "(TagName=\"StatManager.GainedShieldTimes\")" 0 false
        (:OnStatChanged
          _))))
  (bind _output_get_3 (Variables|Default|SetAmmo%20Count _returnvalue_4))
  (Utilities|FlowControl|Delay 0.001)
  (Variables|Default|SetAmmo%20Count (- _output_get_3 1))
  (CallFunction|AmmoUpdate (Math|Conversions|ToFloat(Integer) (Variables|Default|GetAmmo%20Count))))

(event EventOnNewWeapons
  (params (WeaponsAsItems :type Array%20of%20Fort%20Item%20Object%20References))
  (bind _returnvalue (CustomItemWrap|GetWrappedMeshComponents))
  (bind _returnvalue_2 (CustomCharacter|GetAssociatedPlayerPawn))
  (bind _currentweapon (Variables|/Script/FortniteGame.FortPawn|GetCurrentWeapon _returnvalue_2))
  (Utilities|FlowControl|Delay 0.2)
  (bind _currentweapon_2 (Variables|/Script/FortniteGame.FortPawn|GetCurrentWeapon (CustomCharacter|GetAssociatedPlayerPawn))
    (:then
      (EventDispatchers|BindEventtoOnReloadStarted _currentweapon_2 (delegate Custom|ReloadStarted))
      (if (CustomCosmeticModifier|IsinFrontEnd)
        (Variables|Default|Setper%20Clip 8)
        (bind _output_get (Variables|Default|SetAmmo%20Count 8))
        (Rendering|Material|SetScalarParameterValueonMaterials _returnvalue "FDecalV2_Counter__Number" (Math|Conversions|ToFloat(Integer) _output_get))
        (else
          (Variables|Default|Setper%20Clip (FortWeapon|GetBulletsPerClip _currentweapon))
          (bind _output_get_2 (Variables|Default|SetAmmo%20Count (FortWeapon|GetMagazineAmmoCount _currentweapon)))
          (Rendering|Material|SetScalarParameterValueonMaterials _returnvalue "FDecalV2_Counter__Number" (Math|Conversions|ToFloat(Integer) _output_get_2)))))
    (:else)))

(event Custom|ReloadStarted
  (params (ReloadTime :type Float%20%28single-precision%29) (ReloadType :type EFortWeaponReloadType%20Enum :default "ReloadWholeClip"))
  (CallFunction|AmmoUpdate (Math|Conversions|ToFloat(Integer) (Variables|Default|GetAmmo%20Count)))
  (Utilities|FlowControl|Delay ReloadTime)
  (if (not (CustomCosmeticModifier|IsinFrontEnd))
    (bind _output_get (Variables|Default|SetAmmo%20Count (FortWeapon|GetMagazineAmmoCount (Variables|/Script/FortniteGame.FortPawn|GetCurrentWeapon (CustomCharacter|GetAssociatedPlayerPawn)))))
    (CallFunction|AmmoUpdate (Math|Conversions|ToFloat(Integer) _output_get))))

(event EventOnNewPlayerWeaponType
  (params (NewType :type EFortItemType%20Enum :default "WorldItem") (OldType :type EFortItemType%20Enum :default "WorldItem"))
  (bind _returnvalue (CustomItemWrap|GetWrappedMeshComponents))
  (bind _returnvalue_2 (CustomCharacter|GetAssociatedPlayerPawn))
  (bind _currentweapon (Variables|/Script/FortniteGame.FortPawn|GetCurrentWeapon _returnvalue_2))
  (Utilities|FlowControl|Delay 0.2)
  (bind _currentweapon_2 (Variables|/Script/FortniteGame.FortPawn|GetCurrentWeapon (CustomCharacter|GetAssociatedPlayerPawn))
    (:then
      (EventDispatchers|BindEventtoOnReloadStarted _currentweapon_2 (delegate Custom|ReloadStarted))
      (if (CustomCosmeticModifier|IsinFrontEnd)
        (Variables|Default|Setper%20Clip 8)
        (bind _output_get (Variables|Default|SetAmmo%20Count 8))
        (Rendering|Material|SetScalarParameterValueonMaterials _returnvalue "FDecalV2_Counter__Number" (Math|Conversions|ToFloat(Integer) _output_get))
        (else
          (Variables|Default|Setper%20Clip (FortWeapon|GetBulletsPerClip _currentweapon))
          (bind _output_get_2 (Variables|Default|SetAmmo%20Count (FortWeapon|GetMagazineAmmoCount _currentweapon)))
          (Rendering|Material|SetScalarParameterValueonMaterials _returnvalue "FDecalV2_Counter__Number" (Math|Conversions|ToFloat(Integer) _output_get_2)))))
    (:else)))

(event Custom|OnBannerFullyApplied
  (params (Icon :type Texture%202D%20Object%20Reference) (PrimaryColor :type Linear%20Color%20Structure)))

(fn AmmoUpdate
  (params (Ammo :type Float%20%28single-precision%29))
  (bind _returnvalue (CustomItemWrap|GetWrappedMeshComponents))
  (Rendering|Material|SetScalarParameterValueonMaterials _returnvalue "FDecalV2_Counter__Number" Ammo)
  (Rendering|Material|SetScalarParameterValueonMaterials _returnvalue "FDecalV2_Emissive_Brightness_Time_BP" (Utilities|Time|GetGameTimeinSeconds)))

(fn ConstructionScript)

(fn Weapon_Offsets
  (params (Offset_End_Name :type Name) (Offset_End_Value :type Float%20%28double-precision%29) (Offset_Start_Name :type Name) (Offset_State_Value :type Float%20%28double-precision%29))
  (bind _returnvalue (CustomItemWrap|GetWrappedMeshComponents))
  (Rendering|Material|SetVectorParameterValueonMaterials _returnvalue Offset_End_Name (Math|Vector|MakeVector 0.0 0.0 Offset_End_Value))
  (Rendering|Material|SetVectorParameterValueonMaterials _returnvalue Offset_Start_Name (Math|Vector|MakeVector 0.0 0.0 Offset_State_Value)))
