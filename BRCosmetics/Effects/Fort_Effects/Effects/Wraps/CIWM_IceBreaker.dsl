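; Item wrap logic for the IceBreaker weapon wrap, which updates the ice effect on wrapped meshes each time the weapon fires.
; On each shot it adjusts a step count by time since the stored wait time at four steps per second, caps it, and looks up a refreeze value for IB_BaseGrowth.
; It stores game time plus ThawDelay as the next wait time and passes it to IB_InputGameTime.
; A debug "Fired" print remains.
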(event EventOnWeaponFired
  (bind _returnvalue (CustomItemWrap|GetWrappedMeshComponents))
  (bind _returnvalue_2 (Utilities|Time|GetGameTimeinSeconds))
  (bind _thawdelay (Variables|Default|GetThawDelay))
  (bind _returnvalue_3 (+ _returnvalue_2 _thawdelay))
  (Development|PrintString "Fired")
  (bind _output_get (Variables|Default|SetStepCount (Math|Integer|Min(Integer) (+ (- (Variables|Default|GetStepCount) (Math|Integer|Min(Integer) (Variables|Default|GetStepCount) (Math|Float|Floor (* (- (Utilities|Time|GetGameTimeinSeconds) (Variables|Default|GetWaitTime)) 4.0)))) 1) (Variables|Default|GetMaxSteps))))
  (Rendering|Material|SetScalarParameterValueonMaterials _returnvalue "IB_BaseGrowth" (Utilities|Array|Get(aref) (Variables|Default|GetRefreezeStepValues) _output_get))
  (Variables|Default|SetStepCount (+ (Variables|Default|GetStepCount) 1))
  (Rendering|Material|SetScalarParameterValueonMaterials _returnvalue "IB_InputGameTime" _returnvalue_3)
  (Variables|Default|SetWaitTime _returnvalue_3))

(fn ConstructionScript)
