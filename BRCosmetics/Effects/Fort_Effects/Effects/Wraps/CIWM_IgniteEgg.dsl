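; Custom item wrap Blueprint for the IgniteEgg cosmetic, which changes its look in stages as the weapon fires, apparently an egg breaking.
; BeginPlay sets the starting "TopS_Blend_Offset__Master" vector parameter on the wrapped mesh materials.
; Each shot increments a counter that selects the next offset; a shot at the final stage starts a timer and ignores further shots.
; The timer calls Unbreak, restoring the first stage.
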(event EventBeginPlay
  (Rendering|Material|SetVectorParameterValueonMaterials (CustomItemWrap|GetWrappedMeshComponents) "TopS_Blend_Offset__Master" (0.25 0.0 0.0)))

(event Custom|Unbreak
  (Variables|Default|SetIterationCount 0)
  (Variables|Default|SetIsWaiting? false)
  (Rendering|Material|SetVectorParameterValueonMaterials (CustomItemWrap|GetWrappedMeshComponents) "TopS_Blend_Offset__Master" (Math|Vector|MakeVector (Utilities|Select :Index (Variables|Default|GetIterationCount) :Option%200 0.25 :Option%201 0.5 :Option%202 0.75 :Option%203 1.3 :Option%204 2.0))))

(event EventOnWeaponFired
  (if (Variables|Default|GetIsWaiting?)
    _
    (elif (>= (Variables|Default|GetIterationCount) 4)
      (Variables|Default|SetIsWaiting? true)
      (Utilities|Time|SetTimerbyFunctionName 0 "Unbreak" 1e+01)
      (else
        (Math|Integer|IncrementInt (Variables|Default|GetIterationCount)
          (:%20%20
            (Rendering|Material|SetVectorParameterValueonMaterials (CustomItemWrap|GetWrappedMeshComponents) "TopS_Blend_Offset__Master" (Math|Vector|MakeVector (Utilities|Select :Index (Variables|Default|GetIterationCount) :Option%200 0.25 :Option%201 0.5 :Option%202 0.75 :Option%203 1.3 :Option%204 2.0)))))))))

(fn ConstructionScript)
