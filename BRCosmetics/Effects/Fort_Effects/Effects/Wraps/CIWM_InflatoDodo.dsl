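; Weapon wrap cosmetic modifier that moves an emissive meter on wrapped meshes as the player deals damage.
; Damage stat changes increment a counter and kills max it out; wrap_offset maps it to material offset, and in the front end a timeline drives it instead.
; Near full, it pauses damage counting, pulses emissive brightness, then resets.
; Game-time stamps written to scalar parameters trigger the material-side animations.
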(event EventBeginPlay
  (Utilities|FlowControl|Sequence
    (:then_0
      (Utilities|FlowControl|Delay 0.2)
      (if (CustomCosmeticModifier|IsinFrontEnd)
        (bind (_direction _damage_amount) (Timeline|Timeline_0 Play)
          (:Damage%20delt
            (Variables|Default|SetDamage_Current (Math|Float|Truncate _damage_amount))
            (CallFunction|WrapOffset)))
        (else
          (Variables|Default|SetDamage_Current -1))))))

(event EventOnWeaponFired)

(event EventOnPawnSet
  (Utilities|FlowControl|Sequence
    (:then_1
      (Cosmetics|WatchCosmeticStat (CustomCharacter|GetAssociatedPlayerPawn) "(TagName=\"StatManager.DamageDealt\")" 0 false -1.0 -1.0 false
        (:OnStatChanged
          (if (not (Variables|Default|GetPause))
            (Variables|Default|SetDamage_Current (+ (Variables|Default|GetDamage_Current) 1))
            (CallFunction|WrapOffset)))))
    (:then_2
      (Cosmetics|WatchCosmeticStat (CustomCharacter|GetAssociatedPlayerPawn) "(TagName=\"StatManager.AthenaKills\")" 0 false -1.0 -1.0 false
        (:OnStatChanged
          (Variables|Default|SetPause true)
          (Variables|Default|SetDamage_Current 5)
          (CallFunction|WrapOffset))))))

(event Custom|wrap_offset
  (bind _returnvalue (Utilities|Time|GetGameTimeinSeconds))
  (bind _returnvalue_2 (Utilities|Time|GetGameTimeinSeconds))
  (bind _returnvalue_3 (Utilities|Time|GetGameTimeinSeconds))
  (bind _returnvalue_4 (CustomItemWrap|GetWrappedMeshComponents))
  (bind _returnvalue_5 (CustomItemWrap|GetWrappedMeshComponents))
  (bind _returnvalue_6 (CustomItemWrap|GetWrappedMeshComponents))
  (bind _returnvalue_7 (CustomItemWrap|GetWrappedMeshComponents))
  (Rendering|Material|SetScalarParameterValueonMaterials _returnvalue_7 "TopS Emissive_01_Offset---End" (Math|Float|Lerp 0.46 -0.5 (Math|Float|Clamp(Float) (/ (Math|Conversions|ToFloat(Integer) (Variables|Default|GetDamage_Current)) 5.0))))
  (Rendering|Material|SetScalarParameterValueonMaterials _returnvalue_7 "TopS Emissive_01-Brightness-TBP" (Utilities|Time|GetGameTimeinSeconds))
  (if (>= (Variables|Default|GetDamage_Current) 4)
    (Variables|Default|SetPause true)
    (Utilities|FlowControl|Sequence
      (:then_0
        (Utilities|FlowControl|Delay 0.25)
        (Rendering|Material|SetScalarParameterValueonMaterials _returnvalue_6 "TopS Emissive_01-Brightness-TBP" _returnvalue_2)
        (Utilities|FlowControl|Delay 0.25)
        (Rendering|Material|SetScalarParameterValueonMaterials _returnvalue_6 "TopS Emissive_01-Brightness-TBP" _returnvalue_2)
        (Utilities|FlowControl|Delay 0.25)
        (Rendering|Material|SetScalarParameterValueonMaterials _returnvalue_6 "TopS Emissive_01-Brightness-TBP" _returnvalue_2)
        (Utilities|FlowControl|Delay 0.25)
        (Rendering|Material|SetScalarParameterValueonMaterials _returnvalue_6 "TopS Emissive_01_Offset----TBP" _returnvalue_2)
        (Rendering|Material|SetScalarParameterValueonMaterials _returnvalue_6 "TopS Emissive_01_Offset---End" 0.51)
        (Variables|Default|SetDamage_Current -1)
        (Variables|Default|SetPause false))
      (:then_1
        (Rendering|Material|SetScalarParameterValueonMaterials _returnvalue_4 "EDRec_Emissive_A_Brightness_Time_BP" _returnvalue_3)
        (Rendering|Material|SetScalarParameterValueonMaterials _returnvalue_4 "EDRec_Emissive_B_Brightness_Time_BP" _returnvalue_3)
        (Utilities|FlowControl|Delay 0.25)
        (Rendering|Material|SetScalarParameterValueonMaterials _returnvalue_5 "EDRec_Emissive_A_Brightness_Time_BP" _returnvalue)
        (Rendering|Material|SetScalarParameterValueonMaterials _returnvalue_5 "EDRec_Emissive_B_Brightness_Time_BP" _returnvalue)))))

(fn RestPossition
  (params (End :type Float%20%28single-precision%29) (Start :type Float%20%28single-precision%29) (Time :type Float%20%28single-precision%29) (delay :type Float%20%28double-precision%29))
  (bind _returnvalue (CustomItemWrap|GetWrappedMeshComponents))
  (bind _returnvalue_2 (Utilities|Time|GetGameTimeinSeconds))
  (bind _returnvalue_3 (CustomItemWrap|GetWrappedMeshComponents))
  (bind _returnvalue_4 (CustomItemWrap|GetWrappedMeshComponents))
  (bind _returnvalue_5 (+ _returnvalue_2 delay))
  (Rendering|Material|SetScalarParameterValueonMaterials _returnvalue_4 "TopS_01_Blend_Offset---End" End)
  (Rendering|Material|SetScalarParameterValueonMaterials _returnvalue_4 "TopS_01_Blend_Offset---Start" Start)
  (Rendering|Material|SetScalarParameterValueonMaterials _returnvalue_4 "TopS_01_Blend_Offset-TBP" _returnvalue_5)
  (Rendering|Material|SetScalarParameterValueonMaterials _returnvalue_4 "TopS_01_Blend_Offset---Duration" Time)
  (Rendering|Material|SetScalarParameterValueonMaterials _returnvalue_3 "TopS Emissive_02_Offset---End" End)
  (Rendering|Material|SetScalarParameterValueonMaterials _returnvalue_3 "TopS Emissive_02_Offset---Start" Start)
  (Rendering|Material|SetScalarParameterValueonMaterials _returnvalue_3 "TopS Emissive_02_Offset----TBP" _returnvalue_5)
  (Rendering|Material|SetScalarParameterValueonMaterials _returnvalue_3 "TopS Emissive_02_Offset---Duration" Time)
  (Rendering|Material|SetScalarParameterValueonMaterials _returnvalue "TopS Emissive_03_Offset---End" End)
  (Rendering|Material|SetScalarParameterValueonMaterials _returnvalue "TopS Emissive_03_Offset---Start" Start)
  (Rendering|Material|SetScalarParameterValueonMaterials _returnvalue "TopS Emissive_03_Offset----TBP" _returnvalue_5)
  (Rendering|Material|SetScalarParameterValueonMaterials _returnvalue "TopS Emissive_03_Offset---Duration" Time))

(fn ConstructionScript)
