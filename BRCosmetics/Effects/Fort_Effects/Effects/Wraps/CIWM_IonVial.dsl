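; Custom item wrap logic for Ion Vial that applies the player's banner icon and matching character color to wrapped meshes.
; Pawn set, wrap target set and customization finished repeat the same logic, async-loading the banner texture into every wrap material.
; If the equipped character's name matches, it reads IonVial_CustomColor from the Body material.
; SetColors writes that to emissive and decal parameters, deriving a secondary emissive color by rotating RGB as a vector.
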(event EventOnPawnSet
  (bind _returnvalue (CosmeticLoadout|GetCosmeticLoadout))
  (Variables|Default|SetBanner%20Icon%20Id (Utilities|String|StringToName (CosmeticLoadout|GetBannerIconIdfromLoadout _returnvalue)))
  (Variables|Default|SetBanner%20Color%20Id (Utilities|String|StringToName (CosmeticLoadout|GetBannerColorIdfromLoadout _returnvalue)))
  (Utilities|FlowControl|Sequence
    (:then_1
      (bind _returnvalue_2 (Banner|GetBannerLargeIconAssetfromRowName (Variables|Default|GetBanner%20Icon%20Id) "PhysicalBanner"))
      (bind _object (Utilities|AsyncLoadAsset _returnvalue_2)
        (:Completed
          (bind _astexture_2d (Utilities|Casting|CastToTexture2D _object))
          (Variables|Default|SetBanner%20Icon%20Texture _astexture_2d)
          (for _array_element (CustomItemWrap|GetWrappedMeshComponents)
            (for _array_element_2 (Rendering|Material|GetMaterials _array_element)
              (bind _asmaterial_instance_dynamic (Utilities|Casting|CastToMaterialInstanceDynamic _array_element_2))
              (Rendering|Material|SetTextureParameterValue _asmaterial_instance_dynamic (Variables|Default|GetBannerTextureParamName) (Variables|Default|GetBanner%20Icon%20Texture))))
          (CallFunction|OnBannerFullyApplied (Variables|Default|GetBanner%20Icon%20Texture)))))
    (:then_2
      (if (Utilities|String|EqualExactly(String) (Item|GetPersistentName (CosmeticLoadout|GetItemDefinitionfromLoadout (CosmeticLoadout|GetCosmeticLoadout) "(TagName=\"LoadoutSlot.Character\")")) (Variables|Default|GetCharacter-Name))
        (bind _asmaterial_instance_dynamic_2 (Utilities|Casting|CastToMaterialInstanceDynamic (Rendering|Material|GetMaterial (Customization|GetSkeletalMeshforPartType (CustomCharacter|GetAssociatedPlayerPawn) "Body" :!class /Script/FortniteGame.FortPlayerPawn) 1)))
        (bind _returnvalue_3 (Rendering|Material|GetVectorParameterValue _asmaterial_instance_dynamic_2 "IonVial_CustomColor" :!class /Script/Engine.MaterialInstanceDynamic))
        (CallFunction|SetColors _returnvalue_3)))))

(event EventOnSetWrapTarget
  (bind _returnvalue (CosmeticLoadout|GetCosmeticLoadout))
  (Variables|Default|SetBanner%20Icon%20Id (Utilities|String|StringToName (CosmeticLoadout|GetBannerIconIdfromLoadout _returnvalue)))
  (Variables|Default|SetBanner%20Color%20Id (Utilities|String|StringToName (CosmeticLoadout|GetBannerColorIdfromLoadout _returnvalue)))
  (Utilities|FlowControl|Sequence
    (:then_1
      (bind _returnvalue_2 (Banner|GetBannerLargeIconAssetfromRowName (Variables|Default|GetBanner%20Icon%20Id) "PhysicalBanner"))
      (bind _object (Utilities|AsyncLoadAsset _returnvalue_2)
        (:Completed
          (bind _astexture_2d (Utilities|Casting|CastToTexture2D _object))
          (Variables|Default|SetBanner%20Icon%20Texture _astexture_2d)
          (for _array_element (CustomItemWrap|GetWrappedMeshComponents)
            (for _array_element_2 (Rendering|Material|GetMaterials _array_element)
              (bind _asmaterial_instance_dynamic (Utilities|Casting|CastToMaterialInstanceDynamic _array_element_2))
              (Rendering|Material|SetTextureParameterValue _asmaterial_instance_dynamic (Variables|Default|GetBannerTextureParamName) (Variables|Default|GetBanner%20Icon%20Texture))))
          (CallFunction|OnBannerFullyApplied (Variables|Default|GetBanner%20Icon%20Texture)))))
    (:then_2
      (if (Utilities|String|EqualExactly(String) (Item|GetPersistentName (CosmeticLoadout|GetItemDefinitionfromLoadout (CosmeticLoadout|GetCosmeticLoadout) "(TagName=\"LoadoutSlot.Character\")")) (Variables|Default|GetCharacter-Name))
        (bind _asmaterial_instance_dynamic_2 (Utilities|Casting|CastToMaterialInstanceDynamic (Rendering|Material|GetMaterial (Customization|GetSkeletalMeshforPartType (CustomCharacter|GetAssociatedPlayerPawn) "Body" :!class /Script/FortniteGame.FortPlayerPawn) 1)))
        (bind _returnvalue_3 (Rendering|Material|GetVectorParameterValue _asmaterial_instance_dynamic_2 "IonVial_CustomColor" :!class /Script/Engine.MaterialInstanceDynamic))
        (CallFunction|SetColors _returnvalue_3)))))

(event EventOnCharacterCustomizationFinished
  (params (CharacterPartInfo :type Array%20of%20Applied%20Character%20Part%20Info%20Structures))
  (bind _returnvalue (CosmeticLoadout|GetCosmeticLoadout))
  (Variables|Default|SetBanner%20Icon%20Id (Utilities|String|StringToName (CosmeticLoadout|GetBannerIconIdfromLoadout _returnvalue)))
  (Variables|Default|SetBanner%20Color%20Id (Utilities|String|StringToName (CosmeticLoadout|GetBannerColorIdfromLoadout _returnvalue)))
  (Utilities|FlowControl|Sequence
    (:then_1
      (bind _returnvalue_2 (Banner|GetBannerLargeIconAssetfromRowName (Variables|Default|GetBanner%20Icon%20Id) "PhysicalBanner"))
      (bind _object (Utilities|AsyncLoadAsset _returnvalue_2)
        (:Completed
          (bind _astexture_2d (Utilities|Casting|CastToTexture2D _object))
          (Variables|Default|SetBanner%20Icon%20Texture _astexture_2d)
          (for _array_element (CustomItemWrap|GetWrappedMeshComponents)
            (for _array_element_2 (Rendering|Material|GetMaterials _array_element)
              (bind _asmaterial_instance_dynamic (Utilities|Casting|CastToMaterialInstanceDynamic _array_element_2))
              (Rendering|Material|SetTextureParameterValue _asmaterial_instance_dynamic (Variables|Default|GetBannerTextureParamName) (Variables|Default|GetBanner%20Icon%20Texture))))
          (CallFunction|OnBannerFullyApplied (Variables|Default|GetBanner%20Icon%20Texture)))))
    (:then_2
      (if (Utilities|String|EqualExactly(String) (Item|GetPersistentName (CosmeticLoadout|GetItemDefinitionfromLoadout (CosmeticLoadout|GetCosmeticLoadout) "(TagName=\"LoadoutSlot.Character\")")) (Variables|Default|GetCharacter-Name))
        (bind _asmaterial_instance_dynamic_2 (Utilities|Casting|CastToMaterialInstanceDynamic (Rendering|Material|GetMaterial (Customization|GetSkeletalMeshforPartType (CustomCharacter|GetAssociatedPlayerPawn) "Body" :!class /Script/FortniteGame.FortPlayerPawn) 1)))
        (bind _returnvalue_3 (Rendering|Material|GetVectorParameterValue _asmaterial_instance_dynamic_2 "IonVial_CustomColor" :!class /Script/Engine.MaterialInstanceDynamic))
        (CallFunction|SetColors _returnvalue_3)))))

(event Custom|OnBannerFullyApplied
  (params (Icon :type Texture%202D%20Object%20Reference) (PrimaryColor :type Linear%20Color%20Structure)))

(fn setcolors
  (params (Color :type Linear%20Color%20Structure))
  (bind _returnvalue (Math|Conversions|ToVector(LinearColor) Color))
  (bind _returnvalue_2 (Math|Vector|RotateVector _returnvalue "25.000000,25.000000,25.000000"))
  (bind _returnvalue_3 (Math|Conversions|ToLinearColor(Vector) _returnvalue_2))
  (for _array_element (CustomItemWrap|GetWrappedMeshComponents)
    (for _array_element_2 (Rendering|Material|GetMaterials _array_element)
      (bind _asmaterial_instance_dynamic (Utilities|Casting|CastToMaterialInstanceDynamic _array_element_2))
      (Rendering|Material|SetVectorParameterValue _asmaterial_instance_dynamic "Emissive_A_ColorCW" Color :!class /Script/Engine.MaterialInstanceDynamic)
      (Rendering|Material|SetVectorParameterValue _asmaterial_instance_dynamic "Emissive_C_ColorCW" Color :!class /Script/Engine.MaterialInstanceDynamic)
      (Rendering|Material|SetVectorParameterValue _asmaterial_instance_dynamic "FDecal_BaseColor_Simple" Color :!class /Script/Engine.MaterialInstanceDynamic)
      (Rendering|Material|SetVectorParameterValue _asmaterial_instance_dynamic "Emissive_B_ColorCW" _returnvalue_3 :!class /Script/Engine.MaterialInstanceDynamic)
      (Rendering|Material|SetVectorParameterValue _asmaterial_instance_dynamic "Emissive_D_ColorCW" _returnvalue_3 :!class /Script/Engine.MaterialInstanceDynamic))))

(fn ConstructionScript)
