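; Reactive item wrap modifier for IronLilac that builds up emissive glow and flame brightness on the wrapped weapon as it fires.
; Pawn and weapon changes reset the step count and pick max steps and brightness arrays by clip size.
; Each shot advances the step, decayed by idle time, and writes brightness plus game-time stamps to wrap materials.
; At max step a timeline drives flames; a firing pause triggers a material offset and stops it.
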(event EventOnNewWeapons
  (params (WeaponsAsItems :type Array%20of%20Fort%20Item%20Object%20References))
  (Utilities|FlowControl|Delay 0.01)
  (Variables|Default|SetStepCount 0)
  (if (<= (FortWeapon|GetBulletsPerClip (Variables|/Script/FortniteGame.FortPawn|GetCurrentWeapon (CustomCharacter|GetAssociatedPlayerPawn))) 3)
    (Variables|Default|SetMaxSteps 5)
    (Variables|Default|SetBrightness%20value (Variables|Default|GetBrightness%20value_FE))
    (Variables|Default|SetBrightness%20Flames (Variables|Default|GetBrightness%20Flames_FE))
    (else
      (Variables|Default|SetMaxSteps 9)
      (Variables|Default|SetBrightness%20value (Variables|Default|GetBrightness%20value_PIE))
      (Variables|Default|SetBrightness%20Flames (Variables|Default|GetBrightness%20Flames_PIE)))))

(event EventOnNewPlayerWeaponType
  (params (NewType :type EFortItemType%20Enum :default "WorldItem") (OldType :type EFortItemType%20Enum :default "WorldItem"))
  (Utilities|FlowControl|Delay 0.01)
  (Variables|Default|SetStepCount 0)
  (if (<= (FortWeapon|GetBulletsPerClip (Variables|/Script/FortniteGame.FortPawn|GetCurrentWeapon (CustomCharacter|GetAssociatedPlayerPawn))) 3)
    (Variables|Default|SetMaxSteps 5)
    (Variables|Default|SetBrightness%20value (Variables|Default|GetBrightness%20value_FE))
    (Variables|Default|SetBrightness%20Flames (Variables|Default|GetBrightness%20Flames_FE))
    (else
      (Variables|Default|SetMaxSteps 9)
      (Variables|Default|SetBrightness%20value (Variables|Default|GetBrightness%20value_PIE))
      (Variables|Default|SetBrightness%20Flames (Variables|Default|GetBrightness%20Flames_PIE)))))

(event EventOnWeaponChanged
  (params (NewWeapon :type Fort%20Weapon%20Object%20Reference) (OldWeapon :type Fort%20Weapon%20Object%20Reference))
  (Utilities|FlowControl|Delay 0.01)
  (Variables|Default|SetStepCount 0)
  (if (<= (FortWeapon|GetBulletsPerClip (Variables|/Script/FortniteGame.FortPawn|GetCurrentWeapon (CustomCharacter|GetAssociatedPlayerPawn))) 3)
    (Variables|Default|SetMaxSteps 5)
    (Variables|Default|SetBrightness%20value (Variables|Default|GetBrightness%20value_FE))
    (Variables|Default|SetBrightness%20Flames (Variables|Default|GetBrightness%20Flames_FE))
    (else
      (Variables|Default|SetMaxSteps 9)
      (Variables|Default|SetBrightness%20value (Variables|Default|GetBrightness%20value_PIE))
      (Variables|Default|SetBrightness%20Flames (Variables|Default|GetBrightness%20Flames_PIE)))))

(event EventOnPawnSet
  (Utilities|FlowControl|Delay 0.01)
  (Variables|Default|SetStepCount 0)
  (if (<= (FortWeapon|GetBulletsPerClip (Variables|/Script/FortniteGame.FortPawn|GetCurrentWeapon (CustomCharacter|GetAssociatedPlayerPawn))) 3)
    (Variables|Default|SetMaxSteps 5)
    (Variables|Default|SetBrightness%20value (Variables|Default|GetBrightness%20value_FE))
    (Variables|Default|SetBrightness%20Flames (Variables|Default|GetBrightness%20Flames_FE))
    (else
      (Variables|Default|SetMaxSteps 9)
      (Variables|Default|SetBrightness%20value (Variables|Default|GetBrightness%20value_PIE))
      (Variables|Default|SetBrightness%20Flames (Variables|Default|GetBrightness%20Flames_PIE)))))

(event EventBeginPlay
  (Utilities|FlowControl|Delay 0.2)
  (if (not (CustomCosmeticModifier|IsinFrontEnd))
    (Variables|Default|SetStepCount 0)
    (if (not (<= (FortWeapon|GetBulletsPerClip (Variables|/Script/FortniteGame.FortPawn|GetCurrentWeapon (CustomCharacter|GetAssociatedPlayerPawn))) 3))
      (Variables|Default|SetMaxSteps 9)
      (Variables|Default|SetBrightness%20value (Variables|Default|GetBrightness%20value_PIE))
      (Variables|Default|SetBrightness%20Flames (Variables|Default|GetBrightness%20Flames_PIE))))
  (Variables|Default|SetMaxSteps 5)
  (Variables|Default|SetBrightness%20value (Variables|Default|GetBrightness%20value_FE))
  (Variables|Default|SetBrightness%20Flames (Variables|Default|GetBrightness%20Flames_FE)))

(event EventOnWeaponFired
  (bind _returnvalue (CustomItemWrap|GetWrappedMeshComponents))
  (bind _returnvalue_2 (CustomItemWrap|GetWrappedMeshComponents))
  (bind _returnvalue_3 (CustomItemWrap|GetWrappedMeshComponents))
  (Variables|Default|SetPause false)
  (bind _output_get (Variables|Default|SetStepCount (Math|Integer|Min(Integer) (+ (Math|Float|Truncate (Math|Float|Lerp (Math|Conversions|ToFloat(Integer) (Variables|Default|GetStepCount)) 0.0 (Math|Float|Clamp(Float) (/ (- (Utilities|Time|GetGameTimeinSeconds) (Variables|Default|Getwait_timer)) 1.0)))) 1) (Variables|Default|GetMaxSteps))))
  (Utilities|FlowControl|Sequence
    (:then_0
      (Rendering|Material|SetScalarParameterValueonMaterials _returnvalue_2 "TopS Emissive_01-Brightness-TBP" (Utilities|Time|GetGameTimeinSeconds))
      (Rendering|Material|SetScalarParameterValueonMaterials _returnvalue_2 "TopS Emissive_01_Brightness---Start" (Utilities|Array|Get(aref) (Variables|Default|GetBrightness%20value) _output_get))
      (Variables|Default|Setwait_timer (+ (Utilities|Time|GetGameTimeinSeconds) 5.5))
      (Development|PrintText (Utilities|Text|ToText(Integer) (Variables|Default|GetStepCount)) true true "(R=0.610496,G=0.026241,B=0.027321,A=1.000000)")
      (if (== (Variables|Default|GetStepCount) (Variables|Default|GetMaxSteps))
        (Rendering|Material|SetScalarParameterValueonMaterials _returnvalue_3 "TopS Emissive_02-Brightness-TBP" (Utilities|Time|GetGameTimeinSeconds))
        (bind (_direction _newtrack_0) (Timeline|Timeline Play)
          (:Update
            (Rendering|Material|SetScalarParameterValueonMaterials _returnvalue_3 "TopS Emissive_02_Brightness---Start" _newtrack_0)))
        (else
          (Rendering|Material|SetScalarParameterValueonMaterials _returnvalue "TopS Emissive_02-Brightness-TBP" (Utilities|Time|GetGameTimeinSeconds))
          (Rendering|Material|SetScalarParameterValueonMaterials _returnvalue "TopS Emissive_02_Brightness---Start" (Utilities|Array|Get(aref) (Variables|Default|GetBrightness%20Flames) _output_get)))))
    (:then_1
      (if (== (Variables|Default|GetStepCount) (Variables|Default|GetMaxSteps))
        (Utilities|FlowControl|Delay 0.5)
        (Variables|Default|SetPause true)
        (Utilities|FlowControl|Delay 0.3)
        (if (Variables|Default|GetPause)
          (Rendering|Material|SetScalarParameterValueonMaterials (CustomItemWrap|GetWrappedMeshComponents) "TopS Emissive_03_Offset----TBP" (Utilities|Time|GetGameTimeinSeconds))
          (Timeline|Timeline Stop))))))

(fn ConstructionScript)
