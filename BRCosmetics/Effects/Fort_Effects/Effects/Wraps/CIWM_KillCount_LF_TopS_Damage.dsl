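; Cosmetic item wrap modifier that drives a kill counter decal and reactive emissive flashes on wrapped weapons.
; Player pawn DamageDealt and AthenaKills stat watchers update the counter and stamp game time into material scalar parameters, with firing pulsing two other emissives.
; In the front end, fired shots simulate damage flashes and kills, wrapping the preview count.
; Weapon change events reapply the count after a short delay.
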(event EventBeginPlay
  (if (CustomCosmeticModifier|IsinFrontEnd)
    _))

(event EventOnWeaponFired
  (bind _returnvalue (Utilities|Time|GetGameTimeinSeconds))
  (bind _returnvalue_2 (CustomItemWrap|GetWrappedMeshComponents))
  (bind _kill_count (Variables|Default|GetKill%20Count))
  (if (CustomCosmeticModifier|IsinFrontEnd)
    (Utilities|FlowControl|Delay 0.001)
    (Utilities|FlowControl|Sequence
      (:then_0
        _)
      (:then_1
        (bind _output_get (Variables|Default|SetShotsFired (+ (Variables|Default|GetShotsFired) 1)))
        (if (>= _output_get 5)
          (if (== (Variables|Default|GetShotsFired) 7)
            (Variables|Default|SetShotsFired 0)
            (Variables|Default|SetKill%20Count (+ _kill_count 1))
            (CallFunction|OnDamage)
            (CallFunction|KillCounter)
            (if (== _kill_count 8)
              (Utilities|FlowControl|Delay 2.0)
              (Variables|Default|SetKill%20Count 0)
              (CallFunction|KillCounter))
            (else
              (CallFunction|OnDamage)))))))
  (Rendering|Material|SetScalarParameterValueonMaterials _returnvalue_2 "TopS Emissive_02-Brightness-TBP" _returnvalue)
  (Rendering|Material|SetScalarParameterValueonMaterials _returnvalue_2 "TopS Emissive_01-Brightness-TBP" _returnvalue))

(event EventOnPawnSet
  (bind _returnvalue (CustomItemWrap|GetWrappedMeshComponents))
  (bind _returnvalue_2 (Utilities|Time|GetGameTimeinSeconds))
  (Utilities|FlowControl|Sequence
    (:then_0
      (Cosmetics|WatchCosmeticStat (CustomCharacter|GetAssociatedPlayerPawn) "(TagName=\"StatManager.DamageDealt\")" 0 false
        (:OnStatChanged
          (Rendering|Material|SetScalarParameterValueonMaterials (CustomItemWrap|GetWrappedMeshComponents) "TopS Emissive_03-Brightness-TBP" (Utilities|Time|GetGameTimeinSeconds)))))
    (:then_1
      (bind (_statname _statvalue) (Cosmetics|WatchCosmeticStat (CustomCharacter|GetAssociatedPlayerPawn) "(TagName=\"StatManager.AthenaKills\")" 0 false)
        (:OnStatChanged
          (Variables|Default|SetKill%20Count _statvalue)
          (Rendering|Material|SetScalarParameterValueonMaterials _returnvalue "FDecalV2_Counter_Brightness_Time_BP" _returnvalue_2)
          (Rendering|Material|SetScalarParameterValueonMaterials _returnvalue "FDecalV2_Emissive_Brightness_Time_BP" _returnvalue_2)
          (Rendering|Material|SetScalarParameterValueonMaterials _returnvalue "FDecalV2_Counter__Number" (Math|Conversions|ToFloat(Integer) (Variables|Default|GetKill%20Count))))))))

(event EventOnNewWeapons
  (params (WeaponsAsItems :type Array%20of%20Fort%20Item%20Object%20References))
  (Utilities|FlowControl|Delay 0.2)
  (Rendering|Material|SetScalarParameterValueonMaterials (CustomItemWrap|GetWrappedMeshComponents) "FDecalV2_Counter__Number" (Math|Conversions|ToFloat(Integer) (Variables|Default|GetKill%20Count))))

(event EventOnNewPlayerWeaponType
  (params (NewType :type EFortItemType%20Enum :default "WorldItem") (OldType :type EFortItemType%20Enum :default "WorldItem"))
  (Utilities|FlowControl|Delay 0.2)
  (Rendering|Material|SetScalarParameterValueonMaterials (CustomItemWrap|GetWrappedMeshComponents) "FDecalV2_Counter__Number" (Math|Conversions|ToFloat(Integer) (Variables|Default|GetKill%20Count))))

(event EventOnWeaponChanged
  (params (NewWeapon :type Fort%20Weapon%20Object%20Reference) (OldWeapon :type Fort%20Weapon%20Object%20Reference))
  (Utilities|FlowControl|Delay 0.2)
  (Rendering|Material|SetScalarParameterValueonMaterials (CustomItemWrap|GetWrappedMeshComponents) "FDecalV2_Counter__Number" (Math|Conversions|ToFloat(Integer) (Variables|Default|GetKill%20Count))))

(event Custom|KillCounter
  (bind _returnvalue (CustomItemWrap|GetWrappedMeshComponents))
  (bind _returnvalue_2 (Utilities|Time|GetGameTimeinSeconds))
  (Rendering|Material|SetScalarParameterValueonMaterials _returnvalue "FDecalV2_Counter_Brightness_Time_BP" _returnvalue_2)
  (Rendering|Material|SetScalarParameterValueonMaterials _returnvalue "FDecalV2_Emissive_Brightness_Time_BP" _returnvalue_2)
  (Rendering|Material|SetScalarParameterValueonMaterials _returnvalue "FDecalV2_Counter__Number" (Math|Conversions|ToFloat(Integer) (Variables|Default|GetKill%20Count))))

(event Custom|OnDamage
  (Rendering|Material|SetScalarParameterValueonMaterials (CustomItemWrap|GetWrappedMeshComponents) "TopS Emissive_03-Brightness-TBP" (Utilities|Time|GetGameTimeinSeconds)))

(fn ConstructionScript)
