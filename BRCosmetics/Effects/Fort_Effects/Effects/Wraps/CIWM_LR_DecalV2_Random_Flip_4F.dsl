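; Custom item wrap Blueprint that drives a randomized four-frame flipbook decal on the wrapped weapon whenever it fires.
; After a tiny delay, RandomSet writes a normalized random frame value and the current game time to material parameters on the wrapped mesh components.
; The timestamp feeds a parameter named for decal opacity timing.
; The inclusive random range yields five values from 0 to 1, both endpoints included.
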(event EventOnWeaponFired
  (Utilities|FlowControl|Delay 0.001)
  (CallFunction|RandomSet "FDecalV2_FlipBook_Frame #" "FDecalaV2_Opacity_Time_BP_Z" 4))

(fn RandomSet
  (params (FlipBook%20Frame :type Name) (TimeBP :type Name) (FrameCount :type Integer))
  (bind _returnvalue (CustomItemWrap|GetWrappedMeshComponents))
  (Rendering|Material|SetScalarParameterValueonMaterials _returnvalue FlipBook%20Frame (/ (Math|Conversions|ToFloat(Integer) (Math|Random|RandomIntegerinRange 0 FrameCount)) FrameCount))
  (Rendering|Material|SetScalarParameterValueonMaterials _returnvalue TimeBP (Utilities|Time|GetGameTimeinSeconds)))

(fn ConstructionScript)
