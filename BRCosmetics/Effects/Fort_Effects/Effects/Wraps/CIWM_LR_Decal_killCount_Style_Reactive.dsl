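; Reactive item wrap Blueprint that shows the player's kill count as a decal counter on wrapped meshes.
; In the front end a timeline drives the count; on pawn set, it watches the AthenaKills cosmetic stat.
; Count changes write the number, clamped to two digits, and game time to material parameters for time-based brightness and stylized effects.
; Weapon-change events reapply only the number after a short delay.
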(event EventBeginPlay
  (if (CustomCosmeticModifier|IsinFrontEnd)
    (bind (_direction _newtrack_1) (Timeline|Timeline PlayFromStart)
      (:NewTrack_0
        (Variables|Default|SetKill_Count (Math|Float|Truncate _newtrack_1))
        (CallFunction|CounterUpdate)))))

(event Custom|Counter%20Update
  (bind _returnvalue (Utilities|Time|GetGameTimeinSeconds))
  (bind _returnvalue_2 (CustomItemWrap|GetWrappedMeshComponents))
  (Rendering|Material|SetScalarParameterValueonMaterials _returnvalue_2 "FDecalV2_Counter__Number" (Math|Conversions|ToFloat(Integer) (Math|Integer|Clamp(Integer) (Variables|Default|GetKill_Count) 0 99)))
  (Rendering|Material|SetScalarParameterValueonMaterials _returnvalue_2 "FDecalV2_Counter_Brightness_Time_BP" _returnvalue)
  (Rendering|Material|SetScalarParameterValueonMaterials _returnvalue_2 "EfStylise_TFBP-Ourline---" _returnvalue))

(event EventOnPawnSet
  (bind _returnvalue (Utilities|Time|GetGameTimeinSeconds))
  (bind _returnvalue_2 (CustomItemWrap|GetWrappedMeshComponents))
  (bind _asfort_player_pawn (Utilities|Casting|CastToFortPlayerPawn (CustomCharacter|GetAssociatedPlayerPawn)))
  (bind (_statname _statvalue) (Cosmetics|WatchCosmeticStat _asfort_player_pawn "(TagName=\"StatManager.AthenaKills\")" 0 false)
    (:OnStatChanged
      (Variables|Default|SetKill_Count _statvalue)
      (Rendering|Material|SetScalarParameterValueonMaterials _returnvalue_2 "FDecalV2_Counter__Number" (Math|Conversions|ToFloat(Integer) (Math|Integer|Clamp(Integer) (Variables|Default|GetKill_Count) 0 99)))
      (Rendering|Material|SetScalarParameterValueonMaterials _returnvalue_2 "FDecalV2_Counter_Brightness_Time_BP" _returnvalue)
      (Rendering|Material|SetScalarParameterValueonMaterials _returnvalue_2 "EfStylise_TFBP-Ourline---" _returnvalue))))

(event Custom|OnBannerFullyApplied
  (params (Icon :type Texture%202D%20Object%20Reference) (PrimaryColor :type Linear%20Color%20Structure)))

(event EventOnWeaponChanged
  (params (NewWeapon :type Fort%20Weapon%20Object%20Reference) (OldWeapon :type Fort%20Weapon%20Object%20Reference))
  ;[1 1 1] Updated kill count on new weapon
  (Utilities|FlowControl|Delay 0.2)
  (Rendering|Material|SetScalarParameterValueonMaterials (CustomItemWrap|GetWrappedMeshComponents) "FDecalV2_Counter__Number" (Math|Conversions|ToFloat(Integer) (Math|Integer|Clamp(Integer) (Variables|Default|GetKill_Count) 0 99))))

(event EventOnNewWeapons
  (params (WeaponsAsItems :type Array%20of%20Fort%20Item%20Object%20References))
  ;[1 1 1] Updated kill count on new weapon
  (Utilities|FlowControl|Delay 0.2)
  (Rendering|Material|SetScalarParameterValueonMaterials (CustomItemWrap|GetWrappedMeshComponents) "FDecalV2_Counter__Number" (Math|Conversions|ToFloat(Integer) (Math|Integer|Clamp(Integer) (Variables|Default|GetKill_Count) 0 99))))

(event EventOnNewPlayerWeaponType
  (params (NewType :type EFortItemType%20Enum :default "WorldItem") (OldType :type EFortItemType%20Enum :default "WorldItem"))
  ;[1 1 1] Updated kill count on new weapon
  (Utilities|FlowControl|Delay 0.2)
  (Rendering|Material|SetScalarParameterValueonMaterials (CustomItemWrap|GetWrappedMeshComponents) "FDecalV2_Counter__Number" (Math|Conversions|ToFloat(Integer) (Math|Integer|Clamp(Integer) (Variables|Default|GetKill_Count) 0 99))))

(fn ConstructionScript)
