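; Reactive weapon wrap logic for LateralHustle that plays a dragon reveal effect when the player's kill stat changes.
; The DragonReveal timeline drives starburst brightness, decal opacity and decal emissive parameters on the wrapped meshes' materials, with a Playing flag blocking overlapping plays.
; In frontend display, begin play and weapon changes start a timer that simulates a kill, re-armed after each reveal so the preview repeats.
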(event Custom|LockerCount
  (if (Reactive|Frontend|IsCosmeticObjectDisplayedinFrontendContext 0)
    (Utilities|Time|SetTimerbyFunctionName 0 "LockerKill" 3.0)))

(event EventOnWeaponChanged
  (params (NewWeapon :type Fort%20Weapon%20Object%20Reference) (OldWeapon :type Fort%20Weapon%20Object%20Reference))
  (if (Reactive|Frontend|IsCosmeticObjectDisplayedinFrontendContext 0)
    (Utilities|Time|SetTimerbyFunctionName 0 "LockerKill" 3.0)))

(event EventBeginPlay
  (if (Reactive|Frontend|IsCosmeticObjectDisplayedinFrontendContext 0)
    (Utilities|Time|SetTimerbyFunctionName 0 "LockerKill" 3.0)))

(event EventOnWeaponFired)

(event Custom|LockerKill
  (bind _returnvalue (CustomItemWrap|GetWrappedMeshComponents))
  (if (not (Variables|Default|GetPlaying?))
    (Variables|Default|SetPlaying? true)
    (bind (_direction _starburst _dragonreveal _dragonlight) (Timeline|DragonReveal PlayFromStart)
      (:Update
        (Rendering|Material|SetScalarParameterValueonMaterials _returnvalue "TopS Emissive_03_Brightness---End" _starburst)
        (Rendering|Material|SetScalarParameterValueonMaterials _returnvalue "FDecalV2_Opacity_Time__End" _dragonreveal)
        (Rendering|Material|SetScalarParameterValueonMaterials _returnvalue "FDecalV2_Emissive_Brightness_Time__End" _dragonlight))
      (:Finished
        (Variables|Default|SetPlaying? false)
        (CallFunction|LockerCount)))))

(event EventOnPawnSet
  (bind _returnvalue (CustomItemWrap|GetWrappedMeshComponents))
  (Utilities|FlowControl|Sequence
    (:then_0
      (Cosmetics|WatchCosmeticStat (CustomCharacter|GetAssociatedPlayerPawn) "(TagName=\"StatManager.AthenaKills\")" 0 false
        (:OnStatChanged
          (if (not (Variables|Default|GetPlaying?))
            (Variables|Default|SetPlaying? true)
            (Timeline|DragonReveal PlayFromStart)))))))

(fn ConstructionScript)
