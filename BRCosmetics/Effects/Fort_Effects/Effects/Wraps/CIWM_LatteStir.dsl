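; Weapon wrap Blueprint for LatteStir that animates wrapped-mesh materials in response to shots, damage and kills.
; Outside the front end, a shot counter periodically fires the main transition, which kills also trigger, while damage and some intermediate shots pulse a separate emissive layer.
; The transition writes game time into material parameters, alternates mask inversion, then flips base tint after a delay, with Pause_Kill blocking overlapping effects.
; RandomSet randomizes material placement but is never called here.
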(event EventBeginPlay
  (Utilities|FlowControl|Delay 0.2))

(event EventOnWeaponFired
  (bind _returnvalue (Utilities|Time|GetGameTimeinSeconds))
  (bind _returnvalue_2 (CustomItemWrap|GetWrappedMeshComponents))
  (bind _returnvalue_3 (CustomItemWrap|GetWrappedMeshComponents))
  (bind _returnvalue_4 (CustomItemWrap|GetWrappedMeshComponents))
  (if (CustomCosmeticModifier|IsinFrontEnd)
    (Utilities|FlowControl|Delay 0.001)
    (Utilities|FlowControl|Sequence
      (:then_1
        (bind _output_get (Variables|Default|SetShotsFired (+ (Variables|Default|GetShotsFired) 1)))
        (if (>= _output_get 5)
          (if (== (Variables|Default|GetShotsFired) 7)
            (Variables|Default|SetShotsFired 0)
            (if (not (Variables|Default|GetPause_Kill))
              (Variables|Default|SetPause_Kill true)
              (if (Variables|Default|GetPause_A)
                (Rendering|Material|SetScalarParameterValueonMaterials _returnvalue_2 "TopS Blend-InvertMsk   1=invert" 1.0)
                (Variables|Default|SetPause_A false)
                (else
                  (Rendering|Material|SetScalarParameterValueonMaterials _returnvalue_2 "TopS Blend-InvertMsk   1=invert" 0.0)
                  (Variables|Default|SetPause_A true)
                  (Rendering|Material|SetScalarParameterValueonMaterials _returnvalue_4 "FDecalV2_Emissive_Brightness_Time_BP" _returnvalue)))
              (Rendering|Material|SetScalarParameterValueonMaterials _returnvalue_4 "TopS Emissive_01_Offset----TBP" _returnvalue)
              (Rendering|Material|SetScalarParameterValueonMaterials _returnvalue_4 "TopS Emissive_01-Brightness-TBP" _returnvalue)
              (Rendering|Material|SetScalarParameterValueonMaterials _returnvalue_4 "TopS Emissive_02_Offset----TBP" _returnvalue)
              (Rendering|Material|SetScalarParameterValueonMaterials _returnvalue_4 "TopS Emissive_02-Brightness-TBP" _returnvalue)
              (Rendering|Material|SetScalarParameterValueonMaterials _returnvalue_4 "TopS_01_Blend_Offset-TBP" _returnvalue)
              (Utilities|FlowControl|Delay 1.0)
              (Variables|Default|SetPause_Kill false)
              (if (Variables|Default|GetPause_D)
                (Rendering|Material|SetColorParameterValueonMaterials _returnvalue_3 "EF_Anime_BaseColor_Tint" "(R=0.750000,G=0.750000,B=0.750000,A=1.000000)")
                (Variables|Default|SetPause_D false)
                (else
                  (Rendering|Material|SetColorParameterValueonMaterials _returnvalue_3 "EF_Anime_BaseColor_Tint" "(R=0.640000,G=0.600000,B=0.750000,A=1.000000)")
                  (Variables|Default|SetPause_D true))))
            (elif (Variables|Default|GetPause_Kill)
              _
              (else
                (Rendering|Material|SetScalarParameterValueonMaterials (CustomItemWrap|GetWrappedMeshComponents) "TopS Emissive_03-Brightness-TBP" (Utilities|Time|GetGameTimeinSeconds))))))))))

(event EventOnPawnSet
  (bind _returnvalue (Utilities|Time|GetGameTimeinSeconds))
  (bind _returnvalue_2 (CustomItemWrap|GetWrappedMeshComponents))
  (bind _returnvalue_3 (CustomItemWrap|GetWrappedMeshComponents))
  (bind _returnvalue_4 (CustomItemWrap|GetWrappedMeshComponents))
  (Utilities|FlowControl|Sequence
    (:then_0
      (Cosmetics|WatchCosmeticStat (CustomCharacter|GetAssociatedPlayerPawn) "(TagName=\"StatManager.DamageDealt\")" 0 false
        (:OnStatChanged
          (if (not (Variables|Default|GetPause_Kill))
            (Rendering|Material|SetScalarParameterValueonMaterials (CustomItemWrap|GetWrappedMeshComponents) "TopS Emissive_03-Brightness-TBP" (Utilities|Time|GetGameTimeinSeconds))))))
    (:then_1
      (Cosmetics|WatchCosmeticStat (CustomCharacter|GetAssociatedPlayerPawn) "(TagName=\"StatManager.AthenaKills\")" 0 false
        (:OnStatChanged
          (if (not (Variables|Default|GetPause_Kill))
            (Variables|Default|SetPause_Kill true)
            (if (Variables|Default|GetPause_A)
              (Rendering|Material|SetScalarParameterValueonMaterials _returnvalue_2 "TopS Blend-InvertMsk   1=invert" 1.0)
              (Variables|Default|SetPause_A false)
              (else
                (Rendering|Material|SetScalarParameterValueonMaterials _returnvalue_2 "TopS Blend-InvertMsk   1=invert" 0.0)
                (Variables|Default|SetPause_A true)
                (Rendering|Material|SetScalarParameterValueonMaterials _returnvalue_4 "FDecalV2_Emissive_Brightness_Time_BP" _returnvalue)))
            (Rendering|Material|SetScalarParameterValueonMaterials _returnvalue_4 "TopS Emissive_01_Offset----TBP" _returnvalue)
            (Rendering|Material|SetScalarParameterValueonMaterials _returnvalue_4 "TopS Emissive_01-Brightness-TBP" _returnvalue)
            (Rendering|Material|SetScalarParameterValueonMaterials _returnvalue_4 "TopS Emissive_02_Offset----TBP" _returnvalue)
            (Rendering|Material|SetScalarParameterValueonMaterials _returnvalue_4 "TopS Emissive_02-Brightness-TBP" _returnvalue)
            (Rendering|Material|SetScalarParameterValueonMaterials _returnvalue_4 "TopS_01_Blend_Offset-TBP" _returnvalue)
            (Utilities|FlowControl|Delay 1.0)
            (Variables|Default|SetPause_Kill false)
            (if (Variables|Default|GetPause_D)
              (Rendering|Material|SetColorParameterValueonMaterials _returnvalue_3 "EF_Anime_BaseColor_Tint" "(R=0.750000,G=0.750000,B=0.750000,A=1.000000)")
              (Variables|Default|SetPause_D false)
              (else
                (Rendering|Material|SetColorParameterValueonMaterials _returnvalue_3 "EF_Anime_BaseColor_Tint" "(R=0.640000,G=0.600000,B=0.750000,A=1.000000)")
                (Variables|Default|SetPause_D true)))))))))

(fn RandomSet
  (params (TimeBP_Flipbook%20Triger :type Name) (Offset :type Name) (Rotation :type Name) (X=min :type Float%20%28double-precision%29) (X=Max :type Float%20%28double-precision%29) (Y=Min :type Float%20%28double-precision%29) (Y=Max :type Float%20%28double-precision%29) (Scale=Min :type Float%20%28double-precision%29) (Scale=Max :type Float%20%28double-precision%29))
  (bind _returnvalue (CustomItemWrap|GetWrappedMeshComponents))
  (bind _stream (Variables|Default|GetStream))
  (bind _returnvalue_2 (Math|Random|RandomFloatinRangefromStream _stream X=min X=Max))
  (Rendering|Material|SetScalarParameterValueonMaterials _returnvalue TimeBP_Flipbook%20Triger (Utilities|Time|GetGameTimeinSeconds))
  (Rendering|Material|SetColorParameterValueonMaterials _returnvalue Offset (Math|Color|MakeColor _returnvalue_2 (Math|Random|RandomFloatinRangefromStream _stream Y=Min Y=Max) 0.0 (Math|Random|RandomFloatinRangefromStream _stream Scale=Min Scale=Max)))
  (Rendering|Material|SetScalarParameterValueonMaterials _returnvalue Rotation (Math|Random|RandomFloat))
  (Variables|Default|SetPosition _returnvalue_2))

(fn ConstructionScript)
