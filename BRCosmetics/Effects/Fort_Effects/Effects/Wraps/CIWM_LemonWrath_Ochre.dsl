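; Reactive weapon wrap Blueprint (LemonWrath Ochre) that drives emissive material effects on the wrapped meshes when damage is received or the weapon fires.
; Damage-stat watchers and the fire event stamp game time into material scalar parameters, zeroing the End values unless Killed is set.
; In the locker, a repeating timer advances a counter and plays a timeline that lerps emissive parameters up and down as a preview.
; Firing stops that timeline, and retriggerable delays clear the firing state.
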(event EventBeginPlay
  (bind _returnvalue (CustomItemWrap|GetWrappedMeshComponents))
  (bind _returnvalue_2 (Utilities|Time|GetGameTimeinSeconds))
  (if (Reactive|Frontend|IsCosmeticObjectDisplayedinFrontendContext 0)
    (Utilities|Time|SetTimerbyFunctionName 0 "LockerCycle" 2.5 true)
    (else
      (Cosmetics|WatchCosmeticStat (CustomCharacter|GetAssociatedPlayerPawn) "(TagName=\"StatManager.DamageReceived\")" 0 false
        (:OnStatChanged
          (Rendering|Material|SetScalarParameterValueonMaterials _returnvalue "TopS Emissive_01-Brightness-TBP" _returnvalue_2)
          (Rendering|Material|SetScalarParameterValueonMaterials _returnvalue "TopS Emissive_02-Brightness-TBP" _returnvalue_2)
          (Rendering|Material|SetScalarParameterValueonMaterials _returnvalue "TopS Emissive_03-Brightness-TBP" _returnvalue_2)
          (Rendering|Material|SetScalarParameterValueonMaterials _returnvalue "TopS Emissive_01_Brightness---End" _returnvalue_2)
          (Rendering|Material|SetScalarParameterValueonMaterials _returnvalue "TopS Emissive_02_Brightness---End" _returnvalue_2)
          (Rendering|Material|SetScalarParameterValueonMaterials _returnvalue "TopS Emissive_03_Brightness---End" _returnvalue_2)
          (if (not (Variables|Default|GetKilled))
            (Rendering|Material|SetScalarParameterValueonMaterials _returnvalue "TopS Emissive_01_Brightness---End" 0.0)
            (Rendering|Material|SetScalarParameterValueonMaterials _returnvalue "TopS Emissive_02_Brightness---End" 0.0)
            (Rendering|Material|SetScalarParameterValueonMaterials _returnvalue "TopS Emissive_03_Brightness---End" 0.0)))))))

(event EventOnPawnSet
  (bind _returnvalue (CustomItemWrap|GetWrappedMeshComponents))
  (bind _returnvalue_2 (Utilities|Time|GetGameTimeinSeconds))
  (Cosmetics|WatchCosmeticStat (CustomCharacter|GetAssociatedPlayerPawn) "(TagName=\"StatManager.DamageReceived\")" 0 false
    (:OnStatChanged
      (Rendering|Material|SetScalarParameterValueonMaterials _returnvalue "TopS Emissive_01-Brightness-TBP" _returnvalue_2)
      (Rendering|Material|SetScalarParameterValueonMaterials _returnvalue "TopS Emissive_02-Brightness-TBP" _returnvalue_2)
      (Rendering|Material|SetScalarParameterValueonMaterials _returnvalue "TopS Emissive_03-Brightness-TBP" _returnvalue_2)
      (Rendering|Material|SetScalarParameterValueonMaterials _returnvalue "TopS Emissive_01_Brightness---End" _returnvalue_2)
      (Rendering|Material|SetScalarParameterValueonMaterials _returnvalue "TopS Emissive_02_Brightness---End" _returnvalue_2)
      (Rendering|Material|SetScalarParameterValueonMaterials _returnvalue "TopS Emissive_03_Brightness---End" _returnvalue_2)
      (if (not (Variables|Default|GetKilled))
        (Rendering|Material|SetScalarParameterValueonMaterials _returnvalue "TopS Emissive_01_Brightness---End" 0.0)
        (Rendering|Material|SetScalarParameterValueonMaterials _returnvalue "TopS Emissive_02_Brightness---End" 0.0)
        (Rendering|Material|SetScalarParameterValueonMaterials _returnvalue "TopS Emissive_03_Brightness---End" 0.0)))))

(event EventOnWeaponFired
  (bind _returnvalue (Utilities|Time|GetGameTimeinSeconds))
  (bind _returnvalue_2 (CustomItemWrap|GetWrappedMeshComponents))
  (bind _returnvalue_3 (CustomItemWrap|GetWrappedMeshComponents))
  (bind _lockercounter (Variables|Default|GetLockerCounter))
  (Utilities|FlowControl|Sequence
    (:then_0
      (Rendering|Material|SetScalarParameterValueonMaterials _returnvalue_2 "FDecalV2_Emissive_Brightness_Time_BP" _returnvalue)
      (Rendering|Material|SetScalarParameterValueonMaterials _returnvalue_2 "FDecalV2_Emissive_Brightness_Time__End" _returnvalue)
      (if (not (Variables|Default|GetKilled))
        (Rendering|Material|SetScalarParameterValueonMaterials _returnvalue_2 "FDecalV2_Emissive_Brightness_Time__End" 0.0)
        (if (Reactive|Frontend|IsCosmeticObjectDisplayedinFrontendContext 0)
          (Utilities|FlowControl|RetriggerableDelay 1.75)
          (Variables|Default|SetIsFiring false))))
    (:then_1
      (if (CustomCosmeticModifier|IsinFrontEnd)
        (Variables|Default|SetIsFiring true)
        (Utilities|FlowControl|RetriggerableDelay 2.35)
        (Variables|Default|SetIsFiring false)
        (Variables|Default|SetLockerCounter 0)))
    (:then_2
      (if true
        (Variables|Default|SetIsFiring true)
        (bind (_direction _newtrack_0) (Timeline|Timeline Stop)
          (:Update
            (Rendering|Material|SetScalarParameterValueonMaterials _returnvalue_3 "TopS Emissive_01_Brightness---End" (Utilities|Select :Index _lockercounter :Option%200 0.0 :Option%201 (Math|Float|Lerp 0.0 0.3 _newtrack_0) :Option%202 (Math|Float|Lerp 0.3 0.0 _newtrack_0) :Option%203 0.0))
            (Rendering|Material|SetScalarParameterValueonMaterials _returnvalue_3 "TopS Emissive_02_Brightness---End" (Utilities|Select :Index _lockercounter :Option%200 0.0 :Option%201 (Math|Float|Lerp 0.0 0.3 _newtrack_0) :Option%202 (Math|Float|Lerp 0.3 0.0 _newtrack_0) :Option%203 0.0))
            (Rendering|Material|SetScalarParameterValueonMaterials _returnvalue_3 "TopS Emissive_03_Brightness---End" (Utilities|Select :Index _lockercounter :Option%200 0.0 :Option%201 (Math|Float|Lerp 0.0 0.3 _newtrack_0) :Option%202 (Math|Float|Lerp 0.3 0.0 _newtrack_0) :Option%203 0.0))))))))

(event Custom|LockerCycle
  (bind _returnvalue (CustomItemWrap|GetWrappedMeshComponents))
  (bind _lockercounter (Variables|Default|GetLockerCounter))
  (if (>= (Variables|Default|GetLockerCounter) 3)
    (Variables|Default|SetLockerCounter 0)
    (Timeline|Timeline PlayFromStart)
    (else
      (Math|Integer|IncrementInt (Variables|Default|GetLockerCounter)
        (:%20%20
          (Timeline|Timeline PlayFromStart))))))

(fn ConstructionScript)
