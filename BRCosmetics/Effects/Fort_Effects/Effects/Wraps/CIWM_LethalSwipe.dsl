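; Item wrap modifier Blueprint for the Lethal Swipe wrap that shifts material offsets when the weapon fires.
; Each shot plays a timeline from the start, and every update subtracts its track value from two stored offsets, with the second moving 1.5 times as fast.
; The offsets drive the Y component of the Em03 and Em02 offset vector parameters on the wrapped mesh components, accumulating across shots.
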(event EventBeginPlay)

(event EventOnPawnSet)

(event EventOnWeaponFired
  (bind (_direction _newtrack_0) (Timeline|Timeline PlayFromStart :NewTime 0.5)
    (:Update
      (bind _output_get (Variables|Default|SetOffset_A (- (Variables|Default|GetOffset_A) _newtrack_0)))
      (Rendering|Material|SetVectorParameterValueonMaterials (CustomItemWrap|GetWrappedMeshComponents) "TopS_Em03_Offset__Master" (Math|Vector|MakeVector 0.0 _output_get))
      (bind _output_get_2 (Variables|Default|SetOffset_B (- (Variables|Default|GetOffset_B) (* _newtrack_0 1.5))))
      (Rendering|Material|SetVectorParameterValueonMaterials (CustomItemWrap|GetWrappedMeshComponents) "TopS_Em02_Offset__Master" (Math|Vector|MakeVector 0.0 _output_get_2)))))

(fn ConstructionScript)
