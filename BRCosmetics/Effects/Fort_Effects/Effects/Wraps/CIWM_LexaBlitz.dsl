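; LexaBlitz item wrap logic that drives an emissive charge effect on wrapped meshes as the weapon fires.
; Each shot steps the Emissive_03 offset by a seventh, using start/end values and game-time stamps for material-side animation.
; At seven shots a pause flag sets, so further firing triggers Emissive_02, fades Emissive_03, then resets after a retriggerable delay.
; Below seven, a timer runs Dissolve to fade and reset; a debug print remains.
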(event EventOnWeaponFired
  (bind _returnvalue (CustomItemWrap|GetWrappedMeshComponents))
  (bind _returnvalue_2 (Utilities|Time|GetGameTimeinSeconds))
  (bind _returnvalue_3 (CustomItemWrap|GetWrappedMeshComponents))
  (Variables|EnviromentVarable|GetPause
    (:then
      (Rendering|Material|SetScalarParameterValueonMaterials _returnvalue_3 "TopS Emissive_02-Brightness-TBP" _returnvalue_2)
      (Utilities|FlowControl|Delay 0.7)
      (Rendering|Material|SetScalarParameterValueonMaterials _returnvalue_3 "TopS Emissive_03-Brightness-TBP" _returnvalue_2)
      (Rendering|Material|SetScalarParameterValueonMaterials _returnvalue_3 "TopS Emissive_03_Brightness---Start" 1.0)
      (Rendering|Material|SetScalarParameterValueonMaterials _returnvalue_3 "TopS Emissive_03_Brightness---End" 0.0)
      (Utilities|FlowControl|RetriggerableDelay 4.0)
      (Variables|EnviromentVarable|SetPause false)
      (Variables|Default|SetShootsCount 0)
      (Variables|Default|SetNewPosition%20 0.0))
    (:else
      (Variables|Default|SetOldPosition (Variables|Default|GetNewPosition%20))
      (bind _output_get (Variables|Default|SetShootsCount (+ (Variables|Default|GetShootsCount) 1)))
      (Development|PrintString (Utilities|String|ToString(Float) (* (/ _output_get 7) -1.0)))
      (bind _output_get_2 (Variables|Default|SetNewPosition%20 (* (/ _output_get 7) -1.0)))
      (Rendering|Material|SetScalarParameterValueonMaterials _returnvalue "TopS Emissive_03_Offset---End" _output_get_2)
      (Rendering|Material|SetScalarParameterValueonMaterials _returnvalue "TopS Emissive_03_Offset---Start" (Variables|Default|GetOldPosition))
      (Rendering|Material|SetScalarParameterValueonMaterials _returnvalue "TopS Emissive_03_Brightness---End" 1.0)
      (Rendering|Material|SetScalarParameterValueonMaterials _returnvalue "TopS Emissive_03_Offset----TBP" (Utilities|Time|GetGameTimeinSeconds))
      (if (>= (Variables|Default|GetShootsCount) 7)
        (Variables|EnviromentVarable|SetPause true)
        (else
          (Utilities|Time|SetTimerbyEvent (delegate Custom|Dissolve) 5.0 false true))))))

(event Custom|Dissolve
  (bind _returnvalue (CustomItemWrap|GetWrappedMeshComponents))
  (Variables|EnviromentVarable|SetPause true)
  (Rendering|Material|SetScalarParameterValueonMaterials _returnvalue "TopS Emissive_03-Brightness-TBP" (Utilities|Time|GetGameTimeinSeconds))
  (Rendering|Material|SetScalarParameterValueonMaterials _returnvalue "TopS Emissive_03_Brightness---Start" 1.0)
  (Rendering|Material|SetScalarParameterValueonMaterials _returnvalue "TopS Emissive_03_Brightness---End" 0.0)
  (Utilities|FlowControl|Delay 4.0)
  (Variables|EnviromentVarable|SetPause false)
  (Variables|Default|SetShootsCount 0)
  (Variables|Default|SetNewPosition%20 0.0))

(fn ConstructionScript)
