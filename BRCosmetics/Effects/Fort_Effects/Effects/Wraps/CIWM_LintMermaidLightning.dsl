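; Reactive item wrap logic for LintMermaidLightning that drives material effects on wrapped meshes when the weapon fires.
; Each shot stamps game time into scalar parameters for top-side, side-decal, liquid and reactive-decal layers, covering brightness, gradient, opacity and flipbook timing.
; It also randomizes an offset and restarts a looping timer calling Offset_Warp, which steps that offset into top-side vector parameters for each wrap target type.
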(event EventOnWeaponFired
  (bind _returnvalue (Utilities|Time|GetGameTimeinSeconds))
  (bind _returnvalue_2 (Utilities|Time|GetGameTimeinSeconds))
  (bind _returnvalue_3 (Utilities|Time|GetGameTimeinSeconds))
  (bind _returnvalue_4 (CustomItemWrap|GetWrappedMeshComponents))
  (bind _returnvalue_5 (CustomItemWrap|GetWrappedMeshComponents))
  (bind _returnvalue_6 (Utilities|Time|GetGameTimeinSeconds))
  (bind _returnvalue_7 (CustomItemWrap|GetWrappedMeshComponents))
  (bind _returnvalue_8 (CustomItemWrap|GetWrappedMeshComponents))
  (bind _returnvalue_9 (Utilities|Time|GetGameTimeinSeconds))
  (bind _returnvalue_10 (CustomItemWrap|GetWrappedMeshComponents))
  (bind _returnvalue_11 (Utilities|Time|GetGameTimeinSeconds))
  (bind _returnvalue_12 (CustomItemWrap|GetWrappedMeshComponents))
  (bind _returnvalue_13 (Utilities|Time|GetGameTimeinSeconds))
  (bind _returnvalue_14 (CustomItemWrap|GetWrappedMeshComponents))
  (bind _returnvalue_15 (Utilities|Time|GetGameTimeinSeconds))
  (bind _returnvalue_16 (CustomItemWrap|GetWrappedMeshComponents))
  (bind _output_get (Variables|Default|SetTimer))
  (Utilities|Time|ClearandInvalidateTimerbyHandle _output_get)
  (bind _returnvalue_17 (Utilities|Time|SetTimerbyFunctionName 0 "Offset_Warp" 0.1 true true))
  (Variables|Default|SetTimer _returnvalue_17)
  (Variables|Default|SetOffset (Math|Random|RandomFloatinRange 0.0 1.0))
  (Utilities|FlowControl|Sequence
    (:then_0
      ;[1 1 1] TopSide v1
      (Rendering|Material|SetScalarParameterValueonMaterials _returnvalue_16 "TS_Emissive01_Time_BP_Brightness" _returnvalue)
      (Rendering|Material|SetScalarParameterValueonMaterials _returnvalue_16 "TS_Emissive01_Time_BP_Grad_Offset" _returnvalue))
    (:then_1
      (Rendering|Material|SetScalarParameterValueonMaterials _returnvalue_16 "TS_Emissive02_Time_BP_Brightness" _returnvalue)
      (Rendering|Material|SetScalarParameterValueonMaterials _returnvalue_16 "TS_Emissive02_Time_BP_Grad_Offset" _returnvalue))
    (:then_2
      ;[1 1 1] TopsSide V2
      (Rendering|Material|SetScalarParameterValueonMaterials _returnvalue_5 "TopS Emissive_02-Brightness-TBP" _returnvalue_2)
      (Rendering|Material|SetScalarParameterValueonMaterials _returnvalue_5 "TopS Emissive_02_Offset----TBP" _returnvalue_2)
      (Rendering|Material|SetScalarParameterValueonMaterials _returnvalue_5 "TopS_01_Blend_Opacity-TBP" _returnvalue_2))
    (:then_3
      (Rendering|Material|SetScalarParameterValueonMaterials _returnvalue_5 "TopS Emissive_01-Brightness-TBP" _returnvalue_2)
      (Rendering|Material|SetScalarParameterValueonMaterials _returnvalue_5 "TopS Emissive_01_Offset----TBP" _returnvalue_2)
      (Rendering|Material|SetScalarParameterValueonMaterials _returnvalue_5 "TopS Emissive_03-Brightness-TBP" _returnvalue_2)
      (Rendering|Material|SetScalarParameterValueonMaterials _returnvalue_5 "TopS Emissive_03_Offset----TBP" _returnvalue_2))
    (:then_4
      ;[1 1 1] General
      (Rendering|Material|SetScalarParameterValueonMaterials _returnvalue_4 "TimeFrom-BP----------------" _returnvalue_3)
      (Rendering|Material|SetScalarParameterValueonMaterials _returnvalue_4 "FDecal_Emissive_Brightness_Time_BP" _returnvalue_3))
    (:then_5
      ;[1 1 1] Left Right v1
      (Rendering|Material|SetScalarParameterValueonMaterials _returnvalue_8 "FDecal_Emissive_Gradent_Time_BP" _returnvalue_9)
      (Rendering|Material|SetScalarParameterValueonMaterials _returnvalue_8 "FDecal_Opacity_Time_BP_Z" _returnvalue_9))
    (:then_6
      ;[1 1 1] Liquid
      (Rendering|Material|SetScalarParameterValueonMaterials (CustomItemWrap|GetWrappedMeshComponents) "Liquid_TimeFrom-BP" (Utilities|Time|GetGameTimeinSeconds)))
    (:then_7
      ; Left\Right V2
      (Rendering|Material|SetScalarParameterValueonMaterials _returnvalue_10 "FDecalV2_FlipBook_Time_FromBP" _returnvalue_11)
      (Rendering|Material|SetScalarParameterValueonMaterials _returnvalue_10 "FDecalaV2_Opacity_Time_BP_Z" _returnvalue_11)
      (Rendering|Material|SetScalarParameterValueonMaterials _returnvalue_10 "FDecalV2_Emissive_Brightness_Time_BP" _returnvalue_11)
      (Rendering|Material|SetScalarParameterValueonMaterials _returnvalue_10 "FDecalV2_Emissive_Gradent_Time_BP" _returnvalue_11)
      (Rendering|Material|SetScalarParameterValueonMaterials _returnvalue_10 "FDecalV2_Counter_Brightness_Time_BP" _returnvalue_11))
    (:then_8
      ;[1 1 1 0.5] Decal Reactive
      (Rendering|Material|SetScalarParameterValueonMaterials _returnvalue_12 "EDRec_TimeFrom_BP_DecalA" _returnvalue_13)
      (Rendering|Material|SetScalarParameterValueonMaterials _returnvalue_12 "EDRec_TimeFrom_BP_DecalB" _returnvalue_13)
      (Rendering|Material|SetScalarParameterValueonMaterials _returnvalue_12 "EDRec_TimeFrom_BP_DecalC" _returnvalue_13)
      (Rendering|Material|SetScalarParameterValueonMaterials _returnvalue_12 "EDRec_TimeFrom_BP_DecalD" _returnvalue_13)
      (Rendering|Material|SetScalarParameterValueonMaterials _returnvalue_14 "EDRec_Emissive_A_Brightness_Time_BP" _returnvalue_15)
      (Rendering|Material|SetScalarParameterValueonMaterials _returnvalue_14 "EDRec_Emissive_B_Brightness_Time_BP" _returnvalue_15)
      (Rendering|Material|SetScalarParameterValueonMaterials _returnvalue_14 "EDRec_Emissive_C_Brightness_Time_BP" _returnvalue_15)
      (Rendering|Material|SetScalarParameterValueonMaterials _returnvalue_14 "EDRec_Emissive_D_Brightness_Time_BP" _returnvalue_15)
      (Rendering|Material|SetScalarParameterValueonMaterials _returnvalue_7 "EDRec_Emissive_A_Gradent_Time_BP" _returnvalue_6)
      (Rendering|Material|SetScalarParameterValueonMaterials _returnvalue_7 "EDRec_Emissive_B_Gradent_Time_BP" _returnvalue_6)
      (Rendering|Material|SetScalarParameterValueonMaterials _returnvalue_7 "EDRec_Emissive_C_Gradent_Time_BP" _returnvalue_6)
      (Rendering|Material|SetScalarParameterValueonMaterials _returnvalue_7 "EDRec_Emissive_D_Gradent_Time_BP" _returnvalue_6))))

(event Custom|Offset_Warp
  (bind _returnvalue (CustomItemWrap|GetWrappedMeshComponents))
  (bind _output_get (Variables|Default|SetOffset (+ (Variables|Default|GetOffset) 0.1)))
  (Rendering|Material|SetVectorParameterValueonMaterials _returnvalue "TopS_02_Offset_Weapon" (Math|Conversions|ToVector(Float) _output_get))
  (Rendering|Material|SetVectorParameterValueonMaterials _returnvalue "TopS_02_Offset_Backpack" (Math|Conversions|ToVector(Float) _output_get))
  (Rendering|Material|SetVectorParameterValueonMaterials _returnvalue "TopS_02_Offset_Character" (Math|Conversions|ToVector(Float) _output_get))
  (Rendering|Material|SetVectorParameterValueonMaterials _returnvalue "TopS_02_Offset_Mech" (Math|Conversions|ToVector(Float) _output_get))
  (Rendering|Material|SetVectorParameterValueonMaterials _returnvalue "TopS_02_Offset_Pickaxe" (Math|Conversions|ToVector(Float) _output_get))
  (Rendering|Material|SetVectorParameterValueonMaterials _returnvalue "TopS_02_Offset_Vehicle" (Math|Conversions|ToVector(Float) _output_get))
  (Rendering|Material|SetVectorParameterValueonMaterials _returnvalue "TopS_02_Offset_VehicleSide" (Math|Conversions|ToVector(Float) _output_get)))

(fn ConstructionScript)
