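; Custom item wrap logic for a liquid weapon skin that drives surface distortion and emissive brightness from motion and firing.
; A looping timer, started on pawn set or weapon change, measures change in weapon up vector and pawn location.
; Motion raises strength, vector and brightness values via max, while firing sets them directly.
; Each update eases these toward resting values with InterpTo and writes them to the wrapped meshes' material parameters.
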(event EventOnWeaponFired
  (Variables|Default|SetStrength 0.25)
  (Variables|Default|SetBrighntness%20%20Value 1.5)
  (Variables|Default|SetVector_Strenght "0.500000,0.500000,0.000000"))

(event EventTick
  (params (DeltaSeconds :type Float%20%28single-precision%29)))

(event EventOnPawnSet
  (Utilities|Time|SetTimerbyEvent (delegate Custom|update) 0.05 true))

(event EventOnWeaponChanged
  (params (NewWeapon :type Fort%20Weapon%20Object%20Reference) (OldWeapon :type Fort%20Weapon%20Object%20Reference))
  (Utilities|Time|SetTimerbyEvent (delegate Custom|update) 0.05 true))

(event Custom|update
  (bind _returnvalue (CustomItemWrap|GetSingleWrappedMeshComponent))
  (bind _returnvalue_2 (Transformation|GetWorldRotation _returnvalue))
  (bind _returnvalue_3 (Math|Vector|GetUpVector _returnvalue_2))
  (bind _returnvalue_4 (Utilities|Time|GetWorldDeltaSeconds))
  (bind _returnvalue_5 (CustomItemWrap|GetWrappedMeshComponents))
  (bind _returnvalue_6 (CustomItemWrap|GetSingleWrappedMeshComponent))
  (bind _returnvalue_7 (CustomCharacter|GetAssociatedPlayerPawn))
  (bind _returnvalue_8 (Transformation|GetActorLocation _returnvalue_7))
  (bind _prev_location (Variables|Default|Getprev_location))
  (bind _returnvalue_9 (- _prev_location _returnvalue_8))
  (bind _previouse_up_vector (Variables|Default|GetPreviouse%20up%20vector))
  (bind _returnvalue_10 (- _previouse_up_vector _returnvalue_3))
  (bind _returnvalue_11 (Transformation|GetWorldRotation _returnvalue_6))
  (bind _returnvalue_12 (Math|Vector|VectorLength _returnvalue_10))
  (bind _returnvalue_13 (* _returnvalue_9 0.005))
  (bind _returnvalue_14 (Math|Vector|VectorLength _returnvalue_13))
  (bind _brighntness__value (Variables|Default|GetBrighntness%20%20Value))
  (Variables|Default|SetVector_Strenght (Math|Vector|MakeVector (Math|Float|Max(Float) (.x (Variables|Default|GetVector_Strenght)) (.x (Math|Vector|VectorGetAbs _returnvalue_10))) (Math|Float|Max(Float) (.y (Variables|Default|GetVector_Strenght)) (.y (Math|Vector|VectorGetAbs _returnvalue_10))) (Math|Float|Max(Float) (.z (Variables|Default|GetVector_Strenght)) (.z (Math|Vector|VectorGetAbs _returnvalue_10)))))
  (Variables|Default|SetStrength (* (Math|Float|Max(Float) (Math|Float|Max(Float) (Math|Float|Clamp(Float) _returnvalue_12 0.0 0.5) (Math|Float|Clamp(Float) _returnvalue_14 0.0 0.5)) (Variables|Default|GetStrength)) 1.0))
  (Variables|Default|SetBrighntness%20%20Value (Math|Float|Max(Float) (Math|Float|Max(Float) (Math|Float|Clamp(Float) _returnvalue_12) (Math|Float|Clamp(Float) _returnvalue_14)) (Variables|Default|GetBrighntness%20%20Value)))
  (Variables|Default|Setprev_location (Transformation|GetActorLocation _returnvalue_7))
  (Variables|Default|SetPreviouse%20rotation _returnvalue_11)
  (Variables|Default|SetPreviouse%20up%20vector (Math|Vector|GetUpVector _returnvalue_11))
  (Variables|Default|SetStrength (Math|Interpolation|FInterpTo (Variables|Default|GetStrength) 0.025 _returnvalue_4 2.0))
  (Variables|Default|SetVector_Strenght (Math|Interpolation|VInterpTo (Variables|Default|GetVector_Strenght) "0.500000,0.500000,0.000000" _returnvalue_4 2.0))
  (Variables|Default|SetBrighntness%20%20Value (Math|Interpolation|FInterpTo (Variables|Default|GetBrighntness%20%20Value) 0.0 _returnvalue_4 3.0))
  (Rendering|Material|SetScalarParameterValueonMaterials _returnvalue_5 "EFLiq_Surface_Distortion_Strenght" (Variables|Default|GetStrength))
  (Rendering|Material|SetVectorParameterValueonMaterials _returnvalue_5 "EFLiq_Surface_Distortion_Vector Strenght" (Variables|Default|GetVector_Strenght))
  (Rendering|Material|SetScalarParameterValueonMaterials _returnvalue_5 "TopS Emissive_03_Brightness---End" _brighntness__value)
  (Rendering|Material|SetScalarParameterValueonMaterials _returnvalue_5 "TopS Emissive_02_Brightness---End" _brighntness__value))

(fn ConstructionScript)
