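; Madam Moth item wrap Blueprint that swaps the wrap's look to match the progressive stage of a specific equipped character skin.
; On customization finished, wrap target set and pawn set, it compares the loadout character's name to Character-Name, then reads the Stage0/Stage1 variant index.
; That index selects the D or C texture set and emissive color, which Material writes to dynamic material instances on all wrapped meshes.
; WrapMAt, an array-indexed alternative without emissive, is never called here.
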(event EventOnCharacterCustomizationFinished
  (params (CharacterPartInfo :type Array%20of%20Applied%20Character%20Part%20Info%20Structures))
  (if (Utilities|String|EqualExactly(String) (Item|GetPersistentName (CosmeticLoadout|GetItemDefinitionfromLoadout (CosmeticLoadout|GetCosmeticLoadout) "(TagName=\"LoadoutSlot.Character\")")) (Variables|Default|GetCharacter-Name))
    (bind _index (CosmeticLoadout|GetVariantChannelActiveIndexforSlotActor "(TagName=\"LoadoutSlot.Character\")" (CustomCharacter|GetAssociatedPlayerPawn) "(TagName=\"Cosmetics.Variant.Channel.Progressive\")" (Utilities|Array|MakeArray "(TagName=\"Cosmetics.Variant.Property.Stage0\")" "(TagName=\"Cosmetics.Variant.Property.Stage1\")") false false))
    (switch int (Utilities|Array|Get(acopy) (Utilities|Array|MakeArray 0 1) _index)
      (:0
        (CallFunction|Material "/BRCosmetics/Weapons/WeaponSkins/Wraps/Textures/MadamMoth/T_Wrap_MadamMothD_BC.T_Wrap_MadamMothD_BC" "/BRCosmetics/Weapons/WeaponSkins/Wraps/Textures/MadamMoth/T_Wrap_MadamMothD_Msk.T_Wrap_MadamMothD_Msk" "/BRCosmetics/Weapons/WeaponSkins/Wraps/Textures/MadamMoth/T_Wrap_MadamMothD_N.T_Wrap_MadamMothD_N" "(R=0.076185,G=0.174647,B=0.149960,A=1.000000)" 0.55 0.5 0.7 0.0 "(R=0.044610,G=1.500000,B=1.137816,A=1.000000)"))
      (:1
        (CallFunction|Material "/BRCosmetics/Weapons/WeaponSkins/Wraps/Textures/MadamMoth/T_Wrap_MadamMothC_BC.T_Wrap_MadamMothC_BC" "/BRCosmetics/Weapons/WeaponSkins/Wraps/Textures/MadamMoth/T_Wrap_MadamMothC_Msk.T_Wrap_MadamMothC_Msk" "/BRCosmetics/Weapons/WeaponSkins/Wraps/Textures/MadamMoth/T_Wrap_MadamMothC_N.T_Wrap_MadamMothC_N" "(R=0.076185,G=0.174647,B=0.149960,A=1.000000)" 0.55 0.5 0.7 0.0 "(R=0.171171,G=2.000000,B=0.712815,A=1.000000)")))))

(event Custom|OnBannerFullyApplied
  (params (Icon :type Texture%202D%20Object%20Reference) (PrimaryColor :type Linear%20Color%20Structure)))

(event EventOnSetWrapTarget
  (if (Utilities|String|EqualExactly(String) (Item|GetPersistentName (CosmeticLoadout|GetItemDefinitionfromLoadout (CosmeticLoadout|GetCosmeticLoadout) "(TagName=\"LoadoutSlot.Character\")")) (Variables|Default|GetCharacter-Name))
    (bind _index (CosmeticLoadout|GetVariantChannelActiveIndexforSlotActor "(TagName=\"LoadoutSlot.Character\")" (CustomCharacter|GetAssociatedPlayerPawn) "(TagName=\"Cosmetics.Variant.Channel.Progressive\")" (Utilities|Array|MakeArray "(TagName=\"Cosmetics.Variant.Property.Stage0\")" "(TagName=\"Cosmetics.Variant.Property.Stage1\")") false false))
    (switch int (Utilities|Array|Get(acopy) (Utilities|Array|MakeArray 0 1) _index)
      (:0
        (CallFunction|Material "/BRCosmetics/Weapons/WeaponSkins/Wraps/Textures/MadamMoth/T_Wrap_MadamMothD_BC.T_Wrap_MadamMothD_BC" "/BRCosmetics/Weapons/WeaponSkins/Wraps/Textures/MadamMoth/T_Wrap_MadamMothD_Msk.T_Wrap_MadamMothD_Msk" "/BRCosmetics/Weapons/WeaponSkins/Wraps/Textures/MadamMoth/T_Wrap_MadamMothD_N.T_Wrap_MadamMothD_N" "(R=0.076185,G=0.174647,B=0.149960,A=1.000000)" 0.55 0.5 0.7 0.0 "(R=0.044610,G=1.500000,B=1.137816,A=1.000000)"))
      (:1
        (CallFunction|Material "/BRCosmetics/Weapons/WeaponSkins/Wraps/Textures/MadamMoth/T_Wrap_MadamMothC_BC.T_Wrap_MadamMothC_BC" "/BRCosmetics/Weapons/WeaponSkins/Wraps/Textures/MadamMoth/T_Wrap_MadamMothC_Msk.T_Wrap_MadamMothC_Msk" "/BRCosmetics/Weapons/WeaponSkins/Wraps/Textures/MadamMoth/T_Wrap_MadamMothC_N.T_Wrap_MadamMothC_N" "(R=0.076185,G=0.174647,B=0.149960,A=1.000000)" 0.55 0.5 0.7 0.0 "(R=0.171171,G=2.000000,B=0.712815,A=1.000000)")))))

(event EventOnPawnSet
  (if (Utilities|String|EqualExactly(String) (Item|GetPersistentName (CosmeticLoadout|GetItemDefinitionfromLoadout (CosmeticLoadout|GetCosmeticLoadout) "(TagName=\"LoadoutSlot.Character\")")) (Variables|Default|GetCharacter-Name))
    (bind _index (CosmeticLoadout|GetVariantChannelActiveIndexforSlotActor "(TagName=\"LoadoutSlot.Character\")" (CustomCharacter|GetAssociatedPlayerPawn) "(TagName=\"Cosmetics.Variant.Channel.Progressive\")" (Utilities|Array|MakeArray "(TagName=\"Cosmetics.Variant.Property.Stage0\")" "(TagName=\"Cosmetics.Variant.Property.Stage1\")") false false))
    (switch int (Utilities|Array|Get(acopy) (Utilities|Array|MakeArray 0 1) _index)
      (:0
        (CallFunction|Material "/BRCosmetics/Weapons/WeaponSkins/Wraps/Textures/MadamMoth/T_Wrap_MadamMothD_BC.T_Wrap_MadamMothD_BC" "/BRCosmetics/Weapons/WeaponSkins/Wraps/Textures/MadamMoth/T_Wrap_MadamMothD_Msk.T_Wrap_MadamMothD_Msk" "/BRCosmetics/Weapons/WeaponSkins/Wraps/Textures/MadamMoth/T_Wrap_MadamMothD_N.T_Wrap_MadamMothD_N" "(R=0.076185,G=0.174647,B=0.149960,A=1.000000)" 0.55 0.5 0.7 0.0 "(R=0.044610,G=1.500000,B=1.137816,A=1.000000)"))
      (:1
        (CallFunction|Material "/BRCosmetics/Weapons/WeaponSkins/Wraps/Textures/MadamMoth/T_Wrap_MadamMothC_BC.T_Wrap_MadamMothC_BC" "/BRCosmetics/Weapons/WeaponSkins/Wraps/Textures/MadamMoth/T_Wrap_MadamMothC_Msk.T_Wrap_MadamMothC_Msk" "/BRCosmetics/Weapons/WeaponSkins/Wraps/Textures/MadamMoth/T_Wrap_MadamMothC_N.T_Wrap_MadamMothC_N" "(R=0.076185,G=0.174647,B=0.149960,A=1.000000)" 0.55 0.5 0.7 0.0 "(R=0.171171,G=2.000000,B=0.712815,A=1.000000)")))))

(event EventOnPostResIn
  (|CallParentFunction))

(fn Material
  (params (Base_Color :type Texture%20Object%20Reference) (Mask :type Texture%20Object%20Reference) (Normal :type Texture%20Object%20Reference) (Accent%20BC :type Linear%20Color%20Structure) (Accent%20Spec :type Float%20%28single-precision%29) (Accent_Roughness :type Float%20%28single-precision%29) (Accent_Metal :type Float%20%28single-precision%29) (Accent%20Metal%20Lut%20ON :type Float%20%28single-precision%29) (Emissive%20Color :type Linear%20Color%20Structure))
  (for _array_element (CustomItemWrap|GetWrappedMeshComponents)
    (for _array_element_2 (Rendering|Material|GetMaterials _array_element)
      (bind _asmaterial_instance_dynamic (Utilities|Casting|CastToMaterialInstanceDynamic _array_element_2))
      (Rendering|Material|SetTextureParameterValue _asmaterial_instance_dynamic "TopS_01__Tex_BaseColor" Base_Color)
      (Rendering|Material|SetTextureParameterValue _asmaterial_instance_dynamic "TopS_01__Tex_Mask" Mask)
      (Rendering|Material|SetTextureParameterValue _asmaterial_instance_dynamic "TopS_01__Tex_Normal" Normal)
      (Rendering|Material|SetVectorParameterValue _asmaterial_instance_dynamic "Simple_Color" Accent%20BC :!class /Script/Engine.MaterialInstanceDynamic)
      (Rendering|Material|SetScalarParameterValue _asmaterial_instance_dynamic "Simple_Specular" Accent%20Spec :!class /Script/Engine.MaterialInstanceDynamic)
      (Rendering|Material|SetScalarParameterValue _asmaterial_instance_dynamic "Simple_Roughness" Accent_Roughness :!class /Script/Engine.MaterialInstanceDynamic)
      (Rendering|Material|SetScalarParameterValue _asmaterial_instance_dynamic "Simple_Metalness" Accent_Metal :!class /Script/Engine.MaterialInstanceDynamic)
      (Rendering|Material|SetScalarParameterValue _asmaterial_instance_dynamic "Metal_LUT_Msk_Accent   1=On  0=Off" Accent%20Metal%20Lut%20ON :!class /Script/Engine.MaterialInstanceDynamic)
      (Rendering|Material|SetVectorParameterValue _asmaterial_instance_dynamic "TopS Emissive_01_Color-" Emissive%20Color :!class /Script/Engine.MaterialInstanceDynamic))))

(fn ConstructionScript)

(fn WrapMAt
  (params (Stages :type Integer) (A_Set_BaseColor :type Texture%20Object%20Reference) (B_Set_BaseColor :type Texture%20Object%20Reference) (A_Set_Mask :type Texture%20Object%20Reference) (B_Set_Mask :type Texture%20Object%20Reference) (A_Set_Norm :type Texture%20Object%20Reference) (B_Set_Norm :type Texture%20Object%20Reference) (A_Acent_Color :type Linear%20Color%20Structure) (A_Acent_Spec :type Float%20%28double-precision%29) (A_Acent_Roughness :type Float%20%28double-precision%29) (A_Acent_Metal :type Float%20%28double-precision%29) (B_Acent_Color :type Linear%20Color%20Structure) (B_Acent_Spec :type Float%20%28double-precision%29) (B_Acent_Roughness :type Float%20%28double-precision%29) (B_Acent_Metal :type Float%20%28double-precision%29) (A_Lut_MSk_Acent :type Float%20%28single-precision%29) (B_Lut_MSk_Acent :type Float%20%28single-precision%29))
  (for _array_element (CustomItemWrap|GetWrappedMeshComponents)
    (for _array_element_2 (Rendering|Material|GetMaterials _array_element)
      (bind _asmaterial_instance_dynamic (Utilities|Casting|CastToMaterialInstanceDynamic _array_element_2))
      (Rendering|Material|SetTextureParameterValue _asmaterial_instance_dynamic "TopS_01__Tex_BaseColor" (Utilities|Array|Get(acopy) (Utilities|Array|MakeArray A_Set_BaseColor B_Set_BaseColor) Stages))
      (Rendering|Material|SetTextureParameterValue _asmaterial_instance_dynamic "TopS_01__Tex_Mask" (Utilities|Array|Get(acopy) (Utilities|Array|MakeArray A_Set_Mask B_Set_Mask) Stages))
      (Rendering|Material|SetTextureParameterValue _asmaterial_instance_dynamic "TopS_01__Tex_Normal" (Utilities|Array|Get(acopy) (Utilities|Array|MakeArray A_Set_Norm B_Set_Norm) Stages))
      (Rendering|Material|SetVectorParameterValue _asmaterial_instance_dynamic "Simple_Color" (Utilities|Array|Get(acopy) (Utilities|Array|MakeArray A_Acent_Color B_Acent_Color) Stages) :!class /Script/Engine.MaterialInstanceDynamic)
      (Rendering|Material|SetScalarParameterValue _asmaterial_instance_dynamic "Simple_Specular" (Utilities|Array|Get(acopy) (Utilities|Array|MakeArray A_Acent_Spec B_Acent_Spec) Stages) :!class /Script/Engine.MaterialInstanceDynamic)
      (Rendering|Material|SetScalarParameterValue _asmaterial_instance_dynamic "Simple_Roughness" (Utilities|Array|Get(acopy) (Utilities|Array|MakeArray A_Acent_Roughness B_Acent_Roughness) Stages) :!class /Script/Engine.MaterialInstanceDynamic)
      (Rendering|Material|SetScalarParameterValue _asmaterial_instance_dynamic "Simple_Metalness" (Utilities|Array|Get(acopy) (Utilities|Array|MakeArray A_Acent_Metal B_Acent_Metal) Stages) :!class /Script/Engine.MaterialInstanceDynamic)
      (Rendering|Material|SetScalarParameterValue _asmaterial_instance_dynamic "Metal_LUT_Msk_Accent   1=On  0=Off" (Utilities|Array|Get(acopy) (Utilities|Array|MakeArray A_Lut_MSk_Acent B_Lut_MSk_Acent) Stages) :!class /Script/Engine.MaterialInstanceDynamic))))
