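; Custom item wrap Blueprint for the Magic Meadow cosmetic, driving a kill-reactive, fire-reactive material on wrapped weapon meshes.
; Firing stamps game time into a material scalar.
; A watched Athena kills stat indexes a step array, clamped to 0–5, setting an offset scalar and stamping kill time.
; New weapons or weapon-type changes reapply the stored offset after a short delay, keeping the kill look across swaps.
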(event EventOnWeaponFired
  (Rendering|Material|SetScalarParameterValueonMaterials (CustomItemWrap|GetWrappedMeshComponents) "MMead_Time_BP" (Utilities|Time|GetGameTimeinSeconds)))

(event EventOnPawnSet
  (bind _returnvalue (CustomItemWrap|GetWrappedMeshComponents))
  (bind _asfort_player_pawn (Utilities|Casting|CastToFortPlayerPawn (CustomCharacter|GetAssociatedPlayerPawn)))
  (bind (_statname _statvalue) (Cosmetics|WatchCosmeticStat _asfort_player_pawn "(TagName=\"StatManager.AthenaKills\")")
    (:OnStatChanged
      (bind _output_get (Variables|Default|SetKill_Count _statvalue))
      (bind _output_get_2 (Variables|Default|SetCurrent%20kill%20count%20Step (Utilities|Array|Get(aref) (Variables|Default|GetKill%20Count%20steps%20) (Math|Integer|Clamp(Integer) _output_get 0 5))))
      (Rendering|Material|SetScalarParameterValueonMaterials _returnvalue "MMead_KillCount_Offset" _output_get_2)
      (Rendering|Material|SetScalarParameterValueonMaterials _returnvalue "MMead_Time_BP_Kill" (Utilities|Time|GetGameTimeinSeconds)))))

(event EventOnNewWeapons
  (params (WeaponsAsItems :type Array%20of%20Fort%20Item%20Object%20References))
  ;[1 1 1] Updated kill count on new weapon
  (Utilities|FlowControl|Delay 0.2)
  (Rendering|Material|SetScalarParameterValueonMaterials (CustomItemWrap|GetWrappedMeshComponents) "MMead_KillCount_Offset" (Variables|Default|GetCurrent%20kill%20count%20Step)))

(event EventOnNewPlayerWeaponType
  (params (NewType :type EFortItemType%20Enum :default "WorldItem") (OldType :type EFortItemType%20Enum :default "WorldItem"))
  ;[1 1 1] Updated kill count on new weapon
  (Utilities|FlowControl|Delay 0.2)
  (Rendering|Material|SetScalarParameterValueonMaterials (CustomItemWrap|GetWrappedMeshComponents) "MMead_KillCount_Offset" (Variables|Default|GetCurrent%20kill%20count%20Step)))

(fn ConstructionScript)
