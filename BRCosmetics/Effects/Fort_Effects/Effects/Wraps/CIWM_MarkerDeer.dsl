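; Reactive weapon wrap (MarkerDeer) that changes a flipbook decal frame on the wrapped meshes as the player's kill stat changes.
; Each stat change writes the next FrameNumber array entry into a scalar material parameter, wrapping IterationCount after index 5.
; Pawn setup, new weapons, weapon changes and BeginPlay reset to frame zero.
; In frontend display, a looping one-second timer previews the sequence, pausing briefly before restarting.
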(event EventOnPawnSet
  (bind _iterationcount (Variables|Default|GetIterationCount))
  (Utilities|FlowControl|Sequence
    (:then_0
      (Cosmetics|WatchCosmeticStat (CustomCharacter|GetAssociatedPlayerPawn) "(TagName=\"StatManager.AthenaKills\")" 0 false
        (:OnStatChanged
          (Rendering|Material|SetScalarParameterValueonMaterials (CustomItemWrap|GetWrappedMeshComponents) "FDecalV2_FlipBook_Frame #" (Utilities|Array|Get(acopy) (Variables|Default|GetFrameNumber) (Variables|Default|GetIterationCount)))
          (if (>= _iterationcount 5)
            (Variables|Default|SetIterationCount 0)
            (else
              (Math|Integer|IncrementInt _iterationcount))))))
    (:then_1
      (Variables|Default|SetIterationCount 0)
      (Rendering|Material|SetScalarParameterValueonMaterials (CustomItemWrap|GetWrappedMeshComponents) "FDecalV2_FlipBook_Frame #" 0.0))))

(event EventOnNewWeapons
  (params (WeaponsAsItems :type Array%20of%20Fort%20Item%20Object%20References))
  (Variables|Default|SetIterationCount 0)
  (Rendering|Material|SetScalarParameterValueonMaterials (CustomItemWrap|GetWrappedMeshComponents) "FDecalV2_FlipBook_Frame #" 0.0))

(event EventOnWeaponChanged
  (params (NewWeapon :type Fort%20Weapon%20Object%20Reference) (OldWeapon :type Fort%20Weapon%20Object%20Reference))
  (if (Reactive|Frontend|IsCosmeticObjectDisplayedinFrontendContext 0)
    (Variables|Default|SetIterationCount 0)
    (Rendering|Material|SetScalarParameterValueonMaterials (CustomItemWrap|GetWrappedMeshComponents) "FDecalV2_FlipBook_Frame #" 0.0)
    (bind _returnvalue (Utilities|Time|SetTimerbyEvent (delegate Custom|LockerTrigger) 1.0 true))
    (Variables|Default|SetFrontendTimer _returnvalue)
    (else
      (Variables|Default|SetIterationCount 0)
      (Rendering|Material|SetScalarParameterValueonMaterials (CustomItemWrap|GetWrappedMeshComponents) "FDecalV2_FlipBook_Frame #" 0.0))))

(event Custom|RestartLocker
  (bind _returnvalue (Utilities|Time|SetTimerbyEvent (delegate Custom|LockerTrigger) 1.0 true))
  (Variables|Default|SetFrontendTimer _returnvalue))

(event EventBeginPlay
  (if (Reactive|Frontend|IsCosmeticObjectDisplayedinFrontendContext 0)
    (Variables|Default|SetIterationCount 0)
    (Rendering|Material|SetScalarParameterValueonMaterials (CustomItemWrap|GetWrappedMeshComponents) "FDecalV2_FlipBook_Frame #" 0.0)
    (bind _returnvalue (Utilities|Time|SetTimerbyEvent (delegate Custom|LockerTrigger) 1.0 true))
    (Variables|Default|SetFrontendTimer _returnvalue)
    (else
      (Variables|Default|SetIterationCount 0)
      (Rendering|Material|SetScalarParameterValueonMaterials (CustomItemWrap|GetWrappedMeshComponents) "FDecalV2_FlipBook_Frame #" 0.0))))

(event Custom|LockerTrigger
  (bind _iterationcount (Variables|Default|GetIterationCount))
  (Rendering|Material|SetScalarParameterValueonMaterials (CustomItemWrap|GetWrappedMeshComponents) "FDecalV2_FlipBook_Frame #" (Utilities|Array|Get(acopy) (Variables|Default|GetFrameNumber) (Variables|Default|GetIterationCount)))
  (if (>= _iterationcount 5)
    (Utilities|Time|ClearandInvalidateTimerbyHandle (Variables|Default|GetFrontendTimer))
    (Utilities|FlowControl|Delay 2.0)
    (Variables|Default|SetIterationCount 0)
    (CallFunction|RestartLocker)
    (else
      (Math|Integer|IncrementInt _iterationcount))))

(fn ConstructionScript)
