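; Reactive weapon wrap Blueprint for Masako Reign that lights up the wrapped meshes' materials when the weapon fires.
; Each shot stamps game time into emissive parameters and plays a timeline tinting the decal emissive toward orange.
; The first shot also starts a timer-driven ramp of emissive brightness and UV warp.
; Retriggerable three-second delays reverse both effects once firing stops, while a frontend-only LockerCycle timer merely resets a counter.
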(event EventBeginPlay
  (if (Reactive|Frontend|IsCosmeticObjectDisplayedinFrontendContext 0)
    (Utilities|Time|SetTimerbyFunctionName 0 "LockerCycle" 3.0 true)))

(event EventOnWeaponFired
  (bind _returnvalue (CustomItemWrap|GetWrappedMeshComponents))
  (bind _returnvalue_2 (Utilities|Time|GetGameTimeinSeconds))
  (Utilities|FlowControl|Sequence
    (:then_0
      (Rendering|Material|SetScalarParameterValueonMaterials _returnvalue "TopS Emissive_02-Brightness-TBP" _returnvalue_2)
      (Rendering|Material|SetScalarParameterValueonMaterials _returnvalue "TopS Emissive_02_Brightness---End" _returnvalue_2)
      (if (not (Variables|Default|GetKilled))
        (Rendering|Material|SetScalarParameterValueonMaterials _returnvalue "TopS Emissive_02_Brightness---End" 0.0)))
    (:then_1
      (Variables|Default|SetTargetColor "(R=1.000000,G=0.360000,B=0.040000,A=20.000000)")
      (bind (_direction _alpha) (Timeline|Timeline Play)
        (:Update
          (Rendering|Material|SetVectorParameterValueonMaterials _returnvalue "FDecalV2-Emissive-Color--Alpha=Multplyer" (Math|Conversions|ToVector(LinearColor) (Math|Color|Lerp(LinearColor) (Variables|Default|GetBaseColor) (Variables|Default|GetTargetColor) _alpha))))))
    (:then_2
      (Utilities|FlowControl|RetriggerableDelay 3.0)
      (Components|Timeline|SetPlayRate (Variables|CIWM_MasakoReign|GetTimeline) 0.6)
      (Timeline|Timeline Reverse))
    (:then_3
      (if (not (Variables|Default|GetFullAutoWeapons))
        (Variables|Default|SetFullAutoWeapons true)
        (CallFunction|SineAnimation))
      (Utilities|FlowControl|RetriggerableDelay 3.0)
      (CallFunction|SineAnimOFF)
      (Variables|Default|SetFullAutoWeapons false))))

(event Custom|LockerCycle
  (if (>= (Variables|Default|GetLockerCounter) 3)
    (Variables|Default|SetLockerCounter 0)))

(event Custom|SineAnimation
  ; \
  (Variables|Default|SetReverse false)
  (Variables|Default|SetStartTime (Utilities|Time|GetGameTimeinSeconds))
  (bind _returnvalue (Utilities|Time|SetTimerbyEvent (delegate Custom|SineAnimUpdate) 0.02 true))
  (Variables|Default|SetTimerHandle _returnvalue))

(event Custom|SineAnimOFF
  ; \
  (Variables|Default|SetReverse true)
  (Variables|Default|SetStartTime (Utilities|Time|GetGameTimeinSeconds))
  (bind _returnvalue (Utilities|Time|SetTimerbyEvent (delegate Custom|SineAnimUpdate) 0.02 true))
  (Variables|Default|SetTimerHandle _returnvalue))

(event Custom|SineAnimUpdate
  (bind _returnvalue (Math|Float|MapRangeClamped (- (Utilities|Time|GetGameTimeinSeconds) (Variables|Default|GetStartTime)) 0.0 0.25 1.0 25.0))
  (bind _returnvalue_2 (Utilities|Time|GetGameTimeinSeconds))
  (bind _returnvalue_3 (Math|Float|MapRangeClamped (- _returnvalue_2 (Variables|Default|GetStartTime)) 0.0 1.25 25.0 1.0))
  (bind _returnvalue_4 (CustomItemWrap|GetWrappedMeshComponents))
  (bind _starttime (Variables|Default|GetStartTime))
  (bind _returnvalue_5 (- _returnvalue_2 _starttime))
  ; \
  (if (>= _returnvalue_5 5.0)
    (Utilities|Time|ClearandInvalidateTimerbyHandle (Variables|Default|GetTimerHandle))
    (if (Variables|Default|GetReverse)
      ;; Start
      (Rendering|Material|SetScalarParameterValueonMaterials _returnvalue_4 "EDRec_Emissive_A_CC_Brightness" 1.0)
      (Rendering|Material|SetScalarParameterValueonMaterials _returnvalue_4 "EDRec_Emissive_C_CC_Brightness" 1.0)
      (Rendering|Material|SetScalarParameterValueonMaterials _returnvalue_4 "EDRec_Emissive_D_CC_Brightness" 1.0)
      (Rendering|Material|SetScalarParameterValueonMaterials _returnvalue_4 "FDecalV2_UV_Warp-Strength" 0.005)
      (Rendering|Material|SetColorParameterValueonMaterials _returnvalue_4 "EDRec_WarpB_Msk_Decal_Strength" (0.0 0.0 0.0 0.0))
      (else
        ;; Arrive
        (Rendering|Material|SetScalarParameterValueonMaterials _returnvalue_4 "EDRec_Emissive_A_CC_Brightness" 25.0)
        (Rendering|Material|SetScalarParameterValueonMaterials _returnvalue_4 "EDRec_Emissive_C_CC_Brightness" 25.0)
        (Rendering|Material|SetScalarParameterValueonMaterials _returnvalue_4 "EDRec_Emissive_D_CC_Brightness" 25.0)
        (Rendering|Material|SetScalarParameterValueonMaterials _returnvalue_4 "FDecalV2_UV_Warp-Strength" 0.01)
        (Rendering|Material|SetColorParameterValueonMaterials _returnvalue_4 "EDRec_WarpB_Msk_Decal_Strength" (0.5 0.45 0.0 0.5))))
    (elif (Variables|Default|GetReverse)
      ;; FadeIn
      (Rendering|Material|SetScalarParameterValueonMaterials _returnvalue_4 "EDRec_Emissive_A_CC_Brightness" _returnvalue_3)
      (Rendering|Material|SetScalarParameterValueonMaterials _returnvalue_4 "EDRec_Emissive_B_Brightness_Time__End" _returnvalue_3)
      (Rendering|Material|SetScalarParameterValueonMaterials _returnvalue_4 "EDRec_Emissive_C_CC_Brightness" _returnvalue_3)
      (Rendering|Material|SetScalarParameterValueonMaterials _returnvalue_4 "EDRec_Emissive_D_CC_Brightness" _returnvalue_3)
      (Rendering|Material|SetScalarParameterValueonMaterials _returnvalue_4 "FDecalV2_UV_Warp-Strength" (Math|Float|MapRangeClamped _returnvalue_5 0.0 0.15 0.01 0.005))
      (Rendering|Material|SetColorParameterValueonMaterials _returnvalue_4 "EDRec_WarpB_Msk_Decal_Strength" (Math|Color|MakeColor (Math|Float|MapRangeClamped _returnvalue_5 0.0 0.75 0.5) (Math|Float|MapRangeClamped _returnvalue_5 0.0 0.75 0.45) (Math|Float|MapRangeClamped _returnvalue_5 0.0 0.75 0.5) (Math|Float|MapRangeClamped _returnvalue_5 0.0 0.75 0.5)))
      (else
        ;; FadeOut
        (Rendering|Material|SetScalarParameterValueonMaterials _returnvalue_4 "EDRec_Emissive_A_CC_Brightness" _returnvalue)
        (Rendering|Material|SetScalarParameterValueonMaterials _returnvalue_4 "EDRec_Emissive_B_Brightness_Time__End" _returnvalue)
        (Rendering|Material|SetScalarParameterValueonMaterials _returnvalue_4 "EDRec_Emissive_C_CC_Brightness" _returnvalue)
        (Rendering|Material|SetScalarParameterValueonMaterials _returnvalue_4 "EDRec_Emissive_D_CC_Brightness" _returnvalue)
        (Rendering|Material|SetScalarParameterValueonMaterials _returnvalue_4 "FDecalV2_UV_Warp-Strength" (Math|Float|MapRangeClamped _returnvalue_5 0.0 0.15 0.005 0.01))
        (Rendering|Material|SetColorParameterValueonMaterials _returnvalue_4 "EDRec_WarpB_Msk_Decal_Strength" (Math|Color|MakeColor (Math|Float|MapRangeClamped _returnvalue_5 0.0 0.15 0.0 0.5) (Math|Float|MapRangeClamped _returnvalue_5 0.0 0.15 0.0 0.45) (Math|Float|MapRangeClamped _returnvalue_5 0.0 0.15 0.0 0.5) (Math|Float|MapRangeClamped _returnvalue_5 0.0 0.15 0.0 0.5)))))))

(fn ConstructionScript)
