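; Meteor Woman item wrap Blueprint that flashes the emissive layer on wrapped meshes when the weapon fires or the player takes damage.
; Firing sets the emissive color to Fire_Color and stamps game time into a brightness parameter for material-side timing.
; A DamageReceived cosmetic stat watch, registered on pawn set, does the same with Damage_Color and time offset by 0.25s.
; Each write targets two parameter naming sets; weapon changes black out the color after a short delay.
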(event EventOnWeaponFired
  (bind _returnvalue (CustomItemWrap|GetWrappedMeshComponents))
  (Utilities|FlowControl|Sequence
    (:then_0
      (Rendering|Material|SetVectorParameterValueonMaterials _returnvalue "TopS Emissive_01_Color-" (Math|Conversions|ToVector(LinearColor) (Variables|Default|GetFire_Color)))
      (Rendering|Material|SetScalarParameterValueonMaterials _returnvalue "TopS Emissive_01-Brightness-TBP" (Utilities|Time|GetGameTimeinSeconds)))
    (:then_1
      (Rendering|Material|SetVectorParameterValueonMaterials _returnvalue "TS_Emissive01_Color" (Math|Conversions|ToVector(LinearColor) (Variables|Default|GetFire_Color)))
      (Rendering|Material|SetScalarParameterValueonMaterials _returnvalue "TS_Emissive01_Time_BP_Brightness" (Utilities|Time|GetGameTimeinSeconds)))))

(event EventOnPawnSet
  (bind _returnvalue (CustomItemWrap|GetWrappedMeshComponents))
  (Utilities|FlowControl|Sequence
    (:then_0
      (bind _asfort_player_pawn (Utilities|Casting|CastToFortPlayerPawn (CustomCharacter|GetAssociatedPlayerPawn)))
      (Cosmetics|WatchCosmeticStat _asfort_player_pawn "(TagName=\"StatManager.DamageReceived\")"
        (:OnStatChanged
          (Utilities|FlowControl|Sequence
            (:then_0
              (Rendering|Material|SetScalarParameterValueonMaterials _returnvalue "TopS Emissive_01-Brightness-TBP" (+ (Utilities|Time|GetGameTimeinSeconds) 0.25))
              (Rendering|Material|SetVectorParameterValueonMaterials _returnvalue "TopS Emissive_01_Color-" (Math|Conversions|ToVector(LinearColor) (Variables|Default|GetDamage_Color))))
            (:then_1
              (Rendering|Material|SetScalarParameterValueonMaterials _returnvalue "TS_Emissive01_Time_BP_Brightness" (+ (Utilities|Time|GetGameTimeinSeconds) 0.25))
              (Rendering|Material|SetVectorParameterValueonMaterials _returnvalue "TS_Emissive01_Color" (Math|Conversions|ToVector(LinearColor) (Variables|Default|GetDamage_Color))))))))))

(event EventOnNewWeapons
  (params (WeaponsAsItems :type Array%20of%20Fort%20Item%20Object%20References))
  (bind _returnvalue (CustomItemWrap|GetWrappedMeshComponents))
  (Utilities|FlowControl|Delay 0.2)
  (Rendering|Material|SetVectorParameterValueonMaterials _returnvalue "TS_Emissive01_Color" "0, 0, 0")
  (Rendering|Material|SetVectorParameterValueonMaterials _returnvalue "TopS Emissive_01_Color-" "0, 0, 0"))

(event EventOnNewPlayerWeaponType
  (params (NewType :type EFortItemType%20Enum :default "WorldItem") (OldType :type EFortItemType%20Enum :default "WorldItem"))
  (bind _returnvalue (CustomItemWrap|GetWrappedMeshComponents))
  (Utilities|FlowControl|Delay 0.2)
  (Rendering|Material|SetVectorParameterValueonMaterials _returnvalue "TS_Emissive01_Color" "0, 0, 0")
  (Rendering|Material|SetVectorParameterValueonMaterials _returnvalue "TopS Emissive_01_Color-" "0, 0, 0"))

(fn ConstructionScript)
