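; MistMylar weapon wrap Blueprint that drives an ammo-reactive "crack level" on the wrapped mesh materials.
; Spawn, pawn and weapon-change events cache clip data after a short delay and bind the reload event.
; Firing maps spent ammo, or six-step shot counters for clips under two, to blend and emissive offsets, passing game time as a brightness timestamp.
; After reload, or a clipless counter overflow, a timeline lerps the offsets back.
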(event EventBeginPlay
  (bind _returnvalue (CustomCharacter|GetAssociatedPlayerPawn))
  (bind _currentweapon (Variables|/Script/FortniteGame.FortPawn|GetCurrentWeapon _returnvalue))
  (Utilities|FlowControl|Delay 0.1)
  (bind _currentweapon_2 (Variables|/Script/FortniteGame.FortPawn|GetCurrentWeapon _returnvalue)
    (:then
      (EventDispatchers|BindEventtoOnReloadStarted _currentweapon_2 (delegate Custom|ReloadStarted))
      (Variables|Default|Setper%20Clip (FortWeapon|GetBulletsPerClip _currentweapon))
      (Variables|Default|SetCurent%20GameTime -99999999999999.0)
      (Variables|Default|SetAmmo%20Count (FortWeapon|GetMagazineAmmoCount _currentweapon))
      (Variables|Default|SetAmmo-%200%20 (Variables|Default|GetAmmo-%200%20))
      (Variables|Default|SetAmmo-1-2 (Variables|Default|GetAmmo-1-2))
      (CallFunction|CrackLevel))
    (:else)))

(event EventOnWeaponChanged
  (params (NewWeapon :type Fort%20Weapon%20Object%20Reference) (OldWeapon :type Fort%20Weapon%20Object%20Reference))
  (bind _returnvalue (CustomCharacter|GetAssociatedPlayerPawn))
  (bind _currentweapon (Variables|/Script/FortniteGame.FortPawn|GetCurrentWeapon _returnvalue))
  (Utilities|FlowControl|Delay 0.01)
  (bind _currentweapon_2 (Variables|/Script/FortniteGame.FortPawn|GetCurrentWeapon _returnvalue)
    (:then
      (EventDispatchers|BindEventtoOnReloadStarted _currentweapon_2 (delegate Custom|ReloadStarted))
      (Variables|Default|Setper%20Clip (FortWeapon|GetBulletsPerClip _currentweapon))
      (Variables|Default|SetCurent%20GameTime -99999999999999.0)
      (Variables|Default|SetAmmo%20Count (FortWeapon|GetMagazineAmmoCount _currentweapon))
      (Variables|Default|SetAmmo-%200%20 (Variables|Default|GetAmmo-%200%20))
      (Variables|Default|SetAmmo-1-2 (Variables|Default|GetAmmo-1-2))
      (CallFunction|CrackLevel))
    (:else)))

(event EventOnPawnSet
  (bind _returnvalue (CustomCharacter|GetAssociatedPlayerPawn))
  (bind _currentweapon (Variables|/Script/FortniteGame.FortPawn|GetCurrentWeapon _returnvalue))
  (Utilities|FlowControl|Delay 0.01)
  (bind _currentweapon_2 (Variables|/Script/FortniteGame.FortPawn|GetCurrentWeapon _returnvalue)
    (:then
      (EventDispatchers|BindEventtoOnReloadStarted _currentweapon_2 (delegate Custom|ReloadStarted))
      (Variables|Default|Setper%20Clip (FortWeapon|GetBulletsPerClip _currentweapon))
      (Variables|Default|SetCurent%20GameTime -99999999999999.0)
      (Variables|Default|SetAmmo%20Count (FortWeapon|GetMagazineAmmoCount _currentweapon))
      (Variables|Default|SetAmmo-%200%20 (Variables|Default|GetAmmo-%200%20))
      (Variables|Default|SetAmmo-1-2 (Variables|Default|GetAmmo-1-2))
      (CallFunction|CrackLevel))
    (:else)))

(event EventOnWeaponFired
  (bind _per_clip (Variables|Default|Getper%20Clip))
  (bind _output_get (Variables|Default|SetAmmo%20Count (FortWeapon|GetMagazineAmmoCount (Variables|/Script/FortniteGame.FortPawn|GetCurrentWeapon (CustomCharacter|GetAssociatedPlayerPawn)))))
  (Utilities|FlowControl|Delay 0.001)
  (Variables|Default|SetAmmo%20Count (Math|Integer|Clamp(Integer) (- _output_get 1) 0 (Variables|Default|Getper%20Clip)))
  (Variables|Default|SetCurent%20GameTime (Utilities|Time|GetGameTimeinSeconds))
  (Utilities|FlowControl|Sequence
    (:then_0
      (if (== _per_clip 0)
        (Variables|Default|SetAmmo-%200%20 (+ (Variables|Default|GetAmmo-%200%20) 1.0))))
    (:then_1
      (if (< _per_clip 2)
        (Variables|Default|SetAmmo-1-2 (+ (Variables|Default|GetAmmo-1-2) 1.0))))
    (:then_2
      (CallFunction|CrackLevel))))

(event EventOnNewWeapons
  (params (WeaponsAsItems :type Array%20of%20Fort%20Item%20Object%20References))
  (bind _returnvalue (CustomCharacter|GetAssociatedPlayerPawn))
  (bind _currentweapon (Variables|/Script/FortniteGame.FortPawn|GetCurrentWeapon _returnvalue))
  (Utilities|FlowControl|Delay 0.01)
  (Variables|Default|SetAmmo-%200%20 0.0)
  (Variables|Default|SetAmmo-1-2 0.0)
  (bind _currentweapon_2 (Variables|/Script/FortniteGame.FortPawn|GetCurrentWeapon _returnvalue)
    (:then
      (EventDispatchers|BindEventtoOnReloadStarted _currentweapon_2 (delegate Custom|ReloadStarted))
      (Variables|Default|Setper%20Clip (FortWeapon|GetBulletsPerClip _currentweapon))
      (Variables|Default|SetCurent%20GameTime -99999999999999.0)
      (Variables|Default|SetAmmo%20Count (FortWeapon|GetMagazineAmmoCount _currentweapon))
      (Variables|Default|SetAmmo-%200%20 (Variables|Default|GetAmmo-%200%20))
      (Variables|Default|SetAmmo-1-2 (Variables|Default|GetAmmo-1-2))
      (CallFunction|CrackLevel))
    (:else)))

(event EventOnNewPlayerWeaponType
  (params (NewType :type EFortItemType%20Enum :default "WorldItem") (OldType :type EFortItemType%20Enum :default "WorldItem"))
  (bind _returnvalue (CustomCharacter|GetAssociatedPlayerPawn))
  (bind _currentweapon (Variables|/Script/FortniteGame.FortPawn|GetCurrentWeapon _returnvalue))
  (Utilities|FlowControl|Delay 0.01)
  (Variables|Default|SetAmmo-%200%20 0.0)
  (Variables|Default|SetAmmo-1-2 0.0)
  (bind _currentweapon_2 (Variables|/Script/FortniteGame.FortPawn|GetCurrentWeapon _returnvalue)
    (:then
      (EventDispatchers|BindEventtoOnReloadStarted _currentweapon_2 (delegate Custom|ReloadStarted))
      (Variables|Default|Setper%20Clip (FortWeapon|GetBulletsPerClip _currentweapon))
      (Variables|Default|SetCurent%20GameTime -99999999999999.0)
      (Variables|Default|SetAmmo%20Count (FortWeapon|GetMagazineAmmoCount _currentweapon))
      (Variables|Default|SetAmmo-%200%20 (Variables|Default|GetAmmo-%200%20))
      (Variables|Default|SetAmmo-1-2 (Variables|Default|GetAmmo-1-2))
      (CallFunction|CrackLevel))
    (:else)))

(event Custom|Reload%20weapon
  (bind _returnvalue (CustomItemWrap|GetWrappedMeshComponents))
  (bind _position_data (Variables|Default|GetPosition_Data))
  (bind (_direction _newtrack_0 _newtrack_1) (Timeline|Timeline PlayFromStart)
    (:Update
      (Rendering|Material|SetScalarParameterValueonMaterials _returnvalue "TopS_01_Blend_Offset---End" (Math|Float|Lerp _position_data -0.52 _newtrack_1))
      (Rendering|Material|SetScalarParameterValueonMaterials _returnvalue "TopS Emissive_03_Offset---End" (Math|Float|Lerp _position_data -0.52 _newtrack_1))
      (Rendering|Material|SetScalarParameterValueonMaterials _returnvalue "TopS Emissive_03-Brightness-TBP" (Utilities|Time|GetGameTimeinSeconds))
      (Variables|Default|SetAmmo-%200%20 0.0))))

(event Custom|ReloadStarted
  (params (ReloadTime :type Float%20%28single-precision%29) (ReloadType :type EFortWeaponReloadType%20Enum :default "ReloadWholeClip"))
  (bind _returnvalue (CustomItemWrap|GetWrappedMeshComponents))
  (bind _position_data (Variables|Default|GetPosition_Data))
  (if (< (Variables|Default|Getper%20Clip) 2)
    (if (>= (Variables|Default|GetAmmo-1-2) 6.0)
      (Variables|Default|SetAmmo-1-2 0.0))
    (else
      (Variables|Default|SetAmmo%20Count (FortWeapon|GetMagazineAmmoCount (Variables|/Script/FortniteGame.FortPawn|GetCurrentWeapon (CustomCharacter|GetAssociatedPlayerPawn))))))
  (Utilities|FlowControl|Delay ReloadTime)
  (Timeline|Timeline PlayFromStart))

(event Custom|Crack%20level%20
  (bind _returnvalue (CustomItemWrap|GetWrappedMeshComponents))
  (bind _per_clip (Variables|Default|Getper%20Clip))
  (bind _returnvalue_2 (Math|Conversions|ToFloat(Integer) _per_clip))
  (bind _ammo_count (Variables|Default|GetAmmo%20Count))
  (bind _returnvalue_3 (Math|Conversions|ToFloat(Integer) _ammo_count))
  (bind _ammo_1_2 (Variables|Default|GetAmmo-1-2))
  (bind _returnvalue_4 (/ _ammo_1_2 6.0))
  (bind _returnvalue_5 (Math|Float|Lerp -0.52 0.52 _returnvalue_4))
  (bind _returnvalue_6 (/ _returnvalue_3 _returnvalue_2))
  (bind _returnvalue_7 (Math|Float|Lerp 0.52 -0.52 _returnvalue_6))
  (bind _returnvalue_8 (CustomItemWrap|GetWrappedMeshComponents))
  (bind _ammo__0 (Variables|Default|GetAmmo-%200%20))
  (bind _returnvalue_9 (/ _ammo__0 6.0))
  (bind _returnvalue_10 (Math|Float|Lerp -0.52 0.52 _returnvalue_9))
  (bind _returnvalue_11 (CustomItemWrap|GetWrappedMeshComponents))
  (if (< (Variables|Default|Getper%20Clip) 2)
    (if (== (Variables|Default|Getper%20Clip) 0)
      (Rendering|Material|SetScalarParameterValueonMaterials _returnvalue_11 "TopS_01_Blend_Offset---End" _returnvalue_10)
      (Rendering|Material|SetScalarParameterValueonMaterials _returnvalue_11 "TopS Emissive_03_Offset---End" _returnvalue_10)
      (Rendering|Material|SetScalarParameterValueonMaterials _returnvalue_11 "TopS Emissive_03-Brightness-TBP" (Variables|Default|GetCurent%20GameTime))
      (Variables|Default|SetPosition_Data _returnvalue_10)
      (if (== (Variables|Default|GetAmmo-%200%20) 7.0)
        (Utilities|FlowControl|Delay 0.4)
        (CallFunction|ReloadWeapon))
      (else
        (Rendering|Material|SetScalarParameterValueonMaterials _returnvalue "TopS_01_Blend_Offset---End" _returnvalue_5)
        (Rendering|Material|SetScalarParameterValueonMaterials _returnvalue "TopS Emissive_03_Offset---End" _returnvalue_5)
        (Rendering|Material|SetScalarParameterValueonMaterials _returnvalue "TopS Emissive_03-Brightness-TBP" (Variables|Default|GetCurent%20GameTime))
        (Variables|Default|SetPosition_Data _returnvalue_5)))
    (else
      (Rendering|Material|SetScalarParameterValueonMaterials _returnvalue_8 "TopS_01_Blend_Offset---End" _returnvalue_7)
      (Rendering|Material|SetScalarParameterValueonMaterials _returnvalue_8 "TopS Emissive_03_Offset---End" _returnvalue_7)
      (Rendering|Material|SetScalarParameterValueonMaterials _returnvalue_8 "TopS Emissive_03-Brightness-TBP" (Variables|Default|GetCurent%20GameTime))
      (Variables|Default|SetPosition_Data _returnvalue_7))))

(fn ConstructionScript)

(fn Weapon_Offsets
  (params (Offset_End_Name :type Name) (Offset_End_Value :type Float%20%28double-precision%29) (Offset_Start_Name :type Name) (Offset_State_Value :type Float%20%28double-precision%29))
  (bind _returnvalue (CustomItemWrap|GetWrappedMeshComponents))
  (Rendering|Material|SetVectorParameterValueonMaterials _returnvalue Offset_End_Name (Math|Vector|MakeVector 0.0 0.0 Offset_End_Value))
  (Rendering|Material|SetVectorParameterValueonMaterials _returnvalue Offset_Start_Name (Math|Vector|MakeVector 0.0 0.0 Offset_State_Value)))
