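; Custom item wrap Blueprint (MistTycoon) that animates a weapon wrap's material in reaction to firing.
; Each shot stamps game time into emissive and blend parameters on wrapped meshes, staggering three decal timestamps with short delays.
; The first shot starts a firing state, which a retriggerable delay reverses after one second without shots.
; Both transitions share a looping timer that ramps mask alpha, decal opacity and audio static strength, clearing itself after five seconds.
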(event EventOnPawnSet)

(event EventOnWeaponFired
  (bind _returnvalue (Utilities|Time|GetGameTimeinSeconds))
  (bind _returnvalue_2 (CustomItemWrap|GetWrappedMeshComponents))
  (bind _returnvalue_3 (CustomItemWrap|GetWrappedMeshComponents))
  (bind _returnvalue_4 (Utilities|Time|GetGameTimeinSeconds))
  (Utilities|FlowControl|Sequence
    (:then_0
      ;[1 1 1] TopsSide V2
      (Rendering|Material|SetScalarParameterValueonMaterials _returnvalue_3 "TopS Emissive_02-Brightness-TBP" _returnvalue_4)
      (Rendering|Material|SetScalarParameterValueonMaterials _returnvalue_3 "TopS Emissive_02_Offset----TBP" _returnvalue_4)
      (Rendering|Material|SetScalarParameterValueonMaterials _returnvalue_3 "TopS_01_Blend_Opacity-TBP" _returnvalue_4)
      (Rendering|Material|SetScalarParameterValueonMaterials _returnvalue_3 "TopS Emissive_01-Brightness-TBP" _returnvalue_4)
      (Rendering|Material|SetScalarParameterValueonMaterials _returnvalue_3 "TopS Emissive_01_Offset----TBP" _returnvalue_4)
      (Rendering|Material|SetScalarParameterValueonMaterials _returnvalue_3 "TopS Emissive_03-Brightness-TBP" _returnvalue_4)
      (Rendering|Material|SetScalarParameterValueonMaterials _returnvalue_3 "TopS Emissive_03_Offset----TBP" _returnvalue_4)
      (Rendering|Material|SetScalarParameterValueonMaterials _returnvalue_3 "TopS_01_Speed-TBP" _returnvalue_4)
      (Rendering|Material|SetScalarParameterValueonMaterials _returnvalue_3 "TopS_01_Blend_Offset-TBP" _returnvalue_4)
      (Rendering|Material|SetScalarParameterValueonMaterials _returnvalue_2 "EDRec_Emissive_B_Brightness_Time_BP" _returnvalue)
      (Rendering|Material|SetScalarParameterValueonMaterials _returnvalue_2 "EDRec_Emissive_C_Brightness_Time_BP" _returnvalue)
      (Utilities|FlowControl|Delay 0.1)
      (Rendering|Material|SetScalarParameterValueonMaterials _returnvalue_2 "EDRec_TimeFrom_BP_DecalB" _returnvalue)
      (Utilities|FlowControl|Delay 0.1)
      (Rendering|Material|SetScalarParameterValueonMaterials _returnvalue_2 "EDRec_TimeFrom_BP_DecalC" _returnvalue)
      (Utilities|FlowControl|Delay 0.1)
      (Rendering|Material|SetScalarParameterValueonMaterials _returnvalue_2 "EDRec_TimeFrom_BP_DecalD" _returnvalue))
    (:then_1
      (if (not (Variables|Default|GetFullAutoWeapons))
        (Variables|Default|SetFullAutoWeapons true)
        (CallFunction|SineAnimation))
      (Utilities|FlowControl|RetriggerableDelay 1.0)
      (CallFunction|SineAnimOFF)
      (Variables|Default|SetFullAutoWeapons false))))

(event EventBeginPlay)

(event Custom|SineAnimation
  ; Layer Offset
  (Variables|Default|SetReverse false)
  (Variables|Default|SetStartTime (Utilities|Time|GetGameTimeinSeconds))
  (bind _returnvalue (Utilities|Time|SetTimerbyEvent (delegate Custom|SineAnimUpdate) 0.02 true))
  (Variables|Default|SetTimerHandle _returnvalue))

(event Custom|SineAnimOFF
  ; Layer Offset
  (Variables|Default|SetReverse true)
  (Variables|Default|SetStartTime (Utilities|Time|GetGameTimeinSeconds))
  (bind _returnvalue (Utilities|Time|SetTimerbyEvent (delegate Custom|SineAnimUpdate) 0.02 true))
  (Variables|Default|SetTimerHandle _returnvalue))

(event Custom|SineAnimUpdate
  (bind _returnvalue (Utilities|Time|GetGameTimeinSeconds))
  (bind _returnvalue_2 (CustomItemWrap|GetWrappedMeshComponents))
  (bind _starttime (Variables|Default|GetStartTime))
  (bind _returnvalue_3 (- _returnvalue _starttime))
  ; Layer Offset
  (if (>= _returnvalue_3 5.0)
    (Utilities|Time|ClearandInvalidateTimerbyHandle (Variables|Default|GetTimerHandle))
    (if (Variables|Default|GetReverse)
      ;; Start
      (Rendering|Material|SetColorParameterValueonMaterials _returnvalue_2 "TopS Blend-TS 02_Mask_Channel" (0.0 0.0 0.0 0.0))
      (Rendering|Material|SetScalarParameterValueonMaterials _returnvalue_2 "EDRec_DecalA_Opacity_End" 1.0)
      (Rendering|Material|SetScalarParameterValueonMaterials _returnvalue_2 "EDRec_Audio C Static_Strength" 0.1)
      (else
        ;; Arrive
        (Rendering|Material|SetColorParameterValueonMaterials _returnvalue_2 "TopS Blend-TS 02_Mask_Channel" (0.0 0.0 0.0 1.0))
        (Rendering|Material|SetScalarParameterValueonMaterials _returnvalue_2 "EDRec_DecalA_Opacity_End" 0.0)
        (Rendering|Material|SetScalarParameterValueonMaterials _returnvalue_2 "EDRec_Audio C Static_Strength" 0.3)))
    (elif (Variables|Default|GetReverse)
      ;; FadeIn
      (Rendering|Material|SetColorParameterValueonMaterials _returnvalue_2 "TopS Blend-TS 02_Mask_Channel" (Math|Color|MakeColor 0.0 0.0 0.0 (Math|Float|MapRangeClamped _returnvalue_3 0.0 1.0 1.0)))
      (Rendering|Material|SetScalarParameterValueonMaterials _returnvalue_2 "EDRec_DecalA_Opacity_End" (Math|Float|MapRangeClamped _returnvalue_3 0.0 0.25 0.0 1.0))
      (Rendering|Material|SetScalarParameterValueonMaterials _returnvalue_2 "EDRec_Audio C Static_Strength" (Math|Float|MapRangeClamped _returnvalue_3 0.0 0.25 0.3 0.1))
      (else
        ;; FadeOut
        (Rendering|Material|SetColorParameterValueonMaterials _returnvalue_2 "TopS Blend-TS 02_Mask_Channel" (Math|Color|MakeColor 0.0 0.0 0.0 (Math|Float|MapRangeClamped _returnvalue_3 0.0 0.15 0.0 1.0)))
        (Rendering|Material|SetScalarParameterValueonMaterials _returnvalue_2 "EDRec_DecalA_Opacity_End" (Math|Float|MapRangeClamped _returnvalue_3 0.0 0.25 1.0))
        (Rendering|Material|SetScalarParameterValueonMaterials _returnvalue_2 "EDRec_Audio C Static_Strength" (Math|Float|MapRangeClamped _returnvalue_3 0.0 0.25 0.1 0.3))))))

(fn ConstructionScript)
