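; MistyBead reactive item wrap Blueprint that animates wrap decal materials in response to player kills and weapon fire.
; AthenaKills stat changes play a timeline lerping warp strength, opacity and per-item UV offsets, with a Killed flag blocking retriggers.
; Each shot stamps game time into flipbook, emissive and decal material parameters for shader-driven reactions.
; In the front end, a shot counter triggers the same timeline to preview the kill effect.
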(event EventOnPawnSet
  (bind _returnvalue (CustomItemWrap|GetWrappedMeshComponents))
  (Cosmetics|WatchCosmeticStat (CustomCharacter|GetAssociatedPlayerPawn) "(TagName=\"StatManager.AthenaKills\")" 0 false
    (:OnStatChanged
      (Utilities|FlowControl|Sequence
        (:then_0
          ; L/R Offset
          (if (not (Variables|Default|GetKilled))
            (Variables|Default|SetKilled true)
            (bind (_direction _offsetcurve) (Timeline|UVOffsetTimeline PlayFromStart)
              (:Update
                (Rendering|Material|SetScalarParameterValueonMaterials _returnvalue "FDecalV2_UV_Warp-Strength" (Math|Float|Lerp 0.025 0.05 _offsetcurve))
                (Rendering|Material|SetColorParameterValueonMaterials _returnvalue "FDecalV2_UV_Offset_Weapon" (Math|Color|MakeColor 0.0 (Math|Float|Lerp -0.3 -0.1 _offsetcurve)))
                (Rendering|Material|SetColorParameterValueonMaterials _returnvalue "FDecalV2_UV_Offset_Pickaxe" (Math|Color|MakeColor 0.0 (Math|Float|Lerp -0.8 -0.1 _offsetcurve)))
                (Rendering|Material|SetColorParameterValueonMaterials _returnvalue "FDecalV2_UV_Offset_Backpack" (Math|Color|MakeColor 0.0 (Math|Float|Lerp -0.8 -0.1 _offsetcurve)))
                (Rendering|Material|SetColorParameterValueonMaterials _returnvalue "FDecalV2_UV_Offset_VehicleSide" (Math|Color|MakeColor 0.0 (Math|Float|Lerp -0.8 -0.1 _offsetcurve)))
                (Rendering|Material|SetColorParameterValueonMaterials _returnvalue "FDecalV2_UV_Offset_Vehicle" (Math|Color|MakeColor (Math|Float|Lerp -0.8 -0.4 _offsetcurve)))
                (Rendering|Material|SetColorParameterValueonMaterials _returnvalue "FDecalV2_UV_Offset_Character" (Math|Color|MakeColor 0.0 (Math|Float|Lerp -0.7 -0.4 _offsetcurve)))
                (Rendering|Material|SetScalarParameterValueonMaterials _returnvalue "FDecalV2_Opacity Time__End" (Math|Float|Lerp 0.0 1.0 _offsetcurve)))
              (:Finished
                (Variables|Default|SetKilled false)))))))))

(event EventOnWeaponFired
  (bind _returnvalue (Utilities|Time|GetGameTimeinSeconds))
  (bind _returnvalue_2 (CustomItemWrap|GetWrappedMeshComponents))
  (bind _returnvalue_3 (CustomItemWrap|GetWrappedMeshComponents))
  (bind _returnvalue_4 (CustomItemWrap|GetWrappedMeshComponents))
  (bind _returnvalue_5 (Utilities|Time|GetGameTimeinSeconds))
  (bind _returnvalue_6 (CustomItemWrap|GetWrappedMeshComponents))
  (bind _returnvalue_7 (Utilities|Time|GetGameTimeinSeconds))
  (bind _returnvalue_8 (CustomItemWrap|GetWrappedMeshComponents))
  (bind _returnvalue_9 (Utilities|Time|GetGameTimeinSeconds))
  (Utilities|FlowControl|Sequence
    (:then_0
      (Variables|Default|SetRandomOffset_X (Math|Random|RandomFloatinRange -0.5 0.5))
      ; Trigger Waves on FE
      (if (CustomCosmeticModifier|IsinFrontEnd)
        (Utilities|FlowControl|Delay 0.001)
        (bind _output_get (Variables|Default|SetShotsFired (+ (Variables|Default|GetShotsFired) 1)))
        (if (>= _output_get 5)
          (if (== (Variables|Default|GetShotsFired) 7)
            (Variables|Default|SetShotsFired 0)
            (Utilities|FlowControl|Sequence
              (:then_0
                ; L/R Offset
                (if (not (Variables|Default|GetKilled))
                  (Variables|Default|SetKilled true)
                  (Timeline|UVOffsetTimeline PlayFromStart))))))))
    (:then_1
      ; Left\Right V2
      (Rendering|Material|SetScalarParameterValueonMaterials _returnvalue_4 "FDecalV2_FlipBook_Time_FromBP" _returnvalue_5)
      (Rendering|Material|SetScalarParameterValueonMaterials _returnvalue_4 "FDecalV2_Emissive_Brightness_Time_BP" _returnvalue_5)
      (Rendering|Material|SetScalarParameterValueonMaterials _returnvalue_4 "FDecalV2_Emissive_Gradent_Time_BP" _returnvalue_5)
      (Rendering|Material|SetScalarParameterValueonMaterials _returnvalue_4 "FDecalV2_Counter_Brightness_Time_BP" _returnvalue_5))
    (:then_2
      ;[1 1 1 0.5] Decal Reactive
      (Rendering|Material|SetScalarParameterValueonMaterials _returnvalue_8 "EDRec_TimeFrom_BP_DecalA" _returnvalue_9)
      (Rendering|Material|SetScalarParameterValueonMaterials _returnvalue_8 "EDRec_TimeFrom_BP_DecalB" _returnvalue_9)
      (Rendering|Material|SetScalarParameterValueonMaterials _returnvalue_8 "EDRec_TimeFrom_BP_DecalC" _returnvalue_9)
      (Rendering|Material|SetScalarParameterValueonMaterials _returnvalue_8 "EDRec_TimeFrom_BP_DecalD" _returnvalue_9)
      (Rendering|Material|SetScalarParameterValueonMaterials _returnvalue_6 "EDRec_Emissive_A_Brightness_Time_BP" _returnvalue_7)
      (Rendering|Material|SetScalarParameterValueonMaterials _returnvalue_6 "EDRec_Emissive_B_Brightness_Time_BP" _returnvalue_7)
      (Rendering|Material|SetScalarParameterValueonMaterials _returnvalue_6 "EDRec_Emissive_C_Brightness_Time_BP" _returnvalue_7)
      (Rendering|Material|SetScalarParameterValueonMaterials _returnvalue_6 "EDRec_Emissive_D_Brightness_Time_BP" _returnvalue_7)
      (Rendering|Material|SetScalarParameterValueonMaterials _returnvalue_2 "EDRec_Emissive_A_Gradent_Time_BP" _returnvalue)
      (Rendering|Material|SetScalarParameterValueonMaterials _returnvalue_2 "EDRec_Emissive_B_Gradent_Time_BP" _returnvalue)
      (Rendering|Material|SetScalarParameterValueonMaterials _returnvalue_2 "EDRec_Emissive_C_Gradent_Time_BP" _returnvalue)
      (Rendering|Material|SetScalarParameterValueonMaterials _returnvalue_2 "EDRec_Emissive_D_Gradent_Time_BP" _returnvalue))))

(fn ConstructionScript)
