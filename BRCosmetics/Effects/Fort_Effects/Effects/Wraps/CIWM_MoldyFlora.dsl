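; Moldy Flora weapon wrap logic that shifts the wrap's emissive, color and mask material parameters while the weapon fires.
; The first shot starts a fade in, and a retriggerable delay reverses it once firing pauses for three seconds.
; A looping timer drives clamped linear ramps on the wrapped mesh materials, faster in than out despite the "Sine" naming.
; After five seconds, the timer clears and sets the end values.
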(event EventOnWeaponFired
  (if (not (Variables|Default|GetFullAutoWeapons))
    (Variables|Default|SetFullAutoWeapons true)
    (CallFunction|SineAnimation))
  (Utilities|FlowControl|RetriggerableDelay 3.0)
  (CallFunction|SineAnimOFF)
  (Variables|Default|SetFullAutoWeapons false))

(event Custom|SineAnimation
  ; Layer Offset
  (Variables|Default|SetReverse false)
  (Variables|Default|SetStartTime (Utilities|Time|GetGameTimeinSeconds))
  (bind _returnvalue (Utilities|Time|SetTimerbyEvent (delegate Custom|SineAnimUpdate) 0.02 true))
  (Variables|Default|SetTimerHandle _returnvalue))

(event Custom|SineAnimOFF
  ; Layer Offset
  (Variables|Default|SetReverse true)
  (Variables|Default|SetStartTime (Utilities|Time|GetGameTimeinSeconds))
  (bind _returnvalue (Utilities|Time|SetTimerbyEvent (delegate Custom|SineAnimUpdate) 0.02 true))
  (Variables|Default|SetTimerHandle _returnvalue))

(event Custom|SineAnimUpdate
  (bind _returnvalue (Utilities|Time|GetGameTimeinSeconds))
  (bind _returnvalue_2 (CustomItemWrap|GetWrappedMeshComponents))
  (bind _starttime (Variables|Default|GetStartTime))
  (bind _returnvalue_3 (- _returnvalue _starttime))
  ; Layer Offset
  (if (>= _returnvalue_3 5.0)
    (Utilities|Time|ClearandInvalidateTimerbyHandle (Variables|Default|GetTimerHandle))
    (if (Variables|Default|GetReverse)
      ;; Start
      (Rendering|Material|SetColorParameterValueonMaterials _returnvalue_2 "TopS Emissive_03_Grad-Tex_Mask_Channel" (1.0 0.0 0.0 0.0))
      (Rendering|Material|SetColorParameterValueonMaterials _returnvalue_2 "TopS Emissive_02_Grad-Tex_Mask_Channel" (0.7 0.0 0.0 0.0))
      (Rendering|Material|SetScalarParameterValueonMaterials _returnvalue_2 "TopS Emissive_01-Contrast" 2.0)
      (Rendering|Material|SetScalarParameterValueonMaterials _returnvalue_2 "FDecalV2_Emissive------->BaseColor" 0.0)
      (Rendering|Material|SetScalarParameterValueonMaterials _returnvalue_2 "TopS BaseColor----->Emissive" 2.0)
      (Rendering|Material|SetScalarParameterValueonMaterials _returnvalue_2 "TopS Emissive_01_ColorCuve-Brightness" 2.5)
      (Rendering|Material|SetScalarParameterValueonMaterials _returnvalue_2 "HueShift_EmissiveSaturation" -0.35)
      (Rendering|Material|SetScalarParameterValueonMaterials _returnvalue_2 "TopS Emissive_01_ColorCuve_TileRate" 0.15)
      (Rendering|Material|SetVectorParameterValueonMaterials _returnvalue_2 "TopS Emissive_01_ColorCuve Tint" (1.0 1.0 1.0))
      (else
        ;; Arrive
        (Rendering|Material|SetColorParameterValueonMaterials _returnvalue_2 "TopS Emissive_03_Grad-Tex_Mask_Channel" (0.0 0.0 0.0 0.0))
        (Rendering|Material|SetColorParameterValueonMaterials _returnvalue_2 "TopS Emissive_02_Grad-Tex_Mask_Channel" (0.0 0.0 0.0 0.0))
        (Rendering|Material|SetScalarParameterValueonMaterials _returnvalue_2 "TopS Emissive_01-Contrast" 0.5)
        (Rendering|Material|SetScalarParameterValueonMaterials _returnvalue_2 "FDecalV2_Emissive------->BaseColor" 5.0)
        (Rendering|Material|SetScalarParameterValueonMaterials _returnvalue_2 "TopS BaseColor----->Emissive" 5.0)
        (Rendering|Material|SetScalarParameterValueonMaterials _returnvalue_2 "TopS Emissive_01_ColorCuve-Brightness" 1.25)
        (Rendering|Material|SetScalarParameterValueonMaterials _returnvalue_2 "HueShift_EmissiveSaturation" -1.0)
        (Rendering|Material|SetScalarParameterValueonMaterials _returnvalue_2 "TopS Emissive_01_ColorCuve_TileRate" 0.01)
        (Rendering|Material|SetVectorParameterValueonMaterials _returnvalue_2 "TopS Emissive_01_ColorCuve Tint" (1.25 1.4 1.25))))
    (elif (Variables|Default|GetReverse)
      ;; OUT
      (Rendering|Material|SetColorParameterValueonMaterials _returnvalue_2 "TopS Emissive_03_Grad-Tex_Mask_Channel" (Math|Color|MakeColor (Math|Float|MapRangeClamped _returnvalue_3 0.0 0.75 0.0 1.0) 0.0 0.0 0.0))
      (Rendering|Material|SetColorParameterValueonMaterials _returnvalue_2 "TopS Emissive_02_Grad-Tex_Mask_Channel" (Math|Color|MakeColor (Math|Float|MapRangeClamped _returnvalue_3 0.0 0.75 0.0 0.7) 0.0 0.0 0.0))
      (Rendering|Material|SetScalarParameterValueonMaterials _returnvalue_2 "TopS Emissive_01-Contrast" (Math|Float|MapRangeClamped _returnvalue_3 0.0 0.75 0.5 2.0))
      (Rendering|Material|SetScalarParameterValueonMaterials _returnvalue_2 "FDecalV2_Emissive------->BaseColor" (Math|Float|MapRangeClamped _returnvalue_3 0.0 0.75 5.0))
      (Rendering|Material|SetScalarParameterValueonMaterials _returnvalue_2 "TopS BaseColor----->Emissive" (Math|Float|MapRangeClamped _returnvalue_3 0.0 0.75 5.0 2.0))
      (Rendering|Material|SetScalarParameterValueonMaterials _returnvalue_2 "TopS Emissive_01_ColorCuve-Brightness" (Math|Float|MapRangeClamped _returnvalue_3 0.0 0.75 1.25 2.5))
      (Rendering|Material|SetScalarParameterValueonMaterials _returnvalue_2 "HueShift_EmissiveSaturation" (Math|Float|MapRangeClamped _returnvalue_3 0.0 0.75 -1.0 -0.35))
      (Rendering|Material|SetScalarParameterValueonMaterials _returnvalue_2 "TopS Emissive_01_ColorCuve_TileRate" (Math|Float|MapRangeClamped _returnvalue_3 0.0 0.75 0.01 0.15))
      (Rendering|Material|SetVectorParameterValueonMaterials _returnvalue_2 "TopS Emissive_01_ColorCuve Tint" (Math|Vector|MakeVector (Math|Float|MapRangeClamped _returnvalue_3 0.0 0.75 1.25 1.0) (Math|Float|MapRangeClamped _returnvalue_3 0.0 0.75 1.4 1.0) (Math|Float|MapRangeClamped _returnvalue_3 0.0 0.75 1.25 1.0)))
      (else
        ;; IN
        (Rendering|Material|SetColorParameterValueonMaterials _returnvalue_2 "TopS Emissive_03_Grad-Tex_Mask_Channel" (Math|Color|MakeColor (Math|Float|MapRangeClamped _returnvalue_3 0.0 0.25 1.0) 0.0 0.0 0.0))
        (Rendering|Material|SetColorParameterValueonMaterials _returnvalue_2 "TopS Emissive_02_Grad-Tex_Mask_Channel" (Math|Color|MakeColor (Math|Float|MapRangeClamped _returnvalue_3 0.0 0.25 0.7) 0.0 0.0 0.0))
        (Rendering|Material|SetScalarParameterValueonMaterials _returnvalue_2 "TopS Emissive_01-Contrast" (Math|Float|MapRangeClamped _returnvalue_3 0.0 0.25 2.0 0.5))
        (Rendering|Material|SetScalarParameterValueonMaterials _returnvalue_2 "FDecalV2_Emissive------->BaseColor" (Math|Float|MapRangeClamped _returnvalue_3 0.0 0.25 0.0 5.0))
        (Rendering|Material|SetScalarParameterValueonMaterials _returnvalue_2 "TopS BaseColor----->Emissive" (Math|Float|MapRangeClamped _returnvalue_3 0.0 0.25 2.0 5.0))
        (Rendering|Material|SetScalarParameterValueonMaterials _returnvalue_2 "TopS Emissive_01_ColorCuve-Brightness" (Math|Float|MapRangeClamped _returnvalue_3 0.0 0.25 2.5 1.25))
        (Rendering|Material|SetScalarParameterValueonMaterials _returnvalue_2 "HueShift_EmissiveSaturation" (Math|Float|MapRangeClamped _returnvalue_3 0.0 0.25 -0.35 -1.0))
        (Rendering|Material|SetScalarParameterValueonMaterials _returnvalue_2 "TopS Emissive_01_ColorCuve_TileRate" (Math|Float|MapRangeClamped _returnvalue_3 0.0 0.25 0.15 0.01))
        (Rendering|Material|SetVectorParameterValueonMaterials _returnvalue_2 "TopS Emissive_01_ColorCuve Tint" (Math|Vector|MakeVector (Math|Float|MapRangeClamped _returnvalue_3 0.0 0.25 1.0 1.25) (Math|Float|MapRangeClamped _returnvalue_3 0.0 0.25 1.0 1.4) (Math|Float|MapRangeClamped _returnvalue_3 0.0 0.25 1.0 1.25)))))))

(fn ConstructionScript)
