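; Weapon wrap Blueprint that makes the wrapped meshes' materials react to reloading and firing.
; Begin play and pawn or weapon change events bind the current weapon's reload-started dispatcher, with no unbinding.
; Reloading plays a Blink timeline into an emissive brightness parameter until a reload-length timer stops it and zeroes the value.
; Firing writes game time into emissive, offset and opacity parameters, apparently for time-based animation inside the material.
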(event EventBeginPlay
  (bind _returnvalue (CustomCharacter|GetAssociatedPlayerPawn))
  (Utilities|IsValid _returnvalue
    (:Is%20Valid
      (bind _currentweapon (Variables|/Script/FortniteGame.FortPawn|GetCurrentWeapon _returnvalue)
        (:then
          (EventDispatchers|BindEventtoOnReloadStarted _currentweapon (delegate Custom|Reloading)))
        (:else)))))

(event EventOnPawnSet
  (bind _returnvalue (CustomCharacter|GetAssociatedPlayerPawn))
  (Utilities|IsValid _returnvalue
    (:Is%20Valid
      (bind _currentweapon (Variables|/Script/FortniteGame.FortPawn|GetCurrentWeapon _returnvalue)
        (:then
          (EventDispatchers|BindEventtoOnReloadStarted _currentweapon (delegate Custom|Reloading)))
        (:else)))))

(event EventOnNewWeapons
  (params (WeaponsAsItems :type Array%20of%20Fort%20Item%20Object%20References))
  (bind _returnvalue (CustomCharacter|GetAssociatedPlayerPawn))
  (Utilities|IsValid _returnvalue
    (:Is%20Valid
      (bind _currentweapon (Variables|/Script/FortniteGame.FortPawn|GetCurrentWeapon _returnvalue)
        (:then
          (EventDispatchers|BindEventtoOnReloadStarted _currentweapon (delegate Custom|Reloading)))
        (:else)))))

(event EventOnWeaponChanged
  (params (NewWeapon :type Fort%20Weapon%20Object%20Reference) (OldWeapon :type Fort%20Weapon%20Object%20Reference))
  (bind _returnvalue (CustomCharacter|GetAssociatedPlayerPawn))
  (Utilities|IsValid _returnvalue
    (:Is%20Valid
      (bind _currentweapon (Variables|/Script/FortniteGame.FortPawn|GetCurrentWeapon _returnvalue)
        (:then
          (EventDispatchers|BindEventtoOnReloadStarted _currentweapon (delegate Custom|Reloading)))
        (:else)))))

(event EventOnNewPlayerWeaponType
  (params (NewType :type EFortItemType%20Enum :default "WorldItem") (OldType :type EFortItemType%20Enum :default "WorldItem"))
  (bind _returnvalue (CustomCharacter|GetAssociatedPlayerPawn))
  (Utilities|IsValid _returnvalue
    (:Is%20Valid
      (bind _currentweapon (Variables|/Script/FortniteGame.FortPawn|GetCurrentWeapon _returnvalue)
        (:then
          (EventDispatchers|BindEventtoOnReloadStarted _currentweapon (delegate Custom|Reloading)))
        (:else)))))

(event Custom|Reloading
  (params (ReloadTime :type Float%20%28single-precision%29) (ReloadType :type EFortWeaponReloadType%20Enum :default "ReloadWholeClip"))
  (Utilities|Time|SetTimerbyEvent (delegate Custom|EndReload) ReloadTime)
  (bind (_direction _newtrack_0) (Timeline|Blink PlayFromStart)
    (:Update
      (Rendering|Material|SetScalarParameterValueonMaterials (CustomItemWrap|GetWrappedMeshComponents) "TopS Emissive_01_Brightness---End" _newtrack_0))))

(event Custom|EndReload
  (Utilities|FlowControl|Sequence
    (:then_0
      (Timeline|Blink Stop))
    (:then_1
      (Timeline|Blink SetNewTime))
    (:then_2
      (Rendering|Material|SetScalarParameterValueonMaterials (CustomItemWrap|GetWrappedMeshComponents) "TopS Emissive_01_Brightness---End" 0.0))))

(event EventOnWeaponFired
  (bind _returnvalue (CustomItemWrap|GetWrappedMeshComponents))
  (bind _returnvalue_2 (Utilities|Time|GetGameTimeinSeconds))
  (Utilities|FlowControl|Sequence
    (:then_0
      ;[1 1 1] TopsSide V2
      (Rendering|Material|SetScalarParameterValueonMaterials _returnvalue "TopS Emissive_02-Brightness-TBP" _returnvalue_2)
      (Rendering|Material|SetScalarParameterValueonMaterials _returnvalue "TopS Emissive_02_Offset----TBP" _returnvalue_2)
      (Rendering|Material|SetScalarParameterValueonMaterials _returnvalue "TopS_01_Blend_Opacity-TBP" _returnvalue_2)
      (Rendering|Material|SetScalarParameterValueonMaterials _returnvalue "TimeFrom-BP----------------" _returnvalue_2))
    (:then_1
      (Rendering|Material|SetScalarParameterValueonMaterials _returnvalue "TopS Emissive_01-Brightness-TBP" _returnvalue_2)
      (Rendering|Material|SetScalarParameterValueonMaterials _returnvalue "TopS Emissive_01_Offset----TBP" _returnvalue_2)
      (Rendering|Material|SetScalarParameterValueonMaterials _returnvalue "TopS Emissive_03-Brightness-TBP" _returnvalue_2)
      (Rendering|Material|SetScalarParameterValueonMaterials _returnvalue "TopS Emissive_03_Offset----TBP" _returnvalue_2))))

(fn ConstructionScript)
