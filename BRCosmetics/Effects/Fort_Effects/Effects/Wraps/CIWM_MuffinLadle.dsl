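; Weapon wrap logic that animates an emissive saber effect on wrapped meshes in response to firing and kills.
; Firing pulses brightness and slides the emissive offset in, then reverses it after a delayed idle check using pause flags.
; A kill-stat change sweeps out, toggles the color-curve value and sweeps back; the front end simulates this every fifth shot, debug-printing the count.
; Animations are material-driven from game-time stamps, start/end values and durations.
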(event EventOnWeaponFired
  (bind _returnvalue (CustomItemWrap|GetWrappedMeshComponents))
  (bind _returnvalue_2 (CustomItemWrap|GetWrappedMeshComponents))
  (if (CustomCosmeticModifier|IsinFrontEnd)
    (Utilities|FlowControl|Delay 0.05)
    (bind _output_get (Variables|Default|SetIterationCount (+ (Variables|Default|GetIterationCount) 1)))
    (Development|PrintText (Utilities|Text|ToText(Integer) _output_get) true true "(R=0.998750,G=0.001751,B=1.000000,A=1.000000)")
    (if (== _output_get 5)
      (CallFunction|SaberActivate)))
  (if (not (Variables|Default|GetSaber))
    (Variables|Default|SetPause_0 true)
    (Utilities|FlowControl|Sequence
      (:then_0
        (Rendering|Material|SetScalarParameterValueonMaterials (CustomItemWrap|GetWrappedMeshComponents) "TopS Emissive_01-Brightness-TBP" (Utilities|Time|GetGameTimeinSeconds)))
      (:then_1
        (if (not (Variables|Default|GetPause))
          (Variables|Default|SetPause true)
          (Rendering|Material|SetScalarParameterValueonMaterials _returnvalue_2 "TopS Emissive_01_Offset----TBP" (Utilities|Time|GetGameTimeinSeconds))
          (Rendering|Material|SetScalarParameterValueonMaterials _returnvalue_2 "TopS Emissive_01_Offset---End" -0.5)
          (Rendering|Material|SetScalarParameterValueonMaterials _returnvalue_2 "TopS Emissive_01_Offset---Start" 0.5)
          (Rendering|Material|SetScalarParameterValueonMaterials _returnvalue_2 "TopS Emissive_01_Offset---Duration" 0.15)))
      (:then_2
        (Utilities|FlowControl|Delay 2.0)
        (Variables|Default|SetPause_0 false)
        (Utilities|FlowControl|Delay 1.5)
        (if (not (Variables|Default|GetPause_0))
          (Variables|Default|SetPause false)
          (Rendering|Material|SetScalarParameterValueonMaterials _returnvalue "TopS Emissive_01_Offset----TBP" (Utilities|Time|GetGameTimeinSeconds))
          (Rendering|Material|SetScalarParameterValueonMaterials _returnvalue "TopS Emissive_01_Offset---End" 0.5)
          (Rendering|Material|SetScalarParameterValueonMaterials _returnvalue "TopS Emissive_01_Offset---Start" -0.5)
          (Rendering|Material|SetScalarParameterValueonMaterials _returnvalue "TopS Emissive_01_Offset---Duration" 0.5))))))

(event EventOnPawnSet
  (bind _returnvalue (CustomItemWrap|GetWrappedMeshComponents))
  (bind _returnvalue_2 (CustomItemWrap|GetWrappedMeshComponents))
  (bind _returnvalue_3 (CustomItemWrap|GetWrappedMeshComponents))
  (Cosmetics|WatchCosmeticStat (CustomCharacter|GetAssociatedPlayerPawn) "(TagName=\"StatManager.AthenaKills\")" 0 false
    (:OnStatChanged
      (Variables|Default|SetSaber true)
      (Rendering|Material|SetScalarParameterValueonMaterials _returnvalue "TopS Emissive_01_Offset----TBP" (Utilities|Time|GetGameTimeinSeconds))
      (Rendering|Material|SetScalarParameterValueonMaterials _returnvalue "TopS Emissive_01_Offset---End" 0.5)
      (Rendering|Material|SetScalarParameterValueonMaterials _returnvalue "TopS Emissive_01_Offset---Start" -0.5)
      (Rendering|Material|SetScalarParameterValueonMaterials _returnvalue "TopS Emissive_01_Offset---Duration" 0.25)
      (Utilities|FlowControl|Delay 0.25)
      (if (Variables|Default|GetSaber_Color)
        (Rendering|Material|SetScalarParameterValueonMaterials _returnvalue_2 "TopS Emissive_01_ColorCuve" 203.0)
        (Variables|Default|SetSaber_Color false)
        (else
          (Rendering|Material|SetScalarParameterValueonMaterials _returnvalue_2 "TopS Emissive_01_ColorCuve" 204.0)
          (Variables|Default|SetSaber_Color true)))
      (Rendering|Material|SetScalarParameterValueonMaterials _returnvalue_3 "TopS Emissive_01_Offset----TBP" (Utilities|Time|GetGameTimeinSeconds))
      (Rendering|Material|SetScalarParameterValueonMaterials _returnvalue_3 "TopS Emissive_01_Offset---End" -0.5)
      (Rendering|Material|SetScalarParameterValueonMaterials _returnvalue_3 "TopS Emissive_01_Offset---Start" 0.5)
      (Rendering|Material|SetScalarParameterValueonMaterials _returnvalue_3 "TopS Emissive_01_Offset---Duration" 0.15)
      (Utilities|FlowControl|Delay 0.2)
      (Variables|Default|SetSaber false)
      (Variables|Default|SetIterationCount 0))))

(event Custom|Saber_Activate
  (bind _returnvalue (CustomItemWrap|GetWrappedMeshComponents))
  (bind _returnvalue_2 (CustomItemWrap|GetWrappedMeshComponents))
  (bind _returnvalue_3 (CustomItemWrap|GetWrappedMeshComponents))
  (Variables|Default|SetSaber true)
  (Rendering|Material|SetScalarParameterValueonMaterials _returnvalue "TopS Emissive_01_Offset----TBP" (Utilities|Time|GetGameTimeinSeconds))
  (Rendering|Material|SetScalarParameterValueonMaterials _returnvalue "TopS Emissive_01_Offset---End" 0.5)
  (Rendering|Material|SetScalarParameterValueonMaterials _returnvalue "TopS Emissive_01_Offset---Start" -0.5)
  (Rendering|Material|SetScalarParameterValueonMaterials _returnvalue "TopS Emissive_01_Offset---Duration" 0.25)
  (Utilities|FlowControl|Delay 0.25)
  (if (Variables|Default|GetSaber_Color)
    (Rendering|Material|SetScalarParameterValueonMaterials _returnvalue_2 "TopS Emissive_01_ColorCuve" 203.0)
    (Variables|Default|SetSaber_Color false)
    (else
      (Rendering|Material|SetScalarParameterValueonMaterials _returnvalue_2 "TopS Emissive_01_ColorCuve" 204.0)
      (Variables|Default|SetSaber_Color true)))
  (Rendering|Material|SetScalarParameterValueonMaterials _returnvalue_3 "TopS Emissive_01_Offset----TBP" (Utilities|Time|GetGameTimeinSeconds))
  (Rendering|Material|SetScalarParameterValueonMaterials _returnvalue_3 "TopS Emissive_01_Offset---End" -0.5)
  (Rendering|Material|SetScalarParameterValueonMaterials _returnvalue_3 "TopS Emissive_01_Offset---Start" 0.5)
  (Rendering|Material|SetScalarParameterValueonMaterials _returnvalue_3 "TopS Emissive_01_Offset---Duration" 0.15)
  (Utilities|FlowControl|Delay 0.2)
  (Variables|Default|SetSaber false)
  (Variables|Default|SetIterationCount 0))

(fn ConstructionScript)
