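; Reactive weapon wrap (BlockStack) that changes its material display for damage dealt or received, kills, and shield or health gains.
; Cosmetic stat watchers set flipbook frames, emissive color and game-time stamps on wrapped meshes, with timelines crossfading decals back after kills and gains.
; A shared Kill_Pause flag stops reactions from overlapping.
; In the front end, a timeline previews damage, health and shield reactions, while counted shots simulate hits and kills.
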(event EventBeginPlay
  (Utilities|FlowControl|Delay 0.5)
  (if (CustomCosmeticModifier|IsinFrontEnd)
    (Timeline|FrontEnd PlayFromStart
      (:Damage
        (CallFunction|Damage))
      (:Health
        (CallFunction|Health))
      (:Shilds
        (CallFunction|Shilds)))))

(event EventOnWeaponFired
  (bind _returnvalue (CustomItemWrap|GetWrappedMeshComponents))
  (bind _returnvalue_2 (Utilities|Time|GetGameTimeinSeconds))
  (bind _returnvalue_3 (CustomItemWrap|GetWrappedMeshComponents))
  (if (CustomCosmeticModifier|IsinFrontEnd)
    (Utilities|FlowControl|Sequence
      (:then_0
        (Timeline|FrontEnd PlayFromStart))
      (:then_1
        (Variables|Default|SetFront_End true)
        (bind _output_get (Variables|Envirometn_Varable|SetIterationCount (+ (Variables|Envirometn_Varable|GetIterationCount) 1)))
        (if (>= _output_get 5)
          (if (== (Variables|Envirometn_Varable|GetIterationCount) 7)
            (Variables|Envirometn_Varable|SetIterationCount 0)
            (if (not (Variables|Default|GetKill_Pause))
              (Variables|Default|SetKill_Pause true)
              (Rendering|Material|SetScalarParameterValueonMaterials _returnvalue "EDRec_FlipBook_FrameC" 4.0)
              (Rendering|Material|SetScalarParameterValueonMaterials _returnvalue "EDRec_FlipBook_FrameB" 4.0)
              (Rendering|Material|SetScalarParameterValueonMaterials _returnvalue "EDRec_DecalB_Opacity_End" 1.0)
              (Rendering|Material|SetScalarParameterValueonMaterials _returnvalue "EDRec_DecalD_Opacity_End" 0.0)
              (Rendering|Material|SetScalarParameterValueonMaterials _returnvalue "EDRec_TimeFrom_BP_DecalC" _returnvalue_2)
              (Rendering|Material|SetScalarParameterValueonMaterials _returnvalue "EDRec_TimeFrom_BP_DecalB" _returnvalue_2)
              (Rendering|Material|SetScalarParameterValueonMaterials _returnvalue "EDRec_TimeFrom_BP_DecalB" _returnvalue_2)
              (Rendering|Material|SetScalarParameterValueonMaterials _returnvalue "TopS Emissive_01_Offset----TBP" _returnvalue_2)
              (Rendering|Material|SetColorParameterValueonMaterials _returnvalue "TopS Emissive_01_Color-" "(R=5.000000,G=0.000001,B=0.036865,A=1.000000)")
              (Utilities|FlowControl|Sequence
                (:then_0
                  (Utilities|FlowControl|Delay 2.0)
                  (bind (_direction _newtrack_0) (Timeline|Timeline PlayFromStart)
                    (:Update
                      (Rendering|Material|SetScalarParameterValueonMaterials _returnvalue_3 "EDRec_DecalB_Opacity_End" (Math|Float|Lerp 1.0 0.0 _newtrack_0))
                      (Rendering|Material|SetScalarParameterValueonMaterials _returnvalue_3 "EDRec_DecalD_Opacity_End" (Math|Float|Lerp 0.0 1.0 _newtrack_0)))
                    (:Finished
                      (Variables|Default|SetKill_Pause false))))))
            (elif (Variables|Default|GetKill_Pause)
              _
              (else
                (Rendering|Material|SetScalarParameterValueonMaterials (CustomItemWrap|GetWrappedMeshComponents) "EDRec_DecalD__Sine_TimeFrom_BP" (Utilities|Time|GetGameTimeinSeconds))))))))))

(event EventOnPawnSet
  (bind _returnvalue (CustomItemWrap|GetWrappedMeshComponents))
  (bind _returnvalue_2 (Utilities|Time|GetGameTimeinSeconds))
  (bind _returnvalue_3 (CustomItemWrap|GetWrappedMeshComponents))
  (bind _returnvalue_4 (CustomItemWrap|GetWrappedMeshComponents))
  (bind _returnvalue_5 (Utilities|Time|GetGameTimeinSeconds))
  (bind _returnvalue_6 (CustomItemWrap|GetWrappedMeshComponents))
  (bind _returnvalue_7 (CustomCharacter|GetAssociatedPlayerPawn))
  (bind _returnvalue_8 (CustomItemWrap|GetWrappedMeshComponents))
  (bind _returnvalue_9 (Utilities|Time|GetGameTimeinSeconds))
  (bind _returnvalue_10 (Utilities|Time|GetGameTimeinSeconds))
  (bind _returnvalue_11 (CustomItemWrap|GetWrappedMeshComponents))
  (bind _returnvalue_12 (CustomItemWrap|GetWrappedMeshComponents))
  (Utilities|FlowControl|Sequence
    (:then_0
      (Cosmetics|WatchCosmeticStat _returnvalue_7 "(TagName=\"StatManager.DamageDealt\")" 0 false
        (:OnStatChanged
          (if (not (Variables|Default|GetKill_Pause))
            (Rendering|Material|SetScalarParameterValueonMaterials (CustomItemWrap|GetWrappedMeshComponents) "EDRec_DecalD__Sine_TimeFrom_BP" (Utilities|Time|GetGameTimeinSeconds))))))
    (:then_1
      (Cosmetics|WatchCosmeticStat _returnvalue_7 "(TagName=\"StatManager.AthenaKills\")" 0 false
        (:OnStatChanged
          (if (not (Variables|Default|GetKill_Pause))
            (Variables|Default|SetKill_Pause true)
            (Rendering|Material|SetScalarParameterValueonMaterials _returnvalue_4 "EDRec_FlipBook_FrameC" 4.0)
            (Rendering|Material|SetScalarParameterValueonMaterials _returnvalue_4 "EDRec_FlipBook_FrameB" 4.0)
            (Rendering|Material|SetScalarParameterValueonMaterials _returnvalue_4 "EDRec_DecalB_Opacity_End" 1.0)
            (Rendering|Material|SetScalarParameterValueonMaterials _returnvalue_4 "EDRec_DecalD_Opacity_End" 0.0)
            (Rendering|Material|SetScalarParameterValueonMaterials _returnvalue_4 "EDRec_TimeFrom_BP_DecalC" _returnvalue_5)
            (Rendering|Material|SetScalarParameterValueonMaterials _returnvalue_4 "EDRec_TimeFrom_BP_DecalB" _returnvalue_5)
            (Rendering|Material|SetScalarParameterValueonMaterials _returnvalue_4 "EDRec_TimeFrom_BP_DecalB" _returnvalue_5)
            (Rendering|Material|SetScalarParameterValueonMaterials _returnvalue_4 "TopS Emissive_01_Offset----TBP" _returnvalue_5)
            (Rendering|Material|SetColorParameterValueonMaterials _returnvalue_4 "TopS Emissive_01_Color-" "(R=5.000000,G=0.000001,B=0.036865,A=1.000000)")
            (Utilities|FlowControl|Sequence
              (:then_0
                (Utilities|FlowControl|Delay 2.0)
                (Timeline|Timeline PlayFromStart)))))))
    (:then_2
      (Cosmetics|WatchCosmeticStat (CustomCharacter|GetAssociatedPlayerPawn) "(TagName=\"StatManager.GainedShieldTimes\")" 0 false
        (:OnStatChanged
          (if (not (Variables|Default|GetKill_Pause))
            (Variables|Default|SetKill_Pause true)
            (Rendering|Material|SetScalarParameterValueonMaterials _returnvalue_8 "EDRec_FlipBook_FrameC" 3.0)
            (Rendering|Material|SetScalarParameterValueonMaterials _returnvalue_8 "EDRec_FlipBook_FrameB" 3.0)
            (Rendering|Material|SetScalarParameterValueonMaterials _returnvalue_8 "EDRec_DecalB_Opacity_End" 1.0)
            (Rendering|Material|SetScalarParameterValueonMaterials _returnvalue_8 "EDRec_DecalD_Opacity_End" 0.0)
            (Rendering|Material|SetScalarParameterValueonMaterials _returnvalue_8 "EDRec_TimeFrom_BP_DecalB" _returnvalue_9)
            (Rendering|Material|SetScalarParameterValueonMaterials _returnvalue_8 "EDRec_TimeFrom_BP_DecalB" _returnvalue_9)
            (Rendering|Material|SetScalarParameterValueonMaterials _returnvalue_8 "TopS Emissive_01_Offset----TBP" _returnvalue_9)
            (Rendering|Material|SetColorParameterValueonMaterials _returnvalue_8 "TopS Emissive_01_Color-" "(R=0.000000,G=3.434453,B=5.000000,A=1.000000)")
            (Utilities|FlowControl|Sequence
              (:then_0
                (Utilities|FlowControl|Delay 2.0)
                (bind (_direction _newtrack_0) (Timeline|Timeline_0 PlayFromStart)
                  (:Update
                    (Rendering|Material|SetScalarParameterValueonMaterials _returnvalue_11 "EDRec_DecalB_Opacity_End" (Math|Float|Lerp 1.0 0.0 _newtrack_0))
                    (Rendering|Material|SetScalarParameterValueonMaterials _returnvalue_11 "EDRec_DecalD_Opacity_End" (Math|Float|Lerp 0.0 1.0 _newtrack_0)))
                  (:Finished
                    (Variables|Default|SetKill_Pause false))))
              (:then_1
                (Utilities|FlowControl|Delay 1.0)
                (Rendering|Material|SetScalarParameterValueonMaterials (CustomItemWrap|GetWrappedMeshComponents) "TopS Emissive_01_Offset----TBP" (Utilities|Time|GetGameTimeinSeconds))))))))
    (:then_3
      (Cosmetics|WatchCosmeticStat (CustomCharacter|GetAssociatedPlayerPawn) "(TagName=\"StatManager.GainedHealthTimes\")" 0 false
        (:OnStatChanged
          (if (not (Variables|Default|GetKill_Pause))
            (Variables|Default|SetKill_Pause true)
            (Rendering|Material|SetScalarParameterValueonMaterials _returnvalue "EDRec_FlipBook_FrameC" 1.0)
            (Rendering|Material|SetScalarParameterValueonMaterials _returnvalue "EDRec_FlipBook_FrameB" 1.0)
            (Rendering|Material|SetScalarParameterValueonMaterials _returnvalue "EDRec_DecalB_Opacity_End" 1.0)
            (Rendering|Material|SetScalarParameterValueonMaterials _returnvalue "EDRec_DecalD_Opacity_End" 0.0)
            (Rendering|Material|SetScalarParameterValueonMaterials _returnvalue "EDRec_TimeFrom_BP_DecalB" _returnvalue_2)
            (Rendering|Material|SetScalarParameterValueonMaterials _returnvalue "EDRec_TimeFrom_BP_DecalB" _returnvalue_2)
            (Rendering|Material|SetScalarParameterValueonMaterials _returnvalue "TopS Emissive_01_Offset----TBP" _returnvalue_2)
            (Rendering|Material|SetColorParameterValueonMaterials _returnvalue "TopS Emissive_01_Color-" "(R=0.334522,G=5.000000,B=0.000000,A=1.000000)")
            (Utilities|FlowControl|Sequence
              (:then_0
                (Utilities|FlowControl|Delay 2.0)
                (bind (_direction_2 _newtrack_0_2) (Timeline|Timeline_1 PlayFromStart)
                  (:Update
                    (Rendering|Material|SetScalarParameterValueonMaterials _returnvalue_3 "EDRec_DecalB_Opacity_End" (Math|Float|Lerp 1.0 0.0 _newtrack_0_2))
                    (Rendering|Material|SetScalarParameterValueonMaterials _returnvalue_3 "EDRec_DecalD_Opacity_End" (Math|Float|Lerp 0.0 1.0 _newtrack_0_2)))
                  (:Finished
                    (Variables|Default|SetKill_Pause false))))
              (:then_1
                (Utilities|FlowControl|Delay 1.0)
                (Rendering|Material|SetScalarParameterValueonMaterials (CustomItemWrap|GetWrappedMeshComponents) "TopS Emissive_01_Offset----TBP" (Utilities|Time|GetGameTimeinSeconds))))))))
    (:then_4
      (Cosmetics|WatchCosmeticStat (CustomCharacter|GetAssociatedPlayerPawn) "(TagName=\"StatManager.DamageReceived\")" 0 false
        (:OnStatChanged
          (if (not (Variables|Default|GetKill_Pause))
            (Variables|Default|SetKill_Pause true)
            (Rendering|Material|SetScalarParameterValueonMaterials _returnvalue_12 "EDRec_FlipBook_FrameD" 2.0)
            (Rendering|Material|SetScalarParameterValueonMaterials _returnvalue_12 "EDRec_DecalD__Sine_TimeFrom_BP" _returnvalue_10)
            (Rendering|Material|SetScalarParameterValueonMaterials _returnvalue_12 "TopS Emissive_01_Offset----TBP" _returnvalue_10)
            (Rendering|Material|SetColorParameterValueonMaterials _returnvalue_12 "TopS Emissive_01_Color-" "(R=5.000000,G=0.081006,B=0.000458,A=1.000000)")
            (Utilities|FlowControl|Delay 0.75)
            (Rendering|Material|SetScalarParameterValueonMaterials (CustomItemWrap|GetWrappedMeshComponents) "EDRec_FlipBook_FrameD" 0.0)
            (Variables|Default|SetKill_Pause false)))))))

(event Custom|Shilds
  (bind _returnvalue (CustomItemWrap|GetWrappedMeshComponents))
  (bind _returnvalue_2 (Utilities|Time|GetGameTimeinSeconds))
  (bind _returnvalue_3 (CustomItemWrap|GetWrappedMeshComponents))
  (if (not (Variables|Default|GetKill_Pause))
    (Variables|Default|SetKill_Pause true)
    (Rendering|Material|SetScalarParameterValueonMaterials _returnvalue "EDRec_FlipBook_FrameC" 3.0)
    (Rendering|Material|SetScalarParameterValueonMaterials _returnvalue "EDRec_FlipBook_FrameB" 3.0)
    (Rendering|Material|SetScalarParameterValueonMaterials _returnvalue "EDRec_DecalB_Opacity_End" 1.0)
    (Rendering|Material|SetScalarParameterValueonMaterials _returnvalue "EDRec_DecalD_Opacity_End" 0.0)
    (Rendering|Material|SetScalarParameterValueonMaterials _returnvalue "EDRec_TimeFrom_BP_DecalB" _returnvalue_2)
    (Rendering|Material|SetScalarParameterValueonMaterials _returnvalue "EDRec_TimeFrom_BP_DecalB" _returnvalue_2)
    (Rendering|Material|SetScalarParameterValueonMaterials _returnvalue "TopS Emissive_01_Offset----TBP" _returnvalue_2)
    (Rendering|Material|SetColorParameterValueonMaterials _returnvalue "TopS Emissive_01_Color-" "(R=0.000000,G=3.434453,B=5.000000,A=1.000000)")
    (Utilities|FlowControl|Sequence
      (:then_0
        (Utilities|FlowControl|Delay 2.0)
        (Timeline|Timeline_0 PlayFromStart))
      (:then_1
        (Utilities|FlowControl|Delay 1.0)
        (Rendering|Material|SetScalarParameterValueonMaterials (CustomItemWrap|GetWrappedMeshComponents) "TopS Emissive_01_Offset----TBP" (Utilities|Time|GetGameTimeinSeconds))))))

(event Custom|Health
  (bind _returnvalue (CustomItemWrap|GetWrappedMeshComponents))
  (bind _returnvalue_2 (Utilities|Time|GetGameTimeinSeconds))
  (bind _returnvalue_3 (CustomItemWrap|GetWrappedMeshComponents))
  (if (not (Variables|Default|GetKill_Pause))
    (Variables|Default|SetKill_Pause true)
    (Rendering|Material|SetScalarParameterValueonMaterials _returnvalue "EDRec_FlipBook_FrameC" 1.0)
    (Rendering|Material|SetScalarParameterValueonMaterials _returnvalue "EDRec_FlipBook_FrameB" 1.0)
    (Rendering|Material|SetScalarParameterValueonMaterials _returnvalue "EDRec_DecalB_Opacity_End" 1.0)
    (Rendering|Material|SetScalarParameterValueonMaterials _returnvalue "EDRec_DecalD_Opacity_End" 0.0)
    (Rendering|Material|SetScalarParameterValueonMaterials _returnvalue "EDRec_TimeFrom_BP_DecalB" _returnvalue_2)
    (Rendering|Material|SetScalarParameterValueonMaterials _returnvalue "EDRec_TimeFrom_BP_DecalB" _returnvalue_2)
    (Rendering|Material|SetScalarParameterValueonMaterials _returnvalue "TopS Emissive_01_Offset----TBP" _returnvalue_2)
    (Rendering|Material|SetColorParameterValueonMaterials _returnvalue "TopS Emissive_01_Color-" "(R=0.334522,G=5.000000,B=0.000000,A=1.000000)")
    (Utilities|FlowControl|Sequence
      (:then_0
        (Utilities|FlowControl|Delay 2.0)
        (Timeline|Timeline_1 PlayFromStart))
      (:then_1
        (Utilities|FlowControl|Delay 1.0)
        (Rendering|Material|SetScalarParameterValueonMaterials (CustomItemWrap|GetWrappedMeshComponents) "TopS Emissive_01_Offset----TBP" (Utilities|Time|GetGameTimeinSeconds))))))

(event Custom|Damage
  (bind _returnvalue (Utilities|Time|GetGameTimeinSeconds))
  (bind _returnvalue_2 (CustomItemWrap|GetWrappedMeshComponents))
  (if (not (Variables|Default|GetKill_Pause))
    (Variables|Default|SetKill_Pause true)
    (Rendering|Material|SetScalarParameterValueonMaterials _returnvalue_2 "EDRec_FlipBook_FrameD" 2.0)
    (Rendering|Material|SetScalarParameterValueonMaterials _returnvalue_2 "EDRec_DecalD__Sine_TimeFrom_BP" _returnvalue)
    (Rendering|Material|SetScalarParameterValueonMaterials _returnvalue_2 "TopS Emissive_01_Offset----TBP" _returnvalue)
    (Rendering|Material|SetColorParameterValueonMaterials _returnvalue_2 "TopS Emissive_01_Color-" "(R=5.000000,G=0.081006,B=0.000458,A=1.000000)")
    (Utilities|FlowControl|Delay 0.75)
    (Rendering|Material|SetScalarParameterValueonMaterials (CustomItemWrap|GetWrappedMeshComponents) "EDRec_FlipBook_FrameD" 0.0)
    (Variables|Default|SetKill_Pause false)))

(fn ConstructionScript)
