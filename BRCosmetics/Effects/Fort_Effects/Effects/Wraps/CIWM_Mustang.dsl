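; Mustang item wrap Blueprint that makes wrapped weapons react to the player's kill count and show their banner icon.
; The AthenaKills stat drives two clamped offset material parameters plus a game-time stamp, with offsets reapplied shortly after weapon changes.
; In the front end, repeated weapon fire fakes kills and cycles the count.
; On begin play and pawn updates, it asynchronously fetches the banner icon and writes it to dynamic material instances.
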(event EventOnWeaponFired
  (bind _returnvalue (CustomItemWrap|GetWrappedMeshComponents))
  (bind _kill_count (Variables|Default|GetKill_Count))
  ; Frontnd reactive
  (if (CustomCosmeticModifier|IsinFrontEnd)
    (bind _output_get (Variables|Default|Setweapon%20fire (+ (Variables|Default|Getweapon%20fire) 1)))
    (if (>= _output_get 5)
      (if (== (Variables|Default|Getweapon%20fire) 7)
        (bind _output_get_2 (Variables|Default|SetKill_Count (+ (Variables|Default|GetKill_Count) 1)))
        (Variables|Default|Setweapon%20fire 0)
        (if (== _output_get_2 16)
          (Variables|Default|SetKill_Count 0))
        (Rendering|Material|SetScalarParameterValueonMaterials _returnvalue "Mustang_KillC_Offset_A" (Math|Conversions|ToFloat(Integer) (Math|Integer|Clamp(Integer) _kill_count 0 8)))
        (Rendering|Material|SetScalarParameterValueonMaterials _returnvalue "Mustang_KillC_Offset_B" (Math|Conversions|ToFloat(Integer) (- (Math|Integer|Clamp(Integer) _kill_count 7 16) 8)))
        (Rendering|Material|SetScalarParameterValueonMaterials _returnvalue "Mustang_Time_BP" (Utilities|Time|GetGameTimeinSeconds))))))

(event EventOnPawnUpdated
  (bind (_icon _primarycolor) (Utilities|GetIconTextureandColorsAsync (CosmeticLoadout|GetBannerIconIdfromLoadout (CosmeticLoadout|GetCosmeticLoadout)) (CosmeticLoadout|GetBannerColorIdfromLoadout (CosmeticLoadout|GetCosmeticLoadout)) "PhysicalBanner")
    (:Finished
      (for _array_element (CustomItemWrap|GetWrappedMeshComponents)
        (for _index (range (Rendering|Material|GetNumMaterials _array_element))
          (bind _returnvalue (Rendering|Material|CreateDynamicMaterialInstance _array_element _index (Rendering|Material|GetMaterial _array_element _index) "None" :!class /Script/Engine.PrimitiveComponent))
          (Rendering|Material|SetTextureParameterValue _returnvalue (Variables|Default|GetBannerTextureParamName) _icon)))
      (CallFunction|OnBannerFullyApplied _icon _primarycolor))))

(event EventBeginPlay
  (Utilities|FlowControl|Delay (Variables|Default|GetInitialDelayBeforeApplying))
  (bind (_icon _primarycolor) (Utilities|GetIconTextureandColorsAsync (CosmeticLoadout|GetBannerIconIdfromLoadout (CosmeticLoadout|GetCosmeticLoadout)) (CosmeticLoadout|GetBannerColorIdfromLoadout (CosmeticLoadout|GetCosmeticLoadout)) "PhysicalBanner")
    (:Finished
      (for _array_element (CustomItemWrap|GetWrappedMeshComponents)
        (for _index (range (Rendering|Material|GetNumMaterials _array_element))
          (bind _returnvalue (Rendering|Material|CreateDynamicMaterialInstance _array_element _index (Rendering|Material|GetMaterial _array_element _index) "None" :!class /Script/Engine.PrimitiveComponent))
          (Rendering|Material|SetTextureParameterValue _returnvalue (Variables|Default|GetBannerTextureParamName) _icon)))
      (CallFunction|OnBannerFullyApplied _icon _primarycolor))))

(event Custom|OnBannerFullyApplied
  (params (Icon :type Texture%202D%20Object%20Reference) (PrimaryColor :type Linear%20Color%20Structure)))

(event EventOnPawnSet
  (bind _returnvalue (CustomItemWrap|GetWrappedMeshComponents))
  (bind _kill_count (Variables|Default|GetKill_Count))
  (bind _asfort_player_pawn (Utilities|Casting|CastToFortPlayerPawn (CustomCharacter|GetAssociatedPlayerPawn)))
  (bind (_statname _statvalue) (Cosmetics|WatchCosmeticStat _asfort_player_pawn "(TagName=\"StatManager.AthenaKills\")")
    (:OnStatChanged
      (Variables|Default|SetKill_Count _statvalue)
      (Rendering|Material|SetScalarParameterValueonMaterials _returnvalue "Mustang_KillC_Offset_A" (Math|Conversions|ToFloat(Integer) (Math|Integer|Clamp(Integer) _kill_count 0 8)))
      (Rendering|Material|SetScalarParameterValueonMaterials _returnvalue "Mustang_KillC_Offset_B" (Math|Conversions|ToFloat(Integer) (- (Math|Integer|Clamp(Integer) _kill_count 7 16) 8)))
      (Rendering|Material|SetScalarParameterValueonMaterials _returnvalue "Mustang_Time_BP" (Utilities|Time|GetGameTimeinSeconds)))))

(event EventOnNewWeapons
  (params (WeaponsAsItems :type Array%20of%20Fort%20Item%20Object%20References))
  (bind _returnvalue (CustomItemWrap|GetWrappedMeshComponents))
  (bind _kill_count (Variables|Default|GetKill_Count))
  ;[1 1 1] Updated kill count on new weapon
  (Utilities|FlowControl|Delay 0.2)
  (Rendering|Material|SetScalarParameterValueonMaterials _returnvalue "Mustang_KillC_Offset_A" (Math|Conversions|ToFloat(Integer) (Math|Integer|Clamp(Integer) _kill_count 0 8)))
  (Rendering|Material|SetScalarParameterValueonMaterials _returnvalue "Mustang_KillC_Offset_B" (Math|Conversions|ToFloat(Integer) (- (Math|Integer|Clamp(Integer) _kill_count 7 16) 8))))

(event EventOnNewPlayerWeaponType
  (params (NewType :type EFortItemType%20Enum :default "WorldItem") (OldType :type EFortItemType%20Enum :default "WorldItem"))
  (bind _returnvalue (CustomItemWrap|GetWrappedMeshComponents))
  (bind _kill_count (Variables|Default|GetKill_Count))
  ;[1 1 1] Updated kill count on new weapon
  (Utilities|FlowControl|Delay 0.2)
  (Rendering|Material|SetScalarParameterValueonMaterials _returnvalue "Mustang_KillC_Offset_A" (Math|Conversions|ToFloat(Integer) (Math|Integer|Clamp(Integer) _kill_count 0 8)))
  (Rendering|Material|SetScalarParameterValueonMaterials _returnvalue "Mustang_KillC_Offset_B" (Math|Conversions|ToFloat(Integer) (- (Math|Integer|Clamp(Integer) _kill_count 7 16) 8))))

(fn ConstructionScript)
