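; Mystic weapon wrap Blueprint that drives a time-based material effect, apparently an eye, when the weapon fires.
; After a very short delay, firing sets the TEye start, end and timestamp scalars on the wrapped meshes, using game time plus an offset so the material handles the animation.
; When the pawn is set, it watches that player's Athena kills stat, with no response to it defined here.
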(event EventOnWeaponFired
  (bind _returnvalue (CustomItemWrap|GetWrappedMeshComponents))
  (Utilities|FlowControl|Delay 0.01)
  (Rendering|Material|SetScalarParameterValueonMaterials _returnvalue "TEye_Time_End" 1.0)
  (Rendering|Material|SetScalarParameterValueonMaterials _returnvalue "TEye_Time_Start" 0.0)
  (Rendering|Material|SetScalarParameterValueonMaterials _returnvalue "TEye_TimeFrom_BP" (+ (Utilities|Time|GetGameTimeinSeconds) 0.25)))

(event EventOnPawnSet
  (bind _asfort_player_pawn (Utilities|Casting|CastToFortPlayerPawn (CustomCharacter|GetAssociatedPlayerPawn)))
  (Cosmetics|WatchCosmeticStat _asfort_player_pawn "(TagName=\"StatManager.AthenaKills\")"))

(event EventBeginPlay
  (|CallParentFunction))

(fn ConstructionScript)
