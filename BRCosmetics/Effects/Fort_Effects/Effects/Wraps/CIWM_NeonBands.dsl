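; Item wrap Blueprint for the Neon Bands weapon wrap that drives its material from weapon fire.
; On each shot, it writes the current game time to a scalar parameter on the wrapped mesh components' materials.
; An iteration counter rotates round-robin through the MainA, MainB and MainC time parameters, so each of the last three shots keeps its own timestamp.
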(event EventOnWeaponFired
  (bind _returnvalue (CustomItemWrap|GetWrappedMeshComponents))
  (bind _returnvalue_2 (Utilities|Time|GetGameTimeinSeconds))
  (bind _returnvalue_3 (+ _returnvalue_2 0.0))
  (switch int (Variables|Default|Getiteration)
    (:0
      (Rendering|Material|SetScalarParameterValueonMaterials _returnvalue "NeonBnds_MainA_TimeFronBP" _returnvalue_3)
      (Variables|Default|Setiteration 1))
    (:1
      (Rendering|Material|SetScalarParameterValueonMaterials _returnvalue "NeonBnds_MainB_TimeFronBP" _returnvalue_3)
      (Variables|Default|Setiteration 2))
    (:2
      (Rendering|Material|SetScalarParameterValueonMaterials _returnvalue "NeonBnds_MainC_TimeFronBP" _returnvalue_3)
      (Variables|Default|Setiteration 0))))

(fn ConstructionScript)
