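; Weapon wrap modifier that flips wrapped-mesh materials between two blend states on kills and reacts to damage dealt with emissive pulses.
; Kill stat changes toggle the invert state, push blend and offset parameters, then pause reactions for a set duration.
; Damage pulses emissive only while inverted; in the front end, weapon-fire counts simulate both triggers.
; Game time is written into "TBP" scalar parameters, timestamping material-driven transitions instead of ticking.
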(event EventOnWeaponFired
  (bind _returnvalue (CustomItemWrap|GetWrappedMeshComponents))
  (bind _offset_start_pos (Variables|Default|GetOffset%20Start%20pos))
  (bind _invert_state_emiss (Variables|EviromentVarables|Getinvert%20state_Emiss))
  ; Frontnd reactive
  (if (CustomCosmeticModifier|IsinFrontEnd)
    (bind _output_get (Variables|EviromentVarables|Setweapon%20fire (+ (Variables|EviromentVarables|Getweapon%20fire) 1)))
    (if (>= _output_get 5)
      (if (== (Variables|EviromentVarables|Getweapon%20fire) 7)
        (Variables|EviromentVarables|Setweapon%20fire 0)
        (CallFunction|DamageUpdate)
        (elif (Variables|EviromentVarables|GetPause)
          _
          (elif (Variables|EviromentVarables|Getinvert_State)
            (Variables|EviromentVarables|Setinvert%20state_Emiss (Variables|Default|Getinvert%20state_Emiss_Start))
            (Rendering|Material|SetScalarParameterValueonMaterials _returnvalue "TopS Emissive_01_Blend 01---> 02 Invert Msk  1=inverted" _invert_state_emiss)
            (Rendering|Material|SetScalarParameterValueonMaterials _returnvalue "TopS Emissive_02_Blend 01---> 02 Invert Msk  1=inverted" _invert_state_emiss)
            (Rendering|Material|SetScalarParameterValueonMaterials _returnvalue "TopS Emissive_03_Blend 01---> 02 Invert Msk  1=inverted" _invert_state_emiss)
            (Rendering|Material|SetScalarParameterValueonMaterials _returnvalue "TopS Emissive_01_Offset---End" _offset_start_pos)
            (Rendering|Material|SetScalarParameterValueonMaterials _returnvalue "TopS Emissive_02_Offset---End" _offset_start_pos)
            (Rendering|Material|SetScalarParameterValueonMaterials _returnvalue "TopS Emissive_03_Offset---End" _offset_start_pos)
            (CallFunction|ReactiveBrightness)
            (else
              (Variables|EviromentVarables|Setinvert%20state_Emiss (Variables|Default|Getinvert%20state_Emiss_End)))))))))

(event EventOnPawnSet
  (bind _returnvalue (CustomItemWrap|GetWrappedMeshComponents))
  (bind _offset_start_pos (Variables|Default|GetOffset%20Start%20pos))
  (bind _invert_state_emiss (Variables|EviromentVarables|Getinvert%20state_Emiss))
  (Utilities|FlowControl|Sequence
    (:then_0
      (Cosmetics|WatchCosmeticStat (CustomCharacter|GetAssociatedPlayerPawn) "(TagName=\"StatManager.DamageDealt\")" 0 false
        (:OnStatChanged
          (if (Variables|EviromentVarables|GetPause)
            _
            (elif (Variables|EviromentVarables|Getinvert_State)
              (Variables|EviromentVarables|Setinvert%20state_Emiss (Variables|Default|Getinvert%20state_Emiss_Start))
              (Rendering|Material|SetScalarParameterValueonMaterials _returnvalue "TopS Emissive_01_Blend 01---> 02 Invert Msk  1=inverted" _invert_state_emiss)
              (Rendering|Material|SetScalarParameterValueonMaterials _returnvalue "TopS Emissive_02_Blend 01---> 02 Invert Msk  1=inverted" _invert_state_emiss)
              (Rendering|Material|SetScalarParameterValueonMaterials _returnvalue "TopS Emissive_03_Blend 01---> 02 Invert Msk  1=inverted" _invert_state_emiss)
              (Rendering|Material|SetScalarParameterValueonMaterials _returnvalue "TopS Emissive_01_Offset---End" _offset_start_pos)
              (Rendering|Material|SetScalarParameterValueonMaterials _returnvalue "TopS Emissive_02_Offset---End" _offset_start_pos)
              (Rendering|Material|SetScalarParameterValueonMaterials _returnvalue "TopS Emissive_03_Offset---End" _offset_start_pos)
              (CallFunction|ReactiveBrightness)
              (else
                (Variables|EviromentVarables|Setinvert%20state_Emiss (Variables|Default|Getinvert%20state_Emiss_End))))))))
    (:then_1
      (Cosmetics|WatchCosmeticStat (CustomCharacter|GetAssociatedPlayerPawn) "(TagName=\"StatManager.AthenaKills\")" 0 false
        (:OnStatChanged
          (CallFunction|DamageUpdate))))))

(event Custom|Damage_update
  (bind _returnvalue (CustomItemWrap|GetWrappedMeshComponents))
  (if (not (Variables|EviromentVarables|GetPause_B))
    (Variables|EviromentVarables|SetPause_B true)
    (if (Variables|EviromentVarables|Getinvert_State)
      (Variables|EviromentVarables|Setinvert_State false)
      (Variables|EviromentVarables|Setinvert%20state (Variables|Default|Getinvert%20state_Blend_Start))
      (Variables|EviromentVarables|Setinvert%20state_Emiss (Variables|Default|Getinvert%20state_Emiss_Start))
      (else
        (Variables|EviromentVarables|Setinvert_State true)
        (Variables|EviromentVarables|Setinvert%20state (Variables|Default|Getinvert%20state_Blend_End))
        (Variables|EviromentVarables|Setinvert%20state_Emiss (Variables|Default|Getinvert%20state_Emiss_End))))
    (Variables|EviromentVarables|SetPause true)
    (Rendering|Material|SetScalarParameterValueonMaterials _returnvalue "TopS Blend-InvertMsk   1=invert" (Variables|EviromentVarables|Getinvert%20state))
    (Rendering|Material|SetScalarParameterValueonMaterials _returnvalue "TopS Emissive_01_Blend 01---> 02 Invert Msk  1=inverted" (Variables|EviromentVarables|Getinvert%20state_Emiss))
    (Rendering|Material|SetScalarParameterValueonMaterials _returnvalue "TopS Emissive_02_Blend 01---> 02 Invert Msk  1=inverted" (Variables|EviromentVarables|Getinvert%20state_Emiss))
    (Rendering|Material|SetScalarParameterValueonMaterials _returnvalue "TopS Emissive_03_Blend 01---> 02 Invert Msk  1=inverted" (Variables|EviromentVarables|Getinvert%20state_Emiss))
    (CallFunction|UpdateMaterial)
    (CallFunction|ReactiveBrightness)
    (Utilities|FlowControl|Delay (Variables|Default|GetPause%20Duration))
    (Variables|EviromentVarables|SetPause false)
    (Variables|EviromentVarables|SetPause_B false)))

(event EventOnNewWeapons
  (params (WeaponsAsItems :type Array%20of%20Fort%20Item%20Object%20References))
  (Utilities|FlowControl|Delay 0.05))

(event EventOnNewPlayerWeaponType
  (params (NewType :type EFortItemType%20Enum :default "WorldItem") (OldType :type EFortItemType%20Enum :default "WorldItem"))
  (Utilities|FlowControl|Delay 0.05))

(event EventOnWeaponChanged
  (params (NewWeapon :type Fort%20Weapon%20Object%20Reference) (OldWeapon :type Fort%20Weapon%20Object%20Reference))
  (Utilities|FlowControl|Delay 0.05))

(fn ReactiveBrightness
  (bind _returnvalue (CustomItemWrap|GetWrappedMeshComponents))
  (bind _returnvalue_2 (Utilities|Time|GetGameTimeinSeconds))
  (Rendering|Material|SetScalarParameterValueonMaterials _returnvalue "TopS Emissive_01-Brightness-TBP" _returnvalue_2)
  (Rendering|Material|SetScalarParameterValueonMaterials _returnvalue "TopS Emissive_02-Brightness-TBP" _returnvalue_2)
  (Rendering|Material|SetScalarParameterValueonMaterials _returnvalue "TopS Emissive_03-Brightness-TBP" _returnvalue_2))

(fn UpdateMaterial
  (bind _returnvalue (CustomItemWrap|GetWrappedMeshComponents))
  (bind _returnvalue_2 (Utilities|Time|GetGameTimeinSeconds))
  (bind _offset_end_pos (Variables|Default|GetOffset%20End%20pos))
  (bind _offset_start_pos (Variables|Default|GetOffset%20Start%20pos))
  (Rendering|Material|SetScalarParameterValueonMaterials _returnvalue "TopS_01_Blend_Offset---End" _offset_end_pos)
  (Rendering|Material|SetScalarParameterValueonMaterials _returnvalue "TopS_01_Blend_Offset---Start" _offset_start_pos)
  (Rendering|Material|SetScalarParameterValueonMaterials _returnvalue "TopS_01_Blend_Offset-TBP" _returnvalue_2)
  (Rendering|Material|SetScalarParameterValueonMaterials _returnvalue "TopS Emissive_01_Offset---End" _offset_end_pos)
  (Rendering|Material|SetScalarParameterValueonMaterials _returnvalue "TopS Emissive_01_Offset---Start" _offset_start_pos)
  (Rendering|Material|SetScalarParameterValueonMaterials _returnvalue "TopS Emissive_01_Offset----TBP" _returnvalue_2)
  (Rendering|Material|SetScalarParameterValueonMaterials _returnvalue "TopS Emissive_02_Offset---End" _offset_end_pos)
  (Rendering|Material|SetScalarParameterValueonMaterials _returnvalue "TopS Emissive_02_Offset---Start" _offset_start_pos)
  (Rendering|Material|SetScalarParameterValueonMaterials _returnvalue "TopS Emissive_02_Offset----TBP" _returnvalue_2)
  (Rendering|Material|SetScalarParameterValueonMaterials _returnvalue "TopS Emissive_03_Offset---End" _offset_end_pos)
  (Rendering|Material|SetScalarParameterValueonMaterials _returnvalue "TopS Emissive_03_Offset---Start" _offset_start_pos)
  (Rendering|Material|SetScalarParameterValueonMaterials _returnvalue "TopS Emissive_03_Offset----TBP" _returnvalue_2))

(fn ConstructionScript)
