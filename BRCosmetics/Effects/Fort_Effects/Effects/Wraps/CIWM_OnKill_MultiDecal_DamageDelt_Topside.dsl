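; Weapon wrap cosmetic Blueprint that triggers time-driven material effects on wrapped meshes when the player deals damage or gets kills.
; Pawn-set watches the DamageDealt and AthenaKills stats, writing game time into the topside speed parameter or four decal/emissive parameter pairs.
; In the front end, a cycling shot counter previews both effects on weapon fire.
; A RandomSet function randomizes decal offset, scale and rotation, but nothing here calls it.
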(event EventBeginPlay
  (Utilities|FlowControl|Delay 0.2))

(event EventOnWeaponFired
  (bind _returnvalue (CustomItemWrap|GetWrappedMeshComponents))
  (bind _returnvalue_2 (Utilities|Time|GetGameTimeinSeconds))
  (if (CustomCosmeticModifier|IsinFrontEnd)
    (Utilities|FlowControl|Delay 0.001)
    (Utilities|FlowControl|Sequence
      (:then_1
        (bind _output_get (Variables|Default|SetShotsFired (+ (Variables|Default|GetShotsFired) 1)))
        (if (>= _output_get 5)
          (if (== (Variables|Default|GetShotsFired) 7)
            (Variables|Default|SetShotsFired 0)
            (Rendering|Material|SetScalarParameterValueonMaterials _returnvalue "EDRec_TimeFrom_BP_DecalA" _returnvalue_2)
            (Rendering|Material|SetScalarParameterValueonMaterials _returnvalue "EDRec_TimeFrom_BP_DecalB" _returnvalue_2)
            (Rendering|Material|SetScalarParameterValueonMaterials _returnvalue "EDRec_Emissive_A_Brightness_Time_BP" _returnvalue_2)
            (Rendering|Material|SetScalarParameterValueonMaterials _returnvalue "EDRec_Emissive_B_Brightness_Time_BP" _returnvalue_2)
            (Rendering|Material|SetScalarParameterValueonMaterials _returnvalue "EDRec_TimeFrom_BP_DecalC" _returnvalue_2)
            (Rendering|Material|SetScalarParameterValueonMaterials _returnvalue "EDRec_TimeFrom_BP_DecalD" _returnvalue_2)
            (Rendering|Material|SetScalarParameterValueonMaterials _returnvalue "EDRec_Emissive_C_Brightness_Time_BP" _returnvalue_2)
            (Rendering|Material|SetScalarParameterValueonMaterials _returnvalue "EDRec_Emissive_D_Brightness_Time_BP" _returnvalue_2)
            (else
              (Rendering|Material|SetScalarParameterValueonMaterials (CustomItemWrap|GetWrappedMeshComponents) "TopS_01_Speed-TBP" (Utilities|Time|GetGameTimeinSeconds)))))))))

(event EventOnPawnSet
  (bind _returnvalue (CustomItemWrap|GetWrappedMeshComponents))
  (bind _returnvalue_2 (Utilities|Time|GetGameTimeinSeconds))
  (Utilities|FlowControl|Sequence
    (:then_0
      (Cosmetics|WatchCosmeticStat (CustomCharacter|GetAssociatedPlayerPawn) "(TagName=\"StatManager.DamageDealt\")" 0 false
        (:OnStatChanged
          (Rendering|Material|SetScalarParameterValueonMaterials (CustomItemWrap|GetWrappedMeshComponents) "TopS_01_Speed-TBP" (Utilities|Time|GetGameTimeinSeconds)))))
    (:then_1
      (Cosmetics|WatchCosmeticStat (CustomCharacter|GetAssociatedPlayerPawn) "(TagName=\"StatManager.AthenaKills\")" 0 false
        (:OnStatChanged
          (Rendering|Material|SetScalarParameterValueonMaterials _returnvalue "EDRec_TimeFrom_BP_DecalA" _returnvalue_2)
          (Rendering|Material|SetScalarParameterValueonMaterials _returnvalue "EDRec_TimeFrom_BP_DecalB" _returnvalue_2)
          (Rendering|Material|SetScalarParameterValueonMaterials _returnvalue "EDRec_Emissive_A_Brightness_Time_BP" _returnvalue_2)
          (Rendering|Material|SetScalarParameterValueonMaterials _returnvalue "EDRec_Emissive_B_Brightness_Time_BP" _returnvalue_2)
          (Rendering|Material|SetScalarParameterValueonMaterials _returnvalue "EDRec_TimeFrom_BP_DecalC" _returnvalue_2)
          (Rendering|Material|SetScalarParameterValueonMaterials _returnvalue "EDRec_TimeFrom_BP_DecalD" _returnvalue_2)
          (Rendering|Material|SetScalarParameterValueonMaterials _returnvalue "EDRec_Emissive_C_Brightness_Time_BP" _returnvalue_2)
          (Rendering|Material|SetScalarParameterValueonMaterials _returnvalue "EDRec_Emissive_D_Brightness_Time_BP" _returnvalue_2))))))

(fn RandomSet
  (params (TimeBP_Flipbook%20Triger :type Name) (Offset :type Name) (Rotation :type Name) (X=min :type Float%20%28double-precision%29) (X=Max :type Float%20%28double-precision%29) (Y=Min :type Float%20%28double-precision%29) (Y=Max :type Float%20%28double-precision%29) (Scale=Min :type Float%20%28double-precision%29) (Scale=Max :type Float%20%28double-precision%29))
  (bind _returnvalue (CustomItemWrap|GetWrappedMeshComponents))
  (bind _stream (Variables|Default|GetStream))
  (bind _returnvalue_2 (Math|Random|RandomFloatinRangefromStream _stream X=min X=Max))
  (Rendering|Material|SetScalarParameterValueonMaterials _returnvalue TimeBP_Flipbook%20Triger (Utilities|Time|GetGameTimeinSeconds))
  (Rendering|Material|SetColorParameterValueonMaterials _returnvalue Offset (Math|Color|MakeColor _returnvalue_2 (Math|Random|RandomFloatinRangefromStream _stream Y=Min Y=Max) 0.0 (Math|Random|RandomFloatinRangefromStream _stream Scale=Min Scale=Max)))
  (Rendering|Material|SetScalarParameterValueonMaterials _returnvalue Rotation (Math|Random|RandomFloat))
  (Variables|Default|SetPosition _returnvalue_2))

(fn ConstructionScript)
