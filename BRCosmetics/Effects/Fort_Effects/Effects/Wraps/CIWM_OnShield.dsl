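; Item wrap Blueprint that triggers time-based material effects on wrapped meshes when the player gains shield.
; Pawn, res-in and weapon-change events watch the GainedShieldTimes stat, while a timeline drives the same logic in the frontend.
; Each trigger stamps game time into named scalar parameters, plus an offset for a delayed set, gated by a pause flag.
; An optional Timmer event repeats the stamp on a timer with its own offset.
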(event EventBeginPlay
  (if (Class|BlueprintLibrarySTW|IsinFrontend)
    (Timeline|Timeline Play
      (:NewTrack_0
        (Utilities|FlowControl|Sequence
          (:then_0
            (Variables|EnvirometnVarable|SetDelayB (Variables|Default|Shields|GetBP_Delay_On%20Gain%20))
            (if (not (Variables|EnvirometnVarable|GetShield%20Pause%20delay))
              (Variables|EnvirometnVarable|SetShield%20Pause%20delay true)
              (Utilities|FlowControl|Sequence
                (:then_0
                  (for _array_element (Variables|Default|Shields|GetShields%20Names%20BP)
                    (Rendering|Material|SetScalarParameterValueonMaterials (CustomItemWrap|GetWrappedMeshComponents) _array_element (Utilities|Time|GetGameTimeinSeconds)))
                  (Utilities|FlowControl|Delay (Variables|Default|Shields|GetShield%20Pause))
                  (Variables|EnvirometnVarable|SetShield%20Pause%20delay false)
                  (if (Variables|Default|Shields|GetShield%20Timer)
                    (Utilities|Time|SetTimerbyEvent (delegate Custom|Timmer) (Variables|Default|Shields|GetShield%20Timer%20Sec) true)
                    (CallFunction|Timmer)))
                (:then_1
                  (for _array_element_2 (Variables|Default|Shields|GetShields%20Names%20BP_Delay)
                    (Rendering|Material|SetScalarParameterValueonMaterials (CustomItemWrap|GetWrappedMeshComponents) _array_element_2 (+ (Utilities|Time|GetGameTimeinSeconds) (Variables|EnvirometnVarable|GetDelayB))))))))
          (:then_1
            (if (Variables|Default|Shields|GetShield%20Timer)
              (Utilities|Time|SetTimerbyEvent (delegate Custom|Timmer))
              (CallFunction|Timmer)))
          (:then_2
            (Utilities|FlowControl|Delay 0.2)))))))

(event EventOnPawnSet
  (Cosmetics|WatchCosmeticStat (CustomCharacter|GetAssociatedPlayerPawn) "(TagName=\"StatManager.GainedShieldTimes\")" 0 false -1.0 -1.0 false
    (:OnStatChanged
      (Utilities|FlowControl|Sequence
        (:then_0
          (Variables|EnvirometnVarable|SetDelayB (Variables|Default|Shields|GetBP_Delay_On%20Gain%20))
          (if (not (Variables|EnvirometnVarable|GetShield%20Pause%20delay))
            (Variables|EnvirometnVarable|SetShield%20Pause%20delay true)
            (Utilities|FlowControl|Sequence
              (:then_0
                (for _array_element (Variables|Default|Shields|GetShields%20Names%20BP)
                  (Rendering|Material|SetScalarParameterValueonMaterials (CustomItemWrap|GetWrappedMeshComponents) _array_element (Utilities|Time|GetGameTimeinSeconds)))
                (Utilities|FlowControl|Delay (Variables|Default|Shields|GetShield%20Pause))
                (Variables|EnvirometnVarable|SetShield%20Pause%20delay false)
                (if (Variables|Default|Shields|GetShield%20Timer)
                  (Utilities|Time|SetTimerbyEvent (delegate Custom|Timmer) (Variables|Default|Shields|GetShield%20Timer%20Sec) true)
                  (CallFunction|Timmer)))
              (:then_1
                (for _array_element_2 (Variables|Default|Shields|GetShields%20Names%20BP_Delay)
                  (Rendering|Material|SetScalarParameterValueonMaterials (CustomItemWrap|GetWrappedMeshComponents) _array_element_2 (+ (Utilities|Time|GetGameTimeinSeconds) (Variables|EnvirometnVarable|GetDelayB))))))))
        (:then_1
          (if (Variables|Default|Shields|GetShield%20Timer)
            (Utilities|Time|SetTimerbyEvent (delegate Custom|Timmer))
            (CallFunction|Timmer)))
        (:then_2
          (Utilities|FlowControl|Delay 0.2))))))

(event EventOnPostResIn
  (Utilities|FlowControl|Delay 0.1)
  (Cosmetics|WatchCosmeticStat (CustomCharacter|GetAssociatedPlayerPawn) "(TagName=\"StatManager.GainedShieldTimes\")" 0 false -1.0 -1.0 false
    (:OnStatChanged
      (Utilities|FlowControl|Sequence
        (:then_0
          (Variables|EnvirometnVarable|SetDelayB (Variables|Default|Shields|GetBP_Delay_On%20Gain%20))
          (if (not (Variables|EnvirometnVarable|GetShield%20Pause%20delay))
            (Variables|EnvirometnVarable|SetShield%20Pause%20delay true)
            (Utilities|FlowControl|Sequence
              (:then_0
                (for _array_element (Variables|Default|Shields|GetShields%20Names%20BP)
                  (Rendering|Material|SetScalarParameterValueonMaterials (CustomItemWrap|GetWrappedMeshComponents) _array_element (Utilities|Time|GetGameTimeinSeconds)))
                (Utilities|FlowControl|Delay (Variables|Default|Shields|GetShield%20Pause))
                (Variables|EnvirometnVarable|SetShield%20Pause%20delay false)
                (if (Variables|Default|Shields|GetShield%20Timer)
                  (Utilities|Time|SetTimerbyEvent (delegate Custom|Timmer) (Variables|Default|Shields|GetShield%20Timer%20Sec) true)
                  (CallFunction|Timmer)))
              (:then_1
                (for _array_element_2 (Variables|Default|Shields|GetShields%20Names%20BP_Delay)
                  (Rendering|Material|SetScalarParameterValueonMaterials (CustomItemWrap|GetWrappedMeshComponents) _array_element_2 (+ (Utilities|Time|GetGameTimeinSeconds) (Variables|EnvirometnVarable|GetDelayB))))))))
        (:then_1
          (if (Variables|Default|Shields|GetShield%20Timer)
            (Utilities|Time|SetTimerbyEvent (delegate Custom|Timmer))
            (CallFunction|Timmer)))
        (:then_2
          (Utilities|FlowControl|Delay 0.2))))))

(event EventOnNewWeapons
  (params (WeaponsAsItems :type Array%20of%20Fort%20Item%20Object%20References))
  (Utilities|FlowControl|Delay 0.1)
  (Cosmetics|WatchCosmeticStat (CustomCharacter|GetAssociatedPlayerPawn) "(TagName=\"StatManager.GainedShieldTimes\")" 0 false -1.0 -1.0 false
    (:OnStatChanged
      (Utilities|FlowControl|Sequence
        (:then_0
          (Variables|EnvirometnVarable|SetDelayB (Variables|Default|Shields|GetBP_Delay_On%20Gain%20))
          (if (not (Variables|EnvirometnVarable|GetShield%20Pause%20delay))
            (Variables|EnvirometnVarable|SetShield%20Pause%20delay true)
            (Utilities|FlowControl|Sequence
              (:then_0
                (for _array_element (Variables|Default|Shields|GetShields%20Names%20BP)
                  (Rendering|Material|SetScalarParameterValueonMaterials (CustomItemWrap|GetWrappedMeshComponents) _array_element (Utilities|Time|GetGameTimeinSeconds)))
                (Utilities|FlowControl|Delay (Variables|Default|Shields|GetShield%20Pause))
                (Variables|EnvirometnVarable|SetShield%20Pause%20delay false)
                (if (Variables|Default|Shields|GetShield%20Timer)
                  (Utilities|Time|SetTimerbyEvent (delegate Custom|Timmer) (Variables|Default|Shields|GetShield%20Timer%20Sec) true)
                  (CallFunction|Timmer)))
              (:then_1
                (for _array_element_2 (Variables|Default|Shields|GetShields%20Names%20BP_Delay)
                  (Rendering|Material|SetScalarParameterValueonMaterials (CustomItemWrap|GetWrappedMeshComponents) _array_element_2 (+ (Utilities|Time|GetGameTimeinSeconds) (Variables|EnvirometnVarable|GetDelayB))))))))
        (:then_1
          (if (Variables|Default|Shields|GetShield%20Timer)
            (Utilities|Time|SetTimerbyEvent (delegate Custom|Timmer))
            (CallFunction|Timmer)))
        (:then_2
          (Utilities|FlowControl|Delay 0.2))))))

(event EventOnNewPlayerWeaponType
  (params (NewType :type EFortItemType%20Enum :default "WorldItem") (OldType :type EFortItemType%20Enum :default "WorldItem"))
  (Utilities|FlowControl|Delay 0.1)
  (Cosmetics|WatchCosmeticStat (CustomCharacter|GetAssociatedPlayerPawn) "(TagName=\"StatManager.GainedShieldTimes\")" 0 false -1.0 -1.0 false
    (:OnStatChanged
      (Utilities|FlowControl|Sequence
        (:then_0
          (Variables|EnvirometnVarable|SetDelayB (Variables|Default|Shields|GetBP_Delay_On%20Gain%20))
          (if (not (Variables|EnvirometnVarable|GetShield%20Pause%20delay))
            (Variables|EnvirometnVarable|SetShield%20Pause%20delay true)
            (Utilities|FlowControl|Sequence
              (:then_0
                (for _array_element (Variables|Default|Shields|GetShields%20Names%20BP)
                  (Rendering|Material|SetScalarParameterValueonMaterials (CustomItemWrap|GetWrappedMeshComponents) _array_element (Utilities|Time|GetGameTimeinSeconds)))
                (Utilities|FlowControl|Delay (Variables|Default|Shields|GetShield%20Pause))
                (Variables|EnvirometnVarable|SetShield%20Pause%20delay false)
                (if (Variables|Default|Shields|GetShield%20Timer)
                  (Utilities|Time|SetTimerbyEvent (delegate Custom|Timmer) (Variables|Default|Shields|GetShield%20Timer%20Sec) true)
                  (CallFunction|Timmer)))
              (:then_1
                (for _array_element_2 (Variables|Default|Shields|GetShields%20Names%20BP_Delay)
                  (Rendering|Material|SetScalarParameterValueonMaterials (CustomItemWrap|GetWrappedMeshComponents) _array_element_2 (+ (Utilities|Time|GetGameTimeinSeconds) (Variables|EnvirometnVarable|GetDelayB))))))))
        (:then_1
          (if (Variables|Default|Shields|GetShield%20Timer)
            (Utilities|Time|SetTimerbyEvent (delegate Custom|Timmer))
            (CallFunction|Timmer)))
        (:then_2
          (Utilities|FlowControl|Delay 0.2))))))

(event Custom|Timmer
  (if (Variables|Default|Shields|GetShield%20Timer)
    (if true
      (if (not (Variables|EnvirometnVarable|GetShield%20Pause%20delay))
        (Variables|EnvirometnVarable|SetDelayB (Variables|Default|Shields|GetBP_Delay_On%20Timer))
        (Utilities|FlowControl|Sequence
          (:then_0
            (for _array_element (Variables|Default|Shields|GetShields%20Names%20BP)
              (Rendering|Material|SetScalarParameterValueonMaterials (CustomItemWrap|GetWrappedMeshComponents) _array_element (Utilities|Time|GetGameTimeinSeconds))))
          (:then_1
            (for _array_element_2 (Variables|Default|Shields|GetShields%20Names%20BP_Delay)
              (Rendering|Material|SetScalarParameterValueonMaterials (CustomItemWrap|GetWrappedMeshComponents) _array_element_2 (+ (Utilities|Time|GetGameTimeinSeconds) (Variables|EnvirometnVarable|GetDelayB))))))))))

(fn ConstructionScript)
