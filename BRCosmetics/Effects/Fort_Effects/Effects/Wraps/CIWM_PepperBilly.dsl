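; Custom item wrap Blueprint for PepperBilly that animates the wrap's material when the weapon fires.
; A slowed timeline starts on a shot, with a Pause flag blocking restarts until it finishes; shots after its track value reaches one rewind it to a set time.
; Updates lerp offset and rotation parameters on wrapped meshes, covering weapon, pickaxe, backpack, character and vehicle variants.
; A Kill event stops the timeline.
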(event EventBeginPlay
  (Components|Timeline|SetPlayRate (Variables|CIWM_PepperBilly|GetTimeline) 0.4))

(event EventOnPawnSet
  (Components|Timeline|SetPlayRate (Variables|CIWM_PepperBilly|GetTimeline) 0.4))

(event EventOnWeaponFired
  (if (Variables|Default|GetPause)
    (if (>= (Variables|Default|GetFloat_Value) 1.0)
      (bind (_direction _newtrack_0 _rotation) (Timeline|Timeline SetNewTime :NewTime 0.75)
        (:Update
          (Variables|Default|SetFloat_Value _newtrack_0)
          (for _array_element (Utilities|Array|MakeArray (Utilities|Name|MakeLiteralName "TopS_04_Offset_Weapon") (Utilities|Name|MakeLiteralName "TopS_04_Offset_Backpack") (Utilities|Name|MakeLiteralName "TopS_04_Offset_Pickaxe") (Utilities|Name|MakeLiteralName "TopS_04_Offset_Character") (Utilities|Name|MakeLiteralName "TopS_04_Offset_VehicleSide") (Utilities|Name|MakeLiteralName "TopS_04_Offset_Vehicle"))
            (Rendering|Material|SetVectorParameterValueonMaterials (CustomItemWrap|GetWrappedMeshComponents) _array_element (Math|Vector|MakeVector 0.0 (Math|Float|Lerp 0.6 -0.6 _newtrack_0))))
          (for _array_element_2 (Utilities|Array|MakeArray (Utilities|Name|MakeLiteralName "TopS_04_Side_Rotation_Mech\\Character\\Backpack") (Utilities|Name|MakeLiteralName "TopS_04_Side_Rotation_Vehicle") (Utilities|Name|MakeLiteralName "TopS_04_Side_Rotation_Vehicle_Side") (Utilities|Name|MakeLiteralName "TopS_04_Side_Rotation_Weapon\\Pickaxe"))
            (Rendering|Material|SetScalarParameterValueonMaterials (CustomItemWrap|GetWrappedMeshComponents) _array_element_2 (Math|Float|Lerp 0.0 0.5 _rotation))))
        (:Finished
          (Variables|Default|SetPause false))))
    (else
      (Variables|Default|SetPause true)
      (Timeline|Timeline PlayFromStart))))

(event Custom|Kill
  (Timeline|Timeline Stop))

(fn ConstructionScript)
