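; Reactive weapon wrap that charges an emissive meter from the player's damage dealt, with a timeline simulating damage in the front end.
; A DamageDealt stat watcher accumulates deltas, normalizes them and writes timestamped material parameters to wrapped meshes, brightening electricity effects near full charge.
; At full charge, reloading, or a delay on clipless weapons, fires an electricity release and resets the meter.
; Weapon and spawn events rebind the reload delegate and re-push parameters.
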(event EventBeginPlay
  (if (CustomCosmeticModifier|IsinFrontEnd)
    (Timeline|Timeline PlayFromStart
      (:NewTrack_0
        (Variables|Default|SetDamage_Build (+ (Variables|Default|GetDamage_Build) 25))
        (Variables|Default|SetLastKillTime (Utilities|Time|GetGameTimeinSeconds))
        (CallFunction|PushMaterialParameters)
        (if (== (FortWeapon|GetBulletsPerClip (Variables|/Script/FortniteGame.FortPawn|GetCurrentWeapon (CustomCharacter|GetAssociatedPlayerPawn))) 0)
          (if (>= (Variables|Default|GetBar%20Status) 1.0)
            (Utilities|FlowControl|Delay 2.5)
            (CallFunction|ReloadFullMeeter)))))))

(event EventOnPawnSet
  (Utilities|FlowControl|Sequence
    (:then_0
      ;[1 1 1] Updated kill count on new weapon
      (Utilities|FlowControl|Delay 0.05)
      (Variables|Default|SetLastKillTime -999999999999.0)
      (Variables|Default|SetDamage_Build (Variables|Default|GetDamage_Build))
      (EventDispatchers|BindEventtoOnReloadStarted (Variables|/Script/FortniteGame.FortPawn|GetCurrentWeapon (CustomCharacter|GetAssociatedPlayerPawn)) (delegate Custom|ReloadStarted))
      (CallFunction|PushMaterialParameters))
    (:then_1
      (bind (_statname _statvalue) (Cosmetics|WatchCosmeticStat (CustomCharacter|GetAssociatedPlayerPawn) "(TagName=\"StatManager.DamageDealt\")" 0 false)
        (:OnStatChanged
          (Utilities|FlowControl|Delay 0.05)
          (bind _output_get (Variables|Default|SetDamage_Current (- _statvalue (Variables|Default|GetDamage_Last))))
          (Variables|Default|SetDamage_Last _statvalue)
          (Variables|Default|SetDamage_Build (+ _output_get (Variables|Default|GetDamage_Build)))
          (Variables|Default|SetLastKillTime (Utilities|Time|GetGameTimeinSeconds))
          (CallFunction|PushMaterialParameters)
          (if (== (FortWeapon|GetBulletsPerClip (Variables|/Script/FortniteGame.FortPawn|GetCurrentWeapon (CustomCharacter|GetAssociatedPlayerPawn))) 0)
            (if (>= (Variables|Default|GetBar%20Status) 1.0)
              (Utilities|FlowControl|Delay 2.5)
              (CallFunction|ReloadFullMeeter))))))))

(event Custom|Push%20Material%20Parameters
  (bind _damage_build (Variables|Default|GetDamage_Build))
  (bind _returnvalue (Math|Conversions|ToFloat(Integer) _damage_build))
  (bind _returnvalue_2 (/ _returnvalue 2e+02))
  (bind _returnvalue_3 (Math|Float|Clamp(Float) _returnvalue_2))
  (bind _returnvalue_4 (CustomItemWrap|GetWrappedMeshComponents))
  (Variables|Default|SetBar%20Status _returnvalue_3)
  (Rendering|Material|SetScalarParameterValueonMaterials _returnvalue_4 "TopS Emissive_01_Offset---Start" 32.0)
  (Rendering|Material|SetScalarParameterValueonMaterials _returnvalue_4 "TopS Emissive_01_Offset---End" (Math|Float|Lerp -0.9 0.35 _returnvalue_3))
  (Rendering|Material|SetScalarParameterValueonMaterials _returnvalue_4 "TopS Emissive_01-Brightness-TBP" (Variables|Default|GetLastKillTime))
  (if (>= _returnvalue_3 0.75)
    (CallFunction|EletrictyBuildUp)))

(event EventOnPostResIn
  ;[1 1 1] Updated kill count on new weapon
  (Utilities|FlowControl|Delay 0.05)
  (Variables|Default|SetLastKillTime -999999999999.0)
  (Variables|Default|SetDamage_Build (Variables|Default|GetDamage_Build))
  (EventDispatchers|BindEventtoOnReloadStarted (Variables|/Script/FortniteGame.FortPawn|GetCurrentWeapon (CustomCharacter|GetAssociatedPlayerPawn)) (delegate Custom|ReloadStarted))
  (CallFunction|PushMaterialParameters))

(event EventOnNewWeapons
  (params (WeaponsAsItems :type Array%20of%20Fort%20Item%20Object%20References))
  ;[1 1 1] Updated kill count on new weapon
  (Utilities|FlowControl|Delay 0.05)
  (Variables|Default|SetLastKillTime -999999999999.0)
  (Variables|Default|SetDamage_Build (Variables|Default|GetDamage_Build))
  (EventDispatchers|BindEventtoOnReloadStarted (Variables|/Script/FortniteGame.FortPawn|GetCurrentWeapon (CustomCharacter|GetAssociatedPlayerPawn)) (delegate Custom|ReloadStarted))
  (CallFunction|PushMaterialParameters))

(event EventOnNewPlayerWeaponType
  (params (NewType :type EFortItemType%20Enum :default "WorldItem") (OldType :type EFortItemType%20Enum :default "WorldItem"))
  ;[1 1 1] Updated kill count on new weapon
  (Utilities|FlowControl|Delay 0.05)
  (Variables|Default|SetLastKillTime -999999999999.0)
  (Variables|Default|SetDamage_Build (Variables|Default|GetDamage_Build))
  (EventDispatchers|BindEventtoOnReloadStarted (Variables|/Script/FortniteGame.FortPawn|GetCurrentWeapon (CustomCharacter|GetAssociatedPlayerPawn)) (delegate Custom|ReloadStarted))
  (CallFunction|PushMaterialParameters))

(event EventOnWeaponChanged
  (params (NewWeapon :type Fort%20Weapon%20Object%20Reference) (OldWeapon :type Fort%20Weapon%20Object%20Reference))
  ;[1 1 1] Updated kill count on new weapon
  (Utilities|FlowControl|Delay 0.05)
  (Variables|Default|SetLastKillTime -999999999999.0)
  (Variables|Default|SetDamage_Build (Variables|Default|GetDamage_Build))
  (EventDispatchers|BindEventtoOnReloadStarted (Variables|/Script/FortniteGame.FortPawn|GetCurrentWeapon (CustomCharacter|GetAssociatedPlayerPawn)) (delegate Custom|ReloadStarted))
  (CallFunction|PushMaterialParameters))

(event Custom|ReloadStarted
  (params (ReloadTime :type Float%20%28single-precision%29) (ReloadType :type EFortWeaponReloadType%20Enum :default "ReloadWholeClip"))
  (if (>= (Variables|Default|GetBar%20Status) 1.0)
    (Variables|Default|SetReload%20Time ReloadTime)
    (CallFunction|ReloadFullMeeter)))

(event Custom|Reload_FullMeeter
  (bind _returnvalue (CustomItemWrap|GetWrappedMeshComponents))
  (Utilities|FlowControl|Delay (Variables|Default|GetReload%20Time))
  (CallFunction|EletricityRelease)
  (Utilities|FlowControl|Delay 0.2)
  (Rendering|Material|SetScalarParameterValueonMaterials _returnvalue "TopS Emissive_01-Brightness-TBP" (Utilities|Time|GetGameTimeinSeconds))
  (Rendering|Material|SetScalarParameterValueonMaterials _returnvalue "TopS Emissive_01_Offset---Start" 0.35)
  (Rendering|Material|SetScalarParameterValueonMaterials _returnvalue "TopS Emissive_01_Offset---End" -0.9)
  (Rendering|Material|SetScalarParameterValueonMaterials _returnvalue "TopS Emissive_01_Offset----TBP" (Utilities|Time|GetGameTimeinSeconds))
  (Variables|Default|SetDamage_Build 0)
  (Variables|Default|SetBar%20Status 0.0))

(event EventApplyMaterialParameters
  ;[1 1 1] Updated kill count on new weapon
  (Utilities|FlowControl|Delay 0.05)
  (Variables|Default|SetLastKillTime -999999999999.0)
  (Variables|Default|SetDamage_Build (Variables|Default|GetDamage_Build))
  (EventDispatchers|BindEventtoOnReloadStarted (Variables|/Script/FortniteGame.FortPawn|GetCurrentWeapon (CustomCharacter|GetAssociatedPlayerPawn)) (delegate Custom|ReloadStarted))
  (CallFunction|PushMaterialParameters))

(fn EletricityRelease
  (bind _returnvalue (Utilities|Time|GetGameTimeinSeconds))
  (bind _returnvalue_2 (Utilities|Time|GetGameTimeinSeconds))
  (bind _returnvalue_3 (CustomItemWrap|GetWrappedMeshComponents))
  (bind _returnvalue_4 (CustomItemWrap|GetWrappedMeshComponents))
  (bind _returnvalue_5 (CustomItemWrap|GetWrappedMeshComponents))
  (Utilities|FlowControl|Sequence
    (:then_0
      (Rendering|Material|SetScalarParameterValueonMaterials _returnvalue_4 "TopS Emissive_02_Brightness---Duration" 0.75)
      (Rendering|Material|SetScalarParameterValueonMaterials _returnvalue_4 "TopS Emissive_03_Brightness---Duration" 0.75)
      (bind _output_get (Variables|Default|SetElectricty (Variables|Default|GetElectricty)))
      (Rendering|Material|SetScalarParameterValueonMaterials _returnvalue_3 "TopS Emissive_02_Brightness---End" 0.0)
      (Rendering|Material|SetScalarParameterValueonMaterials _returnvalue_3 "TopS Emissive_03_Brightness---End" 0.0)
      (Rendering|Material|SetScalarParameterValueonMaterials _returnvalue_3 "TopS Emissive_02_Brightness---Start" (* _output_get 5.0))
      (Rendering|Material|SetScalarParameterValueonMaterials _returnvalue_3 "TopS Emissive_03_Brightness---Start" (* _output_get 5.0))
      (Rendering|Material|SetScalarParameterValueonMaterials _returnvalue_3 "TopS Emissive_02-Brightness-TBP" _returnvalue_2)
      (Rendering|Material|SetScalarParameterValueonMaterials _returnvalue_3 "TopS Emissive_03-Brightness-TBP" _returnvalue_2)
      (Variables|Default|SetElectricty 0.0))
    (:then_1
      (Rendering|Material|SetScalarParameterValueonMaterials _returnvalue_5 "TopS Emissive_03_Offset---Start" 0.0)
      (Rendering|Material|SetScalarParameterValueonMaterials _returnvalue_5 "TopS Emissive_03_Offset---End" 1.0)
      (Rendering|Material|SetScalarParameterValueonMaterials _returnvalue_5 "TopS Emissive_02_Offset---Start" 0.0)
      (Rendering|Material|SetScalarParameterValueonMaterials _returnvalue_5 "TopS Emissive_02_Offset---End" 1.0)
      (Rendering|Material|SetScalarParameterValueonMaterials _returnvalue_5 "TopS Emissive_02_Offset----TBP" _returnvalue)
      (Rendering|Material|SetScalarParameterValueonMaterials _returnvalue_5 "TopS Emissive_03_Offset----TBP" _returnvalue))))

(fn EletrictyBuildUp
  (bind _returnvalue (CustomItemWrap|GetWrappedMeshComponents))
  (bind _returnvalue_2 (CustomItemWrap|GetWrappedMeshComponents))
  (bind _returnvalue_3 (CustomItemWrap|GetWrappedMeshComponents))
  (bind _lastkilltime (Variables|Default|GetLastKillTime))
  (Rendering|Material|SetScalarParameterValueonMaterials _returnvalue_3 "TopS Emissive_03_Offset---End" 0.0)
  (Rendering|Material|SetScalarParameterValueonMaterials _returnvalue_3 "TopS Emissive_02_Offset---End" 0.0)
  (bind _output_get (Variables|Default|SetElectricty (Math|Float|Clamp(Float) (+ (Variables|Default|GetElectricty) 0.2) 0.0 1.5)))
  (Rendering|Material|SetScalarParameterValueonMaterials _returnvalue "TopS Emissive_02_Brightness---End" _output_get)
  (Rendering|Material|SetScalarParameterValueonMaterials _returnvalue "TopS Emissive_03_Brightness---End" _output_get)
  (Rendering|Material|SetScalarParameterValueonMaterials _returnvalue "TopS Emissive_02_Brightness---Start" (* _output_get 2.0))
  (Rendering|Material|SetScalarParameterValueonMaterials _returnvalue "TopS Emissive_03_Brightness---Start" (* _output_get 2.0))
  (Rendering|Material|SetScalarParameterValueonMaterials _returnvalue "TopS Emissive_02-Brightness-TBP" _lastkilltime)
  (Rendering|Material|SetScalarParameterValueonMaterials _returnvalue "TopS Emissive_03-Brightness-TBP" _lastkilltime)
  (Rendering|Material|SetScalarParameterValueonMaterials _returnvalue_2 "TopS Emissive_02_Brightness---Duration" 0.25)
  (Rendering|Material|SetScalarParameterValueonMaterials _returnvalue_2 "TopS Emissive_03_Brightness---Duration" 0.25))

(fn ConstructionScript)
