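; PigeonChart reactive weapon wrap that changes its flipbook decal look according to the player's kill count.
; A watched AthenaKills stat sets the count, clamped to five, and Kill Update writes flipbook color parameters to the wrapped meshes' materials.
; Weapon, respawn and material events re-run that update after a short delay, while a front-end timeline steps the count through each stage.
; RestPossition sets timed offset parameters but has no caller here.
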(event EventBeginPlay
  (if (CustomCosmeticModifier|IsinFrontEnd)
    (Timeline|Timeline PlayFromStart
      (:Stage1
        (Variables|Default|SetKill%20Count 0))
      (:Stage2
        (Variables|Default|SetKill%20Count 1))
      (:Stage3
        (Variables|Default|SetKill%20Count 2))
      (:Stage4
        (Variables|Default|SetKill%20Count 3))
      (:stage5
        (Variables|Default|SetKill%20Count 4))
      (:stage6
        (Variables|Default|SetKill%20Count 5)))
    (CallFunction|KillUpdate)))

(event EventOnWeaponFired)

(event EventOnPawnSet
  (bind (_statname _statvalue) (Cosmetics|WatchCosmeticStat (CustomCharacter|GetAssociatedPlayerPawn) "(TagName=\"StatManager.AthenaKills\")" 0 false)
    (:OnStatChanged
      (Variables|Default|SetKill%20Count (Math|Integer|Clamp(Integer) _statvalue 0 5))
      (CallFunction|KillUpdate))))

(event Custom|Kill%20Update
  (bind _returnvalue (CustomItemWrap|GetWrappedMeshComponents))
  (switch int (Variables|Default|GetKill%20Count)
    (:1
      _)
    (:2
      (Rendering|Material|SetColorParameterValueonMaterials _returnvalue "EDRec_FlipBook_FrameA_Speed--" (Utilities|Operators|Add :B (2.0 1.0 0.0 1.0))))
    (:3
      (Rendering|Material|SetColorParameterValueonMaterials _returnvalue "EDRec_FlipBook_FrameB_Speed--" (Utilities|Operators|Add :B (2.1 1.0 0.0 1.0))))
    (:4
      (Rendering|Material|SetColorParameterValueonMaterials _returnvalue "EDRec_FlipBook_FrameC_Speed--" (Utilities|Operators|Add :B (2.2 1.0 0.0 1.0))))
    (:5
      (Rendering|Material|SetColorParameterValueonMaterials _returnvalue "EDRec_FlipBook_FrameD_Speed--" (Utilities|Operators|Add :B (2.3 1.0 0.0 1.0)))))
  (Rendering|Material|SetColorParameterValueonMaterials _returnvalue "FDecalV2_FlipBook_Random_Clamp" (Utilities|Operators|Add :B (0.0 1.0 0.0 1.0))))

(event EventOnPostResIn
  (Utilities|FlowControl|Delay 0.05)
  (CallFunction|KillUpdate))

(event EventOnNewWeapons
  (params (WeaponsAsItems :type Array%20of%20Fort%20Item%20Object%20References))
  (Utilities|FlowControl|Delay 0.05)
  (CallFunction|KillUpdate))

(event EventOnNewPlayerWeaponType
  (params (NewType :type EFortItemType%20Enum :default "WorldItem") (OldType :type EFortItemType%20Enum :default "WorldItem"))
  (Utilities|FlowControl|Delay 0.05)
  (CallFunction|KillUpdate))

(event EventOnWeaponChanged
  (params (NewWeapon :type Fort%20Weapon%20Object%20Reference) (OldWeapon :type Fort%20Weapon%20Object%20Reference))
  (Utilities|FlowControl|Delay 0.05)
  (CallFunction|KillUpdate))

(event EventApplyMaterialParameters
  (Utilities|FlowControl|Delay 0.05)
  (CallFunction|KillUpdate))

(fn RestPossition
  (params (End :type Float%20%28single-precision%29) (Start :type Float%20%28single-precision%29) (Time :type Float%20%28single-precision%29) (delay :type Float%20%28double-precision%29))
  (bind _returnvalue (CustomItemWrap|GetWrappedMeshComponents))
  (bind _returnvalue_2 (Utilities|Time|GetGameTimeinSeconds))
  (bind _returnvalue_3 (CustomItemWrap|GetWrappedMeshComponents))
  (bind _returnvalue_4 (CustomItemWrap|GetWrappedMeshComponents))
  (bind _returnvalue_5 (+ _returnvalue_2 delay))
  (Rendering|Material|SetScalarParameterValueonMaterials _returnvalue_4 "TopS_01_Blend_Offset---End" End)
  (Rendering|Material|SetScalarParameterValueonMaterials _returnvalue_4 "TopS_01_Blend_Offset---Start" Start)
  (Rendering|Material|SetScalarParameterValueonMaterials _returnvalue_4 "TopS_01_Blend_Offset-TBP" _returnvalue_5)
  (Rendering|Material|SetScalarParameterValueonMaterials _returnvalue_4 "TopS_01_Blend_Offset---Duration" Time)
  (Rendering|Material|SetScalarParameterValueonMaterials _returnvalue_3 "TopS Emissive_02_Offset---End" End)
  (Rendering|Material|SetScalarParameterValueonMaterials _returnvalue_3 "TopS Emissive_02_Offset---Start" Start)
  (Rendering|Material|SetScalarParameterValueonMaterials _returnvalue_3 "TopS Emissive_02_Offset----TBP" _returnvalue_5)
  (Rendering|Material|SetScalarParameterValueonMaterials _returnvalue_3 "TopS Emissive_02_Offset---Duration" Time)
  (Rendering|Material|SetScalarParameterValueonMaterials _returnvalue "TopS Emissive_03_Offset---End" End)
  (Rendering|Material|SetScalarParameterValueonMaterials _returnvalue "TopS Emissive_03_Offset---Start" Start)
  (Rendering|Material|SetScalarParameterValueonMaterials _returnvalue "TopS Emissive_03_Offset----TBP" _returnvalue_5)
  (Rendering|Material|SetScalarParameterValueonMaterials _returnvalue "TopS Emissive_03_Offset---Duration" Time))

(fn ConstructionScript)
