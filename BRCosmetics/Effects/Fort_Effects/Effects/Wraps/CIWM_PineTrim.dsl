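; Custom item wrap (PineTrim) that animates material parameters on wrapped meshes in response to player kills and weapon fire.
; BeginPlay and PawnSet watch the kill stat; on multiples of three, a DoOnce-gated branch stops the fire timeline and plays a burst that lerps emissive brightness and texture offsets, calling Reopen on finish.
; Weapon fire drives a smaller offset-only timeline, gated by Pause and Float_Value, and prints a debug string.
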(event EventBeginPlay
  (bind _returnvalue (CustomItemWrap|GetWrappedMeshComponents))
  (bind (_statname _statvalue) (Cosmetics|WatchCosmeticStat (CustomCharacter|GetAssociatedPlayerPawn) "(TagName=\"StatManager.AthenaKills\")" 0 false)
    (:OnStatChanged
      (if (== (Math|Integer|%25%28Integer%29 _statvalue 3) 0)
        (Utilities|FlowControl|DoOnce false
          (:Completed
            (CallFunction|Kill)
            (bind (_direction _newtrack_0 _newtrack_2) (Timeline|Timeline_0 PlayFromStart :NewTime 0.75)
              (:Update
                (Rendering|Material|SetScalarParameterValueonMaterials _returnvalue "TopS Emissive_03_Brightness---End" (Math|Float|Lerp 0.0 1.0 _newtrack_2))
                (Rendering|Material|SetScalarParameterValueonMaterials _returnvalue "TopS Emissive_01_Brightness---End" (Math|Float|Lerp 0.8 0.6 _newtrack_0))
                (Rendering|Material|SetScalarParameterValueonMaterials _returnvalue "TopS Emissive_02_Brightness---End" (Math|Float|Lerp 3.0 0.3 _newtrack_0))
                (Rendering|Material|SetVectorParameterValueonMaterials _returnvalue "TopS_03_OffsetS_Weapon" (Math|Vector|MakeVector (Math|Float|Lerp -0.6 0.6 _newtrack_0)))
                (Rendering|Material|SetVectorParameterValueonMaterials _returnvalue "TopS_03_OffsetS_Backpack" (Math|Vector|MakeVector (Math|Float|Lerp -0.6 0.6 _newtrack_0)))
                (Rendering|Material|SetVectorParameterValueonMaterials _returnvalue "TopS_03_OffsetS_Pickaxe" (Math|Vector|MakeVector (Math|Float|Lerp -0.6 0.6 _newtrack_0)))
                (Rendering|Material|SetVectorParameterValueonMaterials _returnvalue "TopS_03_OffsetS_Character" (Math|Vector|MakeVector (Math|Float|Lerp -0.6 0.6 _newtrack_0)))
                (Rendering|Material|SetVectorParameterValueonMaterials _returnvalue "TopS_03_OffsetS_VehicleSide" (Math|Vector|MakeVector (Math|Float|Lerp -0.6 0.6 _newtrack_0)))
                (Rendering|Material|SetVectorParameterValueonMaterials _returnvalue "TopS_03_OffsetS_Vehicle" (Math|Vector|MakeVector (Math|Float|Lerp -0.6 0.6 _newtrack_0))))
              (:Finished
                (CallFunction|Reopen)
                (Variables|Default|SetPause false)))))))))

(event EventOnPawnSet
  (bind _returnvalue (CustomItemWrap|GetWrappedMeshComponents))
  (bind (_statname _statvalue) (Cosmetics|WatchCosmeticStat (CustomCharacter|GetAssociatedPlayerPawn) "(TagName=\"StatManager.AthenaKills\")" 0 false)
    (:OnStatChanged
      (if (== (Math|Integer|%25%28Integer%29 _statvalue 3) 0)
        (Utilities|FlowControl|DoOnce false
          (:Completed
            (CallFunction|Kill)
            (Timeline|Timeline_0 PlayFromStart)))))))

(event Custom|Reopen
  (bind _returnvalue (CustomItemWrap|GetWrappedMeshComponents))
  (Utilities|FlowControl|DoOnce false
    (:Completed
      (CallFunction|Kill)
      (Timeline|Timeline_0 PlayFromStart))))

(event EventOnWeaponFired
  (bind _returnvalue (CustomItemWrap|GetWrappedMeshComponents))
  (Development|PrintString "FireTest" true true "(R=0.000000,G=1.000000,B=0.041611,A=1.000000)")
  (if (Variables|Default|GetPause)
    (if (>= (Variables|Default|GetFloat_Value) 1.0)
      (bind (_direction _newtrack_0) (Timeline|Timeline SetNewTime :NewTime 0.75)
        (:Update
          (Variables|Default|SetFloat_Value _newtrack_0)
          (Rendering|Material|SetVectorParameterValueonMaterials _returnvalue "TopS_03_OffsetS_Weapon" (Math|Vector|MakeVector (Math|Float|Lerp -0.6 -0.3 _newtrack_0)))
          (Rendering|Material|SetVectorParameterValueonMaterials _returnvalue "TopS_03_OffsetS_Backpack" (Math|Vector|MakeVector (Math|Float|Lerp -0.6 -0.3 _newtrack_0)))
          (Rendering|Material|SetVectorParameterValueonMaterials _returnvalue "TopS_03_OffsetS_Pickaxe" (Math|Vector|MakeVector (Math|Float|Lerp -0.6 -0.3 _newtrack_0)))
          (Rendering|Material|SetVectorParameterValueonMaterials _returnvalue "TopS_03_OffsetS_Character" (Math|Vector|MakeVector (Math|Float|Lerp -0.6 -0.2 _newtrack_0)))
          (Rendering|Material|SetVectorParameterValueonMaterials _returnvalue "TopS_03_OffsetS_VehicleSide" (Math|Vector|MakeVector (Math|Float|Lerp -0.6 -0.1 _newtrack_0)))
          (Rendering|Material|SetVectorParameterValueonMaterials _returnvalue "TopS_03_OffsetS_Vehicle" (Math|Vector|MakeVector (Math|Float|Lerp -0.6 -0.2 _newtrack_0))))
        (:Finished
          (Variables|Default|SetPause false))))
    (else
      (Variables|Default|SetPause true)
      (Timeline|Timeline PlayFromStart))))

(event Custom|Kill
  (bind _returnvalue (CustomItemWrap|GetWrappedMeshComponents))
  (Timeline|Timeline Stop))

(fn ConstructionScript)
