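; Weapon wrap modifier for PiteousKicks that changes the wrap's blend and emissive look in five stages as the player's kill count rises.
; It watches the AthenaKills stat, or a preview timeline in the front end, and maps the clamped count to start and end offsets.
; Offsets and last-kill time go to wrapped mesh materials as scalar parameters; weapon changes backdate that time.
; Each stat change also prints the count as debug text.
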(event EventOnWeaponFired)

(event EventBeginPlay
  (if (CustomCosmeticModifier|IsinFrontEnd)
    (Timeline|Timeline PlayFromStart
      (:Stage1
        (Variables|Default|SetKill_Count 0))
      (:Stage2
        (Variables|Default|SetKill_Count 1))
      (:Stage3
        (Variables|Default|SetKill_Count 2))
      (:Stage4
        (Variables|Default|SetKill_Count 3))
      (:stage5
        (Variables|Default|SetKill_Count 4)))
    (Variables|Default|SetLastKillTime (Utilities|Time|GetGameTimeinSeconds))
    (CallFunction|PushMaterialParameters)))

(event EventOnPawnSet
  (Utilities|FlowControl|Sequence
    (:then_1
      (bind (_statname _statvalue) (Cosmetics|WatchCosmeticStat (CustomCharacter|GetAssociatedPlayerPawn) "(TagName=\"StatManager.AthenaKills\")" 0 false)
        (:OnStatChanged
          (bind _output_get (Variables|Default|SetKill_Count _statvalue))
          (Development|PrintText (Utilities|Text|ToText(Integer) _output_get) true true "(R=1.000000,G=0.018970,B=0.000000,A=1.000000)")
          (Variables|Default|SetLastKillTime (Utilities|Time|GetGameTimeinSeconds))
          (CallFunction|PushMaterialParameters))))))

(event Custom|Push%20Material%20Parameters
  (bind _returnvalue (CustomItemWrap|GetWrappedMeshComponents))
  (bind _lastkilltime (Variables|Default|GetLastKillTime))
  (bind _end (Variables|Default|GetEnd))
  (bind _start (Variables|Default|GetStart))
  (switch int (Math|Integer|Clamp(Integer) (Variables|Default|GetKill_Count) 0 4)
    (:0
      (Variables|Default|SetEnd 1.2)
      (Variables|Default|SetStart 1.2))
    (:1
      (Variables|Default|SetEnd 0.15)
      (Variables|Default|SetStart 1.2))
    (:2
      (Variables|Default|SetEnd -0.3)
      (Variables|Default|SetStart 0.15))
    (:3
      (Variables|Default|SetEnd -1.15)
      (Variables|Default|SetStart -0.3))
    (:4
      (Variables|Default|SetEnd -1.15)
      (Variables|Default|SetStart -1.15)))
  (Utilities|FlowControl|Sequence
    (:then_0
      (Rendering|Material|SetScalarParameterValueonMaterials _returnvalue "TopS_01_Blend_Offset---End" _end)
      (Rendering|Material|SetScalarParameterValueonMaterials _returnvalue "TopS_01_Blend_Offset---Start" _start)
      (Rendering|Material|SetScalarParameterValueonMaterials _returnvalue "TopS_01_Blend_Offset-TBP" (Variables|Default|GetLastKillTime)))
    (:then_1
      (Rendering|Material|SetScalarParameterValueonMaterials _returnvalue "TopS Emissive_01_Offset---End" _end)
      (Rendering|Material|SetScalarParameterValueonMaterials _returnvalue "TopS Emissive_01_Offset---Start" _start)
      (Rendering|Material|SetScalarParameterValueonMaterials _returnvalue "TopS Emissive_01_Offset----TBP" (Variables|Default|GetLastKillTime)))
    (:then_2
      (Rendering|Material|SetScalarParameterValueonMaterials _returnvalue "TopS Emissive_02_Offset---End" _end)
      (Rendering|Material|SetScalarParameterValueonMaterials _returnvalue "TopS Emissive_02_Offset---Start" _start)
      (Rendering|Material|SetScalarParameterValueonMaterials _returnvalue "TopS Emissive_02_Offset----TBP" _lastkilltime)
      (Rendering|Material|SetScalarParameterValueonMaterials _returnvalue "TopS Emissive_02-Brightness-TBP" _lastkilltime))))

(event EventOnPostResIn
  (CallFunction|PushMaterialParameters))

(event EventOnNewWeapons
  (params (WeaponsAsItems :type Array%20of%20Fort%20Item%20Object%20References))
  (Utilities|FlowControl|Delay 0.05)
  ;[1 1 1] Updated kill count on new weapon
  (Variables|Default|SetLastKillTime -9999999999.0)
  (CallFunction|PushMaterialParameters))

(event EventOnNewPlayerWeaponType
  (params (NewType :type EFortItemType%20Enum :default "WorldItem") (OldType :type EFortItemType%20Enum :default "WorldItem"))
  (Utilities|FlowControl|Delay 0.05)
  ;[1 1 1] Updated kill count on new weapon
  (Variables|Default|SetLastKillTime -9999999999.0)
  (CallFunction|PushMaterialParameters))

(event EventOnWeaponChanged
  (params (NewWeapon :type Fort%20Weapon%20Object%20Reference) (OldWeapon :type Fort%20Weapon%20Object%20Reference))
  (Utilities|FlowControl|Delay 0.05)
  ;[1 1 1] Updated kill count on new weapon
  (Variables|Default|SetLastKillTime -9999999999.0)
  (CallFunction|PushMaterialParameters))

(event EventApplyMaterialParameters
  (CallFunction|PushMaterialParameters))

(fn ConstructionScript)
