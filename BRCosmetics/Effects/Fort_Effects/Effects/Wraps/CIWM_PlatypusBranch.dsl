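; Kill-reactive item wrap that steps a material offset and flashes emissive brightness as the player gets kills.
; Watched AthenaKills changes advance a counter and target position, and a timeline lerps the per-item-type offset parameters on wrapped meshes.
; A trigger after six steps runs RevertState, which waits, animates back to the start and resets the count.
; A looping frontend timer simulates kills; a Running flag drops triggers during animation.
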(event Custom|Frontend_Trigger
  (bind _returnvalue (CustomItemWrap|GetWrappedMeshComponents))
  (if (not (Variables|Default|GetRunning))
    (Variables|Default|SetRunning true)
    (if (>= (Variables|Default|GetKill_Count) 6)
      (Variables|Default|SetCurrent_Position_Data -0.6)
      (if (not (Variables|Default|GetReverting))
        (CallFunction|RevertState))
      (else
        (Math|Integer|IncrementInt (Variables|Default|GetKill_Count)
          (:%20%20
            (Variables|Default|SetCurrent_Position_Data (+ (Variables|Default|GetCurrent_Position_Data) 0.205))
            (bind (_direction _rays _newtrack_0) (Timeline|Timeline PlayFromStart)
              (:Update
                (Rendering|Material|SetScalarParameterValueonMaterials _returnvalue "TopS Emissive_02_Brightness---End" _rays)
                (Rendering|Material|SetScalarParameterValueonMaterials _returnvalue "TopS Emissive_01_Brightness---End" _rays)
                (for _array_element (Utilities|Array|MakeArray (Utilities|Name|MakeLiteralName "TopS_03_OffsetS_Weapon") (Utilities|Name|MakeLiteralName "TopS_03_OffsetS_Backpack") (Utilities|Name|MakeLiteralName "TopS_03_OffsetS_Character") (Utilities|Name|MakeLiteralName "TopS_03_OffsetS_Mech") (Utilities|Name|MakeLiteralName "TopS_03_OffsetS_Pickaxe") (Utilities|Name|MakeLiteralName "TopS_03_OffsetS_Vehicle") (Utilities|Name|MakeLiteralName "TopS_03_OffsetS_VehicleSide"))
                  (Rendering|Material|SetVectorParameterValueonMaterials _returnvalue _array_element (Math|Vector|MakeVector (Math|Float|Lerp (Variables|Default|GetOld_Position_Data) (Variables|Default|GetCurrent_Position_Data) _newtrack_0)))))
              (:Finished
                (Variables|Default|SetOld_Position_Data (Variables|Default|GetCurrent_Position_Data))
                (Variables|Default|SetRunning false)))))))))

(event EventOnPawnSet
  (bind _returnvalue (CustomItemWrap|GetWrappedMeshComponents))
  (Variables|Default|SetOld_Position_Data -0.6)
  (Variables|Default|SetCurrent_Position_Data -0.6)
  (Cosmetics|WatchCosmeticStat (CustomCharacter|GetAssociatedPlayerPawn) "(TagName=\"StatManager.AthenaKills\")" 0 false
    (:OnStatChanged
      (if (not (Variables|Default|GetRunning))
        (Variables|Default|SetRunning true)
        (if (>= (Variables|Default|GetKill_Count) 6)
          (Variables|Default|SetCurrent_Position_Data -0.6)
          (if (not (Variables|Default|GetReverting))
            (CallFunction|RevertState))
          (else
            (Math|Integer|IncrementInt (Variables|Default|GetKill_Count)
              (:%20%20
                (Variables|Default|SetCurrent_Position_Data (+ (Variables|Default|GetCurrent_Position_Data) 0.205))
                (Timeline|Timeline PlayFromStart)))))))))

(event EventBeginPlay
  (if (Reactive|Frontend|IsCosmeticObjectDisplayedinFrontendContext 0)
    (Variables|Default|SetOld_Position_Data -0.6)
    (Variables|Default|SetCurrent_Position_Data -0.6)
    (bind _returnvalue (Utilities|Time|SetTimerbyFunctionName 0 "Frontend_Trigger" 1.5 true))
    (Variables|Default|SetFrontendTimer _returnvalue)))

(event Custom|RevertState
  (Variables|Default|SetReverting true)
  (Utilities|FlowControl|Delay 2.0)
  (bind (_direction _newtrack_0) (Timeline|Timeline_0 PlayFromStart)
    (:Update
      (for _array_element (Utilities|Array|MakeArray (Utilities|Name|MakeLiteralName "TopS_03_OffsetS_Weapon") (Utilities|Name|MakeLiteralName "TopS_03_OffsetS_Backpack") (Utilities|Name|MakeLiteralName "TopS_03_OffsetS_Character") (Utilities|Name|MakeLiteralName "TopS_03_OffsetS_Mech") (Utilities|Name|MakeLiteralName "TopS_03_OffsetS_Pickaxe") (Utilities|Name|MakeLiteralName "TopS_03_OffsetS_Vehicle") (Utilities|Name|MakeLiteralName "TopS_03_OffsetS_VehicleSide"))
        (Rendering|Material|SetVectorParameterValueonMaterials (CustomItemWrap|GetWrappedMeshComponents) _array_element (Math|Vector|MakeVector (Math|Float|Lerp (Variables|Default|GetOld_Position_Data) (Variables|Default|GetCurrent_Position_Data) _newtrack_0)))))
    (:Finished
      (Variables|Default|SetReverting false)
      (Variables|Default|SetOld_Position_Data (Variables|Default|GetCurrent_Position_Data))
      (Variables|Default|SetKill_Count 0)
      (Variables|Default|SetRunning false))))

(fn ConstructionScript)

(fn Weapon_Offsets
  (params (Offset_End_Name :type Name) (Offset_End_Value :type Float%20%28double-precision%29) (Offset_Start_Name :type Name) (Offset_State_Value :type Float%20%28double-precision%29))
  (bind _returnvalue (CustomItemWrap|GetWrappedMeshComponents))
  (Rendering|Material|SetVectorParameterValueonMaterials _returnvalue Offset_End_Name (Math|Vector|MakeVector 0.0 0.0 Offset_End_Value))
  (Rendering|Material|SetVectorParameterValueonMaterials _returnvalue Offset_Start_Name (Math|Vector|MakeVector 0.0 0.0 Offset_State_Value)))
