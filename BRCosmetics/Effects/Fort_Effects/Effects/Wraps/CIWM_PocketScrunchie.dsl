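; Weapon wrap Blueprint for PocketScrunchie that makes the wrap's materials react when the weapon fires.
; On firing, it sets the decal flipbook frame and emissive end brightness to one on all wrapped mesh components.
; Retriggerable delays then reset the frame first, and later stamp game time into an emissive brightness parameter and zero the end brightness.
; Because the delays retrigger, continued firing postpones the resets.
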(event EventOnWeaponFired
  (bind _returnvalue (CustomItemWrap|GetWrappedMeshComponents))
  (Rendering|Material|SetScalarParameterValueonMaterials _returnvalue "FDecalV2_FlipBook_Frame #" 1.0)
  (Rendering|Material|SetScalarParameterValueonMaterials _returnvalue "TopS Emissive_01_Brightness---End" 1.0)
  (Utilities|FlowControl|RetriggerableDelay 0.5)
  (Rendering|Material|SetScalarParameterValueonMaterials _returnvalue "FDecalV2_FlipBook_Frame #" 0.0)
  (Utilities|FlowControl|RetriggerableDelay 1.0)
  (Rendering|Material|SetScalarParameterValueonMaterials _returnvalue "TopS Emissive_01-Brightness-TBP" (Utilities|Time|GetGameTimeinSeconds))
  (Rendering|Material|SetScalarParameterValueonMaterials _returnvalue "TopS Emissive_01_Brightness---End" 0.0))

(fn ConstructionScript)
