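; Wrap logic Blueprint for the PunkDevilSummer weapon wrap that drives its emissive material parameters when the weapon fires.
; Each shot writes the current game time as a timestamp into three "TopS Emissive" brightness parameters on the wrapped mesh components' materials.
; The three matching offset parameters receive the same time, but a shot flag and short Delay throttle them to at most one update per quarter second.
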(event EventOnWeaponFired
  (bind _returnvalue (Utilities|Time|GetGameTimeinSeconds))
  (bind _returnvalue_2 (CustomItemWrap|GetWrappedMeshComponents))
  (Rendering|Material|SetScalarParameterValueonMaterials _returnvalue_2 "TopS Emissive_01-Brightness-TBP" _returnvalue)
  (Rendering|Material|SetScalarParameterValueonMaterials _returnvalue_2 "TopS Emissive_02-Brightness-TBP" _returnvalue)
  (Rendering|Material|SetScalarParameterValueonMaterials _returnvalue_2 "TopS Emissive_03-Brightness-TBP" _returnvalue)
  (if (not (Variables|Default|Getshot))
    (Variables|Default|Setshot true)
    (Rendering|Material|SetScalarParameterValueonMaterials _returnvalue_2 "TopS Emissive_01_Offset----TBP" _returnvalue)
    (Rendering|Material|SetScalarParameterValueonMaterials _returnvalue_2 "TopS Emissive_02_Offset----TBP" _returnvalue)
    (Rendering|Material|SetScalarParameterValueonMaterials _returnvalue_2 "TopS Emissive_03_Offset----TBP" _returnvalue)
    (Utilities|FlowControl|Delay 0.25)
    (Variables|Default|Setshot false)))

(fn ConstructionScript)
