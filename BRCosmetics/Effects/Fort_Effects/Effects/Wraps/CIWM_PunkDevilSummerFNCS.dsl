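; Custom item wrap Blueprint for the PunkDevilSummerFNCS weapon wrap that animates a material effect when the weapon fires.
; Firing starts the Lightnings timeline, which lerps an offset vector parameter on the wrapped mesh materials and records its value.
; A Pause flag blocks restarts until the timeline finishes.
; Shots fired meanwhile jump it to a fixed time once the recorded value reaches full, apparently sustaining the effect.
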(event EventOnWeaponFired
  (if (Variables|Default|GetPause)
    (if (>= (Variables|Default|GetFloat_Value) 1.0)
      (bind (_direction _newtrack_0) (Timeline|Lightnings SetNewTime :NewTime 0.75)
        (:Update
          (Variables|Default|SetFloat_Value _newtrack_0)
          (Rendering|Material|SetVectorParameterValueonMaterials (CustomItemWrap|GetWrappedMeshComponents) "TopS_Em02_Offset__Master" (Math|Vector|MakeVector 0.0 (Math|Float|Lerp 0.2 -0.5 _newtrack_0))))
        (:Finished
          (Variables|Default|SetPause false))))
    (else
      (Variables|Default|SetPause true)
      (Timeline|Lightnings PlayFromStart))))

(fn ConstructionScript)
