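; Reactive weapon wrap Blueprint for PuzzleSprout that changes its decal frame, motion and emissive color in response to combat.
; A watched DamageReceived stat triggers a spinning frame and blue tint, while weapon fire triggers another frame and pink-red tint unless damage is active.
; The AnimColor timeline lerps colors on wrapped meshes, and retriggerable delays restore idle unless Killed is set.
; In frontend display, a looping LockerCycle timer previews these states instead.
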(event EventBeginPlay
  (bind _returnvalue (Utilities|Time|GetGameTimeinSeconds))
  (bind _returnvalue_2 (CustomItemWrap|GetWrappedMeshComponents))
  (if (Reactive|Frontend|IsCosmeticObjectDisplayedinFrontendContext 0)
    (Utilities|FlowControl|Sequence
      (:then_1
        (Utilities|Time|SetTimerbyFunctionName 0 "LockerCycle" 2.5 true)))
    (else
      (bind _statname (Cosmetics|WatchCosmeticStat (CustomCharacter|GetAssociatedPlayerPawn) "(TagName=\"StatManager.DamageReceived\")" 0 false)
        (:OnStatChanged
          (Utilities|FlowControl|Sequence
            (:then_0
              (Rendering|Material|SetScalarParameterValueonMaterials _returnvalue_2 "EDRec_TimeFrom_BP_DecalA" _returnvalue)
              (Rendering|Material|SetScalarParameterValueonMaterials _returnvalue_2 "EDRec_DecalA__Sine_TimeFrom_BP" _returnvalue)
              (Rendering|Material|SetScalarParameterValueonMaterials _returnvalue_2 "TimeFrom-BP----------------" _returnvalue)
              (Rendering|Material|SetScalarParameterValueonMaterials _returnvalue_2 "EDRec_DecalA_BounceRot_TimeFrom_BP" _returnvalue)
              (Rendering|Material|SetScalarParameterValueonMaterials _returnvalue_2 "EDRec_FlipBook_FrameA" 2.0)
              (Rendering|Material|SetScalarParameterValueonMaterials _returnvalue_2 "EDRec_DecalA_Rotation_Speed --------------" 5.0)
              (Utilities|FlowControl|RetriggerableDelay 0.5)
              (if (not (Variables|Default|GetKilled))
                (Rendering|Material|SetScalarParameterValueonMaterials _returnvalue_2 "EDRec_FlipBook_FrameA" 0.0)
                (Rendering|Material|SetScalarParameterValueonMaterials _returnvalue_2 "EDRec_DecalA_Rotation_Speed --------------" 0.0)))
            (:then_1
              (if (== _statname "DamageReceived")
                (Utilities|FlowControl|Sequence
                  (:then_0
                    (Variables|Default|SetBeingDamaged true)
                    (Variables|Default|SetTargetColor "(R=0.000000,G=0.618010,B=1.000000,A=50.000000)")
                    (Components|Timeline|SetPlayRate (Variables|CIWM_PuzzleSprout|GetAnimColor) 1.5)
                    (bind (_direction _alpha) (Timeline|AnimColor PlayFromStart)
                      (:Update
                        (for _array_element (CustomItemWrap|GetWrappedMeshComponents)
                          (Rendering|Material|SetVectorParameterValueonMaterials _array_element "FDecalV2-Emissive-Color--Alpha=Multplyer" (Math|Conversions|ToVector(LinearColor) (Math|Color|Lerp(LinearColor) (Variables|Default|GetBaseColor) (Variables|Default|GetTargetColor) _alpha)))))))
                  (:then_1
                    (Utilities|FlowControl|RetriggerableDelay 0.75)
                    (Variables|Default|SetBeingDamaged false)
                    (Timeline|AnimColor Reverse)))))))))))

(event EventOnPawnSet
  (bind _returnvalue (Utilities|Time|GetGameTimeinSeconds))
  (bind _returnvalue_2 (CustomItemWrap|GetWrappedMeshComponents))
  (bind _statname (Cosmetics|WatchCosmeticStat (CustomCharacter|GetAssociatedPlayerPawn) "(TagName=\"StatManager.DamageReceived\")" 0 false)
    (:OnStatChanged
      (Utilities|FlowControl|Sequence
        (:then_0
          (Rendering|Material|SetScalarParameterValueonMaterials _returnvalue_2 "EDRec_TimeFrom_BP_DecalA" _returnvalue)
          (Rendering|Material|SetScalarParameterValueonMaterials _returnvalue_2 "EDRec_DecalA__Sine_TimeFrom_BP" _returnvalue)
          (Rendering|Material|SetScalarParameterValueonMaterials _returnvalue_2 "TimeFrom-BP----------------" _returnvalue)
          (Rendering|Material|SetScalarParameterValueonMaterials _returnvalue_2 "EDRec_DecalA_BounceRot_TimeFrom_BP" _returnvalue)
          (Rendering|Material|SetScalarParameterValueonMaterials _returnvalue_2 "EDRec_FlipBook_FrameA" 2.0)
          (Rendering|Material|SetScalarParameterValueonMaterials _returnvalue_2 "EDRec_DecalA_Rotation_Speed --------------" 5.0)
          (Utilities|FlowControl|RetriggerableDelay 0.5)
          (if (not (Variables|Default|GetKilled))
            (Rendering|Material|SetScalarParameterValueonMaterials _returnvalue_2 "EDRec_FlipBook_FrameA" 0.0)
            (Rendering|Material|SetScalarParameterValueonMaterials _returnvalue_2 "EDRec_DecalA_Rotation_Speed --------------" 0.0)))
        (:then_1
          (if (== _statname "DamageReceived")
            (Utilities|FlowControl|Sequence
              (:then_0
                (Variables|Default|SetBeingDamaged true)
                (Variables|Default|SetTargetColor "(R=0.000000,G=0.618010,B=1.000000,A=50.000000)")
                (Components|Timeline|SetPlayRate (Variables|CIWM_PuzzleSprout|GetAnimColor) 1.5)
                (Timeline|AnimColor PlayFromStart))
              (:then_1
                (Utilities|FlowControl|RetriggerableDelay 0.75)
                (Variables|Default|SetBeingDamaged false)
                (Timeline|AnimColor Reverse)))))))))

(event EventOnWeaponFired
  (bind _returnvalue (Utilities|Time|GetGameTimeinSeconds))
  (bind _returnvalue_2 (CustomItemWrap|GetWrappedMeshComponents))
  (if (not (Variables|Default|GetBeingDamaged))
    (Utilities|FlowControl|Sequence
      (:then_0
        (Rendering|Material|SetScalarParameterValueonMaterials _returnvalue_2 "TimeFrom-BP----------------" _returnvalue)
        (Rendering|Material|SetScalarParameterValueonMaterials _returnvalue_2 "EDRec_Emissive_A_Brightness_Time_BP" _returnvalue)
        (if (not (Variables|Default|GetKilled))
          (Rendering|Material|SetScalarParameterValueonMaterials _returnvalue_2 "EDRec_FlipBook_FrameA" 1.0)
          (Utilities|FlowControl|RetriggerableDelay 0.5)
          (if (not (Variables|Default|GetKilled))
            (Rendering|Material|SetScalarParameterValueonMaterials _returnvalue_2 "EDRec_Emissive_A_Brightness_DarkenBase_Time_End" 0.0)
            (Rendering|Material|SetScalarParameterValueonMaterials _returnvalue_2 "EDRec_Emissive_A_Brightness_Time__End" 0.0)
            (Rendering|Material|SetScalarParameterValueonMaterials _returnvalue_2 "EDRec_FlipBook_FrameA" 0.0))))
      (:then_1
        (Variables|Default|SetTargetColor "(R=1.000000,G=0.000000,B=0.260324,A=50.000000)")
        (Components|Timeline|SetPlayRate (Variables|CIWM_PuzzleSprout|GetAnimColor) 2.15)
        (Timeline|AnimColor Play))
      (:then_2
        (Utilities|FlowControl|RetriggerableDelay 0.2)
        (Timeline|AnimColor Reverse)))))

(event Custom|LockerCycle
  (bind _returnvalue (Utilities|Time|GetGameTimeinSeconds))
  (bind _returnvalue_2 (CustomItemWrap|GetWrappedMeshComponents))
  (bind _returnvalue_3 (Utilities|Select :Index (Variables|Default|GetLockerCounter) :Option%200 0.0 :Option%201 0.0 :Option%202 0.0 :Option%203 _returnvalue))
  (bind _lockercounter (Variables|Default|GetLockerCounter))
  (if (>= (Variables|Default|GetLockerCounter) 3)
    (Variables|Default|SetLockerCounter 0)
    (else
      (Math|Integer|IncrementInt (Variables|Default|GetLockerCounter)
        (:%20%20
          (Rendering|Material|SetScalarParameterValueonMaterials _returnvalue_2 "EDRec_FlipBook_FrameA" (Utilities|Select :Index _lockercounter :Option%200 0.0 :Option%201 0.0 :Option%202 1.0 :Option%203 2.0))
          (Rendering|Material|SetVectorParameterValueonMaterials _returnvalue_2 "FDecalV2-Emissive-Color--Alpha=Multplyer" (Math|Conversions|ToVector(LinearColor) (Utilities|Select :Index _lockercounter :Option%200 (Variables|Default|GetBaseColor%20Locker) :Option%201 (Variables|Default|GetBaseColor%20Locker) :Option%202 (Variables|Default|GetDMGdealtColor%20Locker) :Option%203 (Variables|Default|GetDMGrecivedColor%20Locker))))
          (Rendering|Material|SetScalarParameterValueonMaterials _returnvalue_2 "EDRec_DecalA__Sine_TimeFrom_BP" _returnvalue_3)
          (Rendering|Material|SetScalarParameterValueonMaterials _returnvalue_2 "TimeFrom-BP----------------" _returnvalue_3)
          (Rendering|Material|SetScalarParameterValueonMaterials _returnvalue_2 "EDRec_DecalA_BounceRot_TimeFrom_BP" _returnvalue_3)
          (Rendering|Material|SetScalarParameterValueonMaterials _returnvalue_2 "EDRec_TimeFrom_BP_DecalA" _returnvalue_3)
          (Rendering|Material|SetScalarParameterValueonMaterials _returnvalue_2 "EDRec_DecalA_Rotation_Speed --------------" (Utilities|Select :Index _lockercounter :Option%200 0.0 :Option%201 0.0 :Option%202 0.0 :Option%203 5.0)))))))

(fn ConstructionScript)
