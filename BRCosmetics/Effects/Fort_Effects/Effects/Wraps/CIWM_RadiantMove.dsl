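; Reactive item wrap Blueprint for RadiantMove that slides three decals across wrapped meshes when the weapon fires.
; A fast timeline feeds three tracks, gated by IsKilling? against retriggering mid-play; after it finishes, a timer reverses it slowly to retract the decals.
; Each update pushes lerped offset vectors into every wrapped mesh's dynamic material, writing weapon, backpack, character, pickaxe and vehicle parameter sets without checking item type.
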(event EventOnPawnSet
  (if (Reactive|Frontend|IsCosmeticObjectDisplayedinFrontendContext 0)
    _))

(event EventBeginPlay)

(event EventOnWeaponFired
  (if (not (Variables|Default|GetIsKilling?))
    (Variables|Default|SetIsKilling? true)
    (Components|Timeline|SetPlayRate (Variables|CIWM_RadiantMove|GetTimeline) 2.0)
    (bind (_direction _newtrack_0 _newtrack_1 _newtrack_2) (Timeline|Timeline PlayFromStart)
      (:Update
        (Variables|Default|SetDecalA _newtrack_0)
        (Variables|Default|SetDecalB _newtrack_1)
        (Variables|Default|SetDecalC _newtrack_2)
        (for _array_element (CustomItemWrap|GetWrappedMeshComponents)
          (Variables|Default|SetMaterial (Class|BTechArtGlobalLibrary|GetDynamicMaterialInstance _array_element))
          (Utilities|FlowControl|Sequence
            (:then_0
              (CallFunction|Weapon (Variables|Default|GetMaterial) (Variables|Default|GetDecalA) (Variables|Default|GetDecalB) (Variables|Default|GetDecalC)))
            (:then_1
              (CallFunction|Backpack (Variables|Default|GetMaterial) (Variables|Default|GetDecalA) (Variables|Default|GetDecalB) (Variables|Default|GetDecalC)))
            (:then_2
              (CallFunction|Character (Variables|Default|GetMaterial) (Variables|Default|GetDecalA) (Variables|Default|GetDecalB) (Variables|Default|GetDecalC)))
            (:then_3
              (CallFunction|Pickaxe (Variables|Default|GetMaterial) (Variables|Default|GetDecalA) (Variables|Default|GetDecalB) (Variables|Default|GetDecalC)))
            (:then_4
              (CallFunction|Vehicle (Variables|Default|GetMaterial) (Variables|Default|GetDecalA) (Variables|Default|GetDecalB) (Variables|Default|GetDecalC))))))
      (:Finished
        (Variables|Default|SetIsKilling? false)
        (Utilities|Time|SetTimerbyEvent (delegate Custom|Retract) 8.0)))))

(event Custom|Retract
  (Components|Timeline|SetPlayRate (Variables|CIWM_RadiantMove|GetTimeline) 0.3)
  (Timeline|Timeline ReverseFromEnd))

(event Custom|Weapon
  (params (Target :type Material%20Instance%20Dynamic%20Object%20Reference) (A :type Float%20%28double-precision%29) (B :type Float%20%28double-precision%29) (C :type Float%20%28double-precision%29))
  (Rendering|Material|SetVectorParameterValue Target "EDRec_DecalA_Offset_Weapon" (Math|Color|Lerp(LinearColor) (-0.5 -0.125 0.0 0.6) (-0.39 0.0 0.0 0.6) A) :!class /Script/Engine.MaterialInstanceDynamic)
  (Rendering|Material|SetVectorParameterValue Target "EDRec_DecalB_Offset_Weapon" (Math|Color|Lerp(LinearColor) (-0.6 -0.3 0.0 0.8) (-0.38 -0.05 0.0 0.8) B) :!class /Script/Engine.MaterialInstanceDynamic)
  (Rendering|Material|SetVectorParameterValue Target "EDRec_DecalC_Offset_Weapon" (Math|Color|Lerp(LinearColor) (-0.6 -0.4 0.0 0.6) (-0.35 0.065 0.0 0.6) C) :!class /Script/Engine.MaterialInstanceDynamic))

(event Custom|Backpack
  (params (InputPin :type Material%20Instance%20Dynamic%20Object%20Reference) (A :type Float%20%28double-precision%29) (B :type Float%20%28double-precision%29) (C :type Float%20%28double-precision%29))
  (Rendering|Material|SetVectorParameterValue InputPin "EDRec_DecalA_Offset_Backpack----A=Scale" (Math|Color|Lerp(LinearColor) (0.0 -0.64 0.0 0.7) (0.0 -0.4 0.0 0.7) A) :!class /Script/Engine.MaterialInstanceDynamic)
  (Rendering|Material|SetVectorParameterValue InputPin "EDRec_DecalB_Offset_Backpack----A=Scale" (Math|Color|Lerp(LinearColor) (0.0 -0.73 0.0 0.7) (0.0 -0.35 0.0 0.7) B) :!class /Script/Engine.MaterialInstanceDynamic)
  (Rendering|Material|SetVectorParameterValue InputPin "EDRec_DecalC_Offset_Backpack----A=Scale" (Math|Color|Lerp(LinearColor) (0.37 -0.65 0.0 0.8) (0.37 -0.1 0.0 0.8) C) :!class /Script/Engine.MaterialInstanceDynamic))

(event Custom|Character
  (params (InputPin :type Material%20Instance%20Dynamic%20Object%20Reference) (A :type Float%20%28double-precision%29) (B :type Float%20%28double-precision%29) (C :type Float%20%28double-precision%29))
  (Rendering|Material|SetVectorParameterValue InputPin "EDRec_DecalA_Offset_Character" (Math|Color|Lerp(LinearColor) (0.0 -0.58 0.0 0.7) (0.0 -0.39 0.0 0.7) A) :!class /Script/Engine.MaterialInstanceDynamic)
  (Rendering|Material|SetVectorParameterValue InputPin "EDRec_DecalB_Offset_Character" (Math|Color|Lerp(LinearColor) (0.0 -0.66 0.0 0.61) (0.0 -0.34 0.0 0.61) B) :!class /Script/Engine.MaterialInstanceDynamic)
  (Rendering|Material|SetVectorParameterValue InputPin "EDRec_DecalC_Offset_Character" (Math|Color|Lerp(LinearColor) (0.42 -0.6 0.0 0.6) (0.42 -0.15 0.0 0.6) C) :!class /Script/Engine.MaterialInstanceDynamic))

(event Custom|Pickaxe
  (params (InputPin :type Material%20Instance%20Dynamic%20Object%20Reference) (A :type Float%20%28double-precision%29) (B :type Float%20%28double-precision%29) (C :type Float%20%28double-precision%29))
  (Rendering|Material|SetVectorParameterValue InputPin "EDRec_DecalA_Offset_Pickaxe" (Math|Color|Lerp(LinearColor) (-0.2 -0.5 0.0 0.6) (-0.02 -0.4 0.0 0.6) A) :!class /Script/Engine.MaterialInstanceDynamic)
  (Rendering|Material|SetVectorParameterValue InputPin "EDRec_DecalB_Offset_Pickaxe" (Math|Color|Lerp(LinearColor) (-0.3 -0.6 0.0 0.56) (-0.06 -0.34 0.0 0.56) B) :!class /Script/Engine.MaterialInstanceDynamic)
  (Rendering|Material|SetVectorParameterValue InputPin "EDRec_DecalC_Offset_Pickaxe" (Math|Color|Lerp(LinearColor) (-0.4 -0.57 0.0 0.68) (0.09 -0.17 0.0 0.68) C) :!class /Script/Engine.MaterialInstanceDynamic))

(event Custom|Vehicle
  (params (InputPin :type Material%20Instance%20Dynamic%20Object%20Reference) (A :type Float%20%28double-precision%29) (B :type Float%20%28double-precision%29) (C :type Float%20%28double-precision%29))
  (Rendering|Material|SetVectorParameterValue InputPin "EDRec_DecalA_Offset_Vehicle" (Math|Color|Lerp(LinearColor) (-0.58 -0.15 0.0 0.4) (-0.46 -0.05 0.0 0.4) A) :!class /Script/Engine.MaterialInstanceDynamic)
  (Rendering|Material|SetVectorParameterValue InputPin "EDRec_DecalB_Offset_Vehicle" (Math|Color|Lerp(LinearColor) (-0.6 -0.19 0.0 0.3) (-0.45 -0.06 0.0 0.3) B) :!class /Script/Engine.MaterialInstanceDynamic)
  (Rendering|Material|SetVectorParameterValue InputPin "EDRec_DecalC_Offset_Vehicle" (Math|Color|Lerp(LinearColor) (-0.55 -0.2 0.0 0.45) (-0.4 -0.05 0.0 0.45) C) :!class /Script/Engine.MaterialInstanceDynamic))

(fn ConstructionScript)
