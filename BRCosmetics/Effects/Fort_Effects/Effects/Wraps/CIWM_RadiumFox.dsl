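; Reactive weapon wrap logic for RadiumFox that shows the player's kill count on wrapped meshes.
; On pawn set, it watches the AthenaKills stat, caches the count and sets a decal flipbook frame from it, clamped to 0-9 and divided by ten.
; Each stat change also stamps game time into an emissive brightness parameter.
; New weapon and weapon-type events reapply the frame after a short delay.
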(event EventOnPawnSet
  (bind _returnvalue (CustomItemWrap|GetWrappedMeshComponents))
  (bind _asfort_player_pawn (Utilities|Casting|CastToFortPlayerPawn (CustomCharacter|GetAssociatedPlayerPawn)))
  (bind (_statname _statvalue) (Cosmetics|WatchCosmeticStat _asfort_player_pawn "(TagName=\"StatManager.AthenaKills\")")
    (:OnStatChanged
      (Variables|Default|SetKill_Count _statvalue)
      (Rendering|Material|SetScalarParameterValueonMaterials _returnvalue "FDecalV2_FlipBook_Frame #" (/ (Math|Conversions|ToFloat(Integer) (Math|Integer|Clamp(Integer) (Variables|Default|GetKill_Count) 0 9)) 1e+01))
      (Rendering|Material|SetScalarParameterValueonMaterials _returnvalue "FDecalV2_Emissive_Brightness_Time_BP" (Utilities|Time|GetGameTimeinSeconds)))))

(event Custom|OnBannerFullyApplied
  (params (Icon :type Texture%202D%20Object%20Reference) (PrimaryColor :type Linear%20Color%20Structure)))

(event EventOnNewWeapons
  (params (WeaponsAsItems :type Array%20of%20Fort%20Item%20Object%20References))
  ;[1 1 1] Updated kill count on new weapon
  (Utilities|FlowControl|Delay 0.2)
  (Rendering|Material|SetScalarParameterValueonMaterials (CustomItemWrap|GetWrappedMeshComponents) "FDecalV2_FlipBook_Frame #" (/ (Math|Conversions|ToFloat(Integer) (Math|Integer|Clamp(Integer) (Variables|Default|GetKill_Count) 0 9)) 1e+01)))

(event EventOnNewPlayerWeaponType
  (params (NewType :type EFortItemType%20Enum :default "WorldItem") (OldType :type EFortItemType%20Enum :default "WorldItem"))
  ;[1 1 1] Updated kill count on new weapon
  (Utilities|FlowControl|Delay 0.2)
  (Rendering|Material|SetScalarParameterValueonMaterials (CustomItemWrap|GetWrappedMeshComponents) "FDecalV2_FlipBook_Frame #" (/ (Math|Conversions|ToFloat(Integer) (Math|Integer|Clamp(Integer) (Variables|Default|GetKill_Count) 0 9)) 1e+01)))

(fn ConstructionScript)
