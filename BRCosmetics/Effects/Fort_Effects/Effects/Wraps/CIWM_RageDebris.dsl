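; Custom item wrap modifier for the Rage Debris wrap, driving a kill counter and fire-reactive decals through material parameters.
; An AthenaKills stat watch updates the clamped counter and timestamps an emissive effect; weapon changes reapply the count after a delay.
; In the front end, a timeline animates the counter instead.
; Each shot randomly timestamps one of three decals with random rotation after a short delay, printing the pick for debugging.
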(event EventOnWeaponFired
  (bind _returnvalue (Utilities|Time|GetGameTimeinSeconds))
  (bind _returnvalue_2 (CustomItemWrap|GetWrappedMeshComponents))
  (bind _returnvalue_3 (CustomItemWrap|GetWrappedMeshComponents))
  (bind _returnvalue_4 (CustomItemWrap|GetWrappedMeshComponents))
  (bind _returnvalue_5 (CustomItemWrap|GetWrappedMeshComponents))
  (bind _returnvalue_6 (Math|Random|RandomFloat))
  (bind _returnvalue_7 (Math|Float|Lerp -0.1 0.1 _returnvalue_6))
  (bind _returnvalue_8 (Math|Random|RandomFloat))
  (bind _returnvalue_9 (Math|Float|Lerp -0.1 0.1 _returnvalue_8))
  (bind _returnvalue_10 (Math|Random|RandomFloat))
  (bind _returnvalue_11 (Math|Float|Lerp -0.1 0.1 _returnvalue_10))
  (bind _returnvalue_12 (Math|Random|RandomInteger 3))
  (Utilities|FlowControl|Delay 0.05)
  (Development|PrintString (Utilities|String|ToString(Integer) _returnvalue_12))
  (switch int _returnvalue_12
    (:0
      (Rendering|Material|SetScalarParameterValueonMaterials _returnvalue_2 "EDRec_TimeFrom_BP_DecalB" _returnvalue)
      (Rendering|Material|SetScalarParameterValueonMaterials _returnvalue_3 "EDRec_DecalB_Rotation_Backpack" _returnvalue_9)
      (Rendering|Material|SetScalarParameterValueonMaterials _returnvalue_3 "EDRec_DecalB_Rotation_Character" _returnvalue_9)
      (Rendering|Material|SetScalarParameterValueonMaterials _returnvalue_3 "EDRec_DecalB_Rotation_Mech" _returnvalue_9)
      (Rendering|Material|SetScalarParameterValueonMaterials _returnvalue_3 "EDRec_DecalB_Rotation_Pickaxe" _returnvalue_9)
      (Rendering|Material|SetScalarParameterValueonMaterials _returnvalue_3 "EDRec_DecalB_Rotation_Vehicle" _returnvalue_9)
      (Rendering|Material|SetScalarParameterValueonMaterials _returnvalue_3 "EDRec_DecalB_Rotation_Vehicle_Side" _returnvalue_9)
      (Rendering|Material|SetScalarParameterValueonMaterials _returnvalue_3 "EDRec_DecalB_Rotation_Weapon" _returnvalue_9))
    (:1
      (Rendering|Material|SetScalarParameterValueonMaterials _returnvalue_2 "EDRec_TimeFrom_BP_DecalC" _returnvalue)
      (Rendering|Material|SetScalarParameterValueonMaterials _returnvalue_4 "EDRec_DecalC_Rotation_Backpack" _returnvalue_7)
      (Rendering|Material|SetScalarParameterValueonMaterials _returnvalue_4 "EDRec_DecalC_Rotation_Character" _returnvalue_7)
      (Rendering|Material|SetScalarParameterValueonMaterials _returnvalue_4 "EDRec_DecalC_Rotation_Mech" _returnvalue_7)
      (Rendering|Material|SetScalarParameterValueonMaterials _returnvalue_4 "EDRec_DecalC_Rotation_Pickaxe" _returnvalue_7)
      (Rendering|Material|SetScalarParameterValueonMaterials _returnvalue_4 "EDRec_DecalC_Rotation_Vehicle" _returnvalue_7)
      (Rendering|Material|SetScalarParameterValueonMaterials _returnvalue_4 "EDRec_DecalC_Rotation_Vehicle_Side" _returnvalue_7)
      (Rendering|Material|SetScalarParameterValueonMaterials _returnvalue_4 "EDRec_DecalC_Rotation_Weapon" _returnvalue_7))
    (:2
      (Rendering|Material|SetScalarParameterValueonMaterials _returnvalue_2 "EDRec_TimeFrom_BP_DecalD" _returnvalue)
      (Rendering|Material|SetScalarParameterValueonMaterials _returnvalue_5 "EDRec_DecalD_Rotation_Backpack" _returnvalue_11)
      (Rendering|Material|SetScalarParameterValueonMaterials _returnvalue_5 "EDRec_DecalD_Rotation_Character" _returnvalue_11)
      (Rendering|Material|SetScalarParameterValueonMaterials _returnvalue_5 "EDRec_DecalD_Rotation_Mech" _returnvalue_11)
      (Rendering|Material|SetScalarParameterValueonMaterials _returnvalue_5 "EDRec_DecalD_Rotation_Pickaxe" _returnvalue_11)
      (Rendering|Material|SetScalarParameterValueonMaterials _returnvalue_5 "EDRec_DecalD_Rotation_Vehicle" _returnvalue_11)
      (Rendering|Material|SetScalarParameterValueonMaterials _returnvalue_5 "EDRec_DecalD_Rotation_Vehicle_Side" _returnvalue_11)
      (Rendering|Material|SetScalarParameterValueonMaterials _returnvalue_5 "EDRec_DecalD_Rotation_Weapon" _returnvalue_11))))

(event EventBeginPlay
  (bind _returnvalue (CustomItemWrap|GetWrappedMeshComponents))
  (if (CustomCosmeticModifier|IsinFrontEnd)
    (Utilities|FlowControl|Sequence
      (:then_0
        (bind (_direction _newtrack_2) (Timeline|Timeline PlayFromStart)
          (:NewTrack_1
            (Variables|Default|SetKill_Count (Math|Float|Truncate _newtrack_2))
            (Rendering|Material|SetScalarParameterValueonMaterials _returnvalue "EDRec_Counter__Number" (Math|Float|Clamp(Float) (Math|Conversions|ToFloat(Integer) (Variables|Default|GetKill_Count)) 0.0 99.0))
            (Rendering|Material|SetScalarParameterValueonMaterials _returnvalue "EDRec_Emissive_A_Brightness_Time_BP" (Utilities|Time|GetGameTimeinSeconds))))))))

(event EventOnPawnSet
  (bind _returnvalue (CustomItemWrap|GetWrappedMeshComponents))
  (Utilities|FlowControl|Sequence
    (:then_1
      (bind (_statname _statvalue) (Cosmetics|WatchCosmeticStat (CustomCharacter|GetAssociatedPlayerPawn) "(TagName=\"StatManager.AthenaKills\")" 0 false)
        (:OnStatChanged
          (Variables|Default|SetKill_Count _statvalue)
          (Rendering|Material|SetScalarParameterValueonMaterials _returnvalue "EDRec_Counter__Number" (Math|Float|Clamp(Float) (Math|Conversions|ToFloat(Integer) (Variables|Default|GetKill_Count)) 0.0 99.0))
          (Rendering|Material|SetScalarParameterValueonMaterials _returnvalue "EDRec_Emissive_A_Brightness_Time_BP" (Utilities|Time|GetGameTimeinSeconds)))))))

(event EventOnWeaponChanged
  (params (NewWeapon :type Fort%20Weapon%20Object%20Reference) (OldWeapon :type Fort%20Weapon%20Object%20Reference))
  (Utilities|FlowControl|Delay 0.2)
  (Rendering|Material|SetScalarParameterValueonMaterials (CustomItemWrap|GetWrappedMeshComponents) "EDRec_Counter__Number" (Math|Float|Clamp(Float) (Math|Conversions|ToFloat(Integer) (Variables|Default|GetKill_Count)) 0.0 99.0)))

(event EventOnNewWeapons
  (params (WeaponsAsItems :type Array%20of%20Fort%20Item%20Object%20References))
  (Utilities|FlowControl|Delay 0.2)
  (Rendering|Material|SetScalarParameterValueonMaterials (CustomItemWrap|GetWrappedMeshComponents) "EDRec_Counter__Number" (Math|Float|Clamp(Float) (Math|Conversions|ToFloat(Integer) (Variables|Default|GetKill_Count)) 0.0 99.0)))

(event EventOnNewPlayerWeaponType
  (params (NewType :type EFortItemType%20Enum :default "WorldItem") (OldType :type EFortItemType%20Enum :default "WorldItem"))
  (Utilities|FlowControl|Delay 0.2)
  (Rendering|Material|SetScalarParameterValueonMaterials (CustomItemWrap|GetWrappedMeshComponents) "EDRec_Counter__Number" (Math|Float|Clamp(Float) (Math|Conversions|ToFloat(Integer) (Variables|Default|GetKill_Count)) 0.0 99.0)))

(fn ConstructionScript)

(fn Weapon_Offsets)
