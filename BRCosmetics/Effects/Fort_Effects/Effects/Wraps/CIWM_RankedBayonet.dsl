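; Weapon wrap logic for RankedBayonet that animates emissive color and scroll offset on the wrapped meshes when the weapon fires.
; A shot starts a timeline that drives the color from a linear color curve, with alpha as intensity, and accumulates the offset.
; A 0.01-second looping timer adds extra scroll until the timeline finishes and clears the Pause flag, which actually means active.
; Repeat shots jump the timeline to 0.5 once its value reaches 0.3.
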(event EventOnWeaponFired
  (bind _returnvalue (CustomItemWrap|GetWrappedMeshComponents))
  (if (Variables|Default|GetPause)
    (if (>= (Variables|Default|GetFloat_Value) 0.3)
      (bind (_direction _color _offset) (Timeline|Timeline SetNewTime :NewTime 0.5)
        (:Update
          (Variables|Default|SetFloat_Value _color)
          (bind (_returnvalue_r _returnvalue_g _returnvalue_b _returnvalue_a) (Math|Color|BreakColor (Math|Curves|GetLinearColorValue (Variables|Default|GetNewVar) _color)))
          (Rendering|Material|SetVectorParameterValueonMaterials _returnvalue "TopS Emissive_01_Color-" (Math|Conversions|ToVector(LinearColor) (* (Utilities|Struct|MakeLinearColor _returnvalue_r _returnvalue_g _returnvalue_b 1.0) _returnvalue_a)))
          (bind _output_get (Variables|Default|SetOffset (+ _offset (Variables|Default|GetOffset))))
          (Rendering|Material|SetVectorParameterValueonMaterials _returnvalue "TopS_Em01_Offset__Master" (Math|Vector|MakeVector _output_get)))
        (:Finished
          (Variables|Default|SetPause false))))
    (else
      (Variables|Default|SetPause true)
      (bind _returnvalue_2 (Utilities|Time|SetTimerbyFunctionName 0 "Firing" 0.01 true))
      (Variables|Default|SetTimer _returnvalue_2)
      (Timeline|Timeline PlayFromStart))))

(event Custom|Firing
  (if (Variables|Default|GetPause)
    (bind _output_get (Variables|Default|SetOffset (+ (Variables|Default|GetOffset) 0.002)))
    (Rendering|Material|SetVectorParameterValueonMaterials (CustomItemWrap|GetWrappedMeshComponents) "TopS_Em01_Offset__Master" (Math|Vector|MakeVector _output_get))
    (else
      (Utilities|Time|ClearandInvalidateTimerbyHandle (Variables|Default|GetTimer)))))

(fn ConstructionScript)
