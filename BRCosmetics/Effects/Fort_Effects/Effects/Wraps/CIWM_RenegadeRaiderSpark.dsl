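; Blueprint logic for the Renegade Raider Spark weapon wrap, which makes the wrap's material react each time the weapon fires.
; On the fire event it gets the wrapped mesh components and, after a very short delay, stamps the current game time into a material parameter.
; It also sets a random crack offset on those materials, so the timing and crack variation are driven per shot through scalar parameters.
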(event EventOnWeaponFired
  (bind _returnvalue (CustomItemWrap|GetWrappedMeshComponents))
  (Utilities|FlowControl|Delay 0.01)
  (Rendering|Material|SetScalarParameterValueonMaterials _returnvalue "Ren_TimeFrom-BP----------------" (Utilities|Time|GetGameTimeinSeconds))
  (Rendering|Material|SetScalarParameterValueonMaterials _returnvalue "Ren_Crack_Offset" (Math|Random|RandomFloat)))

(fn ConstructionScript)
