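; Custom item wrap Blueprint for RhombCamo that makes the wrap react to the player's kill count.
; On pawn set, it watches the AthenaKills stat, stores each new count and maps it, clamped, to a decal flipbook frame on wrapped meshes.
; Stat changes also write game time into emissive material parameters as animation timestamps.
; New weapons or weapon types reapply the frame after a short delay.
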(event EventOnPawnSet
  (bind _returnvalue (Utilities|Time|GetGameTimeinSeconds))
  (bind _returnvalue_2 (CustomItemWrap|GetWrappedMeshComponents))
  (bind _asfort_player_pawn (Utilities|Casting|CastToFortPlayerPawn (CustomCharacter|GetAssociatedPlayerPawn)))
  (bind (_statname _statvalue) (Cosmetics|WatchCosmeticStat _asfort_player_pawn "(TagName=\"StatManager.AthenaKills\")")
    (:OnStatChanged
      (Variables|Default|SetKill_Count _statvalue)
      (Rendering|Material|SetScalarParameterValueonMaterials _returnvalue_2 "FDecal_FlipBook_Frame #" (/ (Math|Conversions|ToFloat(Integer) (Math|Integer|Clamp(Integer) (Variables|Default|GetKill_Count) 0 13)) 16.0))
      (Rendering|Material|SetScalarParameterValueonMaterials _returnvalue_2 "FDecal_Emissive_Brightness_Time_BP" _returnvalue)
      (Rendering|Material|SetScalarParameterValueonMaterials _returnvalue_2 "TopS Emissive_01_Offset----TBP" _returnvalue))))

(event Custom|OnBannerFullyApplied
  (params (Icon :type Texture%202D%20Object%20Reference) (PrimaryColor :type Linear%20Color%20Structure)))

(event EventOnNewWeapons
  (params (WeaponsAsItems :type Array%20of%20Fort%20Item%20Object%20References))
  ;[1 1 1] Updated kill count on new weapon
  (Utilities|FlowControl|Delay 0.2)
  (Rendering|Material|SetScalarParameterValueonMaterials (CustomItemWrap|GetWrappedMeshComponents) "FDecal_FlipBook_Frame #" (/ (Math|Conversions|ToFloat(Integer) (Math|Integer|Clamp(Integer) (Variables|Default|GetKill_Count) 0 13)) 16.0)))

(event EventOnNewPlayerWeaponType
  (params (NewType :type EFortItemType%20Enum :default "WorldItem") (OldType :type EFortItemType%20Enum :default "WorldItem"))
  ;[1 1 1] Updated kill count on new weapon
  (Utilities|FlowControl|Delay 0.2)
  (Rendering|Material|SetScalarParameterValueonMaterials (CustomItemWrap|GetWrappedMeshComponents) "FDecal_FlipBook_Frame #" (/ (Math|Conversions|ToFloat(Integer) (Math|Integer|Clamp(Integer) (Variables|Default|GetKill_Count) 0 13)) 16.0)))

(fn ConstructionScript)
