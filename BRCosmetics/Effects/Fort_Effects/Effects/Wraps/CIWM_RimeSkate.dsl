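; Rime Skate weapon wrap Blueprint that triggers reactive material animations on wrapped meshes from combat stats or front-end firing.
; On pawn set it watches kills, damage dealt and damage received; in the front end, every seventh shot triggers the effect instead.
; Both paths stamp game time into top emissive and side decal parameters, with side decals delayed, emissive speed gated by a pause flag, and blend opacity by a cooldown.
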(event EventOnWeaponFired
  (bind _returnvalue (Utilities|Time|GetGameTimeinSeconds))
  (bind _returnvalue_2 (CustomItemWrap|GetWrappedMeshComponents))
  (bind _returnvalue_3 (Utilities|Time|GetGameTimeinSeconds))
  (bind _returnvalue_4 (CustomItemWrap|GetWrappedMeshComponents))
  (bind _returnvalue_5 (Utilities|Time|GetGameTimeinSeconds))
  (bind _returnvalue_6 (CustomItemWrap|GetWrappedMeshComponents))
  (if (CustomCosmeticModifier|IsinFrontEnd)
    (Utilities|FlowControl|Delay 0.001)
    (bind _output_get (Variables|Default|SetShotsFired (+ (Variables|Default|GetShotsFired) 1)))
    (if (>= _output_get 5)
      (if (== (Variables|Default|GetShotsFired) 7)
        (Variables|Default|SetShotsFired 0)
        (Utilities|FlowControl|Sequence
          (:then_0
            (if (>= (Utilities|Time|GetGameTimeinSeconds) (Variables|Default|GetFloat_Value))
              (Rendering|Material|SetScalarParameterValueonMaterials (CustomItemWrap|GetWrappedMeshComponents) "TopS_01_Blend_Opacity-TBP" (Utilities|Time|GetGameTimeinSeconds))
              (else
                _))
            ;[1 1 1] TopsSide V2
            (Rendering|Material|SetScalarParameterValueonMaterials _returnvalue_6 "TopS_01_Blend_Offset-TBP" _returnvalue)
            (Rendering|Material|SetScalarParameterValueonMaterials _returnvalue_6 "TimeFrom-BP----------------" _returnvalue)
            (Utilities|FlowControl|Delay 0.5)
            ;[1 1 1] Left Right v1
            (Rendering|Material|SetScalarParameterValueonMaterials _returnvalue_2 "FDecal_Emissive_Gradent_Time_BP" _returnvalue_3)
            (Rendering|Material|SetScalarParameterValueonMaterials _returnvalue_2 "FDecal_Opacity_Time_BP_Z" _returnvalue_3)
            (Rendering|Material|SetScalarParameterValueonMaterials _returnvalue_2 "FDecal_Emissive_Brightness_Time_BP" _returnvalue_3)
            ; Left\Right V2
            (Rendering|Material|SetScalarParameterValueonMaterials _returnvalue_4 "FDecalV2_FlipBook_Time_FromBP" _returnvalue_5)
            (Rendering|Material|SetScalarParameterValueonMaterials _returnvalue_4 "FDecalaV2_Opacity_Time_BP_Z" _returnvalue_5)
            (Rendering|Material|SetScalarParameterValueonMaterials _returnvalue_4 "FDecalV2_Emissive_Brightness_Time_BP" _returnvalue_5)
            (Rendering|Material|SetScalarParameterValueonMaterials _returnvalue_4 "FDecalV2_Emissive_Gradent_Time_BP" _returnvalue_5)
            (Rendering|Material|SetScalarParameterValueonMaterials _returnvalue_4 "FDecalV2_Counter_Brightness_Time_BP" _returnvalue_5))
          (:then_1
            (Rendering|Material|SetScalarParameterValueonMaterials _returnvalue_6 "TopS Emissive_01-Brightness-TBP" _returnvalue)
            (Rendering|Material|SetScalarParameterValueonMaterials _returnvalue_6 "TopS Emissive_01_Offset----TBP" _returnvalue)
            (Rendering|Material|SetScalarParameterValueonMaterials _returnvalue_6 "TopS Emissive_03-Brightness-TBP" _returnvalue)
            (Rendering|Material|SetScalarParameterValueonMaterials _returnvalue_6 "TopS Emissive_03_Offset----TBP" _returnvalue)
            (if (not (Variables|Default|GetPause))
              (Variables|Default|SetPause true)
              (Rendering|Material|SetScalarParameterValueonMaterials _returnvalue_6 "TopS Emissive_01_Speed-TBP" _returnvalue)
              (Rendering|Material|SetScalarParameterValueonMaterials _returnvalue_6 "TopS Emissive_02_Speed-TBP" _returnvalue)
              (Rendering|Material|SetScalarParameterValueonMaterials _returnvalue_6 "TopS Emissive_03_Speed-TBP" _returnvalue)
              (Utilities|FlowControl|Delay 1.0)
              (Variables|Default|SetPause false)))
          (:then_2
            (Rendering|Material|SetScalarParameterValueonMaterials _returnvalue_6 "TopS Emissive_02-Brightness-TBP" _returnvalue)
            (Rendering|Material|SetScalarParameterValueonMaterials _returnvalue_6 "TopS Emissive_02_Offset----TBP" _returnvalue)
            (Variables|Default|SetFloat_Value (+ (Utilities|Time|GetGameTimeinSeconds) 5.0))))))))

(event EventOnPawnSet
  (bind _returnvalue (Utilities|Time|GetGameTimeinSeconds))
  (bind _returnvalue_2 (CustomCharacter|GetAssociatedPlayerPawn))
  (bind _returnvalue_3 (CustomItemWrap|GetWrappedMeshComponents))
  (bind _returnvalue_4 (Utilities|Time|GetGameTimeinSeconds))
  (bind _returnvalue_5 (CustomItemWrap|GetWrappedMeshComponents))
  (bind _returnvalue_6 (Utilities|Time|GetGameTimeinSeconds))
  (bind _returnvalue_7 (CustomItemWrap|GetWrappedMeshComponents))
  (Cosmetics|WatchCosmeticStat _returnvalue_2 "(TagName=\"StatManager.AthenaKills\")" 0 false
    (:then
      (Cosmetics|WatchCosmeticStat _returnvalue_2 "(TagName=\"StatManager.DamageDealt\")" 0 false
        (:then
          (Cosmetics|WatchCosmeticStat _returnvalue_2 "(TagName=\"StatManager.DamageReceived\")" 0 false
            (:OnStatChanged
              _)))
        (:OnStatChanged
          _)))
    (:OnStatChanged
      _))
  (Utilities|FlowControl|Sequence
    (:then_0
      (if (>= (Utilities|Time|GetGameTimeinSeconds) (Variables|Default|GetFloat_Value))
        (Rendering|Material|SetScalarParameterValueonMaterials (CustomItemWrap|GetWrappedMeshComponents) "TopS_01_Blend_Opacity-TBP" (Utilities|Time|GetGameTimeinSeconds))
        (else
          _))
      ;[1 1 1] TopsSide V2
      (Rendering|Material|SetScalarParameterValueonMaterials _returnvalue_7 "TopS_01_Blend_Offset-TBP" _returnvalue)
      (Rendering|Material|SetScalarParameterValueonMaterials _returnvalue_7 "TimeFrom-BP----------------" _returnvalue)
      (Utilities|FlowControl|Delay 0.5)
      ;[1 1 1] Left Right v1
      (Rendering|Material|SetScalarParameterValueonMaterials _returnvalue_3 "FDecal_Emissive_Gradent_Time_BP" _returnvalue_4)
      (Rendering|Material|SetScalarParameterValueonMaterials _returnvalue_3 "FDecal_Opacity_Time_BP_Z" _returnvalue_4)
      (Rendering|Material|SetScalarParameterValueonMaterials _returnvalue_3 "FDecal_Emissive_Brightness_Time_BP" _returnvalue_4)
      ; Left\Right V2
      (Rendering|Material|SetScalarParameterValueonMaterials _returnvalue_5 "FDecalV2_FlipBook_Time_FromBP" _returnvalue_6)
      (Rendering|Material|SetScalarParameterValueonMaterials _returnvalue_5 "FDecalaV2_Opacity_Time_BP_Z" _returnvalue_6)
      (Rendering|Material|SetScalarParameterValueonMaterials _returnvalue_5 "FDecalV2_Emissive_Brightness_Time_BP" _returnvalue_6)
      (Rendering|Material|SetScalarParameterValueonMaterials _returnvalue_5 "FDecalV2_Emissive_Gradent_Time_BP" _returnvalue_6)
      (Rendering|Material|SetScalarParameterValueonMaterials _returnvalue_5 "FDecalV2_Counter_Brightness_Time_BP" _returnvalue_6))
    (:then_1
      (Rendering|Material|SetScalarParameterValueonMaterials _returnvalue_7 "TopS Emissive_01-Brightness-TBP" _returnvalue)
      (Rendering|Material|SetScalarParameterValueonMaterials _returnvalue_7 "TopS Emissive_01_Offset----TBP" _returnvalue)
      (Rendering|Material|SetScalarParameterValueonMaterials _returnvalue_7 "TopS Emissive_03-Brightness-TBP" _returnvalue)
      (Rendering|Material|SetScalarParameterValueonMaterials _returnvalue_7 "TopS Emissive_03_Offset----TBP" _returnvalue)
      (if (not (Variables|Default|GetPause))
        (Variables|Default|SetPause true)
        (Rendering|Material|SetScalarParameterValueonMaterials _returnvalue_7 "TopS Emissive_01_Speed-TBP" _returnvalue)
        (Rendering|Material|SetScalarParameterValueonMaterials _returnvalue_7 "TopS Emissive_02_Speed-TBP" _returnvalue)
        (Rendering|Material|SetScalarParameterValueonMaterials _returnvalue_7 "TopS Emissive_03_Speed-TBP" _returnvalue)
        (Utilities|FlowControl|Delay 1.0)
        (Variables|Default|SetPause false)))
    (:then_2
      (Rendering|Material|SetScalarParameterValueonMaterials _returnvalue_7 "TopS Emissive_02-Brightness-TBP" _returnvalue)
      (Rendering|Material|SetScalarParameterValueonMaterials _returnvalue_7 "TopS Emissive_02_Offset----TBP" _returnvalue)
      (Variables|Default|SetFloat_Value (+ (Utilities|Time|GetGameTimeinSeconds) 5.0)))))

(fn ConstructionScript)
