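; Custom item wrap logic for the RockParty weapon wrap that changes the wrap's look each time the weapon fires.
; On fire, it writes game time into an emissive brightness material parameter, then waits through a short Delay.
; A switch on MskNumber picks the next of five mask textures, looping back to the start.
; MskTextureSwap applies that texture to every dynamic material on the wrapped meshes and stores the next index.
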(event EventOnWeaponFired
  (Rendering|Material|SetScalarParameterValueonMaterials (CustomItemWrap|GetWrappedMeshComponents) "TopS Emissive_02-Brightness-TBP" (Utilities|Time|GetGameTimeinSeconds))
  (Utilities|FlowControl|Delay 0.2)
  (switch int (Variables|Default|GetMskNumber)
    (:0
      (CallFunction|MskTextureSwap "/BRCosmetics/Weapons/WeaponSkins/Wraps/Textures/RockParty/T_Wrap_RockParty_MskC.T_Wrap_RockParty_MskC" 1))
    (:1
      (CallFunction|MskTextureSwap "/BRCosmetics/Weapons/WeaponSkins/Wraps/Textures/RockParty/T_Wrap_RockParty_MskD.T_Wrap_RockParty_MskD" 2))
    (:2
      (CallFunction|MskTextureSwap "/BRCosmetics/Weapons/WeaponSkins/Wraps/Textures/RockParty/T_Wrap_RockParty_MskE.T_Wrap_RockParty_MskE" 3))
    (:3
      (CallFunction|MskTextureSwap "/BRCosmetics/Weapons/WeaponSkins/Wraps/Textures/RockParty/T_Wrap_RockParty_MskF.T_Wrap_RockParty_MskF" 4))
    (:4
      (CallFunction|MskTextureSwap "/BRCosmetics/Weapons/WeaponSkins/Wraps/Textures/RockParty/T_Wrap_RockParty_MskB.T_Wrap_RockParty_MskB"))))

(fn MskTextureSwap
  (params (Msk_Texture :type Texture%20Object%20Reference) (MskNumber :type Integer))
  (for _array_element (CustomItemWrap|GetWrappedMeshComponents)
    (for _array_element_2 (Rendering|Material|GetMaterials _array_element)
      (bind _asmaterial_instance_dynamic (Utilities|Casting|CastToMaterialInstanceDynamic _array_element_2))
      (Rendering|Material|SetTextureParameterValue _asmaterial_instance_dynamic "TopS_00A_Sharerd_Tex_Msk" Msk_Texture)
      (Variables|Default|SetMskNumber MskNumber))))

(fn ConstructionScript)
