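; RoomReader weapon wrap modifier that flashes a decal effect on wrapped meshes when the player's kill stat changes, or every fifth shot in the front end.
; Each trigger stamps game time into material parameters for decal, offset and emissive brightness, setting opacity endpoints to fade in, then out five seconds later.
; A Pause flag blocks retriggers until the cycle ends.
; In the front end, BeginPlay also starts a timeline after a short delay.
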(event EventBeginPlay
  (Utilities|FlowControl|Delay 0.2)
  (if (CustomCosmeticModifier|IsinFrontEnd)
    (Timeline|Timeline PlayFromStart)))

(event EventOnWeaponFired
  (bind _returnvalue (Utilities|Time|GetGameTimeinSeconds))
  (bind _returnvalue_2 (CustomItemWrap|GetWrappedMeshComponents))
  (bind _returnvalue_3 (CustomItemWrap|GetWrappedMeshComponents))
  (if (CustomCosmeticModifier|IsinFrontEnd)
    (bind _output_get (Variables|Default|SetIterationCount (+ (Variables|Default|GetIterationCount) 1)))
    (if (>= _output_get 5)
      (Variables|Default|SetIterationCount 0)
      (if (not (Variables|Default|GetPause))
        (Variables|Default|SetPause true)
        (Utilities|FlowControl|Sequence
          (:then_0
            (Rendering|Material|SetScalarParameterValueonMaterials _returnvalue_3 "EDRec_TimeFrom_BP_DecalA" _returnvalue)
            (Rendering|Material|SetScalarParameterValueonMaterials _returnvalue_3 "EDRec_DecalA_Opacity_Start" 0.0)
            (Rendering|Material|SetScalarParameterValueonMaterials _returnvalue_3 "EDRec_DecalA_Opacity_End" 1.0)
            (Utilities|FlowControl|Delay 5.0)
            (Rendering|Material|SetScalarParameterValueonMaterials _returnvalue_2 "EDRec_TimeFrom_BP_DecalA" (Utilities|Time|GetGameTimeinSeconds))
            (Rendering|Material|SetScalarParameterValueonMaterials _returnvalue_2 "EDRec_DecalA_Opacity_Start" 1.0)
            (Rendering|Material|SetScalarParameterValueonMaterials _returnvalue_2 "EDRec_DecalA_Opacity_End" 0.0)
            (Utilities|FlowControl|Delay 1.0)
            (Variables|Default|SetPause false))
          (:then_1
            (Rendering|Material|SetScalarParameterValueonMaterials _returnvalue_3 "EDRec_TimeFrom_BP_Offset_DecalA" _returnvalue)
            (Rendering|Material|SetScalarParameterValueonMaterials _returnvalue_3 "TopS Emissive_01-Brightness-TBP" _returnvalue)))))))

(event EventOnPawnSet
  (bind _returnvalue (Utilities|Time|GetGameTimeinSeconds))
  (bind _returnvalue_2 (CustomItemWrap|GetWrappedMeshComponents))
  (bind _returnvalue_3 (CustomItemWrap|GetWrappedMeshComponents))
  (Cosmetics|WatchCosmeticStat (CustomCharacter|GetAssociatedPlayerPawn) "(TagName=\"StatManager.AthenaKills\")" 0 false
    (:OnStatChanged
      (if (not (Variables|Default|GetPause))
        (Variables|Default|SetPause true)
        (Utilities|FlowControl|Sequence
          (:then_0
            (Rendering|Material|SetScalarParameterValueonMaterials _returnvalue_3 "EDRec_TimeFrom_BP_DecalA" _returnvalue)
            (Rendering|Material|SetScalarParameterValueonMaterials _returnvalue_3 "EDRec_DecalA_Opacity_Start" 0.0)
            (Rendering|Material|SetScalarParameterValueonMaterials _returnvalue_3 "EDRec_DecalA_Opacity_End" 1.0)
            (Utilities|FlowControl|Delay 5.0)
            (Rendering|Material|SetScalarParameterValueonMaterials _returnvalue_2 "EDRec_TimeFrom_BP_DecalA" (Utilities|Time|GetGameTimeinSeconds))
            (Rendering|Material|SetScalarParameterValueonMaterials _returnvalue_2 "EDRec_DecalA_Opacity_Start" 1.0)
            (Rendering|Material|SetScalarParameterValueonMaterials _returnvalue_2 "EDRec_DecalA_Opacity_End" 0.0)
            (Utilities|FlowControl|Delay 1.0)
            (Variables|Default|SetPause false))
          (:then_1
            (Rendering|Material|SetScalarParameterValueonMaterials _returnvalue_3 "EDRec_TimeFrom_BP_Offset_DecalA" _returnvalue)
            (Rendering|Material|SetScalarParameterValueonMaterials _returnvalue_3 "TopS Emissive_01-Brightness-TBP" _returnvalue)))))))

(event EventOnNewWeapons
  (params (WeaponsAsItems :type Array%20of%20Fort%20Item%20Object%20References))
  (Utilities|FlowControl|Delay 0.2))

(event EventOnNewPlayerWeaponType
  (params (NewType :type EFortItemType%20Enum :default "WorldItem") (OldType :type EFortItemType%20Enum :default "WorldItem"))
  (Utilities|FlowControl|Delay 0.2))

(fn ConstructionScript)
