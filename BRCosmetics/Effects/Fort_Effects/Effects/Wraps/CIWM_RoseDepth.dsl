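; Weapon wrap logic for the RoseDepth cosmetic that steps up emissive glow, flame and ember brightness as the weapon fires.
; On begin play it picks front-end or in-game brightness arrays and step caps, with a timeline driving a time-of-day override in the front end.
; Each shot advances a capped step, lerped toward zero after a pause in firing, then writes that step's brightness and game-time stamps to wrapped-mesh materials.
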(event EventBeginPlay
  (Utilities|FlowControl|Delay 0.2)
  (if (CustomCosmeticModifier|IsinFrontEnd)
    (Utilities|FlowControl|Sequence
      (:then_0
        (Variables|Default|SetMaxSteps 5)
        (Variables|Default|SetBrightness%20value (Variables|Default|GetBrightness%20value_FE))
        (Variables|Default|SetBrightness%20Flames (Variables|Default|GetBrightness%20Flames_FE))
        (Variables|Default|SetEmbers (Variables|Default|GetEmbers_FE)))
      (:then_1
        (bind (_direction _tod_value) (Timeline|TOD PlayFromStart)
          (:Update
            (Rendering|Material|SetScalarParameterValueonMaterials (CustomItemWrap|GetWrappedMeshComponents) "00-FE-TODCC-Override" _tod_value)))))
    (else
      (Variables|Default|SetMaxSteps 9)
      (Variables|Default|SetBrightness%20value (Variables|Default|GetBrightness%20value_PIE))
      (Variables|Default|SetBrightness%20Flames (Variables|Default|GetBrightness%20Flames_PIE))
      (Variables|Default|SetEmbers (Variables|Default|GetEmbers_PIE)))))

(event EventOnWeaponChanged
  (params (NewWeapon :type Fort%20Weapon%20Object%20Reference) (OldWeapon :type Fort%20Weapon%20Object%20Reference))
  (Variables|Default|SetMaxSteps 9)
  (Variables|Default|SetBrightness%20value (Variables|Default|GetBrightness%20value_PIE))
  (Variables|Default|SetBrightness%20Flames (Variables|Default|GetBrightness%20Flames_PIE))
  (Variables|Default|SetEmbers (Variables|Default|GetEmbers_PIE)))

(event EventOnWeaponFired
  (bind _returnvalue (CustomItemWrap|GetWrappedMeshComponents))
  (bind _returnvalue_2 (Utilities|Time|GetGameTimeinSeconds))
  (bind _returnvalue_3 (Utilities|Time|GetGameTimeinSeconds))
  (bind _returnvalue_4 (CustomItemWrap|GetWrappedMeshComponents))
  (bind _returnvalue_5 (CustomItemWrap|GetWrappedMeshComponents))
  (bind _returnvalue_6 (Utilities|Time|GetGameTimeinSeconds))
  (bind _stepcount (Variables|Default|GetStepCount))
  (bind _embers (Variables|Default|GetEmbers))
  (bind _output (Utilities|Array|Get(aref) _embers _stepcount))
  (bind _stepcount_2 (Variables|Default|GetStepCount))
  (bind _brightness_flames (Variables|Default|GetBrightness%20Flames))
  (bind _output_2 (Utilities|Array|Get(aref) _brightness_flames _stepcount_2))
  (bind _brightness_value (Variables|Default|GetBrightness%20value))
  (Utilities|FlowControl|Delay 0.01)
  (bind _output_get (Variables|Default|SetStepCount (Math|Integer|Min(Integer) (+ (Math|Float|Truncate (Math|Float|Lerp (Math|Conversions|ToFloat(Integer) (Variables|Default|GetStepCount)) 0.0 (Math|Float|Clamp(Float) (/ (- (Utilities|Time|GetGameTimeinSeconds) (Variables|Default|Getwait_timer)) 1.0)))) 1) (Variables|Default|GetMaxSteps))))
  (Rendering|Material|SetScalarParameterValueonMaterials _returnvalue "EDRec_Emissive_C_Brightness_Time__Start" (Utilities|Array|Get(aref) _brightness_value _output_get))
  (Rendering|Material|SetScalarParameterValueonMaterials _returnvalue "EDRec_Emissive_D_Brightness_Time__Start" (Utilities|Array|Get(aref) _brightness_value _output_get))
  (Rendering|Material|SetScalarParameterValueonMaterials _returnvalue "EDRec_Emissive_C_Brightness_Time_BP" _returnvalue_3)
  (Rendering|Material|SetScalarParameterValueonMaterials _returnvalue "EDRec_Emissive_D_Brightness_Time_BP" _returnvalue_3)
  (Variables|Default|Setwait_timer (+ (Utilities|Time|GetGameTimeinSeconds) 5.5))
  (Utilities|FlowControl|Sequence
    (:then_0
      (Rendering|Material|SetScalarParameterValueonMaterials _returnvalue_5 "TopS Emissive_01-Brightness-TBP" _returnvalue_6)
      (Rendering|Material|SetScalarParameterValueonMaterials _returnvalue_5 "TopS Emissive_02-Brightness-TBP" _returnvalue_6)
      (Rendering|Material|SetScalarParameterValueonMaterials _returnvalue_5 "TopS Emissive_01_Brightness---Start" _output)
      (Rendering|Material|SetScalarParameterValueonMaterials _returnvalue_5 "TopS Emissive_02_Brightness---Start" _output))
    (:then_1
      (Rendering|Material|SetScalarParameterValueonMaterials _returnvalue_4 "EDRec_Emissive_A_Brightness_Time_BP" _returnvalue_2)
      (Rendering|Material|SetScalarParameterValueonMaterials _returnvalue_4 "EDRec_Emissive_A_Brightness_Time__Start" _output_2)
      (Rendering|Material|SetScalarParameterValueonMaterials _returnvalue_4 "EDRec_Emissive_B_Brightness_Time_BP" _returnvalue_2)
      (Rendering|Material|SetScalarParameterValueonMaterials _returnvalue_4 "EDRec_Emissive_B_Brightness_Time__Start" _output_2))))

(fn ConstructionScript)
