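; A custom item wrap modifier whose weapon wrap materials animate in response to firing and player stats.
; On weapon fire outside the front end, it counts shots, with threshold checks driving emissive pulses and a counter reset that also triggers a decal flipbook burst.
; On pawn set, it watches damage and kill stats, and kills trigger the same burst.
; Effects start by writing game time into material scalar parameters, leaving the animation to the shader.
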(event EventOnWeaponFired
  (bind _returnvalue (Utilities|Time|GetGameTimeinSeconds))
  (bind _returnvalue_2 (CustomItemWrap|GetWrappedMeshComponents))
  (bind _returnvalue_3 (CustomItemWrap|GetWrappedMeshComponents))
  (if (CustomCosmeticModifier|IsinFrontEnd)
    (Utilities|FlowControl|Delay 0.001)
    (Utilities|FlowControl|Sequence
      (:then_1
        (bind _output_get (Variables|Default|SetShotsFired (+ (Variables|Default|GetShotsFired) 1)))
        (if (>= _output_get 5)
          (if (== (Variables|Default|GetShotsFired) 7)
            (Variables|Default|SetShotsFired 0)
            (Utilities|FlowControl|Sequence
              (:then_0
                _)
              (:then_1
                (Rendering|Material|SetScalarParameterValueonMaterials _returnvalue_2 "FDecalV2_FlipBook_Time_FromBP" (Utilities|Time|GetGameTimeinSeconds))
                (Rendering|Material|SetScalarParameterValueonMaterials _returnvalue_2 "EDRec_Emissive_C_Brightness_Time_BP" _returnvalue)
                (Rendering|Material|SetScalarParameterValueonMaterials _returnvalue_2 "EDRec_Emissive_B_Brightness_Time_BP" _returnvalue)))
            (elif (Variables|Default|GetPause_Kill)
              _))
          (Rendering|Material|SetScalarParameterValueonMaterials _returnvalue_3 "FDecalV2_Emissive_Brightness_Time_BP" (Utilities|Time|GetGameTimeinSeconds))
          (Rendering|Material|SetScalarParameterValueonMaterials _returnvalue_3 "FDecalV2_Emissive_Gradent_Time_BP" (Utilities|Time|GetGameTimeinSeconds)))))))

(event EventOnPawnSet
  (bind _returnvalue (Utilities|Time|GetGameTimeinSeconds))
  (bind _returnvalue_2 (CustomItemWrap|GetWrappedMeshComponents))
  (bind _returnvalue_3 (CustomItemWrap|GetWrappedMeshComponents))
  (Utilities|FlowControl|Sequence
    (:then_0
      (Cosmetics|WatchCosmeticStat (CustomCharacter|GetAssociatedPlayerPawn) "(TagName=\"StatManager.DamageDealt\")" 0 false
        (:OnStatChanged
          _)))
    (:then_1
      (Cosmetics|WatchCosmeticStat (CustomCharacter|GetAssociatedPlayerPawn) "(TagName=\"StatManager.AthenaKills\")" 0 false
        (:OnStatChanged
          (Utilities|FlowControl|Sequence
            (:then_0
              _)
            (:then_1
              (Rendering|Material|SetScalarParameterValueonMaterials _returnvalue_2 "FDecalV2_FlipBook_Time_FromBP" (Utilities|Time|GetGameTimeinSeconds))
              (Rendering|Material|SetScalarParameterValueonMaterials _returnvalue_2 "EDRec_Emissive_C_Brightness_Time_BP" _returnvalue)
              (Rendering|Material|SetScalarParameterValueonMaterials _returnvalue_2 "EDRec_Emissive_B_Brightness_Time_BP" _returnvalue)))))))
  (Rendering|Material|SetScalarParameterValueonMaterials _returnvalue_3 "FDecalV2_Emissive_Brightness_Time_BP" (Utilities|Time|GetGameTimeinSeconds))
  (Rendering|Material|SetScalarParameterValueonMaterials _returnvalue_3 "FDecalV2_Emissive_Gradent_Time_BP" (Utilities|Time|GetGameTimeinSeconds)))

(fn ConstructionScript)
