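; Custom item wrap Blueprint for the RoyalWonk wrap, animating the material on wrapped meshes when the weapon fires.
; Each shot starts a timeline, which subtracts its offset output from the stored Offset_A variable on every update.
; The running value is pushed to the emissive color-curve offset scalar parameter on all wrapped mesh components' materials, shifting the glow with firing.
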(event EventOnWeaponFired
  (bind (_direction _offset) (Timeline|Timeline PlayFromStart :NewTime 0.5)
    (:Update
      (bind _output_get (Variables|Default|SetOffset_A (- (Variables|Default|GetOffset_A) _offset)))
      (Rendering|Material|SetScalarParameterValueonMaterials (CustomItemWrap|GetWrappedMeshComponents) "TopS Emissive_01_ColorCuve_Offset" _output_get))))

(fn ConstructionScript)
