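; CIWM_RumbleBeat is a reactive weapon wrap modifier that drives staged emissive effects on wrapped meshes from firing and kills.
; Weapon fire timestamps emissive layer A, while watched AthenaKills stat changes light layers B, C and D in turn.
; After D, a pause flag blocks progression while delays hold, reset and timestamp those layers, then restart the cycle.
; In the front end, every seventh shot advances the same duplicated stage logic.
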(event EventOnWeaponFired
  (bind _returnvalue (CustomItemWrap|GetWrappedMeshComponents))
  (bind _returnvalue_2 (CustomItemWrap|GetWrappedMeshComponents))
  (bind _returnvalue_3 (CustomItemWrap|GetWrappedMeshComponents))
  (bind _returnvalue_4 (Utilities|Time|GetGameTimeinSeconds))
  (if (CustomCosmeticModifier|IsinFrontEnd)
    (Utilities|FlowControl|Sequence
      (:then_0
        _)
      (:then_1
        (Utilities|FlowControl|Delay 0.001)
        (Variables|Default|SetShotsFired (+ (Variables|Default|GetShotsFired) 1))
        (if (== (Variables|Default|GetShotsFired) 7)
          (Variables|Default|SetShotsFired 0)
          (if (not (Variables|Default|GetPause))
            (switch int (Variables|Default|GetIterationCount)
              (:0
                (Math|Integer|IncrementInt (Variables|Default|GetIterationCount)
                  (:%20%20
                    (Rendering|Material|SetScalarParameterValueonMaterials _returnvalue_2 "EDRec_Emissive_B_Brightness_Time__End" 1.0))))
              (:1
                (Math|Integer|IncrementInt (Variables|Default|GetIterationCount)
                  (:%20%20
                    (Rendering|Material|SetScalarParameterValueonMaterials _returnvalue_2 "EDRec_Emissive_C_Brightness_Time__End" 1.0))))
              (:2
                (Variables|Default|SetPause true)
                (Rendering|Material|SetScalarParameterValueonMaterials _returnvalue_2 "EDRec_Emissive_D_Brightness_Time__End" 1.0)
                (Utilities|FlowControl|Delay 3.0)
                (Rendering|Material|SetScalarParameterValueonMaterials _returnvalue_3 "EDRec_Emissive_B_Brightness_Time__End" 0.0)
                (Rendering|Material|SetScalarParameterValueonMaterials _returnvalue_3 "EDRec_Emissive_C_Brightness_Time__End" 0.0)
                (Rendering|Material|SetScalarParameterValueonMaterials _returnvalue_3 "EDRec_Emissive_D_Brightness_Time__End" 0.0)
                (Rendering|Material|SetScalarParameterValueonMaterials _returnvalue "EDRec_Emissive_B_Brightness_Time_BP" _returnvalue_4)
                (Rendering|Material|SetScalarParameterValueonMaterials _returnvalue "EDRec_Emissive_C_Brightness_Time_BP" _returnvalue_4)
                (Rendering|Material|SetScalarParameterValueonMaterials _returnvalue "EDRec_Emissive_D_Brightness_Time_BP" _returnvalue_4)
                (Utilities|FlowControl|Delay 1.5)
                (Variables|Default|SetPause false)
                (Variables|Default|SetIterationCount 0))))))))
  (Rendering|Material|SetScalarParameterValueonMaterials (CustomItemWrap|GetWrappedMeshComponents) "EDRec_Emissive_A_Brightness_Time_BP" (Utilities|Time|GetGameTimeinSeconds)))

(event EventOnPawnSet
  (bind _returnvalue (CustomItemWrap|GetWrappedMeshComponents))
  (bind _returnvalue_2 (CustomItemWrap|GetWrappedMeshComponents))
  (bind _returnvalue_3 (CustomItemWrap|GetWrappedMeshComponents))
  (bind _returnvalue_4 (Utilities|Time|GetGameTimeinSeconds))
  (Cosmetics|WatchCosmeticStat (CustomCharacter|GetAssociatedPlayerPawn) "(TagName=\"StatManager.AthenaKills\")" 0 false
    (:OnStatChanged
      (if (not (Variables|Default|GetPause))
        (switch int (Variables|Default|GetIterationCount)
          (:0
            (Math|Integer|IncrementInt (Variables|Default|GetIterationCount)
              (:%20%20
                (Rendering|Material|SetScalarParameterValueonMaterials _returnvalue_2 "EDRec_Emissive_B_Brightness_Time__End" 1.0))))
          (:1
            (Math|Integer|IncrementInt (Variables|Default|GetIterationCount)
              (:%20%20
                (Rendering|Material|SetScalarParameterValueonMaterials _returnvalue_2 "EDRec_Emissive_C_Brightness_Time__End" 1.0))))
          (:2
            (Variables|Default|SetPause true)
            (Rendering|Material|SetScalarParameterValueonMaterials _returnvalue_2 "EDRec_Emissive_D_Brightness_Time__End" 1.0)
            (Utilities|FlowControl|Delay 3.0)
            (Rendering|Material|SetScalarParameterValueonMaterials _returnvalue_3 "EDRec_Emissive_B_Brightness_Time__End" 0.0)
            (Rendering|Material|SetScalarParameterValueonMaterials _returnvalue_3 "EDRec_Emissive_C_Brightness_Time__End" 0.0)
            (Rendering|Material|SetScalarParameterValueonMaterials _returnvalue_3 "EDRec_Emissive_D_Brightness_Time__End" 0.0)
            (Rendering|Material|SetScalarParameterValueonMaterials _returnvalue "EDRec_Emissive_B_Brightness_Time_BP" _returnvalue_4)
            (Rendering|Material|SetScalarParameterValueonMaterials _returnvalue "EDRec_Emissive_C_Brightness_Time_BP" _returnvalue_4)
            (Rendering|Material|SetScalarParameterValueonMaterials _returnvalue "EDRec_Emissive_D_Brightness_Time_BP" _returnvalue_4)
            (Utilities|FlowControl|Delay 1.5)
            (Variables|Default|SetPause false)
            (Variables|Default|SetIterationCount 0)))))))

(fn ConstructionScript)
