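; Item wrap Blueprint that cycles wrapped meshes through four SDF blend looks as the pawn's kill stat changes.
; Each stat change copies Selection into Kill_Stat, sets SDFBL channel masks and blend endpoints, stamps game time, then advances Selection, wrapping after four.
; Weapon and weapon-type changes reapply the current state after a one-second delay, without restamping time.
; Debug prints remain, and NewFunction duplicates the switch without being called here.
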(event EventOnPawnSet
  (bind _returnvalue (CustomItemWrap|GetWrappedMeshComponents))
  (bind _returnvalue_2 (CustomItemWrap|GetWrappedMeshComponents))
  (bind _returnvalue_3 (CustomItemWrap|GetWrappedMeshComponents))
  (bind _returnvalue_4 (CustomItemWrap|GetWrappedMeshComponents))
  (bind _kill_stat (Variables|Default|GetKill_Stat))
  (bind _asfort_player_pawn (Utilities|Casting|CastToFortPlayerPawn (CustomCharacter|GetAssociatedPlayerPawn)))
  (Cosmetics|WatchCosmeticStat _asfort_player_pawn "(TagName=\"StatManager.AthenaKills\")"
    (:OnStatChanged
      (Variables|Default|SetKill_Stat (Variables|Default|GetSelection))
      (Development|PrintString (Utilities|String|ToString(Integer) _kill_stat) true true "(R=1.000000,G=0.015037,B=0.000000,A=1.000000)" 5.0)
      (switch int _kill_stat
        (:0
          (Rendering|Material|SetScalarParameterValueonMaterials _returnvalue "SDFBL_Start" 1.0)
          (Rendering|Material|SetScalarParameterValueonMaterials _returnvalue "SDFBL_End" 0.0)
          (Rendering|Material|SetVectorParameterValueonMaterials _returnvalue "SDFBL_Channel_Mask_1,2,3,_A" "0.000000,1.000000,0.000000")
          (Rendering|Material|SetScalarParameterValueonMaterials _returnvalue "SDFBL_Channel_Mask_4_A" 0.0)
          (Rendering|Material|SetVectorParameterValueonMaterials _returnvalue "SDFBL_Channel_Mask_1,2,3,_B" "0, 0,1.000000")
          (Rendering|Material|SetScalarParameterValueonMaterials _returnvalue "SDFBL_Channel_Mask_4_B" 0.0)
          (Rendering|Material|SetScalarParameterValueonMaterials _returnvalue "SDFBL_Time_BP" (Utilities|Time|GetGameTimeinSeconds))
          (Variables|Default|SetSelection 1))
        (:1
          (Rendering|Material|SetScalarParameterValueonMaterials _returnvalue_2 "SDFBL_Start" 0.0)
          (Rendering|Material|SetScalarParameterValueonMaterials _returnvalue_2 "SDFBL_End" 1.0)
          (Rendering|Material|SetVectorParameterValueonMaterials _returnvalue_2 "SDFBL_Channel_Mask_1,2,3,_A" "1.000000,0.000000,0.000000")
          (Rendering|Material|SetScalarParameterValueonMaterials _returnvalue_2 "SDFBL_Channel_Mask_4_A" 0.0)
          (Rendering|Material|SetVectorParameterValueonMaterials _returnvalue_2 "SDFBL_Channel_Mask_1,2,3,_B" "0.000000, 0,1.000000")
          (Rendering|Material|SetScalarParameterValueonMaterials _returnvalue_2 "SDFBL_Channel_Mask_4_B" 0.0)
          (Rendering|Material|SetScalarParameterValueonMaterials _returnvalue_2 "SDFBL_Time_BP" (Utilities|Time|GetGameTimeinSeconds))
          (Variables|Default|SetSelection 2))
        (:2
          (Rendering|Material|SetScalarParameterValueonMaterials _returnvalue_3 "SDFBL_Start" 1.0)
          (Rendering|Material|SetScalarParameterValueonMaterials _returnvalue_3 "SDFBL_End" 0.0)
          (Rendering|Material|SetVectorParameterValueonMaterials _returnvalue_3 "SDFBL_Channel_Mask_1,2,3,_A" "1.000000,0.000000,0.000000")
          (Rendering|Material|SetScalarParameterValueonMaterials _returnvalue_3 "SDFBL_Channel_Mask_4_A" 0.0)
          (Rendering|Material|SetVectorParameterValueonMaterials _returnvalue_3 "SDFBL_Channel_Mask_1,2,3,_B" "0, 0, 0")
          (Rendering|Material|SetScalarParameterValueonMaterials _returnvalue_3 "SDFBL_Channel_Mask_4_B" 1.0)
          (Rendering|Material|SetScalarParameterValueonMaterials _returnvalue_3 "SDFBL_Time_BP" (Utilities|Time|GetGameTimeinSeconds))
          (Variables|Default|SetSelection 3))
        (:3
          (Rendering|Material|SetScalarParameterValueonMaterials _returnvalue_4 "SDFBL_Start" 0.0)
          (Rendering|Material|SetScalarParameterValueonMaterials _returnvalue_4 "SDFBL_End" 1.0)
          (Rendering|Material|SetVectorParameterValueonMaterials _returnvalue_4 "SDFBL_Channel_Mask_1,2,3,_A" "0.000000,1.000000,0.000000")
          (Rendering|Material|SetScalarParameterValueonMaterials _returnvalue_4 "SDFBL_Channel_Mask_4_A" 0.0)
          (Rendering|Material|SetVectorParameterValueonMaterials _returnvalue_4 "SDFBL_Channel_Mask_1,2,3,_B" "0.000000, 0,1.000000")
          (Rendering|Material|SetScalarParameterValueonMaterials _returnvalue_4 "SDFBL_Channel_Mask_4_B" 1.0)
          (Rendering|Material|SetScalarParameterValueonMaterials _returnvalue_4 "SDFBL_Time_BP" (Utilities|Time|GetGameTimeinSeconds))
          (Variables|Default|SetSelection 0))))))

(event EventBeginPlay
  (|CallParentFunction))

(event EventOnNewWeapons
  (params (WeaponsAsItems :type Array%20of%20Fort%20Item%20Object%20References))
  (bind _returnvalue (CustomItemWrap|GetWrappedMeshComponents))
  (bind _returnvalue_2 (CustomItemWrap|GetWrappedMeshComponents))
  (bind _returnvalue_3 (CustomItemWrap|GetWrappedMeshComponents))
  (bind _returnvalue_4 (CustomItemWrap|GetWrappedMeshComponents))
  (bind _kill_stat (Variables|Default|GetKill_Stat))
  (Utilities|FlowControl|Delay 1.0)
  (Development|PrintString (Utilities|String|ToString(Integer) _kill_stat) true true "(R=1.000000,G=0.015037,B=0.000000,A=1.000000)" 5.0)
  (switch int _kill_stat
    (:0
      (Rendering|Material|SetScalarParameterValueonMaterials _returnvalue "SDFBL_Start" 1.0)
      (Rendering|Material|SetScalarParameterValueonMaterials _returnvalue "SDFBL_End" 0.0)
      (Rendering|Material|SetVectorParameterValueonMaterials _returnvalue "SDFBL_Channel_Mask_1,2,3,_A" "0.000000,1.000000,0.000000")
      (Rendering|Material|SetScalarParameterValueonMaterials _returnvalue "SDFBL_Channel_Mask_4_A" 0.0)
      (Rendering|Material|SetVectorParameterValueonMaterials _returnvalue "SDFBL_Channel_Mask_1,2,3,_B" "0, 0,1.000000")
      (Rendering|Material|SetScalarParameterValueonMaterials _returnvalue "SDFBL_Channel_Mask_4_B" 0.0))
    (:1
      (Rendering|Material|SetScalarParameterValueonMaterials _returnvalue_2 "SDFBL_Start" 0.0)
      (Rendering|Material|SetScalarParameterValueonMaterials _returnvalue_2 "SDFBL_End" 1.0)
      (Rendering|Material|SetVectorParameterValueonMaterials _returnvalue_2 "SDFBL_Channel_Mask_1,2,3,_A" "1.000000,0.000000,0.000000")
      (Rendering|Material|SetScalarParameterValueonMaterials _returnvalue_2 "SDFBL_Channel_Mask_4_A" 0.0)
      (Rendering|Material|SetVectorParameterValueonMaterials _returnvalue_2 "SDFBL_Channel_Mask_1,2,3,_B" "0.000000, 0,1.000000")
      (Rendering|Material|SetScalarParameterValueonMaterials _returnvalue_2 "SDFBL_Channel_Mask_4_B" 0.0))
    (:2
      (Rendering|Material|SetScalarParameterValueonMaterials _returnvalue_3 "SDFBL_Start" 1.0)
      (Rendering|Material|SetScalarParameterValueonMaterials _returnvalue_3 "SDFBL_End" 0.0)
      (Rendering|Material|SetVectorParameterValueonMaterials _returnvalue_3 "SDFBL_Channel_Mask_1,2,3,_A" "1.000000,0.000000,0.000000")
      (Rendering|Material|SetScalarParameterValueonMaterials _returnvalue_3 "SDFBL_Channel_Mask_4_A" 0.0)
      (Rendering|Material|SetVectorParameterValueonMaterials _returnvalue_3 "SDFBL_Channel_Mask_1,2,3,_B" "0, 0, 0")
      (Rendering|Material|SetScalarParameterValueonMaterials _returnvalue_3 "SDFBL_Channel_Mask_4_B" 1.0))
    (:3
      (Rendering|Material|SetScalarParameterValueonMaterials _returnvalue_4 "SDFBL_Start" 0.0)
      (Rendering|Material|SetScalarParameterValueonMaterials _returnvalue_4 "SDFBL_End" 1.0)
      (Rendering|Material|SetVectorParameterValueonMaterials _returnvalue_4 "SDFBL_Channel_Mask_1,2,3,_A" "0.000000,1.000000,0.000000")
      (Rendering|Material|SetScalarParameterValueonMaterials _returnvalue_4 "SDFBL_Channel_Mask_4_A" 0.0)
      (Rendering|Material|SetVectorParameterValueonMaterials _returnvalue_4 "SDFBL_Channel_Mask_1,2,3,_B" "0.000000, 0,1.000000")
      (Rendering|Material|SetScalarParameterValueonMaterials _returnvalue_4 "SDFBL_Channel_Mask_4_B" 1.0))))

(event EventOnNewPlayerWeaponType
  (params (NewType :type EFortItemType%20Enum :default "WorldItem") (OldType :type EFortItemType%20Enum :default "WorldItem"))
  (bind _returnvalue (CustomItemWrap|GetWrappedMeshComponents))
  (bind _returnvalue_2 (CustomItemWrap|GetWrappedMeshComponents))
  (bind _returnvalue_3 (CustomItemWrap|GetWrappedMeshComponents))
  (bind _returnvalue_4 (CustomItemWrap|GetWrappedMeshComponents))
  (bind _kill_stat (Variables|Default|GetKill_Stat))
  (Utilities|FlowControl|Delay 1.0)
  (Development|PrintString (Utilities|String|ToString(Integer) _kill_stat) true true "(R=1.000000,G=0.015037,B=0.000000,A=1.000000)" 5.0)
  (switch int _kill_stat
    (:0
      (Rendering|Material|SetScalarParameterValueonMaterials _returnvalue "SDFBL_Start" 1.0)
      (Rendering|Material|SetScalarParameterValueonMaterials _returnvalue "SDFBL_End" 0.0)
      (Rendering|Material|SetVectorParameterValueonMaterials _returnvalue "SDFBL_Channel_Mask_1,2,3,_A" "0.000000,1.000000,0.000000")
      (Rendering|Material|SetScalarParameterValueonMaterials _returnvalue "SDFBL_Channel_Mask_4_A" 0.0)
      (Rendering|Material|SetVectorParameterValueonMaterials _returnvalue "SDFBL_Channel_Mask_1,2,3,_B" "0, 0,1.000000")
      (Rendering|Material|SetScalarParameterValueonMaterials _returnvalue "SDFBL_Channel_Mask_4_B" 0.0))
    (:1
      (Rendering|Material|SetScalarParameterValueonMaterials _returnvalue_2 "SDFBL_Start" 0.0)
      (Rendering|Material|SetScalarParameterValueonMaterials _returnvalue_2 "SDFBL_End" 1.0)
      (Rendering|Material|SetVectorParameterValueonMaterials _returnvalue_2 "SDFBL_Channel_Mask_1,2,3,_A" "1.000000,0.000000,0.000000")
      (Rendering|Material|SetScalarParameterValueonMaterials _returnvalue_2 "SDFBL_Channel_Mask_4_A" 0.0)
      (Rendering|Material|SetVectorParameterValueonMaterials _returnvalue_2 "SDFBL_Channel_Mask_1,2,3,_B" "0.000000, 0,1.000000")
      (Rendering|Material|SetScalarParameterValueonMaterials _returnvalue_2 "SDFBL_Channel_Mask_4_B" 0.0))
    (:2
      (Rendering|Material|SetScalarParameterValueonMaterials _returnvalue_3 "SDFBL_Start" 1.0)
      (Rendering|Material|SetScalarParameterValueonMaterials _returnvalue_3 "SDFBL_End" 0.0)
      (Rendering|Material|SetVectorParameterValueonMaterials _returnvalue_3 "SDFBL_Channel_Mask_1,2,3,_A" "1.000000,0.000000,0.000000")
      (Rendering|Material|SetScalarParameterValueonMaterials _returnvalue_3 "SDFBL_Channel_Mask_4_A" 0.0)
      (Rendering|Material|SetVectorParameterValueonMaterials _returnvalue_3 "SDFBL_Channel_Mask_1,2,3,_B" "0, 0, 0")
      (Rendering|Material|SetScalarParameterValueonMaterials _returnvalue_3 "SDFBL_Channel_Mask_4_B" 1.0))
    (:3
      (Rendering|Material|SetScalarParameterValueonMaterials _returnvalue_4 "SDFBL_Start" 0.0)
      (Rendering|Material|SetScalarParameterValueonMaterials _returnvalue_4 "SDFBL_End" 1.0)
      (Rendering|Material|SetVectorParameterValueonMaterials _returnvalue_4 "SDFBL_Channel_Mask_1,2,3,_A" "0.000000,1.000000,0.000000")
      (Rendering|Material|SetScalarParameterValueonMaterials _returnvalue_4 "SDFBL_Channel_Mask_4_A" 0.0)
      (Rendering|Material|SetVectorParameterValueonMaterials _returnvalue_4 "SDFBL_Channel_Mask_1,2,3,_B" "0.000000, 0,1.000000")
      (Rendering|Material|SetScalarParameterValueonMaterials _returnvalue_4 "SDFBL_Channel_Mask_4_B" 1.0))))

(fn NewFunction
  (bind _returnvalue (CustomItemWrap|GetWrappedMeshComponents))
  (bind _returnvalue_2 (CustomItemWrap|GetWrappedMeshComponents))
  (bind _returnvalue_3 (CustomItemWrap|GetWrappedMeshComponents))
  (bind _returnvalue_4 (CustomItemWrap|GetWrappedMeshComponents))
  (switch int (Variables|Default|GetKill_Stat)
    (:0
      (Rendering|Material|SetScalarParameterValueonMaterials _returnvalue "SDFBL_Start" 1.0)
      (Rendering|Material|SetScalarParameterValueonMaterials _returnvalue "SDFBL_End" 0.0)
      (Rendering|Material|SetVectorParameterValueonMaterials _returnvalue "SDFBL_Channel_Mask_1,2,3,_A" "0.000000,1.000000,0.000000")
      (Rendering|Material|SetScalarParameterValueonMaterials _returnvalue "SDFBL_Channel_Mask_4_A" 0.0)
      (Rendering|Material|SetVectorParameterValueonMaterials _returnvalue "SDFBL_Channel_Mask_1,2,3,_B" "0, 0,1.000000")
      (Rendering|Material|SetScalarParameterValueonMaterials _returnvalue "SDFBL_Channel_Mask_4_B" 0.0)
      (Variables|Default|SetSelection 1))
    (:1
      (Rendering|Material|SetScalarParameterValueonMaterials _returnvalue_2 "SDFBL_Start" 0.0)
      (Rendering|Material|SetScalarParameterValueonMaterials _returnvalue_2 "SDFBL_End" 1.0)
      (Rendering|Material|SetVectorParameterValueonMaterials _returnvalue_2 "SDFBL_Channel_Mask_1,2,3,_A" "1.000000,0.000000,0.000000")
      (Rendering|Material|SetScalarParameterValueonMaterials _returnvalue_2 "SDFBL_Channel_Mask_4_A" 0.0)
      (Rendering|Material|SetVectorParameterValueonMaterials _returnvalue_2 "SDFBL_Channel_Mask_1,2,3,_B" "0.000000, 0,1.000000")
      (Rendering|Material|SetScalarParameterValueonMaterials _returnvalue_2 "SDFBL_Channel_Mask_4_B" 0.0)
      (Variables|Default|SetSelection 2))
    (:2
      (Rendering|Material|SetScalarParameterValueonMaterials _returnvalue_3 "SDFBL_Start" 1.0)
      (Rendering|Material|SetScalarParameterValueonMaterials _returnvalue_3 "SDFBL_End" 0.0)
      (Rendering|Material|SetVectorParameterValueonMaterials _returnvalue_3 "SDFBL_Channel_Mask_1,2,3,_A" "1.000000,0.000000,0.000000")
      (Rendering|Material|SetScalarParameterValueonMaterials _returnvalue_3 "SDFBL_Channel_Mask_4_A" 0.0)
      (Rendering|Material|SetVectorParameterValueonMaterials _returnvalue_3 "SDFBL_Channel_Mask_1,2,3,_B" "0, 0, 0")
      (Rendering|Material|SetScalarParameterValueonMaterials _returnvalue_3 "SDFBL_Channel_Mask_4_B" 1.0)
      (Variables|Default|SetSelection 3))
    (:3
      (Rendering|Material|SetScalarParameterValueonMaterials _returnvalue_4 "SDFBL_Start" 0.0)
      (Rendering|Material|SetScalarParameterValueonMaterials _returnvalue_4 "SDFBL_End" 1.0)
      (Rendering|Material|SetVectorParameterValueonMaterials _returnvalue_4 "SDFBL_Channel_Mask_1,2,3,_A" "0.000000,1.000000,0.000000")
      (Rendering|Material|SetScalarParameterValueonMaterials _returnvalue_4 "SDFBL_Channel_Mask_4_A" 0.0)
      (Rendering|Material|SetVectorParameterValueonMaterials _returnvalue_4 "SDFBL_Channel_Mask_1,2,3,_B" "0.000000, 0,1.000000")
      (Rendering|Material|SetScalarParameterValueonMaterials _returnvalue_4 "SDFBL_Channel_Mask_4_B" 1.0)
      (Variables|Default|SetSelection 0))))

(fn ConstructionScript)
