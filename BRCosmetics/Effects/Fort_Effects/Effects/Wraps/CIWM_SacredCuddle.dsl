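; SacredCuddle reactive item wrap that shows a kill counter and time-driven emissive effects on wrapped weapon materials.
; Watched damage stats stamp game time into emissive D, while kills update the 0–99 counter and stamp emissive A and C.
; In the front end, a timeline drives the same parameters as a preview.
; Weapon change events reapply the count after a short delay, and an uncalled Weapon_Offsets helper sets Z offset vectors.
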(event EventBeginPlay
  (bind _returnvalue (Utilities|Time|GetGameTimeinSeconds))
  (bind _returnvalue_2 (CustomItemWrap|GetWrappedMeshComponents))
  (if (CustomCosmeticModifier|IsinFrontEnd)
    (Utilities|FlowControl|Sequence
      (:then_0
        (bind (_direction _newtrack_2) (Timeline|Timeline PlayFromStart)
          (:NewTrack_0
            (Rendering|Material|SetScalarParameterValueonMaterials (CustomItemWrap|GetWrappedMeshComponents) "EDRec_Emissive_D_Brightness_Time_BP" (Utilities|Time|GetGameTimeinSeconds)))
          (:NewTrack_1
            (Variables|Default|SetKill_Count (Math|Float|Truncate _newtrack_2))
            (Rendering|Material|SetScalarParameterValueonMaterials _returnvalue_2 "EDRec_Counter__Number" (Math|Float|Clamp(Float) (Math|Conversions|ToFloat(Integer) (Variables|Default|GetKill_Count)) 0.0 99.0))
            (Rendering|Material|SetScalarParameterValueonMaterials _returnvalue_2 "EDRec_Emissive_A_Brightness_Time_BP" _returnvalue)
            (Rendering|Material|SetScalarParameterValueonMaterials _returnvalue_2 "EDRec_Emissive_C_Brightness_Time_BP" _returnvalue)))))))

(event EventOnPawnSet
  (bind _returnvalue (Utilities|Time|GetGameTimeinSeconds))
  (bind _returnvalue_2 (CustomItemWrap|GetWrappedMeshComponents))
  (Utilities|FlowControl|Sequence
    (:then_0
      (Cosmetics|WatchCosmeticStat (CustomCharacter|GetAssociatedPlayerPawn) "(TagName=\"StatManager.DamageDealt\")" 0 false
        (:OnStatChanged
          (Rendering|Material|SetScalarParameterValueonMaterials (CustomItemWrap|GetWrappedMeshComponents) "EDRec_Emissive_D_Brightness_Time_BP" (Utilities|Time|GetGameTimeinSeconds)))))
    (:then_1
      (bind (_statname _statvalue) (Cosmetics|WatchCosmeticStat (CustomCharacter|GetAssociatedPlayerPawn) "(TagName=\"StatManager.AthenaKills\")" 0 false)
        (:OnStatChanged
          (Variables|Default|SetKill_Count _statvalue)
          (Rendering|Material|SetScalarParameterValueonMaterials _returnvalue_2 "EDRec_Counter__Number" (Math|Float|Clamp(Float) (Math|Conversions|ToFloat(Integer) (Variables|Default|GetKill_Count)) 0.0 99.0))
          (Rendering|Material|SetScalarParameterValueonMaterials _returnvalue_2 "EDRec_Emissive_A_Brightness_Time_BP" _returnvalue)
          (Rendering|Material|SetScalarParameterValueonMaterials _returnvalue_2 "EDRec_Emissive_C_Brightness_Time_BP" _returnvalue))))))

(event EventOnWeaponChanged
  (params (NewWeapon :type Fort%20Weapon%20Object%20Reference) (OldWeapon :type Fort%20Weapon%20Object%20Reference))
  (Utilities|FlowControl|Delay 0.2)
  (Rendering|Material|SetScalarParameterValueonMaterials (CustomItemWrap|GetWrappedMeshComponents) "EDRec_Counter__Number" (Math|Float|Clamp(Float) (Math|Conversions|ToFloat(Integer) (Variables|Default|GetKill_Count)) 0.0 99.0)))

(event EventOnNewWeapons
  (params (WeaponsAsItems :type Array%20of%20Fort%20Item%20Object%20References))
  (Utilities|FlowControl|Delay 0.2)
  (Rendering|Material|SetScalarParameterValueonMaterials (CustomItemWrap|GetWrappedMeshComponents) "EDRec_Counter__Number" (Math|Float|Clamp(Float) (Math|Conversions|ToFloat(Integer) (Variables|Default|GetKill_Count)) 0.0 99.0)))

(event EventOnNewPlayerWeaponType
  (params (NewType :type EFortItemType%20Enum :default "WorldItem") (OldType :type EFortItemType%20Enum :default "WorldItem"))
  (Utilities|FlowControl|Delay 0.2)
  (Rendering|Material|SetScalarParameterValueonMaterials (CustomItemWrap|GetWrappedMeshComponents) "EDRec_Counter__Number" (Math|Float|Clamp(Float) (Math|Conversions|ToFloat(Integer) (Variables|Default|GetKill_Count)) 0.0 99.0)))

(fn ConstructionScript)

(fn Weapon_Offsets
  (params (Offset_End_Name :type Name) (Offset_End_Value :type Float%20%28double-precision%29) (Offset_Start_Name :type Name) (Offset_State_Value :type Float%20%28double-precision%29))
  (bind _returnvalue (CustomItemWrap|GetWrappedMeshComponents))
  (Rendering|Material|SetVectorParameterValueonMaterials _returnvalue Offset_End_Name (Math|Vector|MakeVector 0.0 0.0 Offset_End_Value))
  (Rendering|Material|SetVectorParameterValueonMaterials _returnvalue Offset_Start_Name (Math|Vector|MakeVector 0.0 0.0 Offset_State_Value)))
