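; Reactive logic for the SafariGnome weapon wrap, which switches on a decal effect after the player gets a kill.
; A watched AthenaKills stat change sets a decal opacity scalar on the wrapped mesh materials, then a timer reverts it after ten seconds.
; In frontend display, weapon-fired events are counted and the fifth simulates a kill through LockerKill.
; A Playing flag blocks retriggering, and Revert clears it along with the shot counter.
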(event Custom|LockerKill
  (if (not (Variables|Default|GetPlaying?))
    (Variables|Default|SetPlaying? true)
    (Rendering|Material|SetScalarParameterValueonMaterials (CustomItemWrap|GetWrappedMeshComponents) "FDecalV2_Opacity_Time__End" 1.0)
    (Utilities|Time|SetTimerbyEvent (delegate Custom|Revert) 1e+01)))

(event EventOnPawnSet
  (Utilities|FlowControl|Sequence
    (:then_0
      (Cosmetics|WatchCosmeticStat (CustomCharacter|GetAssociatedPlayerPawn) "(TagName=\"StatManager.AthenaKills\")" 0 false
        (:OnStatChanged
          (if (not (Variables|Default|GetPlaying?))
            (Variables|Default|SetPlaying? true)
            (Rendering|Material|SetScalarParameterValueonMaterials (CustomItemWrap|GetWrappedMeshComponents) "FDecalV2_Opacity_Time__End" 1.0)
            (Utilities|Time|SetTimerbyEvent (delegate Custom|Revert) 1e+01)))))))

(event Custom|Revert
  (Rendering|Material|SetScalarParameterValueonMaterials (CustomItemWrap|GetWrappedMeshComponents) "FDecalV2_Opacity_Time__End" 0.0)
  (Variables|Default|SetPlaying? false)
  (Variables|Default|SetIterationCount 0))

(event EventOnWeaponFired
  (if (Reactive|Frontend|IsCosmeticObjectDisplayedinFrontendContext 0)
    (if (Variables|Default|GetPlaying?)
      _
      (elif (>= (Variables|Default|GetIterationCount) 4)
        (CallFunction|LockerKill)
        (else
          (Math|Integer|IncrementInt (Variables|Default|GetIterationCount)))))))

(fn ConstructionScript)
