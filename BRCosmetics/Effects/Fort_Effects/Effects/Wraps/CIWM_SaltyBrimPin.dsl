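; Weapon wrap Blueprint (SaltyBrimPin) that animates emissive material parameters on the wrapped meshes when the weapon fires.
; Each shot stamps game time into emissive parameters and picks a random array value, unused elsewhere here.
; The first shot starts a looping timer that fades two TopS emissive layers' offset and brightness in.
; A retriggerable delay reverses the fade once firing pauses; the timer snaps end values and clears after five seconds.
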(event EventOnWeaponFired
  (bind _returnvalue (CustomItemWrap|GetWrappedMeshComponents))
  (bind _returnvalue_2 (Utilities|Time|GetGameTimeinSeconds))
  (Variables|Default|SetCurrentRandomValue (Utilities|Array|Get(acopy) (Variables|Default|GetFloat%20array) (Math|Random|RandomIntegerinRange 0 (Utilities|Array|Length (Variables|Default|GetFloat%20array)))))
  (Utilities|FlowControl|Sequence
    (:then_0
      ;[1 1 1] TopsSide V2
      (Rendering|Material|SetScalarParameterValueonMaterials _returnvalue "EDRec_Emissive_A_Brightness_Time_BP" _returnvalue_2)
      (Rendering|Material|SetScalarParameterValueonMaterials _returnvalue "EDRec_Emissive_A_Brightness_Time__End" _returnvalue_2)
      (Rendering|Material|SetScalarParameterValueonMaterials _returnvalue "FDecalV2_Emissive_Brightness_Time_BP" _returnvalue_2)
      (Rendering|Material|SetScalarParameterValueonMaterials _returnvalue "FDecalV2_Emissive_Brightness_Time__End" _returnvalue_2)
      (if (not (Variables|Default|GetKilled))
        (Rendering|Material|SetScalarParameterValueonMaterials _returnvalue "EDRec_Emissive_A_Brightness_Time__End" 0.0)
        (Rendering|Material|SetScalarParameterValueonMaterials _returnvalue "FDecalV2_Emissive_Brightness_Time__End" 0.0)))
    (:then_1
      (if (not (Variables|Default|GetFullAutoWeapons))
        (Variables|Default|SetFullAutoWeapons true)
        (CallFunction|SineAnimation))
      (Utilities|FlowControl|RetriggerableDelay 1.5)
      (CallFunction|SineAnimOFF)
      (Variables|Default|SetFullAutoWeapons false))))

(event Custom|SineAnimation
  ; Layer Offset
  (Variables|Default|SetReverse false)
  (Variables|Default|SetStartTime (Utilities|Time|GetGameTimeinSeconds))
  (bind _returnvalue (Utilities|Time|SetTimerbyEvent (delegate Custom|SineAnimUpdate) 0.02 true))
  (Variables|Default|SetTimerHandle _returnvalue))

(event Custom|SineAnimOFF
  ; Layer Offset
  (Variables|Default|SetReverse true)
  (Variables|Default|SetStartTime (Utilities|Time|GetGameTimeinSeconds))
  (bind _returnvalue (Utilities|Time|SetTimerbyEvent (delegate Custom|SineAnimUpdate) 0.02 true))
  (Variables|Default|SetTimerHandle _returnvalue))

(event Custom|SineAnimUpdate
  (bind _returnvalue (Math|Float|MapRangeClamped (- (Utilities|Time|GetGameTimeinSeconds) (Variables|Default|GetStartTime)) 0.0 0.75 1.0))
  (bind _returnvalue_2 (Math|Float|MapRangeClamped (- (Utilities|Time|GetGameTimeinSeconds) (Variables|Default|GetStartTime)) 0.0 0.75 0.325 1.0))
  (bind _returnvalue_3 (Utilities|Time|GetGameTimeinSeconds))
  (bind _returnvalue_4 (CustomItemWrap|GetWrappedMeshComponents))
  (bind _returnvalue_5 (Math|Float|MapRangeClamped (- _returnvalue_3 (Variables|Default|GetStartTime)) 0.0 0.1 1.0 0.325))
  (bind _returnvalue_6 (Math|Float|MapRangeClamped (- _returnvalue_3 (Variables|Default|GetStartTime)) 0.0 0.75 0.0 1.0))
  (bind _starttime (Variables|Default|GetStartTime))
  (bind _returnvalue_7 (- _returnvalue_3 _starttime))
  ; Layer Offset
  (if (>= _returnvalue_7 5.0)
    (Utilities|Time|ClearandInvalidateTimerbyHandle (Variables|Default|GetTimerHandle))
    (if (Variables|Default|GetReverse)
      ;; Start
      (Rendering|Material|SetScalarParameterValueonMaterials _returnvalue_4 "TopS Emissive_03_Offset---End" 1.0)
      (Rendering|Material|SetScalarParameterValueonMaterials _returnvalue_4 "TopS Emissive_03_Brightness---End" 0.0)
      (Rendering|Material|SetScalarParameterValueonMaterials _returnvalue_4 "TopS Emissive_02_Offset---End" 1.0)
      (Rendering|Material|SetScalarParameterValueonMaterials _returnvalue_4 "TopS Emissive_02_Brightness---End" 0.0)
      (else
        ;; Arrive
        (Rendering|Material|SetScalarParameterValueonMaterials _returnvalue_4 "TopS Emissive_03_Offset---End" 0.325)
        (Rendering|Material|SetScalarParameterValueonMaterials _returnvalue_4 "TopS Emissive_03_Brightness---End" 1.0)
        (Rendering|Material|SetScalarParameterValueonMaterials _returnvalue_4 "TopS Emissive_02_Offset---End" 0.325)
        (Rendering|Material|SetScalarParameterValueonMaterials _returnvalue_4 "TopS Emissive_02_Brightness---End" 1.0)))
    (elif (Variables|Default|GetReverse)
      ;; FadeOut
      (Rendering|Material|SetScalarParameterValueonMaterials _returnvalue_4 "TopS Emissive_03_Offset---End" _returnvalue_2)
      (Rendering|Material|SetScalarParameterValueonMaterials _returnvalue_4 "TopS Emissive_03_Brightness---End" _returnvalue)
      (Rendering|Material|SetScalarParameterValueonMaterials _returnvalue_4 "TopS Emissive_02_Offset---End" _returnvalue_2)
      (Rendering|Material|SetScalarParameterValueonMaterials _returnvalue_4 "TopS Emissive_02_Brightness---End" _returnvalue)
      (else
        ;; FadeIn
        (Rendering|Material|SetScalarParameterValueonMaterials _returnvalue_4 "TopS Emissive_03_Offset---End" _returnvalue_5)
        (Rendering|Material|SetScalarParameterValueonMaterials _returnvalue_4 "TopS Emissive_03_Brightness---End" _returnvalue_6)
        (Rendering|Material|SetScalarParameterValueonMaterials _returnvalue_4 "TopS Emissive_02_Offset---End" _returnvalue_5)
        (Rendering|Material|SetScalarParameterValueonMaterials _returnvalue_4 "TopS Emissive_02_Brightness---End" _returnvalue_6)))))

(fn ConstructionScript)
