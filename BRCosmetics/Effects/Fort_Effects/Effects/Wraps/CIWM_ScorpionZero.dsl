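; Wrap modifier Blueprint for the ScorpionZero cosmetic that matches anime-style shading on wrapped meshes to the character's selected style.
; When customization finishes, the wrap target is set or the pawn is set, it checks that the loadout character's name matches a configured name.
; It then reads the active Particle variant index and maps it to EF_Anime_PBR_Shading, setting that scalar on every dynamic material of each wrapped mesh component.
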(event EventOnCharacterCustomizationFinished
  (params (CharacterPartInfo :type Array%20of%20Applied%20Character%20Part%20Info%20Structures))
  (if (Utilities|String|EqualExactly(String) (Item|GetPersistentName (CosmeticLoadout|GetItemDefinitionfromLoadout (CosmeticLoadout|GetCosmeticLoadout) "(TagName=\"LoadoutSlot.Character\")")) (Variables|Default|GetCharacter-Name))
    (bind _index (Class|FortCosmeticFunctionLibrary|%28Deprecated%29GetVariantChannelActiveIndexForSlot "(TagName=\"LoadoutSlot.Character\")" (Variables|/Script/FortniteGame.FortPlayerPawn|GetCosmeticLoadout (CustomCharacter|GetAssociatedPlayerPawn)) "(TagName=\"Cosmetics.Variant.Channel.Particle\")" (Utilities|Array|MakeArray "(TagName=\"Cosmetics.Variant.Property.Particle0\")" "(TagName=\"Cosmetics.Variant.Property.Particle1\")" "(TagName=\"Cosmetics.Variant.Property.Particle2\")") false false))
    (for _array_element (CustomItemWrap|GetWrappedMeshComponents)
      (for _array_element_2 (Rendering|Material|GetMaterials _array_element)
        (bind _asmaterial_instance_dynamic (Utilities|Casting|CastToMaterialInstanceDynamic _array_element_2))
        (Rendering|Material|SetScalarParameterValue _asmaterial_instance_dynamic "EF_Anime_PBR_Shading" (Utilities|Array|Get(acopy) (Utilities|Array|MakeArray 0.0 1.0 1.0) _index) :!class /Script/Engine.MaterialInstanceDynamic)))))

(event Custom|OnBannerFullyApplied
  (params (Icon :type Texture%202D%20Object%20Reference) (PrimaryColor :type Linear%20Color%20Structure)))

(event EventOnSetWrapTarget
  (if (Utilities|String|EqualExactly(String) (Item|GetPersistentName (CosmeticLoadout|GetItemDefinitionfromLoadout (CosmeticLoadout|GetCosmeticLoadout) "(TagName=\"LoadoutSlot.Character\")")) (Variables|Default|GetCharacter-Name))
    (bind _index (Class|FortCosmeticFunctionLibrary|%28Deprecated%29GetVariantChannelActiveIndexForSlot "(TagName=\"LoadoutSlot.Character\")" (Variables|/Script/FortniteGame.FortPlayerPawn|GetCosmeticLoadout (CustomCharacter|GetAssociatedPlayerPawn)) "(TagName=\"Cosmetics.Variant.Channel.Particle\")" (Utilities|Array|MakeArray "(TagName=\"Cosmetics.Variant.Property.Particle0\")" "(TagName=\"Cosmetics.Variant.Property.Particle1\")" "(TagName=\"Cosmetics.Variant.Property.Particle2\")") false false))
    (for _array_element (CustomItemWrap|GetWrappedMeshComponents)
      (for _array_element_2 (Rendering|Material|GetMaterials _array_element)
        (bind _asmaterial_instance_dynamic (Utilities|Casting|CastToMaterialInstanceDynamic _array_element_2))
        (Rendering|Material|SetScalarParameterValue _asmaterial_instance_dynamic "EF_Anime_PBR_Shading" (Utilities|Array|Get(acopy) (Utilities|Array|MakeArray 0.0 1.0 1.0) _index) :!class /Script/Engine.MaterialInstanceDynamic)))))

(event EventOnPawnSet
  (if (Utilities|String|EqualExactly(String) (Item|GetPersistentName (CosmeticLoadout|GetItemDefinitionfromLoadout (CosmeticLoadout|GetCosmeticLoadout) "(TagName=\"LoadoutSlot.Character\")")) (Variables|Default|GetCharacter-Name))
    (bind _index (Class|FortCosmeticFunctionLibrary|%28Deprecated%29GetVariantChannelActiveIndexForSlot "(TagName=\"LoadoutSlot.Character\")" (Variables|/Script/FortniteGame.FortPlayerPawn|GetCosmeticLoadout (CustomCharacter|GetAssociatedPlayerPawn)) "(TagName=\"Cosmetics.Variant.Channel.Particle\")" (Utilities|Array|MakeArray "(TagName=\"Cosmetics.Variant.Property.Particle0\")" "(TagName=\"Cosmetics.Variant.Property.Particle1\")" "(TagName=\"Cosmetics.Variant.Property.Particle2\")") false false))
    (for _array_element (CustomItemWrap|GetWrappedMeshComponents)
      (for _array_element_2 (Rendering|Material|GetMaterials _array_element)
        (bind _asmaterial_instance_dynamic (Utilities|Casting|CastToMaterialInstanceDynamic _array_element_2))
        (Rendering|Material|SetScalarParameterValue _asmaterial_instance_dynamic "EF_Anime_PBR_Shading" (Utilities|Array|Get(acopy) (Utilities|Array|MakeArray 0.0 1.0 1.0) _index) :!class /Script/Engine.MaterialInstanceDynamic)))))

(event EventOnPostResIn
  (|CallParentFunction))

(fn ConstructionScript)
