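; Seared Scurf weapon wrap Blueprint that drives a reactive emissive effect which builds up as the weapon fires.
; Each shot increments a step count capped at MaxSteps, which is set shortly after BeginPlay.
; Buildup decays without Tick: the next shot lerps the count toward zero once a post-shot grace period has passed.
; Shots stamp game time into wrapped-mesh materials and scale the emissive offset's start and duration by buildup.
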(event EventBeginPlay
  (Utilities|FlowControl|Delay 0.2)
  (Variables|Default|SetMaxSteps 12))

(event EventOnWeaponFired
  (bind _returnvalue (CustomItemWrap|GetWrappedMeshComponents))
  (bind _returnvalue_2 (CustomItemWrap|GetWrappedMeshComponents))
  (bind _maxsteps (Variables|Default|GetMaxSteps))
  (bind _output_get (Variables|Default|SetStepCount (Math|Integer|Min(Integer) (+ (Math|Float|Truncate (Math|Float|Lerp (Math|Conversions|ToFloat(Integer) (Variables|Default|GetStepCount)) 0.0 (Math|Float|Clamp(Float) (/ (- (Utilities|Time|GetGameTimeinSeconds) (Variables|Default|Getwait_timer)) 1.0)))) 1) (Variables|Default|GetMaxSteps))))
  (Rendering|Material|SetScalarParameterValueonMaterials _returnvalue_2 "TopS Emissive_03-Brightness-TBP" (Utilities|Time|GetGameTimeinSeconds))
  (Rendering|Material|SetScalarParameterValueonMaterials _returnvalue_2 "TopS Emissive_02_Offset---Start" (Math|Float|Lerp -0.5 0.5 (/ (Math|Conversions|ToFloat(Integer) _output_get) _maxsteps)))
  (Variables|Default|Setwait_timer (+ (Utilities|Time|GetGameTimeinSeconds) 5.5))
  (Rendering|Material|SetScalarParameterValueonMaterials _returnvalue "TopS Emissive_02_Offset----TBP" (Utilities|Time|GetGameTimeinSeconds))
  (Rendering|Material|SetScalarParameterValueonMaterials _returnvalue "TopS Emissive_02_Offset---Duration" (Math|Float|Lerp 4.0 3.0 (/ (Math|Conversions|ToFloat(Integer) _output_get) _maxsteps))))

(fn ConstructionScript)
