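; Season 42 weapon wrap modifier that evolves its material decals as the player gets kills.
; A watched AthenaKills stat advances a counter capped at three, and a pause-gated timeline crossfades lettered decals while pulsing warp and emissive parameters.
; Weapon fire plays a second warp timeline and stamps game time into emissive parameters; in the front end, every fourth shot mimics a kill.
; The Restart timer target is empty, while Reverse' resets the front-end counter after a delay.
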(event EventOnPawnSet
  (bind _returnvalue (CustomItemWrap|GetWrappedMeshComponents))
  (bind _decals (Variables|Default|GetDecals))
  (Cosmetics|WatchCosmeticStat (CustomCharacter|GetAssociatedPlayerPawn) "(TagName=\"StatManager.AthenaKills\")" 0 false
    (:OnStatChanged
      (Rendering|Material|SetScalarParameterValueonMaterials (CustomItemWrap|GetWrappedMeshComponents) "TopS Emissive_01-Brightness-TBP" (Utilities|Time|GetGameTimeinSeconds))
      (if (Variables|Default|GetPause)
        _
        (elif (< (Variables|Default|GetKill_Count) 3)
          (Math|Integer|IncrementInt (Variables|Default|GetKill_Count)
            (:%20%20
              (Variables|Default|SetPause true)
              (bind (_direction _fadein _fadeout _inout _inoutbright) (Timeline|Anim PlayFromStart)
                (:Update
                  (Utilities|FlowControl|Sequence
                    (:then_0
                      (Rendering|Material|SetScalarParameterValueonMaterials _returnvalue (Utilities|String|StringToName (Utilities|String|ToString(Text) (Utilities|Text|FormatText "NSLOCTEXT(\"[F0E2587FF1EB337E9EE7B851C0893E4B]\", \"2E07D9874199329303B856ABB9D5A1FD\", \"EDRec_Decal{Letter}_Opacity_End\")" (Utilities|Array|Get(acopy) _decals (- (Variables|Default|GetKill_Count) 1))))) _fadein))
                    (:then_1
                      (if (>= (Variables|Default|GetKill_Count) 2)
                        (Rendering|Material|SetScalarParameterValueonMaterials _returnvalue (Utilities|String|StringToName (Utilities|String|ToString(Text) (Utilities|Text|FormatText "NSLOCTEXT(\"[F0E2587FF1EB337E9EE7B851C0893E4B]\", \"2E07D9874199329303B856ABB9D5A1FD\", \"EDRec_Decal{Letter}_Opacity_End\")" (Utilities|Array|Get(acopy) _decals (- (Variables|Default|GetKill_Count) 2))))) _fadeout)))
                    (:then_2
                      (Rendering|Material|SetScalarParameterValueonMaterials _returnvalue "EDRec_Emissive_A_Brightness_Time__End" _inoutbright)
                      (Rendering|Material|SetScalarParameterValueonMaterials _returnvalue "EDRec_Emissive_B_Brightness_Time__End" _inoutbright)
                      (Rendering|Material|SetScalarParameterValueonMaterials _returnvalue "EDRec_Emissive_C_Brightness_Time__End" _inoutbright)
                      (Rendering|Material|SetScalarParameterValueonMaterials _returnvalue "EDRec_WarpA_Strength" _inout)
                      (Rendering|Material|SetScalarParameterValueonMaterials _returnvalue "EDRec_WarpB_Strength" _inout)
                      (Rendering|Material|SetScalarParameterValueonMaterials _returnvalue "EDRec_WarpC_Strength" _inout))))
                (:Finished
                  (Variables|Default|SetPause false)
                  (if (not (CustomCosmeticModifier|IsinFrontEnd))
                    (Utilities|Time|SetTimerbyFunctionName 0 "Restart" 5.0)))))))))))

(event EventOnWeaponFired
  (bind _returnvalue (CustomItemWrap|GetWrappedMeshComponents))
  (bind _returnvalue_2 (Utilities|Time|GetGameTimeinSeconds))
  (bind _returnvalue_3 (CustomItemWrap|GetWrappedMeshComponents))
  (bind _decals (Variables|Default|GetDecals))
  (if (CustomCosmeticModifier|IsinFrontEnd)
    (Utilities|FlowControl|Sequence
      (:then_0
        _)
      (:then_1
        (bind _output_get (Variables|Default|SetShotsFired (+ (Variables|Default|GetShotsFired) 1)))
        (if (== (Math|Integer|%25%28Integer%29 _output_get 4) 0)
          (Rendering|Material|SetScalarParameterValueonMaterials _returnvalue "TopS Emissive_01-Brightness-TBP" (Utilities|Time|GetGameTimeinSeconds))
          (if (Variables|Default|GetPause)
            _
            (elif (< (Variables|Default|GetKill_Count) 3)
              (Math|Integer|IncrementInt (Variables|Default|GetKill_Count)
                (:%20%20
                  (Variables|Default|SetPause true)
                  (Timeline|Anim PlayFromStart)))))))
      (:then_2
        (bind (_direction _inout) (Timeline|Anim_02 PlayFromStart)
          (:Update
            (Rendering|Material|SetScalarParameterValueonMaterials _returnvalue "FDecalV2_UV_Warp-Strength" _inout)
            (Utilities|FlowControl|Delay 0.15)
            (Rendering|Material|SetScalarParameterValueonMaterials _returnvalue "EDRec_WarpA_Strength" _inout)
            (Rendering|Material|SetScalarParameterValueonMaterials _returnvalue "EDRec_WarpB_Strength" _inout)
            (Rendering|Material|SetScalarParameterValueonMaterials _returnvalue "EDRec_WarpC_Strength" _inout)))))
    (else
      (Utilities|FlowControl|Sequence
        (:then_0
          _)
        (:then_1
          (Timeline|Anim_02 PlayFromStart)))))
  (Rendering|Material|SetScalarParameterValueonMaterials _returnvalue "FDecalV2_Emissive_Brightness_Time_BP" _returnvalue_2)
  (Rendering|Material|SetScalarParameterValueonMaterials _returnvalue "FDecalV2_Emissive_Brightness_Time__End" _returnvalue_2)
  (if (not (Variables|Default|GetKilled))
    (Rendering|Material|SetScalarParameterValueonMaterials _returnvalue "FDecalV2_Emissive_Brightness_Time__End" 0.0)))

(event Custom|Restart)

(event Custom|Reverse'
  (if (CustomCosmeticModifier|IsinFrontEnd)
    (if (Variables|Default|GetPause)
      _
      (elif (>= (Variables|Default|GetKill_Count) 0)
        (Variables|Default|SetPause true)
        (Variables|Default|SetKill_Count 4)
        (Utilities|FlowControl|Sequence
          (:then_0
            (Components|Timeline|SetPlayRate (Variables|CIWM_Season42|GetAnim) 0.5))
          (:then_1
            (Utilities|FlowControl|RetriggerableDelay 3.0)
            (Variables|Default|SetKill_Count 0)
            (Variables|Default|SetPause false)))))))

(fn ConstructionScript)
