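; Season 41 reactive weapon wrap that drives decal and emissive material parameters on the wrapped meshes when the weapon fires.
; Each shot stamps game time into the animation parameters, with some end values reset when not Killed.
; A timeline ramps emissive brightness up, then reverses at reduced rate after a retriggerable delay.
; In frontend display, a looping timer calls LockerCycle, and a separate custom event only wraps LockerCounter back to zero.
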(event Custom|CustomEvent
  (if (Reactive|Frontend|IsCosmeticObjectDisplayedinFrontendContext 0)
    (Utilities|Time|SetTimerbyFunctionName 0 "LockerCycle" 3.0 true)))

(event EventOnWeaponFired
  (bind _returnvalue (CustomItemWrap|GetWrappedMeshComponents))
  (bind _returnvalue_2 (Utilities|Time|GetGameTimeinSeconds))
  (Utilities|FlowControl|Sequence
    (:then_0
      (Rendering|Material|SetScalarParameterValueonMaterials _returnvalue "EDRec_Decal_A_Animation_Time_BP" _returnvalue_2)
      (Rendering|Material|SetScalarParameterValueonMaterials _returnvalue "EDRec_TimeFrom_BP_DecalA" _returnvalue_2)
      (Rendering|Material|SetScalarParameterValueonMaterials _returnvalue "EDRec_Emissive_A_Brightness_Time_BP" _returnvalue_2)
      (Rendering|Material|SetScalarParameterValueonMaterials _returnvalue "EDRec_Emissive_A_Brightness_Time__End" _returnvalue_2)
      (Rendering|Material|SetScalarParameterValueonMaterials _returnvalue "EDRec_DecalA_Opacity_End" _returnvalue_2)
      (if (not (Variables|Default|GetKilled))
        (Rendering|Material|SetScalarParameterValueonMaterials _returnvalue "EDRec_Emissive_A_Brightness_Time__End" 0.0)
        (Rendering|Material|SetScalarParameterValueonMaterials _returnvalue "EDRec_DecalA_Opacity_End" 0.0)
        (Rendering|Material|SetScalarParameterValueonMaterials _returnvalue "TopS Emissive_01_Brightness---End" 1.0)))
    (:then_1
      (bind (_direction _newtrack_0) (Timeline|Timeline_0 Play)
        (:Update
          (Rendering|Material|SetScalarParameterValueonMaterials _returnvalue "TopS Emissive_01_Brightness---End" (Math|Float|Lerp 1.0 5.0 _newtrack_0))
          (Rendering|Material|SetScalarParameterValueonMaterials _returnvalue "FDecalV2_Emissive_Brightness_Time__End" (Math|Float|Lerp 1.0 5.0 _newtrack_0)))))
    (:then_2
      (Utilities|FlowControl|RetriggerableDelay 2.0)
      (Components|Timeline|SetPlayRate (Variables|Default|GetANim) 0.5)
      (Timeline|Timeline_0 Reverse))))

(event Custom|CustomEvent_2
  (if (>= (Variables|Default|GetLockerCounter) 3)
    (Variables|Default|SetLockerCounter 0)))

(fn ConstructionScript)
