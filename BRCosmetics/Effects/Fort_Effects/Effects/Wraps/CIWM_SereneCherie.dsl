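; Weapon wrap Blueprint for SereneCherie that shifts the wrap's base color hue when the weapon fires.
; Each shot rerolls a random positive or negative hue until it differs sufficiently from the last, then makes it the target.
; A timeline update eases the current hue toward that target and writes HueShift_BaseColor to the wrapped meshes' materials.
; A retriggerable delay resets the target to zero once firing pauses briefly.
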(event EventOnWeaponFired
  (while (< (Math|Float|Absolute(Float) (- (Variables|Default|GetNewHueValue) (Variables|Default|GetLastHueValue))) 0.2)
    (Variables|Default|SetNewHueValue (* (Math|Random|RandomFloatinRange 0.15 0.95) (Math|Float|SelectFloat 1.25 -0.5 (Math|Random|RandomBool)))))
  (Variables|Default|SetTargetHue (Variables|Default|GetNewHueValue))
  (Variables|Default|SetLastHueValue (Variables|Default|GetNewHueValue))
  (Utilities|FlowControl|Sequence
    (:then_0
      (Timeline|Timeline_0 Play
        (:Update
          (Variables|Default|SetCurrentHue (Math|Interpolation|FInterpTo (Variables|Default|GetCurrentHue) (Variables|Default|GetTargetHue) (Utilities|Time|GetWorldDeltaSeconds) 1.0))
          (Rendering|Material|SetScalarParameterValueonMaterials (CustomItemWrap|GetWrappedMeshComponents) "HueShift_BaseColor" (Variables|Default|GetCurrentHue)))))
    (:then_1
      (Utilities|FlowControl|RetriggerableDelay 0.5)
      (Variables|Default|SetTargetHue 0.0))))

(fn SineAnimation)

(fn ConstructionScript)
