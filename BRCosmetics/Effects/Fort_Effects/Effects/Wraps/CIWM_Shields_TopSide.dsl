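; Item wrap modifier that lights emissive layers on wrapped meshes according to the player's shield level.
; A slowed timeline simulates shield values in the front end; in game, pawn-set and weapon-change events read the pawn's shield percentage after a short delay.
; Doubled, clamped shields drive brightness parameters, game time stamps material animations, and a pause flag marks zero shields until a shield-gain stat clears it.
; Debug prints and an uncalled random flipbook helper remain.
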(event EventBeginPlay
  (bind _returnvalue (CustomItemWrap|GetWrappedMeshComponents))
  (bind _returnvalue_2 (Utilities|Time|GetGameTimeinSeconds))
  (bind _returnvalue_3 (CustomItemWrap|GetWrappedMeshComponents))
  (bind _shields (Variables|Default|GetShields))
  (bind _returnvalue_4 (* _shields 2.0))
  (bind _returnvalue_5 (Math|Float|Clamp(Float) _returnvalue_4))
  (Utilities|FlowControl|Delay 0.2)
  (if (CustomCosmeticModifier|IsinFrontEnd)
    (Components|Timeline|SetPlayRate (Variables|CIWM_Shields_TopSide|GetTimeline) 0.25)
    (bind (_direction _newtrack_1) (Timeline|Timeline PlayFromStart)
      (:NewTrack_0
        (Variables|Default|SetShields _newtrack_1)
        (Utilities|FlowControl|Sequence
          (:then_0
            ;[1 1 1] TopsSide V2
            (Rendering|Material|SetScalarParameterValueonMaterials _returnvalue "TopS Emissive_02-Brightness-TBP" _returnvalue_2)
            (Rendering|Material|SetScalarParameterValueonMaterials _returnvalue "TopS Emissive_02_Offset----TBP" _returnvalue_2)
            (Rendering|Material|SetScalarParameterValueonMaterials _returnvalue "TopS_01_Blend_Opacity-TBP" _returnvalue_2)
            (Rendering|Material|SetScalarParameterValueonMaterials _returnvalue "TimeFrom-BP----------------" _returnvalue_2))
          (:then_1
            (Rendering|Material|SetScalarParameterValueonMaterials _returnvalue "TopS Emissive_01-Brightness-TBP" _returnvalue_2)
            (Rendering|Material|SetScalarParameterValueonMaterials _returnvalue "TopS Emissive_01_Offset----TBP" _returnvalue_2)
            (Rendering|Material|SetScalarParameterValueonMaterials _returnvalue "TopS Emissive_03-Brightness-TBP" _returnvalue_2)
            (Rendering|Material|SetScalarParameterValueonMaterials _returnvalue "TopS Emissive_03_Offset----TBP" _returnvalue_2)
            (Rendering|Material|SetScalarParameterValueonMaterials _returnvalue_3 "TopS Emissive_01_Brightness---End" _returnvalue_5)
            (Rendering|Material|SetScalarParameterValueonMaterials _returnvalue_3 "TopS Emissive_02_Brightness---End" _returnvalue_5)
            (Rendering|Material|SetScalarParameterValueonMaterials _returnvalue_3 "TopS Emissive_03_Brightness---End" _returnvalue_5)
            (Rendering|Material|SetScalarParameterValueonMaterials _returnvalue_3 "TopS Emissive_01_Brightness_Darken_Base End" _returnvalue_5)
            (Rendering|Material|SetScalarParameterValueonMaterials _returnvalue_3 "TopS Emissive_02_Brightness_Darken_Base End" _returnvalue_5)
            (Variables|Default|Setpause (== _shields 0))))))))

(event EventOnPawnSet
  (bind _returnvalue (CustomItemWrap|GetWrappedMeshComponents))
  (bind _returnvalue_2 (Utilities|Time|GetGameTimeinSeconds))
  (bind _returnvalue_3 (CustomItemWrap|GetWrappedMeshComponents))
  (bind _shields (Variables|Default|GetShields))
  (bind _returnvalue_4 (* _shields 2.0))
  (bind _returnvalue_5 (Math|Float|Clamp(Float) _returnvalue_4))
  (bind _returnvalue_6 (CustomCharacter|GetAssociatedPlayerPawn))
  (Utilities|FlowControl|Sequence
    (:then_1
      (Cosmetics|WatchCosmeticStat _returnvalue_6 "(TagName=\"StatManager.GainedShieldTimes\")" 0 false
        (:OnStatChanged
          (Variables|Default|Setpause false))))
    (:then_2
      (Cosmetics|WatchCosmeticStat _returnvalue_6 "(TagName=\"StatManager.DamageReceived\")" 0 false
        (:OnStatChanged
          _))))
  (Utilities|FlowControl|Delay 0.1)
  (Variables|Default|SetShields (Class|SlurpSoldierHelpers|GetShieldPercentagefromPawn _returnvalue_6))
  (Development|PrintText "INVTEXT(\"Hello\")" true true "(R=1.000000,G=0.011126,B=0.000105,A=1.000000)")
  (if (not (Variables|Default|Getpause))
    (Utilities|FlowControl|Sequence
      (:then_0
        ;[1 1 1] TopsSide V2
        (Rendering|Material|SetScalarParameterValueonMaterials _returnvalue "TopS Emissive_02-Brightness-TBP" _returnvalue_2)
        (Rendering|Material|SetScalarParameterValueonMaterials _returnvalue "TopS Emissive_02_Offset----TBP" _returnvalue_2)
        (Rendering|Material|SetScalarParameterValueonMaterials _returnvalue "TopS_01_Blend_Opacity-TBP" _returnvalue_2)
        (Rendering|Material|SetScalarParameterValueonMaterials _returnvalue "TimeFrom-BP----------------" _returnvalue_2))
      (:then_1
        (Rendering|Material|SetScalarParameterValueonMaterials _returnvalue "TopS Emissive_01-Brightness-TBP" _returnvalue_2)
        (Rendering|Material|SetScalarParameterValueonMaterials _returnvalue "TopS Emissive_01_Offset----TBP" _returnvalue_2)
        (Rendering|Material|SetScalarParameterValueonMaterials _returnvalue "TopS Emissive_03-Brightness-TBP" _returnvalue_2)
        (Rendering|Material|SetScalarParameterValueonMaterials _returnvalue "TopS Emissive_03_Offset----TBP" _returnvalue_2)
        (Rendering|Material|SetScalarParameterValueonMaterials _returnvalue_3 "TopS Emissive_01_Brightness---End" _returnvalue_5)
        (Rendering|Material|SetScalarParameterValueonMaterials _returnvalue_3 "TopS Emissive_02_Brightness---End" _returnvalue_5)
        (Rendering|Material|SetScalarParameterValueonMaterials _returnvalue_3 "TopS Emissive_03_Brightness---End" _returnvalue_5)
        (Rendering|Material|SetScalarParameterValueonMaterials _returnvalue_3 "TopS Emissive_01_Brightness_Darken_Base End" _returnvalue_5)
        (Rendering|Material|SetScalarParameterValueonMaterials _returnvalue_3 "TopS Emissive_02_Brightness_Darken_Base End" _returnvalue_5)
        (Variables|Default|Setpause (== _shields 0))))))

(event EventOnWeaponChanged
  (params (NewWeapon :type Fort%20Weapon%20Object%20Reference) (OldWeapon :type Fort%20Weapon%20Object%20Reference))
  (bind _returnvalue (CustomItemWrap|GetWrappedMeshComponents))
  (bind _shields (Variables|Default|GetShields))
  (bind _returnvalue_2 (* _shields 2.0))
  (bind _returnvalue_3 (Math|Float|Clamp(Float) _returnvalue_2))
  (Utilities|FlowControl|Delay 0.1)
  (Variables|Default|SetShields (Class|SlurpSoldierHelpers|GetShieldPercentagefromPawn (CustomCharacter|GetAssociatedPlayerPawn)))
  (Rendering|Material|SetScalarParameterValueonMaterials _returnvalue "TopS Emissive_01_Brightness---End" _returnvalue_3)
  (Rendering|Material|SetScalarParameterValueonMaterials _returnvalue "TopS Emissive_02_Brightness---End" _returnvalue_3)
  (Rendering|Material|SetScalarParameterValueonMaterials _returnvalue "TopS Emissive_03_Brightness---End" _returnvalue_3)
  (Rendering|Material|SetScalarParameterValueonMaterials _returnvalue "TopS Emissive_01_Brightness_Darken_Base End" _returnvalue_3)
  (Rendering|Material|SetScalarParameterValueonMaterials _returnvalue "TopS Emissive_02_Brightness_Darken_Base End" _returnvalue_3)
  (Variables|Default|Setpause (== _shields 0)))

(event EventOnNewWeapons
  (params (WeaponsAsItems :type Array%20of%20Fort%20Item%20Object%20References))
  (bind _returnvalue (CustomItemWrap|GetWrappedMeshComponents))
  (bind _shields (Variables|Default|GetShields))
  (bind _returnvalue_2 (* _shields 2.0))
  (bind _returnvalue_3 (Math|Float|Clamp(Float) _returnvalue_2))
  (Utilities|FlowControl|Delay 0.1)
  (Variables|Default|SetShields (Class|SlurpSoldierHelpers|GetShieldPercentagefromPawn (CustomCharacter|GetAssociatedPlayerPawn)))
  (Rendering|Material|SetScalarParameterValueonMaterials _returnvalue "TopS Emissive_01_Brightness---End" _returnvalue_3)
  (Rendering|Material|SetScalarParameterValueonMaterials _returnvalue "TopS Emissive_02_Brightness---End" _returnvalue_3)
  (Rendering|Material|SetScalarParameterValueonMaterials _returnvalue "TopS Emissive_03_Brightness---End" _returnvalue_3)
  (Rendering|Material|SetScalarParameterValueonMaterials _returnvalue "TopS Emissive_01_Brightness_Darken_Base End" _returnvalue_3)
  (Rendering|Material|SetScalarParameterValueonMaterials _returnvalue "TopS Emissive_02_Brightness_Darken_Base End" _returnvalue_3)
  (Variables|Default|Setpause (== _shields 0)))

(event EventOnNewPlayerWeaponType
  (params (NewType :type EFortItemType%20Enum :default "WorldItem") (OldType :type EFortItemType%20Enum :default "WorldItem"))
  (bind _returnvalue (CustomItemWrap|GetWrappedMeshComponents))
  (bind _shields (Variables|Default|GetShields))
  (bind _returnvalue_2 (* _shields 2.0))
  (bind _returnvalue_3 (Math|Float|Clamp(Float) _returnvalue_2))
  (Utilities|FlowControl|Delay 0.1)
  (Variables|Default|SetShields (Class|SlurpSoldierHelpers|GetShieldPercentagefromPawn (CustomCharacter|GetAssociatedPlayerPawn)))
  (Rendering|Material|SetScalarParameterValueonMaterials _returnvalue "TopS Emissive_01_Brightness---End" _returnvalue_3)
  (Rendering|Material|SetScalarParameterValueonMaterials _returnvalue "TopS Emissive_02_Brightness---End" _returnvalue_3)
  (Rendering|Material|SetScalarParameterValueonMaterials _returnvalue "TopS Emissive_03_Brightness---End" _returnvalue_3)
  (Rendering|Material|SetScalarParameterValueonMaterials _returnvalue "TopS Emissive_01_Brightness_Darken_Base End" _returnvalue_3)
  (Rendering|Material|SetScalarParameterValueonMaterials _returnvalue "TopS Emissive_02_Brightness_Darken_Base End" _returnvalue_3)
  (Variables|Default|Setpause (== _shields 0)))

(fn RandomSet
  (params (FlipBook%20Frame :type Name) (TimeBP :type Name) (Offset :type Name) (Rotation :type Name) (X=min :type Float%20%28double-precision%29) (X=Max :type Float%20%28double-precision%29) (Y=Min :type Float%20%28double-precision%29) (Y=Max :type Float%20%28double-precision%29) (Rot=Min :type Float%20%28double-precision%29) (Rot=Max :type Float%20%28double-precision%29) (TimeBP-first :type Name))
  (bind _returnvalue (CustomItemWrap|GetWrappedMeshComponents))
  (bind _stream (Variables|Default|GetStream))
  (bind _returnvalue_2 (Math|Random|RandomFloatinRangefromStream _stream X=min X=Max))
  (bind _returnvalue_3 (Math|Random|RandomIntegerinRange 0 8))
  (bind _returnvalue_4 (Math|Conversions|ToFloat(Integer) _returnvalue_3))
  (bind _returnvalue_5 (/ _returnvalue_4 8.0))
  (Rendering|Material|SetScalarParameterValueonMaterials _returnvalue FlipBook%20Frame _returnvalue_5)
  (Rendering|Material|SetScalarParameterValueonMaterials _returnvalue TimeBP-first (Utilities|Time|GetGameTimeinSeconds))
  (Rendering|Material|SetVectorParameterValueonMaterials _returnvalue Offset (Math|Vector|MakeVector _returnvalue_2 (Math|Random|RandomFloatinRangefromStream _stream Y=Min Y=Max)))
  (Rendering|Material|SetScalarParameterValueonMaterials _returnvalue Rotation (Math|Random|RandomFloatinRange Rot=Min Rot=Max))
  (Rendering|Material|SetScalarParameterValueonMaterials _returnvalue TimeBP (Utilities|Time|GetGameTimeinSeconds))
  (Variables|Default|SetShields _returnvalue_2)
  (Development|PrintString (Utilities|String|ToString(Float) _returnvalue_5) true true "(R=0.060001,G=1.000000,B=0.000000,A=1.000000)" 3.0))

(fn ConstructionScript)
