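; Weapon wrap modifier for ShinToll that builds an emissive charge as the player deals damage and flips the wrap's look when full.
; DamageDealt stat changes advance a ten-step counter driving the gradient of emissive C or D, with game-time stamps for material-side animation.
; At ten, Shift inverts the mask and emissive blends, then the counter resets after a pause.
; In the front end, periodic weapon fire simulates damage.
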(event EventOnWeaponFired
  (bind _damage_buildup (Variables|EnvirometnVarable|GetDamage_BuildUp))
  (bind _returnvalue (Math|Conversions|ToFloat(Integer) _damage_buildup))
  (bind _returnvalue_2 (/ _returnvalue 1e+01))
  (bind _returnvalue_3 (Math|Float|Lerp -0.85 -0.1 _returnvalue_2))
  (bind _returnvalue_4 (CustomItemWrap|GetWrappedMeshComponents))
  (bind _returnvalue_5 (Utilities|Time|GetGameTimeinSeconds))
  (bind _returnvalue_6 (CustomItemWrap|GetWrappedMeshComponents))
  (bind _returnvalue_7 (Utilities|Time|GetGameTimeinSeconds))
  (bind _returnvalue_8 (CustomItemWrap|GetWrappedMeshComponents))
  (if (CustomCosmeticModifier|IsinFrontEnd)
    (bind _output_get (Variables|Envirometn_Varable|SetIterationCount (+ (Variables|Envirometn_Varable|GetIterationCount) 1)))
    (if (>= _output_get 5)
      (if (== (Variables|Envirometn_Varable|GetIterationCount) 7)
        (Variables|Envirometn_Varable|SetIterationCount 0))
      (if (not (Variables|EnvirometnVarable|GetPause%20Damage%20Delt))
        (Utilities|FlowControl|Delay 0.05)
        (Variables|EnvirometnVarable|SetDamage_BuildUp (+ (Variables|EnvirometnVarable|GetDamage_BuildUp) 1))
        (if (Variables|Default|Getinvert_State)
          (Rendering|Material|SetScalarParameterValueonMaterials _returnvalue_6 "TopS Emissive_01_Offset----TBP" _returnvalue_7)
          (Rendering|Material|SetScalarParameterValueonMaterials _returnvalue_6 "EDRec_Emissive_C_Brightness_Time_BP" _returnvalue_7)
          (Rendering|Material|SetScalarParameterValueonMaterials _returnvalue_6 "EDRec_Emissive_C_Gradent_Time_End" (Math|Float|Lerp -0.85 -0.1 (/ (Math|Conversions|ToFloat(Integer) (Variables|EnvirometnVarable|GetDamage_BuildUp)) 1e+01)))
          (else
            (Rendering|Material|SetScalarParameterValueonMaterials _returnvalue_8 "TopS Emissive_02_Offset----TBP" _returnvalue_5)
            (Rendering|Material|SetScalarParameterValueonMaterials _returnvalue_8 "EDRec_Emissive_D_Brightness_Time_BP" _returnvalue_5)
            (Rendering|Material|SetScalarParameterValueonMaterials _returnvalue_8 "EDRec_Emissive_D_Gradent_Time_End" (Math|Float|Lerp -0.85 -0.1 (/ (Math|Conversions|ToFloat(Integer) (Variables|EnvirometnVarable|GetDamage_BuildUp)) 1e+01)))))
        (if (== (Variables|EnvirometnVarable|GetDamage_BuildUp) 10)
          (Variables|EnvirometnVarable|SetPause%20Damage%20Delt true)
          (Utilities|FlowControl|Delay 0.3)
          (CallFunction|Shift)
          (Utilities|FlowControl|Delay (Variables|Default|DamageDelt|GetDamage_Delt%20Pause))
          (Variables|EnvirometnVarable|SetDamage_BuildUp 0)
          (Rendering|Material|SetScalarParameterValueonMaterials _returnvalue_4 "EDRec_Emissive_D_Gradent_Time_End" _returnvalue_3)
          (Rendering|Material|SetScalarParameterValueonMaterials _returnvalue_4 "EDRec_Emissive_C_Gradent_Time_End" _returnvalue_3)
          (Variables|EnvirometnVarable|SetPause%20Damage%20Delt false))))))

(event EventOnPostResIn
  (bind _damage_buildup (Variables|EnvirometnVarable|GetDamage_BuildUp))
  (bind _returnvalue (Math|Conversions|ToFloat(Integer) _damage_buildup))
  (bind _returnvalue_2 (/ _returnvalue 1e+01))
  (bind _returnvalue_3 (Math|Float|Lerp -0.85 -0.1 _returnvalue_2))
  (bind _returnvalue_4 (CustomItemWrap|GetWrappedMeshComponents))
  (bind _returnvalue_5 (Utilities|Time|GetGameTimeinSeconds))
  (bind _returnvalue_6 (CustomItemWrap|GetWrappedMeshComponents))
  (bind _returnvalue_7 (Utilities|Time|GetGameTimeinSeconds))
  (bind _returnvalue_8 (CustomItemWrap|GetWrappedMeshComponents))
  (bind _output_get (Variables|Default|SetNewVar (CustomCharacter|GetAssociatedPlayerPawn)))
  (Cosmetics|WatchCosmeticStat _output_get "(TagName=\"StatManager.DamageDealt\")" 0 false
    (:OnStatChanged
      (if (not (Variables|EnvirometnVarable|GetPause%20Damage%20Delt))
        (Utilities|FlowControl|Delay 0.05)
        (Variables|EnvirometnVarable|SetDamage_BuildUp (+ (Variables|EnvirometnVarable|GetDamage_BuildUp) 1))
        (if (Variables|Default|Getinvert_State)
          (Rendering|Material|SetScalarParameterValueonMaterials _returnvalue_6 "TopS Emissive_01_Offset----TBP" _returnvalue_7)
          (Rendering|Material|SetScalarParameterValueonMaterials _returnvalue_6 "EDRec_Emissive_C_Brightness_Time_BP" _returnvalue_7)
          (Rendering|Material|SetScalarParameterValueonMaterials _returnvalue_6 "EDRec_Emissive_C_Gradent_Time_End" (Math|Float|Lerp -0.85 -0.1 (/ (Math|Conversions|ToFloat(Integer) (Variables|EnvirometnVarable|GetDamage_BuildUp)) 1e+01)))
          (else
            (Rendering|Material|SetScalarParameterValueonMaterials _returnvalue_8 "TopS Emissive_02_Offset----TBP" _returnvalue_5)
            (Rendering|Material|SetScalarParameterValueonMaterials _returnvalue_8 "EDRec_Emissive_D_Brightness_Time_BP" _returnvalue_5)
            (Rendering|Material|SetScalarParameterValueonMaterials _returnvalue_8 "EDRec_Emissive_D_Gradent_Time_End" (Math|Float|Lerp -0.85 -0.1 (/ (Math|Conversions|ToFloat(Integer) (Variables|EnvirometnVarable|GetDamage_BuildUp)) 1e+01)))))
        (if (== (Variables|EnvirometnVarable|GetDamage_BuildUp) 10)
          (Variables|EnvirometnVarable|SetPause%20Damage%20Delt true)
          (Utilities|FlowControl|Delay 0.3)
          (CallFunction|Shift)
          (Utilities|FlowControl|Delay (Variables|Default|DamageDelt|GetDamage_Delt%20Pause))
          (Variables|EnvirometnVarable|SetDamage_BuildUp 0)
          (Rendering|Material|SetScalarParameterValueonMaterials _returnvalue_4 "EDRec_Emissive_D_Gradent_Time_End" _returnvalue_3)
          (Rendering|Material|SetScalarParameterValueonMaterials _returnvalue_4 "EDRec_Emissive_C_Gradent_Time_End" _returnvalue_3)
          (Variables|EnvirometnVarable|SetPause%20Damage%20Delt false))))))

(event EventOnPawnSet
  (bind _damage_buildup (Variables|EnvirometnVarable|GetDamage_BuildUp))
  (bind _returnvalue (Math|Conversions|ToFloat(Integer) _damage_buildup))
  (bind _returnvalue_2 (/ _returnvalue 1e+01))
  (bind _returnvalue_3 (Math|Float|Lerp -0.85 -0.1 _returnvalue_2))
  (bind _returnvalue_4 (CustomItemWrap|GetWrappedMeshComponents))
  (bind _returnvalue_5 (Utilities|Time|GetGameTimeinSeconds))
  (bind _returnvalue_6 (CustomItemWrap|GetWrappedMeshComponents))
  (bind _returnvalue_7 (Utilities|Time|GetGameTimeinSeconds))
  (bind _returnvalue_8 (CustomItemWrap|GetWrappedMeshComponents))
  (bind _output_get (Variables|Default|SetNewVar (CustomCharacter|GetAssociatedPlayerPawn)))
  (Cosmetics|WatchCosmeticStat _output_get "(TagName=\"StatManager.DamageDealt\")" 0 false
    (:OnStatChanged
      (if (not (Variables|EnvirometnVarable|GetPause%20Damage%20Delt))
        (Utilities|FlowControl|Delay 0.05)
        (Variables|EnvirometnVarable|SetDamage_BuildUp (+ (Variables|EnvirometnVarable|GetDamage_BuildUp) 1))
        (if (Variables|Default|Getinvert_State)
          (Rendering|Material|SetScalarParameterValueonMaterials _returnvalue_6 "TopS Emissive_01_Offset----TBP" _returnvalue_7)
          (Rendering|Material|SetScalarParameterValueonMaterials _returnvalue_6 "EDRec_Emissive_C_Brightness_Time_BP" _returnvalue_7)
          (Rendering|Material|SetScalarParameterValueonMaterials _returnvalue_6 "EDRec_Emissive_C_Gradent_Time_End" (Math|Float|Lerp -0.85 -0.1 (/ (Math|Conversions|ToFloat(Integer) (Variables|EnvirometnVarable|GetDamage_BuildUp)) 1e+01)))
          (else
            (Rendering|Material|SetScalarParameterValueonMaterials _returnvalue_8 "TopS Emissive_02_Offset----TBP" _returnvalue_5)
            (Rendering|Material|SetScalarParameterValueonMaterials _returnvalue_8 "EDRec_Emissive_D_Brightness_Time_BP" _returnvalue_5)
            (Rendering|Material|SetScalarParameterValueonMaterials _returnvalue_8 "EDRec_Emissive_D_Gradent_Time_End" (Math|Float|Lerp -0.85 -0.1 (/ (Math|Conversions|ToFloat(Integer) (Variables|EnvirometnVarable|GetDamage_BuildUp)) 1e+01)))))
        (if (== (Variables|EnvirometnVarable|GetDamage_BuildUp) 10)
          (Variables|EnvirometnVarable|SetPause%20Damage%20Delt true)
          (Utilities|FlowControl|Delay 0.3)
          (CallFunction|Shift)
          (Utilities|FlowControl|Delay (Variables|Default|DamageDelt|GetDamage_Delt%20Pause))
          (Variables|EnvirometnVarable|SetDamage_BuildUp 0)
          (Rendering|Material|SetScalarParameterValueonMaterials _returnvalue_4 "EDRec_Emissive_D_Gradent_Time_End" _returnvalue_3)
          (Rendering|Material|SetScalarParameterValueonMaterials _returnvalue_4 "EDRec_Emissive_C_Gradent_Time_End" _returnvalue_3)
          (Variables|EnvirometnVarable|SetPause%20Damage%20Delt false))))))

(event Custom|Shift
  (bind _returnvalue (CustomItemWrap|GetWrappedMeshComponents))
  (bind _invert_state_emiss (Variables|Default|Getinvert%20state_Emiss))
  (if (Variables|Default|Getinvert_State)
    (Variables|Default|Setinvert_State false)
    (Variables|Default|Setinvert%20state 0.0)
    (Variables|Default|Setinvert%20state_Emiss 1.0)
    (else
      (Variables|Default|Setinvert_State true)
      (Variables|Default|Setinvert%20state 1.0)
      (Variables|Default|Setinvert%20state_Emiss 0.0)))
  (Rendering|Material|SetScalarParameterValueonMaterials _returnvalue "TopS Blend-InvertMsk   1=invert" (Variables|Default|Getinvert%20state))
  (Rendering|Material|SetScalarParameterValueonMaterials _returnvalue "TopS Emissive_02_Blend 01---> 02 Affects Mask" _invert_state_emiss)
  (Rendering|Material|SetScalarParameterValueonMaterials _returnvalue "EDRec-Emissive-D_TopS Blend 01<--->02 MSk" _invert_state_emiss)
  (Rendering|Material|SetScalarParameterValueonMaterials (CustomItemWrap|GetWrappedMeshComponents) "TopS_01_Blend_Offset-TBP" (Utilities|Time|GetGameTimeinSeconds)))

(fn ConstructionScript)
