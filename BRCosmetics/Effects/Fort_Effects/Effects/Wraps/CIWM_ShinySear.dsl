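; Weapon wrap Blueprint for ShinySear that drives wrap material effects when the weapon fires.
; On each shot it writes current game time into TopS emissive, blend and time-from scalar parameters on the wrapped mesh components.
; A Pause flag with a Delay limits emissive speed stamps to once per second.
; A stored next-allowed time in Float_Value similarly throttles the third emissive layer's brightness and offset stamps.
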(event EventOnWeaponFired
  (bind _returnvalue (CustomItemWrap|GetWrappedMeshComponents))
  (bind _returnvalue_2 (Utilities|Time|GetGameTimeinSeconds))
  (Utilities|FlowControl|Sequence
    (:then_0
      ;[1 1 1] TopsSide V2
      (Rendering|Material|SetScalarParameterValueonMaterials _returnvalue "TopS Emissive_02-Brightness-TBP" _returnvalue_2)
      (Rendering|Material|SetScalarParameterValueonMaterials _returnvalue "TopS Emissive_02_Offset----TBP" _returnvalue_2)
      (Rendering|Material|SetScalarParameterValueonMaterials _returnvalue "TopS_01_Blend_Opacity-TBP" _returnvalue_2)
      (Rendering|Material|SetScalarParameterValueonMaterials _returnvalue "TopS_01_Blend_Offset-TBP" _returnvalue_2)
      (Rendering|Material|SetScalarParameterValueonMaterials _returnvalue "TimeFrom-BP----------------" _returnvalue_2))
    (:then_1
      (Rendering|Material|SetScalarParameterValueonMaterials _returnvalue "TopS Emissive_01-Brightness-TBP" _returnvalue_2)
      (Rendering|Material|SetScalarParameterValueonMaterials _returnvalue "TopS Emissive_01_Offset----TBP" _returnvalue_2)
      (if (not (Variables|Default|GetPause))
        (Variables|Default|SetPause true)
        (Rendering|Material|SetScalarParameterValueonMaterials _returnvalue "TopS Emissive_01_Speed-TBP" _returnvalue_2)
        (Rendering|Material|SetScalarParameterValueonMaterials _returnvalue "TopS Emissive_02_Speed-TBP" _returnvalue_2)
        (Rendering|Material|SetScalarParameterValueonMaterials _returnvalue "TopS Emissive_03_Speed-TBP" _returnvalue_2)
        (Utilities|FlowControl|Delay 1.0)
        (Variables|Default|SetPause false)))
    (:then_2
      (if (>= (Utilities|Time|GetGameTimeinSeconds) (Variables|Default|GetFloat_Value))
        (Rendering|Material|SetScalarParameterValueonMaterials _returnvalue "TopS Emissive_03-Brightness-TBP" _returnvalue_2)
        (Rendering|Material|SetScalarParameterValueonMaterials _returnvalue "TopS Emissive_03_Offset----TBP" _returnvalue_2)
        (Variables|Default|SetFloat_Value (+ (Utilities|Time|GetGameTimeinSeconds) 1.0))))))

(fn ConstructionScript)
