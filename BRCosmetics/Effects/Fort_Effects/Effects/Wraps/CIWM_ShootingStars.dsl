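; Custom item wrap Blueprint for the Shooting Stars weapon wrap, which triggers material pattern effects whenever the weapon fires.
; Each shot writes game time and a seeded random unit vector to one of three iteration parameter slots on the wrapped meshes' materials.
; The slot counter advances to the third, then keeps overwriting it until 0.66 seconds have passed since the first was written, when it loops back.
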(event EventOnWeaponFired
  (bind _returnvalue (Utilities|Time|GetGameTimeinSeconds))
  (bind _randomstream (Variables|Default|GetRandomStream))
  (bind _returnvalue_2 (Math|Random|RandomUnitVectorfromStream _randomstream))
  (bind _returnvalue_3 (CustomItemWrap|GetWrappedMeshComponents))
  (if (< (Variables|Default|GetIterationCount) 2)
    (Variables|Default|SetIterationCount (+ (Variables|Default|GetIterationCount) 1))
    (elif (>= (- (Utilities|Time|GetGameTimeinSeconds) (Variables|Default|GetWaitForLoop)) 0.66)
      (Variables|Default|SetIterationCount 0)
      (else
        (Variables|Default|SetIterationCount 2))))
  (switch int (Variables|Default|GetIterationCount)
    (:0
      (Rendering|Material|SetScalarParameterValueonMaterials _returnvalue_3 "GameTimeFromBP_Iteration1" _returnvalue)
      (Rendering|Material|SetVectorParameterValueonMaterials _returnvalue_3 "TriggeredPatterns_RandomHashFromBP_Iteration1" _returnvalue_2)
      (Variables|Default|SetWaitForLoop _returnvalue))
    (:1
      (Rendering|Material|SetScalarParameterValueonMaterials _returnvalue_3 "GameTimeFromBP_Iteration2" _returnvalue)
      (Rendering|Material|SetVectorParameterValueonMaterials _returnvalue_3 "TriggeredPatterns_RandomHashFromBP_Iteration2" _returnvalue_2))
    (:2
      (Rendering|Material|SetScalarParameterValueonMaterials _returnvalue_3 "GameTimeFromBP_Iteration3" _returnvalue)
      (Rendering|Material|SetVectorParameterValueonMaterials _returnvalue_3 "TriggeredPatterns_RandomHashFromBP_Iteration3" _returnvalue_2))))

(fn ConstructionScript)
