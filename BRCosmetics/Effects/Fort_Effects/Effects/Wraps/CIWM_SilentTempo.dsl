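; Item wrap modifier for the SilentTempo wrap that matches its textures to the active style of a specific equipped character.
; When customization finishes, a wrap target is set or a pawn is set, identical duplicated logic compares the loadout character's persistent name to a Character-Name variable.
; On a match, the Progressive channel's Stage1/Stage2 index picks the base color and clockwork textures, applied to dynamic material instances on all wrapped meshes.
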(event EventOnCharacterCustomizationFinished
  (params (CharacterPartInfo :type Array%20of%20Applied%20Character%20Part%20Info%20Structures))
  (if (Utilities|String|EqualExactly(String) (Item|GetPersistentName (CosmeticLoadout|GetItemDefinitionfromLoadout (CosmeticLoadout|GetCosmeticLoadout) "(TagName=\"LoadoutSlot.Character\")")) (Variables|Default|GetCharacter-Name))
    (bind _index (CosmeticLoadout|GetVariantChannelActiveIndexforSlotActor "(TagName=\"LoadoutSlot.Character\")" (CustomCharacter|GetAssociatedPlayerPawn) "(TagName=\"Cosmetics.Variant.Channel.Progressive\")" (Utilities|Array|MakeArray "(TagName=\"Cosmetics.Variant.Property.Stage1\")" "(TagName=\"Cosmetics.Variant.Property.Stage2\")") false false))
    (for _array_element (CustomItemWrap|GetWrappedMeshComponents)
      (for _array_element_2 (Rendering|Material|GetMaterials _array_element)
        (bind _asmaterial_instance_dynamic (Utilities|Casting|CastToMaterialInstanceDynamic _array_element_2))
        (Rendering|Material|SetTextureParameterValue _asmaterial_instance_dynamic "TopS_01__Tex_BaseColor" (Utilities|Array|Get(acopy) (Utilities|Array|MakeArray "/BRCosmetics/Weapons/WeaponSkins/Wraps/Textures/SilentTempo/T_Wrap_SilentTempo_BcD.T_Wrap_SilentTempo_BcD" "/BRCosmetics/Weapons/WeaponSkins/Wraps/Textures/SilentTempo/T_Wrap_SilentTempo_BcL.T_Wrap_SilentTempo_BcL") _index))
        (Rendering|Material|SetTextureParameterValue _asmaterial_instance_dynamic "BC_Clockwork" (Utilities|Array|Get(acopy) (Utilities|Array|MakeArray "/BRCosmetics/Weapons/WeaponSkins/Wraps/Textures/SilentTempo/T_Wrap_SilentTempo_BFD.T_Wrap_SilentTempo_BFD" "/BRCosmetics/Weapons/WeaponSkins/Wraps/Textures/SilentTempo/T_Wrap_SilentTempo_BFL.T_Wrap_SilentTempo_BFL") _index))))))

(event Custom|OnBannerFullyApplied
  (params (Icon :type Texture%202D%20Object%20Reference) (PrimaryColor :type Linear%20Color%20Structure)))

(event EventOnSetWrapTarget
  (if (Utilities|String|EqualExactly(String) (Item|GetPersistentName (CosmeticLoadout|GetItemDefinitionfromLoadout (CosmeticLoadout|GetCosmeticLoadout) "(TagName=\"LoadoutSlot.Character\")")) (Variables|Default|GetCharacter-Name))
    (bind _index (CosmeticLoadout|GetVariantChannelActiveIndexforSlotActor "(TagName=\"LoadoutSlot.Character\")" (CustomCharacter|GetAssociatedPlayerPawn) "(TagName=\"Cosmetics.Variant.Channel.Progressive\")" (Utilities|Array|MakeArray "(TagName=\"Cosmetics.Variant.Property.Stage1\")" "(TagName=\"Cosmetics.Variant.Property.Stage2\")") false false))
    (for _array_element (CustomItemWrap|GetWrappedMeshComponents)
      (for _array_element_2 (Rendering|Material|GetMaterials _array_element)
        (bind _asmaterial_instance_dynamic (Utilities|Casting|CastToMaterialInstanceDynamic _array_element_2))
        (Rendering|Material|SetTextureParameterValue _asmaterial_instance_dynamic "TopS_01__Tex_BaseColor" (Utilities|Array|Get(acopy) (Utilities|Array|MakeArray "/BRCosmetics/Weapons/WeaponSkins/Wraps/Textures/SilentTempo/T_Wrap_SilentTempo_BcD.T_Wrap_SilentTempo_BcD" "/BRCosmetics/Weapons/WeaponSkins/Wraps/Textures/SilentTempo/T_Wrap_SilentTempo_BcL.T_Wrap_SilentTempo_BcL") _index))
        (Rendering|Material|SetTextureParameterValue _asmaterial_instance_dynamic "BC_Clockwork" (Utilities|Array|Get(acopy) (Utilities|Array|MakeArray "/BRCosmetics/Weapons/WeaponSkins/Wraps/Textures/SilentTempo/T_Wrap_SilentTempo_BFD.T_Wrap_SilentTempo_BFD" "/BRCosmetics/Weapons/WeaponSkins/Wraps/Textures/SilentTempo/T_Wrap_SilentTempo_BFL.T_Wrap_SilentTempo_BFL") _index))))))

(event EventOnPawnSet
  (if (Utilities|String|EqualExactly(String) (Item|GetPersistentName (CosmeticLoadout|GetItemDefinitionfromLoadout (CosmeticLoadout|GetCosmeticLoadout) "(TagName=\"LoadoutSlot.Character\")")) (Variables|Default|GetCharacter-Name))
    (bind _index (CosmeticLoadout|GetVariantChannelActiveIndexforSlotActor "(TagName=\"LoadoutSlot.Character\")" (CustomCharacter|GetAssociatedPlayerPawn) "(TagName=\"Cosmetics.Variant.Channel.Progressive\")" (Utilities|Array|MakeArray "(TagName=\"Cosmetics.Variant.Property.Stage1\")" "(TagName=\"Cosmetics.Variant.Property.Stage2\")") false false))
    (for _array_element (CustomItemWrap|GetWrappedMeshComponents)
      (for _array_element_2 (Rendering|Material|GetMaterials _array_element)
        (bind _asmaterial_instance_dynamic (Utilities|Casting|CastToMaterialInstanceDynamic _array_element_2))
        (Rendering|Material|SetTextureParameterValue _asmaterial_instance_dynamic "TopS_01__Tex_BaseColor" (Utilities|Array|Get(acopy) (Utilities|Array|MakeArray "/BRCosmetics/Weapons/WeaponSkins/Wraps/Textures/SilentTempo/T_Wrap_SilentTempo_BcD.T_Wrap_SilentTempo_BcD" "/BRCosmetics/Weapons/WeaponSkins/Wraps/Textures/SilentTempo/T_Wrap_SilentTempo_BcL.T_Wrap_SilentTempo_BcL") _index))
        (Rendering|Material|SetTextureParameterValue _asmaterial_instance_dynamic "BC_Clockwork" (Utilities|Array|Get(acopy) (Utilities|Array|MakeArray "/BRCosmetics/Weapons/WeaponSkins/Wraps/Textures/SilentTempo/T_Wrap_SilentTempo_BFD.T_Wrap_SilentTempo_BFD" "/BRCosmetics/Weapons/WeaponSkins/Wraps/Textures/SilentTempo/T_Wrap_SilentTempo_BFL.T_Wrap_SilentTempo_BFL") _index))))))

(event EventOnPostResIn
  (|CallParentFunction))

(fn ConstructionScript)
