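; Weapon wrap logic for the SilverBullet cosmetic wrap that updates decal material parameters each time the weapon fires.
; On fire, a random index node feeds two selects that choose among decal slots A to D on the wrapped mesh components' materials.
; One select's parameter receives a random flipbook frame, the other's receives the current game time, presumably so the material animates the decal from that moment.
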(event EventOnWeaponFired
  (bind _returnvalue (CustomItemWrap|GetWrappedMeshComponents))
  (bind _returnvalue_2 (Math|Random|RandomIntegerinRange 0 3))
  (Rendering|Material|SetScalarParameterValueonMaterials _returnvalue (Utilities|Select :Index _returnvalue_2 :Option%200 "EDRec_FlipBook_FrameA" :Option%201 "EDRec_FlipBook_FrameB" :Option%202 "EDRec_FlipBook_FrameC" :Option%203 "EDRec_FlipBook_FrameD") (Math|Conversions|ToFloat(Integer) (Math|Random|RandomIntegerinRange 0 3)))
  (Rendering|Material|SetScalarParameterValueonMaterials _returnvalue (Utilities|Select :Index _returnvalue_2 :Option%200 "EDRec_TimeFrom_BP_DecalA" :Option%201 "EDRec_TimeFrom_BP_DecalB" :Option%202 "EDRec_TimeFrom_BP_DecalC" :Option%203 "EDRec_TimeFrom_BP_DecalD") (Utilities|Time|GetGameTimeinSeconds)))

(fn ConstructionScript)
