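; Custom item wrap Blueprint CIWM_Simple that feeds weapon fire timing into wrap materials for effects.
; On the weapon-fired event it gets the wrapped mesh components and sets a time scalar parameter on their materials to the current game time in seconds.
; The construction script is empty, so any visual response to that timestamp lives in the materials.
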(event EventOnWeaponFired
  (Rendering|Material|SetScalarParameterValueonMaterials (CustomItemWrap|GetWrappedMeshComponents) "TimeFrom-BP----------------" (Utilities|Time|GetGameTimeinSeconds)))

(fn ConstructionScript)
