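; Item wrap Blueprint (CIWM_SimpleDelay) that drives a time-based material effect on a wrapped weapon when it fires.
; On the weapon-fired event it waits a very short delay, then writes game time plus a small fixed offset into the "TimeFrom-BP" scalar parameter on all wrapped mesh components.
; The wrap material presumably animates from that timestamp; the construction script is empty.
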(event EventOnWeaponFired
  (Utilities|FlowControl|Delay 0.01)
  (Rendering|Material|SetScalarParameterValueonMaterials (CustomItemWrap|GetWrappedMeshComponents) "TimeFrom-BP----------------" (+ (Utilities|Time|GetGameTimeinSeconds) 0.35)))

(fn ConstructionScript)
