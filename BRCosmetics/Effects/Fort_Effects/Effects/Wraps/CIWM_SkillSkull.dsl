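; Custom item wrap Blueprint for SkillSkull that makes the wrap's animated decal react to weapon fire.
; Each shot stamps game time into opacity and emissive material parameters and raises a capped fire counter that adds to flipbook speed.
; Chained delays then halve the boost and later reset it to the base speed.
; Shots during a pending delay update the materials without restarting that timer.
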(event EventOnWeaponFired
  (bind _returnvalue (Utilities|Time|GetGameTimeinSeconds))
  (bind _returnvalue_2 (CustomItemWrap|GetWrappedMeshComponents))
  (Variables|Default|SetFireCounter (Math|Float|Clamp(Float) (+ (Variables|Default|GetFireCounter) 1.0) 0.0 4.0))
  (Rendering|Material|SetScalarParameterValueonMaterials _returnvalue_2 "FDecalaV2_Opacity_Time_BP_Z" _returnvalue)
  (Rendering|Material|SetScalarParameterValueonMaterials _returnvalue_2 "FDecalV2_Emissive_Brightness_Time_BP" _returnvalue)
  (Rendering|Material|SetScalarParameterValueonMaterials _returnvalue_2 "FDecalV2_FlipBook_Speed" (+ (Variables|Default|GetFireCounter) (Variables|Default|GetbaseValue)))
  (Utilities|FlowControl|Delay 2.0)
  (bind _output_get (Variables|Default|SetFireCounter (* 0.5 (Variables|Default|GetFireCounter))))
  (Rendering|Material|SetScalarParameterValueonMaterials _returnvalue_2 "FDecalV2_FlipBook_Speed" (+ _output_get (Variables|Default|GetbaseValue)))
  (Utilities|FlowControl|Delay 2.0)
  (Rendering|Material|SetScalarParameterValueonMaterials _returnvalue_2 "FDecalV2_FlipBook_Speed" (Variables|Default|GetbaseValue))
  (Variables|Default|SetFireCounter 0.0))

(fn ConstructionScript)
