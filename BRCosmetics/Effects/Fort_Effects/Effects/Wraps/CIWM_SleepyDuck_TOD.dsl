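; Item wrap cosmetic modifier Blueprint for the SleepyDuck TOD wrap, animating material parameters on the wrapped mesh components.
; On BeginPlay, only in the front end, a timeline drives the "00-FE-TODCC-Override" scalar with its TOD value.
; Each weapon fire restarts a Bounce timeline, whose output is lerped into a small "FDecalV2_UV_Rotation_Speed Bounce Strength" value on the wrap materials.
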(event EventBeginPlay
  (if (CustomCosmeticModifier|IsinFrontEnd)
    (bind (_direction _tod_value) (Timeline|Timeline PlayFromStart)
      (:Update
        (Rendering|Material|SetScalarParameterValueonMaterials (CustomItemWrap|GetWrappedMeshComponents) "00-FE-TODCC-Override" _tod_value)))))

(event EventOnWeaponFired
  (bind (_direction _newtrack_0) (Timeline|Bounce PlayFromStart)
    (:Update
      (Rendering|Material|SetScalarParameterValueonMaterials (CustomItemWrap|GetWrappedMeshComponents) "FDecalV2_UV_Rotation_Speed Bounce Strength" (Math|Float|Lerp 0.0 0.1 _newtrack_0)))))

(fn ConstructionScript)
