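; Custom item wrap Blueprint for the SleetReceipt wrap that animates the wrap material whenever the weapon fires.
; On fire, FireTimeline restarts and its tracks drive UV warp amount and emissive brightness on the wrapped mesh components.
; Each update also adds the offset track to a stored Offset, so movement accumulates across shots.
; That value feeds offset vector parameters for backpack, weapon, character, mech, pickaxe and vehicle variants.
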(event EventOnPawnSet)

(event EventOnWeaponFired
  (bind _returnvalue (CustomItemWrap|GetWrappedMeshComponents))
  (bind (_direction _warp _newtrack_0 _offset) (Timeline|FireTimeline PlayFromStart)
    (:Update
      (Rendering|Material|SetScalarParameterValueonMaterials _returnvalue "TopS_03_Warp_UV_Amount" _warp)
      (Rendering|Material|SetScalarParameterValueonMaterials _returnvalue "TopS Emissive_02_Brightness---End" _newtrack_0)
      (bind _output_get (Variables|Default|SetOffset (+ _offset (Variables|Default|GetOffset))))
      (for _array_element (Utilities|Array|MakeArray (Utilities|Name|MakeLiteralName "TopS_03_OffsetS_Backpack") (Utilities|Name|MakeLiteralName "TopS_03_OffsetS_Weapon") (Utilities|Name|MakeLiteralName "TopS_03_OffsetS_Character") (Utilities|Name|MakeLiteralName "TopS_03_OffsetS_Mech") (Utilities|Name|MakeLiteralName "TopS_03_OffsetS_Pickaxe") (Utilities|Name|MakeLiteralName "TopS_03_OffsetS_Vehicle") (Utilities|Name|MakeLiteralName "TopS_03_OffsetS_VehicleSide"))
        (Rendering|Material|SetVectorParameterValueonMaterials _returnvalue _array_element (Math|Vector|MakeVector _output_get))))))

(fn ConstructionScript)
