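; Item wrap Blueprint for the SlickSwish cosmetic that animates the wrap's materials in reaction to weapon fire.
; On each shot it stamps the current game time into several emissive, blend and time scalar parameters on the wrapped meshes.
; It also randomizes an offset and restarts a looping timer for Offset_Warp.
; Each tick of that timer increments the offset and writes it to per-target vector parameters for weapon, backpack, character, mech, pickaxe and vehicle.
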(event EventOnWeaponFired
  (bind _returnvalue (Utilities|Time|GetGameTimeinSeconds))
  (bind _returnvalue_2 (CustomItemWrap|GetWrappedMeshComponents))
  (bind _output_get (Variables|Default|SetTimer))
  (Utilities|Time|ClearandInvalidateTimerbyHandle _output_get)
  (bind _returnvalue_3 (Utilities|Time|SetTimerbyFunctionName 0 "Offset_Warp" 0.1 true true))
  (Variables|Default|SetTimer _returnvalue_3)
  (Variables|Default|SetOffset (Math|Random|RandomFloatinRange 0.0 1.0))
  (Utilities|FlowControl|Sequence
    (:then_0
      ;[1 1 1] TopsSide V2
      (Rendering|Material|SetScalarParameterValueonMaterials _returnvalue_2 "TopS Emissive_02-Brightness-TBP" _returnvalue)
      (Rendering|Material|SetScalarParameterValueonMaterials _returnvalue_2 "TopS Emissive_02_Offset----TBP" _returnvalue)
      (Rendering|Material|SetScalarParameterValueonMaterials _returnvalue_2 "TopS_01_Blend_Opacity-TBP" _returnvalue)
      (Rendering|Material|SetScalarParameterValueonMaterials _returnvalue_2 "TimeFrom-BP----------------" _returnvalue))
    (:then_1
      (Rendering|Material|SetScalarParameterValueonMaterials _returnvalue_2 "TopS Emissive_01-Brightness-TBP" _returnvalue)
      (Rendering|Material|SetScalarParameterValueonMaterials _returnvalue_2 "TopS Emissive_01_Offset----TBP" _returnvalue)
      (Rendering|Material|SetScalarParameterValueonMaterials _returnvalue_2 "TopS Emissive_03-Brightness-TBP" _returnvalue)
      (Rendering|Material|SetScalarParameterValueonMaterials _returnvalue_2 "TopS Emissive_03_Offset----TBP" _returnvalue))))

(event Custom|Offset_Warp
  (bind _returnvalue (CustomItemWrap|GetWrappedMeshComponents))
  (bind _output_get (Variables|Default|SetOffset (+ (Variables|Default|GetOffset) 0.1)))
  (Rendering|Material|SetVectorParameterValueonMaterials _returnvalue "TopS_04_Offset_Weapon" (Math|Conversions|ToVector(Float) _output_get))
  (Rendering|Material|SetVectorParameterValueonMaterials _returnvalue "TopS_04_Offset_Backpack" (Math|Conversions|ToVector(Float) _output_get))
  (Rendering|Material|SetVectorParameterValueonMaterials _returnvalue "TopS_04_Offset_Character" (Math|Conversions|ToVector(Float) _output_get))
  (Rendering|Material|SetVectorParameterValueonMaterials _returnvalue "TopS_04_Offset_Mech" (Math|Conversions|ToVector(Float) _output_get))
  (Rendering|Material|SetVectorParameterValueonMaterials _returnvalue "TopS_04_Offset_Pickaxe" (Math|Conversions|ToVector(Float) _output_get))
  (Rendering|Material|SetVectorParameterValueonMaterials _returnvalue "TopS_04_Offset_Vehicle" (Math|Conversions|ToVector(Float) _output_get))
  (Rendering|Material|SetVectorParameterValueonMaterials _returnvalue "TopS_04_Offset_VehicleSide" (Math|Conversions|ToVector(Float) _output_get)))

(fn ConstructionScript)
