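; Custom item wrap modifier for SlugRipple that changes the wrap's decal flipbook frame in reaction to weapon fire and kills.
; BeginPlay and OnPawnSet watch the Athena kills stat, showing a kill frame and setting Pause so shots cannot override it.
; Shots show a separate frame while unpaused, and every seventh shot in the front end shows the kill frame.
; Retriggerable delays return the frame to zero.
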(event EventBeginPlay
  (bind _returnvalue (CustomItemWrap|GetWrappedMeshComponents))
  (Cosmetics|WatchCosmeticStat (CustomCharacter|GetAssociatedPlayerPawn) "(TagName=\"StatManager.AthenaKills\")" 0 false
    (:OnStatChanged
      (Variables|Default|SetPause true)
      (Rendering|Material|SetScalarParameterValueonMaterials _returnvalue "FDecalV2_FlipBook_Frame #" 0.5)
      (Utilities|FlowControl|RetriggerableDelay 2.0)
      (Rendering|Material|SetScalarParameterValueonMaterials _returnvalue "FDecalV2_FlipBook_Frame #" 0.0)
      (Variables|Default|SetPause false))))

(event EventOnPawnSet
  (bind _returnvalue (CustomItemWrap|GetWrappedMeshComponents))
  (Cosmetics|WatchCosmeticStat (CustomCharacter|GetAssociatedPlayerPawn) "(TagName=\"StatManager.AthenaKills\")" 0 false
    (:OnStatChanged
      (Variables|Default|SetPause true)
      (Rendering|Material|SetScalarParameterValueonMaterials _returnvalue "FDecalV2_FlipBook_Frame #" 0.5)
      (Utilities|FlowControl|RetriggerableDelay 2.0)
      (Rendering|Material|SetScalarParameterValueonMaterials _returnvalue "FDecalV2_FlipBook_Frame #" 0.0)
      (Variables|Default|SetPause false))))

(event EventOnWeaponFired
  (bind _returnvalue (CustomItemWrap|GetWrappedMeshComponents))
  (bind _returnvalue_2 (CustomItemWrap|GetWrappedMeshComponents))
  (Utilities|FlowControl|Sequence
    (:then_0
      (if (not (Variables|Default|GetPause))
        (Rendering|Material|SetScalarParameterValueonMaterials _returnvalue "FDecalV2_FlipBook_Frame #" 0.3)
        (Utilities|FlowControl|RetriggerableDelay 2.0)
        (Rendering|Material|SetScalarParameterValueonMaterials _returnvalue "FDecalV2_FlipBook_Frame #" 0.0)))
    (:then_1
      (bind _result (Math|Integer|IncrementInt (Variables|Default|GetIterationCount))
        (:%20%20
          (if (and (== (Math|Integer|%25%28Integer%29 _result 7) 0) (CustomCosmeticModifier|IsinFrontEnd))
            (Rendering|Material|SetScalarParameterValueonMaterials _returnvalue_2 "FDecalV2_FlipBook_Frame #" 0.5)
            (Utilities|FlowControl|RetriggerableDelay 2.0)
            (Rendering|Material|SetScalarParameterValueonMaterials _returnvalue_2 "FDecalV2_FlipBook_Frame #" 0.0)))))))

(fn ConstructionScript)
