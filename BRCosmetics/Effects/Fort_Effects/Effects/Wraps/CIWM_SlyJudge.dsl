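; Cosmetic wrap modifier for SlyJudge that shows a rainbow effect on wrapped meshes once the pawn's OpenChests stat is above zero.
; Stat changes feed CheckDisplayRainbow, which stamps game time and an end offset into material parameters once, latched by wasTriggered, or zeroes the end offset.
; Weapon changes clear the latch and reapply after a short delay.
; In the front end, a timeline fakes the stat to preview showing and hiding.
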(event EventBeginPlay
  (Utilities|FlowControl|Delay 0.2)
  (if (CustomCosmeticModifier|IsinFrontEnd)
    (Timeline|ShowRainbowTimeline Play
      (:DisplayRainbow
        (Variables|Default|SetOpenChestValue 1)
        (Variables|Default|SetwasTriggered false)
        (CallFunction|CheckDisplayRainbow))
      (:HideRainbow
        (Variables|Default|SetOpenChestValue 0)
        (Variables|Default|SetwasTriggered true)
        (CallFunction|CheckDisplayRainbow)))))

(event EventOnPawnSet
  (bind (_statname _statvalue) (Cosmetics|WatchCosmeticStat (CustomCharacter|GetAssociatedPlayerPawn) "(TagName=\"StatManager.OpenChests\")")
    (:OnStatChanged
      (Variables|Default|SetOpenChestValue _statvalue)
      (CallFunction|CheckDisplayRainbow))))

(event Custom|CheckDisplayRainbow
  (bind _returnvalue (CustomItemWrap|GetWrappedMeshComponents))
  (if (> (Variables|Default|GetOpenChestValue) 0)
    (if (not (Variables|Default|GetwasTriggered))
      (Rendering|Material|SetScalarParameterValueonMaterials _returnvalue "TopS_01_Blend_Offset-TBP" (Utilities|Time|GetGameTimeinSeconds))
      (Rendering|Material|SetScalarParameterValueonMaterials _returnvalue "TopS_01_Blend_Offset---End" 0.6)
      (Variables|Default|SetwasTriggered true))
    (else
      (Rendering|Material|SetScalarParameterValueonMaterials (CustomItemWrap|GetWrappedMeshComponents) "TopS_01_Blend_Offset---End" 0.0)
      (Variables|Default|SetwasTriggered false))))

(event EventOnNewWeapons
  (params (WeaponsAsItems :type Array%20of%20Fort%20Item%20Object%20References))
  (Utilities|FlowControl|Delay 0.1)
  (Variables|Default|SetwasTriggered false)
  (CallFunction|CheckDisplayRainbow))

(event EventOnNewPlayerWeaponType
  (params (NewType :type EFortItemType%20Enum :default "WorldItem") (OldType :type EFortItemType%20Enum :default "WorldItem"))
  (Utilities|FlowControl|Delay 0.1)
  (Variables|Default|SetwasTriggered false)
  (CallFunction|CheckDisplayRainbow))

(fn ConstructionScript)
