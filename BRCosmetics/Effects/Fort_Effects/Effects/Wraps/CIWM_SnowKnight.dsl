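; SnowKnight item wrap script that tints wrapped meshes to match the player's outfit when it is the configured character.
; On pawn set, wrap target set or customization finished, identical logic reads crystal, trim, secondary and primary colors from the Body material.
; It writes them to wrap channel tints, deriving time-of-day colors from the crystal color.
; Banner IDs are stored, but the banner-applied event is empty and `setcolors` is never called here.
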(event EventOnPawnSet
  (bind _returnvalue (CosmeticLoadout|GetCosmeticLoadout))
  (Variables|Default|SetBanner%20Icon%20Id (Utilities|String|StringToName (CosmeticLoadout|GetBannerIconIdfromLoadout _returnvalue)))
  (Variables|Default|SetBanner%20Color%20Id (Utilities|String|StringToName (CosmeticLoadout|GetBannerColorIdfromLoadout _returnvalue)))
  (if (Utilities|String|EqualExactly(String) (Item|GetPersistentName (CosmeticLoadout|GetItemDefinitionfromLoadout (CosmeticLoadout|GetCosmeticLoadout) "(TagName=\"LoadoutSlot.Character\")")) (Variables|Default|GetCharacter-Name))
    (bind _asmaterial_instance_dynamic (Utilities|Casting|CastToMaterialInstanceDynamic (Rendering|Material|GetMaterial (Customization|GetSkeletalMeshforPartType (CustomCharacter|GetAssociatedPlayerPawn) "Body" :!class /Script/FortniteGame.FortPlayerPawn) 0)))
    (Utilities|FlowControl|Sequence
      (:then_0
        (bind _returnvalue_2 (Rendering|Material|GetVectorParameterValue _asmaterial_instance_dynamic "Crystal_Color" :!class /Script/Engine.MaterialInstanceDynamic))
        (for _array_element (CustomItemWrap|GetWrappedMeshComponents)
          (for _array_element_2 (Rendering|Material|GetMaterials _array_element)
            (bind _asmaterial_instance_dynamic_2 (Utilities|Casting|CastToMaterialInstanceDynamic _array_element_2))
            (Rendering|Material|SetVectorParameterValue _asmaterial_instance_dynamic_2 "TopS_01_ColorTint_CH_Blue  A=0=off" (* _returnvalue_2 "(R=1.322917,G=1.645964,B=2.000000,A=1.000000)") :!class /Script/Engine.MaterialInstanceDynamic)
            (Rendering|Material|SetVectorParameterValue _asmaterial_instance_dynamic_2 "TS_TOD_Alpha_Channel_Color_Night" (* _returnvalue_2 "(R=1.000000,G=1.000000,B=1.000000,A=1.000000)") :!class /Script/Engine.MaterialInstanceDynamic)
            (Rendering|Material|SetVectorParameterValue _asmaterial_instance_dynamic_2 "TS_TOD_Alpha_Channel_Color_Evening" (* (Math|Color|Desaturate(LinearColor) _returnvalue_2 0.35) "(R=0.500000,G=0.500000,B=0.500000,A=1.000000)") :!class /Script/Engine.MaterialInstanceDynamic)
            (Rendering|Material|SetVectorParameterValue _asmaterial_instance_dynamic_2 "TS_TOD_Alpha_Channel_Color_Morning" (* (Math|Color|Desaturate(LinearColor) _returnvalue_2 0.25) "(R=1.000000,G=1.000000,B=1.000000,A=1.000000)") :!class /Script/Engine.MaterialInstanceDynamic))))
      (:then_1
        (bind _returnvalue_3 (Rendering|Material|GetVectorParameterValue _asmaterial_instance_dynamic "Trim_Color" :!class /Script/Engine.MaterialInstanceDynamic))
        (for _array_element_3 (CustomItemWrap|GetWrappedMeshComponents)
          (for _array_element_4 (Rendering|Material|GetMaterials _array_element_3)
            (bind _asmaterial_instance_dynamic_3 (Utilities|Casting|CastToMaterialInstanceDynamic _array_element_4))
            (Rendering|Material|SetVectorParameterValue _asmaterial_instance_dynamic_3 "TopS_01_ColorTint_CH_Green A=0=off" _returnvalue_3 :!class /Script/Engine.MaterialInstanceDynamic))))
      (:then_2
        (bind _returnvalue_4 (Rendering|Material|GetVectorParameterValue _asmaterial_instance_dynamic "Secondary_Color" :!class /Script/Engine.MaterialInstanceDynamic))
        (for _array_element_5 (CustomItemWrap|GetWrappedMeshComponents)
          (for _array_element_6 (Rendering|Material|GetMaterials _array_element_5)
            (bind _asmaterial_instance_dynamic_4 (Utilities|Casting|CastToMaterialInstanceDynamic _array_element_6))
            (Rendering|Material|SetVectorParameterValue _asmaterial_instance_dynamic_4 "TopS_01_ColorTint_CH_Alpha   A=0=off" _returnvalue_4 :!class /Script/Engine.MaterialInstanceDynamic))))
      (:then_3
        (bind _returnvalue_5 (Rendering|Material|GetVectorParameterValue _asmaterial_instance_dynamic "Primary_Color" :!class /Script/Engine.MaterialInstanceDynamic))
        (for _array_element_7 (CustomItemWrap|GetWrappedMeshComponents)
          (for _array_element_8 (Rendering|Material|GetMaterials _array_element_7)
            (bind _asmaterial_instance_dynamic_5 (Utilities|Casting|CastToMaterialInstanceDynamic _array_element_8))
            (Rendering|Material|SetVectorParameterValue _asmaterial_instance_dynamic_5 "TopS_01_ColorTint_CH_Red  A=0=off" _returnvalue_5 :!class /Script/Engine.MaterialInstanceDynamic)))))))

(event EventOnSetWrapTarget
  (bind _returnvalue (CosmeticLoadout|GetCosmeticLoadout))
  (Variables|Default|SetBanner%20Icon%20Id (Utilities|String|StringToName (CosmeticLoadout|GetBannerIconIdfromLoadout _returnvalue)))
  (Variables|Default|SetBanner%20Color%20Id (Utilities|String|StringToName (CosmeticLoadout|GetBannerColorIdfromLoadout _returnvalue)))
  (if (Utilities|String|EqualExactly(String) (Item|GetPersistentName (CosmeticLoadout|GetItemDefinitionfromLoadout (CosmeticLoadout|GetCosmeticLoadout) "(TagName=\"LoadoutSlot.Character\")")) (Variables|Default|GetCharacter-Name))
    (bind _asmaterial_instance_dynamic (Utilities|Casting|CastToMaterialInstanceDynamic (Rendering|Material|GetMaterial (Customization|GetSkeletalMeshforPartType (CustomCharacter|GetAssociatedPlayerPawn) "Body" :!class /Script/FortniteGame.FortPlayerPawn) 0)))
    (Utilities|FlowControl|Sequence
      (:then_0
        (bind _returnvalue_2 (Rendering|Material|GetVectorParameterValue _asmaterial_instance_dynamic "Crystal_Color" :!class /Script/Engine.MaterialInstanceDynamic))
        (for _array_element (CustomItemWrap|GetWrappedMeshComponents)
          (for _array_element_2 (Rendering|Material|GetMaterials _array_element)
            (bind _asmaterial_instance_dynamic_2 (Utilities|Casting|CastToMaterialInstanceDynamic _array_element_2))
            (Rendering|Material|SetVectorParameterValue _asmaterial_instance_dynamic_2 "TopS_01_ColorTint_CH_Blue  A=0=off" (* _returnvalue_2 "(R=1.322917,G=1.645964,B=2.000000,A=1.000000)") :!class /Script/Engine.MaterialInstanceDynamic)
            (Rendering|Material|SetVectorParameterValue _asmaterial_instance_dynamic_2 "TS_TOD_Alpha_Channel_Color_Night" (* _returnvalue_2 "(R=1.000000,G=1.000000,B=1.000000,A=1.000000)") :!class /Script/Engine.MaterialInstanceDynamic)
            (Rendering|Material|SetVectorParameterValue _asmaterial_instance_dynamic_2 "TS_TOD_Alpha_Channel_Color_Evening" (* (Math|Color|Desaturate(LinearColor) _returnvalue_2 0.35) "(R=0.500000,G=0.500000,B=0.500000,A=1.000000)") :!class /Script/Engine.MaterialInstanceDynamic)
            (Rendering|Material|SetVectorParameterValue _asmaterial_instance_dynamic_2 "TS_TOD_Alpha_Channel_Color_Morning" (* (Math|Color|Desaturate(LinearColor) _returnvalue_2 0.25) "(R=1.000000,G=1.000000,B=1.000000,A=1.000000)") :!class /Script/Engine.MaterialInstanceDynamic))))
      (:then_1
        (bind _returnvalue_3 (Rendering|Material|GetVectorParameterValue _asmaterial_instance_dynamic "Trim_Color" :!class /Script/Engine.MaterialInstanceDynamic))
        (for _array_element_3 (CustomItemWrap|GetWrappedMeshComponents)
          (for _array_element_4 (Rendering|Material|GetMaterials _array_element_3)
            (bind _asmaterial_instance_dynamic_3 (Utilities|Casting|CastToMaterialInstanceDynamic _array_element_4))
            (Rendering|Material|SetVectorParameterValue _asmaterial_instance_dynamic_3 "TopS_01_ColorTint_CH_Green A=0=off" _returnvalue_3 :!class /Script/Engine.MaterialInstanceDynamic))))
      (:then_2
        (bind _returnvalue_4 (Rendering|Material|GetVectorParameterValue _asmaterial_instance_dynamic "Secondary_Color" :!class /Script/Engine.MaterialInstanceDynamic))
        (for _array_element_5 (CustomItemWrap|GetWrappedMeshComponents)
          (for _array_element_6 (Rendering|Material|GetMaterials _array_element_5)
            (bind _asmaterial_instance_dynamic_4 (Utilities|Casting|CastToMaterialInstanceDynamic _array_element_6))
            (Rendering|Material|SetVectorParameterValue _asmaterial_instance_dynamic_4 "TopS_01_ColorTint_CH_Alpha   A=0=off" _returnvalue_4 :!class /Script/Engine.MaterialInstanceDynamic))))
      (:then_3
        (bind _returnvalue_5 (Rendering|Material|GetVectorParameterValue _asmaterial_instance_dynamic "Primary_Color" :!class /Script/Engine.MaterialInstanceDynamic))
        (for _array_element_7 (CustomItemWrap|GetWrappedMeshComponents)
          (for _array_element_8 (Rendering|Material|GetMaterials _array_element_7)
            (bind _asmaterial_instance_dynamic_5 (Utilities|Casting|CastToMaterialInstanceDynamic _array_element_8))
            (Rendering|Material|SetVectorParameterValue _asmaterial_instance_dynamic_5 "TopS_01_ColorTint_CH_Red  A=0=off" _returnvalue_5 :!class /Script/Engine.MaterialInstanceDynamic)))))))

(event EventOnCharacterCustomizationFinished
  (params (CharacterPartInfo :type Array%20of%20Applied%20Character%20Part%20Info%20Structures))
  (bind _returnvalue (CosmeticLoadout|GetCosmeticLoadout))
  (Variables|Default|SetBanner%20Icon%20Id (Utilities|String|StringToName (CosmeticLoadout|GetBannerIconIdfromLoadout _returnvalue)))
  (Variables|Default|SetBanner%20Color%20Id (Utilities|String|StringToName (CosmeticLoadout|GetBannerColorIdfromLoadout _returnvalue)))
  (if (Utilities|String|EqualExactly(String) (Item|GetPersistentName (CosmeticLoadout|GetItemDefinitionfromLoadout (CosmeticLoadout|GetCosmeticLoadout) "(TagName=\"LoadoutSlot.Character\")")) (Variables|Default|GetCharacter-Name))
    (bind _asmaterial_instance_dynamic (Utilities|Casting|CastToMaterialInstanceDynamic (Rendering|Material|GetMaterial (Customization|GetSkeletalMeshforPartType (CustomCharacter|GetAssociatedPlayerPawn) "Body" :!class /Script/FortniteGame.FortPlayerPawn) 0)))
    (Utilities|FlowControl|Sequence
      (:then_0
        (bind _returnvalue_2 (Rendering|Material|GetVectorParameterValue _asmaterial_instance_dynamic "Crystal_Color" :!class /Script/Engine.MaterialInstanceDynamic))
        (for _array_element (CustomItemWrap|GetWrappedMeshComponents)
          (for _array_element_2 (Rendering|Material|GetMaterials _array_element)
            (bind _asmaterial_instance_dynamic_2 (Utilities|Casting|CastToMaterialInstanceDynamic _array_element_2))
            (Rendering|Material|SetVectorParameterValue _asmaterial_instance_dynamic_2 "TopS_01_ColorTint_CH_Blue  A=0=off" (* _returnvalue_2 "(R=1.322917,G=1.645964,B=2.000000,A=1.000000)") :!class /Script/Engine.MaterialInstanceDynamic)
            (Rendering|Material|SetVectorParameterValue _asmaterial_instance_dynamic_2 "TS_TOD_Alpha_Channel_Color_Night" (* _returnvalue_2 "(R=1.000000,G=1.000000,B=1.000000,A=1.000000)") :!class /Script/Engine.MaterialInstanceDynamic)
            (Rendering|Material|SetVectorParameterValue _asmaterial_instance_dynamic_2 "TS_TOD_Alpha_Channel_Color_Evening" (* (Math|Color|Desaturate(LinearColor) _returnvalue_2 0.35) "(R=0.500000,G=0.500000,B=0.500000,A=1.000000)") :!class /Script/Engine.MaterialInstanceDynamic)
            (Rendering|Material|SetVectorParameterValue _asmaterial_instance_dynamic_2 "TS_TOD_Alpha_Channel_Color_Morning" (* (Math|Color|Desaturate(LinearColor) _returnvalue_2 0.25) "(R=1.000000,G=1.000000,B=1.000000,A=1.000000)") :!class /Script/Engine.MaterialInstanceDynamic))))
      (:then_1
        (bind _returnvalue_3 (Rendering|Material|GetVectorParameterValue _asmaterial_instance_dynamic "Trim_Color" :!class /Script/Engine.MaterialInstanceDynamic))
        (for _array_element_3 (CustomItemWrap|GetWrappedMeshComponents)
          (for _array_element_4 (Rendering|Material|GetMaterials _array_element_3)
            (bind _asmaterial_instance_dynamic_3 (Utilities|Casting|CastToMaterialInstanceDynamic _array_element_4))
            (Rendering|Material|SetVectorParameterValue _asmaterial_instance_dynamic_3 "TopS_01_ColorTint_CH_Green A=0=off" _returnvalue_3 :!class /Script/Engine.MaterialInstanceDynamic))))
      (:then_2
        (bind _returnvalue_4 (Rendering|Material|GetVectorParameterValue _asmaterial_instance_dynamic "Secondary_Color" :!class /Script/Engine.MaterialInstanceDynamic))
        (for _array_element_5 (CustomItemWrap|GetWrappedMeshComponents)
          (for _array_element_6 (Rendering|Material|GetMaterials _array_element_5)
            (bind _asmaterial_instance_dynamic_4 (Utilities|Casting|CastToMaterialInstanceDynamic _array_element_6))
            (Rendering|Material|SetVectorParameterValue _asmaterial_instance_dynamic_4 "TopS_01_ColorTint_CH_Alpha   A=0=off" _returnvalue_4 :!class /Script/Engine.MaterialInstanceDynamic))))
      (:then_3
        (bind _returnvalue_5 (Rendering|Material|GetVectorParameterValue _asmaterial_instance_dynamic "Primary_Color" :!class /Script/Engine.MaterialInstanceDynamic))
        (for _array_element_7 (CustomItemWrap|GetWrappedMeshComponents)
          (for _array_element_8 (Rendering|Material|GetMaterials _array_element_7)
            (bind _asmaterial_instance_dynamic_5 (Utilities|Casting|CastToMaterialInstanceDynamic _array_element_8))
            (Rendering|Material|SetVectorParameterValue _asmaterial_instance_dynamic_5 "TopS_01_ColorTint_CH_Red  A=0=off" _returnvalue_5 :!class /Script/Engine.MaterialInstanceDynamic)))))))

(event Custom|OnBannerFullyApplied
  (params (Icon :type Texture%202D%20Object%20Reference) (PrimaryColor :type Linear%20Color%20Structure)))

(fn setcolors
  (params (Color :type Linear%20Color%20Structure))
  (bind _returnvalue (Math|Conversions|ToVector(LinearColor) Color))
  (bind _returnvalue_2 (Math|Vector|RotateVector _returnvalue "25.000000,25.000000,25.000000"))
  (bind _returnvalue_3 (Math|Conversions|ToLinearColor(Vector) _returnvalue_2))
  (for _array_element (CustomItemWrap|GetWrappedMeshComponents)
    (for _array_element_2 (Rendering|Material|GetMaterials _array_element)
      (bind _asmaterial_instance_dynamic (Utilities|Casting|CastToMaterialInstanceDynamic _array_element_2))
      (Rendering|Material|SetVectorParameterValue _asmaterial_instance_dynamic "Emissive_A_ColorCW" Color :!class /Script/Engine.MaterialInstanceDynamic)
      (Rendering|Material|SetVectorParameterValue _asmaterial_instance_dynamic "Emissive_C_ColorCW" Color :!class /Script/Engine.MaterialInstanceDynamic)
      (Rendering|Material|SetVectorParameterValue _asmaterial_instance_dynamic "FDecal_BaseColor_Simple" Color :!class /Script/Engine.MaterialInstanceDynamic)
      (Rendering|Material|SetVectorParameterValue _asmaterial_instance_dynamic "Emissive_B_ColorCW" _returnvalue_3 :!class /Script/Engine.MaterialInstanceDynamic)
      (Rendering|Material|SetVectorParameterValue _asmaterial_instance_dynamic "Emissive_D_ColorCW" _returnvalue_3 :!class /Script/Engine.MaterialInstanceDynamic))))

(fn ConstructionScript)
