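; Reactive weapon wrap (SoapPocket) that shifts two emissive material offsets on wrapped meshes as the player deals damage and gets kills.
; Cosmetic stat watchers accumulate damage to a cap and map it to the offset.
; After an idle delay, a timeline drains it at a rate scaled by the total.
; Kills max it out, then hand the material timed reset parameters; in the front end, a timeline simulates both events.
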(event EventBeginPlay
  (if (CustomCosmeticModifier|IsinFrontEnd)
    (bind (_direction _damage_amount) (Timeline|Timeline_0 PlayFromStart)
      (:Damage%20delt
        (CallFunction|Stop)
        (Variables|Default|SetDamageAddedUp (Math|Float|Lerp 0.0 4e+02 _damage_amount))
        (Variables|Default|SetDamage%20Ref (Math|Float|Lerp 0.0 4e+02 _damage_amount))
        (Variables|Default|SetLastKillTime (Utilities|Time|GetGameTimeinSeconds))
        (CallFunction|WrapOffset)
        (Variables|Default|Setkill false)
        (CallFunction|SlowReset))
      (:Kill
        (Variables|Default|SetPause false)
        (Variables|Default|Setkill true)
        (Variables|Default|SetDamageAddedUp 4e+02)
        (CallFunction|Stop)
        (CallFunction|WrapOffset)
        (Utilities|FlowControl|Delay 0.1)
        (CallFunction|RestPossition -0.7 0.1 3.0 1.0)
        (Variables|Default|SetDamageAddedUp 0.0)
        (Utilities|FlowControl|Delay 2.0)
        (CallFunction|Stop)
        (Variables|Default|Setkill true)))))

(event EventOnWeaponFired)

(event EventOnPawnSet
  (bind _damageaddedup (Variables|Default|GetDamageAddedUp))
  (Utilities|FlowControl|Sequence
    (:then_0
      (Utilities|FlowControl|Delay 0.05)
      (Variables|Default|SetLastKillTime -999999999999.0))
    (:then_1
      (bind (_statname _statvalue) (Cosmetics|WatchCosmeticStat (CustomCharacter|GetAssociatedPlayerPawn) "(TagName=\"StatManager.DamageDealt\")" 0 false -1.0 -1.0 false)
        (:OnStatChanged
          (Utilities|FlowControl|Delay 0.05)
          (CallFunction|Stop)
          (bind _output_get (Variables|Default|SetDamage_Current (- _statvalue (Variables|Default|GetDamage_Last))))
          (Variables|Default|SetDamage_Last _statvalue)
          (bind _output_get_2 (Variables|Default|SetDamageAddedUp (Math|Float|Clamp(Float) (+ (Math|Conversions|ToFloat(Integer) _output_get) _damageaddedup) 0.0 4e+02)))
          (Variables|Default|SetDamage%20Ref (Math|Float|Clamp(Float) (+ (Math|Conversions|ToFloat(Integer) _output_get) _damageaddedup) 0.0 4e+02))
          (Variables|Default|SetLastKillTime (Utilities|Time|GetGameTimeinSeconds))
          (CallFunction|WrapOffset)
          (Variables|Default|Setkill false)
          (CallFunction|SlowReset))))
    (:then_2
      (Cosmetics|WatchCosmeticStat (CustomCharacter|GetAssociatedPlayerPawn) "(TagName=\"StatManager.AthenaKills\")" 0 false -1.0 -1.0 false
        (:OnStatChanged
          (Utilities|FlowControl|Delay 0.05)
          (Variables|Default|SetPause false)
          (Variables|Default|Setkill true)
          (Variables|Default|SetDamageAddedUp 4e+02)
          (CallFunction|Stop)
          (CallFunction|WrapOffset)
          (Utilities|FlowControl|Delay 0.1)
          (CallFunction|RestPossition -0.7 0.1 3.0 1.0)
          (Variables|Default|SetDamageAddedUp 0.0)
          (Utilities|FlowControl|Delay 2.0)
          (CallFunction|Stop)
          (Variables|Default|Setkill true))))))

(event Custom|wrap_offset
  (bind _returnvalue (CustomItemWrap|GetWrappedMeshComponents))
  (bind _damageaddedup (Variables|Default|GetDamageAddedUp))
  (bind _returnvalue_2 (/ _damageaddedup 4e+02))
  (bind _returnvalue_3 (Math|Float|Clamp(Float) _returnvalue_2))
  (bind _returnvalue_4 (Math|Float|Lerp -0.7 0.1 _returnvalue_3))
  (Utilities|FlowControl|Sequence
    (:then_0
      (Variables|Default|SetPause false)
      (Rendering|Material|SetScalarParameterValueonMaterials _returnvalue "TopS Emissive_01_Offset---End" _returnvalue_4)
      (Rendering|Material|SetScalarParameterValueonMaterials _returnvalue "TopS Emissive_02_Offset---End" _returnvalue_4)
      (Variables|Default|SetCurrent%20Position _returnvalue_4))))

(event Custom|Slow%20Reset
  (bind _returnvalue (CustomItemWrap|GetWrappedMeshComponents))
  (bind _current_position (Variables|Default|GetCurrent%20Position))
  (Utilities|FlowControl|Delay 0.5)
  (Variables|Default|SetPause true)
  (Utilities|FlowControl|Delay 4.0)
  (if (Variables|Default|GetPause)
    (if (not (Variables|Default|Getkill))
      (Components|Timeline|SetPlayRate (Variables|CIWM_SoapPocket|GetTimeline) (/ 1.0 (Math|Float|Lerp 0.5 35.0 (Math|Float|Clamp(Float) (/ (Variables|Default|GetDamageAddedUp) 4e+02)))))
      (bind (_direction _timeline) (Timeline|Timeline PlayFromStart)
        (:Update
          (Rendering|Material|SetScalarParameterValueonMaterials _returnvalue "TopS Emissive_01_Offset---End" (Math|Float|Lerp -0.7 _current_position _timeline))
          (Rendering|Material|SetScalarParameterValueonMaterials _returnvalue "TopS Emissive_02_Offset---End" (Math|Float|Lerp -0.7 _current_position _timeline))
          (Variables|Default|SetDamageAddedUp (Math|Float|Lerp 0.0 (Variables|Default|GetDamage%20Ref) _timeline)))))))

(event Custom|Stop
  (bind _returnvalue (CustomItemWrap|GetWrappedMeshComponents))
  (bind _current_position (Variables|Default|GetCurrent%20Position))
  (Timeline|Timeline Stop))

(fn RestPossition
  (params (End :type Float%20%28single-precision%29) (Start :type Float%20%28single-precision%29) (Time :type Float%20%28single-precision%29) (delay :type Float%20%28double-precision%29))
  (bind _returnvalue (CustomItemWrap|GetWrappedMeshComponents))
  (bind _returnvalue_2 (Utilities|Time|GetGameTimeinSeconds))
  (bind _returnvalue_3 (CustomItemWrap|GetWrappedMeshComponents))
  (bind _returnvalue_4 (+ _returnvalue_2 delay))
  (Rendering|Material|SetScalarParameterValueonMaterials _returnvalue_3 "TopS Emissive_02_Offset---End" End)
  (Rendering|Material|SetScalarParameterValueonMaterials _returnvalue_3 "TopS Emissive_02_Offset---Start" Start)
  (Rendering|Material|SetScalarParameterValueonMaterials _returnvalue_3 "TopS Emissive_02_Offset----TBP" _returnvalue_4)
  (Rendering|Material|SetScalarParameterValueonMaterials _returnvalue_3 "TopS Emissive_02_Offset---Duration" Time)
  (Rendering|Material|SetScalarParameterValueonMaterials _returnvalue "TopS Emissive_01_Offset---End" End)
  (Rendering|Material|SetScalarParameterValueonMaterials _returnvalue "TopS Emissive_01_Offset---Start" Start)
  (Rendering|Material|SetScalarParameterValueonMaterials _returnvalue "TopS Emissive_01_Offset----TBP" _returnvalue_4)
  (Rendering|Material|SetScalarParameterValueonMaterials _returnvalue "TopS Emissive_01_Offset---Duration" Time))

(fn ConstructionScript)
