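; SoloSnooze item wrap Blueprint that matches wrapped items to the styling of one named character skin.
; On pawn set, wrap target set and customization finished, identical logic compares the loadout character's name to Character-Name.
; On a match, it copies PrimaryColor, SecondaryColor and SprayTexture from the Body mesh's first material onto wrapped mesh materials.
; Each event also debug-prints the character name; the banner event is empty.
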(event EventOnPawnSet
  (bind _returnvalue (CosmeticLoadout|GetCosmeticLoadout))
  (bind _returnvalue_2 (CosmeticLoadout|GetItemDefinitionfromLoadout _returnvalue "(TagName=\"LoadoutSlot.Character\")"))
  (bind _returnvalue_3 (Item|GetPersistentName _returnvalue_2))
  (bind _character_material (Variables|Default|GetCharacter%20Material))
  (Development|PrintText (Utilities|Text|ToText(String) _returnvalue_3) true true "(R=0.610496,G=0.026241,B=0.027321,A=1.000000)" 5.0)
  (if (Utilities|String|Equal,CaseInsensitive(String) _returnvalue_3 (Variables|Default|GetCharacter-Name))
    (bind _asmaterial_instance_dynamic (Utilities|Casting|CastToMaterialInstanceDynamic (Rendering|Material|GetMaterial (Customization|GetSkeletalMeshforPartType (CustomCharacter|GetAssociatedPlayerPawn) "Body" :!class /Script/FortniteGame.FortPlayerPawn) 0)))
    (Variables|Default|SetCharacter%20Material _asmaterial_instance_dynamic)
    (Utilities|FlowControl|Sequence
      (:then_0
        (bind _returnvalue_4 (Rendering|Material|GetVectorParameterValue _character_material "PrimaryColor" :!class /Script/Engine.MaterialInstanceDynamic))
        (bind _returnvalue_5 (Rendering|Material|GetVectorParameterValue _character_material "SecondaryColor" :!class /Script/Engine.MaterialInstanceDynamic))
        (for _array_element (CustomItemWrap|GetWrappedMeshComponents)
          (for _array_element_2 (Rendering|Material|GetMaterials _array_element)
            (bind _asmaterial_instance_dynamic_2 (Utilities|Casting|CastToMaterialInstanceDynamic _array_element_2))
            (Rendering|Material|SetVectorParameterValue _asmaterial_instance_dynamic_2 "TopS_01_ColorTint_CH_Blue  A=0=off" (* "(R=1.000000,G=1.000000,B=1.000000,A=1.000000)" _returnvalue_4) :!class /Script/Engine.MaterialInstanceDynamic)
            (Rendering|Material|SetVectorParameterValue _asmaterial_instance_dynamic_2 "FDecalV2_Base_RGBA_Color=Green" (* "(R=1.000000,G=1.000000,B=1.000000,A=1.000000)" _returnvalue_5) :!class /Script/Engine.MaterialInstanceDynamic)
            (Rendering|Material|SetVectorParameterValue _asmaterial_instance_dynamic_2 "AccentV2_BaseColor_" (* "(R=1.000000,G=1.000000,B=1.000000,A=1.000000)" _returnvalue_5) :!class /Script/Engine.MaterialInstanceDynamic))))
      (:then_1
        (bind _returnvalue_6 (Rendering|Material|GetTextureParameterValue (Variables|Default|GetCharacter%20Material) "SprayTexture" :!class /Script/Engine.MaterialInstanceDynamic))
        (for _array_element_3 (CustomItemWrap|GetWrappedMeshComponents)
          (for _array_element_4 (Rendering|Material|GetMaterials _array_element_3)
            (bind _asmaterial_instance_dynamic_3 (Utilities|Casting|CastToMaterialInstanceDynamic _array_element_4))
            (Rendering|Material|SetTextureParameterValue _asmaterial_instance_dynamic_3 "EDRec_Decal_BC__Tex" _returnvalue_6)))))))

(event EventOnSetWrapTarget
  (bind _returnvalue (CosmeticLoadout|GetCosmeticLoadout))
  (bind _returnvalue_2 (CosmeticLoadout|GetItemDefinitionfromLoadout _returnvalue "(TagName=\"LoadoutSlot.Character\")"))
  (bind _returnvalue_3 (Item|GetPersistentName _returnvalue_2))
  (bind _character_material (Variables|Default|GetCharacter%20Material))
  (Development|PrintText (Utilities|Text|ToText(String) _returnvalue_3) true true "(R=0.610496,G=0.026241,B=0.027321,A=1.000000)" 5.0)
  (if (Utilities|String|Equal,CaseInsensitive(String) _returnvalue_3 (Variables|Default|GetCharacter-Name))
    (bind _asmaterial_instance_dynamic (Utilities|Casting|CastToMaterialInstanceDynamic (Rendering|Material|GetMaterial (Customization|GetSkeletalMeshforPartType (CustomCharacter|GetAssociatedPlayerPawn) "Body" :!class /Script/FortniteGame.FortPlayerPawn) 0)))
    (Variables|Default|SetCharacter%20Material _asmaterial_instance_dynamic)
    (Utilities|FlowControl|Sequence
      (:then_0
        (bind _returnvalue_4 (Rendering|Material|GetVectorParameterValue _character_material "PrimaryColor" :!class /Script/Engine.MaterialInstanceDynamic))
        (bind _returnvalue_5 (Rendering|Material|GetVectorParameterValue _character_material "SecondaryColor" :!class /Script/Engine.MaterialInstanceDynamic))
        (for _array_element (CustomItemWrap|GetWrappedMeshComponents)
          (for _array_element_2 (Rendering|Material|GetMaterials _array_element)
            (bind _asmaterial_instance_dynamic_2 (Utilities|Casting|CastToMaterialInstanceDynamic _array_element_2))
            (Rendering|Material|SetVectorParameterValue _asmaterial_instance_dynamic_2 "TopS_01_ColorTint_CH_Blue  A=0=off" (* "(R=1.000000,G=1.000000,B=1.000000,A=1.000000)" _returnvalue_4) :!class /Script/Engine.MaterialInstanceDynamic)
            (Rendering|Material|SetVectorParameterValue _asmaterial_instance_dynamic_2 "FDecalV2_Base_RGBA_Color=Green" (* "(R=1.000000,G=1.000000,B=1.000000,A=1.000000)" _returnvalue_5) :!class /Script/Engine.MaterialInstanceDynamic)
            (Rendering|Material|SetVectorParameterValue _asmaterial_instance_dynamic_2 "AccentV2_BaseColor_" (* "(R=1.000000,G=1.000000,B=1.000000,A=1.000000)" _returnvalue_5) :!class /Script/Engine.MaterialInstanceDynamic))))
      (:then_1
        (bind _returnvalue_6 (Rendering|Material|GetTextureParameterValue (Variables|Default|GetCharacter%20Material) "SprayTexture" :!class /Script/Engine.MaterialInstanceDynamic))
        (for _array_element_3 (CustomItemWrap|GetWrappedMeshComponents)
          (for _array_element_4 (Rendering|Material|GetMaterials _array_element_3)
            (bind _asmaterial_instance_dynamic_3 (Utilities|Casting|CastToMaterialInstanceDynamic _array_element_4))
            (Rendering|Material|SetTextureParameterValue _asmaterial_instance_dynamic_3 "EDRec_Decal_BC__Tex" _returnvalue_6)))))))

(event EventOnCharacterCustomizationFinished
  (params (CharacterPartInfo :type Array%20of%20Applied%20Character%20Part%20Info%20Structures))
  (bind _returnvalue (CosmeticLoadout|GetCosmeticLoadout))
  (bind _returnvalue_2 (CosmeticLoadout|GetItemDefinitionfromLoadout _returnvalue "(TagName=\"LoadoutSlot.Character\")"))
  (bind _returnvalue_3 (Item|GetPersistentName _returnvalue_2))
  (bind _character_material (Variables|Default|GetCharacter%20Material))
  (Development|PrintText (Utilities|Text|ToText(String) _returnvalue_3) true true "(R=0.610496,G=0.026241,B=0.027321,A=1.000000)" 5.0)
  (if (Utilities|String|Equal,CaseInsensitive(String) _returnvalue_3 (Variables|Default|GetCharacter-Name))
    (bind _asmaterial_instance_dynamic (Utilities|Casting|CastToMaterialInstanceDynamic (Rendering|Material|GetMaterial (Customization|GetSkeletalMeshforPartType (CustomCharacter|GetAssociatedPlayerPawn) "Body" :!class /Script/FortniteGame.FortPlayerPawn) 0)))
    (Variables|Default|SetCharacter%20Material _asmaterial_instance_dynamic)
    (Utilities|FlowControl|Sequence
      (:then_0
        (bind _returnvalue_4 (Rendering|Material|GetVectorParameterValue _character_material "PrimaryColor" :!class /Script/Engine.MaterialInstanceDynamic))
        (bind _returnvalue_5 (Rendering|Material|GetVectorParameterValue _character_material "SecondaryColor" :!class /Script/Engine.MaterialInstanceDynamic))
        (for _array_element (CustomItemWrap|GetWrappedMeshComponents)
          (for _array_element_2 (Rendering|Material|GetMaterials _array_element)
            (bind _asmaterial_instance_dynamic_2 (Utilities|Casting|CastToMaterialInstanceDynamic _array_element_2))
            (Rendering|Material|SetVectorParameterValue _asmaterial_instance_dynamic_2 "TopS_01_ColorTint_CH_Blue  A=0=off" (* "(R=1.000000,G=1.000000,B=1.000000,A=1.000000)" _returnvalue_4) :!class /Script/Engine.MaterialInstanceDynamic)
            (Rendering|Material|SetVectorParameterValue _asmaterial_instance_dynamic_2 "FDecalV2_Base_RGBA_Color=Green" (* "(R=1.000000,G=1.000000,B=1.000000,A=1.000000)" _returnvalue_5) :!class /Script/Engine.MaterialInstanceDynamic)
            (Rendering|Material|SetVectorParameterValue _asmaterial_instance_dynamic_2 "AccentV2_BaseColor_" (* "(R=1.000000,G=1.000000,B=1.000000,A=1.000000)" _returnvalue_5) :!class /Script/Engine.MaterialInstanceDynamic))))
      (:then_1
        (bind _returnvalue_6 (Rendering|Material|GetTextureParameterValue (Variables|Default|GetCharacter%20Material) "SprayTexture" :!class /Script/Engine.MaterialInstanceDynamic))
        (for _array_element_3 (CustomItemWrap|GetWrappedMeshComponents)
          (for _array_element_4 (Rendering|Material|GetMaterials _array_element_3)
            (bind _asmaterial_instance_dynamic_3 (Utilities|Casting|CastToMaterialInstanceDynamic _array_element_4))
            (Rendering|Material|SetTextureParameterValue _asmaterial_instance_dynamic_3 "EDRec_Decal_BC__Tex" _returnvalue_6)))))))

(event Custom|OnBannerFullyApplied
  (params (Icon :type Texture%202D%20Object%20Reference) (PrimaryColor :type Linear%20Color%20Structure)))

(fn ConstructionScript)
