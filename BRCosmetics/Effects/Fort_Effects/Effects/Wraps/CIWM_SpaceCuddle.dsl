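; Custom item wrap Blueprint for SpaceCuddle that drives a staged material effect on wrapped weapon meshes as the weapon fires.
; Each shot, after a tiny delay, raises a capped step count, first subtracting decay computed lazily from time elapsed past the previous reset deadline.
; It pushes the stage plus shot and reset timestamps to material scalars.
; An accumulator counts shots at maximum stage, resetting at four or below maximum; weapon changes zero the stage.
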(event EventBeginPlay
  (|CallParentFunction))

(event EventOnWeaponChanged
  (params (NewWeapon :type Fort%20Weapon%20Object%20Reference) (OldWeapon :type Fort%20Weapon%20Object%20Reference))
  (Variables|Default|SetStepCount 0)
  (Rendering|Material|SetScalarParameterValueonMaterials (CustomItemWrap|GetWrappedMeshComponents) "SCud_Progression_Stage" 0.0))

(event EventOnWeaponFired
  (bind _returnvalue (CustomItemWrap|GetWrappedMeshComponents))
  (bind _returnvalue_2 (Utilities|Time|GetGameTimeinSeconds))
  (bind _resetdelay (Variables|Default|GetResetDelay))
  (bind _returnvalue_3 (+ _returnvalue_2 _resetdelay))
  (bind _maxsteps (Variables|Default|GetMaxSteps))
  (Utilities|FlowControl|Delay 0.01)
  (bind _output_get (Variables|Default|SetStepCount (Math|Integer|Min(Integer) (+ (- (Variables|Default|GetStepCount) (Math|Float|Round (Math|Float|Clamp(Float) (/ (- (Utilities|Time|GetGameTimeinSeconds) (Variables|Default|GetWaitTime)) 0.25) 0.0 (Math|Conversions|ToFloat(Integer) (Variables|Default|GetStepCount))))) 1) _maxsteps)))
  (Rendering|Material|SetScalarParameterValueonMaterials _returnvalue "SCud_Progression_Stage" (Math|Conversions|ToFloat(Integer) _output_get))
  (Utilities|FlowControl|Sequence
    (:then_0
      (Rendering|Material|SetScalarParameterValueonMaterials _returnvalue "SCud_Time_BP" _returnvalue_2)
      (Rendering|Material|SetScalarParameterValueonMaterials _returnvalue "SCud_Time_BP_Reset" _returnvalue_3)
      (Variables|Default|SetWaitTime _returnvalue_3))
    (:then_1
      (if (>= (Math|Conversions|ToFloat(Integer) _output_get) (Math|Conversions|ToFloat(Integer) _maxsteps))
        (bind _output_get_2 (Variables|Default|SetAcumulator (+ (Variables|Default|GetAcumulator) 1)))
        (Rendering|Material|SetScalarParameterValueonMaterials _returnvalue "SCud_Progression_Stage_Acumulator" (Math|Conversions|ToFloat(Integer) _output_get_2))
        (if (>= (Math|Conversions|ToFloat(Integer) _output_get_2) 4.0)
          (Variables|Default|SetAcumulator (Variables|Default|GetAcumulator_0)))
        (else
          (bind _output_get_3 (Variables|Default|SetAcumulator (Variables|Default|GetAcumulator_0)))
          (Rendering|Material|SetScalarParameterValueonMaterials _returnvalue "SCud_Progression_Stage_Acumulator" (Math|Conversions|ToFloat(Integer) _output_get_3)))))))

(fn ConstructionScript)
