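; Reactive weapon wrap logic for Space Plunge that animates the wrapped meshes' materials on weapon fire and kills.
; It writes current game time into scalar material parameters as trigger timestamps, with speed parameters updated behind a one-second Pause gate on firing.
; A kill-stat watcher, registered on BeginPlay and pawn set, stamps the Emissive_03 brightness.
; In the front end, every third shot fires that kill effect through FrontendRay instead of the normal firing updates.
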(event EventOnWeaponFired
  (bind _returnvalue (Utilities|Time|GetGameTimeinSeconds))
  (bind _returnvalue_2 (CustomItemWrap|GetWrappedMeshComponents))
  (if (CustomCosmeticModifier|IsinFrontEnd)
    (bind _result (Math|Integer|IncrementInt (Variables|Default|GetIterationCount))
      (:%20%20
        (if (== (Math|Integer|%25%28Integer%29 _result 3) 0)
          (CallFunction|FrontendRay))))
    (else
      (Utilities|FlowControl|Sequence
        (:then_0
          ;[1 1 1] TopsSide V2
          (Rendering|Material|SetScalarParameterValueonMaterials _returnvalue_2 "TopS Emissive_02-Brightness-TBP" _returnvalue)
          (Rendering|Material|SetScalarParameterValueonMaterials _returnvalue_2 "TopS Emissive_02_Offset----TBP" _returnvalue)
          (Rendering|Material|SetScalarParameterValueonMaterials _returnvalue_2 "TopS_01_Blend_Opacity-TBP" _returnvalue)
          (Rendering|Material|SetScalarParameterValueonMaterials _returnvalue_2 "TopS_01_Blend_Offset-TBP" _returnvalue)
          (Rendering|Material|SetScalarParameterValueonMaterials _returnvalue_2 "TimeFrom-BP----------------" _returnvalue))
        (:then_1
          (Rendering|Material|SetScalarParameterValueonMaterials _returnvalue_2 "TopS Emissive_01-Brightness-TBP" _returnvalue)
          (Rendering|Material|SetScalarParameterValueonMaterials _returnvalue_2 "TopS Emissive_01_Offset----TBP" _returnvalue)
          (Rendering|Material|SetScalarParameterValueonMaterials _returnvalue_2 "TopS Emissive_03_Offset----TBP" _returnvalue)
          (if (not (Variables|Default|GetPause))
            (Variables|Default|SetPause true)
            (Rendering|Material|SetScalarParameterValueonMaterials _returnvalue_2 "TopS Emissive_01_Speed-TBP" _returnvalue)
            (Rendering|Material|SetScalarParameterValueonMaterials _returnvalue_2 "TopS Emissive_02_Speed-TBP" _returnvalue)
            (Rendering|Material|SetScalarParameterValueonMaterials _returnvalue_2 "TopS Emissive_03_Speed-TBP" _returnvalue)
            (Utilities|FlowControl|Delay 1.0)
            (Variables|Default|SetPause false)))))))

(event EventBeginPlay
  (Cosmetics|WatchCosmeticStat (CustomCharacter|GetAssociatedPlayerPawn) "(TagName=\"StatManager.AthenaKills\")" 0 false
    (:OnStatChanged
      (Rendering|Material|SetScalarParameterValueonMaterials (CustomItemWrap|GetWrappedMeshComponents) "TopS Emissive_03-Brightness-TBP" (Utilities|Time|GetGameTimeinSeconds)))))

(event Custom|FrontendRay
  (Rendering|Material|SetScalarParameterValueonMaterials (CustomItemWrap|GetWrappedMeshComponents) "TopS Emissive_03-Brightness-TBP" (Utilities|Time|GetGameTimeinSeconds)))

(event EventOnPawnSet
  (Cosmetics|WatchCosmeticStat (CustomCharacter|GetAssociatedPlayerPawn) "(TagName=\"StatManager.AthenaKills\")" 0 false
    (:OnStatChanged
      (Rendering|Material|SetScalarParameterValueonMaterials (CustomItemWrap|GetWrappedMeshComponents) "TopS Emissive_03-Brightness-TBP" (Utilities|Time|GetGameTimeinSeconds)))))

(fn ConstructionScript)
