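; Item wrap modifier that changes the wrapped weapon's material as the player's kill count rises through five clamped stages.
; A watched AthenaKills stat sets the count and kill time, while a front-end timeline steps through the stages.
; PushMaterialParameters derives start/end blend values via CountPerctange, writing them and the kill timestamp to materials in the stage-3 branch.
; Weapon changes backdate the timestamp before re-pushing after a short delay.
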(event EventOnWeaponFired)

(event EventBeginPlay
  (if (CustomCosmeticModifier|IsinFrontEnd)
    (Timeline|Timeline PlayFromStart
      (:Stage1
        (Variables|Default|SetKill_Count 0))
      (:Stage2
        (Variables|Default|SetKill_Count 1))
      (:Stage3
        (Variables|Default|SetKill_Count 2))
      (:Stage4
        (Variables|Default|SetKill_Count 3))
      (:stage5
        (Variables|Default|SetKill_Count 4)))
    (Variables|Default|SetLastKillTime (Utilities|Time|GetGameTimeinSeconds))
    (CallFunction|PushMaterialParameters)))

(event EventOnPawnSet
  (Utilities|FlowControl|Sequence
    (:then_1
      (bind (_statname _statvalue) (Cosmetics|WatchCosmeticStat (CustomCharacter|GetAssociatedPlayerPawn) "(TagName=\"StatManager.AthenaKills\")")
        (:OnStatChanged
          (bind _output_get (Variables|Default|SetKill_Count _statvalue))
          (Variables|Default|SetLastKillTime (Utilities|Time|GetGameTimeinSeconds))
          (CallFunction|PushMaterialParameters))))))

(event EventOnPostResIn
  (CallFunction|PushMaterialParameters))

(event EventOnNewWeapons
  (params (WeaponsAsItems :type Array%20of%20Fort%20Item%20Object%20References))
  (Utilities|FlowControl|Delay 0.05)
  ;[1 1 1] Updated kill count on new weapon
  (Variables|Default|SetLastKillTime -999999999999.0)
  (CallFunction|PushMaterialParameters))

(event EventOnNewPlayerWeaponType
  (params (NewType :type EFortItemType%20Enum :default "WorldItem") (OldType :type EFortItemType%20Enum :default "WorldItem"))
  (Utilities|FlowControl|Delay 0.05)
  ;[1 1 1] Updated kill count on new weapon
  (Variables|Default|SetLastKillTime -999999999999.0)
  (CallFunction|PushMaterialParameters))

(event EventOnWeaponChanged
  (params (NewWeapon :type Fort%20Weapon%20Object%20Reference) (OldWeapon :type Fort%20Weapon%20Object%20Reference))
  (Utilities|FlowControl|Delay 0.05)
  ;[1 1 1] Updated kill count on new weapon
  (Variables|Default|SetLastKillTime -999999999999.0)
  (CallFunction|PushMaterialParameters))

(event EventApplyMaterialParameters
  (CallFunction|PushMaterialParameters))

(event Custom|Push%20Material%20Parameters
  (bind _returnvalue (CustomItemWrap|GetWrappedMeshComponents))
  (bind _lastkilltime (Variables|Default|GetLastKillTime))
  (switch int (Math|Integer|Clamp(Integer) (Variables|Default|GetKill_Count) 0 4)
    (:0
      (CallFunction|CountPerctange 0.0 0.0 -1.0 -1.0 4.0 4.0))
    (:1
      (CallFunction|CountPerctange 1.0 0.0 -1.0 -1.0 4.0 4.0))
    (:2
      (CallFunction|CountPerctange 2.0 1.0 -1.0 -1.0 4.0 4.0))
    (:3
      (CallFunction|CountPerctange 3.0 2.0 -3.0 -1.0 3.0 4.0)
      (Rendering|Material|SetScalarParameterValueonMaterials (CustomItemWrap|GetWrappedMeshComponents) "TopS_01_Blend_Offset---Duration" 8.0)
      (Utilities|FlowControl|Sequence
        (:then_0
          (Rendering|Material|SetScalarParameterValueonMaterials _returnvalue "TopS_01_Blend_Offset---End" (Variables|Default|GetEnd))
          (Rendering|Material|SetScalarParameterValueonMaterials _returnvalue "TopS_01_Blend_Offset---Start" (Variables|Default|GetStart))
          (Rendering|Material|SetScalarParameterValueonMaterials _returnvalue "TopS_01_Blend_Offset-TBP" _lastkilltime)
          (Rendering|Material|SetScalarParameterValueonMaterials _returnvalue "TopS Emissive_01-Brightness-TBP" _lastkilltime))))
    (:4
      (CallFunction|CountPerctange 3.0 3.0 -3.0 -3.0 3.0 3.0)))
  (Rendering|Material|SetScalarParameterValueonMaterials (CustomItemWrap|GetWrappedMeshComponents) "TopS_01_Blend_Offset---Duration" 3.0))

(fn CountPerctange
  (params (End%20count :type Float%20%28double-precision%29) (StartCount :type Float%20%28double-precision%29) (End-offset :type Float%20%28double-precision%29) (Start_End :type Float%20%28double-precision%29) (KillAmount-End :type Float%20%28double-precision%29) (KillAmount-Start :type Float%20%28double-precision%29))
  (Variables|Default|SetBackPack%20offset 0.0)
  (Variables|Default|SetCharacter%20offset 0.0)
  (Variables|Default|SetEnd (Math|Float|Lerp 0.35 End-offset (/ End%20count KillAmount-End)))
  (Variables|Default|SetStart (Math|Float|Lerp 0.35 Start_End (/ StartCount KillAmount-Start))))

(fn ConstructionScript)
