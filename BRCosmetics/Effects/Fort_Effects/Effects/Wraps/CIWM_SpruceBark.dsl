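; Weapon wrap cosmetic logic that progressively reveals lettered emissive decals as the player's kill count rises.
; A kill-stat watcher pulses emissive brightness and, unless paused, increments a capped counter, driving the Anim timeline into material parameters named from the Decals array.
; In the front end, every fourth shot simulates a kill, and a Reverse event resets the counter after a delay.
; Weapon fire also updates decal emissive timing; the timer-called Restart event is empty.
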(event EventBeginPlay
  (if (CustomCosmeticModifier|IsinFrontEnd)
    (Timeline|Timeline PlayFromStart)))

(event EventOnPawnSet
  (bind _returnvalue (CustomItemWrap|GetWrappedMeshComponents))
  (bind _kill_count (Variables|Default|GetKill_Count))
  (bind _returnvalue_2 (- _kill_count 1))
  (bind _decals (Variables|Default|GetDecals))
  (bind _output (Utilities|Array|Get(acopy) _decals _returnvalue_2))
  (Cosmetics|WatchCosmeticStat (CustomCharacter|GetAssociatedPlayerPawn) "(TagName=\"StatManager.AthenaKills\")" 0 false
    (:OnStatChanged
      (Rendering|Material|SetScalarParameterValueonMaterials (CustomItemWrap|GetWrappedMeshComponents) "TopS Emissive_01-Brightness-TBP" (Utilities|Time|GetGameTimeinSeconds))
      (if (Variables|Default|GetPause)
        _
        (elif (<= (Variables|Default|GetKill_Count) 4)
          (Math|Integer|IncrementInt (Variables|Default|GetKill_Count)
            (:%20%20
              (Variables|Default|SetPause true)
              (bind (_direction _newtrack_0) (Timeline|Anim PlayFromStart)
                (:Update
                  (Rendering|Material|SetScalarParameterValueonMaterials _returnvalue (Utilities|String|StringToName (Utilities|String|ToString(Text) (Utilities|Text|FormatText "NSLOCTEXT(\"[459DF3BDC3E22A79A0447588AA35669A]\", \"C3FEF5F845C97F39F87552A89009CF42\", \"EDRec_Emissive_{Letter}_Brightness_Time__End\")" _output))) _newtrack_0)
                  (Rendering|Material|SetScalarParameterValueonMaterials _returnvalue (Utilities|String|StringToName (Utilities|String|ToString(Text) (Utilities|Text|FormatText "NSLOCTEXT(\"[459DF3BDC3E22A79A0447588AA35669A]\", \"F192A12947D7C018621EF986D194D6A9\", \"EDRec_Decal{Letter}_Opacity_End\")" _output))) _newtrack_0))
                (:Finished
                  (Variables|Default|SetPause false)
                  (if (not (CustomCosmeticModifier|IsinFrontEnd))
                    (Utilities|Time|SetTimerbyFunctionName 0 "Restart" 5.0)))))))))))

(event EventOnWeaponFired
  (bind _returnvalue (Utilities|Time|GetGameTimeinSeconds))
  (bind _returnvalue_2 (CustomItemWrap|GetWrappedMeshComponents))
  (bind _returnvalue_3 (CustomItemWrap|GetWrappedMeshComponents))
  (bind _kill_count (Variables|Default|GetKill_Count))
  (bind _returnvalue_4 (- _kill_count 1))
  (bind _decals (Variables|Default|GetDecals))
  (bind _output (Utilities|Array|Get(acopy) _decals _returnvalue_4))
  (if (CustomCosmeticModifier|IsinFrontEnd)
    (Utilities|FlowControl|Sequence
      (:then_0
        _)
      (:then_1
        (bind _output_get (Variables|Default|SetShotsFired (+ (Variables|Default|GetShotsFired) 1)))
        (if (== (Math|Integer|%25%28Integer%29 _output_get 4) 0)
          (Rendering|Material|SetScalarParameterValueonMaterials _returnvalue_2 "TopS Emissive_01-Brightness-TBP" (Utilities|Time|GetGameTimeinSeconds))
          (if (Variables|Default|GetPause)
            _
            (elif (<= (Variables|Default|GetKill_Count) 4)
              (Math|Integer|IncrementInt (Variables|Default|GetKill_Count)
                (:%20%20
                  (Variables|Default|SetPause true)
                  (Timeline|Anim PlayFromStart)))))))))
  (Rendering|Material|SetScalarParameterValueonMaterials _returnvalue_2 "FDecalV2_Emissive_Brightness_Time_BP" _returnvalue)
  (Rendering|Material|SetScalarParameterValueonMaterials _returnvalue_2 "FDecalV2_Emissive_Brightness_Time__End" _returnvalue)
  (if (not (Variables|Default|GetKilled))
    (Rendering|Material|SetScalarParameterValueonMaterials _returnvalue_2 "FDecalV2_Emissive_Brightness_Time__End" 0.0)))

(event Custom|Restart)

(event Custom|Reverse'
  (if (CustomCosmeticModifier|IsinFrontEnd)
    (if (Variables|Default|GetPause)
      _
      (elif (>= (Variables|Default|GetKill_Count) 0)
        (Variables|Default|SetPause true)
        (Variables|Default|SetKill_Count 4)
        (Utilities|FlowControl|Sequence
          (:then_0
            (Components|Timeline|SetPlayRate (Variables|CIWM_SpruceBark|GetAnim) 0.5))
          (:then_1
            (Utilities|FlowControl|RetriggerableDelay 3.0)
            (Variables|Default|SetKill_Count 0)
            (Variables|Default|SetPause false)))))))

(fn ConstructionScript)
