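; SquidGlis weapon wrap modifier that plays a timed glow-style material animation when the player's kill stat changes, or on the fifth shot in the front end.
; It stamps game time and start/end scalars onto wrapped mesh materials, so shaders animate a top-layer blend and decal fade.
; Two delayed Sequence branches stage the fade-in and fade-out in parallel.
; A Pause flag blocks retriggers until the effect finishes and the counter resets.
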(event EventOnWeaponFired
  (bind _returnvalue (CustomItemWrap|GetWrappedMeshComponents))
  (bind _returnvalue_2 (CustomItemWrap|GetWrappedMeshComponents))
  (bind _returnvalue_3 (CustomItemWrap|GetWrappedMeshComponents))
  (bind _returnvalue_4 (CustomItemWrap|GetWrappedMeshComponents))
  (if (CustomCosmeticModifier|IsinFrontEnd)
    (if (not (Variables|Default|GetPause))
      (bind _output_get (Variables|Default|SetKill_Count (+ (Variables|Default|GetKill_Count) 1)))
      (if (>= _output_get 5)
        (Variables|Default|SetPause true)
        (Rendering|Material|SetScalarParameterValueonMaterials _returnvalue_2 "TopS_01_Blend_Opacity---End" 1.0)
        (Rendering|Material|SetScalarParameterValueonMaterials _returnvalue_2 "FDecalV2_Opacity_Time__End" 0.0)
        (Rendering|Material|SetScalarParameterValueonMaterials _returnvalue_2 "TopS_01_Blend_Offset-TBP" (Utilities|Time|GetGameTimeinSeconds))
        (Rendering|Material|SetScalarParameterValueonMaterials _returnvalue_2 "TopS_01_Blend_Offset---Start" 0.5)
        (Rendering|Material|SetScalarParameterValueonMaterials _returnvalue_2 "TopS_01_Blend_Offset---End" -1.0)
        (Utilities|FlowControl|Sequence
          (:then_0
            (Utilities|FlowControl|Delay 3.0)
            (Rendering|Material|SetScalarParameterValueonMaterials _returnvalue_3 "TopS_01_Blend_Opacity-TBP" (Utilities|Time|GetGameTimeinSeconds))
            (Rendering|Material|SetScalarParameterValueonMaterials _returnvalue_3 "TopS_01_Blend_Opacity---Start" 1.0)
            (Rendering|Material|SetScalarParameterValueonMaterials _returnvalue_3 "TopS_01_Blend_Opacity---End" 0.0)
            (Utilities|FlowControl|Delay 2.0)
            (Variables|Default|SetPause false)
            (Variables|Default|SetKill_Count 0))
          (:then_1
            (Rendering|Material|SetScalarParameterValueonMaterials _returnvalue_4 "FDecalaV2_Opacity_Time_BP_Z" (Utilities|Time|GetGameTimeinSeconds))
            (Rendering|Material|SetScalarParameterValueonMaterials _returnvalue_4 "FDecalV2_Opacity_Time__Start" 0.0)
            (Rendering|Material|SetScalarParameterValueonMaterials _returnvalue_4 "FDecalV2_Opacity_Time__End" 1.0)
            (Rendering|Material|SetScalarParameterValueonMaterials _returnvalue_4 "FDecalaV2_Opacity_Duration" 1.0)
            (Utilities|FlowControl|Delay 1.0)
            (Rendering|Material|SetScalarParameterValueonMaterials _returnvalue "FDecalaV2_Opacity_Time_BP_Z" (Utilities|Time|GetGameTimeinSeconds))
            (Rendering|Material|SetScalarParameterValueonMaterials _returnvalue "FDecalV2_Opacity_Time__Start" 1.0)
            (Rendering|Material|SetScalarParameterValueonMaterials _returnvalue "FDecalV2_Opacity_Time__End" 0.0)
            (Rendering|Material|SetScalarParameterValueonMaterials _returnvalue "FDecalaV2_Opacity_Duration" 5.0)))))))

(event EventBeginPlay
  (Utilities|FlowControl|Delay 0.2))

(event EventOnPawnSet
  (bind _returnvalue (CustomItemWrap|GetWrappedMeshComponents))
  (bind _returnvalue_2 (CustomItemWrap|GetWrappedMeshComponents))
  (bind _returnvalue_3 (CustomItemWrap|GetWrappedMeshComponents))
  (bind _returnvalue_4 (CustomItemWrap|GetWrappedMeshComponents))
  (Cosmetics|WatchCosmeticStat (CustomCharacter|GetAssociatedPlayerPawn) "(TagName=\"StatManager.AthenaKills\")" 0 false
    (:OnStatChanged
      (if (not (Variables|Default|GetPause))
        (Variables|Default|SetPause true)
        (Rendering|Material|SetScalarParameterValueonMaterials _returnvalue_2 "TopS_01_Blend_Opacity---End" 1.0)
        (Rendering|Material|SetScalarParameterValueonMaterials _returnvalue_2 "FDecalV2_Opacity_Time__End" 0.0)
        (Rendering|Material|SetScalarParameterValueonMaterials _returnvalue_2 "TopS_01_Blend_Offset-TBP" (Utilities|Time|GetGameTimeinSeconds))
        (Rendering|Material|SetScalarParameterValueonMaterials _returnvalue_2 "TopS_01_Blend_Offset---Start" 0.5)
        (Rendering|Material|SetScalarParameterValueonMaterials _returnvalue_2 "TopS_01_Blend_Offset---End" -1.0)
        (Utilities|FlowControl|Sequence
          (:then_0
            (Utilities|FlowControl|Delay 3.0)
            (Rendering|Material|SetScalarParameterValueonMaterials _returnvalue_3 "TopS_01_Blend_Opacity-TBP" (Utilities|Time|GetGameTimeinSeconds))
            (Rendering|Material|SetScalarParameterValueonMaterials _returnvalue_3 "TopS_01_Blend_Opacity---Start" 1.0)
            (Rendering|Material|SetScalarParameterValueonMaterials _returnvalue_3 "TopS_01_Blend_Opacity---End" 0.0)
            (Utilities|FlowControl|Delay 2.0)
            (Variables|Default|SetPause false)
            (Variables|Default|SetKill_Count 0))
          (:then_1
            (Rendering|Material|SetScalarParameterValueonMaterials _returnvalue_4 "FDecalaV2_Opacity_Time_BP_Z" (Utilities|Time|GetGameTimeinSeconds))
            (Rendering|Material|SetScalarParameterValueonMaterials _returnvalue_4 "FDecalV2_Opacity_Time__Start" 0.0)
            (Rendering|Material|SetScalarParameterValueonMaterials _returnvalue_4 "FDecalV2_Opacity_Time__End" 1.0)
            (Rendering|Material|SetScalarParameterValueonMaterials _returnvalue_4 "FDecalaV2_Opacity_Duration" 1.0)
            (Utilities|FlowControl|Delay 1.0)
            (Rendering|Material|SetScalarParameterValueonMaterials _returnvalue "FDecalaV2_Opacity_Time_BP_Z" (Utilities|Time|GetGameTimeinSeconds))
            (Rendering|Material|SetScalarParameterValueonMaterials _returnvalue "FDecalV2_Opacity_Time__Start" 1.0)
            (Rendering|Material|SetScalarParameterValueonMaterials _returnvalue "FDecalV2_Opacity_Time__End" 0.0)
            (Rendering|Material|SetScalarParameterValueonMaterials _returnvalue "FDecalaV2_Opacity_Duration" 5.0)))))))

(event EventOnNewWeapons
  (params (WeaponsAsItems :type Array%20of%20Fort%20Item%20Object%20References)))

(event EventOnNewPlayerWeaponType
  (params (NewType :type EFortItemType%20Enum :default "WorldItem") (OldType :type EFortItemType%20Enum :default "WorldItem")))

(fn ConstructionScript)
