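; Custom item wrap Blueprint for the StageCue wrap that triggers emissive material effects when the weapon fires.
; After a short delay, each shot stamps game time into a brightness parameter on the wrapped meshes' materials.
; A Selection switch alternates between two sets of brightness and offset timestamps, each gated by its own pause flag.
; Delays based on Pause length clear those flags, while selection changes either immediately or at half that length.
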(event EventOnWeaponFired
  (bind _returnvalue (Utilities|Time|GetGameTimeinSeconds))
  (bind _returnvalue_2 (CustomItemWrap|GetWrappedMeshComponents))
  (bind _returnvalue_3 (CustomItemWrap|GetWrappedMeshComponents))
  (bind _returnvalue_4 (Utilities|Time|GetGameTimeinSeconds))
  (bind _pause_length (Variables|Default|GetPause%20length))
  (Utilities|FlowControl|Delay 0.05)
  (Rendering|Material|SetScalarParameterValueonMaterials (CustomItemWrap|GetWrappedMeshComponents) "TopS Emissive_02-Brightness-TBP" (Utilities|Time|GetGameTimeinSeconds))
  (switch int (Variables|Default|GetSelection)
    (:0
      (if (not (Variables|Default|GetPause))
        (Variables|Default|SetPause true)
        (Rendering|Material|SetScalarParameterValueonMaterials _returnvalue_3 "TopS Emissive_01-Brightness-TBP" _returnvalue_4)
        (Rendering|Material|SetScalarParameterValueonMaterials _returnvalue_3 "TopS Emissive_01_Offset----TBP" _returnvalue_4)
        (Utilities|FlowControl|Sequence
          (:then_0
            (Utilities|FlowControl|Delay _pause_length)
            (Variables|Default|SetPause false))
          (:then_1
            (Utilities|FlowControl|Delay (/ _pause_length 2.0))
            (Variables|Default|SetSelection 1)))))
    (:1
      (if (not (Variables|Default|GetPause_B))
        (Variables|Default|SetPause_B true)
        (Rendering|Material|SetScalarParameterValueonMaterials _returnvalue_2 "TopS Emissive_02-Brightness-TBP" _returnvalue)
        (Rendering|Material|SetScalarParameterValueonMaterials _returnvalue_2 "TopS Emissive_03_Offset----TBP" _returnvalue)
        (Utilities|FlowControl|Sequence
          (:then_0
            (Variables|Default|SetSelection 0)
            (Utilities|FlowControl|Delay _pause_length)
            (Variables|Default|SetPause_B false))
          (:then_1
            (Utilities|FlowControl|Delay (/ _pause_length 2.0))
            (Variables|Default|SetSelection 0)))))))

(fn ConstructionScript)
