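; Star Aces weapon wrap logic that feeds shot timing to the wrap materials whenever the weapon fires.
; Each shot writes game time to the SAces_Game_TimeBPE scalar parameter on the wrapped mesh components immediately.
; After a short delay, a counter switch writes game time to one of parameters BPA through BPD in round-robin order, giving the material four staggered timestamps alongside the immediate one.
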(event EventOnWeaponFired
  (bind _returnvalue (Utilities|Time|GetGameTimeinSeconds))
  (bind _returnvalue_2 (CustomItemWrap|GetWrappedMeshComponents))
  (Rendering|Material|SetScalarParameterValueonMaterials _returnvalue_2 "SAces_Game_TimeBPE" _returnvalue)
  (Utilities|FlowControl|Delay 0.15)
  (switch int (Variables|Default|GetSet_Varable)
    (:0
      (Rendering|Material|SetScalarParameterValueonMaterials _returnvalue_2 "SAces_Game_TimeBPA" _returnvalue)
      (Variables|Default|SetSet_Varable 1))
    (:1
      (Rendering|Material|SetScalarParameterValueonMaterials _returnvalue_2 "SAces_Game_TimeBPB" _returnvalue)
      (Variables|Default|SetSet_Varable 2))
    (:2
      (Rendering|Material|SetScalarParameterValueonMaterials _returnvalue_2 "SAces_Game_TimeBPC" _returnvalue)
      (Variables|Default|SetSet_Varable 3))
    (:3
      (Rendering|Material|SetScalarParameterValueonMaterials _returnvalue_2 "SAces_Game_TimeBPD" _returnvalue)
      (Variables|Default|SetSet_Varable 0))))

(fn ConstructionScript)
