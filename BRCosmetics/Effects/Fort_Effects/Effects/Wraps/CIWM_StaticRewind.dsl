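; CIWM_StaticRewind is a weapon wrap cosmetic Blueprint that drives time-based material effects when the weapon fires.
; On the fire event it writes game time to four decal time parameters (A through D) on the materials of the wrapped mesh components.
; Short delays separate the writes, so the decals update in sequence, leaving any animation to the materials themselves.
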(event EventOnWeaponFired
  (bind _returnvalue (Utilities|Time|GetGameTimeinSeconds))
  (bind _returnvalue_2 (CustomItemWrap|GetWrappedMeshComponents))
  (Rendering|Material|SetScalarParameterValueonMaterials _returnvalue_2 "EDRec_TimeFrom_BP_DecalA" _returnvalue)
  (Utilities|FlowControl|Delay 0.1)
  (Rendering|Material|SetScalarParameterValueonMaterials _returnvalue_2 "EDRec_TimeFrom_BP_DecalB" _returnvalue)
  (Utilities|FlowControl|Delay 0.1)
  (Rendering|Material|SetScalarParameterValueonMaterials _returnvalue_2 "EDRec_TimeFrom_BP_DecalC" _returnvalue)
  (Utilities|FlowControl|Delay 0.1)
  (Rendering|Material|SetScalarParameterValueonMaterials _returnvalue_2 "EDRec_TimeFrom_BP_DecalD" _returnvalue))

(fn ConstructionScript)
