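; StealthWoodland custom item wrap Blueprint that shows the player's kill count on wrapped items.
; When its pawn is set, it casts the associated pawn to FortPlayerPawn and watches the AthenaKills cosmetic stat.
; On each stat change it prints the value and writes it as a float to a number-display scalar parameter on the wrapped meshes' materials.
; The weapon-fired event is empty, and BeginPlay only calls the parent.
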(event EventOnPawnSet
  (bind _asfort_player_pawn (Utilities|Casting|CastToFortPlayerPawn (CustomCharacter|GetAssociatedPlayerPawn)))
  (bind (_statname _statvalue) (Cosmetics|WatchCosmeticStat _asfort_player_pawn "(TagName=\"StatManager.AthenaKills\")")
    (:OnStatChanged
      (Development|PrintString (Utilities|String|ToString(Integer) _statvalue))
      (Rendering|Material|SetScalarParameterValueonMaterials (CustomItemWrap|GetWrappedMeshComponents) "FNumDisplay_Digits_NumerValue" (Math|Conversions|ToFloat(Integer) _statvalue)))))

(event EventBeginPlay
  (|CallParentFunction))

(event EventOnWeaponFired)

(fn ConstructionScript)
