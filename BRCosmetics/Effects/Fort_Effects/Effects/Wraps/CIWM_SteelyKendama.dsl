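; Custom item wrap modifier for SteelyKendama that makes the wrap's materials on weapon meshes react to firing.
; On BeginPlay in the front end only, a timeline drives decal opacity and top-side emissive scalars.
; When the weapon fires, a Sequence stamps current game time into scalar parameters for top-side emissives, side decals, liquid and reactive decals.
; The Blueprint only supplies these timestamps, leaving the materials to animate from them.
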(event EventBeginPlay
  (bind _returnvalue (CustomItemWrap|GetWrappedMeshComponents))
  (if (CustomCosmeticModifier|IsinFrontEnd)
    (bind (_direction _decal _emissive) (Timeline|Timeline PlayFromStart)
      (:Update
        (Rendering|Material|SetScalarParameterValueonMaterials _returnvalue "FDecalV2_Opacity_TOD_Day" _decal)
        (Rendering|Material|SetScalarParameterValueonMaterials _returnvalue "TopS Emissivs_TOD_Day" _emissive)))))

(event EventOnWeaponFired
  (bind _returnvalue (Utilities|Time|GetGameTimeinSeconds))
  (bind _returnvalue_2 (Utilities|Time|GetGameTimeinSeconds))
  (bind _returnvalue_3 (Utilities|Time|GetGameTimeinSeconds))
  (bind _returnvalue_4 (CustomItemWrap|GetWrappedMeshComponents))
  (bind _returnvalue_5 (CustomItemWrap|GetWrappedMeshComponents))
  (bind _returnvalue_6 (Utilities|Time|GetGameTimeinSeconds))
  (bind _returnvalue_7 (CustomItemWrap|GetWrappedMeshComponents))
  (bind _returnvalue_8 (Utilities|Time|GetGameTimeinSeconds))
  (bind _returnvalue_9 (CustomItemWrap|GetWrappedMeshComponents))
  (bind _returnvalue_10 (Utilities|Time|GetGameTimeinSeconds))
  (bind _returnvalue_11 (CustomItemWrap|GetWrappedMeshComponents))
  (bind _returnvalue_12 (Utilities|Time|GetGameTimeinSeconds))
  (bind _returnvalue_13 (CustomItemWrap|GetWrappedMeshComponents))
  (bind _returnvalue_14 (Utilities|Time|GetGameTimeinSeconds))
  (bind _returnvalue_15 (CustomItemWrap|GetWrappedMeshComponents))
  (bind _returnvalue_16 (CustomItemWrap|GetWrappedMeshComponents))
  (Utilities|FlowControl|Sequence
    (:then_0
      ;[1 1 1] TopSide v1
      (Rendering|Material|SetScalarParameterValueonMaterials _returnvalue_15 "TS_Emissive01_Time_BP_Brightness" _returnvalue)
      (Rendering|Material|SetScalarParameterValueonMaterials _returnvalue_15 "TS_Emissive01_Time_BP_Grad_Offset" _returnvalue))
    (:then_1
      (Rendering|Material|SetScalarParameterValueonMaterials _returnvalue_15 "TS_Emissive02_Time_BP_Brightness" _returnvalue)
      (Rendering|Material|SetScalarParameterValueonMaterials _returnvalue_15 "TS_Emissive02_Time_BP_Grad_Offset" _returnvalue))
    (:then_2
      ;[1 1 1] TopsSide V2
      (Rendering|Material|SetScalarParameterValueonMaterials _returnvalue_5 "TopS Emissive_02-Brightness-TBP" _returnvalue_2)
      (Rendering|Material|SetScalarParameterValueonMaterials _returnvalue_5 "TopS Emissive_02_Offset----TBP" _returnvalue_2)
      (Rendering|Material|SetScalarParameterValueonMaterials _returnvalue_5 "TopS_01_Blend_Opacity-TBP" _returnvalue_2))
    (:then_3
      (Rendering|Material|SetScalarParameterValueonMaterials _returnvalue_5 "TopS Emissive_01-Brightness-TBP" _returnvalue_2)
      (Rendering|Material|SetScalarParameterValueonMaterials _returnvalue_5 "TopS Emissive_01_Offset----TBP" _returnvalue_2)
      (Rendering|Material|SetScalarParameterValueonMaterials _returnvalue_5 "TopS Emissive_03-Brightness-TBP" _returnvalue_2)
      (Rendering|Material|SetScalarParameterValueonMaterials _returnvalue_5 "TopS Emissive_03_Offset----TBP" _returnvalue_2))
    (:then_4
      ;[1 1 1] General
      (Rendering|Material|SetScalarParameterValueonMaterials _returnvalue_4 "TimeFrom-BP----------------" _returnvalue_3)
      (Rendering|Material|SetScalarParameterValueonMaterials _returnvalue_4 "FDecal_Emissive_Brightness_Time_BP" _returnvalue_3))
    (:then_5
      ;[1 1 1] Left Right v1
      (Rendering|Material|SetScalarParameterValueonMaterials _returnvalue_7 "FDecal_Emissive_Gradent_Time_BP" _returnvalue_8)
      (Rendering|Material|SetScalarParameterValueonMaterials _returnvalue_7 "FDecal_Opacity_Time_BP_Z" _returnvalue_8))
    (:then_6
      ;[1 1 1] Liquid
      (Rendering|Material|SetScalarParameterValueonMaterials (CustomItemWrap|GetWrappedMeshComponents) "Liquid_TimeFrom-BP" (Utilities|Time|GetGameTimeinSeconds)))
    (:then_7
      ; Left\Right V2
      (Rendering|Material|SetScalarParameterValueonMaterials _returnvalue_9 "FDecalV2_FlipBook_Time_FromBP" _returnvalue_10)
      (Rendering|Material|SetScalarParameterValueonMaterials _returnvalue_9 "FDecalaV2_Opacity_Time_BP_Z" _returnvalue_10)
      (Rendering|Material|SetScalarParameterValueonMaterials _returnvalue_9 "FDecalV2_Emissive_Brightness_Time_BP" _returnvalue_10)
      (Rendering|Material|SetScalarParameterValueonMaterials _returnvalue_9 "FDecalV2_Emissive_Gradent_Time_BP" _returnvalue_10)
      (Rendering|Material|SetScalarParameterValueonMaterials _returnvalue_9 "FDecalV2_Counter_Brightness_Time_BP" _returnvalue_10))
    (:then_8
      ;[1 1 1 0.5] Decal Reactive
      (Rendering|Material|SetScalarParameterValueonMaterials _returnvalue_11 "EDRec_TimeFrom_BP_DecalA" _returnvalue_12)
      (Rendering|Material|SetScalarParameterValueonMaterials _returnvalue_11 "EDRec_TimeFrom_BP_DecalB" _returnvalue_12)
      (Rendering|Material|SetScalarParameterValueonMaterials _returnvalue_11 "EDRec_TimeFrom_BP_DecalC" _returnvalue_12)
      (Rendering|Material|SetScalarParameterValueonMaterials _returnvalue_11 "EDRec_TimeFrom_BP_DecalD" _returnvalue_12)
      (Rendering|Material|SetScalarParameterValueonMaterials _returnvalue_13 "EDRec_Emissive_A_Brightness_Time_BP" _returnvalue_14)
      (Rendering|Material|SetScalarParameterValueonMaterials _returnvalue_13 "EDRec_Emissive_B_Brightness_Time_BP" _returnvalue_14)
      (Rendering|Material|SetScalarParameterValueonMaterials _returnvalue_13 "EDRec_Emissive_C_Brightness_Time_BP" _returnvalue_14)
      (Rendering|Material|SetScalarParameterValueonMaterials _returnvalue_13 "EDRec_Emissive_D_Brightness_Time_BP" _returnvalue_14)
      (Rendering|Material|SetScalarParameterValueonMaterials _returnvalue_16 "EDRec_Emissive_A_Gradent_Time_BP" _returnvalue_6)
      (Rendering|Material|SetScalarParameterValueonMaterials _returnvalue_16 "EDRec_Emissive_B_Gradent_Time_BP" _returnvalue_6)
      (Rendering|Material|SetScalarParameterValueonMaterials _returnvalue_16 "EDRec_Emissive_C_Gradent_Time_BP" _returnvalue_6)
      (Rendering|Material|SetScalarParameterValueonMaterials _returnvalue_16 "EDRec_Emissive_D_Gradent_Time_BP" _returnvalue_6))))

(fn ConstructionScript)
