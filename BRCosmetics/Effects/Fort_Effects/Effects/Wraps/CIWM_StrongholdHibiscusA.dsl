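; Custom weapon wrap Blueprint that makes the wrapped meshes' materials react to firing with stepped brightness levels.
; Each shot increments a capped step count, which decays toward zero when the gap since the previous shot exceeds a hold window.
; The step indexes two arrays, scaled by a start multiplier, to set blend-opacity and emissive brightness start values.
; Game time is also written as a material timestamp, so no Tick is used.
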(event EventOnWeaponFired
  (bind _returnvalue (CustomItemWrap|GetWrappedMeshComponents))
  (bind _returnvalue_2 (CustomItemWrap|GetWrappedMeshComponents))
  (bind _output_get (Variables|Default|SetStepCount (Math|Integer|Min(Integer) (+ (Math|Float|Truncate (Math|Float|Lerp (Math|Conversions|ToFloat(Integer) (Variables|Default|GetStepCount)) 0.0 (Math|Float|Clamp(Float) (/ (- (Utilities|Time|GetGameTimeinSeconds) (Variables|Default|Getwait_timer)) 1.0)))) 1) (Variables|Default|GetMaxSteps))))
  (Rendering|Material|SetScalarParameterValueonMaterials _returnvalue_2 "TopS_01_Blend_Opacity-TBP" (Utilities|Time|GetGameTimeinSeconds))
  (Rendering|Material|SetScalarParameterValueonMaterials _returnvalue_2 "TopS_01_Blend_Opacity---Start" (* (Utilities|Array|Get(aref) (Variables|Default|GetBrightness%20value) _output_get) (Variables|Default|GetStartMultiplier)))
  (Variables|Default|Setwait_timer (+ (Utilities|Time|GetGameTimeinSeconds) 5.5))
  (Rendering|Material|SetScalarParameterValueonMaterials _returnvalue "TopS Emissive_01-Brightness-TBP" (Utilities|Time|GetGameTimeinSeconds))
  (Rendering|Material|SetScalarParameterValueonMaterials _returnvalue "TopS Emissive_01_Brightness---Start" (* (Variables|Default|GetStartMultiplier) (Utilities|Array|Get(aref) (Variables|Default|GetBrightness%20Flames) _output_get))))

(fn ConstructionScript)
