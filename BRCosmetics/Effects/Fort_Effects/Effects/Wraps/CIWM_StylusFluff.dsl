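; Custom item wrap Blueprint that sparks a material effect on wrapped meshes when the player takes damage.
; Pawn and weapon events register watches on the DamageReceived cosmetic stat; in the front end, a FrontEnd timeline track drives the effect instead.
; Each trigger shuffles the Properties array and restarts the SparksInit timeline, whose four tracks write to the first four shuffled scalar parameter names, varying which parameter gets which curve.
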(event Custom|ReloadStarted
  (params (ReloadTime :type Float%20%28single-precision%29) (ReloadType :type EFortWeaponReloadType%20Enum :default "ReloadWholeClip")))

(event EventBeginPlay
  (bind _returnvalue (CustomItemWrap|GetWrappedMeshComponents))
  (bind _properties (Variables|Default|GetProperties))
  (if (CustomCosmeticModifier|IsinFrontEnd)
    (Timeline|FrontEnd PlayFromStart
      (:NewTrack_0
        _))
    (else
      (Cosmetics|WatchCosmeticStat (CustomCharacter|GetAssociatedPlayerPawn) "(TagName=\"StatManager.DamageReceived\")" 0 false -1.0 -1.0 false
        (:OnStatChanged
          _))))
  (Utilities|Array|Shuffle (Variables|Default|GetProperties))
  (bind (_direction _newtrack_0 _newtrack_1 _newtrack_2 _newtrack_3) (Timeline|SparksInit PlayFromStart)
    (:Update
      (Rendering|Material|SetScalarParameterValueonMaterials _returnvalue (Utilities|Array|Get(acopy) _properties) _newtrack_0)
      (Rendering|Material|SetScalarParameterValueonMaterials _returnvalue (Utilities|Array|Get(acopy) _properties 1) _newtrack_1)
      (Rendering|Material|SetScalarParameterValueonMaterials _returnvalue (Utilities|Array|Get(acopy) _properties 2) _newtrack_2)
      (Rendering|Material|SetScalarParameterValueonMaterials _returnvalue (Utilities|Array|Get(acopy) _properties 3) _newtrack_3))))

(event EventOnPawnSet
  (bind _returnvalue (CustomItemWrap|GetWrappedMeshComponents))
  (bind _properties (Variables|Default|GetProperties))
  (Cosmetics|WatchCosmeticStat (CustomCharacter|GetAssociatedPlayerPawn) "(TagName=\"StatManager.DamageReceived\")" 0 false -1.0 -1.0 false
    (:OnStatChanged
      (Utilities|Array|Shuffle (Variables|Default|GetProperties))
      (Timeline|SparksInit PlayFromStart))))

(event EventOnNewWeapons
  (params (WeaponsAsItems :type Array%20of%20Fort%20Item%20Object%20References))
  (bind _returnvalue (CustomItemWrap|GetWrappedMeshComponents))
  (bind _properties (Variables|Default|GetProperties))
  (Cosmetics|WatchCosmeticStat (CustomCharacter|GetAssociatedPlayerPawn) "(TagName=\"StatManager.DamageReceived\")" 0 false -1.0 -1.0 false
    (:OnStatChanged
      (Utilities|Array|Shuffle (Variables|Default|GetProperties))
      (Timeline|SparksInit PlayFromStart))))

(event EventOnWeaponChanged
  (params (NewWeapon :type Fort%20Weapon%20Object%20Reference) (OldWeapon :type Fort%20Weapon%20Object%20Reference))
  (bind _returnvalue (CustomItemWrap|GetWrappedMeshComponents))
  (bind _properties (Variables|Default|GetProperties))
  (Cosmetics|WatchCosmeticStat (CustomCharacter|GetAssociatedPlayerPawn) "(TagName=\"StatManager.DamageReceived\")" 0 false -1.0 -1.0 false
    (:OnStatChanged
      (Utilities|Array|Shuffle (Variables|Default|GetProperties))
      (Timeline|SparksInit PlayFromStart))))

(event EventOnNewPlayerWeaponType
  (params (NewType :type EFortItemType%20Enum :default "WorldItem") (OldType :type EFortItemType%20Enum :default "WorldItem"))
  (bind _returnvalue (CustomItemWrap|GetWrappedMeshComponents))
  (bind _properties (Variables|Default|GetProperties))
  (Cosmetics|WatchCosmeticStat (CustomCharacter|GetAssociatedPlayerPawn) "(TagName=\"StatManager.DamageReceived\")" 0 false -1.0 -1.0 false
    (:OnStatChanged
      (Utilities|Array|Shuffle (Variables|Default|GetProperties))
      (Timeline|SparksInit PlayFromStart))))

(fn ConstructionScript)
