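; Reactive SummerBummer weapon wrap that changes its materials as the player takes damage and heals.
; DamageReceived stat changes advance StatePhase up to 3, with a timeline lerping material offsets between per-phase values, then setting ice flipbook frames.
; GainedHealthTimes changes run a DoOnce-gated Heal timeline back to phase zero.
; Watchers register on BeginPlay and PawnSet; in frontend display, BeginPlay instead starts a looping LockerBreak timer, whose function is absent here.
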(event EventOnWeaponChanged
  (params (NewWeapon :type Fort%20Weapon%20Object%20Reference) (OldWeapon :type Fort%20Weapon%20Object%20Reference)))

(event EventOnNewWeapons
  (params (WeaponsAsItems :type Array%20of%20Fort%20Item%20Object%20References)))

(event EventOnNewPlayerWeaponType
  (params (NewType :type EFortItemType%20Enum :default "WorldItem") (OldType :type EFortItemType%20Enum :default "WorldItem")))

(event EventBeginPlay
  (bind _returnvalue (CustomItemWrap|GetWrappedMeshComponents))
  (bind _returnvalue_2 (CustomItemWrap|GetWrappedMeshComponents))
  (bind _returnvalue_3 (CustomItemWrap|GetWrappedMeshComponents))
  (bind _returnvalue_4 (CustomItemWrap|GetWrappedMeshComponents))
  (bind _returnvalue_5 (CustomCharacter|GetAssociatedPlayerPawn))
  (bind _statephase (Variables|Default|GetStatePhase))
  (bind _iceframes (Variables|Default|GetIceFrames))
  (bind _output (Utilities|Array|Get(acopy) _iceframes _statephase))
  (bind _statephase_2 (Variables|Default|GetStatePhase))
  (if (Reactive|Frontend|IsCosmeticObjectDisplayedinFrontendContext 0)
    (Utilities|Time|SetTimerbyFunctionName 0 "LockerBreak" 1.0 true)
    (else
      (Utilities|FlowControl|Sequence
        (:then_0
          (Cosmetics|WatchCosmeticStat _returnvalue_5 "(TagName=\"StatManager.DamageReceived\")" 0 false
            (:OnStatChanged
              (if (not (>= _statephase_2 3))
                (Variables|Default|SetStatePhase (Math|Integer|Clamp(Integer) (+ _statephase_2 1) 0 3))
                (bind (_direction _newtrack_0) (Timeline|WaterLe%20vel PlayFromStart)
                  (:Update
                    (Rendering|Material|SetVectorParameterValueonMaterials _returnvalue_2 "TopS_01_OffsetS_Weapon" (Math|Vector|MakeVector 0.0 (Math|Float|Lerp (Utilities|Array|Get(acopy) (Variables|Default|GetWeaponOffsets) (- (Variables|Default|GetStatePhase) 1)) (Utilities|Array|Get(acopy) (Variables|Default|GetWeaponOffsets) (Variables|Default|GetStatePhase)) _newtrack_0)))
                    (for _array_element (Variables|Default|GetParameter%20Name)
                      (Rendering|Material|SetVectorParameterValueonMaterials _returnvalue_2 _array_element (Math|Vector|MakeVector 0.0 (Math|Float|Lerp (Utilities|Array|Get(acopy) (Variables|Default|GetGeneralOffsets) (- (Variables|Default|GetStatePhase) 1)) (Utilities|Array|Get(acopy) (Variables|Default|GetGeneralOffsets) (Variables|Default|GetStatePhase)) _newtrack_0)))))
                  (:Finished
                    (Rendering|Material|SetScalarParameterValueonMaterials _returnvalue "EDRec_FlipBook_FrameA" _output)
                    (Rendering|Material|SetScalarParameterValueonMaterials _returnvalue "EDRec_FlipBook_FrameB" _output)))))))
        (:then_1
          (Cosmetics|WatchCosmeticStat _returnvalue_5 "(TagName=\"StatManager.GainedHealthTimes\")" 0 false
            (:OnStatChanged
              (if (not (<= (Variables|Default|GetStatePhase) 0))
                (Utilities|FlowControl|DoOnce false
                  (:Completed
                    (bind (_direction_2 _newtrack_0_2) (Timeline|Heal PlayFromStart)
                      (:Update
                        (Rendering|Material|SetVectorParameterValueonMaterials _returnvalue_4 "TopS_01_OffsetS_Weapon" (Math|Vector|MakeVector 0.0 (Math|Float|Lerp (Utilities|Array|Get(acopy) (Variables|Default|GetWeaponOffsets) (Variables|Default|GetStatePhase)) (Utilities|Array|Get(acopy) (Variables|Default|GetWeaponOffsets)) _newtrack_0_2)))
                        (for _array_element_2 (Variables|Default|GetParameter%20Name)
                          (Rendering|Material|SetVectorParameterValueonMaterials _returnvalue_4 _array_element_2 (Math|Vector|MakeVector 0.0 (Math|Float|Lerp (Utilities|Array|Get(acopy) (Variables|Default|GetGeneralOffsets) (Variables|Default|GetStatePhase)) (Utilities|Array|Get(acopy) (Variables|Default|GetGeneralOffsets)) _newtrack_0_2)))))
                      (:Finished
                        (Variables|Default|SetStatePhase 0)
                        (CallFunction|ReopenHealth)
                        (Rendering|Material|SetScalarParameterValueonMaterials _returnvalue_3 "EDRec_FlipBook_FrameA" 0.0)
                        (Rendering|Material|SetScalarParameterValueonMaterials _returnvalue_3 "EDRec_FlipBook_FrameB" 0.0)))))))))))))

(event EventOnPawnSet
  (bind _returnvalue (CustomItemWrap|GetWrappedMeshComponents))
  (bind _returnvalue_2 (CustomItemWrap|GetWrappedMeshComponents))
  (bind _returnvalue_3 (CustomItemWrap|GetWrappedMeshComponents))
  (bind _returnvalue_4 (CustomItemWrap|GetWrappedMeshComponents))
  (bind _returnvalue_5 (CustomCharacter|GetAssociatedPlayerPawn))
  (bind _statephase (Variables|Default|GetStatePhase))
  (bind _iceframes (Variables|Default|GetIceFrames))
  (bind _output (Utilities|Array|Get(acopy) _iceframes _statephase))
  (bind _statephase_2 (Variables|Default|GetStatePhase))
  (Utilities|FlowControl|Sequence
    (:then_0
      (Cosmetics|WatchCosmeticStat _returnvalue_5 "(TagName=\"StatManager.DamageReceived\")" 0 false
        (:OnStatChanged
          (if (not (>= _statephase_2 3))
            (Variables|Default|SetStatePhase (Math|Integer|Clamp(Integer) (+ _statephase_2 1) 0 3))
            (Timeline|WaterLe%20vel PlayFromStart)))))
    (:then_1
      (Cosmetics|WatchCosmeticStat _returnvalue_5 "(TagName=\"StatManager.GainedHealthTimes\")" 0 false
        (:OnStatChanged
          (if (not (<= (Variables|Default|GetStatePhase) 0))
            (Utilities|FlowControl|DoOnce false
              (:Completed
                (Timeline|Heal PlayFromStart)))))))))

(event Custom|Unbreak
  (if (Reactive|Frontend|IsCosmeticObjectDisplayedinFrontendContext 0)
    _))

(event Custom|Reopen%20Health
  (bind _returnvalue (CustomItemWrap|GetWrappedMeshComponents))
  (bind _returnvalue_2 (CustomItemWrap|GetWrappedMeshComponents))
  (Utilities|FlowControl|DoOnce false
    (:Completed
      (Timeline|Heal PlayFromStart))))

(fn ConstructionScript)
