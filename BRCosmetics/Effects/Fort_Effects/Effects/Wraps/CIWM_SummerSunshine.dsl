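; Custom item wrap Blueprint for the Summer Sunshine wrap that makes its material react when the weapon fires.
; Each shot stops and replays a timeline whose track lerps four EDRec warp strength scalars (A to D) between 0 and 0.05 on the wrapped mesh components.
; Timeline completion clears IsKilling?, which nothing here sets to true.
; BeginPlay and OnPawnSet only check for frontend display, with empty branches.
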(event EventOnPawnSet
  (if (Reactive|Frontend|IsCosmeticObjectDisplayedinFrontendContext 0)
    _))

(event EventBeginPlay
  (if (Reactive|Frontend|IsCosmeticObjectDisplayedinFrontendContext 0)
    _))

(event EventOnWeaponFired
  (bind _returnvalue (CustomItemWrap|GetWrappedMeshComponents))
  (Utilities|FlowControl|Sequence
    (:then_0
      (bind (_direction _newtrack_0) (Timeline|Timeline Stop)
        (:Update
          (Rendering|Material|SetScalarParameterValueonMaterials _returnvalue "EDRec_WarpD_Strength" (Math|Float|Lerp 0.0 0.05 _newtrack_0))
          (Rendering|Material|SetScalarParameterValueonMaterials _returnvalue "EDRec_WarpC_Strength" (Math|Float|Lerp 0.0 0.05 _newtrack_0))
          (Rendering|Material|SetScalarParameterValueonMaterials _returnvalue "EDRec_WarpA_Strength" (Math|Float|Lerp 0.0 0.05 _newtrack_0))
          (Rendering|Material|SetScalarParameterValueonMaterials _returnvalue "EDRec_WarpB_Strength" (Math|Float|Lerp 0.0 0.05 _newtrack_0)))
        (:Finished
          (Variables|Default|SetIsKilling? false))))
    (:then_1
      (Timeline|Timeline PlayFromStart))))

(fn ConstructionScript)
