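; Item wrap Blueprint for SummerTideB that makes a liquid level in the wrapped meshes' materials follow the current weapon's magazine.
; Spawn and weapon-change events cache ammo and clip size and bind the reload delegate; firing updates the count.
; WaterLevel lerps Z offset parameters per mesh category by ammo ratio and stamps game time.
; After the reload delay, paired offsets and a reset timestamp are pushed, and ammo resets to full.
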(event EventBeginPlay
  (bind _returnvalue (CustomCharacter|GetAssociatedPlayerPawn))
  (bind _currentweapon (Variables|/Script/FortniteGame.FortPawn|GetCurrentWeapon _returnvalue))
  (Utilities|FlowControl|Delay 0.1)
  (bind _currentweapon_2 (Variables|/Script/FortniteGame.FortPawn|GetCurrentWeapon _returnvalue)
    (:then
      (EventDispatchers|BindEventtoOnReloadStarted _currentweapon_2 (delegate Custom|ReloadStarted))
      (Variables|Default|Setper%20Clip (FortWeapon|GetBulletsPerClip _currentweapon))
      (Variables|Default|SetAmmo%20Count (FortWeapon|GetMagazineAmmoCount _currentweapon))
      (CallFunction|WaterLevel))
    (:else)))

(event EventOnWeaponChanged
  (params (NewWeapon :type Fort%20Weapon%20Object%20Reference) (OldWeapon :type Fort%20Weapon%20Object%20Reference))
  (bind _returnvalue (CustomCharacter|GetAssociatedPlayerPawn))
  (bind _currentweapon (Variables|/Script/FortniteGame.FortPawn|GetCurrentWeapon _returnvalue))
  (Utilities|FlowControl|Delay 0.01)
  (bind _currentweapon_2 (Variables|/Script/FortniteGame.FortPawn|GetCurrentWeapon _returnvalue)
    (:then
      (EventDispatchers|BindEventtoOnReloadStarted _currentweapon_2 (delegate Custom|ReloadStarted))
      (Variables|Default|Setper%20Clip (FortWeapon|GetBulletsPerClip _currentweapon))
      (Variables|Default|SetAmmo%20Count (FortWeapon|GetMagazineAmmoCount _currentweapon))
      (CallFunction|WaterLevel))
    (:else)))

(event EventOnWeaponFired
  (Utilities|FlowControl|Delay 0.01)
  (bind _output_get (Variables|Default|SetAmmo%20Count (FortWeapon|GetMagazineAmmoCount (Variables|/Script/FortniteGame.FortPawn|GetCurrentWeapon (CustomCharacter|GetAssociatedPlayerPawn)))))
  (Utilities|FlowControl|Delay 0.01)
  (bind _output_get_2 (Variables|Default|SetAmmo%20Count (Math|Integer|Clamp(Integer) (- _output_get 1) 0 (Variables|Default|Getper%20Clip))))
  (CallFunction|WaterLevel))

(event EventOnPawnSet
  (bind _returnvalue (CustomCharacter|GetAssociatedPlayerPawn))
  (bind _currentweapon (Variables|/Script/FortniteGame.FortPawn|GetCurrentWeapon _returnvalue))
  (Utilities|FlowControl|Delay 0.01)
  (bind _currentweapon_2 (Variables|/Script/FortniteGame.FortPawn|GetCurrentWeapon _returnvalue)
    (:then
      (EventDispatchers|BindEventtoOnReloadStarted _currentweapon_2 (delegate Custom|ReloadStarted))
      (Variables|Default|Setper%20Clip (FortWeapon|GetBulletsPerClip _currentweapon))
      (Variables|Default|SetAmmo%20Count (FortWeapon|GetMagazineAmmoCount _currentweapon))
      (CallFunction|WaterLevel))
    (:else)))

(event EventOnNewWeapons
  (params (WeaponsAsItems :type Array%20of%20Fort%20Item%20Object%20References))
  (bind _returnvalue (CustomCharacter|GetAssociatedPlayerPawn))
  (bind _currentweapon (Variables|/Script/FortniteGame.FortPawn|GetCurrentWeapon _returnvalue))
  (Utilities|FlowControl|Delay 0.01)
  (bind _currentweapon_2 (Variables|/Script/FortniteGame.FortPawn|GetCurrentWeapon _returnvalue)
    (:then
      (EventDispatchers|BindEventtoOnReloadStarted _currentweapon_2 (delegate Custom|ReloadStarted))
      (Variables|Default|Setper%20Clip (FortWeapon|GetBulletsPerClip _currentweapon))
      (Variables|Default|SetAmmo%20Count (FortWeapon|GetMagazineAmmoCount _currentweapon))
      (CallFunction|WaterLevel))
    (:else)))

(event EventOnNewPlayerWeaponType
  (params (NewType :type EFortItemType%20Enum :default "WorldItem") (OldType :type EFortItemType%20Enum :default "WorldItem"))
  (bind _returnvalue (CustomCharacter|GetAssociatedPlayerPawn))
  (bind _currentweapon (Variables|/Script/FortniteGame.FortPawn|GetCurrentWeapon _returnvalue))
  (Utilities|FlowControl|Delay 0.01)
  (bind _currentweapon_2 (Variables|/Script/FortniteGame.FortPawn|GetCurrentWeapon _returnvalue)
    (:then
      (EventDispatchers|BindEventtoOnReloadStarted _currentweapon_2 (delegate Custom|ReloadStarted))
      (Variables|Default|Setper%20Clip (FortWeapon|GetBulletsPerClip _currentweapon))
      (Variables|Default|SetAmmo%20Count (FortWeapon|GetMagazineAmmoCount _currentweapon))
      (CallFunction|WaterLevel))
    (:else)))

(event Custom|ReloadStarted
  (params (ReloadTime :type Float%20%28single-precision%29) (ReloadType :type EFortWeaponReloadType%20Enum :default "ReloadWholeClip"))
  (bind _per_clip (Variables|Default|Getper%20Clip))
  (bind _returnvalue (Math|Conversions|ToFloat(Integer) _per_clip))
  (bind _ammo_count (Variables|Default|GetAmmo%20Count))
  (bind _returnvalue_2 (Math|Conversions|ToFloat(Integer) _ammo_count))
  (bind _returnvalue_3 (/ _returnvalue_2 _returnvalue))
  (Utilities|FlowControl|Delay ReloadTime)
  (CallFunction|WaterLevel)
  (CallFunction|WeaponOffsets "DarkA_End_Offset_Weapon" (Math|Float|Lerp 0.14 -0.1 _returnvalue_3) "DarkA_Offset_Weapon" -0.1)
  (CallFunction|WeaponOffsets "DarkA_End_Offset_Backpack" (Math|Float|Lerp 0.4 -0.55 _returnvalue_3) "DarkA_Offset_Backpack" -0.55)
  (CallFunction|WeaponOffsets "DarkA_End_Offset_Character" (Math|Float|Lerp 0.3 -0.5 _returnvalue_3) "DarkA_Offset_Character" -0.5)
  (CallFunction|WeaponOffsets "DarkA_End_Offset_Mech" (Math|Float|Lerp 0.3 -0.5 _returnvalue_3) "DarkA_Offset_Mech" -0.5)
  (CallFunction|WeaponOffsets "DarkA_End_Offset_Pickaxe" (Math|Float|Lerp 0.2 -0.3 _returnvalue_3) "DarkA_Offset_Pickaxe" -0.3)
  (CallFunction|WeaponOffsets "DarkA_End_Offset_Vehicle" (Math|Float|Lerp -0.03 -0.1 _returnvalue_3) "DarkA_Offset_Vehicle" -0.1)
  (CallFunction|WeaponOffsets "DarkA_End_Offset_VehicleSide" (Math|Float|Lerp 0.25 0.075 _returnvalue_3) "DarkA_Offset_VehicleSide" 0.075)
  (Rendering|Material|SetScalarParameterValueonMaterials (CustomItemWrap|GetWrappedMeshComponents) "Liquid_Reset_TimeFrom-BP" (Utilities|Time|GetGameTimeinSeconds))
  (Variables|Default|SetAmmo%20Count (FortWeapon|GetBulletsPerClip (Variables|/Script/FortniteGame.FortPawn|GetCurrentWeapon (CustomCharacter|GetAssociatedPlayerPawn)))))

(event Custom|Water_Level
  (bind _per_clip (Variables|Default|Getper%20Clip))
  (bind _returnvalue (Math|Conversions|ToFloat(Integer) _per_clip))
  (bind _returnvalue_2 (CustomItemWrap|GetWrappedMeshComponents))
  (bind _ammo_count (Variables|Default|GetAmmo%20Count))
  (bind _returnvalue_3 (Math|Conversions|ToFloat(Integer) _ammo_count))
  (bind _returnvalue_4 (/ _returnvalue_3 _returnvalue))
  (Rendering|Material|SetVectorParameterValueonMaterials _returnvalue_2 "DarkA_Offset_Weapon" (Math|Vector|MakeVector 0.0 0.0 (Math|Float|Lerp 0.14 -0.1 _returnvalue_4)))
  (Rendering|Material|SetVectorParameterValueonMaterials _returnvalue_2 "DarkA_Offset_Backpack" (Math|Vector|MakeVector 0.0 0.0 (Math|Float|Lerp 0.4 -0.55 _returnvalue_4)))
  (Rendering|Material|SetVectorParameterValueonMaterials _returnvalue_2 "DarkA_Offset_Character" (Math|Vector|MakeVector 0.0 0.0 (Math|Float|Lerp 0.3 -0.5 _returnvalue_4)))
  (Rendering|Material|SetVectorParameterValueonMaterials _returnvalue_2 "DarkA_Offset_Mech" (Math|Vector|MakeVector 0.0 0.0 (Math|Float|Lerp 0.3 -0.5 _returnvalue_4)))
  (Rendering|Material|SetVectorParameterValueonMaterials _returnvalue_2 "DarkA_Offset_Pickaxe" (Math|Vector|MakeVector 0.0 0.0 (Math|Float|Lerp 0.2 -0.3 _returnvalue_4)))
  (Rendering|Material|SetVectorParameterValueonMaterials _returnvalue_2 "DarkA_Offset_Vehicle" (Math|Vector|MakeVector 0.0 0.0 (Math|Float|Lerp -0.03 -0.1 _returnvalue_4)))
  (Rendering|Material|SetVectorParameterValueonMaterials _returnvalue_2 "DarkA_Offset_VehicleSide" (Math|Vector|MakeVector 0.0 0.0 (Math|Float|Lerp 0.25 0.075 _returnvalue_4)))
  (Rendering|Material|SetScalarParameterValueonMaterials _returnvalue_2 "Liquid_TimeFrom-BP" (Utilities|Time|GetGameTimeinSeconds)))

(fn ConstructionScript)

(fn Weapon_Offsets
  (params (Offset_End_Name :type Name) (Offset_End_Value :type Float%20%28double-precision%29) (Offset_Start_Name :type Name) (Offset_State_Value :type Float%20%28double-precision%29))
  (bind _returnvalue (CustomItemWrap|GetWrappedMeshComponents))
  (Rendering|Material|SetVectorParameterValueonMaterials _returnvalue Offset_End_Name (Math|Vector|MakeVector 0.0 0.0 Offset_End_Value))
  (Rendering|Material|SetVectorParameterValueonMaterials _returnvalue Offset_Start_Name (Math|Vector|MakeVector 0.0 0.0 Offset_State_Value)))
