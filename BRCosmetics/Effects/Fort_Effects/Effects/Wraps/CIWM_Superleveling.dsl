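; Custom item wrap Blueprint for the Superleveling wrap, which levels up its material as the player's elimination count rises.
; It watches the AthenaKills cosmetic stat on the owning pawn, caches the count and writes it, clamped to six, into SL_CurrentLevel on wrapped meshes.
; Changes at or below the cap also stamp game time into SL_GameTimeFromBP.
; New weapons or weapon-type changes reapply the cached level after a short delay.
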(event Custom|OnActualStatChange
  (Rendering|Material|SetScalarParameterValueonMaterials (CustomItemWrap|GetWrappedMeshComponents) "SL_GameTimeFromBP" (Utilities|Time|GetGameTimeinSeconds)))

(event EventOnPawnSet
  (bind _asfort_player_pawn (Utilities|Casting|CastToFortPlayerPawn (CustomCharacter|GetAssociatedPlayerPawn)))
  (bind (_statname _statvalue) (Cosmetics|WatchCosmeticStat _asfort_player_pawn "(TagName=\"StatManager.AthenaKills\")")
    (:OnStatChanged
      (bind _output_get (Variables|Default|SetKill_Count _statvalue))
      (Rendering|Material|SetScalarParameterValueonMaterials (CustomItemWrap|GetWrappedMeshComponents) "SL_CurrentLevel" (Math|Float|Clamp(Float) (Math|Conversions|ToFloat(Integer) _output_get) 0.0 6.0))
      (if (<= (Math|Conversions|ToFloat(Integer) _output_get) 6.0)
        (CallFunction|OnActualStatChange)))))

(event EventOnNewWeapons
  (params (WeaponsAsItems :type Array%20of%20Fort%20Item%20Object%20References))
  ;[1 1 1] Updated kill count on new weapon
  (Utilities|FlowControl|Delay 0.2)
  (Rendering|Material|SetScalarParameterValueonMaterials (CustomItemWrap|GetWrappedMeshComponents) "SL_CurrentLevel" (Math|Float|Clamp(Float) (Math|Conversions|ToFloat(Integer) (Variables|Default|GetKill_Count)) 0.0 6.0)))

(event EventOnNewPlayerWeaponType
  (params (NewType :type EFortItemType%20Enum :default "WorldItem") (OldType :type EFortItemType%20Enum :default "WorldItem"))
  ;[1 1 1] Updated kill count on new weapon
  (Utilities|FlowControl|Delay 0.2)
  (Rendering|Material|SetScalarParameterValueonMaterials (CustomItemWrap|GetWrappedMeshComponents) "SL_CurrentLevel" (Math|Float|Clamp(Float) (Math|Conversions|ToFloat(Integer) (Variables|Default|GetKill_Count)) 0.0 6.0)))

(fn ConstructionScript)
