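; CIWM_SweetCuddle is an item wrap Blueprint that recolors emissives to match the player's character style.
; On customization finished, wrap target set or pawn set, identical logic checks the character name and reads two variant channels, with debug prints.
; MaterialFront and MaterialBack write emissive colors 01 and 02 respectively to the wrapped meshes' dynamic materials; blue is the front fallback.
; WrapMAt, a two-stage texture and accent setter, is never called here.
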(event EventOnCharacterCustomizationFinished
  (params (CharacterPartInfo :type Array%20of%20Applied%20Character%20Part%20Info%20Structures))
  (bind _returnvalue (CosmeticLoadout|GetCosmeticLoadout))
  (bind _returnvalue_2 (CosmeticLoadout|GetItemDefinitionfromLoadout _returnvalue "(TagName=\"LoadoutSlot.Character\")"))
  (bind _returnvalue_3 (CustomCharacter|GetAssociatedPlayerPawn))
  (bind _returnvalue_4 (Item|GetPersistentName _returnvalue_2))
  (Development|PrintText (Utilities|Text|ToText(String) _returnvalue_4) true true "(R=1.000000,G=0.682705,B=0.000000,A=1.000000)" 5e+01)
  (if (Utilities|String|EqualExactly(String) _returnvalue_4 (Variables|Default|GetCharacter-Name))
    (Utilities|FlowControl|Sequence
      (:then_0
        (bind _index (CosmeticLoadout|GetVariantChannelActiveIndexforSlotActor "(TagName=\"LoadoutSlot.Character\")" _returnvalue_3 "(TagName=\"Cosmetics.Variant.Channel.Slot.0\")" (Utilities|Array|MakeArray "(TagName=\"Cosmetics.Variant.Property.Mat10\")" "(TagName=\"Cosmetics.Variant.Property.Mat11\")" "(TagName=\"Cosmetics.Variant.Property.Mat12\")" "(TagName=\"Cosmetics.Variant.Property.Mat13\")") false false))
        (Development|PrintText (Utilities|Text|ToText(Float) (Math|Conversions|ToFloat(Integer) _index)) true true "(R=0.006593,G=0.000000,B=1.000000,A=1.000000)")
        (switch int (Utilities|Array|Get(acopy) (Utilities|Array|MakeArray 0 1 2 3) _index)
          (:0
            (CallFunction|MaterialFront "(R=0.875000,G=0.042150,B=0.527978,A=1.000000)"))
          (:1
            _)
          (:2
            (CallFunction|MaterialFront "(R=1.000000,G=0.317629,B=0.000000,A=1.000000)"))
          (:3
            (CallFunction|MaterialFront "(R=0.131771,G=0.718750,B=0.000000,A=1.000000)"))))
      (:then_1
        (bind _index_2 (CosmeticLoadout|GetVariantChannelActiveIndexforSlotActor "(TagName=\"LoadoutSlot.Character\")" _returnvalue_3 "(TagName=\"Cosmetics.Variant.Channel.Slot.1\")" (Utilities|Array|MakeArray "(TagName=\"Cosmetics.Variant.Property.Mat20\")" "(TagName=\"Cosmetics.Variant.Property.Mat21\")" "(TagName=\"Cosmetics.Variant.Property.Mat22\")" "(TagName=\"Cosmetics.Variant.Property.Mat23\")") false false))
        (switch int (Utilities|Array|Get(acopy) (Utilities|Array|MakeArray 0 1 2 3) _index_2)
          (:0
            (CallFunction|MaterialBack "(R=0.593750,G=0.028602,B=0.465225,A=1.000000)" "(R=0.549000,G=0.000000,B=0.045750,A=1.000000)"))
          (:1
            (CallFunction|MaterialBack "(R=0.004436,G=0.206813,B=0.791667,A=1.000000)" "(R=0.026209,G=0.116385,B=0.786458,A=1.000000)"))
          (:2
            (CallFunction|MaterialBack "(R=0.004436,G=0.206813,B=0.791667,A=1.000000)" "(R=1.000000,G=0.080133,B=0.000000,A=1.000000)"))
          (:3
            (CallFunction|MaterialBack "(R=0.004436,G=0.206813,B=0.791667,A=1.000000)" "(R=0.000000,G=0.312500,B=0.010417,A=1.000000)")))))
    (else
      (Utilities|FlowControl|Sequence
        (:then_0
          _)
        (:then_1
          _))))
  (CallFunction|MaterialFront "(R=0.000000,G=0.394829,B=1.000000,A=1.000000)"))

(event Custom|OnBannerFullyApplied
  (params (Icon :type Texture%202D%20Object%20Reference) (PrimaryColor :type Linear%20Color%20Structure)))

(event EventOnSetWrapTarget
  (bind _returnvalue (CosmeticLoadout|GetCosmeticLoadout))
  (bind _returnvalue_2 (CosmeticLoadout|GetItemDefinitionfromLoadout _returnvalue "(TagName=\"LoadoutSlot.Character\")"))
  (bind _returnvalue_3 (CustomCharacter|GetAssociatedPlayerPawn))
  (bind _returnvalue_4 (Item|GetPersistentName _returnvalue_2))
  (Development|PrintText (Utilities|Text|ToText(String) _returnvalue_4) true true "(R=1.000000,G=0.682705,B=0.000000,A=1.000000)" 5e+01)
  (if (Utilities|String|EqualExactly(String) _returnvalue_4 (Variables|Default|GetCharacter-Name))
    (Utilities|FlowControl|Sequence
      (:then_0
        (bind _index (CosmeticLoadout|GetVariantChannelActiveIndexforSlotActor "(TagName=\"LoadoutSlot.Character\")" _returnvalue_3 "(TagName=\"Cosmetics.Variant.Channel.Slot.0\")" (Utilities|Array|MakeArray "(TagName=\"Cosmetics.Variant.Property.Mat10\")" "(TagName=\"Cosmetics.Variant.Property.Mat11\")" "(TagName=\"Cosmetics.Variant.Property.Mat12\")" "(TagName=\"Cosmetics.Variant.Property.Mat13\")") false false))
        (Development|PrintText (Utilities|Text|ToText(Float) (Math|Conversions|ToFloat(Integer) _index)) true true "(R=0.006593,G=0.000000,B=1.000000,A=1.000000)")
        (switch int (Utilities|Array|Get(acopy) (Utilities|Array|MakeArray 0 1 2 3) _index)
          (:0
            (CallFunction|MaterialFront "(R=0.875000,G=0.042150,B=0.527978,A=1.000000)"))
          (:1
            _)
          (:2
            (CallFunction|MaterialFront "(R=1.000000,G=0.317629,B=0.000000,A=1.000000)"))
          (:3
            (CallFunction|MaterialFront "(R=0.131771,G=0.718750,B=0.000000,A=1.000000)"))))
      (:then_1
        (bind _index_2 (CosmeticLoadout|GetVariantChannelActiveIndexforSlotActor "(TagName=\"LoadoutSlot.Character\")" _returnvalue_3 "(TagName=\"Cosmetics.Variant.Channel.Slot.1\")" (Utilities|Array|MakeArray "(TagName=\"Cosmetics.Variant.Property.Mat20\")" "(TagName=\"Cosmetics.Variant.Property.Mat21\")" "(TagName=\"Cosmetics.Variant.Property.Mat22\")" "(TagName=\"Cosmetics.Variant.Property.Mat23\")") false false))
        (switch int (Utilities|Array|Get(acopy) (Utilities|Array|MakeArray 0 1 2 3) _index_2)
          (:0
            (CallFunction|MaterialBack "(R=0.593750,G=0.028602,B=0.465225,A=1.000000)" "(R=0.549000,G=0.000000,B=0.045750,A=1.000000)"))
          (:1
            (CallFunction|MaterialBack "(R=0.004436,G=0.206813,B=0.791667,A=1.000000)" "(R=0.026209,G=0.116385,B=0.786458,A=1.000000)"))
          (:2
            (CallFunction|MaterialBack "(R=0.004436,G=0.206813,B=0.791667,A=1.000000)" "(R=1.000000,G=0.080133,B=0.000000,A=1.000000)"))
          (:3
            (CallFunction|MaterialBack "(R=0.004436,G=0.206813,B=0.791667,A=1.000000)" "(R=0.000000,G=0.312500,B=0.010417,A=1.000000)")))))
    (else
      (Utilities|FlowControl|Sequence
        (:then_0
          _)
        (:then_1
          _))))
  (CallFunction|MaterialFront "(R=0.000000,G=0.394829,B=1.000000,A=1.000000)"))

(event EventOnPawnSet
  (bind _returnvalue (CosmeticLoadout|GetCosmeticLoadout))
  (bind _returnvalue_2 (CosmeticLoadout|GetItemDefinitionfromLoadout _returnvalue "(TagName=\"LoadoutSlot.Character\")"))
  (bind _returnvalue_3 (CustomCharacter|GetAssociatedPlayerPawn))
  (bind _returnvalue_4 (Item|GetPersistentName _returnvalue_2))
  (Development|PrintText (Utilities|Text|ToText(String) _returnvalue_4) true true "(R=1.000000,G=0.682705,B=0.000000,A=1.000000)" 5e+01)
  (if (Utilities|String|EqualExactly(String) _returnvalue_4 (Variables|Default|GetCharacter-Name))
    (Utilities|FlowControl|Sequence
      (:then_0
        (bind _index (CosmeticLoadout|GetVariantChannelActiveIndexforSlotActor "(TagName=\"LoadoutSlot.Character\")" _returnvalue_3 "(TagName=\"Cosmetics.Variant.Channel.Slot.0\")" (Utilities|Array|MakeArray "(TagName=\"Cosmetics.Variant.Property.Mat10\")" "(TagName=\"Cosmetics.Variant.Property.Mat11\")" "(TagName=\"Cosmetics.Variant.Property.Mat12\")" "(TagName=\"Cosmetics.Variant.Property.Mat13\")") false false))
        (Development|PrintText (Utilities|Text|ToText(Float) (Math|Conversions|ToFloat(Integer) _index)) true true "(R=0.006593,G=0.000000,B=1.000000,A=1.000000)")
        (switch int (Utilities|Array|Get(acopy) (Utilities|Array|MakeArray 0 1 2 3) _index)
          (:0
            (CallFunction|MaterialFront "(R=0.875000,G=0.042150,B=0.527978,A=1.000000)"))
          (:1
            _)
          (:2
            (CallFunction|MaterialFront "(R=1.000000,G=0.317629,B=0.000000,A=1.000000)"))
          (:3
            (CallFunction|MaterialFront "(R=0.131771,G=0.718750,B=0.000000,A=1.000000)"))))
      (:then_1
        (bind _index_2 (CosmeticLoadout|GetVariantChannelActiveIndexforSlotActor "(TagName=\"LoadoutSlot.Character\")" _returnvalue_3 "(TagName=\"Cosmetics.Variant.Channel.Slot.1\")" (Utilities|Array|MakeArray "(TagName=\"Cosmetics.Variant.Property.Mat20\")" "(TagName=\"Cosmetics.Variant.Property.Mat21\")" "(TagName=\"Cosmetics.Variant.Property.Mat22\")" "(TagName=\"Cosmetics.Variant.Property.Mat23\")") false false))
        (switch int (Utilities|Array|Get(acopy) (Utilities|Array|MakeArray 0 1 2 3) _index_2)
          (:0
            (CallFunction|MaterialBack "(R=0.593750,G=0.028602,B=0.465225,A=1.000000)" "(R=0.549000,G=0.000000,B=0.045750,A=1.000000)"))
          (:1
            (CallFunction|MaterialBack "(R=0.004436,G=0.206813,B=0.791667,A=1.000000)" "(R=0.026209,G=0.116385,B=0.786458,A=1.000000)"))
          (:2
            (CallFunction|MaterialBack "(R=0.004436,G=0.206813,B=0.791667,A=1.000000)" "(R=1.000000,G=0.080133,B=0.000000,A=1.000000)"))
          (:3
            (CallFunction|MaterialBack "(R=0.004436,G=0.206813,B=0.791667,A=1.000000)" "(R=0.000000,G=0.312500,B=0.010417,A=1.000000)")))))
    (else
      (Utilities|FlowControl|Sequence
        (:then_0
          _)
        (:then_1
          _))))
  (CallFunction|MaterialFront "(R=0.000000,G=0.394829,B=1.000000,A=1.000000)"))

(fn Material_Back
  (params (TopS%20Emissive_01_Color- :type Linear%20Color%20Structure) (TopS%20Emissive_02_Color- :type Linear%20Color%20Structure))
  (for _array_element (CustomItemWrap|GetWrappedMeshComponents)
    (for _array_element_2 (Rendering|Material|GetMaterials _array_element)
      (bind _asmaterial_instance_dynamic (Utilities|Casting|CastToMaterialInstanceDynamic _array_element_2))
      (Rendering|Material|SetVectorParameterValue _asmaterial_instance_dynamic "TopS Emissive_02_Color-" TopS%20Emissive_02_Color- :!class /Script/Engine.MaterialInstanceDynamic))))

(fn Material_Front
  (params (TopS%20Emissive_01_Color- :type Linear%20Color%20Structure) (TopS%20Emissive_02_Color- :type Linear%20Color%20Structure))
  (for _array_element (CustomItemWrap|GetWrappedMeshComponents)
    (for _array_element_2 (Rendering|Material|GetMaterials _array_element)
      (bind _asmaterial_instance_dynamic (Utilities|Casting|CastToMaterialInstanceDynamic _array_element_2))
      (Rendering|Material|SetVectorParameterValue _asmaterial_instance_dynamic "TopS Emissive_01_Color-" TopS%20Emissive_01_Color- :!class /Script/Engine.MaterialInstanceDynamic))))

(fn ConstructionScript)

(fn WrapMAt
  (params (Stages :type Integer) (A_Set_BaseColor :type Texture%20Object%20Reference) (B_Set_BaseColor :type Texture%20Object%20Reference) (A_Set_Mask :type Texture%20Object%20Reference) (B_Set_Mask :type Texture%20Object%20Reference) (A_Set_Norm :type Texture%20Object%20Reference) (B_Set_Norm :type Texture%20Object%20Reference) (A_Acent_Color :type Linear%20Color%20Structure) (A_Acent_Spec :type Float%20%28double-precision%29) (A_Acent_Roughness :type Float%20%28double-precision%29) (A_Acent_Metal :type Float%20%28double-precision%29) (B_Acent_Color :type Linear%20Color%20Structure) (B_Acent_Spec :type Float%20%28double-precision%29) (B_Acent_Roughness :type Float%20%28double-precision%29) (B_Acent_Metal :type Float%20%28double-precision%29) (A_Lut_MSk_Acent :type Float%20%28single-precision%29) (B_Lut_MSk_Acent :type Float%20%28single-precision%29))
  (for _array_element (CustomItemWrap|GetWrappedMeshComponents)
    (for _array_element_2 (Rendering|Material|GetMaterials _array_element)
      (bind _asmaterial_instance_dynamic (Utilities|Casting|CastToMaterialInstanceDynamic _array_element_2))
      (Rendering|Material|SetTextureParameterValue _asmaterial_instance_dynamic "TopS_01__Tex_BaseColor" (Utilities|Array|Get(acopy) (Utilities|Array|MakeArray A_Set_BaseColor B_Set_BaseColor) Stages))
      (Rendering|Material|SetTextureParameterValue _asmaterial_instance_dynamic "TopS_01__Tex_Mask" (Utilities|Array|Get(acopy) (Utilities|Array|MakeArray A_Set_Mask B_Set_Mask) Stages))
      (Rendering|Material|SetTextureParameterValue _asmaterial_instance_dynamic "TopS_01__Tex_Normal" (Utilities|Array|Get(acopy) (Utilities|Array|MakeArray A_Set_Norm B_Set_Norm) Stages))
      (Rendering|Material|SetVectorParameterValue _asmaterial_instance_dynamic "Simple_Color" (Utilities|Array|Get(acopy) (Utilities|Array|MakeArray A_Acent_Color B_Acent_Color) Stages) :!class /Script/Engine.MaterialInstanceDynamic)
      (Rendering|Material|SetScalarParameterValue _asmaterial_instance_dynamic "Simple_Specular" (Utilities|Array|Get(acopy) (Utilities|Array|MakeArray A_Acent_Spec B_Acent_Spec) Stages) :!class /Script/Engine.MaterialInstanceDynamic)
      (Rendering|Material|SetScalarParameterValue _asmaterial_instance_dynamic "Simple_Roughness" (Utilities|Array|Get(acopy) (Utilities|Array|MakeArray A_Acent_Roughness B_Acent_Roughness) Stages) :!class /Script/Engine.MaterialInstanceDynamic)
      (Rendering|Material|SetScalarParameterValue _asmaterial_instance_dynamic "Simple_Metalness" (Utilities|Array|Get(acopy) (Utilities|Array|MakeArray A_Acent_Metal B_Acent_Metal) Stages) :!class /Script/Engine.MaterialInstanceDynamic)
      (Rendering|Material|SetScalarParameterValue _asmaterial_instance_dynamic "Metal_LUT_Msk_Accent   1=On  0=Off" (Utilities|Array|Get(acopy) (Utilities|Array|MakeArray A_Lut_MSk_Acent B_Lut_MSk_Acent) Stages) :!class /Script/Engine.MaterialInstanceDynamic))))
