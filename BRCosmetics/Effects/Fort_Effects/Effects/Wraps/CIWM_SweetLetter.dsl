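; Weapon wrap modifier for the SweetLetter cosmetic wrap that changes its material each time the weapon fires.
; After a short delay, a step counter cycles through four states, each storing a different blend offset.
; The offset and the current game time are then written as scalar parameters to the materials on the wrapped mesh components, with game time going to the parameter named TopS_01_Blend_Opacity-TBP.
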(event EventOnWeaponFired
  (bind _returnvalue (CustomItemWrap|GetWrappedMeshComponents))
  (Utilities|FlowControl|Delay 0.1)
  (switch int (Variables|Default|Getsteps)
    (:0
      (Variables|Default|SetOffset 0.1)
      (Variables|Default|Setsteps 1))
    (:1
      (Variables|Default|SetOffset -0.1)
      (Variables|Default|Setsteps 2))
    (:2
      (Variables|Default|SetOffset -0.4)
      (Variables|Default|Setsteps 3))
    (:3
      (Variables|Default|SetOffset -0.6)
      (Variables|Default|Setsteps 0)))
  (Rendering|Material|SetScalarParameterValueonMaterials _returnvalue "TopS_01_Blend_Offset---End" (Variables|Default|GetOffset))
  (Rendering|Material|SetScalarParameterValueonMaterials _returnvalue "TopS_01_Blend_Opacity-TBP" (Utilities|Time|GetGameTimeinSeconds)))

(fn ConstructionScript)
