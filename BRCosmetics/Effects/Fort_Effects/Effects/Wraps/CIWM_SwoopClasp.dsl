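; Custom item wrap Blueprint for the SwoopClasp wrap that animates its emissive material effects whenever the weapon fires.
; On fire, it writes the current game time into TopS brightness, offset, blend and timing scalar parameters on all wrapped meshes.
; It also restarts a timeline that drives separate Em03 offset vectors for weapon, backpack, character, pickaxe and vehicle variants, some with their own fixed second component.
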(event EventOnWeaponFired
  (bind _returnvalue (Utilities|Time|GetGameTimeinSeconds))
  (bind _returnvalue_2 (CustomItemWrap|GetWrappedMeshComponents))
  (Utilities|FlowControl|Sequence
    (:then_0
      ;[1 1 1] TopsSide V2
      (Rendering|Material|SetScalarParameterValueonMaterials _returnvalue_2 "TopS Emissive_02-Brightness-TBP" _returnvalue)
      (Rendering|Material|SetScalarParameterValueonMaterials _returnvalue_2 "TopS Emissive_02_Offset----TBP" _returnvalue)
      (Rendering|Material|SetScalarParameterValueonMaterials _returnvalue_2 "TopS_01_Blend_Opacity-TBP" _returnvalue)
      (Rendering|Material|SetScalarParameterValueonMaterials _returnvalue_2 "TimeFrom-BP----------------" _returnvalue)
      (bind (_direction _newtrack_0) (Timeline|Timeline PlayFromStart)
        (:Update
          (Rendering|Material|SetVectorParameterValueonMaterials _returnvalue_2 "TopS_Em03_Offset_Weapon" (Math|Vector|MakeVector _newtrack_0))
          (Rendering|Material|SetVectorParameterValueonMaterials _returnvalue_2 "TopS_Em03_Offset_Backpack" (Math|Vector|MakeVector _newtrack_0 0.21))
          (Rendering|Material|SetVectorParameterValueonMaterials _returnvalue_2 "TopS_Em03_Offset_Character" (Math|Vector|MakeVector _newtrack_0 0.21))
          (Rendering|Material|SetVectorParameterValueonMaterials _returnvalue_2 "TopS_Em03_Offset_Pickaxe" (Math|Vector|MakeVector _newtrack_0 0.165))
          (Rendering|Material|SetVectorParameterValueonMaterials _returnvalue_2 "TopS_Em03_Offset_Vehicle" (Math|Vector|MakeVector _newtrack_0 0.124))
          (Rendering|Material|SetVectorParameterValueonMaterials _returnvalue_2 "TopS_Em03_Offset_VehicleSide" (Math|Vector|MakeVector _newtrack_0)))))
    (:then_1
      (Rendering|Material|SetScalarParameterValueonMaterials _returnvalue_2 "TopS Emissive_01-Brightness-TBP" _returnvalue)
      (Rendering|Material|SetScalarParameterValueonMaterials _returnvalue_2 "TopS Emissive_01_Offset----TBP" _returnvalue)
      (Rendering|Material|SetScalarParameterValueonMaterials _returnvalue_2 "TopS Emissive_03-Brightness-TBP" _returnvalue)
      (Rendering|Material|SetScalarParameterValueonMaterials _returnvalue_2 "TopS Emissive_03_Offset----TBP" _returnvalue))))

(fn ConstructionScript)
