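; Custom item wrap Blueprint for the TNTina cosmetic wrap that makes its materials react when the weapon fires.
; On each shot it writes the current game time into two scalar parameters on the wrapped mesh components' materials.
; It also sets a vector seed from a random unit vector times a scale variable, giving materials a timestamp and per-shot variation.
; The construction script is empty.
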(event EventOnWeaponFired
  (bind _returnvalue (Utilities|Time|GetGameTimeinSeconds))
  (bind _returnvalue_2 (CustomItemWrap|GetWrappedMeshComponents))
  (Rendering|Material|SetScalarParameterValueonMaterials _returnvalue_2 "TNT_GameTimeFromBP" _returnvalue)
  (Rendering|Material|SetScalarParameterValueonMaterials _returnvalue_2 "TNT_GameTimeFromBP2" _returnvalue)
  (Rendering|Material|SetVectorParameterValueonMaterials _returnvalue_2 "TNT_Vec3SeedFromBP" (* (Math|Random|RandomUnitVector) (Variables|Default|GetRandomVectorScale))))

(fn ConstructionScript)
