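; Cosmetic weapon wrap Blueprint that shifts a material blend on wrapped meshes according to the player's health.
; Pawn-set and weapon events read health after short delays, while the front end animates it with a timeline.
; DamageAmount maps health to a blend offset, passes a stored game time to an emissive parameter, and prints debug text.
; Damage and heal stat watchers have empty callbacks, and the Weapon_Offsets helper is never called here.
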(event EventBeginPlay
  (if (CustomCosmeticModifier|IsinFrontEnd)
    (bind (_direction _newtrack_0) (Timeline|Timeline PlayFromStart)
      (:NewTrack_1
        (Variables|Default|SetHealth _newtrack_0)
        (Variables|Default|SetCurent%20GameTime (Utilities|Time|GetGameTimeinSeconds))
        (CallFunction|DamageAmount)))
    (else
      (Utilities|FlowControl|Delay 0.1)
      (Utilities|FlowControl|Delay 0.05)
      (Variables|Default|SetHealth (FortPawn|GetHealth (CustomCharacter|GetAssociatedPlayerPawn)))
      (CallFunction|DamageAmount))))

(event EventOnWeaponChanged
  (params (NewWeapon :type Fort%20Weapon%20Object%20Reference) (OldWeapon :type Fort%20Weapon%20Object%20Reference))
  (Utilities|FlowControl|Delay 0.05)
  (Variables|Default|SetHealth (FortPawn|GetHealth (CustomCharacter|GetAssociatedPlayerPawn)))
  (CallFunction|DamageAmount))

(event EventOnNewWeapons
  (params (WeaponsAsItems :type Array%20of%20Fort%20Item%20Object%20References))
  (Utilities|FlowControl|Delay 0.01)
  (Utilities|FlowControl|Delay 0.05)
  (Variables|Default|SetHealth (FortPawn|GetHealth (CustomCharacter|GetAssociatedPlayerPawn)))
  (CallFunction|DamageAmount))

(event EventOnNewPlayerWeaponType
  (params (NewType :type EFortItemType%20Enum :default "WorldItem") (OldType :type EFortItemType%20Enum :default "WorldItem"))
  (Utilities|FlowControl|Delay 0.01)
  (Utilities|FlowControl|Delay 0.05)
  (Variables|Default|SetHealth (FortPawn|GetHealth (CustomCharacter|GetAssociatedPlayerPawn)))
  (CallFunction|DamageAmount))

(event EventOnPawnSet
  (bind _returnvalue (CustomCharacter|GetAssociatedPlayerPawn))
  (Utilities|FlowControl|Sequence
    (:then_0
      (Cosmetics|WatchCosmeticStat _returnvalue "(TagName=\"StatManager.DamageReceived\")" 0 false
        (:OnStatChanged
          _)))
    (:then_1
      (Cosmetics|WatchCosmeticStat _returnvalue "(TagName=\"StatManager.GainedHealthTimes\")" 0 false
        (:OnStatChanged
          _))))
  (Utilities|FlowControl|Delay 0.1)
  (Variables|Default|SetHealth (FortPawn|GetHealth (CustomCharacter|GetAssociatedPlayerPawn)))
  (Variables|Default|SetCurent%20GameTime (Utilities|Time|GetGameTimeinSeconds))
  (CallFunction|DamageAmount))

(event Custom|DamageAmount
  (bind _health (Variables|Default|GetHealth))
  (bind _returnvalue (/ _health 1e+02))
  (bind _returnvalue_2 (Math|Float|Lerp -0.498 0.41 _returnvalue))
  (bind _returnvalue_3 (CustomItemWrap|GetWrappedMeshComponents))
  (Rendering|Material|SetScalarParameterValueonMaterials _returnvalue_3 "TopS_01_Blend_Offset---End" _returnvalue_2)
  (Rendering|Material|SetScalarParameterValueonMaterials _returnvalue_3 "TopS Emissive_01-Brightness-TBP" (Variables|Default|GetCurent%20GameTime))
  (Variables|Default|SetPosition_Data _returnvalue_2)
  (Development|PrintText (Utilities|Text|ToText(Float) _returnvalue) true true "(R=0.610496,G=0.026241,B=0.027321,A=1.000000)" 5.0)
  (Development|PrintText (Utilities|Text|ToText(Float) (DisplayHelper|RoundFloat _health)) true true "(R=0.201821,G=0.560810,B=0.610496,A=1.000000)" 5.0))

(fn ConstructionScript)

(fn Weapon_Offsets
  (params (Offset_End_Name :type Name) (Offset_End_Value :type Float%20%28double-precision%29) (Offset_Start_Name :type Name) (Offset_State_Value :type Float%20%28double-precision%29))
  (bind _returnvalue (CustomItemWrap|GetWrappedMeshComponents))
  (Rendering|Material|SetVectorParameterValueonMaterials _returnvalue Offset_End_Name (Math|Vector|MakeVector 0.0 0.0 Offset_End_Value))
  (Rendering|Material|SetVectorParameterValueonMaterials _returnvalue Offset_Start_Name (Math|Vector|MakeVector 0.0 0.0 Offset_State_Value)))
