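; Item wrap modifier that builds up emissive glow on the wrapped weapon as it fires, with a step cap set at start from front-end state and clip size.
; Each shot increments the step count, which decays toward zero once 5.5 seconds pass without firing, and weapon changes reset it.
; The step ratio drives brightness on two emissive layers, with game time written to timestamp parameters for material-side animation.
; A third layer's timestamp updates when the second reaches full brightness.
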(event EventBeginPlay
  (Utilities|FlowControl|Delay 0.2)
  (Utilities|FlowControl|Sequence
    (:then_0
      (if (CustomCosmeticModifier|IsinFrontEnd)
        _
        (elif (<= (FortWeapon|GetBulletsPerClip (Variables|/Script/FortniteGame.FortPawn|GetCurrentWeapon (CustomCharacter|GetAssociatedPlayerPawn))) 3)
          _
          (else
            (Variables|Default|SetMaxSteps 9))))
      (Variables|Default|SetMaxSteps 5))
    (:then_1
      (Utilities|FlowControl|Sequence
        (:then_1
          (Variables|Default|SetStepCount 0))))))

(event EventOnNewWeapons
  (params (WeaponsAsItems :type Array%20of%20Fort%20Item%20Object%20References))
  (Utilities|FlowControl|Delay 0.05)
  (Utilities|FlowControl|Sequence
    (:then_1
      (Variables|Default|SetStepCount 0))))

(event EventOnNewPlayerWeaponType
  (params (NewType :type EFortItemType%20Enum :default "WorldItem") (OldType :type EFortItemType%20Enum :default "WorldItem"))
  (Utilities|FlowControl|Delay 0.05)
  (Utilities|FlowControl|Sequence
    (:then_1
      (Variables|Default|SetStepCount 0))))

(event EventOnWeaponChanged
  (params (NewWeapon :type Fort%20Weapon%20Object%20Reference) (OldWeapon :type Fort%20Weapon%20Object%20Reference))
  (Utilities|FlowControl|Delay 0.05)
  (Utilities|FlowControl|Sequence
    (:then_1
      (Variables|Default|SetStepCount 0))))

(event EventOnWeaponFired
  (bind _returnvalue (CustomItemWrap|GetWrappedMeshComponents))
  (bind _stepcount (Variables|Default|GetStepCount))
  (bind _returnvalue_2 (Math|Conversions|ToFloat(Integer) _stepcount))
  (bind _maxsteps (Variables|Default|GetMaxSteps))
  (bind _returnvalue_3 (Math|Conversions|ToFloat(Integer) _maxsteps))
  (bind _returnvalue_4 (/ _returnvalue_2 _returnvalue_3))
  (bind _returnvalue_5 (Math|Float|Lerp -0.2 1.0 _returnvalue_4))
  (bind _returnvalue_6 (CustomItemWrap|GetWrappedMeshComponents))
  (bind _returnvalue_7 (Math|Float|Clamp(Float) _returnvalue_5))
  (bind _returnvalue_8 (Math|Float|Lerp 0.0 2.0 _returnvalue_7))
  (bind _returnvalue_9 (CustomItemWrap|GetWrappedMeshComponents))
  (Variables|Default|SetStepCount (Math|Integer|Min(Integer) (+ (Math|Float|Truncate (Math|Float|Lerp (Math|Conversions|ToFloat(Integer) (Variables|Default|GetStepCount)) 0.0 (Math|Float|Clamp(Float) (/ (- (Utilities|Time|GetGameTimeinSeconds) (Variables|Default|Getwait_timer)) 1.0)))) 1) (Variables|Default|GetMaxSteps)))
  (Rendering|Material|SetScalarParameterValueonMaterials _returnvalue_6 "TopS Emissive_01-Brightness-TBP" (Utilities|Time|GetGameTimeinSeconds))
  (Rendering|Material|SetScalarParameterValueonMaterials _returnvalue_6 "TopS Emissive_01_Brightness---Start" (Math|Float|Lerp 0.3 1.5 _returnvalue_4))
  (Rendering|Material|SetScalarParameterValueonMaterials _returnvalue_9 "TopS Emissive_02-Brightness-TBP" (Utilities|Time|GetGameTimeinSeconds))
  (Rendering|Material|SetScalarParameterValueonMaterials _returnvalue_9 "TopS Emissive_02_Brightness---Start" _returnvalue_8)
  (Utilities|FlowControl|Sequence
    (:then_0
      (Variables|Default|Setwait_timer (+ (Utilities|Time|GetGameTimeinSeconds) 5.5)))
    (:then_1
      (if (== _returnvalue_8 2.0)
        (Rendering|Material|SetScalarParameterValueonMaterials _returnvalue "TopS Emissive_03-Brightness-TBP" (Utilities|Time|GetGameTimeinSeconds))
        (Rendering|Material|SetScalarParameterValueonMaterials _returnvalue "TopS Emissive_03_Brightness---Start" 1.5)))))

(fn ConstructionScript)
