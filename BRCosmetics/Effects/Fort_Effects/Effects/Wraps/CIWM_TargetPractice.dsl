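; Item wrap Blueprint for the Target Practice cosmetic that makes the wrapped item's material react to the player's eliminations.
; When the pawn is set, it casts the associated pawn to a Fort player pawn and watches the AthenaKills cosmetic stat.
; Each stat change writes the current game time into a scalar parameter on the wrapped mesh materials, presumably so the shader can time an effect from that moment.
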(event EventOnPawnSet
  (bind _asfort_player_pawn (Utilities|Casting|CastToFortPlayerPawn (CustomCharacter|GetAssociatedPlayerPawn)))
  (Cosmetics|WatchCosmeticStat _asfort_player_pawn "(TagName=\"StatManager.AthenaKills\")"
    (:OnStatChanged
      (Rendering|Material|SetScalarParameterValueonMaterials (CustomItemWrap|GetWrappedMeshComponents) "TP_GameTimeFromBP" (Utilities|Time|GetGameTimeinSeconds)))))

(event EventBeginPlay
  (|CallParentFunction))

(event EventOnWeaponFired)

(fn ConstructionScript)
