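; Weapon wrap Blueprint that advances a material "crack level" effect as the player fires, resetting it after reloads.
; Weapon change and spawn events cache clip size and ammo and bind reload start; firing updates them and reruns the crack logic.
; Magazine fraction drives material offsets, while zero- and one-round clips use six-step shot counters.
; Game-time stamps and random colors from three presets drive emissive flashes; a debug PrintText remains in the magazine branch.
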(event EventBeginPlay
  (bind _returnvalue (CustomCharacter|GetAssociatedPlayerPawn))
  (bind _currentweapon (Variables|/Script/FortniteGame.FortPawn|GetCurrentWeapon _returnvalue))
  (Utilities|FlowControl|Delay 0.1)
  (bind _currentweapon_2 (Variables|/Script/FortniteGame.FortPawn|GetCurrentWeapon _returnvalue)
    (:then
      (EventDispatchers|BindEventtoOnReloadStarted _currentweapon_2 (delegate Custom|ReloadStarted))
      (Variables|Default|Setper%20Clip (FortWeapon|GetBulletsPerClip _currentweapon))
      (Variables|Default|SetCurent%20GameTime -99999999999999.0)
      (Variables|Default|SetAmmo%20Count (FortWeapon|GetMagazineAmmoCount _currentweapon))
      (Variables|Default|SetAmmo-%200%20 (Variables|Default|GetAmmo-%200%20))
      (Variables|Default|SetAmmo-1-2 (Variables|Default|GetAmmo-1-2))
      (CallFunction|CrackLevel))
    (:else)))

(event EventOnWeaponChanged
  (params (NewWeapon :type Fort%20Weapon%20Object%20Reference) (OldWeapon :type Fort%20Weapon%20Object%20Reference))
  (bind _returnvalue (CustomCharacter|GetAssociatedPlayerPawn))
  (bind _currentweapon (Variables|/Script/FortniteGame.FortPawn|GetCurrentWeapon _returnvalue))
  (bind _currentweapon_2 (Variables|/Script/FortniteGame.FortPawn|GetCurrentWeapon _returnvalue)
    (:then
      (EventDispatchers|BindEventtoOnReloadStarted _currentweapon_2 (delegate Custom|ReloadStarted))
      (Variables|Default|Setper%20Clip (FortWeapon|GetBulletsPerClip _currentweapon))
      (Variables|Default|SetCurent%20GameTime -99999999999999.0)
      (Variables|Default|SetAmmo%20Count (FortWeapon|GetMagazineAmmoCount _currentweapon))
      (Variables|Default|SetAmmo-%200%20 (Variables|Default|GetAmmo-%200%20))
      (Variables|Default|SetAmmo-1-2 (Variables|Default|GetAmmo-1-2))
      (CallFunction|CrackLevel))
    (:else)))

(event EventOnPawnSet
  (bind _returnvalue (CustomCharacter|GetAssociatedPlayerPawn))
  (bind _currentweapon (Variables|/Script/FortniteGame.FortPawn|GetCurrentWeapon _returnvalue))
  (bind _currentweapon_2 (Variables|/Script/FortniteGame.FortPawn|GetCurrentWeapon _returnvalue)
    (:then
      (EventDispatchers|BindEventtoOnReloadStarted _currentweapon_2 (delegate Custom|ReloadStarted))
      (Variables|Default|Setper%20Clip (FortWeapon|GetBulletsPerClip _currentweapon))
      (Variables|Default|SetCurent%20GameTime -99999999999999.0)
      (Variables|Default|SetAmmo%20Count (FortWeapon|GetMagazineAmmoCount _currentweapon))
      (Variables|Default|SetAmmo-%200%20 (Variables|Default|GetAmmo-%200%20))
      (Variables|Default|SetAmmo-1-2 (Variables|Default|GetAmmo-1-2))
      (CallFunction|CrackLevel))
    (:else)))

(event EventOnWeaponFired
  (bind _per_clip (Variables|Default|Getper%20Clip))
  (bind _output_get (Variables|Default|SetAmmo%20Count (FortWeapon|GetMagazineAmmoCount (Variables|/Script/FortniteGame.FortPawn|GetCurrentWeapon (CustomCharacter|GetAssociatedPlayerPawn)))))
  (Utilities|FlowControl|Delay 0.001)
  (Variables|Default|SetAmmo%20Count (Math|Integer|Clamp(Integer) (- _output_get 1) 0 (Variables|Default|Getper%20Clip)))
  (Variables|Default|SetCurent%20GameTime (Utilities|Time|GetGameTimeinSeconds))
  (Utilities|FlowControl|Sequence
    (:then_0
      (if (== _per_clip 0)
        (Variables|Default|SetAmmo-%200%20 (+ (Variables|Default|GetAmmo-%200%20) 1.0))))
    (:then_1
      (if (< _per_clip 2)
        (Variables|Default|SetAmmo-1-2 (+ (Variables|Default|GetAmmo-1-2) 1.0))))
    (:then_2
      (CallFunction|CrackLevel))))

(event EventOnNewWeapons
  (params (WeaponsAsItems :type Array%20of%20Fort%20Item%20Object%20References))
  (bind _returnvalue (CustomCharacter|GetAssociatedPlayerPawn))
  (bind _currentweapon (Variables|/Script/FortniteGame.FortPawn|GetCurrentWeapon _returnvalue))
  (Utilities|FlowControl|Delay 0.01)
  (Variables|Default|SetAmmo-%200%20 0.0)
  (Variables|Default|SetAmmo-1-2 0.0)
  (bind _currentweapon_2 (Variables|/Script/FortniteGame.FortPawn|GetCurrentWeapon _returnvalue)
    (:then
      (EventDispatchers|BindEventtoOnReloadStarted _currentweapon_2 (delegate Custom|ReloadStarted))
      (Variables|Default|Setper%20Clip (FortWeapon|GetBulletsPerClip _currentweapon))
      (Variables|Default|SetCurent%20GameTime -99999999999999.0)
      (Variables|Default|SetAmmo%20Count (FortWeapon|GetMagazineAmmoCount _currentweapon))
      (Variables|Default|SetAmmo-%200%20 (Variables|Default|GetAmmo-%200%20))
      (Variables|Default|SetAmmo-1-2 (Variables|Default|GetAmmo-1-2))
      (CallFunction|CrackLevel))
    (:else)))

(event EventOnNewPlayerWeaponType
  (params (NewType :type EFortItemType%20Enum :default "WorldItem") (OldType :type EFortItemType%20Enum :default "WorldItem"))
  (bind _returnvalue (CustomCharacter|GetAssociatedPlayerPawn))
  (bind _currentweapon (Variables|/Script/FortniteGame.FortPawn|GetCurrentWeapon _returnvalue))
  (Utilities|FlowControl|Delay 0.01)
  (Variables|Default|SetAmmo-%200%20 0.0)
  (Variables|Default|SetAmmo-1-2 0.0)
  (bind _currentweapon_2 (Variables|/Script/FortniteGame.FortPawn|GetCurrentWeapon _returnvalue)
    (:then
      (EventDispatchers|BindEventtoOnReloadStarted _currentweapon_2 (delegate Custom|ReloadStarted))
      (Variables|Default|Setper%20Clip (FortWeapon|GetBulletsPerClip _currentweapon))
      (Variables|Default|SetCurent%20GameTime -99999999999999.0)
      (Variables|Default|SetAmmo%20Count (FortWeapon|GetMagazineAmmoCount _currentweapon))
      (Variables|Default|SetAmmo-%200%20 (Variables|Default|GetAmmo-%200%20))
      (Variables|Default|SetAmmo-1-2 (Variables|Default|GetAmmo-1-2))
      (CallFunction|CrackLevel))
    (:else)))

(event Custom|Reload%20weapon
  (bind _returnvalue (CustomItemWrap|GetWrappedMeshComponents))
  (Rendering|Material|SetScalarParameterValueonMaterials _returnvalue "TopS_01_Blend_Offset---End" -0.3)
  (Rendering|Material|SetScalarParameterValueonMaterials _returnvalue "TopS Emissive_02_Offset---End" -0.3)
  (Rendering|Material|SetScalarParameterValueonMaterials _returnvalue "TopS Emissive_01_Offset---End" -0.3)
  (Rendering|Material|SetScalarParameterValueonMaterials _returnvalue "TopS Emissive_03-Brightness-TBP" (Utilities|Time|GetGameTimeinSeconds))
  (CallFunction|ColorRandom "TopS Emissive_03_Color-")
  (Variables|Default|SetAmmo-%200%20 0.0))

(event Custom|ReloadStarted
  (params (ReloadTime :type Float%20%28single-precision%29) (ReloadType :type EFortWeaponReloadType%20Enum :default "ReloadWholeClip"))
  (bind _returnvalue (CustomItemWrap|GetWrappedMeshComponents))
  (if (< (Variables|Default|Getper%20Clip) 2)
    (if (>= (Variables|Default|GetAmmo-1-2) 6.0)
      (Variables|Default|SetAmmo-1-2 0.0))
    (else
      (Variables|Default|SetAmmo%20Count (FortWeapon|GetMagazineAmmoCount (Variables|/Script/FortniteGame.FortPawn|GetCurrentWeapon (CustomCharacter|GetAssociatedPlayerPawn))))))
  (Utilities|FlowControl|Delay ReloadTime)
  (Rendering|Material|SetScalarParameterValueonMaterials _returnvalue "TopS_01_Blend_Offset---End" -0.3)
  (Rendering|Material|SetScalarParameterValueonMaterials _returnvalue "TopS Emissive_02_Offset---End" -0.3)
  (Rendering|Material|SetScalarParameterValueonMaterials _returnvalue "TopS Emissive_01_Offset---End" -0.3)
  (Rendering|Material|SetScalarParameterValueonMaterials _returnvalue "TopS Emissive_03-Brightness-TBP" (Utilities|Time|GetGameTimeinSeconds))
  (CallFunction|ColorRandom "TopS Emissive_03_Color-")
  (Variables|Default|SetAmmo-%200%20 0.0))

(event Custom|Crack%20level%20
  (bind _returnvalue (CustomItemWrap|GetWrappedMeshComponents))
  (bind _returnvalue_2 (CustomItemWrap|GetWrappedMeshComponents))
  (bind _per_clip (Variables|Default|Getper%20Clip))
  (bind _returnvalue_3 (Math|Conversions|ToFloat(Integer) _per_clip))
  (bind _ammo_count (Variables|Default|GetAmmo%20Count))
  (bind _returnvalue_4 (Math|Conversions|ToFloat(Integer) _ammo_count))
  (bind _ammo_1_2 (Variables|Default|GetAmmo-1-2))
  (bind _returnvalue_5 (/ _ammo_1_2 6.0))
  (bind _returnvalue_6 (Math|Float|Lerp -0.3 0.6 _returnvalue_5))
  (bind _returnvalue_7 (/ _returnvalue_4 _returnvalue_3))
  (bind _returnvalue_8 (Math|Float|Lerp 0.6 -0.3 _returnvalue_7))
  (bind _ammo__0 (Variables|Default|GetAmmo-%200%20))
  (bind _returnvalue_9 (/ _ammo__0 6.0))
  (bind _returnvalue_10 (Math|Float|Lerp -0.3 0.6 _returnvalue_9))
  (bind _returnvalue_11 (CustomItemWrap|GetWrappedMeshComponents))
  (if (< (Variables|Default|Getper%20Clip) 2)
    (if (== (Variables|Default|Getper%20Clip) 0)
      (Rendering|Material|SetScalarParameterValueonMaterials _returnvalue_11 "TopS_01_Blend_Offset---End" _returnvalue_10)
      (Rendering|Material|SetScalarParameterValueonMaterials _returnvalue_11 "TopS Emissive_02_Offset---End" _returnvalue_10)
      (Rendering|Material|SetScalarParameterValueonMaterials _returnvalue_11 "TopS Emissive_01_Offset---End" _returnvalue_10)
      (Rendering|Material|SetScalarParameterValueonMaterials _returnvalue_11 "TopS Emissive_02-Brightness-TBP" (Variables|Default|GetCurent%20GameTime))
      (Rendering|Material|SetVectorParameterValueonMaterials _returnvalue_11 "TopS_Em02_Offset__Master" (Math|Vector|MakeVector 0.0 (Math|Random|RandomFloat)))
      (CallFunction|ColorRandom "TopS Emissive_02_Color-")
      (Variables|Default|SetPosition_Data _returnvalue_10)
      (if (== (Variables|Default|GetAmmo-%200%20) 7.0)
        (Utilities|FlowControl|Delay 0.4)
        (CallFunction|ReloadWeapon))
      (else
        (Rendering|Material|SetScalarParameterValueonMaterials _returnvalue_2 "TopS_01_Blend_Offset---End" _returnvalue_6)
        (Rendering|Material|SetScalarParameterValueonMaterials _returnvalue_2 "TopS Emissive_02_Offset---End" _returnvalue_6)
        (Rendering|Material|SetScalarParameterValueonMaterials _returnvalue_2 "TopS Emissive_01_Offset---End" _returnvalue_6)
        (Rendering|Material|SetScalarParameterValueonMaterials _returnvalue_2 "TopS Emissive_02-Brightness-TBP" (Variables|Default|GetCurent%20GameTime))
        (Rendering|Material|SetVectorParameterValueonMaterials _returnvalue_2 "TopS_Em02_Offset__Master" (Math|Vector|MakeVector 0.0 (Math|Random|RandomFloat)))
        (CallFunction|ColorRandom "TopS Emissive_02_Color-")
        (Variables|Default|SetPosition_Data _returnvalue_6)))
    (else
      (Rendering|Material|SetScalarParameterValueonMaterials _returnvalue "TopS_01_Blend_Offset---End" _returnvalue_8)
      (Rendering|Material|SetScalarParameterValueonMaterials _returnvalue "TopS Emissive_02_Offset---End" _returnvalue_8)
      (Rendering|Material|SetScalarParameterValueonMaterials _returnvalue "TopS Emissive_01_Offset---End" _returnvalue_8)
      (Rendering|Material|SetScalarParameterValueonMaterials _returnvalue "TopS Emissive_02-Brightness-TBP" (Variables|Default|GetCurent%20GameTime))
      (Rendering|Material|SetVectorParameterValueonMaterials _returnvalue "TopS_Em02_Offset__Master" (Math|Vector|MakeVector 0.0 (Math|Random|RandomFloat)))
      (CallFunction|ColorRandom "TopS Emissive_02_Color-")
      (Variables|Default|SetPosition_Data _returnvalue_8)
      (Development|PrintText (Utilities|Text|ToText(Float) _returnvalue_8) true true "(R=0.610496,G=0.026241,B=0.027321,A=1.000000)"))))

(fn ColorRandom
  (params (ParameterName :type Name))
  (bind _returnvalue (CustomItemWrap|GetWrappedMeshComponents))
  (switch int (Math|Random|RandomIntegerinRange 0 2)
    (:0
      (Rendering|Material|SetColorParameterValueonMaterials _returnvalue ParameterName (* "(R=5.000000,G=0.207200,B=0.306678,A=1.000000)" "(R=1.000000,G=1.000000,B=1.000000,A=1.000000)")))
    (:1
      (Rendering|Material|SetColorParameterValueonMaterials _returnvalue ParameterName (* "(R=0.160732,G=5.000000,B=0.485944,A=1.000000)" "(R=1.000000,G=1.000000,B=1.000000,A=1.000000)")))
    (:2
      (Rendering|Material|SetColorParameterValueonMaterials _returnvalue ParameterName (* "(R=5.000000,G=4.082335,B=0.203241,A=1.000000)" "(R=1.000000,G=1.000000,B=1.000000,A=1.000000)")))))

(fn ConstructionScript)
