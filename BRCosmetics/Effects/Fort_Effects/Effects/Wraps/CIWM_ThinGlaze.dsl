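; Custom item wrap Blueprint for ThinGlaze that animates the wrap's material when the weapon fires.
; A timeline lerps two layers of TopS offset parameters on wrapped meshes, with separate ranges for weapon, backpack, pickaxe, character and vehicle variants.
; The Pause flag marks a running animation: firing starts it, or rewinds to a fixed time once its value peaks.
; Finishing clears the flag, Kill stops the timeline, and each shot prints a debug string.
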(event EventOnWeaponFired
  (bind _returnvalue (CustomItemWrap|GetWrappedMeshComponents))
  (bind _returnvalue_2 (CustomItemWrap|GetWrappedMeshComponents))
  (Development|PrintString "FireTest" true true "(R=0.000000,G=1.000000,B=0.041611,A=1.000000)")
  (if (Variables|Default|GetPause)
    (if (>= (Variables|Default|GetFloat_Value) 1.0)
      (bind (_direction _newtrack_0) (Timeline|Timeline SetNewTime :NewTime 0.75)
        (:Update
          (Variables|Default|SetFloat_Value _newtrack_0)
          (Rendering|Material|SetVectorParameterValueonMaterials _returnvalue_2 "TopS_01_OffsetS_Weapon" (Math|Vector|MakeVector 0.0 (Math|Float|Lerp 0.25 0.0 _newtrack_0)))
          (Rendering|Material|SetVectorParameterValueonMaterials _returnvalue_2 "TopS_01_OffsetS_Backpack" (Math|Vector|MakeVector (Math|Float|Lerp 0.7 0.1 _newtrack_0)))
          (Rendering|Material|SetVectorParameterValueonMaterials _returnvalue_2 "TopS_01_OffsetS_Pickaxe" (Math|Vector|MakeVector (Math|Float|Lerp 0.45 0.0 _newtrack_0)))
          (Rendering|Material|SetVectorParameterValueonMaterials _returnvalue_2 "TopS_01_OffsetS_Character" (Math|Vector|MakeVector (Math|Float|Lerp 0.45 0.0 _newtrack_0)))
          (Rendering|Material|SetVectorParameterValueonMaterials _returnvalue_2 "TopS_01_OffsetS_Vehicle" (Math|Vector|MakeVector (Math|Float|Lerp 0.128 -0.1 _newtrack_0)))
          (Rendering|Material|SetVectorParameterValueonMaterials _returnvalue_2 "TopS_01_OffsetS_VehicleSide" (Math|Vector|MakeVector (Math|Float|Lerp 0.15 -0.1 _newtrack_0)))
          (Rendering|Material|SetVectorParameterValueonMaterials _returnvalue "TopS_02_OffsetS_Weapon" (Math|Vector|MakeVector 0.0 (Math|Float|Lerp 0.48 0.1 _newtrack_0)))
          (Rendering|Material|SetVectorParameterValueonMaterials _returnvalue "TopS_02_OffsetS_Backpack" (Math|Vector|MakeVector (Math|Float|Lerp 0.52 0.0 _newtrack_0)))
          (Rendering|Material|SetVectorParameterValueonMaterials _returnvalue "TopS_02_OffsetS_Pickaxe" (Math|Vector|MakeVector (Math|Float|Lerp 0.49 0.0 _newtrack_0)))
          (Rendering|Material|SetVectorParameterValueonMaterials _returnvalue "TopS_02_OffsetS_Character" (Math|Vector|MakeVector (Math|Float|Lerp 0.5 0.0 _newtrack_0)))
          (Rendering|Material|SetVectorParameterValueonMaterials _returnvalue "TopS_02_OffsetS_Vehicle" (Math|Vector|MakeVector (Math|Float|Lerp 0.2 0.0 _newtrack_0)))
          (Rendering|Material|SetVectorParameterValueonMaterials _returnvalue "TopS_02_OffsetS_VehicleSide" (Math|Vector|MakeVector (Math|Float|Lerp 0.18 0.0 _newtrack_0))))
        (:Finished
          (Variables|Default|SetPause false))))
    (else
      (Variables|Default|SetPause true)
      (Timeline|Timeline PlayFromStart))))

(event Custom|Kill
  (bind _returnvalue (CustomItemWrap|GetWrappedMeshComponents))
  (bind _returnvalue_2 (CustomItemWrap|GetWrappedMeshComponents))
  (Timeline|Timeline Stop))

(fn ConstructionScript)
