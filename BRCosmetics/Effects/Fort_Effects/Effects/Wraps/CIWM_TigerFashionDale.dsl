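; Tiger Fashion Dale weapon wrap modifier that makes wrapped weapons react to the player's elimination count in three tiers.
; The watched AthenaKills stat, or a timeline in the front end, sets a tier index that DisplayKillCount turns into emissive visibility, flame color, warp and mask texture.
; Weapon changes reapply the look across wrapped mesh materials.
; Firing writes game time, after a short delay, into an emissive material parameter.
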(event EventBeginPlay
  (if (CustomCosmeticModifier|IsinFrontEnd)
    (bind (_direction _normalizedfloat) (Timeline|Timeline Play)
      (:Update
        (Variables|Default|SetMskNumber (- (Math|Integer|Clamp(Integer) (Math|Float|Truncate _normalizedfloat) 0 3) 1))
        (CallFunction|DisplayKillCount)))))

(event EventOnPawnSet
  (bind (_statname _statvalue) (Cosmetics|WatchCosmeticStat (CustomCharacter|GetAssociatedPlayerPawn) "(TagName=\"StatManager.AthenaKills\")")
    (:OnStatChanged
      (Variables|Default|SetMskNumber (- (Math|Integer|Clamp(Integer) _statvalue 0 3) 1))
      (CallFunction|DisplayKillCount))))

(event EventOnNewWeapons
  (params (WeaponsAsItems :type Array%20of%20Fort%20Item%20Object%20References))
  (CallFunction|DisplayKillCount))

(event EventOnNewPlayerWeaponType
  (params (NewType :type EFortItemType%20Enum :default "WorldItem") (OldType :type EFortItemType%20Enum :default "WorldItem"))
  (CallFunction|DisplayKillCount))

(event EventOnWeaponFired
  (Utilities|FlowControl|Delay 0.1)
  (Rendering|Material|SetScalarParameterValueonMaterials (CustomItemWrap|GetWrappedMeshComponents) "TopS Emissive_01-Brightness-TBP" (Utilities|Time|GetGameTimeinSeconds)))

(event Custom|DisplayKillCount
  (bind _returnvalue (CustomItemWrap|GetWrappedMeshComponents))
  (if (>= (Variables|Default|GetMskNumber) 0)
    (Utilities|FlowControl|Sequence
      (:then_0
        (Rendering|Material|SetScalarParameterValueonMaterials _returnvalue "TopS Emissive_01_Brightness---End" 1.0)
        (Rendering|Material|SetVectorParameterValueonMaterials _returnvalue "TopS Emissive_01_Color-" (Math|Conversions|ToVector(LinearColor) (Utilities|Array|Get(aref) (Variables|Default|GetFlameColors) (Math|Integer|Clamp(Integer) (Variables|Default|GetMskNumber) 0 2))))
        (Rendering|Material|SetVectorParameterValueonMaterials _returnvalue "TopS_02_Warp_UV_Panning" (Math|Vector|MakeVector (Utilities|Select :Index (Variables|Default|GetMskNumber) :Option%200 0.2 :Option%201 0.22 :Option%202 0.25)))
        (Rendering|Material|SetScalarParameterValueonMaterials _returnvalue "TopS_02_Warp_UV_Amount" (Utilities|Select :Index (Variables|Default|GetMskNumber) :Option%200 0.05 :Option%201 0.055 :Option%202 0.06))
        (CallFunction|MskTextureSwap (Utilities|Array|Get(acopy) (Variables|Default|GetMskTextures) (Math|Integer|Clamp(Integer) (Variables|Default|GetMskNumber) 0 2))))
      (:then_1
        _))
    (else
      (Rendering|Material|SetScalarParameterValueonMaterials _returnvalue "TopS Emissive_01_Brightness---End" 0.0)))
  (Utilities|FlowControl|DoOnce true
    (:Completed
      (CallFunction|MskTextureSwap (Utilities|Array|Get(acopy) (Variables|Default|GetMskTextures))))))

(fn MskTextureSwap
  (params (Msk_Texture :type Texture%20Object%20Reference))
  (for _array_element (CustomItemWrap|GetWrappedMeshComponents)
    (for _array_element_2 (Rendering|Material|GetMaterials _array_element)
      (bind _asmaterial_instance_dynamic (Utilities|Casting|CastToMaterialInstanceDynamic _array_element_2))
      (Rendering|Material|SetTextureParameterValue _asmaterial_instance_dynamic "TopS_02__Tex_Mask" Msk_Texture))))

(fn ConstructionScript)
