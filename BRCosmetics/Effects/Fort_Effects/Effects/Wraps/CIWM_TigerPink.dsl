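; Custom item wrap Blueprint for the TigerPink wrap that drives reactive material effects on wrapped meshes when the weapon fires.
; Each shot stamps game time into a top emissive brightness parameter, then after a brief delay a Selection switch cycles through four decal channels, A to D.
; Each channel stamps its own time parameter and holds a Pause flag for a half-second cooldown; a paused channel only advances the selection.
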(event EventOnWeaponFired
  (Rendering|Material|SetScalarParameterValueonMaterials (CustomItemWrap|GetWrappedMeshComponents) "TopS Emissive_01-Brightness-TBP" (Utilities|Time|GetGameTimeinSeconds))
  (Utilities|FlowControl|Delay 0.05)
  (switch int (Variables|Default|GetSelection)
    (:0
      (if (Variables|Default|GetPause)
        (Variables|Default|SetSelection 1)
        (else
          (Variables|Default|SetPause true)
          (Rendering|Material|SetScalarParameterValueonMaterials (CustomItemWrap|GetWrappedMeshComponents) "EDRec_TimeFrom_BP_DecalA" (Utilities|Time|GetGameTimeinSeconds))
          (Variables|Default|SetSelection 1)
          (Utilities|FlowControl|Delay 0.5)
          (Variables|Default|SetPause false))))
    (:1
      (if (Variables|Default|GetPause_0)
        (Variables|Default|SetSelection 2)
        (else
          (Variables|Default|SetPause_0 true)
          (Rendering|Material|SetScalarParameterValueonMaterials (CustomItemWrap|GetWrappedMeshComponents) "EDRec_TimeFrom_BP_DecalB" (Utilities|Time|GetGameTimeinSeconds))
          (Variables|Default|SetSelection 2)
          (Utilities|FlowControl|Delay 0.5)
          (Variables|Default|SetPause_0 false))))
    (:2
      (if (Variables|Default|GetPause_1)
        (Variables|Default|SetSelection 3)
        (else
          (Variables|Default|SetPause_1 true)
          (Rendering|Material|SetScalarParameterValueonMaterials (CustomItemWrap|GetWrappedMeshComponents) "EDRec_TimeFrom_BP_DecalC" (Utilities|Time|GetGameTimeinSeconds))
          (Variables|Default|SetSelection 3)
          (Utilities|FlowControl|Delay 0.5)
          (Variables|Default|SetPause_1 false))))
    (:3
      (if (Variables|Default|GetPause_2)
        (Variables|Default|SetSelection 0)
        (else
          (Variables|Default|SetPause_2 true)
          (Rendering|Material|SetScalarParameterValueonMaterials (CustomItemWrap|GetWrappedMeshComponents) "EDRec_TimeFrom_BP_DecalD" (Utilities|Time|GetGameTimeinSeconds))
          (Variables|Default|SetSelection 0)
          (Utilities|FlowControl|Delay 0.5)
          (Variables|Default|SetPause_2 false))))))

(fn ConstructionScript)
