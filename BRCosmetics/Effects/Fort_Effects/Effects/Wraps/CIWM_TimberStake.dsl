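; Custom item wrap Blueprint for the TimberStake wrap that reacts each time the weapon fires.
; Each shot advances a stored float by about a third, resetting to zero near one, and writes it to the decal flipbook frame parameter on the wrapped mesh materials.
; The same event restarts the Glow timeline, which lerps a top-layer emissive brightness parameter on those materials to animate a glow.
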(event EventOnWeaponFired
  (bind _returnvalue (CustomItemWrap|GetWrappedMeshComponents))
  (if (>= (Variables|Default|GetFloat_Value) 0.99)
    (Variables|Default|SetFloat_Value 0.0)
    (else
      (Variables|Default|SetFloat_Value (+ (Variables|Default|GetFloat_Value) 0.33))))
  (Rendering|Material|SetScalarParameterValueonMaterials _returnvalue "FDecalV2_FlipBook_Frame #" (Variables|Default|GetFloat_Value))
  (bind (_direction _newtrack_0) (Timeline|Glow PlayFromStart)
    (:Update
      (Rendering|Material|SetScalarParameterValueonMaterials _returnvalue "TopS_02 Emissive Brightness" (Math|Float|Lerp 0.0 1e+01 _newtrack_0)))))

(fn ConstructionScript)
