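; Wrap modifier Blueprint for the TopicProud cosmetic wrap that rotates a material decal whenever the weapon fires.
; Each shot fetches the wrapped mesh components and restarts a timeline, whose every update adds the scaled track value to a stored float.
; That accumulated value drives the DecalC rotation scalars for weapon, both vehicle variants, pickaxe, character and backpack.
; Rotation carries over between shots, and its per-update increment depends on frame rate.
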(event EventOnWeaponFired
  (bind _returnvalue (CustomItemWrap|GetWrappedMeshComponents))
  (bind (_direction _newtrack_0) (Timeline|Timeline PlayFromStart)
    (:Update
      (bind _output_get (Variables|Default|SetFloat_Value (+ (Variables|Default|GetFloat_Value) (* _newtrack_0 0.1))))
      (Rendering|Material|SetScalarParameterValueonMaterials _returnvalue "EDRec_DecalC_Rotation_Weapon" _output_get)
      (Rendering|Material|SetScalarParameterValueonMaterials _returnvalue "EDRec_DecalC_Rotation_Vehicle_Side" _output_get)
      (Rendering|Material|SetScalarParameterValueonMaterials _returnvalue "EDRec_DecalC_Rotation_Vehicle" _output_get)
      (Rendering|Material|SetScalarParameterValueonMaterials _returnvalue "EDRec_DecalC_Rotation_Pickaxe" _output_get)
      (Rendering|Material|SetScalarParameterValueonMaterials _returnvalue "EDRec_DecalC_Rotation_Character" _output_get)
      (Rendering|Material|SetScalarParameterValueonMaterials _returnvalue "EDRec_DecalC_Rotation_Backpack" _output_get))))

(fn ConstructionScript)
