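; Reactive weapon wrap modifier that drives emissive material parameters on wrapped meshes from firing and kills.
; Each shot stamps game time into the materials and raises a capped step count that scales glow brightness, decaying once a wait timer expires.
; Kill stat changes swap color curves and set a flag that hue-shifts emissive when weapon events reset the steps.
; BeginPlay picks the step cap by clip size and plays a front-end color timeline.
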(event EventBeginPlay
  (bind _returnvalue (CustomItemWrap|GetWrappedMeshComponents))
  (bind _stepcount (Variables|Default|GetStepCount))
  (bind _returnvalue_2 (Math|Conversions|ToFloat(Integer) _stepcount))
  (bind _maxsteps (Variables|Default|GetMaxSteps))
  (bind _returnvalue_3 (Math|Conversions|ToFloat(Integer) _maxsteps))
  (bind _returnvalue_4 (CustomItemWrap|GetWrappedMeshComponents))
  (bind _returnvalue_5 (CustomItemWrap|GetWrappedMeshComponents))
  (bind _returnvalue_6 (CustomItemWrap|GetWrappedMeshComponents))
  (bind _returnvalue_7 (/ _returnvalue_2 _returnvalue_3))
  (bind _kill (Variables|Default|Getkill))
  (Utilities|FlowControl|Delay 0.2)
  (Utilities|FlowControl|Sequence
    (:then_0
      (if (CustomCosmeticModifier|IsinFrontEnd)
        (Utilities|FlowControl|Sequence
          (:then_0
            (Timeline|Timeline Play
              (:NewTrack_0
                (Rendering|Material|SetScalarParameterValueonMaterials _returnvalue_6 "EDRec_Emissive_A_CC" 197.0)
                (Rendering|Material|SetScalarParameterValueonMaterials _returnvalue_6 "EDRec_Emissive_B_CC" 198.0)
                (Rendering|Material|SetScalarParameterValueonMaterials _returnvalue_6 "TopS Emissive_01_ColorCuve" 198.0)
                (Rendering|Material|SetScalarParameterValueonMaterials _returnvalue_6 "TopS Emissive_02_ColorCuve" 198.0)
                (Rendering|Material|SetScalarParameterValueonMaterials _returnvalue_6 "TopS Emissive_03_ColorCuve" 198.0))
              (:NewTrack_1
                (Rendering|Material|SetScalarParameterValueonMaterials _returnvalue_4 "EDRec_Emissive_A_CC" 194.0)
                (Rendering|Material|SetScalarParameterValueonMaterials _returnvalue_4 "EDRec_Emissive_B_CC" 195.0)
                (Rendering|Material|SetScalarParameterValueonMaterials _returnvalue_4 "TopS Emissive_01_ColorCuve" 195.0)
                (Rendering|Material|SetScalarParameterValueonMaterials _returnvalue_4 "TopS Emissive_02_ColorCuve" 195.0)
                (Rendering|Material|SetScalarParameterValueonMaterials _returnvalue_4 "TopS Emissive_03_ColorCuve" 195.0))))
          (:then_1
            _))
        (elif (<= (FortWeapon|GetBulletsPerClip (Variables|/Script/FortniteGame.FortPawn|GetCurrentWeapon (CustomCharacter|GetAssociatedPlayerPawn))) 3)
          _
          (else
            (Variables|Default|SetMaxSteps 9))))
      (Variables|Default|SetMaxSteps 5))
    (:then_1
      (Utilities|FlowControl|Sequence
        (:then_0
          (bind _output_get (Variables|Default|Setkill _kill))
          (if _output_get
            (Rendering|Material|SetScalarParameterValueonMaterials _returnvalue_5 "HueShift_EmissiveColor" 0.35)
            (Development|PrintString (Utilities|String|ToString(Boolean) _kill) true true "(R=1.000000,G=0.064978,B=0.000000,A=1.000000)")))
        (:then_1
          (Variables|Default|SetStepCount 0)
          (Rendering|Material|SetScalarParameterValueonMaterials _returnvalue_5 "TopS Emissive_02-Brightness-TBP" (Utilities|Time|GetGameTimeinSeconds))
          (Rendering|Material|SetScalarParameterValueonMaterials _returnvalue_5 "TopS Emissive_02_Brightness---Start" (Math|Float|Lerp 0.0 1.5 _returnvalue_7))
          (Rendering|Material|SetScalarParameterValueonMaterials _returnvalue "TopS Emissive_03-Brightness-TBP" (Utilities|Time|GetGameTimeinSeconds))
          (Rendering|Material|SetScalarParameterValueonMaterials _returnvalue "TopS Emissive_03_Brightness---Start" (Math|Float|Lerp 0.0 1.5 (Math|Float|Clamp(Float) (Math|Float|Lerp -0.5 1.0 _returnvalue_7))))
          (Variables|Default|Setwait_timer (+ (Utilities|Time|GetGameTimeinSeconds) 5.5)))))))

(event EventOnNewWeapons
  (params (WeaponsAsItems :type Array%20of%20Fort%20Item%20Object%20References))
  (bind _returnvalue (CustomItemWrap|GetWrappedMeshComponents))
  (bind _stepcount (Variables|Default|GetStepCount))
  (bind _returnvalue_2 (Math|Conversions|ToFloat(Integer) _stepcount))
  (bind _maxsteps (Variables|Default|GetMaxSteps))
  (bind _returnvalue_3 (Math|Conversions|ToFloat(Integer) _maxsteps))
  (bind _returnvalue_4 (CustomItemWrap|GetWrappedMeshComponents))
  (bind _returnvalue_5 (/ _returnvalue_2 _returnvalue_3))
  (bind _kill (Variables|Default|Getkill))
  (Utilities|FlowControl|Delay 0.05)
  (Utilities|FlowControl|Sequence
    (:then_0
      (bind _output_get (Variables|Default|Setkill _kill))
      (if _output_get
        (Rendering|Material|SetScalarParameterValueonMaterials _returnvalue_4 "HueShift_EmissiveColor" 0.35)
        (Development|PrintString (Utilities|String|ToString(Boolean) _kill) true true "(R=1.000000,G=0.064978,B=0.000000,A=1.000000)")))
    (:then_1
      (Variables|Default|SetStepCount 0)
      (Rendering|Material|SetScalarParameterValueonMaterials _returnvalue_4 "TopS Emissive_02-Brightness-TBP" (Utilities|Time|GetGameTimeinSeconds))
      (Rendering|Material|SetScalarParameterValueonMaterials _returnvalue_4 "TopS Emissive_02_Brightness---Start" (Math|Float|Lerp 0.0 1.5 _returnvalue_5))
      (Rendering|Material|SetScalarParameterValueonMaterials _returnvalue "TopS Emissive_03-Brightness-TBP" (Utilities|Time|GetGameTimeinSeconds))
      (Rendering|Material|SetScalarParameterValueonMaterials _returnvalue "TopS Emissive_03_Brightness---Start" (Math|Float|Lerp 0.0 1.5 (Math|Float|Clamp(Float) (Math|Float|Lerp -0.5 1.0 _returnvalue_5))))
      (Variables|Default|Setwait_timer (+ (Utilities|Time|GetGameTimeinSeconds) 5.5)))))

(event EventOnNewPlayerWeaponType
  (params (NewType :type EFortItemType%20Enum :default "WorldItem") (OldType :type EFortItemType%20Enum :default "WorldItem"))
  (bind _returnvalue (CustomItemWrap|GetWrappedMeshComponents))
  (bind _stepcount (Variables|Default|GetStepCount))
  (bind _returnvalue_2 (Math|Conversions|ToFloat(Integer) _stepcount))
  (bind _maxsteps (Variables|Default|GetMaxSteps))
  (bind _returnvalue_3 (Math|Conversions|ToFloat(Integer) _maxsteps))
  (bind _returnvalue_4 (CustomItemWrap|GetWrappedMeshComponents))
  (bind _returnvalue_5 (/ _returnvalue_2 _returnvalue_3))
  (bind _kill (Variables|Default|Getkill))
  (Utilities|FlowControl|Delay 0.05)
  (Utilities|FlowControl|Sequence
    (:then_0
      (bind _output_get (Variables|Default|Setkill _kill))
      (if _output_get
        (Rendering|Material|SetScalarParameterValueonMaterials _returnvalue_4 "HueShift_EmissiveColor" 0.35)
        (Development|PrintString (Utilities|String|ToString(Boolean) _kill) true true "(R=1.000000,G=0.064978,B=0.000000,A=1.000000)")))
    (:then_1
      (Variables|Default|SetStepCount 0)
      (Rendering|Material|SetScalarParameterValueonMaterials _returnvalue_4 "TopS Emissive_02-Brightness-TBP" (Utilities|Time|GetGameTimeinSeconds))
      (Rendering|Material|SetScalarParameterValueonMaterials _returnvalue_4 "TopS Emissive_02_Brightness---Start" (Math|Float|Lerp 0.0 1.5 _returnvalue_5))
      (Rendering|Material|SetScalarParameterValueonMaterials _returnvalue "TopS Emissive_03-Brightness-TBP" (Utilities|Time|GetGameTimeinSeconds))
      (Rendering|Material|SetScalarParameterValueonMaterials _returnvalue "TopS Emissive_03_Brightness---Start" (Math|Float|Lerp 0.0 1.5 (Math|Float|Clamp(Float) (Math|Float|Lerp -0.5 1.0 _returnvalue_5))))
      (Variables|Default|Setwait_timer (+ (Utilities|Time|GetGameTimeinSeconds) 5.5)))))

(event EventOnWeaponChanged
  (params (NewWeapon :type Fort%20Weapon%20Object%20Reference) (OldWeapon :type Fort%20Weapon%20Object%20Reference))
  (bind _returnvalue (CustomItemWrap|GetWrappedMeshComponents))
  (bind _stepcount (Variables|Default|GetStepCount))
  (bind _returnvalue_2 (Math|Conversions|ToFloat(Integer) _stepcount))
  (bind _maxsteps (Variables|Default|GetMaxSteps))
  (bind _returnvalue_3 (Math|Conversions|ToFloat(Integer) _maxsteps))
  (bind _returnvalue_4 (CustomItemWrap|GetWrappedMeshComponents))
  (bind _returnvalue_5 (/ _returnvalue_2 _returnvalue_3))
  (bind _kill (Variables|Default|Getkill))
  (Utilities|FlowControl|Delay 0.05)
  (Utilities|FlowControl|Sequence
    (:then_0
      (bind _output_get (Variables|Default|Setkill _kill))
      (if _output_get
        (Rendering|Material|SetScalarParameterValueonMaterials _returnvalue_4 "HueShift_EmissiveColor" 0.35)
        (Development|PrintString (Utilities|String|ToString(Boolean) _kill) true true "(R=1.000000,G=0.064978,B=0.000000,A=1.000000)")))
    (:then_1
      (Variables|Default|SetStepCount 0)
      (Rendering|Material|SetScalarParameterValueonMaterials _returnvalue_4 "TopS Emissive_02-Brightness-TBP" (Utilities|Time|GetGameTimeinSeconds))
      (Rendering|Material|SetScalarParameterValueonMaterials _returnvalue_4 "TopS Emissive_02_Brightness---Start" (Math|Float|Lerp 0.0 1.5 _returnvalue_5))
      (Rendering|Material|SetScalarParameterValueonMaterials _returnvalue "TopS Emissive_03-Brightness-TBP" (Utilities|Time|GetGameTimeinSeconds))
      (Rendering|Material|SetScalarParameterValueonMaterials _returnvalue "TopS Emissive_03_Brightness---Start" (Math|Float|Lerp 0.0 1.5 (Math|Float|Clamp(Float) (Math|Float|Lerp -0.5 1.0 _returnvalue_5))))
      (Variables|Default|Setwait_timer (+ (Utilities|Time|GetGameTimeinSeconds) 5.5)))))

(event EventOnWeaponFired
  (bind _returnvalue (Utilities|Time|GetGameTimeinSeconds))
  (bind _returnvalue_2 (CustomItemWrap|GetWrappedMeshComponents))
  (bind _stepcount (Variables|Default|GetStepCount))
  (bind _returnvalue_3 (Math|Conversions|ToFloat(Integer) _stepcount))
  (bind _maxsteps (Variables|Default|GetMaxSteps))
  (bind _returnvalue_4 (Math|Conversions|ToFloat(Integer) _maxsteps))
  (bind _returnvalue_5 (CustomItemWrap|GetWrappedMeshComponents))
  (bind _returnvalue_6 (CustomItemWrap|GetWrappedMeshComponents))
  (bind _returnvalue_7 (/ _returnvalue_3 _returnvalue_4))
  (Rendering|Material|SetScalarParameterValueonMaterials _returnvalue_5 "TopS Emissive_01-Brightness-TBP" _returnvalue)
  (Rendering|Material|SetScalarParameterValueonMaterials _returnvalue_5 "EDRec_Emissive_B_Brightness_Time_BP" _returnvalue)
  (bind _output_get (Variables|Default|SetStepCount (Math|Integer|Min(Integer) (+ (Math|Float|Truncate (Math|Float|Lerp (Math|Conversions|ToFloat(Integer) (Variables|Default|GetStepCount)) 0.0 (Math|Float|Clamp(Float) (/ (- (Utilities|Time|GetGameTimeinSeconds) (Variables|Default|Getwait_timer)) 1.0)))) 1) (Variables|Default|GetMaxSteps))))
  (Rendering|Material|SetScalarParameterValueonMaterials _returnvalue_6 "TopS Emissive_02-Brightness-TBP" (Utilities|Time|GetGameTimeinSeconds))
  (Rendering|Material|SetScalarParameterValueonMaterials _returnvalue_6 "TopS Emissive_02_Brightness---Start" (Math|Float|Lerp 0.0 1.5 _returnvalue_7))
  (Rendering|Material|SetScalarParameterValueonMaterials _returnvalue_2 "TopS Emissive_03-Brightness-TBP" (Utilities|Time|GetGameTimeinSeconds))
  (Rendering|Material|SetScalarParameterValueonMaterials _returnvalue_2 "TopS Emissive_03_Brightness---Start" (Math|Float|Lerp 0.0 1.5 (Math|Float|Clamp(Float) (Math|Float|Lerp -0.5 1.0 _returnvalue_7))))
  (Variables|Default|Setwait_timer (+ (Utilities|Time|GetGameTimeinSeconds) 5.5)))

(event EventOnPawnSet
  (bind _returnvalue (CustomItemWrap|GetWrappedMeshComponents))
  (Utilities|FlowControl|Sequence
    (:then_1
      (bind (_statname _statvalue) (Cosmetics|WatchCosmeticStat (CustomCharacter|GetAssociatedPlayerPawn) "(TagName=\"StatManager.AthenaKills\")")
        (:OnStatChanged
          (Rendering|Material|SetScalarParameterValueonMaterials _returnvalue "EDRec_Emissive_A_CC" 197.0)
          (Rendering|Material|SetScalarParameterValueonMaterials _returnvalue "EDRec_Emissive_B_CC" 198.0)
          (Rendering|Material|SetScalarParameterValueonMaterials _returnvalue "TopS Emissive_01_ColorCuve" 198.0)
          (Rendering|Material|SetScalarParameterValueonMaterials _returnvalue "TopS Emissive_02_ColorCuve" 198.0)
          (Rendering|Material|SetScalarParameterValueonMaterials _returnvalue "TopS Emissive_03_ColorCuve" 198.0)
          (Variables|Default|Setkill (> _statvalue 0)))))))

(fn ConstructionScript)
