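; Reactive weapon wrap (ToughPack) that switches a decal flipbook frame on its wrapped meshes in response to combat.
; BeginPlay and OnPawnSet both register pawn stat watchers for damage received and kills, while OnWeaponFired handles shots.
; Retriggerable delays restore the idle frame, with a Killed flag blocking fire reactions and early resets during the longer kill display.
; In frontend display, BeginPlay instead starts a looping timer that cycles through the frames.
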(event EventBeginPlay
  (bind _returnvalue (CustomItemWrap|GetWrappedMeshComponents))
  (bind _returnvalue_2 (CustomItemWrap|GetWrappedMeshComponents))
  (bind _returnvalue_3 (CustomCharacter|GetAssociatedPlayerPawn))
  (if (Reactive|Frontend|IsCosmeticObjectDisplayedinFrontendContext 0)
    (Utilities|Time|SetTimerbyFunctionName 0 "LockerCycle" 2.5 true)
    (else
      (Utilities|FlowControl|Sequence
        (:then_0
          (Cosmetics|WatchCosmeticStat _returnvalue_3 "(TagName=\"StatManager.DamageReceived\")" 0 false
            (:OnStatChanged
              (Rendering|Material|SetScalarParameterValueonMaterials _returnvalue "FDecalV2_FlipBook_Frame #" 1.0)
              (Utilities|FlowControl|RetriggerableDelay 0.5)
              (if (not (Variables|Default|GetKilled))
                (Rendering|Material|SetScalarParameterValueonMaterials _returnvalue "FDecalV2_FlipBook_Frame #" 0.0)))))
        (:then_1
          (Cosmetics|WatchCosmeticStat _returnvalue_3 "(TagName=\"StatManager.AthenaKills\")" 0 false
            (:OnStatChanged
              (Rendering|Material|SetScalarParameterValueonMaterials _returnvalue_2 "FDecalV2_FlipBook_Frame #" 0.66)
              (Variables|Default|SetKilled true)
              (Utilities|FlowControl|RetriggerableDelay 5.0)
              (Rendering|Material|SetScalarParameterValueonMaterials _returnvalue_2 "FDecalV2_FlipBook_Frame #" 0.0)
              (Variables|Default|SetKilled false))))))))

(event EventOnPawnSet
  (bind _returnvalue (CustomItemWrap|GetWrappedMeshComponents))
  (bind _returnvalue_2 (CustomItemWrap|GetWrappedMeshComponents))
  (bind _returnvalue_3 (CustomCharacter|GetAssociatedPlayerPawn))
  (Utilities|FlowControl|Sequence
    (:then_0
      (Cosmetics|WatchCosmeticStat _returnvalue_3 "(TagName=\"StatManager.DamageReceived\")" 0 false
        (:OnStatChanged
          (Rendering|Material|SetScalarParameterValueonMaterials _returnvalue "FDecalV2_FlipBook_Frame #" 1.0)
          (Utilities|FlowControl|RetriggerableDelay 0.5)
          (if (not (Variables|Default|GetKilled))
            (Rendering|Material|SetScalarParameterValueonMaterials _returnvalue "FDecalV2_FlipBook_Frame #" 0.0)))))
    (:then_1
      (Cosmetics|WatchCosmeticStat _returnvalue_3 "(TagName=\"StatManager.AthenaKills\")" 0 false
        (:OnStatChanged
          (Rendering|Material|SetScalarParameterValueonMaterials _returnvalue_2 "FDecalV2_FlipBook_Frame #" 0.66)
          (Variables|Default|SetKilled true)
          (Utilities|FlowControl|RetriggerableDelay 5.0)
          (Rendering|Material|SetScalarParameterValueonMaterials _returnvalue_2 "FDecalV2_FlipBook_Frame #" 0.0)
          (Variables|Default|SetKilled false))))))

(event EventOnWeaponFired
  (bind _returnvalue (CustomItemWrap|GetWrappedMeshComponents))
  (if (not (Variables|Default|GetKilled))
    (Rendering|Material|SetScalarParameterValueonMaterials _returnvalue "FDecalV2_FlipBook_Frame #" 0.33)
    (Utilities|FlowControl|RetriggerableDelay 0.5)
    (if (not (Variables|Default|GetKilled))
      (Rendering|Material|SetScalarParameterValueonMaterials _returnvalue "FDecalV2_FlipBook_Frame #" 0.0))))

(event Custom|LockerCycle
  (if (>= (Variables|Default|GetLockerCounter) 4)
    (Variables|Default|SetLockerCounter 0)
    (else
      (Math|Integer|IncrementInt (Variables|Default|GetLockerCounter)
        (:%20%20
          (Rendering|Material|SetScalarParameterValueonMaterials (CustomItemWrap|GetWrappedMeshComponents) "FDecalV2_FlipBook_Frame #" (Utilities|Select :Index (Variables|Default|GetLockerCounter) :Option%200 0.0 :Option%201 0.0 :Option%202 0.33 :Option%203 0.66 :Option%204 1.0)))))))

(fn ConstructionScript)
