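; TreeFrog weapon wrap logic that shifts the wrap's color each time the weapon fires.
; On each shot it adds a fixed step to a stored float, with no wrap or reset in this graph.
; It writes that value to three color-curve and emissive color offset scalar parameters on the materials of the wrapped mesh components.
; The construction script is empty.
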(event EventOnWeaponFired
  (bind _returnvalue (CustomItemWrap|GetWrappedMeshComponents))
  (bind _output_get (Variables|Default|SetFloat_Value (+ (Variables|Default|GetFloat_Value) 0.1)))
  (Rendering|Material|SetScalarParameterValueonMaterials _returnvalue "TSCC_ColorCurve_Offset" _output_get)
  (Rendering|Material|SetScalarParameterValueonMaterials _returnvalue "TS-Emissive-01-Color-Offset" _output_get)
  (Rendering|Material|SetScalarParameterValueonMaterials _returnvalue "TopS Emissive_01_ColorCuve_Offset" _output_get))

(fn ConstructionScript)
