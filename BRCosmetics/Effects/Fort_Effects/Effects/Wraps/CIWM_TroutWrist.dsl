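; TroutWrist item wrap cosmetic logic that changes the wrapped meshes' materials in stages as the owning pawn takes damage.
; A watched DamageReceived stat, after a brief delay, raises a capped step count while health is below 99, otherwise resetting it.
; The count decays if hits are spaced beyond a timed window.
; The step selects a blend offset, game time stamps a second material parameter, and health prints for debugging.
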(event EventOnPawnSet
  (bind _returnvalue (CustomCharacter|GetAssociatedPlayerPawn))
  (bind _returnvalue_2 (FortPawn|GetHealth _returnvalue))
  (bind _returnvalue_3 (CustomItemWrap|GetWrappedMeshComponents))
  (Utilities|FlowControl|Sequence
    (:then_0
      (Cosmetics|WatchCosmeticStat (CustomCharacter|GetAssociatedPlayerPawn) "(TagName=\"StatManager.DamageReceived\")"
        (:OnStatChanged
          (Utilities|FlowControl|Delay 0.05)
          (if (< _returnvalue_2 99.0)
            (Variables|Default|SetStepCount (Math|Integer|Clamp(Integer) (+ (Math|Float|Truncate (Math|Float|Lerp (Math|Conversions|ToFloat(Integer) (Variables|Default|GetStepCount)) 0.0 (Math|Float|Clamp(Float) (/ (- (Utilities|Time|GetGameTimeinSeconds) (Variables|Default|Getwait_timer)) 1.0)))) 1) 0 3))
            (else
              (Variables|Default|SetStepCount 0)))
          (Development|PrintText (Utilities|Text|ToText(Integer) (Math|Float|Truncate _returnvalue_2)) true true "(R=1.000000,G=0.126024,B=0.000000,A=1.000000)" 5.0)
          (switch int (Variables|Default|GetStepCount)
            (:0
              (Rendering|Material|SetScalarParameterValueonMaterials _returnvalue_3 "TopS_01_Blend_Offset---End" 0.3))
            (:1
              (Rendering|Material|SetScalarParameterValueonMaterials _returnvalue_3 "TopS_01_Blend_Offset---End" 0.1))
            (:2
              (Rendering|Material|SetScalarParameterValueonMaterials _returnvalue_3 "TopS_01_Blend_Offset---End" -0.3))
            (:3
              (Rendering|Material|SetScalarParameterValueonMaterials _returnvalue_3 "TopS_01_Blend_Offset---End" -1.0)))
          (Rendering|Material|SetScalarParameterValueonMaterials _returnvalue_3 "TopS_01_Blend_Opacity-TBP" (Utilities|Time|GetGameTimeinSeconds))
          (Variables|Default|Setwait_timer (+ (Utilities|Time|GetGameTimeinSeconds) 5.5)))))))

(fn ConstructionScript)
