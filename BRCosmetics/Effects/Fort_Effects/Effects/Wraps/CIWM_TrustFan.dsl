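; Reactive item wrap modifier for the TrustFan cosmetic that changes flipbook and decal material parameters on wrapped meshes in response to combat.
; A DamageReceived stat watch sets a damage look and a Taking Damage flag for two seconds, then restores Default State.
; Firing shows a separate one-second look unless that flag is set.
; In the front end, a timeline's stage6 event previews the damage reaction.
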(event EventBeginPlay
  (bind _returnvalue (CustomItemWrap|GetWrappedMeshComponents))
  (if (CustomCosmeticModifier|IsinFrontEnd)
    (Timeline|Timeline PlayFromStart
      (:stage6
        (Variables|Default|SetTaking%20Damage true)
        (Rendering|Material|SetColorParameterValueonMaterials _returnvalue "EDRec_FlipBook_FrameA_Speed--" (Utilities|Operators|Add :B (0.0 0.0 0.5 0.79)))
        (Rendering|Material|SetScalarParameterValueonMaterials _returnvalue "EDRec_FlipBook_FrameA" 9.0)
        (Rendering|Material|SetColorParameterValueonMaterials _returnvalue "EDRec_DecalC_Scale" (Utilities|Operators|Add :B (0.26 1.0 1.0 1.0)))
        (Rendering|Material|SetScalarParameterValueonMaterials _returnvalue "EDRec_FlipBook_FrameD" 4.0)
        (Utilities|FlowControl|Delay 2.0)
        (CallFunction|DefaultState)
        (Variables|Default|SetTaking%20Damage false)))))

(event EventOnWeaponFired
  (bind _returnvalue (CustomItemWrap|GetWrappedMeshComponents))
  (if (not (Variables|Default|GetTaking%20Damage))
    (Rendering|Material|SetScalarParameterValueonMaterials (CustomItemWrap|GetWrappedMeshComponents) "EDRec_FlipBook_FrameD" 3.0)
    (Rendering|Material|SetColorParameterValueonMaterials _returnvalue "EDRec_FlipBook_FrameA_Speed--" (Utilities|Operators|Add :B (0.0 0.0 0.0 1.0)))
    (Rendering|Material|SetScalarParameterValueonMaterials _returnvalue "EDRec_FlipBook_FrameA" 8.0)
    (Utilities|FlowControl|Delay 1.0)
    (CallFunction|DefaultState)))

(event EventOnPawnSet
  (bind _returnvalue (CustomItemWrap|GetWrappedMeshComponents))
  (Utilities|FlowControl|Sequence
    (:then_1
      (Cosmetics|WatchCosmeticStat (CustomCharacter|GetAssociatedPlayerPawn) "(TagName=\"StatManager.DamageReceived\")" 0 false
        (:OnStatChanged
          (Variables|Default|SetTaking%20Damage true)
          (Rendering|Material|SetColorParameterValueonMaterials _returnvalue "EDRec_FlipBook_FrameA_Speed--" (Utilities|Operators|Add :B (0.0 0.0 0.5 0.79)))
          (Rendering|Material|SetScalarParameterValueonMaterials _returnvalue "EDRec_FlipBook_FrameA" 9.0)
          (Rendering|Material|SetColorParameterValueonMaterials _returnvalue "EDRec_DecalC_Scale" (Utilities|Operators|Add :B (0.26 1.0 1.0 1.0)))
          (Rendering|Material|SetScalarParameterValueonMaterials _returnvalue "EDRec_FlipBook_FrameD" 4.0)
          (Utilities|FlowControl|Delay 2.0)
          (CallFunction|DefaultState)
          (Variables|Default|SetTaking%20Damage false))))))

(event Custom|Default%20State
  (bind _returnvalue (CustomItemWrap|GetWrappedMeshComponents))
  (Rendering|Material|SetColorParameterValueonMaterials _returnvalue "EDRec_FlipBook_FrameA_Speed--" (Utilities|Operators|Add :B (2.0 1.0 0.5 0.79)))
  (Rendering|Material|SetScalarParameterValueonMaterials _returnvalue "EDRec_FlipBook_FrameA" 0.0)
  (Rendering|Material|SetColorParameterValueonMaterials _returnvalue "EDRec_DecalC_Scale" (Utilities|Operators|Add :B (0.5 1.0 1.0 1.0)))
  (Rendering|Material|SetScalarParameterValueonMaterials _returnvalue "EDRec_FlipBook_FrameD" 0.0))

(fn ConstructionScript)
