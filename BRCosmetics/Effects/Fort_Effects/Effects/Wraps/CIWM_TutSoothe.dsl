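; Weapon wrap script for the TutSoothe cosmetic that drives emissive material effects on the wrapped meshes.
; Firing stamps game time into emissive brightness start and end parameters, then zeroes the end values unless Killed is set.
; SineAnimation and SineAnimOFF, uncalled here, start a looping timer that ramps an emissive gradient mask's green channel in opposite directions.
; After five seconds the update clears the timer and snaps the mask to the opposite end of its ramp.
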(event EventOnWeaponFired
  (bind _returnvalue (CustomItemWrap|GetWrappedMeshComponents))
  (bind _returnvalue_2 (Utilities|Time|GetGameTimeinSeconds))
  (Utilities|FlowControl|Sequence
    (:then_0
      (Rendering|Material|SetScalarParameterValueonMaterials _returnvalue "EDRec_Emissive_B_Brightness_Time_BP" _returnvalue_2)
      (Rendering|Material|SetScalarParameterValueonMaterials _returnvalue "EDRec_Emissive_A_Brightness_Time_BP" _returnvalue_2)
      (Rendering|Material|SetScalarParameterValueonMaterials _returnvalue "TopS Emissive_01-Brightness-TBP" _returnvalue_2)
      (Rendering|Material|SetScalarParameterValueonMaterials _returnvalue "TopS Emissive_01_Brightness---End" _returnvalue_2)
      (Rendering|Material|SetScalarParameterValueonMaterials _returnvalue "EDRec_Emissive_B_Brightness_Time__End" _returnvalue_2)
      (Rendering|Material|SetScalarParameterValueonMaterials _returnvalue "EDRec_Emissive_A_Brightness_Time__End" _returnvalue_2)
      (if (not (Variables|Default|GetKilled))
        (Rendering|Material|SetScalarParameterValueonMaterials _returnvalue "EDRec_Emissive_B_Brightness_Time__End" 0.0)
        (Rendering|Material|SetScalarParameterValueonMaterials _returnvalue "EDRec_Emissive_A_Brightness_Time__End" 0.0)
        (Rendering|Material|SetScalarParameterValueonMaterials _returnvalue "TopS Emissive_01_Brightness---End" 0.0)))))

(event Custom|SineAnimation
  ; Layer Offset
  (Variables|Default|SetReverse false)
  (Variables|Default|SetStartTime (Utilities|Time|GetGameTimeinSeconds))
  (bind _returnvalue (Utilities|Time|SetTimerbyEvent (delegate Custom|SineAnimUpdate) 0.02 true))
  (Variables|Default|SetTimerHandle _returnvalue))

(event Custom|SineAnimOFF
  ; Layer Offset
  (Variables|Default|SetReverse true)
  (Variables|Default|SetStartTime (Utilities|Time|GetGameTimeinSeconds))
  (bind _returnvalue (Utilities|Time|SetTimerbyEvent (delegate Custom|SineAnimUpdate) 0.02 true))
  (Variables|Default|SetTimerHandle _returnvalue))

(event Custom|SineAnimUpdate
  (bind _returnvalue (Utilities|Time|GetGameTimeinSeconds))
  (bind _returnvalue_2 (CustomItemWrap|GetWrappedMeshComponents))
  (bind _starttime (Variables|Default|GetStartTime))
  (bind _returnvalue_3 (- _returnvalue _starttime))
  ; Layer Offset
  (if (>= _returnvalue_3 5.0)
    (Utilities|Time|ClearandInvalidateTimerbyHandle (Variables|Default|GetTimerHandle))
    (if (Variables|Default|GetReverse)
      ;; Start
      (Rendering|Material|SetColorParameterValueonMaterials _returnvalue_2 "TopS Emissive_01_Grad-Tex_Mask_Channel" (0.0 0.0 0.0 0.0))
      (else
        ;; Arrive
        (Rendering|Material|SetColorParameterValueonMaterials _returnvalue_2 "TopS Emissive_01_Grad-Tex_Mask_Channel" (0.0 1.7 0.0 0.0))))
    (elif (Variables|Default|GetReverse)
      ;; FadeIn
      (Rendering|Material|SetColorParameterValueonMaterials _returnvalue_2 "TopS Emissive_01_Grad-Tex_Mask_Channel" (Math|Color|MakeColor 0.0 (Math|Float|MapRangeClamped _returnvalue_3 0.0 0.5 0.0 1.7) 0.0 0.0))
      (else
        ;; FadeOut
        (Rendering|Material|SetColorParameterValueonMaterials _returnvalue_2 "TopS Emissive_01_Grad-Tex_Mask_Channel" (Math|Color|MakeColor 0.0 (Math|Float|MapRangeClamped _returnvalue_3 0.0 0.5 1.7) 0.0 0.0))))))

(fn ConstructionScript)
