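; UniqueBeetle weapon wrap Blueprint that drives a live ammo counter decal and firing-reactive emissive effects on the wrapped meshes.
; Weapon, pawn and reload events refresh the count from magazine ammo, while the front end simulates an eight-round clip.
; Shots update the counter and stamp game time into material parameters for brightness effects.
; A retriggerable delay detects when firing stops, with looping timers ramping decal glow up and back down linearly, despite the SineAnimation naming.
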(event EventOnWeaponFired
  (bind _returnvalue (CustomItemWrap|GetWrappedMeshComponents))
  (bind _returnvalue_2 (CustomCharacter|GetAssociatedPlayerPawn))
  (bind _returnvalue_3 (Utilities|Time|GetGameTimeinSeconds))
  (bind _currentweapon (Variables|/Script/FortniteGame.FortPawn|GetCurrentWeapon _returnvalue_2))
  (bind _returnvalue_4 (FortWeapon|GetMagazineAmmoCount _currentweapon))
  (Utilities|FlowControl|Sequence
    (:then_0
      (Rendering|Material|SetScalarParameterValueonMaterials _returnvalue "TopS Emissive_01-Brightness-TBP" _returnvalue_3)
      (Rendering|Material|SetScalarParameterValueonMaterials _returnvalue "TopS Emissive_01_Brightness---End" _returnvalue_3)
      (if (not (Variables|Default|GetKilled))
        (Rendering|Material|SetScalarParameterValueonMaterials _returnvalue "TopS Emissive_01_Brightness---End" 0.0)))
    (:then_1
      (if (CustomCosmeticModifier|IsinFrontEnd)
        (bind _output_get (Variables|Default|SetAmmo%20Count (Variables|Default|GetAmmo%20Count)))
        (Utilities|FlowControl|Delay 0.001)
        (bind _output_get_2 (Variables|Default|SetAmmo%20Count (- _output_get 1)))
        (if (== _output_get_2 0)
          (bind _output_get_3 (Variables|Default|SetAmmo%20Count 8))
          (CallFunction|AmmoUpdate (Math|Conversions|ToFloat(Integer) _output_get_3))
          (else
            _))
        (elif (== _returnvalue_4 0)
          _
          (else
            (bind _output_get_4 (Variables|Default|SetAmmo%20Count _returnvalue_4))
            (Utilities|FlowControl|Delay 0.001)
            (Variables|Default|SetAmmo%20Count (- _output_get_4 1)))))
      (CallFunction|AmmoUpdate (Math|Conversions|ToFloat(Integer) (Variables|Default|GetAmmo%20Count))))
    (:then_2
      (if (not (Variables|Default|GetFullAutoWeapons))
        (Variables|Default|SetFullAutoWeapons true)
        (CallFunction|SineAnimation))
      (Utilities|FlowControl|RetriggerableDelay 0.5)
      (CallFunction|SineAnimOFF)
      (Variables|Default|SetFullAutoWeapons false))))

(event EventBeginPlay
  (bind _returnvalue (CustomItemWrap|GetWrappedMeshComponents))
  (bind _returnvalue_2 (CustomCharacter|GetAssociatedPlayerPawn))
  (bind _currentweapon (Variables|/Script/FortniteGame.FortPawn|GetCurrentWeapon _returnvalue_2))
  (if (CustomCosmeticModifier|IsinFrontEnd)
    (Utilities|FlowControl|Delay 0.2)
    (else
      (Utilities|FlowControl|Delay 0.2)
      (bind _currentweapon_2 (Variables|/Script/FortniteGame.FortPawn|GetCurrentWeapon (CustomCharacter|GetAssociatedPlayerPawn))
        (:then
          (EventDispatchers|BindEventtoOnReloadStarted _currentweapon_2 (delegate Custom|ReloadStarted))
          (if (not (CustomCosmeticModifier|IsinFrontEnd))
            (Variables|Default|Setper%20Clip (FortWeapon|GetBulletsPerClip _currentweapon))
            (bind _output_get (Variables|Default|SetAmmo%20Count (FortWeapon|GetMagazineAmmoCount _currentweapon)))
            (Rendering|Material|SetScalarParameterValueonMaterials _returnvalue "FDecalV2_Counter__Number" (Math|Conversions|ToFloat(Integer) _output_get))))
        (:else))))
  (Variables|Default|Setper%20Clip 8)
  (bind _output_get_2 (Variables|Default|SetAmmo%20Count 8))
  (Rendering|Material|SetScalarParameterValueonMaterials _returnvalue "FDecalV2_Counter__Number" (Math|Conversions|ToFloat(Integer) _output_get_2)))

(event EventOnWeaponChanged
  (params (NewWeapon :type Fort%20Weapon%20Object%20Reference) (OldWeapon :type Fort%20Weapon%20Object%20Reference))
  (bind _returnvalue (CustomItemWrap|GetWrappedMeshComponents))
  (bind _returnvalue_2 (CustomCharacter|GetAssociatedPlayerPawn))
  (bind _currentweapon (Variables|/Script/FortniteGame.FortPawn|GetCurrentWeapon _returnvalue_2))
  (Utilities|FlowControl|Delay 0.2)
  (bind _currentweapon_2 (Variables|/Script/FortniteGame.FortPawn|GetCurrentWeapon (CustomCharacter|GetAssociatedPlayerPawn))
    (:then
      (EventDispatchers|BindEventtoOnReloadStarted _currentweapon_2 (delegate Custom|ReloadStarted))
      (if (CustomCosmeticModifier|IsinFrontEnd)
        (Variables|Default|Setper%20Clip 8)
        (bind _output_get (Variables|Default|SetAmmo%20Count 8))
        (Rendering|Material|SetScalarParameterValueonMaterials _returnvalue "FDecalV2_Counter__Number" (Math|Conversions|ToFloat(Integer) _output_get))
        (else
          (Variables|Default|Setper%20Clip (FortWeapon|GetBulletsPerClip _currentweapon))
          (bind _output_get_2 (Variables|Default|SetAmmo%20Count (FortWeapon|GetMagazineAmmoCount _currentweapon)))
          (Rendering|Material|SetScalarParameterValueonMaterials _returnvalue "FDecalV2_Counter__Number" (Math|Conversions|ToFloat(Integer) _output_get_2)))))
    (:else)))

(event EventOnPawnSet
  (bind _returnvalue (CustomItemWrap|GetWrappedMeshComponents))
  (bind _returnvalue_2 (CustomCharacter|GetAssociatedPlayerPawn))
  (bind _currentweapon (Variables|/Script/FortniteGame.FortPawn|GetCurrentWeapon _returnvalue_2))
  (Utilities|FlowControl|Delay 0.2)
  (bind _currentweapon_2 (Variables|/Script/FortniteGame.FortPawn|GetCurrentWeapon (CustomCharacter|GetAssociatedPlayerPawn))
    (:then
      (EventDispatchers|BindEventtoOnReloadStarted _currentweapon_2 (delegate Custom|ReloadStarted))
      (if (CustomCosmeticModifier|IsinFrontEnd)
        (Variables|Default|Setper%20Clip 8)
        (bind _output_get (Variables|Default|SetAmmo%20Count 8))
        (Rendering|Material|SetScalarParameterValueonMaterials _returnvalue "FDecalV2_Counter__Number" (Math|Conversions|ToFloat(Integer) _output_get))
        (else
          (Variables|Default|Setper%20Clip (FortWeapon|GetBulletsPerClip _currentweapon))
          (bind _output_get_2 (Variables|Default|SetAmmo%20Count (FortWeapon|GetMagazineAmmoCount _currentweapon)))
          (Rendering|Material|SetScalarParameterValueonMaterials _returnvalue "FDecalV2_Counter__Number" (Math|Conversions|ToFloat(Integer) _output_get_2)))))
    (:else)))

(event Custom|ReloadStarted
  (params (ReloadTime :type Float%20%28single-precision%29) (ReloadType :type EFortWeaponReloadType%20Enum :default "ReloadWholeClip"))
  (CallFunction|AmmoUpdate (Math|Conversions|ToFloat(Integer) (Variables|Default|GetAmmo%20Count)))
  (Utilities|FlowControl|Delay ReloadTime)
  (if (not (CustomCosmeticModifier|IsinFrontEnd))
    (bind _output_get (Variables|Default|SetAmmo%20Count (FortWeapon|GetMagazineAmmoCount (Variables|/Script/FortniteGame.FortPawn|GetCurrentWeapon (CustomCharacter|GetAssociatedPlayerPawn)))))
    (CallFunction|AmmoUpdate (Math|Conversions|ToFloat(Integer) _output_get))))

(event EventOnNewWeapons
  (params (WeaponsAsItems :type Array%20of%20Fort%20Item%20Object%20References))
  (bind _returnvalue (CustomItemWrap|GetWrappedMeshComponents))
  (bind _returnvalue_2 (CustomCharacter|GetAssociatedPlayerPawn))
  (bind _currentweapon (Variables|/Script/FortniteGame.FortPawn|GetCurrentWeapon _returnvalue_2))
  (Utilities|FlowControl|Delay 0.2)
  (bind _currentweapon_2 (Variables|/Script/FortniteGame.FortPawn|GetCurrentWeapon (CustomCharacter|GetAssociatedPlayerPawn))
    (:then
      (EventDispatchers|BindEventtoOnReloadStarted _currentweapon_2 (delegate Custom|ReloadStarted))
      (if (CustomCosmeticModifier|IsinFrontEnd)
        (Variables|Default|Setper%20Clip 8)
        (bind _output_get (Variables|Default|SetAmmo%20Count 8))
        (Rendering|Material|SetScalarParameterValueonMaterials _returnvalue "FDecalV2_Counter__Number" (Math|Conversions|ToFloat(Integer) _output_get))
        (else
          (Variables|Default|Setper%20Clip (FortWeapon|GetBulletsPerClip _currentweapon))
          (bind _output_get_2 (Variables|Default|SetAmmo%20Count (FortWeapon|GetMagazineAmmoCount _currentweapon)))
          (Rendering|Material|SetScalarParameterValueonMaterials _returnvalue "FDecalV2_Counter__Number" (Math|Conversions|ToFloat(Integer) _output_get_2)))))
    (:else)))

(event EventOnNewPlayerWeaponType
  (params (NewType :type EFortItemType%20Enum :default "WorldItem") (OldType :type EFortItemType%20Enum :default "WorldItem"))
  (bind _returnvalue (CustomItemWrap|GetWrappedMeshComponents))
  (bind _returnvalue_2 (CustomCharacter|GetAssociatedPlayerPawn))
  (bind _currentweapon (Variables|/Script/FortniteGame.FortPawn|GetCurrentWeapon _returnvalue_2))
  (Utilities|FlowControl|Delay 0.2)
  (bind _currentweapon_2 (Variables|/Script/FortniteGame.FortPawn|GetCurrentWeapon (CustomCharacter|GetAssociatedPlayerPawn))
    (:then
      (EventDispatchers|BindEventtoOnReloadStarted _currentweapon_2 (delegate Custom|ReloadStarted))
      (if (CustomCosmeticModifier|IsinFrontEnd)
        (Variables|Default|Setper%20Clip 8)
        (bind _output_get (Variables|Default|SetAmmo%20Count 8))
        (Rendering|Material|SetScalarParameterValueonMaterials _returnvalue "FDecalV2_Counter__Number" (Math|Conversions|ToFloat(Integer) _output_get))
        (else
          (Variables|Default|Setper%20Clip (FortWeapon|GetBulletsPerClip _currentweapon))
          (bind _output_get_2 (Variables|Default|SetAmmo%20Count (FortWeapon|GetMagazineAmmoCount _currentweapon)))
          (Rendering|Material|SetScalarParameterValueonMaterials _returnvalue "FDecalV2_Counter__Number" (Math|Conversions|ToFloat(Integer) _output_get_2)))))
    (:else)))

(event Custom|SineAnimation
  (Variables|Default|SetReverse false)
  (Variables|Default|SetStartTime (Utilities|Time|GetGameTimeinSeconds))
  (bind _returnvalue (Utilities|Time|SetTimerbyEvent (delegate Custom|SineAnimUpdate) 0.02 true))
  (Variables|Default|SetTimerHandle _returnvalue))

(event Custom|SineAnimOFF
  (Variables|Default|SetReverse true)
  (Variables|Default|SetStartTime (Utilities|Time|GetGameTimeinSeconds))
  (bind _returnvalue (Utilities|Time|SetTimerbyEvent (delegate Custom|SineAnimUpdate) 0.02 true))
  (Variables|Default|SetTimerHandle _returnvalue))

(event Custom|SineAnimUpdate
  (bind _returnvalue (Utilities|Time|GetGameTimeinSeconds))
  (bind _returnvalue_2 (CustomItemWrap|GetWrappedMeshComponents))
  (bind _starttime (Variables|Default|GetStartTime))
  (bind _returnvalue_3 (- _returnvalue _starttime))
  (if (>= _returnvalue_3 5.0)
    (Utilities|Time|ClearandInvalidateTimerbyHandle (Variables|Default|GetTimerHandle))
    (if (Variables|Default|GetReverse)
      ; Start
      (Rendering|Material|SetColorParameterValueonMaterials _returnvalue_2 "FDecalV2-Emissive-Color--Alpha=Multplyer" (1.0 0.052 0.131 1.0))
      (else
        ; Arrive
        (Rendering|Material|SetColorParameterValueonMaterials _returnvalue_2 "FDecalV2-Emissive-Color--Alpha=Multplyer" (1.0 0.052 0.131 6.0))))
    (elif (Variables|Default|GetReverse)
      ; FadeOut
      (Rendering|Material|SetColorParameterValueonMaterials _returnvalue_2 "FDecalV2-Emissive-Color--Alpha=Multplyer" (Math|Color|MakeColor 1.0 0.052 0.131 (Math|Float|MapRangeClamped _returnvalue_3 0.0 0.5 6.0 1.0)))
      (else
        ; FadeIn
        (Rendering|Material|SetColorParameterValueonMaterials _returnvalue_2 "FDecalV2-Emissive-Color--Alpha=Multplyer" (Math|Color|MakeColor 1.0 0.052 0.131 (Math|Float|MapRangeClamped _returnvalue_3 0.0 0.25 1.0 6.0)))))))

(fn AmmoUpdate
  (params (Ammo :type Float%20%28single-precision%29))
  (bind _returnvalue (CustomItemWrap|GetWrappedMeshComponents))
  (Rendering|Material|SetScalarParameterValueonMaterials _returnvalue "FDecalV2_Counter__Number" Ammo)
  (Rendering|Material|SetScalarParameterValueonMaterials _returnvalue "FDecalV2_Emissive_Brightness_Time_BP" (Utilities|Time|GetGameTimeinSeconds)))

(fn ConstructionScript)

(fn Weapon_Offsets
  (params (Offset_End_Name :type Name) (Offset_End_Value :type Float%20%28double-precision%29) (Offset_Start_Name :type Name) (Offset_State_Value :type Float%20%28double-precision%29))
  (bind _returnvalue (CustomItemWrap|GetWrappedMeshComponents))
  (Rendering|Material|SetVectorParameterValueonMaterials _returnvalue Offset_End_Name (Math|Vector|MakeVector 0.0 0.0 Offset_End_Value))
  (Rendering|Material|SetVectorParameterValueonMaterials _returnvalue Offset_Start_Name (Math|Vector|MakeVector 0.0 0.0 Offset_State_Value)))
