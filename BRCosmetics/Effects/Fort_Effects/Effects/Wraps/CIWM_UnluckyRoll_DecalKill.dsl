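; Kill-reactive Unlucky Roll weapon wrap that escalates its decals, emissives and flames as the player's kills rise, up to four stages.
; The AthenaKills stat sets kill count and time in match, while a timeline previews the stages in the front end.
; KillMaterial swaps the flame mask, scales warp and panning, and timestamps decal parameters for material-driven animation, reapplying on weapon changes.
; UpdatMAt is an uncalled near-duplicate of UpdateDecal.
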(event EventBeginPlay
  (bind _returnvalue (Utilities|Time|GetGameTimeinSeconds))
  (Utilities|FlowControl|Delay 0.2)
  (if (CustomCosmeticModifier|IsinFrontEnd)
    (bind (_direction _timevalue) (Timeline|Timeline PlayFromStart)
      (:Update
        (if (!= (Variables|Default|GetKill_Count) (Math|Integer|Clamp(Integer) (Math|Float|Truncate _timevalue) 0 4))
          (Variables|Default|SetKill_Count (Math|Integer|Clamp(Integer) (Math|Float|Truncate _timevalue) 0 4))
          (Rendering|Material|SetScalarParameterValueonMaterials (CustomItemWrap|GetWrappedMeshComponents) "TopS Emissive_01-Brightness-TBP" _returnvalue)
          (Variables|Default|SetLastKillTime _returnvalue)
          (CallFunction|KillMaterial))))
    (else
      (CallFunction|ResetSealValues))))

(event Custom|ResetSealValues
  (bind _returnvalue (CustomItemWrap|GetWrappedMeshComponents))
  (Rendering|Material|SetScalarParameterValueonMaterials _returnvalue "EDRec_DecalA_Opacity_End" 1.0)
  (Rendering|Material|SetScalarParameterValueonMaterials _returnvalue "EDRec_DecalB_Opacity_End" 1.0)
  (Rendering|Material|SetScalarParameterValueonMaterials _returnvalue "EDRec_DecalC_Opacity_End" 1.0)
  (Rendering|Material|SetScalarParameterValueonMaterials _returnvalue "EDRec_Emissive_C_Brightness_Time__End" 0.0)
  (Rendering|Material|SetScalarParameterValueonMaterials _returnvalue "EDRec_Emissive_B_Brightness_Time__End" 0.0)
  (Rendering|Material|SetScalarParameterValueonMaterials _returnvalue "EDRec_Emissive_A_Brightness_Time__End" 0.0))

(event Custom|UpdateDecal
  (bind _returnvalue (CustomItemWrap|GetWrappedMeshComponents))
  (bind _returnvalue_2 (CustomItemWrap|GetWrappedMeshComponents))
  (bind _kill_count (Variables|Default|GetKill_Count))
  (bind _returnvalue_3 (>= _kill_count 4))
  (bind _lastkilltime (Variables|Default|GetLastKillTime))
  (bind _lastkilltime_2 (Variables|Default|GetLastKillTime))
  (if _returnvalue_3
    (Rendering|Material|SetScalarParameterValueonMaterials _returnvalue "EDRec_Emissive_C_Brightness_Time__End" 0.0)
    (Rendering|Material|SetScalarParameterValueonMaterials _returnvalue "EDRec_DecalC_Opacity_End" 0.0)
    (Rendering|Material|SetScalarParameterValueonMaterials _returnvalue "EDRec_Emissive_B_Brightness_Time__End" 0.0)
    (Rendering|Material|SetScalarParameterValueonMaterials _returnvalue "EDRec_DecalB_Opacity_End" 0.0)
    (else
      (switch int _kill_count
        (:0
          (CallFunction|ResetSealValues))
        (:1
          (Rendering|Material|SetScalarParameterValueonMaterials _returnvalue "EDRec_TimeFrom_BP_DecalA" _lastkilltime_2)
          (Rendering|Material|SetScalarParameterValueonMaterials _returnvalue "EDRec_Emissive_A_Brightness_Time_BP" _lastkilltime_2))
        (:2
          (Rendering|Material|SetScalarParameterValueonMaterials _returnvalue "EDRec_TimeFrom_BP_DecalB" _lastkilltime_2)
          (Rendering|Material|SetScalarParameterValueonMaterials _returnvalue "EDRec_Emissive_B_Brightness_Time_BP" _lastkilltime_2))
        (:3
          (Rendering|Material|SetScalarParameterValueonMaterials _returnvalue "EDRec_TimeFrom_BP_DecalC" _lastkilltime_2)
          (Rendering|Material|SetScalarParameterValueonMaterials _returnvalue "EDRec_Emissive_C_Brightness_Time_BP" _lastkilltime_2)))))
  (Rendering|Material|SetScalarParameterValueonMaterials _returnvalue "EDRec_Emissive_A_Brightness_Time__End" 0.0)
  (Rendering|Material|SetScalarParameterValueonMaterials _returnvalue "EDRec_DecalA_Opacity_End" 0.0)
  (if _returnvalue_3
    (Rendering|Material|SetScalarParameterValueonMaterials _returnvalue_2 "EDRec_TimeFrom_BP_DecalA" _lastkilltime)
    (Rendering|Material|SetScalarParameterValueonMaterials _returnvalue_2 "EDRec_TimeFrom_BP_DecalB" _lastkilltime)
    (Rendering|Material|SetScalarParameterValueonMaterials _returnvalue_2 "EDRec_TimeFrom_BP_DecalC" _lastkilltime)
    (Rendering|Material|SetScalarParameterValueonMaterials _returnvalue_2 "EDRec_TimeFrom_BP_DecalD" _lastkilltime)))

(event Custom|UpdatMAt
  (bind _returnvalue (CustomItemWrap|GetWrappedMeshComponents))
  (bind _returnvalue_2 (CustomItemWrap|GetWrappedMeshComponents))
  (bind _lastkilltime (Variables|Default|GetLastKillTime))
  (bind _kill_count (Variables|Default|GetKill_Count))
  (bind _lastkilltime_2 (Variables|Default|GetLastKillTime))
  (switch int _kill_count
    (:0
      (CallFunction|ResetSealValues))
    (:1
      (Rendering|Material|SetScalarParameterValueonMaterials _returnvalue "EDRec_TimeFrom_BP_DecalA" _lastkilltime_2)
      (Rendering|Material|SetScalarParameterValueonMaterials _returnvalue "EDRec_Emissive_A_Brightness_Time_BP" _lastkilltime_2))
    (:2
      (Rendering|Material|SetScalarParameterValueonMaterials _returnvalue "EDRec_TimeFrom_BP_DecalB" _lastkilltime_2)
      (Rendering|Material|SetScalarParameterValueonMaterials _returnvalue "EDRec_Emissive_B_Brightness_Time_BP" _lastkilltime_2)
      (Rendering|Material|SetScalarParameterValueonMaterials _returnvalue "EDRec_Emissive_B_Brightness_Time__End" 0.0)
      (Rendering|Material|SetScalarParameterValueonMaterials _returnvalue "EDRec_DecalB_Opacity_End" 0.0))
    (:3
      (Rendering|Material|SetScalarParameterValueonMaterials _returnvalue "EDRec_TimeFrom_BP_DecalC" _lastkilltime_2)
      (Rendering|Material|SetScalarParameterValueonMaterials _returnvalue "EDRec_Emissive_C_Brightness_Time_BP" _lastkilltime_2)
      (Rendering|Material|SetScalarParameterValueonMaterials _returnvalue "EDRec_Emissive_C_Brightness_Time__End" 0.0)
      (Rendering|Material|SetScalarParameterValueonMaterials _returnvalue "EDRec_DecalC_Opacity_End" 0.0)))
  (Rendering|Material|SetScalarParameterValueonMaterials _returnvalue "EDRec_Emissive_A_Brightness_Time__End" 0.0)
  (Rendering|Material|SetScalarParameterValueonMaterials _returnvalue "EDRec_DecalA_Opacity_End" 0.0)
  (if (>= _kill_count 4)
    (Rendering|Material|SetScalarParameterValueonMaterials _returnvalue_2 "EDRec_TimeFrom_BP_DecalA" _lastkilltime)
    (Rendering|Material|SetScalarParameterValueonMaterials _returnvalue_2 "EDRec_TimeFrom_BP_DecalB" _lastkilltime)
    (Rendering|Material|SetScalarParameterValueonMaterials _returnvalue_2 "EDRec_TimeFrom_BP_DecalC" _lastkilltime)
    (Rendering|Material|SetScalarParameterValueonMaterials _returnvalue_2 "EDRec_TimeFrom_BP_DecalD" _lastkilltime)))

(event Custom|UpdateWarpAndSpeed
  (bind _returnvalue (CustomItemWrap|GetWrappedMeshComponents))
  (bind _kill_count (Variables|Default|GetKill_Count))
  (bind _returnvalue_2 (Math|Integer|Clamp(Integer) _kill_count 0 4))
  (bind _returnvalue_3 (Math|Conversions|ToFloat(Integer) _returnvalue_2))
  (bind _output (Utilities|RemapFloat _returnvalue_3 "(X=0.000000,Y=4.000000)" "(X=-0.080000,Y=-0.200000)"))
  (Rendering|Material|SetScalarParameterValueonMaterials _returnvalue "TopS_01_Blend_Warp_Amount" (Utilities|RemapFloat _returnvalue_3 "(X=0.000000,Y=4.000000)" "(X=0.020000,Y=0.100000)"))
  (Rendering|Material|SetScalarParameterValueonMaterials _returnvalue "TopS_01_Warp_UV_Amount" (Utilities|RemapFloat _returnvalue_3 "(X=0.000000,Y=4.000000)" "(X=0.050000,Y=0.150000)"))
  (Rendering|Material|SetScalarParameterValueonMaterials _returnvalue "TopS Emissive_02_Brightness---End" (Utilities|RemapFloat _returnvalue_3 "(X=0.000000,Y=4.000000)" "(X=0.000000,Y=0.500000)"))
  (Rendering|Material|SetVectorParameterValueonMaterials _returnvalue "TopS_01_Warp_UV_Panning" (Math|Vector|MakeVector 0.0 _output))
  (Rendering|Material|SetVectorParameterValueonMaterials _returnvalue "TopS_01_PanningS_Speed------------" (Math|Vector|MakeVector 0.0 (Utilities|RemapFloat _output "(X=0.000000,Y=4.000000)" "(X=0.050000,Y=0.150000)"))))

(event Custom|UpdateFlames
  (for _array_element (CustomItemWrap|GetWrappedMeshComponents)
    (for _array_element_2 (Rendering|Material|GetMaterials _array_element)
      (bind _asmaterial_instance_dynamic (Utilities|Casting|CastToMaterialInstanceDynamic _array_element_2))
      (Rendering|Material|SetTextureParameterValue _asmaterial_instance_dynamic "TopS Blend_Grad-Tex_Mask" (select (> (Variables|Default|GetKill_Count) 2) "/BRCosmetics/Weapons/WeaponSkins/Wraps/Textures/UnluckyRoll/T_Wrap_UnluckyRoll_msk3.T_Wrap_UnluckyRoll_msk3" "/BRCosmetics/Weapons/WeaponSkins/Wraps/Textures/UnluckyRoll/T_Wrap_UnluckyRoll_msk2.T_Wrap_UnluckyRoll_msk2")))))

(event EventOnPawnSet
  (bind _returnvalue (Utilities|Time|GetGameTimeinSeconds))
  (bind (_statname _statvalue) (Cosmetics|WatchCosmeticStat (CustomCharacter|GetAssociatedPlayerPawn) "(TagName=\"StatManager.AthenaKills\")")
    (:OnStatChanged
      (Variables|Default|SetKill_Count _statvalue)
      (Rendering|Material|SetScalarParameterValueonMaterials (CustomItemWrap|GetWrappedMeshComponents) "TopS Emissive_01-Brightness-TBP" _returnvalue)
      (Variables|Default|SetLastKillTime _returnvalue)
      (CallFunction|KillMaterial))))

(event EventOnNewWeapons
  (params (WeaponsAsItems :type Array%20of%20Fort%20Item%20Object%20References))
  (Utilities|FlowControl|Delay 0.2)
  (CallFunction|KillMaterial))

(event EventOnNewPlayerWeaponType
  (params (NewType :type EFortItemType%20Enum :default "WorldItem") (OldType :type EFortItemType%20Enum :default "WorldItem"))
  (Utilities|FlowControl|Delay 0.2)
  (CallFunction|KillMaterial))

(event Custom|Kill%20material
  (CallFunction|UpdateFlames)
  (CallFunction|UpdateWarpandSpeed)
  (CallFunction|UpdateDecal))

(fn ConstructionScript)
