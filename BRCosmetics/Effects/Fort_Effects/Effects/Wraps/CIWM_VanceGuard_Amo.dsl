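; Wrap cosmetic modifier for the VanceGuard weapon wrap that shows a live magazine ammo counter on the wrapped meshes through material parameters.
; Weapon change, pawn set and reload events re-read the magazine; firing decrements it, and AmmoUpdate writes the number plus a flash timestamp.
; Kill and damage-dealt stat watchers timestamp other emissive and decal layers, while health-gain also refreshes the counter.
; In the front end it fakes an eight-round clip and plays the effects on a timeline.
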(event EventOnWeaponFired
  (bind _returnvalue (CustomCharacter|GetAssociatedPlayerPawn))
  (bind _currentweapon (Variables|/Script/FortniteGame.FortPawn|GetCurrentWeapon _returnvalue))
  (bind _returnvalue_2 (FortWeapon|GetMagazineAmmoCount _currentweapon))
  (if (CustomCosmeticModifier|IsinFrontEnd)
    (bind _output_get (Variables|Default|SetAmmo%20Count (Variables|Default|GetAmmo%20Count)))
    (Utilities|FlowControl|Delay 0.001)
    (bind _output_get_2 (Variables|Default|SetAmmo%20Count (- _output_get 1)))
    (if (== _output_get_2 0)
      (bind _output_get_3 (Variables|Default|SetAmmo%20Count 8))
      (CallFunction|AmmoUpdate (Math|Conversions|ToFloat(Integer) _output_get_3))
      (else
        _))
    (elif (== _returnvalue_2 0)
      _
      (else
        (bind _output_get_4 (Variables|Default|SetAmmo%20Count _returnvalue_2))
        (Utilities|FlowControl|Delay 0.001)
        (Variables|Default|SetAmmo%20Count (- _output_get_4 1)))))
  (CallFunction|AmmoUpdate (Math|Conversions|ToFloat(Integer) (Variables|Default|GetAmmo%20Count))))

(event EventBeginPlay
  (bind _returnvalue (CustomItemWrap|GetWrappedMeshComponents))
  (bind _returnvalue_2 (Utilities|Time|GetGameTimeinSeconds))
  (bind _returnvalue_3 (CustomItemWrap|GetWrappedMeshComponents))
  (bind _returnvalue_4 (CustomCharacter|GetAssociatedPlayerPawn))
  (bind _currentweapon (Variables|/Script/FortniteGame.FortPawn|GetCurrentWeapon _returnvalue_4))
  (if (CustomCosmeticModifier|IsinFrontEnd)
    (Utilities|FlowControl|Delay 0.2)
    (Utilities|FlowControl|Sequence
      (:then_0
        _)
      (:then_1
        (Timeline|Timeline Play
          (:NewTrack_0
            (Rendering|Material|SetScalarParameterValueonMaterials _returnvalue_3 "EDRec_Emissive_B_Brightness_Time_BP" _returnvalue_2))
          (:NewTrack_1
            (Rendering|Material|SetScalarParameterValueonMaterials _returnvalue_3 "EDRec_Emissive_C_Brightness_Time_BP" _returnvalue_2)
            (Rendering|Material|SetScalarParameterValueonMaterials _returnvalue_3 "EDRec_TimeFrom_BP_DecalC" _returnvalue_2)
            (Utilities|FlowControl|Delay 0.2)
            (Rendering|Material|SetScalarParameterValueonMaterials _returnvalue_3 "EDRec_Emissive_D_Brightness_Time_BP" _returnvalue_2)
            (Rendering|Material|SetScalarParameterValueonMaterials _returnvalue_3 "EDRec_TimeFrom_BP_DecalD" _returnvalue_2)))))
    (else
      (Utilities|FlowControl|Delay 0.2)
      (bind _currentweapon_2 (Variables|/Script/FortniteGame.FortPawn|GetCurrentWeapon (CustomCharacter|GetAssociatedPlayerPawn))
        (:then
          (EventDispatchers|BindEventtoOnReloadStarted _currentweapon_2 (delegate Custom|ReloadStarted))
          (if (not (CustomCosmeticModifier|IsinFrontEnd))
            (Variables|Default|Setper%20Clip (FortWeapon|GetBulletsPerClip _currentweapon))
            (bind _output_get (Variables|Default|SetAmmo%20Count (FortWeapon|GetMagazineAmmoCount _currentweapon)))
            (Rendering|Material|SetScalarParameterValueonMaterials _returnvalue "EDRec_Counter__Number" (Math|Conversions|ToFloat(Integer) _output_get))))
        (:else))))
  (Variables|Default|Setper%20Clip 8)
  (bind _output_get_2 (Variables|Default|SetAmmo%20Count 8))
  (Rendering|Material|SetScalarParameterValueonMaterials _returnvalue "EDRec_Counter__Number" (Math|Conversions|ToFloat(Integer) _output_get_2)))

(event Custom|OnBannerFullyApplied
  (params (Icon :type Texture%202D%20Object%20Reference) (PrimaryColor :type Linear%20Color%20Structure)))

(event EventOnWeaponChanged
  (params (NewWeapon :type Fort%20Weapon%20Object%20Reference) (OldWeapon :type Fort%20Weapon%20Object%20Reference))
  (bind _returnvalue (CustomItemWrap|GetWrappedMeshComponents))
  (bind _returnvalue_2 (CustomCharacter|GetAssociatedPlayerPawn))
  (bind _currentweapon (Variables|/Script/FortniteGame.FortPawn|GetCurrentWeapon _returnvalue_2))
  (Utilities|FlowControl|Delay 0.2)
  (bind _currentweapon_2 (Variables|/Script/FortniteGame.FortPawn|GetCurrentWeapon (CustomCharacter|GetAssociatedPlayerPawn))
    (:then
      (EventDispatchers|BindEventtoOnReloadStarted _currentweapon_2 (delegate Custom|ReloadStarted))
      (if (CustomCosmeticModifier|IsinFrontEnd)
        (Variables|Default|Setper%20Clip 8)
        (bind _output_get (Variables|Default|SetAmmo%20Count 8))
        (Rendering|Material|SetScalarParameterValueonMaterials _returnvalue "EDRec_Counter__Number" (Math|Conversions|ToFloat(Integer) _output_get))
        (else
          (Variables|Default|Setper%20Clip (FortWeapon|GetBulletsPerClip _currentweapon))
          (bind _output_get_2 (Variables|Default|SetAmmo%20Count (FortWeapon|GetMagazineAmmoCount _currentweapon)))
          (Rendering|Material|SetScalarParameterValueonMaterials _returnvalue "EDRec_Counter__Number" (Math|Conversions|ToFloat(Integer) _output_get_2)))))
    (:else)))

(event EventOnPawnSet
  (bind _returnvalue (CustomItemWrap|GetWrappedMeshComponents))
  (bind _returnvalue_2 (Utilities|Time|GetGameTimeinSeconds))
  (bind _returnvalue_3 (CustomItemWrap|GetWrappedMeshComponents))
  (bind _returnvalue_4 (CustomCharacter|GetAssociatedPlayerPawn))
  (bind _returnvalue_5 (CustomCharacter|GetAssociatedPlayerPawn))
  (bind _currentweapon (Variables|/Script/FortniteGame.FortPawn|GetCurrentWeapon _returnvalue_4))
  (bind _returnvalue_6 (FortWeapon|GetMagazineAmmoCount _currentweapon))
  (Utilities|FlowControl|Sequence
    (:then_0
      (Utilities|FlowControl|Delay 0.2)
      (bind _currentweapon_2 (Variables|/Script/FortniteGame.FortPawn|GetCurrentWeapon _returnvalue_5)
        (:then
          (EventDispatchers|BindEventtoOnReloadStarted _currentweapon_2 (delegate Custom|ReloadStarted))
          (if (CustomCosmeticModifier|IsinFrontEnd)
            (Variables|Default|Setper%20Clip 8)
            (bind _output_get (Variables|Default|SetAmmo%20Count 8))
            (Rendering|Material|SetScalarParameterValueonMaterials _returnvalue "EDRec_Counter__Number" (Math|Conversions|ToFloat(Integer) _output_get))
            (else
              (Variables|Default|Setper%20Clip (FortWeapon|GetBulletsPerClip _currentweapon))
              (bind _output_get_2 (Variables|Default|SetAmmo%20Count _returnvalue_6))
              (Rendering|Material|SetScalarParameterValueonMaterials _returnvalue "EDRec_Counter__Number" (Math|Conversions|ToFloat(Integer) _output_get_2)))))
        (:else)))
    (:then_1
      (Cosmetics|WatchCosmeticStat _returnvalue_5 "(TagName=\"StatManager.GainedHealthTimes\")" 0 false
        (:OnStatChanged
          (bind _output_get_3 (Variables|Default|SetAmmo%20Count _returnvalue_6))
          (Utilities|FlowControl|Delay 0.001)
          (Variables|Default|SetAmmo%20Count (- _output_get_3 1))
          (CallFunction|AmmoUpdate (Math|Conversions|ToFloat(Integer) (Variables|Default|GetAmmo%20Count))))))
    (:then_2
      (Cosmetics|WatchCosmeticStat (CustomCharacter|GetAssociatedPlayerPawn) "(TagName=\"StatManager.AthenaKills\")" 0 false
        (:OnStatChanged
          (Rendering|Material|SetScalarParameterValueonMaterials _returnvalue_3 "EDRec_Emissive_C_Brightness_Time_BP" _returnvalue_2)
          (Rendering|Material|SetScalarParameterValueonMaterials _returnvalue_3 "EDRec_TimeFrom_BP_DecalC" _returnvalue_2)
          (Utilities|FlowControl|Delay 0.2)
          (Rendering|Material|SetScalarParameterValueonMaterials _returnvalue_3 "EDRec_Emissive_D_Brightness_Time_BP" _returnvalue_2)
          (Rendering|Material|SetScalarParameterValueonMaterials _returnvalue_3 "EDRec_TimeFrom_BP_DecalD" _returnvalue_2))))
    (:then_3
      (Cosmetics|WatchCosmeticStat (CustomCharacter|GetAssociatedPlayerPawn) "(TagName=\"StatManager.DamageDealt\")" 0 false
        (:OnStatChanged
          (Rendering|Material|SetScalarParameterValueonMaterials _returnvalue_3 "EDRec_Emissive_B_Brightness_Time_BP" _returnvalue_2))))))

(event Custom|ReloadStarted
  (params (ReloadTime :type Float%20%28single-precision%29) (ReloadType :type EFortWeaponReloadType%20Enum :default "ReloadWholeClip"))
  (CallFunction|AmmoUpdate (Math|Conversions|ToFloat(Integer) (Variables|Default|GetAmmo%20Count)))
  (Utilities|FlowControl|Delay ReloadTime)
  (if (not (CustomCosmeticModifier|IsinFrontEnd))
    (bind _output_get (Variables|Default|SetAmmo%20Count (FortWeapon|GetMagazineAmmoCount (Variables|/Script/FortniteGame.FortPawn|GetCurrentWeapon (CustomCharacter|GetAssociatedPlayerPawn)))))
    (CallFunction|AmmoUpdate (Math|Conversions|ToFloat(Integer) _output_get))))

(event EventOnNewWeapons
  (params (WeaponsAsItems :type Array%20of%20Fort%20Item%20Object%20References))
  (bind _returnvalue (CustomItemWrap|GetWrappedMeshComponents))
  (bind _returnvalue_2 (CustomCharacter|GetAssociatedPlayerPawn))
  (bind _currentweapon (Variables|/Script/FortniteGame.FortPawn|GetCurrentWeapon _returnvalue_2))
  (Utilities|FlowControl|Delay 0.2)
  (bind _currentweapon_2 (Variables|/Script/FortniteGame.FortPawn|GetCurrentWeapon (CustomCharacter|GetAssociatedPlayerPawn))
    (:then
      (EventDispatchers|BindEventtoOnReloadStarted _currentweapon_2 (delegate Custom|ReloadStarted))
      (if (CustomCosmeticModifier|IsinFrontEnd)
        (Variables|Default|Setper%20Clip 8)
        (bind _output_get (Variables|Default|SetAmmo%20Count 8))
        (Rendering|Material|SetScalarParameterValueonMaterials _returnvalue "EDRec_Counter__Number" (Math|Conversions|ToFloat(Integer) _output_get))
        (else
          (Variables|Default|Setper%20Clip (FortWeapon|GetBulletsPerClip _currentweapon))
          (bind _output_get_2 (Variables|Default|SetAmmo%20Count (FortWeapon|GetMagazineAmmoCount _currentweapon)))
          (Rendering|Material|SetScalarParameterValueonMaterials _returnvalue "EDRec_Counter__Number" (Math|Conversions|ToFloat(Integer) _output_get_2)))))
    (:else)))

(event EventOnNewPlayerWeaponType
  (params (NewType :type EFortItemType%20Enum :default "WorldItem") (OldType :type EFortItemType%20Enum :default "WorldItem"))
  (bind _returnvalue (CustomItemWrap|GetWrappedMeshComponents))
  (bind _returnvalue_2 (CustomCharacter|GetAssociatedPlayerPawn))
  (bind _currentweapon (Variables|/Script/FortniteGame.FortPawn|GetCurrentWeapon _returnvalue_2))
  (Utilities|FlowControl|Delay 0.2)
  (bind _currentweapon_2 (Variables|/Script/FortniteGame.FortPawn|GetCurrentWeapon (CustomCharacter|GetAssociatedPlayerPawn))
    (:then
      (EventDispatchers|BindEventtoOnReloadStarted _currentweapon_2 (delegate Custom|ReloadStarted))
      (if (CustomCosmeticModifier|IsinFrontEnd)
        (Variables|Default|Setper%20Clip 8)
        (bind _output_get (Variables|Default|SetAmmo%20Count 8))
        (Rendering|Material|SetScalarParameterValueonMaterials _returnvalue "EDRec_Counter__Number" (Math|Conversions|ToFloat(Integer) _output_get))
        (else
          (Variables|Default|Setper%20Clip (FortWeapon|GetBulletsPerClip _currentweapon))
          (bind _output_get_2 (Variables|Default|SetAmmo%20Count (FortWeapon|GetMagazineAmmoCount _currentweapon)))
          (Rendering|Material|SetScalarParameterValueonMaterials _returnvalue "EDRec_Counter__Number" (Math|Conversions|ToFloat(Integer) _output_get_2)))))
    (:else)))

(fn AmmoUpdate
  (params (Ammo :type Float%20%28single-precision%29))
  (bind _returnvalue (CustomItemWrap|GetWrappedMeshComponents))
  (Rendering|Material|SetScalarParameterValueonMaterials _returnvalue "EDRec_Counter__Number" Ammo)
  (Rendering|Material|SetScalarParameterValueonMaterials _returnvalue "EDRec_Emissive_A_Brightness_Time_BP" (Utilities|Time|GetGameTimeinSeconds)))

(fn ConstructionScript)

(fn Weapon_Offsets
  (params (Offset_End_Name :type Name) (Offset_End_Value :type Float%20%28double-precision%29) (Offset_Start_Name :type Name) (Offset_State_Value :type Float%20%28double-precision%29))
  (bind _returnvalue (CustomItemWrap|GetWrappedMeshComponents))
  (Rendering|Material|SetVectorParameterValueonMaterials _returnvalue Offset_End_Name (Math|Vector|MakeVector 0.0 0.0 Offset_End_Value))
  (Rendering|Material|SetVectorParameterValueonMaterials _returnvalue Offset_Start_Name (Math|Vector|MakeVector 0.0 0.0 Offset_State_Value)))
