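; Veiled Glitch item wrap logic that makes the wrapped weapon's glitch material react to player eliminations.
; On pawn set it watches the AthenaKills stat, using odd/even parity to flip the mask-invert parameter and stamping game time to start the transition.
; Begin play, res-in and weapon-change events re-read the count after a short delay, setting the mask with a far-future time.
; In the front end, a cycling shot counter toggles the effect instead.
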(event EventOnWeaponFired
  (bind _returnvalue (CustomItemWrap|GetWrappedMeshComponents))
  ; Frontnd reactive
  (if (CustomCosmeticModifier|IsinFrontEnd)
    (bind _output_get (Variables|Default|Setweapon%20fire (+ (Variables|Default|Getweapon%20fire) 1)))
    (if (>= _output_get 5)
      (if (== (Variables|Default|Getweapon%20fire) 7)
        (Variables|Default|Setweapon%20fire 0)
        (if (Variables|Default|GetActive)
          (Rendering|Material|SetScalarParameterValueonMaterials _returnvalue "VGlit_Reactive_Msk_invert" 1.0)
          (Rendering|Material|SetScalarParameterValueonMaterials _returnvalue "VGlit_Reactive_Time_BP" (Utilities|Time|GetGameTimeinSeconds))
          (Variables|Default|SetActive false)
          (else
            (Rendering|Material|SetScalarParameterValueonMaterials _returnvalue "VGlit_Reactive_Msk_invert" 0.0)
            (Rendering|Material|SetScalarParameterValueonMaterials _returnvalue "VGlit_Reactive_Time_BP" (Utilities|Time|GetGameTimeinSeconds))
            (Variables|Default|SetActive true)))))))

(event EventOnPawnSet
  (bind _returnvalue (CustomItemWrap|GetWrappedMeshComponents))
  (bind _returnvalue_2 (CustomItemWrap|GetWrappedMeshComponents))
  (bind (_statname _statvalue) (Cosmetics|WatchCosmeticStat (CustomCharacter|GetAssociatedPlayerPawn) "(TagName=\"StatManager.AthenaKills\")")
    (:OnStatChanged
      (bind _output_get (Variables|Default|Setkill%20count _statvalue))
      (if (>= _output_get 1)
        (if (> (Math|Integer|%25%28Integer%29 _statvalue 2) 0)
          (Rendering|Material|SetScalarParameterValueonMaterials _returnvalue "VGlit_Reactive_Msk_invert" 0.0)
          (Rendering|Material|SetScalarParameterValueonMaterials _returnvalue "VGlit_Reactive_Time_BP" (Utilities|Time|GetGameTimeinSeconds))
          (Variables|Default|SetActive true)
          (else
            (Rendering|Material|SetScalarParameterValueonMaterials _returnvalue "VGlit_Reactive_Msk_invert" 1.0)
            (Rendering|Material|SetScalarParameterValueonMaterials _returnvalue "VGlit_Reactive_Time_BP" (Utilities|Time|GetGameTimeinSeconds))
            (Variables|Default|SetActive false)))
        (else
          (Rendering|Material|SetScalarParameterValueonMaterials _returnvalue_2 "VGlit_Reactive_Msk_invert" 0.0)
          (Rendering|Material|SetScalarParameterValueonMaterials _returnvalue_2 "VGlit_Reactive_Time_BP" 99999999.0))))))

(event EventOnPostResIn
  (bind _returnvalue (CustomItemWrap|GetWrappedMeshComponents))
  (Utilities|FlowControl|Delay 0.05)
  (bind (_statname _statvalue) (Cosmetics|WatchCosmeticStat (CustomCharacter|GetAssociatedPlayerPawn) "(TagName=\"StatManager.AthenaKills\")" 0 true -1.0 -1.0 false))
  (bind _output_get (Variables|Default|Setkill%20count _statvalue))
  (if (> (Math|Integer|%25%28Integer%29 (Variables|Default|Getkill%20count) 2) 0)
    (Rendering|Material|SetScalarParameterValueonMaterials _returnvalue "VGlit_Reactive_Msk_invert" 1.0)
    (Rendering|Material|SetScalarParameterValueonMaterials _returnvalue "VGlit_Reactive_Time_BP" 99999999.0)
    (else
      (Rendering|Material|SetScalarParameterValueonMaterials _returnvalue "VGlit_Reactive_Msk_invert" 0.0)
      (Rendering|Material|SetScalarParameterValueonMaterials _returnvalue "VGlit_Reactive_Time_BP" 99999999.0))))

(event EventBeginPlay
  (bind _returnvalue (CustomItemWrap|GetWrappedMeshComponents))
  (|CallParentFunction)
  (Utilities|FlowControl|Delay 0.05)
  (bind (_statname _statvalue) (Cosmetics|WatchCosmeticStat (CustomCharacter|GetAssociatedPlayerPawn) "(TagName=\"StatManager.AthenaKills\")" 0 true -1.0 -1.0 false))
  (bind _output_get (Variables|Default|Setkill%20count _statvalue))
  (if (> (Math|Integer|%25%28Integer%29 (Variables|Default|Getkill%20count) 2) 0)
    (Rendering|Material|SetScalarParameterValueonMaterials _returnvalue "VGlit_Reactive_Msk_invert" 1.0)
    (Rendering|Material|SetScalarParameterValueonMaterials _returnvalue "VGlit_Reactive_Time_BP" 99999999.0)
    (else
      (Rendering|Material|SetScalarParameterValueonMaterials _returnvalue "VGlit_Reactive_Msk_invert" 0.0)
      (Rendering|Material|SetScalarParameterValueonMaterials _returnvalue "VGlit_Reactive_Time_BP" 99999999.0))))

(event EventOnNewWeapons
  (params (WeaponsAsItems :type Array%20of%20Fort%20Item%20Object%20References))
  (bind _returnvalue (CustomItemWrap|GetWrappedMeshComponents))
  (Utilities|FlowControl|Delay 0.05)
  (bind (_statname _statvalue) (Cosmetics|WatchCosmeticStat (CustomCharacter|GetAssociatedPlayerPawn) "(TagName=\"StatManager.AthenaKills\")" 0 true -1.0 -1.0 false))
  (bind _output_get (Variables|Default|Setkill%20count _statvalue))
  (if (> (Math|Integer|%25%28Integer%29 (Variables|Default|Getkill%20count) 2) 0)
    (Rendering|Material|SetScalarParameterValueonMaterials _returnvalue "VGlit_Reactive_Msk_invert" 1.0)
    (Rendering|Material|SetScalarParameterValueonMaterials _returnvalue "VGlit_Reactive_Time_BP" 99999999.0)
    (else
      (Rendering|Material|SetScalarParameterValueonMaterials _returnvalue "VGlit_Reactive_Msk_invert" 0.0)
      (Rendering|Material|SetScalarParameterValueonMaterials _returnvalue "VGlit_Reactive_Time_BP" 99999999.0))))

(event EventOnNewPlayerWeaponType
  (params (NewType :type EFortItemType%20Enum :default "WorldItem") (OldType :type EFortItemType%20Enum :default "WorldItem"))
  (bind _returnvalue (CustomItemWrap|GetWrappedMeshComponents))
  (Utilities|FlowControl|Delay 0.05)
  (bind (_statname _statvalue) (Cosmetics|WatchCosmeticStat (CustomCharacter|GetAssociatedPlayerPawn) "(TagName=\"StatManager.AthenaKills\")" 0 true -1.0 -1.0 false))
  (bind _output_get (Variables|Default|Setkill%20count _statvalue))
  (if (> (Math|Integer|%25%28Integer%29 (Variables|Default|Getkill%20count) 2) 0)
    (Rendering|Material|SetScalarParameterValueonMaterials _returnvalue "VGlit_Reactive_Msk_invert" 1.0)
    (Rendering|Material|SetScalarParameterValueonMaterials _returnvalue "VGlit_Reactive_Time_BP" 99999999.0)
    (else
      (Rendering|Material|SetScalarParameterValueonMaterials _returnvalue "VGlit_Reactive_Msk_invert" 0.0)
      (Rendering|Material|SetScalarParameterValueonMaterials _returnvalue "VGlit_Reactive_Time_BP" 99999999.0))))

(event EventOnWeaponChanged
  (params (NewWeapon :type Fort%20Weapon%20Object%20Reference) (OldWeapon :type Fort%20Weapon%20Object%20Reference))
  (bind _returnvalue (CustomItemWrap|GetWrappedMeshComponents))
  (Utilities|FlowControl|Delay 0.05)
  (bind (_statname _statvalue) (Cosmetics|WatchCosmeticStat (CustomCharacter|GetAssociatedPlayerPawn) "(TagName=\"StatManager.AthenaKills\")" 0 true -1.0 -1.0 false))
  (bind _output_get (Variables|Default|Setkill%20count _statvalue))
  (if (> (Math|Integer|%25%28Integer%29 (Variables|Default|Getkill%20count) 2) 0)
    (Rendering|Material|SetScalarParameterValueonMaterials _returnvalue "VGlit_Reactive_Msk_invert" 1.0)
    (Rendering|Material|SetScalarParameterValueonMaterials _returnvalue "VGlit_Reactive_Time_BP" 99999999.0)
    (else
      (Rendering|Material|SetScalarParameterValueonMaterials _returnvalue "VGlit_Reactive_Msk_invert" 0.0)
      (Rendering|Material|SetScalarParameterValueonMaterials _returnvalue "VGlit_Reactive_Time_BP" 99999999.0))))

(fn ConstructionScript)
