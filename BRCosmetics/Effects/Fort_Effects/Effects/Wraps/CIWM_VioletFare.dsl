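; VioletFare weapon wrap cosmetic modifier that changes emissive flame brightness as the weapon fires.
; BeginPlay picks the step cap and two brightness arrays for front end or gameplay, while weapon and pawn changes reset the count.
; Each shot increments the capped step, indexes both arrays, and writes start brightness and game-time stamps to wrapped materials.
; Decay needs no tick: each shot lerps the count toward zero based on time past a hold timer.
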(event EventBeginPlay
  (Utilities|FlowControl|Delay 0.2)
  (Variables|Default|SetStepCount 0)
  (if (CustomCosmeticModifier|IsinFrontEnd)
    (Variables|Default|SetMaxSteps 5)
    (Variables|Default|SetBrightness%20Flames (Variables|Default|GetBrightness%20Flames_FE))
    (Variables|Default|SetBrightness%20Flames_A (Variables|Default|GetBrightness%20Flames%20A%20FE))
    (else
      (Variables|Default|SetMaxSteps 9)
      (Variables|Default|SetBrightness%20Flames (Variables|Default|GetBrightness%20Flames_Pie))
      (Variables|Default|SetBrightness%20Flames_A (Variables|Default|GetBrightness%20Flames%20A%20PIE)))))

(event EventOnNewWeapons
  (params (WeaponsAsItems :type Array%20of%20Fort%20Item%20Object%20References))
  (Utilities|FlowControl|Delay 0.01)
  (Variables|Default|SetStepCount 0)
  (Variables|Default|Setwait_timer 0.0))

(event EventOnNewPlayerWeaponType
  (params (NewType :type EFortItemType%20Enum :default "WorldItem") (OldType :type EFortItemType%20Enum :default "WorldItem"))
  (Utilities|FlowControl|Delay 0.01)
  (Variables|Default|SetStepCount 0)
  (Variables|Default|Setwait_timer 0.0))

(event EventOnWeaponChanged
  (params (NewWeapon :type Fort%20Weapon%20Object%20Reference) (OldWeapon :type Fort%20Weapon%20Object%20Reference))
  (Utilities|FlowControl|Delay 0.01)
  (Variables|Default|SetStepCount 0)
  (Variables|Default|Setwait_timer 0.0))

(event EventOnPawnSet
  (Utilities|FlowControl|Delay 0.01)
  (Variables|Default|SetStepCount 0)
  (Variables|Default|Setwait_timer 0.0))

(event EventOnWeaponFired
  (bind _returnvalue (CustomItemWrap|GetWrappedMeshComponents))
  (bind _returnvalue_2 (CustomItemWrap|GetWrappedMeshComponents))
  (bind _output_get (Variables|Default|SetStepCount (Math|Integer|Min(Integer) (+ (Math|Float|Truncate (Math|Float|Lerp (Math|Conversions|ToFloat(Integer) (Variables|Default|GetStepCount)) 0.0 (Math|Float|Clamp(Float) (/ (- (Utilities|Time|GetGameTimeinSeconds) (Variables|Default|Getwait_timer)) 1.0)))) 1) (Variables|Default|GetMaxSteps))))
  (Rendering|Material|SetScalarParameterValueonMaterials _returnvalue_2 "TopS Emissive_02-Brightness-TBP" (Utilities|Time|GetGameTimeinSeconds))
  (Rendering|Material|SetScalarParameterValueonMaterials _returnvalue_2 "TopS Emissive_02_Brightness---Start" (Utilities|Array|Get(aref) (Variables|Default|GetBrightness%20Flames_A) _output_get))
  (Rendering|Material|SetScalarParameterValueonMaterials _returnvalue "TopS Emissive_03-Brightness-TBP" (Utilities|Time|GetGameTimeinSeconds))
  (Rendering|Material|SetScalarParameterValueonMaterials _returnvalue "TopS Emissive_03_Brightness---Start" (Utilities|Array|Get(aref) (Variables|Default|GetBrightness%20Flames) _output_get))
  (Variables|Default|Setwait_timer (+ (Utilities|Time|GetGameTimeinSeconds) 5.5)))

(fn ConstructionScript)
