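; Blueprint logic for the VioletInk item wrap that drives its material's reaction to the weapon being fired.
; On each shot, it writes the current game time into scalar parameters on the wrapped mesh components' materials for emissive brightness, emissive offset and a time reference.
; A stored timestamp, pushed two seconds ahead each shot, decides whether that time also goes to the blend opacity parameter (window expired) or blend offset (still within it).
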(event EventOnWeaponFired
  (bind _returnvalue (Utilities|Time|GetGameTimeinSeconds))
  (bind _returnvalue_2 (CustomItemWrap|GetWrappedMeshComponents))
  (Utilities|FlowControl|Sequence
    (:then_0
      (if (>= (Utilities|Time|GetGameTimeinSeconds) (Variables|Default|GetFloat_Value))
        (Rendering|Material|SetScalarParameterValueonMaterials _returnvalue_2 "TopS_01_Blend_Opacity-TBP" _returnvalue)
        (Rendering|Material|SetScalarParameterValueonMaterials _returnvalue_2 "TopS_01_Blend_Offset-TBP" _returnvalue)))
    (:then_1
      ;[1 1 1] TopsSide V2
      (Rendering|Material|SetScalarParameterValueonMaterials _returnvalue_2 "TopS Emissive_01-Brightness-TBP" _returnvalue)
      (Rendering|Material|SetScalarParameterValueonMaterials _returnvalue_2 "TopS Emissive_01_Offset----TBP" _returnvalue)
      (Rendering|Material|SetScalarParameterValueonMaterials _returnvalue_2 "TimeFrom-BP----------------" _returnvalue)
      (Variables|Default|SetFloat_Value (+ (Utilities|Time|GetGameTimeinSeconds) 2.0)))))

(fn ConstructionScript)
