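; Reactive Voyager Cuboid weapon wrap that animates decal and emissive material parameters on shots and player kills.
; Each shot stamps game time into flipbook decal parameters; a watched AthenaKills stat stamps reactive decal times and, when rising, triggers a timer-driven emissive dip and recovery.
; A looping timer sine-pulses a third emissive layer.
; In the front end, repeated shots mimic the kill reaction using a shot counter.
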(event Custom|FrontendKO
  (bind _returnvalue (Utilities|Time|SetTimerbyEvent (delegate Custom|FEExtended_KO) 0.5 true))
  (Variables|Default|SetCurrentTimer _returnvalue))

(event Custom|CustomEvent)

(event Custom|FEExtended_KO
  (bind _result (Math|Integer|IncrementInt (Variables|Default|GetIterationCount))
    (:%20%20
      (if (>= _result 7)
        (Utilities|Time|ClearandInvalidateTimerbyHandle (Variables|Default|GetCurrentTimer))
        (Variables|Default|SetIterationCount 0)))))

(event EventBeginPlay
  (Utilities|FlowControl|Delay 0.2))

(event EventOnWeaponFired
  (bind _returnvalue (CustomItemWrap|GetWrappedMeshComponents))
  (bind _returnvalue_2 (Utilities|Time|GetGameTimeinSeconds))
  (bind _returnvalue_3 (Utilities|Time|GetGameTimeinSeconds))
  (bind _returnvalue_4 (CustomItemWrap|GetWrappedMeshComponents))
  (bind _returnvalue_5 (CustomItemWrap|GetWrappedMeshComponents))
  (bind _returnvalue_6 (Utilities|Time|GetGameTimeinSeconds))
  (bind _returnvalue_7 (CustomItemWrap|GetWrappedMeshComponents))
  (bind _returnvalue_8 (Utilities|Time|GetGameTimeinSeconds))
  (Rendering|Material|SetScalarParameterValueonMaterials _returnvalue_5 "FDecalV2_FlipBook_Time_FromBP" _returnvalue_6)
  (Rendering|Material|SetScalarParameterValueonMaterials _returnvalue_5 "FDecalV2_Emissive_Brightness_Time_BP" _returnvalue_6)
  (Rendering|Material|SetScalarParameterValueonMaterials _returnvalue_5 "FDecalaV2_Opacity_Time_BP_Z" _returnvalue_6)
  (if (CustomCosmeticModifier|IsinFrontEnd)
    (Utilities|FlowControl|Delay 0.001)
    (bind _output_get (Variables|Default|SetShotsFired (+ (Variables|Default|GetShotsFired) 1)))
    (if (>= _output_get 5)
      (if (not (== (Variables|Default|GetShotsFired) 7))
        (Variables|Default|SetShotsFired 0))
      (Utilities|FlowControl|Sequence
        (:then_0
          ;[1 1 1 0.5] Decal Reactive
          (Rendering|Material|SetScalarParameterValueonMaterials _returnvalue "EDRec_TimeFrom_BP_DecalA" _returnvalue_2)
          (Rendering|Material|SetScalarParameterValueonMaterials _returnvalue "EDRec_TimeFrom_BP_DecalB" _returnvalue_2)
          (Rendering|Material|SetScalarParameterValueonMaterials _returnvalue "EDRec_TimeFrom_BP_DecalC" _returnvalue_2)
          (Rendering|Material|SetScalarParameterValueonMaterials _returnvalue "EDRec_TimeFrom_BP_DecalD" _returnvalue_2)
          (Rendering|Material|SetScalarParameterValueonMaterials _returnvalue_7 "EDRec_Emissive_A_Brightness_Time_BP" _returnvalue_8)
          (Rendering|Material|SetScalarParameterValueonMaterials _returnvalue_7 "EDRec_Emissive_B_Brightness_Time_BP" _returnvalue_8)
          (Rendering|Material|SetScalarParameterValueonMaterials _returnvalue_7 "EDRec_Emissive_C_Brightness_Time_BP" _returnvalue_8)
          (Rendering|Material|SetScalarParameterValueonMaterials _returnvalue_7 "EDRec_Emissive_D_Brightness_Time_BP" _returnvalue_8)
          (Rendering|Material|SetScalarParameterValueonMaterials _returnvalue_4 "EDRec_Emissive_A_Gradent_Time_BP" _returnvalue_3)
          (Rendering|Material|SetScalarParameterValueonMaterials _returnvalue_4 "EDRec_Emissive_B_Gradent_Time_BP" _returnvalue_3)
          (Rendering|Material|SetScalarParameterValueonMaterials _returnvalue_4 "EDRec_Emissive_C_Gradent_Time_BP" _returnvalue_3)
          (Rendering|Material|SetScalarParameterValueonMaterials _returnvalue_4 "EDRec_Emissive_D_Gradent_Time_BP" _returnvalue_3)
          (Rendering|Material|SetScalarParameterValueonMaterials _returnvalue_4 "EDRec_Decal_B_Animation_Time_BP" _returnvalue_3)
          (Rendering|Material|SetScalarParameterValueonMaterials _returnvalue_4 "EDRec_Decal_C_Animation_Time_BP" _returnvalue_3)
          (Rendering|Material|SetScalarParameterValueonMaterials _returnvalue_4 "EDRec_DecalA__Sine_TimeFrom_BP" _returnvalue_3))
        (:then_1
          (if (> (stale EventOnPawnSet _statvalue) (Variables|Default|GetKillCounter))
            (CallFunction|SineAnimation)
            (Utilities|FlowControl|Delay 0.75)
            (CallFunction|SineAnimOFF)))
        (:then_2
          (Variables|Default|SetKillCounter (stale EventOnPawnSet _statvalue)))))))

(event EventOnPawnSet
  (bind _returnvalue (CustomItemWrap|GetWrappedMeshComponents))
  (bind _returnvalue_2 (Utilities|Time|GetGameTimeinSeconds))
  (bind _returnvalue_3 (Utilities|Time|GetGameTimeinSeconds))
  (bind _returnvalue_4 (CustomItemWrap|GetWrappedMeshComponents))
  (bind _returnvalue_5 (CustomItemWrap|GetWrappedMeshComponents))
  (bind _returnvalue_6 (Utilities|Time|GetGameTimeinSeconds))
  (Utilities|FlowControl|Sequence
    (:then_0
      (bind (_statname _statvalue) (Cosmetics|WatchCosmeticStat (CustomCharacter|GetAssociatedPlayerPawn) "(TagName=\"StatManager.AthenaKills\")")
        (:OnStatChanged
          (Utilities|FlowControl|Sequence
            (:then_0
              ;[1 1 1 0.5] Decal Reactive
              (Rendering|Material|SetScalarParameterValueonMaterials _returnvalue "EDRec_TimeFrom_BP_DecalA" _returnvalue_2)
              (Rendering|Material|SetScalarParameterValueonMaterials _returnvalue "EDRec_TimeFrom_BP_DecalB" _returnvalue_2)
              (Rendering|Material|SetScalarParameterValueonMaterials _returnvalue "EDRec_TimeFrom_BP_DecalC" _returnvalue_2)
              (Rendering|Material|SetScalarParameterValueonMaterials _returnvalue "EDRec_TimeFrom_BP_DecalD" _returnvalue_2)
              (Rendering|Material|SetScalarParameterValueonMaterials _returnvalue_5 "EDRec_Emissive_A_Brightness_Time_BP" _returnvalue_6)
              (Rendering|Material|SetScalarParameterValueonMaterials _returnvalue_5 "EDRec_Emissive_B_Brightness_Time_BP" _returnvalue_6)
              (Rendering|Material|SetScalarParameterValueonMaterials _returnvalue_5 "EDRec_Emissive_C_Brightness_Time_BP" _returnvalue_6)
              (Rendering|Material|SetScalarParameterValueonMaterials _returnvalue_5 "EDRec_Emissive_D_Brightness_Time_BP" _returnvalue_6)
              (Rendering|Material|SetScalarParameterValueonMaterials _returnvalue_4 "EDRec_Emissive_A_Gradent_Time_BP" _returnvalue_3)
              (Rendering|Material|SetScalarParameterValueonMaterials _returnvalue_4 "EDRec_Emissive_B_Gradent_Time_BP" _returnvalue_3)
              (Rendering|Material|SetScalarParameterValueonMaterials _returnvalue_4 "EDRec_Emissive_C_Gradent_Time_BP" _returnvalue_3)
              (Rendering|Material|SetScalarParameterValueonMaterials _returnvalue_4 "EDRec_Emissive_D_Gradent_Time_BP" _returnvalue_3)
              (Rendering|Material|SetScalarParameterValueonMaterials _returnvalue_4 "EDRec_Decal_B_Animation_Time_BP" _returnvalue_3)
              (Rendering|Material|SetScalarParameterValueonMaterials _returnvalue_4 "EDRec_Decal_C_Animation_Time_BP" _returnvalue_3)
              (Rendering|Material|SetScalarParameterValueonMaterials _returnvalue_4 "EDRec_DecalA__Sine_TimeFrom_BP" _returnvalue_3))
            (:then_1
              (if (> _statvalue (Variables|Default|GetKillCounter))
                (CallFunction|SineAnimation)
                (Utilities|FlowControl|Delay 0.75)
                (CallFunction|SineAnimOFF)))
            (:then_2
              (Variables|Default|SetKillCounter _statvalue))))))
    (:then_1
      (Utilities|Time|SetTimerbyEvent (delegate Custom|LoopAnim) 0.02 true))))

(event Custom|LoopAnim
  (for _array_element (CustomItemWrap|GetWrappedMeshComponents)
    (Rendering|Material|SetScalarParameterValueonMaterials _array_element "TopS Emissive_03_Brightness---End" (Math|Float|MapRangeClamped (Math|Trig|Sin(Radians) (* (Utilities|Time|GetGameTimeinSeconds) (Variables|Default|GetAnim%20Velocity))) -1.0 1.0 0.15 1.0))))

(event Custom|SineAnimation
  (Variables|Default|SetReverse false)
  (Variables|Default|SetStartTime (Utilities|Time|GetGameTimeinSeconds))
  (bind _returnvalue (Utilities|Time|SetTimerbyEvent (delegate Custom|SineAnimUpdate) 0.02 true))
  (Variables|Default|SetTimerHandle _returnvalue))

(event Custom|SineAnimOFF
  (Variables|Default|SetReverse true)
  (Variables|Default|SetStartTime (Utilities|Time|GetGameTimeinSeconds))
  (bind _returnvalue (Utilities|Time|SetTimerbyEvent (delegate Custom|SineAnimUpdate) 0.02 true))
  (Variables|Default|SetTimerHandle _returnvalue))

(event Custom|SineAnimUpdate
  (bind _returnvalue (CustomItemWrap|GetWrappedMeshComponents))
  (bind _returnvalue_2 (Utilities|Time|GetGameTimeinSeconds))
  (bind _starttime (Variables|Default|GetStartTime))
  (bind _returnvalue_3 (- _returnvalue_2 _starttime))
  (if (>= _returnvalue_3 0.75)
    (Utilities|Time|ClearandInvalidateTimerbyHandle (Variables|Default|GetTimerHandle))
    (if (Variables|Default|GetReverse)
      (Rendering|Material|SetScalarParameterValueonMaterials _returnvalue "TopS Emissive_01_Brightness---End" 0.5)
      (Rendering|Material|SetScalarParameterValueonMaterials _returnvalue "TopS Emissive_02_Brightness---End" 2.0)
      (else
        (Rendering|Material|SetScalarParameterValueonMaterials _returnvalue "TopS Emissive_01_Brightness---End" 0.2)
        (Rendering|Material|SetScalarParameterValueonMaterials _returnvalue "TopS Emissive_02_Brightness---End" 0.2)))
    (elif (Variables|Default|GetReverse)
      ; FadeIn
      (Rendering|Material|SetScalarParameterValueonMaterials _returnvalue "TopS Emissive_01_Brightness---End" (Math|Float|MapRangeClamped _returnvalue_3 0.0 0.5 0.2 0.5))
      (Rendering|Material|SetScalarParameterValueonMaterials _returnvalue "TopS Emissive_02_Brightness---End" (Math|Float|MapRangeClamped _returnvalue_3 0.0 0.5 0.2 2.0))
      (else
        ; FadeOut
        (Rendering|Material|SetScalarParameterValueonMaterials _returnvalue "TopS Emissive_01_Brightness---End" (Math|Float|MapRangeClamped _returnvalue_3 0.0 0.5 0.5 0.2))
        (Rendering|Material|SetScalarParameterValueonMaterials _returnvalue "TopS Emissive_02_Brightness---End" (Math|Float|MapRangeClamped _returnvalue_3 0.0 0.5 2.0 0.2))))))

(fn RestartAnim)

(fn ConstructionScript)
