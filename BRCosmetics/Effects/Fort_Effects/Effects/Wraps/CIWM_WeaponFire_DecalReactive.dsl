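; Item wrap modifier that makes a weapon wrap's materials react to firing, with an optional front-end time-of-day preview.
; On weapon fire, after a short delay, it writes game time in seconds to each scalar parameter named in an array, leaving the animation to the material.
; On begin play, if time-of-day is enabled and in the front end, it sets a TOD flag and runs a timeline driving scalar and color overrides on the wrapped meshes.
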(event EventBeginPlay
  (bind _returnvalue (CustomItemWrap|GetWrappedMeshComponents))
  (Utilities|FlowControl|Delay 0.2)
  (if (Variables|Default|GetEnable%20TOD)
    (if (CustomCosmeticModifier|IsinFrontEnd)
      (Rendering|Material|SetScalarParameterValueonMaterials (CustomItemWrap|GetWrappedMeshComponents) "00-FE-TOD--" 1.0)
      (bind (_direction _tod_value _newtrack_1 _newtrack_0) (Timeline|Timeline PlayFromStart)
        (:Update
          (Rendering|Material|SetScalarParameterValueonMaterials _returnvalue "00-FE-TODCC-Override" _tod_value)
          (Rendering|Material|SetColorParameterValueonMaterials _returnvalue "00-FE-TODV4-Override" _newtrack_0))))))

(event EventOnWeaponFired
  (Utilities|FlowControl|Delay 0.05)
  (for _array_element (Variables|Default|GetArray)
    (Rendering|Material|SetScalarParameterValueonMaterials (CustomItemWrap|GetWrappedMeshComponents) _array_element (Utilities|Time|GetGameTimeinSeconds))))

(fn ConstructionScript)
