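; Custom item wrap modifier that makes a weapon wrap's materials react each time the weapon fires.
; Each shot adds a small step to a stored float and writes it as the color-curve offset of emissive channels A–D, so the shift accumulates.
; It also stamps the current game time into the decal, emissive brightness and gradient time parameters on the wrapped meshes, leaving any time-based animation to the material.
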(event EventOnWeaponFired
  (bind _returnvalue (Utilities|Time|GetGameTimeinSeconds))
  (bind _returnvalue_2 (CustomItemWrap|GetWrappedMeshComponents))
  (bind _returnvalue_3 (Utilities|Time|GetGameTimeinSeconds))
  (bind _returnvalue_4 (CustomItemWrap|GetWrappedMeshComponents))
  (bind _returnvalue_5 (CustomItemWrap|GetWrappedMeshComponents))
  (bind _returnvalue_6 (Utilities|Time|GetGameTimeinSeconds))
  (bind _returnvalue_7 (CustomItemWrap|GetWrappedMeshComponents))
  (Utilities|FlowControl|Sequence
    (:then_0
      (bind _output_get (Variables|Default|SetFloat_Value (+ (Variables|Default|GetFloat_Value) 0.05)))
      ;[1 1 1] TopsSide V2
      (Rendering|Material|SetScalarParameterValueonMaterials _returnvalue_2 "EDRec_Emissive_A_CC_Offset" _output_get)
      (Rendering|Material|SetScalarParameterValueonMaterials _returnvalue_2 "EDRec_Emissive_B_CC_Offset" _output_get)
      (Rendering|Material|SetScalarParameterValueonMaterials _returnvalue_2 "EDRec_Emissive_C_CC_Offset" _output_get)
      (Rendering|Material|SetScalarParameterValueonMaterials _returnvalue_2 "EDRec_Emissive_D_CC_Offset" _output_get))
    (:then_1
      ;[1 1 1 0.5] Decal Reactive
      (Rendering|Material|SetScalarParameterValueonMaterials _returnvalue_5 "EDRec_TimeFrom_BP_DecalA" _returnvalue_6)
      (Rendering|Material|SetScalarParameterValueonMaterials _returnvalue_5 "EDRec_TimeFrom_BP_DecalB" _returnvalue_6)
      (Rendering|Material|SetScalarParameterValueonMaterials _returnvalue_5 "EDRec_TimeFrom_BP_DecalC" _returnvalue_6)
      (Rendering|Material|SetScalarParameterValueonMaterials _returnvalue_5 "EDRec_TimeFrom_BP_DecalD" _returnvalue_6)
      (Rendering|Material|SetScalarParameterValueonMaterials _returnvalue_7 "EDRec_Emissive_A_Brightness_Time_BP" _returnvalue)
      (Rendering|Material|SetScalarParameterValueonMaterials _returnvalue_7 "EDRec_Emissive_B_Brightness_Time_BP" _returnvalue)
      (Rendering|Material|SetScalarParameterValueonMaterials _returnvalue_7 "EDRec_Emissive_C_Brightness_Time_BP" _returnvalue)
      (Rendering|Material|SetScalarParameterValueonMaterials _returnvalue_7 "EDRec_Emissive_D_Brightness_Time_BP" _returnvalue)
      (Rendering|Material|SetScalarParameterValueonMaterials _returnvalue_4 "EDRec_Emissive_A_Gradent_Time_BP" _returnvalue_3)
      (Rendering|Material|SetScalarParameterValueonMaterials _returnvalue_4 "EDRec_Emissive_B_Gradent_Time_BP" _returnvalue_3)
      (Rendering|Material|SetScalarParameterValueonMaterials _returnvalue_4 "EDRec_Emissive_C_Gradent_Time_BP" _returnvalue_3)
      (Rendering|Material|SetScalarParameterValueonMaterials _returnvalue_4 "EDRec_Emissive_D_Gradent_Time_BP" _returnvalue_3))))

(fn ConstructionScript)
