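; Custom item wrap Blueprint that makes weapon wrap decals react each time the weapon fires.
; On the weapon-fired event, it writes the current game time into time parameters on the materials of the wrapped mesh components.
; A sequence covers three parameter families: Left/Right decal v1, Left/Right V2, and four-channel (A-D) reactive decals.
; The timestamps feed parameters named for flipbook, opacity, emissive gradient and brightness, leaving any animation to the materials.
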(event EventOnWeaponFired
  (bind _returnvalue (Utilities|Time|GetGameTimeinSeconds))
  (bind _returnvalue_2 (CustomItemWrap|GetWrappedMeshComponents))
  (bind _returnvalue_3 (CustomItemWrap|GetWrappedMeshComponents))
  (bind _returnvalue_4 (Utilities|Time|GetGameTimeinSeconds))
  (bind _returnvalue_5 (CustomItemWrap|GetWrappedMeshComponents))
  (bind _returnvalue_6 (Utilities|Time|GetGameTimeinSeconds))
  (bind _returnvalue_7 (CustomItemWrap|GetWrappedMeshComponents))
  (bind _returnvalue_8 (Utilities|Time|GetGameTimeinSeconds))
  (bind _returnvalue_9 (CustomItemWrap|GetWrappedMeshComponents))
  (bind _returnvalue_10 (Utilities|Time|GetGameTimeinSeconds))
  (Utilities|FlowControl|Sequence
    (:then_0
      ;[1 1 1] Left Right v1
      (Rendering|Material|SetScalarParameterValueonMaterials _returnvalue_3 "FDecal_Emissive_Gradent_Time_BP" _returnvalue_4)
      (Rendering|Material|SetScalarParameterValueonMaterials _returnvalue_3 "FDecal_Opacity_Time_BP_Z" _returnvalue_4)
      (Rendering|Material|SetScalarParameterValueonMaterials _returnvalue_3 "FDecal_Emissive_Brightness_Time_BP" _returnvalue_4))
    (:then_1
      ; Left\Right V2
      (Rendering|Material|SetScalarParameterValueonMaterials _returnvalue_5 "FDecalV2_FlipBook_Time_FromBP" _returnvalue_6)
      (Rendering|Material|SetScalarParameterValueonMaterials _returnvalue_5 "FDecalaV2_Opacity_Time_BP_Z" _returnvalue_6)
      (Rendering|Material|SetScalarParameterValueonMaterials _returnvalue_5 "FDecalV2_Emissive_Brightness_Time_BP" _returnvalue_6)
      (Rendering|Material|SetScalarParameterValueonMaterials _returnvalue_5 "FDecalV2_Emissive_Gradent_Time_BP" _returnvalue_6)
      (Rendering|Material|SetScalarParameterValueonMaterials _returnvalue_5 "FDecalV2_Counter_Brightness_Time_BP" _returnvalue_6))
    (:then_2
      ;[1 1 1 0.5] Decal Reactive
      (Rendering|Material|SetScalarParameterValueonMaterials _returnvalue_7 "EDRec_TimeFrom_BP_DecalA" _returnvalue_8)
      (Rendering|Material|SetScalarParameterValueonMaterials _returnvalue_7 "EDRec_TimeFrom_BP_DecalB" _returnvalue_8)
      (Rendering|Material|SetScalarParameterValueonMaterials _returnvalue_7 "EDRec_TimeFrom_BP_DecalC" _returnvalue_8)
      (Rendering|Material|SetScalarParameterValueonMaterials _returnvalue_7 "EDRec_TimeFrom_BP_DecalD" _returnvalue_8)
      (Rendering|Material|SetScalarParameterValueonMaterials _returnvalue_9 "EDRec_Emissive_A_Brightness_Time_BP" _returnvalue_10)
      (Rendering|Material|SetScalarParameterValueonMaterials _returnvalue_9 "EDRec_Emissive_B_Brightness_Time_BP" _returnvalue_10)
      (Rendering|Material|SetScalarParameterValueonMaterials _returnvalue_9 "EDRec_Emissive_C_Brightness_Time_BP" _returnvalue_10)
      (Rendering|Material|SetScalarParameterValueonMaterials _returnvalue_9 "EDRec_Emissive_D_Brightness_Time_BP" _returnvalue_10)
      (Rendering|Material|SetScalarParameterValueonMaterials _returnvalue_2 "EDRec_Emissive_A_Gradent_Time_BP" _returnvalue)
      (Rendering|Material|SetScalarParameterValueonMaterials _returnvalue_2 "EDRec_Emissive_B_Gradent_Time_BP" _returnvalue)
      (Rendering|Material|SetScalarParameterValueonMaterials _returnvalue_2 "EDRec_Emissive_C_Gradent_Time_BP" _returnvalue)
      (Rendering|Material|SetScalarParameterValueonMaterials _returnvalue_2 "EDRec_Emissive_D_Gradent_Time_BP" _returnvalue))))

(fn ConstructionScript)
