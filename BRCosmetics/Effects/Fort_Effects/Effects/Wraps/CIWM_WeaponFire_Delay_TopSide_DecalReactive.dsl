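; Weapon-wrap Blueprint that makes wrap materials react to weapon fire through TopSide emissive and reactive decal effects.
; Each shot writes game time into decal A-D scalar parameters on wrapped meshes.
; TopSide parameters get the same timestamp, gated by a Pause flag that clears after a delay computed from pause time and falloff.
; Pawn and weapon events set those values to constants, despite comments saying they come from the material.
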(event EventOnPawnSet
  ; gets duration and fall off from material to set pause length
  (Utilities|FlowControl|Delay 0.2)
  (Variables|Default|Setpause%20time 2.0)
  (Variables|Default|Setfalloff 0.25))

(event EventOnWeaponChanged
  (params (NewWeapon :type Fort%20Weapon%20Object%20Reference) (OldWeapon :type Fort%20Weapon%20Object%20Reference))
  ; gets duration and fall off from material to set pause length
  (Utilities|FlowControl|Delay 0.2)
  (Variables|Default|Setpause%20time 2.0)
  (Variables|Default|Setfalloff 0.25))

(event EventOnNewWeapons
  (params (WeaponsAsItems :type Array%20of%20Fort%20Item%20Object%20References))
  ; gets duration and fall off from material to set pause length
  (Utilities|FlowControl|Delay 0.2)
  (Variables|Default|Setpause%20time 2.0)
  (Variables|Default|Setfalloff 0.25))

(event EventOnWeaponFired
  (bind _returnvalue (CustomItemWrap|GetWrappedMeshComponents))
  (bind _returnvalue_2 (Utilities|Time|GetGameTimeinSeconds))
  (bind _returnvalue_3 (Utilities|Time|GetGameTimeinSeconds))
  (bind _returnvalue_4 (Utilities|Time|GetGameTimeinSeconds))
  (bind _returnvalue_5 (CustomItemWrap|GetWrappedMeshComponents))
  (bind _returnvalue_6 (CustomItemWrap|GetWrappedMeshComponents))
  (bind _returnvalue_7 (Utilities|Time|GetGameTimeinSeconds))
  (bind _returnvalue_8 (CustomItemWrap|GetWrappedMeshComponents))
  (Utilities|FlowControl|Sequence
    (:then_0
      (if (not (Variables|Default|GetPause))
        (Variables|Default|SetPause true)
        ;[1 1 1] TopsSide V2
        (Rendering|Material|SetScalarParameterValueonMaterials _returnvalue "TopS Emissive_02-Brightness-TBP" _returnvalue_3)
        (Rendering|Material|SetScalarParameterValueonMaterials _returnvalue "TopS Emissive_02_Offset----TBP" _returnvalue_3)
        (Rendering|Material|SetScalarParameterValueonMaterials _returnvalue "TopS_01_Blend_Opacity-TBP" _returnvalue_3)
        (Rendering|Material|SetScalarParameterValueonMaterials _returnvalue "TopS Emissive_01-Brightness-TBP" _returnvalue_3)
        (Rendering|Material|SetScalarParameterValueonMaterials _returnvalue "TopS Emissive_01_Offset----TBP" _returnvalue_3)
        (Rendering|Material|SetScalarParameterValueonMaterials _returnvalue "TopS Emissive_03-Brightness-TBP" _returnvalue_3)
        (Rendering|Material|SetScalarParameterValueonMaterials _returnvalue "TopS Emissive_03_Offset----TBP" _returnvalue_3)
        (Rendering|Material|SetScalarParameterValueonMaterials _returnvalue "TopS Emissive_01_Speed-TBP" _returnvalue_3)
        (Rendering|Material|SetScalarParameterValueonMaterials _returnvalue "TopS Emissive_02_Speed-TBP" _returnvalue_3)
        (Rendering|Material|SetScalarParameterValueonMaterials _returnvalue "TopS Emissive_03_Speed-TBP" _returnvalue_3)
        (Utilities|FlowControl|Delay (* (+ (Variables|Default|Getpause%20time) 0.25) (Variables|Default|Getfalloff)))
        (Variables|Default|SetPause false)))
    (:then_1
      ;[1 1 1 0.5] Decal Reactive
      (Rendering|Material|SetScalarParameterValueonMaterials _returnvalue_6 "EDRec_TimeFrom_BP_DecalA" _returnvalue_7)
      (Rendering|Material|SetScalarParameterValueonMaterials _returnvalue_6 "EDRec_TimeFrom_BP_DecalB" _returnvalue_7)
      (Rendering|Material|SetScalarParameterValueonMaterials _returnvalue_6 "EDRec_TimeFrom_BP_DecalC" _returnvalue_7)
      (Rendering|Material|SetScalarParameterValueonMaterials _returnvalue_6 "EDRec_TimeFrom_BP_DecalD" _returnvalue_7)
      (Rendering|Material|SetScalarParameterValueonMaterials _returnvalue_8 "EDRec_Emissive_A_Brightness_Time_BP" _returnvalue_2)
      (Rendering|Material|SetScalarParameterValueonMaterials _returnvalue_8 "EDRec_Emissive_B_Brightness_Time_BP" _returnvalue_2)
      (Rendering|Material|SetScalarParameterValueonMaterials _returnvalue_8 "EDRec_Emissive_C_Brightness_Time_BP" _returnvalue_2)
      (Rendering|Material|SetScalarParameterValueonMaterials _returnvalue_8 "EDRec_Emissive_D_Brightness_Time_BP" _returnvalue_2)
      (Rendering|Material|SetScalarParameterValueonMaterials _returnvalue_5 "EDRec_Emissive_A_Gradent_Time_BP" _returnvalue_4)
      (Rendering|Material|SetScalarParameterValueonMaterials _returnvalue_5 "EDRec_Emissive_B_Gradent_Time_BP" _returnvalue_4)
      (Rendering|Material|SetScalarParameterValueonMaterials _returnvalue_5 "EDRec_Emissive_C_Gradent_Time_BP" _returnvalue_4)
      (Rendering|Material|SetScalarParameterValueonMaterials _returnvalue_5 "EDRec_Emissive_D_Gradent_Time_BP" _returnvalue_4))))

(fn ConstructionScript)
