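; Item wrap modifier Blueprint that shifts the color of a wrap each time the wrapped weapon fires.
; On each fire event it adds Step Amount to a stored float, which accumulates with no clamp or reset.
; The new value goes to five differently named hue and color shift scalar parameters on the materials of all wrapped mesh components.
; The construction script is empty.
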(event EventOnWeaponFired
  (bind _returnvalue (CustomItemWrap|GetWrappedMeshComponents))
  (bind _output_get (Variables|Default|SetFloat_Value (+ (Variables|Default|GetFloat_Value) (Variables|Default|GetStep%20Amount))))
  (Rendering|Material|SetScalarParameterValueonMaterials _returnvalue "NeonBL_CC_Shift" _output_get)
  (Rendering|Material|SetScalarParameterValueonMaterials _returnvalue "HueShift_EmissiveColor" _output_get)
  (Rendering|Material|SetScalarParameterValueonMaterials _returnvalue "HueShift_BaseColor" _output_get)
  (Rendering|Material|SetScalarParameterValueonMaterials _returnvalue "SP_Gradent_Color_Shift" _output_get)
  (Rendering|Material|SetScalarParameterValueonMaterials _returnvalue "00-Color_Shift" _output_get))

(fn ConstructionScript)
