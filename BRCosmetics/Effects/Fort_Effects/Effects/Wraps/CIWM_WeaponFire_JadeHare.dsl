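; Blueprint for the JadeHare weapon wrap that makes the wrap's materials react when the weapon fires.
; Each shot stamps current game time into brightness time parameters on the wrapped mesh components' materials.
; Two pause flags with half-second delays act as a two-slot cooldown for emissive offset triggers.
; A shot fires offset 01 if free, otherwise offset 02, and skips both while both are busy.
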(event EventOnWeaponFired
  (bind _returnvalue (CustomItemWrap|GetWrappedMeshComponents))
  (bind _returnvalue_2 (Utilities|Time|GetGameTimeinSeconds))
  ;[1 1 1 0.5] Decal Reactive
  (Rendering|Material|SetScalarParameterValueonMaterials _returnvalue "EDRec_Emissive_A_Brightness_Time_BP" _returnvalue_2)
  (Rendering|Material|SetScalarParameterValueonMaterials _returnvalue "EDRec_Emissive_B_Brightness_Time_BP" _returnvalue_2)
  (Rendering|Material|SetScalarParameterValueonMaterials _returnvalue "TopS Emissive_03-Brightness-TBP" _returnvalue_2)
  (if (Variables|Default|GetPause)
    (if (not (Variables|Default|GetPause_B))
      (Variables|Default|SetPause_B true)
      (Rendering|Material|SetScalarParameterValueonMaterials (CustomItemWrap|GetWrappedMeshComponents) "TopS Emissive_02_Offset----TBP" (Utilities|Time|GetGameTimeinSeconds))
      (Utilities|FlowControl|Delay 0.5)
      (Variables|Default|SetPause_B false))
    (else
      (Variables|Default|SetPause true)
      (Rendering|Material|SetScalarParameterValueonMaterials (CustomItemWrap|GetWrappedMeshComponents) "TopS Emissive_01_Offset----TBP" (Utilities|Time|GetGameTimeinSeconds))
      (Utilities|FlowControl|Delay 0.5)
      (Variables|Default|SetPause false))))

(fn ConstructionScript)
