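; Weapon wrap effect Blueprint that triggers animated left/right decal effects on a wrapped weapon's materials when the weapon fires.
; On each fire event, it writes the current game time into scalar parameters on the wrapped mesh components, so material-side animations can time themselves from the shot.
; Separate parameter sets drive the V1 and V2 decal variants for flipbook, opacity, emissive gradient and brightness.
; The V1 flipbook timestamp is set after a brief delay.
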(event EventOnWeaponFired
  (bind _returnvalue (CustomItemWrap|GetWrappedMeshComponents))
  (bind _returnvalue_2 (Utilities|Time|GetGameTimeinSeconds))
  (bind _returnvalue_3 (CustomItemWrap|GetWrappedMeshComponents))
  (bind _returnvalue_4 (Utilities|Time|GetGameTimeinSeconds))
  (Utilities|FlowControl|Sequence
    (:then_0
      (Utilities|FlowControl|Delay 0.01)
      (Rendering|Material|SetScalarParameterValueonMaterials (CustomItemWrap|GetWrappedMeshComponents) "FDecal_FlipBook_Time_FromBP" (Utilities|Time|GetGameTimeinSeconds)))
    (:then_1
      ;[1 1 1] Left Right v1
      (Rendering|Material|SetScalarParameterValueonMaterials _returnvalue "FDecal_Emissive_Gradent_Time_BP" _returnvalue_2)
      (Rendering|Material|SetScalarParameterValueonMaterials _returnvalue "FDecal_Opacity_Time_BP_Z" _returnvalue_2)
      (Rendering|Material|SetScalarParameterValueonMaterials _returnvalue "FDecal_Emissive_Brightness_Time_BP" _returnvalue_2))
    (:then_2
      ; Left\Right V2
      (Rendering|Material|SetScalarParameterValueonMaterials _returnvalue_3 "FDecalV2_FlipBook_Time_FromBP" _returnvalue_4)
      (Rendering|Material|SetScalarParameterValueonMaterials _returnvalue_3 "FDecalaV2_Opacity_Time_BP_Z" _returnvalue_4)
      (Rendering|Material|SetScalarParameterValueonMaterials _returnvalue_3 "FDecalV2_Emissive_Brightness_Time_BP" _returnvalue_4)
      (Rendering|Material|SetScalarParameterValueonMaterials _returnvalue_3 "FDecalV2_Emissive_Gradent_Time_BP" _returnvalue_4)
      (Rendering|Material|SetScalarParameterValueonMaterials _returnvalue_3 "FDecalV2_Counter_Brightness_Time_BP" _returnvalue_4))))

(fn ConstructionScript)
