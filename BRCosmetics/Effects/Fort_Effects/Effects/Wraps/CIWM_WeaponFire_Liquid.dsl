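; Custom item wrap Blueprint for a liquid weapon wrap that reacts to the weapon being fired.
; On each weapon-fired event, it writes the current game time into the Liquid_TimeFrom-BP scalar parameter on the materials of all wrapped mesh components.
; The Blueprint only supplies that timestamp; it has no other graph logic.
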(event EventOnWeaponFired
  ;[1 1 1] Liquid
  (Rendering|Material|SetScalarParameterValueonMaterials (CustomItemWrap|GetWrappedMeshComponents) "Liquid_TimeFrom-BP" (Utilities|Time|GetGameTimeinSeconds)))

(fn ConstructionScript)
