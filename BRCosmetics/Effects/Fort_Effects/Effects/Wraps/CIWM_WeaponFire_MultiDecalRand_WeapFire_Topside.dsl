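; Custom item wrap Blueprint that drives animated emissive and decal material effects on wrapped meshes when the weapon fires.
; After a short delay, it writes game time to top-side emissive parameters and cycles through decal slots A–D, each with a half-second re-trigger lock.
; RandomSet timestamps each decal's animation and randomizes rotation and offset per item type (weapon, pickaxe, vehicle, etc.), packing scale into the offset color's alpha.
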(event EventBeginPlay
  (Utilities|FlowControl|Delay 0.2))

(event EventOnWeaponFired
  (bind _returnvalue (Utilities|Time|GetGameTimeinSeconds))
  (bind _returnvalue_2 (CustomItemWrap|GetWrappedMeshComponents))
  (Utilities|FlowControl|Delay 0.2)
  (Utilities|FlowControl|Sequence
    (:then_0
      (Utilities|FlowControl|Delay 0.001)
      (Utilities|FlowControl|Sequence
        (:then_0
          ;[1 1 1] TopsSide V2
          (Rendering|Material|SetScalarParameterValueonMaterials _returnvalue_2 "TopS Emissive_02-Brightness-TBP" _returnvalue)
          (Rendering|Material|SetScalarParameterValueonMaterials _returnvalue_2 "TopS Emissive_02_Offset----TBP" _returnvalue)
          (Rendering|Material|SetScalarParameterValueonMaterials _returnvalue_2 "TopS_01_Blend_Opacity-TBP" _returnvalue)
          (Rendering|Material|SetScalarParameterValueonMaterials _returnvalue_2 "TimeFrom-BP----------------" _returnvalue))
        (:then_1
          (Rendering|Material|SetScalarParameterValueonMaterials _returnvalue_2 "TopS Emissive_01-Brightness-TBP" _returnvalue)
          (Rendering|Material|SetScalarParameterValueonMaterials _returnvalue_2 "TopS Emissive_01_Offset----TBP" _returnvalue)
          (Rendering|Material|SetScalarParameterValueonMaterials _returnvalue_2 "TopS Emissive_03-Brightness-TBP" _returnvalue)
          (Rendering|Material|SetScalarParameterValueonMaterials _returnvalue_2 "TopS Emissive_03_Offset----TBP" _returnvalue))))
    (:then_1
      (switch int (Variables|Default|GetFire%20seq)
        (:0
          (if (not (Variables|Default|GetPause_A))
            (Variables|Default|SetPause_A true)
            (Variables|Default|SetFire%20seq 1)
            ;[1 1 1] multi A
            (CallFunction|RandomSet "EDRec_Decal_A_Animation_Time_BP" "EDRec_DecalA_Offset_Weapon" "EDRec_DecalA_Rotation_Weapon" -0.38 -0.1 -0.01 0.01 0.25 0.4 "EDRec_TimeFrom_BP_DecalA")
            (CallFunction|RandomSet "0" "EDRec_DecalA_Offset_Pickaxe" "EDRec_DecalA_Rotation_Pickaxe" -0.3 -0.1 0.2 -0.2 0.4 0.7)
            (CallFunction|RandomSet "0" "EDRec_DecalA_Offset_Backpack----A=Scale" "EDRec_DecalA_Rotation_Backpack" -0.15 0.15 -0.25 0.0 0.4 0.8)
            (CallFunction|RandomSet "0" "EDRec_DecalA_Offset_Character" "EDRec_DecalA_Rotation_Character" -0.15 0.15 -0.2 0.0 0.4 0.8)
            (CallFunction|RandomSet "0" "EDRec_DecalA_Offset_Vehicle" "EDRec_DecalA_Rotation_Vehicle" 0.15 0.4 0.0 0.15 0.25 0.5)
            (CallFunction|RandomSet "0" "EDRec_DecalA_Offset_VehicleSide" "EDRec_DecalA_Rotation_Vehicle_Side" 0.3 0.42 -0.1 -0.13 0.2 0.3)
            (CallFunction|RandomSet "0" "EDRec_DecalA_Offset_Mech" "EDRec_DecalA_Rotation_Mech" -0.1 0.1 0.32 0.2 0.25 0.35)
            (Utilities|FlowControl|Delay 0.5)
            (Variables|Default|SetPause_A false)))
        (:1
          (if (not (Variables|Default|GetPause_B))
            (Variables|Default|SetPause_B true)
            (Variables|Default|SetFire%20seq 2)
            ;[1 1 1] multi B
            ;[1 1 1] Comment
            (CallFunction|RandomSet "EDRec_Decal_B_Animation_Time_BP" "EDRec_DecalB_Offset_Weapon" "EDRec_DecalB_Rotation_Weapon" -0.1 0.2 -0.01 0.01 0.25 0.4 "EDRec_TimeFrom_BP_DecalB")
            (CallFunction|RandomSet "0" "EDRec_DecalB_Offset_Pickaxe" "EDRec_DecalB_Rotation_Pickaxe" -0.1 0.15 0.2 -0.2 0.4 0.7)
            (CallFunction|RandomSet "0" "EDRec_DecalB_Offset_Backpack----A=Scale" "EDRec_DecalB_Rotation_Backpack" -0.15 0.15 0.0 0.25 0.4 0.8)
            (CallFunction|RandomSet "0" "EDRec_DecalB_Offset_Character" "EDRec_DecalB_Rotation_Character" -0.15 0.15 0.0 0.28 0.4 0.8)
            (CallFunction|RandomSet "0" "EDRec_DecalB_Offset_Vehicle" "EDRec_DecalB_Rotation_Vehicle" 0.15 0.4 -0.15 0.0 0.25 0.5)
            (CallFunction|RandomSet "0" "EDRec_DecalB_Offset_VehicleSide" "EDRec_DecalB_Rotation_Vehicle_Side" 0.3 0.42 -0.1 -0.13 0.2 0.3)
            (CallFunction|RandomSet "0" "EDRec_DecalB_Offset_Mech" "EDRec_DecalB_Rotation_Mech" -0.1 0.1 0.32 0.2 0.25 0.35)
            (Utilities|FlowControl|Delay 0.5)
            (Variables|Default|SetPause_B false)))
        (:2
          (if (not (Variables|Default|GetPause_C))
            (Variables|Default|SetPause_C true)
            (Variables|Default|SetFire%20seq 3)
            ;[1 1 1] C
            (CallFunction|RandomSet "EDRec_Decal_C_Animation_Time_BP" "EDRec_DecalC_Offset_Weapon" "EDRec_DecalC_Rotation_Weapon" -0.38 -0.1 -0.01 0.01 0.25 0.4 "EDRec_TimeFrom_BP_DecalC")
            (CallFunction|RandomSet "0" "EDRec_DecalC_Offset_Pickaxe" "EDRec_DecalC_Rotation_Pickaxe" -0.3 -0.1 0.2 -0.2 0.4 0.7)
            (CallFunction|RandomSet "0" "EDRec_DecalC_Offset_Backpack----A=Scale" "EDRec_DecalC_Rotation_Backpack" -0.15 0.15 -0.25 0.0 0.4 0.8)
            (CallFunction|RandomSet "0" "EDRec_DecalC_Offset_Character" "EDRec_DecalC_Rotation_Character" -0.15 0.15 -0.2 0.0 0.4 0.8)
            (CallFunction|RandomSet "0" "EDRec_DecalC_Offset_Vehicle" "EDRec_DecalC_Rotation_Vehicle" 0.15 0.4 0.0 0.15 0.25 0.5)
            (CallFunction|RandomSet "0" "EDRec_DecalC_Offset_VehicleSide" "EDRec_DecalC_Rotation_Vehicle_Side" 0.3 0.42 -0.1 -0.13 0.2 0.3)
            (CallFunction|RandomSet "0" "EDRec_DecalC_Offset_Mech" "EDRec_DecalC_Rotation_Mech" -0.1 0.1 0.32 0.2 0.25 0.35)
            (Utilities|FlowControl|Delay 0.5)
            (Variables|Default|SetPause_C false)))
        (:3
          (if (not (Variables|Default|GetPause_D))
            (Variables|Default|SetPause_D true)
            (Variables|Default|SetFire%20seq 0)
            ;[1 1 1] C
            (CallFunction|RandomSet "EDRec_Decal_D_Animation_Time_BP" "EDRec_DecalD_Offset_Weapon" "EDRec_DecalD_Rotation_Weapon" -0.1 0.2 -0.01 0.01 0.25 0.4 "EDRec_TimeFrom_BP_DecalD")
            (CallFunction|RandomSet "0" "EDRec_DecalD_Offset_Pickaxe" "EDRec_DecalD_Rotation_Pickaxe" -0.1 0.15 0.2 -0.2 0.4 0.7)
            (CallFunction|RandomSet "0" "EDRec_DecalD_Offset_Backpack----A=Scale" "EDRec_DecalD_Rotation_Backpack" -0.15 0.15 0.0 0.25 0.4 0.8)
            (CallFunction|RandomSet "0" "EDRec_DecalD_Offset_Character" "EDRec_DecalD_Rotation_Character" -0.15 0.15 0.0 0.28 0.4 0.8)
            (CallFunction|RandomSet "0" "EDRec_DecalD_Offset_Vehicle" "EDRec_DecalD_Rotation_Vehicle" 0.15 0.4 -0.15 0.0 0.25 0.5)
            (CallFunction|RandomSet "0" "EDRec_DecalD_Offset_VehicleSide" "EDRec_DecalD_Rotation_Vehicle_Side" 0.3 0.42 -0.1 -0.13 0.2 0.3)
            (CallFunction|RandomSet "0" "EDRec_DecalD_Offset_Mech" "EDRec_DecalD_Rotation_Mech" -0.1 0.1 0.32 0.2 0.25 0.35)
            (Utilities|FlowControl|Delay 0.5)
            (Variables|Default|SetPause_D false)))))))

(fn RandomSet
  (params (TimeBP_Flipbook%20Triger :type Name) (Offset :type Name) (Rotation :type Name) (X=min :type Float%20%28double-precision%29) (X=Max :type Float%20%28double-precision%29) (Y=Min :type Float%20%28double-precision%29) (Y=Max :type Float%20%28double-precision%29) (Scale=Min :type Float%20%28double-precision%29) (Scale=Max :type Float%20%28double-precision%29) (opacity :type Name))
  (bind _returnvalue (CustomItemWrap|GetWrappedMeshComponents))
  (bind _returnvalue_2 (Utilities|Time|GetGameTimeinSeconds))
  (bind _stream (Variables|Default|GetStream))
  (bind _returnvalue_3 (Math|Random|RandomFloatinRangefromStream _stream X=min X=Max))
  (Rendering|Material|SetScalarParameterValueonMaterials _returnvalue TimeBP_Flipbook%20Triger _returnvalue_2)
  (Rendering|Material|SetScalarParameterValueonMaterials _returnvalue opacity _returnvalue_2)
  (Rendering|Material|SetColorParameterValueonMaterials _returnvalue Offset (Math|Color|MakeColor _returnvalue_3 (Math|Random|RandomFloatinRangefromStream _stream Y=Min Y=Max) 0.0 (Math|Random|RandomFloatinRangefromStream _stream Scale=Min Scale=Max)))
  (Rendering|Material|SetScalarParameterValueonMaterials _returnvalue Rotation (Math|Float|Lerp 0.1 -0.1 (Math|Random|RandomFloat)))
  (Variables|Default|SetPosition _returnvalue_3))

(fn ConstructionScript)
