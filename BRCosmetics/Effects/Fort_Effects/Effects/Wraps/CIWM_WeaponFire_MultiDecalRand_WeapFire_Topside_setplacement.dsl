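; Item wrap Blueprint that animates top-side emissives and four randomized decals on wrapped meshes when the weapon fires.
; After a short delay, it stamps game time into emissive material parameters and cycles decal slots A–D, each locked for half a second.
; RandomSet timestamps the decal animation and jitters hand-set placements per item type (weapon, pickaxe, backpack, character, vehicle, mech), packing offset and scale into a color parameter and randomizing rotation.
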(event EventBeginPlay
  (Utilities|FlowControl|Delay 0.2))

(event EventOnWeaponFired
  (bind _returnvalue (Utilities|Time|GetGameTimeinSeconds))
  (bind _returnvalue_2 (CustomItemWrap|GetWrappedMeshComponents))
  (Utilities|FlowControl|Delay 0.2)
  (Utilities|FlowControl|Sequence
    (:then_0
      (Utilities|FlowControl|Delay 0.001)
      (Utilities|FlowControl|Sequence
        (:then_0
          ;[1 1 1] TopsSide V2
          (Rendering|Material|SetScalarParameterValueonMaterials _returnvalue_2 "TopS Emissive_02-Brightness-TBP" _returnvalue)
          (Rendering|Material|SetScalarParameterValueonMaterials _returnvalue_2 "TopS Emissive_02_Offset----TBP" _returnvalue)
          (Rendering|Material|SetScalarParameterValueonMaterials _returnvalue_2 "TopS_01_Blend_Opacity-TBP" _returnvalue)
          (Rendering|Material|SetScalarParameterValueonMaterials _returnvalue_2 "TimeFrom-BP----------------" _returnvalue))
        (:then_1
          (Rendering|Material|SetScalarParameterValueonMaterials _returnvalue_2 "TopS Emissive_01-Brightness-TBP" _returnvalue)
          (Rendering|Material|SetScalarParameterValueonMaterials _returnvalue_2 "TopS Emissive_01_Offset----TBP" _returnvalue)
          (Rendering|Material|SetScalarParameterValueonMaterials _returnvalue_2 "TopS Emissive_03-Brightness-TBP" _returnvalue)
          (Rendering|Material|SetScalarParameterValueonMaterials _returnvalue_2 "TopS Emissive_03_Offset----TBP" _returnvalue))))
    (:then_1
      (switch int (Variables|Default|GetFire%20seq)
        (:0
          (if (not (Variables|Default|GetPause_A))
            (Variables|Default|SetPause_A true)
            (Variables|Default|SetFire%20seq 1)
            ;[1 1 1] multi A
            (CallFunction|RandomSet "EDRec_Decal_A_Animation_Time_BP" "EDRec_DecalA_Offset_Weapon" "EDRec_DecalA_Rotation_Weapon" -0.35 0.0 0.25 "EDRec_TimeFrom_BP_DecalA" 0.03 0.05)
            (CallFunction|RandomSet "0" "EDRec_DecalA_Offset_Pickaxe" "EDRec_DecalA_Rotation_Pickaxe" -0.35 0.05 0.35 "None" 0.05 0.05)
            (CallFunction|RandomSet "0" "EDRec_DecalA_Offset_Backpack----A=Scale" "EDRec_DecalA_Rotation_Backpack" -0.15 0.3 0.45 "None" 0.05 0.05)
            (CallFunction|RandomSet "0" "EDRec_DecalA_Offset_Character" "EDRec_DecalA_Rotation_Character" -0.12 0.025 0.35 "None" 0.05 0.05)
            (CallFunction|RandomSet "0" "EDRec_DecalA_Offset_Vehicle" "EDRec_DecalA_Rotation_Vehicle" 0.25 -0.125 0.2 "None" 0.05 0.05)
            (CallFunction|RandomSet "0" "EDRec_DecalA_Offset_VehicleSide" "EDRec_DecalA_Rotation_Vehicle_Side" 0.33 -0.125 0.14 "None" 0.05 0.05)
            (CallFunction|RandomSet "0" "EDRec_DecalA_Offset_Mech" "EDRec_DecalA_Rotation_Mech" 0.02 0.23 0.18 "None" 0.05 0.05)
            (Utilities|FlowControl|Delay 0.5)
            (Variables|Default|SetPause_A false)))
        (:1
          (if (not (Variables|Default|GetPause_B))
            (Variables|Default|SetPause_B true)
            (Variables|Default|SetFire%20seq 2)
            (CallFunction|RandomSet "EDRec_Decal_B_Animation_Time_BP" "EDRec_DecalB_Offset_Weapon" "EDRec_DecalB_Rotation_Weapon" 0.13 -0.025 0.3 "EDRec_TimeFrom_BP_DecalB" 0.03 0.05)
            ;[1 1 1] multi B
            ;[1 1 1] Comment
            (CallFunction|RandomSet "0" "EDRec_DecalB_Offset_Pickaxe" "EDRec_DecalB_Rotation_Pickaxe" -0.08 0.08 0.35 "None" 0.05 0.05)
            (CallFunction|RandomSet "0" "EDRec_DecalB_Offset_Backpack----A=Scale" "EDRec_DecalB_Rotation_Backpack" 0.15 0.1 0.45 "None" 0.05 0.05)
            (CallFunction|RandomSet "0" "EDRec_DecalB_Offset_Character" "EDRec_DecalB_Rotation_Character" 0.1 0.125 0.35 "None" 0.05 0.05)
            (CallFunction|RandomSet "0" "EDRec_DecalB_Offset_Vehicle" "EDRec_DecalB_Rotation_Vehicle" 0.35 0.125 0.2 "None" 0.05 0.05)
            (CallFunction|RandomSet "0" "EDRec_DecalB_Offset_VehicleSide" "EDRec_DecalB_Rotation_Vehicle_Side" 0.0 -0.08 0.14 "None" 0.05 0.05)
            (CallFunction|RandomSet "0" "EDRec_DecalB_Offset_Mech" "EDRec_DecalB_Rotation_Mech" 0.1 0.3 0.18 "None" 0.05 0.05)
            (Utilities|FlowControl|Delay 0.5)
            (Variables|Default|SetPause_B false)))
        (:2
          (if (not (Variables|Default|GetPause_C))
            (Variables|Default|SetPause_C true)
            (Variables|Default|SetFire%20seq 3)
            ;[1 1 1] C
            (CallFunction|RandomSet "EDRec_Decal_C_Animation_Time_BP" "EDRec_DecalC_Offset_Weapon" "EDRec_DecalC_Rotation_Weapon" 0.4 -0.02 0.25 "EDRec_TimeFrom_BP_DecalC" 0.03 0.05)
            (CallFunction|RandomSet "0" "EDRec_DecalC_Offset_Pickaxe" "EDRec_DecalC_Rotation_Pickaxe" 0.05 -0.05 0.03 "None" 0.05 0.05)
            (CallFunction|RandomSet "0" "EDRec_DecalC_Offset_Backpack----A=Scale" "EDRec_DecalC_Rotation_Backpack" 0.13 -0.25 0.5 "None" 0.05 0.05)
            (CallFunction|RandomSet "0" "EDRec_DecalC_Offset_Character" "EDRec_DecalC_Rotation_Character" 0.11 -0.15 0.3 "None" 0.05 0.05)
            (CallFunction|RandomSet "0" "EDRec_DecalC_Offset_Vehicle" "EDRec_DecalC_Rotation_Vehicle" 0.12 -0.125 0.18 "None" 0.05 0.05)
            (CallFunction|RandomSet "0" "EDRec_DecalC_Offset_VehicleSide" "EDRec_DecalC_Rotation_Vehicle_Side" -0.14 0.0 0.14 "None" 0.05 0.05)
            (CallFunction|RandomSet "0" "EDRec_DecalC_Offset_Mech" "EDRec_DecalC_Rotation_Mech" -0.1 0.3 0.18 "None" 0.05 0.05)
            (Utilities|FlowControl|Delay 0.5)
            (Variables|Default|SetPause_C false)))
        (:3
          (if (not (Variables|Default|GetPause_D))
            (Variables|Default|SetPause_D true)
            (Variables|Default|SetFire%20seq 0)
            ;[1 1 1] C
            (CallFunction|RandomSet "EDRec_Decal_D_Animation_Time_BP" "EDRec_DecalD_Offset_Weapon" "EDRec_DecalD_Rotation_Weapon" -0.13 -0.01 0.25 "EDRec_TimeFrom_BP_DecalD" 0.03 0.05)
            (CallFunction|RandomSet "0" "EDRec_DecalD_Offset_Pickaxe" "EDRec_DecalD_Rotation_Pickaxe" -0.15 -0.1 0.3 "None" 0.05 0.05)
            (CallFunction|RandomSet "0" "EDRec_DecalD_Offset_Backpack----A=Scale" "EDRec_DecalD_Rotation_Backpack" -0.15 -0.1 0.5 "None" 0.05 0.05)
            (CallFunction|RandomSet "0" "EDRec_DecalD_Offset_Character" "EDRec_DecalD_Rotation_Character" -0.1 0.0 0.3 "None" 0.05 0.05)
            (CallFunction|RandomSet "0" "EDRec_DecalD_Offset_Vehicle" "EDRec_DecalD_Rotation_Vehicle" 0.15 0.125 0.2 "None" 0.05 0.05)
            (CallFunction|RandomSet "0" "EDRec_DecalD_Offset_VehicleSide" "EDRec_DecalD_Rotation_Vehicle_Side" -0.03 -0.01 0.14 "None" 0.05 0.05)
            (CallFunction|RandomSet "0" "EDRec_DecalD_Offset_Mech" "EDRec_DecalD_Rotation_Mech" -0.03 0.25 0.18 "None" 0.05 0.05)
            (Utilities|FlowControl|Delay 0.5)
            (Variables|Default|SetPause_D false)))))))

(fn RandomSet
  (params (TimeBP_Flipbook%20Triger :type Name) (Offset :type Name) (Rotation :type Name) (R :type Float%20%28double-precision%29) (G :type Float%20%28double-precision%29) (Scale :type Float%20%28double-precision%29) (opacity :type Name) (offset%20Range :type Float%20%28single-precision%29) (Scale%20Range :type Float%20%28single-precision%29))
  (bind _returnvalue (CustomItemWrap|GetWrappedMeshComponents))
  (bind _returnvalue_2 (Utilities|Time|GetGameTimeinSeconds))
  (bind _returnvalue_3 (* offset%20Range -1.0))
  (bind _stream (Variables|Default|GetStream))
  (bind _returnvalue_4 (Math|Random|RandomFloatinRangefromStream _stream _returnvalue_3 offset%20Range))
  (bind _returnvalue_5 (+ R _returnvalue_4))
  (Rendering|Material|SetScalarParameterValueonMaterials _returnvalue TimeBP_Flipbook%20Triger _returnvalue_2)
  (Rendering|Material|SetScalarParameterValueonMaterials _returnvalue opacity _returnvalue_2)
  (Rendering|Material|SetColorParameterValueonMaterials _returnvalue Offset (Math|Color|MakeColor _returnvalue_5 (+ G (Math|Random|RandomFloatinRangefromStream _stream (* offset%20Range -1.0) offset%20Range)) 0.0 (+ Scale (Math|Random|RandomFloatinRangefromStream _stream (* Scale%20Range -1.0) Scale%20Range))))
  (Rendering|Material|SetScalarParameterValueonMaterials _returnvalue Rotation (Math|Float|Lerp 0.1 -0.1 (Math|Random|RandomFloat)))
  (Variables|Default|SetPosition _returnvalue_5))

(fn ConstructionScript)
