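; Weapon wrap cosmetic modifier that triggers one of four decal groups on the wrapped meshes when the weapon fires.
; A shuffled index array picks the group per shot, reshuffling after four, and game time is written to that group's material parameters.
; Per-group pause flags block retriggers for about half a second.
; RandomSet, uncalled here, randomizes flipbook frame, offset and rotation; a damage stat watch has no visible handler.
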(event EventBeginPlay
  (Utilities|FlowControl|Delay 0.2)
  (Utilities|Array|Shuffle (Variables|Default|GetAray))
  (if (CustomCosmeticModifier|IsinFrontEnd)
    (Timeline|Timeline PlayFromStart)))

(event EventOnPawnSet
  (Cosmetics|WatchCosmeticStat (CustomCharacter|GetAssociatedPlayerPawn) "(TagName=\"StatManager.DamageReceived\")" 0 false))

(event EventOnWeaponFired
  (Utilities|FlowControl|Delay 0.05)
  (bind _output_get (Variables|Default|SetFire%20seq (Variables|Default|GetFire%20seq)))
  (switch int (Utilities|Array|Get(aref) (Variables|Default|GetAray) _output_get)
    (:0
      (if (not (Variables|Default|Getpause_A))
        (Variables|Default|Setpause_A true)
        (for _array_element (Variables|Default|GetDeacal%20A)
          (Rendering|Material|SetScalarParameterValueonMaterials (CustomItemWrap|GetWrappedMeshComponents) _array_element (Utilities|Time|GetGameTimeinSeconds)))
        (Utilities|FlowControl|Sequence
          (:then_0
            _)
          (:then_1
            (Utilities|FlowControl|Delay 0.6)
            (Variables|Default|Setpause_A false)))))
    (:1
      (if (not (Variables|Default|Getpause_B))
        (Variables|Default|Setpause_B true)
        (for _array_element_2 (Variables|Default|GetDecal_B)
          (Rendering|Material|SetScalarParameterValueonMaterials (CustomItemWrap|GetWrappedMeshComponents) _array_element_2 (Utilities|Time|GetGameTimeinSeconds)))
        (Utilities|FlowControl|Sequence
          (:then_0
            _)
          (:then_1
            (Utilities|FlowControl|Delay 0.5)
            (Variables|Default|Setpause_B false)))))
    (:2
      (if (not (Variables|Default|Getpause_C))
        (Variables|Default|Setpause_C true)
        (for _array_element_3 (Variables|Default|GetDecal_C)
          (Rendering|Material|SetScalarParameterValueonMaterials (CustomItemWrap|GetWrappedMeshComponents) _array_element_3 (Utilities|Time|GetGameTimeinSeconds)))
        (Utilities|FlowControl|Sequence
          (:then_0
            _)
          (:then_1
            (Utilities|FlowControl|Delay 0.6)
            (Variables|Default|Setpause_C false)))))
    (:3
      (if (not (Variables|Default|Getpause_D))
        (Variables|Default|Setpause_D true)
        (for _array_element_4 (Variables|Default|GetDecal_D)
          (Rendering|Material|SetScalarParameterValueonMaterials (CustomItemWrap|GetWrappedMeshComponents) _array_element_4 (Utilities|Time|GetGameTimeinSeconds)))
        (Utilities|FlowControl|Sequence
          (:then_0
            _)
          (:then_1
            (Utilities|FlowControl|Delay 0.6)
            (Variables|Default|Setpause_D false))))))
  (bind _output_get_2 (Variables|Default|SetFire%20seq (+ (Variables|Default|GetFire%20seq) 1)))
  (if (== _output_get_2 4)
    (Utilities|Array|Shuffle (Variables|Default|GetAray))
    (Variables|Default|SetFire%20seq 0)))

(fn RandomSet
  (params (FlipBook%20Frame :type Name) (TimeBP :type Name) (Offset :type Name) (Rotation :type Name) (X=min :type Float%20%28double-precision%29) (X=Max :type Float%20%28double-precision%29) (Y=Min :type Float%20%28double-precision%29) (Y=Max :type Float%20%28double-precision%29) (Rot=Min :type Float%20%28double-precision%29) (Rot=Max :type Float%20%28double-precision%29) (TimeBP-first :type Name))
  (bind _returnvalue (CustomItemWrap|GetWrappedMeshComponents))
  (bind _stream (Variables|Default|GetStream))
  (bind _returnvalue_2 (Math|Random|RandomFloatinRangefromStream _stream X=min X=Max))
  (bind _returnvalue_3 (Math|Random|RandomIntegerinRange 0 8))
  (bind _returnvalue_4 (Math|Conversions|ToFloat(Integer) _returnvalue_3))
  (bind _returnvalue_5 (/ _returnvalue_4 8.0))
  (Rendering|Material|SetScalarParameterValueonMaterials _returnvalue FlipBook%20Frame _returnvalue_5)
  (Rendering|Material|SetScalarParameterValueonMaterials _returnvalue TimeBP-first (Utilities|Time|GetGameTimeinSeconds))
  (Rendering|Material|SetVectorParameterValueonMaterials _returnvalue Offset (Math|Vector|MakeVector _returnvalue_2 (Math|Random|RandomFloatinRangefromStream _stream Y=Min Y=Max)))
  (Rendering|Material|SetScalarParameterValueonMaterials _returnvalue Rotation (Math|Random|RandomFloatinRange Rot=Min Rot=Max))
  (Rendering|Material|SetScalarParameterValueonMaterials _returnvalue TimeBP (Utilities|Time|GetGameTimeinSeconds))
  (Variables|Default|SetPosition _returnvalue_2)
  (Development|PrintString (Utilities|String|ToString(Float) _returnvalue_5) true true "(R=0.060001,G=1.000000,B=0.000000,A=1.000000)" 3.0))

(fn ConstructionScript)
