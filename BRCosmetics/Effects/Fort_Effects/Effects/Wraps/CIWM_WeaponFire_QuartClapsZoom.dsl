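; Weapon wrap modifier that drives animated decal and emissive effects on the wrapped meshes when the weapon fires.
; Each shot stamps game time into material parameters, with channels B, C and D throttled by separate cooldowns of increasing length.
; Channel A brightness builds with a shot counter that decays over time after a pause.
; The counter's cap depends on the current weapon's clip size, or on front-end state at BeginPlay.
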(event EventBeginPlay
  (Utilities|FlowControl|Delay 0.2)
  (Utilities|FlowControl|Delay 0.2)
  (if (CustomCosmeticModifier|IsinFrontEnd)
    (Variables|Default|SetMaxSteps 5)
    (else
      (Variables|Default|SetMaxSteps 10))))

(event EventOnWeaponChanged
  (params (NewWeapon :type Fort%20Weapon%20Object%20Reference) (OldWeapon :type Fort%20Weapon%20Object%20Reference))
  (Utilities|FlowControl|Delay 0.2)
  (if (> (FortWeapon|GetBulletsPerClip (Variables|/Script/FortniteGame.FortPawn|GetCurrentWeapon (CustomCharacter|GetAssociatedPlayerPawn))) 5)
    (Variables|Default|SetMaxSteps 10)
    (else
      (Variables|Default|SetMaxSteps 5))))

(event EventOnNewWeapons
  (params (WeaponsAsItems :type Array%20of%20Fort%20Item%20Object%20References))
  (Utilities|FlowControl|Delay 0.2)
  (if (> (FortWeapon|GetBulletsPerClip (Variables|/Script/FortniteGame.FortPawn|GetCurrentWeapon (CustomCharacter|GetAssociatedPlayerPawn))) 5)
    (Variables|Default|SetMaxSteps 10)
    (else
      (Variables|Default|SetMaxSteps 5))))

(event EventOnNewPlayerWeaponType
  (params (NewType :type EFortItemType%20Enum :default "WorldItem") (OldType :type EFortItemType%20Enum :default "WorldItem"))
  (Utilities|FlowControl|Delay 0.2)
  (if (> (FortWeapon|GetBulletsPerClip (Variables|/Script/FortniteGame.FortPawn|GetCurrentWeapon (CustomCharacter|GetAssociatedPlayerPawn))) 5)
    (Variables|Default|SetMaxSteps 10)
    (else
      (Variables|Default|SetMaxSteps 5))))

(event EventOnWeaponFired
  (bind _returnvalue (Utilities|Time|GetGameTimeinSeconds))
  (bind _returnvalue_2 (CustomItemWrap|GetWrappedMeshComponents))
  (bind _returnvalue_3 (CustomItemWrap|GetWrappedMeshComponents))
  (Utilities|FlowControl|Delay 0.05)
  (Utilities|FlowControl|Sequence
    (:then_0
      (if (not (Variables|Default|GetPause))
        (Variables|Default|SetPause true)
        (Rendering|Material|SetScalarParameterValueonMaterials _returnvalue_2 "EDRec_TimeFrom_BP_DecalB" _returnvalue)
        (Rendering|Material|SetScalarParameterValueonMaterials _returnvalue_2 "EDRec_Emissive_B_Brightness_Time_BP" _returnvalue)
        (Utilities|FlowControl|Delay 0.4)
        (Variables|Default|SetPause false)))
    (:then_1
      (if (not (Variables|Default|GetPause_0))
        (Variables|Default|SetPause_0 true)
        (Rendering|Material|SetScalarParameterValueonMaterials _returnvalue_2 "EDRec_TimeFrom_BP_DecalC" _returnvalue)
        (Rendering|Material|SetScalarParameterValueonMaterials _returnvalue_2 "EDRec_Emissive_C_Brightness_Time_BP" _returnvalue)
        (Utilities|FlowControl|Delay 0.75)
        (Variables|Default|SetPause_0 false)))
    (:then_2
      (if (not (Variables|Default|GetPause_1))
        (Variables|Default|SetPause_1 true)
        (Rendering|Material|SetScalarParameterValueonMaterials _returnvalue_2 "EDRec_TimeFrom_BP_DecalD" _returnvalue)
        (Rendering|Material|SetScalarParameterValueonMaterials _returnvalue_2 "EDRec_Emissive_D_Brightness_Time_BP" _returnvalue)
        (Utilities|FlowControl|Delay 1.25)
        (Variables|Default|SetPause_1 false)))
    (:then_3
      (bind _output_get (Variables|Default|SetStepCount (Math|Integer|Min(Integer) (+ (Math|Float|Truncate (Math|Float|Lerp (Math|Conversions|ToFloat(Integer) (Variables|Default|GetStepCount)) 0.0 (Math|Float|Clamp(Float) (/ (- (Utilities|Time|GetGameTimeinSeconds) (Variables|Default|Getwait_timer)) 4.0)))) 1) (Variables|Default|GetMaxSteps))))
      (Rendering|Material|SetScalarParameterValueonMaterials _returnvalue_3 "EDRec_Emissive_A_Brightness_Time__Start" (Math|Float|Lerp 0.0 5e+01 (/ (Math|Conversions|ToFloat(Integer) _output_get) (Variables|Default|GetMaxSteps))))
      (Rendering|Material|SetScalarParameterValueonMaterials _returnvalue_3 "EDRec_Emissive_A_Brightness_Time_BP" (Utilities|Time|GetGameTimeinSeconds))
      (Variables|Default|Setwait_timer (+ (Utilities|Time|GetGameTimeinSeconds) (Variables|Default|GetPauseTime))))))

(fn ConstructionScript)
