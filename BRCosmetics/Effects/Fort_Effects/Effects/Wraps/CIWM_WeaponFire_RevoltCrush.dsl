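; Custom item wrap Blueprint, apparently for the RevoltCrush wrap, that animates the wrapped weapon's materials when it fires.
; A shot with Pause false sets Pause, switches the decal flipbook to frame 1 and plays the Frosting timeline from the start.
; The timeline lerps emissive brightness and contrast in opposite directions, then clears Pause and restores frame 0 on finish.
; Shots during Pause set the timeline's time to 1.0 instead of restarting it.
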(event EventOnWeaponFired
  (bind _returnvalue (CustomItemWrap|GetWrappedMeshComponents))
  (if (Variables|Default|GetPause)
    (bind (_direction _frostingcontrast) (Timeline|Frosting SetNewTime :NewTime 1.0)
      (:Update
        (Rendering|Material|SetScalarParameterValueonMaterials _returnvalue "TopS Emissive_02_Brightness---End" (Math|Float|Lerp 0.0 1.0 _frostingcontrast))
        (Rendering|Material|SetScalarParameterValueonMaterials _returnvalue "TopS Emissive_02-Contrast" (Math|Float|Lerp 1.0 0.0 _frostingcontrast)))
      (:Finished
        (Variables|Default|SetPause false)
        (Rendering|Material|SetScalarParameterValueonMaterials (CustomItemWrap|GetWrappedMeshComponents) "FDecalV2_FlipBook_Frame #" 0.0)))
    (else
      (Variables|Default|SetPause true)
      (Rendering|Material|SetScalarParameterValueonMaterials (CustomItemWrap|GetWrappedMeshComponents) "FDecalV2_FlipBook_Frame #" 1.0)
      (Timeline|Frosting PlayFromStart))))

(fn ConstructionScript)
