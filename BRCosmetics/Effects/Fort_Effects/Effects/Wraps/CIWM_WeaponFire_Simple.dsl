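; Blueprint for a simple Custom Item Wrap modifier that makes wrap materials react to weapon fire.
; On the weapon-fired event it gets the wrapped mesh components and writes the current game time in seconds into a scalar material parameter, named by a game-time variable.
; The Blueprint only supplies the timestamp, leaving any visual response to the wrap materials, and its construction script is empty.
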(event EventOnWeaponFired
  (Rendering|Material|SetScalarParameterValueonMaterials (CustomItemWrap|GetWrappedMeshComponents) (Variables|Default|GetName_GameTime_TBP) (Utilities|Time|GetGameTimeinSeconds)))

(fn ConstructionScript)
