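; Weapon wrap modifier that builds up opacity and emissive glow on the wrapped meshes as the weapon fires.
; Each shot increments a capped step count, first decayed by time elapsed past a pause-plus-cooldown window, and lerps opacity, brightness and darken-base parameters by that fraction.
; It also writes game time into material parameters as timestamps.
; Step cap and pause are chosen by front-end state on BeginPlay and by clip size on weapon changes.
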(event EventOnWeaponFired
  (CallFunction|WeaponFireBase))

(event EventBeginPlay
  (Utilities|FlowControl|Delay 0.2)
  (if (CustomCosmeticModifier|IsinFrontEnd)
    (Variables|EnviromentVarable|SetMaxSteps (Variables|Default|GetShot%20Count%20FrontEnd))
    (Variables|EnviromentVarable|SetPauseTime (Variables|Default|GetPause%20Timer%20FrontEnd))
    (else
      (Variables|EnviromentVarable|SetMaxSteps (Variables|Default|GetShot%20Count%20PIE)))))

(event EventOnWeaponChanged
  (params (NewWeapon :type Fort%20Weapon%20Object%20Reference) (OldWeapon :type Fort%20Weapon%20Object%20Reference))
  (Utilities|FlowControl|Delay 0.2)
  (if (> (FortWeapon|GetBulletsPerClip (Variables|/Script/FortniteGame.FortPawn|GetCurrentWeapon (CustomCharacter|GetAssociatedPlayerPawn))) 5)
    (Variables|EnviromentVarable|SetMaxSteps (Variables|Default|GetShot%20Count%20PIE))
    (Variables|EnviromentVarable|SetPauseTime (Variables|Default|GetPause%20Timer%20Ammo%20>5))
    (else
      (Variables|EnviromentVarable|SetMaxSteps (Variables|Default|GetShot%20Count%20Ammo%20<%205))
      (Variables|EnviromentVarable|SetPauseTime (Variables|Default|GetPause%20Timer%20Ammo%20<5)))))

(event EventOnNewWeapons
  (params (WeaponsAsItems :type Array%20of%20Fort%20Item%20Object%20References))
  (Utilities|FlowControl|Delay 0.2)
  (if (> (FortWeapon|GetBulletsPerClip (Variables|/Script/FortniteGame.FortPawn|GetCurrentWeapon (CustomCharacter|GetAssociatedPlayerPawn))) 5)
    (Variables|EnviromentVarable|SetMaxSteps (Variables|Default|GetShot%20Count%20PIE))
    (Variables|EnviromentVarable|SetPauseTime (Variables|Default|GetPause%20Timer%20Ammo%20>5))
    (else
      (Variables|EnviromentVarable|SetMaxSteps (Variables|Default|GetShot%20Count%20Ammo%20<%205))
      (Variables|EnviromentVarable|SetPauseTime (Variables|Default|GetPause%20Timer%20Ammo%20<5)))))

(event EventOnNewPlayerWeaponType
  (params (NewType :type EFortItemType%20Enum :default "WorldItem") (OldType :type EFortItemType%20Enum :default "WorldItem"))
  (Utilities|FlowControl|Delay 0.2)
  (if (> (FortWeapon|GetBulletsPerClip (Variables|/Script/FortniteGame.FortPawn|GetCurrentWeapon (CustomCharacter|GetAssociatedPlayerPawn))) 5)
    (Variables|EnviromentVarable|SetMaxSteps (Variables|Default|GetShot%20Count%20PIE))
    (Variables|EnviromentVarable|SetPauseTime (Variables|Default|GetPause%20Timer%20Ammo%20>5))
    (else
      (Variables|EnviromentVarable|SetMaxSteps (Variables|Default|GetShot%20Count%20Ammo%20<%205))
      (Variables|EnviromentVarable|SetPauseTime (Variables|Default|GetPause%20Timer%20Ammo%20<5)))))

(event Custom|WeaponFire_Base
  (bind _returnvalue (CustomItemWrap|GetWrappedMeshComponents))
  (bind _returnvalue_2 (CustomItemWrap|GetWrappedMeshComponents))
  (bind _returnvalue_3 (CustomItemWrap|GetWrappedMeshComponents))
  (bind _returnvalue_4 (Utilities|Time|GetGameTimeinSeconds))
  (bind _returnvalue_5 (CustomItemWrap|GetWrappedMeshComponents))
  (bind _stepcount (Variables|EnviromentVarable|GetStepCount))
  (bind _returnvalue_6 (Math|Conversions|ToFloat(Integer) _stepcount))
  (bind _stepcount_2 (Variables|EnviromentVarable|GetStepCount))
  (bind _returnvalue_7 (Math|Conversions|ToFloat(Integer) _stepcount_2))
  (bind _maxsteps (Variables|EnviromentVarable|GetMaxSteps))
  (bind _returnvalue_8 (/ _returnvalue_6 _maxsteps))
  (bind _maxsteps_2 (Variables|EnviromentVarable|GetMaxSteps))
  (bind _returnvalue_9 (/ _returnvalue_7 _maxsteps_2))
  (bind _output_get (Variables|EnviromentVarable|SetStepCount (Math|Integer|Min(Integer) (+ (Math|Float|Truncate (Math|Float|Lerp (Math|Conversions|ToFloat(Integer) (Variables|EnviromentVarable|GetStepCount)) 0.0 (Math|Float|Clamp(Float) (/ (- (Utilities|Time|GetGameTimeinSeconds) (Variables|EnviromentVarable|Getwait_timer)) (Variables|Default|GetCool%20Down%20timer\Duration))))) 1) (Variables|EnviromentVarable|GetMaxSteps))))
  (Rendering|Material|SetScalarParameterValueonMaterials _returnvalue (Variables|Default|GetOpacity=Pameter-Name-GameTime-TBP) (Utilities|Time|GetGameTimeinSeconds))
  (Rendering|Material|SetScalarParameterValueonMaterials _returnvalue (Variables|Default|GetOpacity=Pameter-Name) (Math|Float|Lerp (Variables|Default|GetOpacity<---Start%20Value) (Variables|Default|GetOpacity<---End%20Value) (/ (Math|Conversions|ToFloat(Integer) _output_get) (Variables|EnviromentVarable|GetMaxSteps))))
  (Rendering|Material|SetScalarParameterValueonMaterials _returnvalue_2 "TopS Emissive_01-Brightness-TBP" _returnvalue_4)
  (Rendering|Material|SetScalarParameterValueonMaterials _returnvalue_2 "TopS Emissive_02-Brightness-TBP" _returnvalue_4)
  (Rendering|Material|SetScalarParameterValueonMaterials _returnvalue_2 "TopS Emissive_02-Brightness-TBP" _returnvalue_4)
  (Rendering|Material|SetScalarParameterValueonMaterials _returnvalue_5 "TopS Emissive_01_Brightness---Start" (Math|Float|Lerp (Variables|Default|GetEmissive01%20Brightness<---Start) (Variables|Default|GetEmissive01%20Brightness<---End) _returnvalue_9))
  (Rendering|Material|SetScalarParameterValueonMaterials _returnvalue_5 "TopS Emissive_02_Brightness---Start" (Math|Float|Lerp (Variables|Default|GetEmissive02%20Brightness<---Start) (Variables|Default|GetEmissive02%20Brightness<---End) _returnvalue_9))
  (Rendering|Material|SetScalarParameterValueonMaterials _returnvalue_5 "TopS Emissive_03_Brightness---Start" (Math|Float|Lerp (Variables|Default|GetEmissive03%20Brightness<---Start) (Variables|Default|GetEmissive03%20Brightness<---End) _returnvalue_9))
  (Rendering|Material|SetScalarParameterValueonMaterials _returnvalue_3 "TopS Emissive_01_Brightness_Darken_Base Start" (Math|Float|Lerp (Variables|Default|GetEmissive01%20DarkenBS<---Start) (Variables|Default|GetEmissive01%20DarkenBS<---End) _returnvalue_8))
  (Rendering|Material|SetScalarParameterValueonMaterials _returnvalue_3 "TopS Emissive_02_Brightness_Darken_Base Start" (Math|Float|Lerp (Variables|Default|GetEmissive02%20DarkenBS<---Start) (Variables|Default|GetEmissive02%20DarkenBS<---End) _returnvalue_8))
  (Rendering|Material|SetScalarParameterValueonMaterials _returnvalue_3 "TopS Emissive_03_Brightness_Darken_Base Start" (Math|Float|Lerp (Variables|Default|GetEmissive03%20DarkenBS<---Start) (Variables|Default|GetEmissive03%20DarkenBS<---End) _returnvalue_8))
  (Variables|EnviromentVarable|Setwait_timer (+ (Utilities|Time|GetGameTimeinSeconds) (Variables|EnviromentVarable|GetPauseTime))))

(fn ConstructionScript)
