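; Weapon wrap cosmetic modifier that drives a material buildup on wrapped meshes as the weapon fires and deals damage.
; Each shot steps an offset between start and end values up to a maximum, with stored steps decaying by elapsed time after a pause.
; Clip size selects step limit and pause; front end has its own settings and fakes damage on some shots.
; A watched DamageDealt stat timestamps a separate material parameter.
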(event EventOnWeaponFired
  (if (CustomCosmeticModifier|IsinFrontEnd)
    (bind _output_get (Variables|EnviromentVarable|SetIterationCount (+ (Variables|EnviromentVarable|GetIterationCount) 1)))
    (if (>= _output_get 5)
      (if (== (Variables|EnviromentVarable|GetIterationCount) 7)
        (Variables|EnviromentVarable|SetIterationCount 0)
        (else
          (CallFunction|WeaponFireBase)
          (CallFunction|DamageDelt)))
      (else
        _)))
  (CallFunction|WeaponFireBase))

(event EventBeginPlay
  (Utilities|FlowControl|Delay 0.2)
  (if (CustomCosmeticModifier|IsinFrontEnd)
    (Variables|EnviromentVarable|SetMaxSteps (Variables|Default|GetShot%20Count%20FrontEnd))
    (Variables|EnviromentVarable|SetPauseTime (Variables|Default|GetPause%20Timer%20FrontEnd))
    (else
      (Variables|EnviromentVarable|SetMaxSteps (Variables|Default|GetShot%20Count%20PIE)))))

(event EventOnWeaponChanged
  (params (NewWeapon :type Fort%20Weapon%20Object%20Reference) (OldWeapon :type Fort%20Weapon%20Object%20Reference))
  (Utilities|FlowControl|Delay 0.2)
  (if (> (FortWeapon|GetBulletsPerClip (Variables|/Script/FortniteGame.FortPawn|GetCurrentWeapon (CustomCharacter|GetAssociatedPlayerPawn))) 5)
    (Variables|EnviromentVarable|SetMaxSteps (Variables|Default|GetShot%20Count%20PIE))
    (Variables|EnviromentVarable|SetPauseTime (Variables|Default|GetPause%20Timer%20Ammo%20>5))
    (else
      (Variables|EnviromentVarable|SetMaxSteps (Variables|Default|GetShot%20Count%20Ammo%20<%205))
      (Variables|EnviromentVarable|SetPauseTime (Variables|Default|GetPause%20Timer%20Ammo%20<5)))))

(event EventOnNewWeapons
  (params (WeaponsAsItems :type Array%20of%20Fort%20Item%20Object%20References))
  (Utilities|FlowControl|Delay 0.2)
  (if (> (FortWeapon|GetBulletsPerClip (Variables|/Script/FortniteGame.FortPawn|GetCurrentWeapon (CustomCharacter|GetAssociatedPlayerPawn))) 5)
    (Variables|EnviromentVarable|SetMaxSteps (Variables|Default|GetShot%20Count%20PIE))
    (Variables|EnviromentVarable|SetPauseTime (Variables|Default|GetPause%20Timer%20Ammo%20>5))
    (else
      (Variables|EnviromentVarable|SetMaxSteps (Variables|Default|GetShot%20Count%20Ammo%20<%205))
      (Variables|EnviromentVarable|SetPauseTime (Variables|Default|GetPause%20Timer%20Ammo%20<5)))))

(event EventOnNewPlayerWeaponType
  (params (NewType :type EFortItemType%20Enum :default "WorldItem") (OldType :type EFortItemType%20Enum :default "WorldItem"))
  (Utilities|FlowControl|Delay 0.2)
  (if (> (FortWeapon|GetBulletsPerClip (Variables|/Script/FortniteGame.FortPawn|GetCurrentWeapon (CustomCharacter|GetAssociatedPlayerPawn))) 5)
    (Variables|EnviromentVarable|SetMaxSteps (Variables|Default|GetShot%20Count%20PIE))
    (Variables|EnviromentVarable|SetPauseTime (Variables|Default|GetPause%20Timer%20Ammo%20>5))
    (else
      (Variables|EnviromentVarable|SetMaxSteps (Variables|Default|GetShot%20Count%20Ammo%20<%205))
      (Variables|EnviromentVarable|SetPauseTime (Variables|Default|GetPause%20Timer%20Ammo%20<5)))))

(event Custom|WeaponFire_Base
  (bind _returnvalue (CustomItemWrap|GetWrappedMeshComponents))
  (bind _output_get (Variables|EnviromentVarable|SetStepCount (Math|Integer|Min(Integer) (+ (Math|Float|Truncate (Math|Float|Lerp (Math|Conversions|ToFloat(Integer) (Variables|EnviromentVarable|GetStepCount)) 0.0 (Math|Float|Clamp(Float) (/ (- (Utilities|Time|GetGameTimeinSeconds) (Variables|EnviromentVarable|Getwait_timer)) (Variables|Default|GetCool%20Down%20timer\Duration))))) 1) (Variables|EnviromentVarable|GetMaxSteps))))
  (Rendering|Material|SetScalarParameterValueonMaterials _returnvalue (Variables|Default|GetOffset=Pameter-Name-GameTime-TBP) (Utilities|Time|GetGameTimeinSeconds))
  (Rendering|Material|SetScalarParameterValueonMaterials _returnvalue (Variables|Default|GetOffset=Pameter-Name) (Math|Float|Lerp (Variables|Default|GetOffset<---Start%20Pos) (Variables|Default|GetOffset--->End%20Pos) (/ (Math|Conversions|ToFloat(Integer) _output_get) (Variables|EnviromentVarable|GetMaxSteps))))
  (Variables|EnviromentVarable|Setwait_timer (+ (Utilities|Time|GetGameTimeinSeconds) (Variables|EnviromentVarable|GetPauseTime))))

(event Custom|DamageDelt
  (Rendering|Material|SetScalarParameterValueonMaterials (CustomItemWrap|GetWrappedMeshComponents) (Variables|Default|GetDamageDelt-Name-GameTime-TBP) (Utilities|Time|GetGameTimeinSeconds)))

(event EventOnPawnSet
  (Cosmetics|WatchCosmeticStat (CustomCharacter|GetAssociatedPlayerPawn) "(TagName=\"StatManager.DamageDealt\")" 0 false
    (:OnStatChanged
      (Rendering|Material|SetScalarParameterValueonMaterials (CustomItemWrap|GetWrappedMeshComponents) (Variables|Default|GetDamageDelt-Name-GameTime-TBP) (Utilities|Time|GetGameTimeinSeconds)))))

(fn ConstructionScript)
