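; Custom item wrap Blueprint that makes a weapon wrap's materials react when the weapon fires.
; On EventOnWeaponFired, a Sequence writes the current game time into scalar parameters on the wrapped mesh components' materials.
; The first branch stamps TopSide V2 emissive, blend and speed parameters; the second stamps reactive decal A–D timing, brightness and gradient parameters.
; It only passes timestamps, with no tick or timeline, leaving any animation to the materials.
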(event EventOnWeaponFired
  (bind _returnvalue (Utilities|Time|GetGameTimeinSeconds))
  (bind _returnvalue_2 (CustomItemWrap|GetWrappedMeshComponents))
  (bind _returnvalue_3 (Utilities|Time|GetGameTimeinSeconds))
  (bind _returnvalue_4 (CustomItemWrap|GetWrappedMeshComponents))
  (bind _returnvalue_5 (CustomItemWrap|GetWrappedMeshComponents))
  (bind _returnvalue_6 (Utilities|Time|GetGameTimeinSeconds))
  (bind _returnvalue_7 (CustomItemWrap|GetWrappedMeshComponents))
  (bind _returnvalue_8 (Utilities|Time|GetGameTimeinSeconds))
  (Utilities|FlowControl|Sequence
    (:then_0
      ;[1 1 1] TopsSide V2
      (Rendering|Material|SetScalarParameterValueonMaterials _returnvalue_2 "TopS Emissive_02-Brightness-TBP" _returnvalue)
      (Rendering|Material|SetScalarParameterValueonMaterials _returnvalue_2 "TopS Emissive_02_Offset----TBP" _returnvalue)
      (Rendering|Material|SetScalarParameterValueonMaterials _returnvalue_2 "TopS_01_Blend_Opacity-TBP" _returnvalue)
      (Rendering|Material|SetScalarParameterValueonMaterials _returnvalue_2 "TopS_01_Blend_Offset-TBP" _returnvalue)
      (Rendering|Material|SetScalarParameterValueonMaterials _returnvalue_2 "TopS Emissive_01-Brightness-TBP" _returnvalue)
      (Rendering|Material|SetScalarParameterValueonMaterials _returnvalue_2 "TopS Emissive_01_Offset----TBP" _returnvalue)
      (Rendering|Material|SetScalarParameterValueonMaterials _returnvalue_2 "TopS Emissive_03-Brightness-TBP" _returnvalue)
      (Rendering|Material|SetScalarParameterValueonMaterials _returnvalue_2 "TopS Emissive_03_Offset----TBP" _returnvalue)
      (Rendering|Material|SetScalarParameterValueonMaterials _returnvalue_2 "TopS_02_Speed-TBP" _returnvalue)
      (Rendering|Material|SetScalarParameterValueonMaterials _returnvalue_2 "TopS_03_Speed-TBP" _returnvalue)
      (Rendering|Material|SetScalarParameterValueonMaterials _returnvalue_2 "TopS_04_Speed-TBP" _returnvalue)
      (Rendering|Material|SetScalarParameterValueonMaterials _returnvalue_2 "TopS Emissive_01_Speed-TBP" _returnvalue)
      (Rendering|Material|SetScalarParameterValueonMaterials _returnvalue_2 "TopS Emissive_02_Speed-TBP" _returnvalue))
    (:then_1
      ;[1 1 1 0.5] Decal Reactive
      (Rendering|Material|SetScalarParameterValueonMaterials _returnvalue_5 "EDRec_TimeFrom_BP_DecalA" _returnvalue_6)
      (Rendering|Material|SetScalarParameterValueonMaterials _returnvalue_5 "EDRec_TimeFrom_BP_DecalB" _returnvalue_6)
      (Rendering|Material|SetScalarParameterValueonMaterials _returnvalue_5 "EDRec_TimeFrom_BP_DecalC" _returnvalue_6)
      (Rendering|Material|SetScalarParameterValueonMaterials _returnvalue_5 "EDRec_TimeFrom_BP_DecalD" _returnvalue_6)
      (Rendering|Material|SetScalarParameterValueonMaterials _returnvalue_7 "EDRec_Emissive_A_Brightness_Time_BP" _returnvalue_8)
      (Rendering|Material|SetScalarParameterValueonMaterials _returnvalue_7 "EDRec_Emissive_B_Brightness_Time_BP" _returnvalue_8)
      (Rendering|Material|SetScalarParameterValueonMaterials _returnvalue_7 "EDRec_Emissive_C_Brightness_Time_BP" _returnvalue_8)
      (Rendering|Material|SetScalarParameterValueonMaterials _returnvalue_7 "EDRec_Emissive_D_Brightness_Time_BP" _returnvalue_8)
      (Rendering|Material|SetScalarParameterValueonMaterials _returnvalue_4 "EDRec_Emissive_A_Gradent_Time_BP" _returnvalue_3)
      (Rendering|Material|SetScalarParameterValueonMaterials _returnvalue_4 "EDRec_Emissive_B_Gradent_Time_BP" _returnvalue_3)
      (Rendering|Material|SetScalarParameterValueonMaterials _returnvalue_4 "EDRec_Emissive_C_Gradent_Time_BP" _returnvalue_3)
      (Rendering|Material|SetScalarParameterValueonMaterials _returnvalue_4 "EDRec_Emissive_D_Gradent_Time_BP" _returnvalue_3))))

(fn ConstructionScript)
