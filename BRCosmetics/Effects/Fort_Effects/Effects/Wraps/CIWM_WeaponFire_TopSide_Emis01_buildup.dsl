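; Weapon wrap cosmetic modifier that builds up a material effect on the wrapped meshes as the weapon fires.
; Each shot adds a step up to a maximum, then writes the game time and an interpolated offset to material parameters.
; Decay needs no tick: each shot first scales the count down by time elapsed since the last pause ended.
; Delayed begin-play and weapon-change events set step limit and pause by front-end state or clip size.
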(event EventBeginPlay
  (Utilities|FlowControl|Delay 0.2)
  (if (CustomCosmeticModifier|IsinFrontEnd)
    (Variables|EviromentVarables|SetMaxSteps (Variables|Default|GetShot%20Count%20FrontEnd))
    (Variables|EviromentVarables|SetPauseTime 2.0)
    (else
      (Variables|EviromentVarables|SetMaxSteps (Variables|Default|GetShot%20Count%20Pie)))))

(event EventOnPawnSet)

(event EventOnWeaponChanged
  (params (NewWeapon :type Fort%20Weapon%20Object%20Reference) (OldWeapon :type Fort%20Weapon%20Object%20Reference))
  (Utilities|FlowControl|Delay 0.2)
  (if (> (FortWeapon|GetBulletsPerClip (Variables|/Script/FortniteGame.FortPawn|GetCurrentWeapon (CustomCharacter|GetAssociatedPlayerPawn))) 5)
    (Variables|EviromentVarables|SetMaxSteps (Variables|Default|GetShot%20Count%20Pie))
    (Variables|EviromentVarables|SetPauseTime (Variables|Default|GetPauseTime_Ammo%20>5))
    (else
      (Variables|EviromentVarables|SetMaxSteps (Variables|Default|GetShot%20Count%20Ammo%20<5))
      (Variables|EviromentVarables|SetPauseTime (Variables|Default|GetPauseTime_Ammo%20<5)))))

(event EventOnNewWeapons
  (params (WeaponsAsItems :type Array%20of%20Fort%20Item%20Object%20References))
  (Utilities|FlowControl|Delay 0.2)
  (if (> (FortWeapon|GetBulletsPerClip (Variables|/Script/FortniteGame.FortPawn|GetCurrentWeapon (CustomCharacter|GetAssociatedPlayerPawn))) 5)
    (Variables|EviromentVarables|SetMaxSteps (Variables|Default|GetShot%20Count%20Pie))
    (Variables|EviromentVarables|SetPauseTime (Variables|Default|GetPauseTime_Ammo%20>5))
    (else
      (Variables|EviromentVarables|SetMaxSteps (Variables|Default|GetShot%20Count%20Ammo%20<5))
      (Variables|EviromentVarables|SetPauseTime (Variables|Default|GetPauseTime_Ammo%20<5)))))

(event EventOnNewPlayerWeaponType
  (params (NewType :type EFortItemType%20Enum :default "WorldItem") (OldType :type EFortItemType%20Enum :default "WorldItem"))
  (Utilities|FlowControl|Delay 0.2)
  (if (> (FortWeapon|GetBulletsPerClip (Variables|/Script/FortniteGame.FortPawn|GetCurrentWeapon (CustomCharacter|GetAssociatedPlayerPawn))) 5)
    (Variables|EviromentVarables|SetMaxSteps (Variables|Default|GetShot%20Count%20Pie))
    (Variables|EviromentVarables|SetPauseTime (Variables|Default|GetPauseTime_Ammo%20>5))
    (else
      (Variables|EviromentVarables|SetMaxSteps (Variables|Default|GetShot%20Count%20Ammo%20<5))
      (Variables|EviromentVarables|SetPauseTime (Variables|Default|GetPauseTime_Ammo%20<5)))))

(event EventOnWeaponFired
  (bind _returnvalue (CustomItemWrap|GetWrappedMeshComponents))
  (bind _output_get (Variables|EviromentVarables|SetStepCount (Math|Integer|Min(Integer) (+ (Math|Float|Truncate (Math|Float|Lerp (Math|Conversions|ToFloat(Integer) (Variables|EviromentVarables|GetStepCount)) 0.0 (Math|Float|Clamp(Float) (/ (- (Utilities|Time|GetGameTimeinSeconds) (Variables|EviromentVarables|Getwait_timer)) (Variables|Default|GetCool%20Down%20timer\Duration))))) 1) (Variables|EviromentVarables|GetMaxSteps))))
  (Rendering|Material|SetScalarParameterValueonMaterials _returnvalue (Variables|Default|GetOffset=Pameter-Name-GameTime%20-TBP) (Utilities|Time|GetGameTimeinSeconds))
  (Rendering|Material|SetScalarParameterValueonMaterials _returnvalue (Variables|Default|GetOffset=Pameter-Name) (Math|Float|Lerp (Variables|Default|GetOffset<---Start%20Pos) (Variables|Default|GetOffset--->End%20Pos) (/ (Math|Conversions|ToFloat(Integer) _output_get) (Variables|EviromentVarables|GetMaxSteps))))
  (Variables|EviromentVarables|Setwait_timer (+ (Utilities|Time|GetGameTimeinSeconds) (Variables|EviromentVarables|GetPauseTime))))

(fn ConstructionScript)
