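; Weapon wrap modifier for GnatGala that animates wrapped mesh materials when the player takes damage, or on a looping timer in the frontend.
; A Progress timeline lerps a blend offset scalar down, and a delayed Revert timer plays it back in reverse.
; Each trigger also shuffles a list of parameter names and stamps each with game time in a staggered loop.
; Damage stat watchers are bound on both BeginPlay and pawn set.
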(event Custom|Count
  (Utilities|Time|SetTimerbyFunctionName 0 "Revert" 1.5))

(event Custom|FrontendKO
  (Utilities|FlowControl|Sequence
    (:then_0
      (if (<= (Variables|Default|GetCurrentOffset) -0.8)
        (Utilities|Time|SetTimerbyFunctionName 0 "Revert" 1.5)
        (else
          (bind (_direction _newtrack_0) (Timeline|Progress PlayFromStart)
            (:Update
              (bind _output_get (Variables|Default|SetCurrentOffset (Math|Float|Lerp 0.5 -0.8 _newtrack_0)))
              (Rendering|Material|SetScalarParameterValueonMaterials (CustomItemWrap|GetWrappedMeshComponents) "TopS_01_Blend_Offset---End" _output_get))
            (:Finished
              (if (not (>= (Variables|Default|GetCurrentOffset) 0.5))
                (CallFunction|Count)))))))
    (:then_1
      (Utilities|Array|Shuffle (Variables|Default|GetNames))
      (bind _arrayitem (Utilities|ForEachLoopwithDealy (Variables|Default|GetNames) 0.2)
        (:LoopBody
          (Rendering|Material|SetScalarParameterValueonMaterials (CustomItemWrap|GetWrappedMeshComponents) _arrayitem (Utilities|Time|GetGameTimeinSeconds)))))))

(event EventBeginPlay
  (if (CustomCosmeticModifier|IsinFrontEnd)
    (Utilities|Time|SetTimerbyEvent (delegate Custom|FrontendKO) 7.0 true)
    (else
      (Cosmetics|WatchCosmeticStat (CustomCharacter|GetAssociatedPlayerPawn) "(TagName=\"StatManager.DamageReceived\")" 0 false
        (:OnStatChanged
          (Utilities|FlowControl|Sequence
            (:then_0
              (if (<= (Variables|Default|GetCurrentOffset) -0.8)
                (Utilities|Time|SetTimerbyFunctionName 0 "Revert" 1.5)
                (else
                  (Timeline|Progress PlayFromStart))))
            (:then_1
              (Utilities|Array|Shuffle (Variables|Default|GetNames))
              (bind _arrayitem (Utilities|ForEachLoopwithDealy (Variables|Default|GetNames) 0.2)
                (:LoopBody
                  (Rendering|Material|SetScalarParameterValueonMaterials (CustomItemWrap|GetWrappedMeshComponents) _arrayitem (Utilities|Time|GetGameTimeinSeconds)))))))))))

(event Custom|Revert
  (Timeline|Progress ReverseFromEnd))

(event EventOnPawnSet
  (Cosmetics|WatchCosmeticStat (CustomCharacter|GetAssociatedPlayerPawn) "(TagName=\"StatManager.DamageReceived\")" 0 false
    (:OnStatChanged
      (Utilities|FlowControl|Sequence
        (:then_0
          (if (<= (Variables|Default|GetCurrentOffset) -0.8)
            (Utilities|Time|SetTimerbyFunctionName 0 "Revert" 1.5)
            (else
              (Timeline|Progress PlayFromStart))))
        (:then_1
          (Utilities|Array|Shuffle (Variables|Default|GetNames))
          (bind _arrayitem (Utilities|ForEachLoopwithDealy (Variables|Default|GetNames) 0.2)
            (:LoopBody
              (Rendering|Material|SetScalarParameterValueonMaterials (CustomItemWrap|GetWrappedMeshComponents) _arrayitem (Utilities|Time|GetGameTimeinSeconds)))))))))

(event EventOnWeaponFired)

(fn ConstructionScript)
