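; WingBath reactive item wrap Blueprint that restyles wrapped meshes to match the wearer's customized character, if that character's name matches.
; Pawn set, wrap target and customization-finished events run identical logic, caching Head and Hat materials.
; Metal base, accent and decal colors and surface biases are copied, scaled or offset, onto wrap material instances.
; Loadout variant channels select the decal mask texture and sigil flipbook frame, tinted from Hat sigil color.
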(event EventOnPawnSet
  (bind _character_material (Variables|Default|GetCharacter%20Material))
  (bind _character_material_2 (Variables|Default|GetCharacter%20Material))
  (if (Utilities|String|Equal,CaseInsensitive(String) (Item|GetPersistentName (CosmeticLoadout|GetItemDefinitionfromLoadout (CosmeticLoadout|GetCosmeticLoadout) "(TagName=\"LoadoutSlot.Character\")")) (Variables|Default|GetCharacter-Name))
    (bind _asmaterial_instance_dynamic (Utilities|Casting|CastToMaterialInstanceDynamic (Rendering|Material|GetMaterial (Customization|GetSkeletalMeshforPartType (CustomCharacter|GetAssociatedPlayerPawn) "Head" :!class /Script/FortniteGame.FortPlayerPawn) 0)))
    (Variables|Default|SetCharacter%20Material _asmaterial_instance_dynamic)
    (bind _asmaterial_instance_dynamic_2 (Utilities|Casting|CastToMaterialInstanceDynamic (Rendering|Material|GetMaterial (Customization|GetSkeletalMeshforPartType (CustomCharacter|GetAssociatedPlayerPawn) "Hat" :!class /Script/FortniteGame.FortPlayerPawn) 0)))
    (Variables|Default|SetSigelMat _asmaterial_instance_dynamic_2)
    (Utilities|FlowControl|Sequence
      (:then_0
        (bind _returnvalue (Rendering|Material|GetVectorParameterValue _asmaterial_instance_dynamic "MetalBase_Color" :!class /Script/Engine.MaterialInstanceDynamic))
        (bind _returnvalue_2 (Rendering|Material|GetScalarParameterValue _asmaterial_instance_dynamic "MetalBase_MetallicBias" :!class /Script/Engine.MaterialInstanceDynamic))
        (bind _returnvalue_3 (Rendering|Material|GetScalarParameterValue _asmaterial_instance_dynamic "MetalBase_SpecularBias" :!class /Script/Engine.MaterialInstanceDynamic))
        (bind _returnvalue_4 (Rendering|Material|GetScalarParameterValue _asmaterial_instance_dynamic "MetalBase_RoughnessBias" :!class /Script/Engine.MaterialInstanceDynamic))
        (for _array_element (CustomItemWrap|GetWrappedMeshComponents)
          (for _array_element_2 (Rendering|Material|GetMaterials _array_element)
            (bind _asmaterial_instance_dynamic_3 (Utilities|Casting|CastToMaterialInstanceDynamic _array_element_2))
            (Rendering|Material|SetVectorParameterValue _asmaterial_instance_dynamic_3 "TopS_01_ColorTint_CH_Green A=0=off" (* "(R=0.750000,G=0.750000,B=0.750000,A=1.000000)" _returnvalue) :!class /Script/Engine.MaterialInstanceDynamic)
            (Rendering|Material|SetScalarParameterValue _asmaterial_instance_dynamic_3 "TopS_01_Metalic_CH_Green" _returnvalue_2 :!class /Script/Engine.MaterialInstanceDynamic)
            (Rendering|Material|SetScalarParameterValue _asmaterial_instance_dynamic_3 "TopS_01_Roughtness_CH_Green" (+ _returnvalue_4 1.0) :!class /Script/Engine.MaterialInstanceDynamic)
            (Rendering|Material|SetScalarParameterValue _asmaterial_instance_dynamic_3 "TopS_01_Spec_CH_Green" (+ _returnvalue_3 0.5) :!class /Script/Engine.MaterialInstanceDynamic))))
      (:then_1
        (bind _returnvalue_5 (Rendering|Material|GetVectorParameterValue _character_material_2 "MetalAccent_Color" :!class /Script/Engine.MaterialInstanceDynamic))
        (bind _returnvalue_6 (Rendering|Material|GetScalarParameterValue _character_material_2 "MetalAccent_MetallicBias" :!class /Script/Engine.MaterialInstanceDynamic))
        (bind _returnvalue_7 (Rendering|Material|GetScalarParameterValue _character_material_2 "MetalAccent_SpecularBias" :!class /Script/Engine.MaterialInstanceDynamic))
        (bind _returnvalue_8 (Rendering|Material|GetScalarParameterValue _character_material_2 "MetalAccent_RoughnessBias" :!class /Script/Engine.MaterialInstanceDynamic))
        (for _array_element_3 (CustomItemWrap|GetWrappedMeshComponents)
          (for _array_element_4 (Rendering|Material|GetMaterials _array_element_3)
            (bind _asmaterial_instance_dynamic_4 (Utilities|Casting|CastToMaterialInstanceDynamic _array_element_4))
            (Rendering|Material|SetVectorParameterValue _asmaterial_instance_dynamic_4 "TopS_01_ColorTint_CH_Blue  A=0=off" (* "(R=0.750000,G=0.750000,B=0.750000,A=1.000000)" _returnvalue_5) :!class /Script/Engine.MaterialInstanceDynamic)
            (Rendering|Material|SetScalarParameterValue _asmaterial_instance_dynamic_4 "TopS_01_Metalic_CH_Blue" _returnvalue_6 :!class /Script/Engine.MaterialInstanceDynamic)
            (Rendering|Material|SetScalarParameterValue _asmaterial_instance_dynamic_4 "TopS_01_Roughtness_CH_Blue" (+ _returnvalue_8 1.0) :!class /Script/Engine.MaterialInstanceDynamic)
            (Rendering|Material|SetScalarParameterValue _asmaterial_instance_dynamic_4 "TopS_01_Spec_CH_Blue" (+ _returnvalue_7 0.5) :!class /Script/Engine.MaterialInstanceDynamic))))
      (:then_2
        (bind _index (CosmeticLoadout|GetVariantChannelActiveIndexforSlotActor "(TagName=\"LoadoutSlot.Character\")" (CustomCharacter|GetAssociatedPlayerPawn) "(TagName=\"Cosmetics.Variant.Channel.Slot.3\")" (Utilities|Array|MakeArray "(TagName=\"Cosmetics.Variant.Property.Mat27\")" "(TagName=\"Cosmetics.Variant.Property.Mat16\")" "(TagName=\"Cosmetics.Variant.Property.Mat17\")" "(TagName=\"Cosmetics.Variant.Property.Mat18\")" "(TagName=\"Cosmetics.Variant.Property.Mat19\")" "(TagName=\"Cosmetics.Variant.Property.Mat20\")" "(TagName=\"Cosmetics.Variant.Property.Mat21\")" "(TagName=\"Cosmetics.Variant.Property.Mat22\")" "(TagName=\"Cosmetics.Variant.Property.Mat23\")") false false))
        (Variables|Default|SetDecal_index _index)
        (Utilities|FlowControl|Sequence
          (:then_0
            (switch int (Utilities|Array|Get(acopy) (Utilities|Array|MakeArray 0 1 1 1 1 1 1 1 1) (Variables|Default|GetDecal_index))
              (:0
                (for _array_element_5 (CustomItemWrap|GetWrappedMeshComponents)
                  (for _array_element_6 (Rendering|Material|GetMaterials _array_element_5)
                    (bind _asmaterial_instance_dynamic_5 (Utilities|Casting|CastToMaterialInstanceDynamic _array_element_6))
                    (bind _returnvalue_9 (Rendering|Material|GetVectorParameterValue _asmaterial_instance_dynamic_5 "AccentV2_BaseColor_" :!class /Script/Engine.MaterialInstanceDynamic))
                    (bind _returnvalue_10 (Rendering|Material|GetScalarParameterValue _asmaterial_instance_dynamic_5 "AccentV2_Metalness" :!class /Script/Engine.MaterialInstanceDynamic))
                    (bind _returnvalue_11 (Rendering|Material|GetScalarParameterValue _asmaterial_instance_dynamic_5 "AccentV2_Roughness" :!class /Script/Engine.MaterialInstanceDynamic))
                    (bind _returnvalue_12 (Rendering|Material|GetScalarParameterValue _asmaterial_instance_dynamic_5 "AccentV2_Specular" :!class /Script/Engine.MaterialInstanceDynamic))
                    (Rendering|Material|SetVectorParameterValue _asmaterial_instance_dynamic_5 "AccentV2_BaseColor_" _returnvalue_9 :!class /Script/Engine.MaterialInstanceDynamic)
                    (Rendering|Material|SetScalarParameterValue _asmaterial_instance_dynamic_5 "AccentV2_Metalness" _returnvalue_10 :!class /Script/Engine.MaterialInstanceDynamic)
                    (Rendering|Material|SetScalarParameterValue _asmaterial_instance_dynamic_5 "AccentV2_Roughness" _returnvalue_11 :!class /Script/Engine.MaterialInstanceDynamic)
                    (Rendering|Material|SetScalarParameterValue _asmaterial_instance_dynamic_5 "AccentV2_Specular" _returnvalue_12 :!class /Script/Engine.MaterialInstanceDynamic))))
              (:1
                (bind _returnvalue_13 (Rendering|Material|GetVectorParameterValue _character_material "Decal_Color" :!class /Script/Engine.MaterialInstanceDynamic))
                (bind _returnvalue_14 (Rendering|Material|GetScalarParameterValue _character_material "Decal_MetallicBias" :!class /Script/Engine.MaterialInstanceDynamic))
                (bind _returnvalue_15 (Rendering|Material|GetScalarParameterValue _character_material "Decal_SpecularBias" :!class /Script/Engine.MaterialInstanceDynamic))
                (bind _returnvalue_16 (Rendering|Material|GetScalarParameterValue _character_material "Decal_RoughnessBias" :!class /Script/Engine.MaterialInstanceDynamic))
                (for _array_element_7 (CustomItemWrap|GetWrappedMeshComponents)
                  (for _array_element_8 (Rendering|Material|GetMaterials _array_element_7)
                    (bind _asmaterial_instance_dynamic_6 (Utilities|Casting|CastToMaterialInstanceDynamic _array_element_8))
                    (bind _returnvalue_17 (Rendering|Material|GetScalarParameterValue _asmaterial_instance_dynamic_6 "AccentV2_Metalness" :!class /Script/Engine.MaterialInstanceDynamic))
                    (bind _returnvalue_18 (Rendering|Material|GetScalarParameterValue _asmaterial_instance_dynamic_6 "AccentV2_Roughness" :!class /Script/Engine.MaterialInstanceDynamic))
                    (bind _returnvalue_19 (Rendering|Material|GetScalarParameterValue _asmaterial_instance_dynamic_6 "AccentV2_Specular" :!class /Script/Engine.MaterialInstanceDynamic))
                    (Rendering|Material|SetVectorParameterValue _asmaterial_instance_dynamic_6 "AccentV2_BaseColor_" (* _returnvalue_13 "(R=0.300000,G=0.300000,B=0.300000,A=1.000000)") :!class /Script/Engine.MaterialInstanceDynamic)
                    (Rendering|Material|SetScalarParameterValue _asmaterial_instance_dynamic_6 "AccentV2_Metalness" (+ _returnvalue_17 _returnvalue_14) :!class /Script/Engine.MaterialInstanceDynamic)
                    (Rendering|Material|SetScalarParameterValue _asmaterial_instance_dynamic_6 "AccentV2_Roughness" (+ _returnvalue_18 _returnvalue_16) :!class /Script/Engine.MaterialInstanceDynamic)
                    (Rendering|Material|SetScalarParameterValue _asmaterial_instance_dynamic_6 "AccentV2_Specular" (+ _returnvalue_19 _returnvalue_15) :!class /Script/Engine.MaterialInstanceDynamic))))))
          (:then_1
            (CallFunction|Decal))))
      (:then_3
        (CallFunction|Icon)))))

(event EventOnSetWrapTarget
  (bind _character_material (Variables|Default|GetCharacter%20Material))
  (bind _character_material_2 (Variables|Default|GetCharacter%20Material))
  (if (Utilities|String|Equal,CaseInsensitive(String) (Item|GetPersistentName (CosmeticLoadout|GetItemDefinitionfromLoadout (CosmeticLoadout|GetCosmeticLoadout) "(TagName=\"LoadoutSlot.Character\")")) (Variables|Default|GetCharacter-Name))
    (bind _asmaterial_instance_dynamic (Utilities|Casting|CastToMaterialInstanceDynamic (Rendering|Material|GetMaterial (Customization|GetSkeletalMeshforPartType (CustomCharacter|GetAssociatedPlayerPawn) "Head" :!class /Script/FortniteGame.FortPlayerPawn) 0)))
    (Variables|Default|SetCharacter%20Material _asmaterial_instance_dynamic)
    (bind _asmaterial_instance_dynamic_2 (Utilities|Casting|CastToMaterialInstanceDynamic (Rendering|Material|GetMaterial (Customization|GetSkeletalMeshforPartType (CustomCharacter|GetAssociatedPlayerPawn) "Hat" :!class /Script/FortniteGame.FortPlayerPawn) 0)))
    (Variables|Default|SetSigelMat _asmaterial_instance_dynamic_2)
    (Utilities|FlowControl|Sequence
      (:then_0
        (bind _returnvalue (Rendering|Material|GetVectorParameterValue _asmaterial_instance_dynamic "MetalBase_Color" :!class /Script/Engine.MaterialInstanceDynamic))
        (bind _returnvalue_2 (Rendering|Material|GetScalarParameterValue _asmaterial_instance_dynamic "MetalBase_MetallicBias" :!class /Script/Engine.MaterialInstanceDynamic))
        (bind _returnvalue_3 (Rendering|Material|GetScalarParameterValue _asmaterial_instance_dynamic "MetalBase_SpecularBias" :!class /Script/Engine.MaterialInstanceDynamic))
        (bind _returnvalue_4 (Rendering|Material|GetScalarParameterValue _asmaterial_instance_dynamic "MetalBase_RoughnessBias" :!class /Script/Engine.MaterialInstanceDynamic))
        (for _array_element (CustomItemWrap|GetWrappedMeshComponents)
          (for _array_element_2 (Rendering|Material|GetMaterials _array_element)
            (bind _asmaterial_instance_dynamic_3 (Utilities|Casting|CastToMaterialInstanceDynamic _array_element_2))
            (Rendering|Material|SetVectorParameterValue _asmaterial_instance_dynamic_3 "TopS_01_ColorTint_CH_Green A=0=off" (* "(R=0.750000,G=0.750000,B=0.750000,A=1.000000)" _returnvalue) :!class /Script/Engine.MaterialInstanceDynamic)
            (Rendering|Material|SetScalarParameterValue _asmaterial_instance_dynamic_3 "TopS_01_Metalic_CH_Green" _returnvalue_2 :!class /Script/Engine.MaterialInstanceDynamic)
            (Rendering|Material|SetScalarParameterValue _asmaterial_instance_dynamic_3 "TopS_01_Roughtness_CH_Green" (+ _returnvalue_4 1.0) :!class /Script/Engine.MaterialInstanceDynamic)
            (Rendering|Material|SetScalarParameterValue _asmaterial_instance_dynamic_3 "TopS_01_Spec_CH_Green" (+ _returnvalue_3 0.5) :!class /Script/Engine.MaterialInstanceDynamic))))
      (:then_1
        (bind _returnvalue_5 (Rendering|Material|GetVectorParameterValue _character_material_2 "MetalAccent_Color" :!class /Script/Engine.MaterialInstanceDynamic))
        (bind _returnvalue_6 (Rendering|Material|GetScalarParameterValue _character_material_2 "MetalAccent_MetallicBias" :!class /Script/Engine.MaterialInstanceDynamic))
        (bind _returnvalue_7 (Rendering|Material|GetScalarParameterValue _character_material_2 "MetalAccent_SpecularBias" :!class /Script/Engine.MaterialInstanceDynamic))
        (bind _returnvalue_8 (Rendering|Material|GetScalarParameterValue _character_material_2 "MetalAccent_RoughnessBias" :!class /Script/Engine.MaterialInstanceDynamic))
        (for _array_element_3 (CustomItemWrap|GetWrappedMeshComponents)
          (for _array_element_4 (Rendering|Material|GetMaterials _array_element_3)
            (bind _asmaterial_instance_dynamic_4 (Utilities|Casting|CastToMaterialInstanceDynamic _array_element_4))
            (Rendering|Material|SetVectorParameterValue _asmaterial_instance_dynamic_4 "TopS_01_ColorTint_CH_Blue  A=0=off" (* "(R=0.750000,G=0.750000,B=0.750000,A=1.000000)" _returnvalue_5) :!class /Script/Engine.MaterialInstanceDynamic)
            (Rendering|Material|SetScalarParameterValue _asmaterial_instance_dynamic_4 "TopS_01_Metalic_CH_Blue" _returnvalue_6 :!class /Script/Engine.MaterialInstanceDynamic)
            (Rendering|Material|SetScalarParameterValue _asmaterial_instance_dynamic_4 "TopS_01_Roughtness_CH_Blue" (+ _returnvalue_8 1.0) :!class /Script/Engine.MaterialInstanceDynamic)
            (Rendering|Material|SetScalarParameterValue _asmaterial_instance_dynamic_4 "TopS_01_Spec_CH_Blue" (+ _returnvalue_7 0.5) :!class /Script/Engine.MaterialInstanceDynamic))))
      (:then_2
        (bind _index (CosmeticLoadout|GetVariantChannelActiveIndexforSlotActor "(TagName=\"LoadoutSlot.Character\")" (CustomCharacter|GetAssociatedPlayerPawn) "(TagName=\"Cosmetics.Variant.Channel.Slot.3\")" (Utilities|Array|MakeArray "(TagName=\"Cosmetics.Variant.Property.Mat27\")" "(TagName=\"Cosmetics.Variant.Property.Mat16\")" "(TagName=\"Cosmetics.Variant.Property.Mat17\")" "(TagName=\"Cosmetics.Variant.Property.Mat18\")" "(TagName=\"Cosmetics.Variant.Property.Mat19\")" "(TagName=\"Cosmetics.Variant.Property.Mat20\")" "(TagName=\"Cosmetics.Variant.Property.Mat21\")" "(TagName=\"Cosmetics.Variant.Property.Mat22\")" "(TagName=\"Cosmetics.Variant.Property.Mat23\")") false false))
        (Variables|Default|SetDecal_index _index)
        (Utilities|FlowControl|Sequence
          (:then_0
            (switch int (Utilities|Array|Get(acopy) (Utilities|Array|MakeArray 0 1 1 1 1 1 1 1 1) (Variables|Default|GetDecal_index))
              (:0
                (for _array_element_5 (CustomItemWrap|GetWrappedMeshComponents)
                  (for _array_element_6 (Rendering|Material|GetMaterials _array_element_5)
                    (bind _asmaterial_instance_dynamic_5 (Utilities|Casting|CastToMaterialInstanceDynamic _array_element_6))
                    (bind _returnvalue_9 (Rendering|Material|GetVectorParameterValue _asmaterial_instance_dynamic_5 "AccentV2_BaseColor_" :!class /Script/Engine.MaterialInstanceDynamic))
                    (bind _returnvalue_10 (Rendering|Material|GetScalarParameterValue _asmaterial_instance_dynamic_5 "AccentV2_Metalness" :!class /Script/Engine.MaterialInstanceDynamic))
                    (bind _returnvalue_11 (Rendering|Material|GetScalarParameterValue _asmaterial_instance_dynamic_5 "AccentV2_Roughness" :!class /Script/Engine.MaterialInstanceDynamic))
                    (bind _returnvalue_12 (Rendering|Material|GetScalarParameterValue _asmaterial_instance_dynamic_5 "AccentV2_Specular" :!class /Script/Engine.MaterialInstanceDynamic))
                    (Rendering|Material|SetVectorParameterValue _asmaterial_instance_dynamic_5 "AccentV2_BaseColor_" _returnvalue_9 :!class /Script/Engine.MaterialInstanceDynamic)
                    (Rendering|Material|SetScalarParameterValue _asmaterial_instance_dynamic_5 "AccentV2_Metalness" _returnvalue_10 :!class /Script/Engine.MaterialInstanceDynamic)
                    (Rendering|Material|SetScalarParameterValue _asmaterial_instance_dynamic_5 "AccentV2_Roughness" _returnvalue_11 :!class /Script/Engine.MaterialInstanceDynamic)
                    (Rendering|Material|SetScalarParameterValue _asmaterial_instance_dynamic_5 "AccentV2_Specular" _returnvalue_12 :!class /Script/Engine.MaterialInstanceDynamic))))
              (:1
                (bind _returnvalue_13 (Rendering|Material|GetVectorParameterValue _character_material "Decal_Color" :!class /Script/Engine.MaterialInstanceDynamic))
                (bind _returnvalue_14 (Rendering|Material|GetScalarParameterValue _character_material "Decal_MetallicBias" :!class /Script/Engine.MaterialInstanceDynamic))
                (bind _returnvalue_15 (Rendering|Material|GetScalarParameterValue _character_material "Decal_SpecularBias" :!class /Script/Engine.MaterialInstanceDynamic))
                (bind _returnvalue_16 (Rendering|Material|GetScalarParameterValue _character_material "Decal_RoughnessBias" :!class /Script/Engine.MaterialInstanceDynamic))
                (for _array_element_7 (CustomItemWrap|GetWrappedMeshComponents)
                  (for _array_element_8 (Rendering|Material|GetMaterials _array_element_7)
                    (bind _asmaterial_instance_dynamic_6 (Utilities|Casting|CastToMaterialInstanceDynamic _array_element_8))
                    (bind _returnvalue_17 (Rendering|Material|GetScalarParameterValue _asmaterial_instance_dynamic_6 "AccentV2_Metalness" :!class /Script/Engine.MaterialInstanceDynamic))
                    (bind _returnvalue_18 (Rendering|Material|GetScalarParameterValue _asmaterial_instance_dynamic_6 "AccentV2_Roughness" :!class /Script/Engine.MaterialInstanceDynamic))
                    (bind _returnvalue_19 (Rendering|Material|GetScalarParameterValue _asmaterial_instance_dynamic_6 "AccentV2_Specular" :!class /Script/Engine.MaterialInstanceDynamic))
                    (Rendering|Material|SetVectorParameterValue _asmaterial_instance_dynamic_6 "AccentV2_BaseColor_" (* _returnvalue_13 "(R=0.300000,G=0.300000,B=0.300000,A=1.000000)") :!class /Script/Engine.MaterialInstanceDynamic)
                    (Rendering|Material|SetScalarParameterValue _asmaterial_instance_dynamic_6 "AccentV2_Metalness" (+ _returnvalue_17 _returnvalue_14) :!class /Script/Engine.MaterialInstanceDynamic)
                    (Rendering|Material|SetScalarParameterValue _asmaterial_instance_dynamic_6 "AccentV2_Roughness" (+ _returnvalue_18 _returnvalue_16) :!class /Script/Engine.MaterialInstanceDynamic)
                    (Rendering|Material|SetScalarParameterValue _asmaterial_instance_dynamic_6 "AccentV2_Specular" (+ _returnvalue_19 _returnvalue_15) :!class /Script/Engine.MaterialInstanceDynamic))))))
          (:then_1
            (CallFunction|Decal))))
      (:then_3
        (CallFunction|Icon)))))

(event EventOnCharacterCustomizationFinished
  (params (CharacterPartInfo :type Array%20of%20Applied%20Character%20Part%20Info%20Structures))
  (bind _character_material (Variables|Default|GetCharacter%20Material))
  (bind _character_material_2 (Variables|Default|GetCharacter%20Material))
  (if (Utilities|String|Equal,CaseInsensitive(String) (Item|GetPersistentName (CosmeticLoadout|GetItemDefinitionfromLoadout (CosmeticLoadout|GetCosmeticLoadout) "(TagName=\"LoadoutSlot.Character\")")) (Variables|Default|GetCharacter-Name))
    (bind _asmaterial_instance_dynamic (Utilities|Casting|CastToMaterialInstanceDynamic (Rendering|Material|GetMaterial (Customization|GetSkeletalMeshforPartType (CustomCharacter|GetAssociatedPlayerPawn) "Head" :!class /Script/FortniteGame.FortPlayerPawn) 0)))
    (Variables|Default|SetCharacter%20Material _asmaterial_instance_dynamic)
    (bind _asmaterial_instance_dynamic_2 (Utilities|Casting|CastToMaterialInstanceDynamic (Rendering|Material|GetMaterial (Customization|GetSkeletalMeshforPartType (CustomCharacter|GetAssociatedPlayerPawn) "Hat" :!class /Script/FortniteGame.FortPlayerPawn) 0)))
    (Variables|Default|SetSigelMat _asmaterial_instance_dynamic_2)
    (Utilities|FlowControl|Sequence
      (:then_0
        (bind _returnvalue (Rendering|Material|GetVectorParameterValue _asmaterial_instance_dynamic "MetalBase_Color" :!class /Script/Engine.MaterialInstanceDynamic))
        (bind _returnvalue_2 (Rendering|Material|GetScalarParameterValue _asmaterial_instance_dynamic "MetalBase_MetallicBias" :!class /Script/Engine.MaterialInstanceDynamic))
        (bind _returnvalue_3 (Rendering|Material|GetScalarParameterValue _asmaterial_instance_dynamic "MetalBase_SpecularBias" :!class /Script/Engine.MaterialInstanceDynamic))
        (bind _returnvalue_4 (Rendering|Material|GetScalarParameterValue _asmaterial_instance_dynamic "MetalBase_RoughnessBias" :!class /Script/Engine.MaterialInstanceDynamic))
        (for _array_element (CustomItemWrap|GetWrappedMeshComponents)
          (for _array_element_2 (Rendering|Material|GetMaterials _array_element)
            (bind _asmaterial_instance_dynamic_3 (Utilities|Casting|CastToMaterialInstanceDynamic _array_element_2))
            (Rendering|Material|SetVectorParameterValue _asmaterial_instance_dynamic_3 "TopS_01_ColorTint_CH_Green A=0=off" (* "(R=0.750000,G=0.750000,B=0.750000,A=1.000000)" _returnvalue) :!class /Script/Engine.MaterialInstanceDynamic)
            (Rendering|Material|SetScalarParameterValue _asmaterial_instance_dynamic_3 "TopS_01_Metalic_CH_Green" _returnvalue_2 :!class /Script/Engine.MaterialInstanceDynamic)
            (Rendering|Material|SetScalarParameterValue _asmaterial_instance_dynamic_3 "TopS_01_Roughtness_CH_Green" (+ _returnvalue_4 1.0) :!class /Script/Engine.MaterialInstanceDynamic)
            (Rendering|Material|SetScalarParameterValue _asmaterial_instance_dynamic_3 "TopS_01_Spec_CH_Green" (+ _returnvalue_3 0.5) :!class /Script/Engine.MaterialInstanceDynamic))))
      (:then_1
        (bind _returnvalue_5 (Rendering|Material|GetVectorParameterValue _character_material_2 "MetalAccent_Color" :!class /Script/Engine.MaterialInstanceDynamic))
        (bind _returnvalue_6 (Rendering|Material|GetScalarParameterValue _character_material_2 "MetalAccent_MetallicBias" :!class /Script/Engine.MaterialInstanceDynamic))
        (bind _returnvalue_7 (Rendering|Material|GetScalarParameterValue _character_material_2 "MetalAccent_SpecularBias" :!class /Script/Engine.MaterialInstanceDynamic))
        (bind _returnvalue_8 (Rendering|Material|GetScalarParameterValue _character_material_2 "MetalAccent_RoughnessBias" :!class /Script/Engine.MaterialInstanceDynamic))
        (for _array_element_3 (CustomItemWrap|GetWrappedMeshComponents)
          (for _array_element_4 (Rendering|Material|GetMaterials _array_element_3)
            (bind _asmaterial_instance_dynamic_4 (Utilities|Casting|CastToMaterialInstanceDynamic _array_element_4))
            (Rendering|Material|SetVectorParameterValue _asmaterial_instance_dynamic_4 "TopS_01_ColorTint_CH_Blue  A=0=off" (* "(R=0.750000,G=0.750000,B=0.750000,A=1.000000)" _returnvalue_5) :!class /Script/Engine.MaterialInstanceDynamic)
            (Rendering|Material|SetScalarParameterValue _asmaterial_instance_dynamic_4 "TopS_01_Metalic_CH_Blue" _returnvalue_6 :!class /Script/Engine.MaterialInstanceDynamic)
            (Rendering|Material|SetScalarParameterValue _asmaterial_instance_dynamic_4 "TopS_01_Roughtness_CH_Blue" (+ _returnvalue_8 1.0) :!class /Script/Engine.MaterialInstanceDynamic)
            (Rendering|Material|SetScalarParameterValue _asmaterial_instance_dynamic_4 "TopS_01_Spec_CH_Blue" (+ _returnvalue_7 0.5) :!class /Script/Engine.MaterialInstanceDynamic))))
      (:then_2
        (bind _index (CosmeticLoadout|GetVariantChannelActiveIndexforSlotActor "(TagName=\"LoadoutSlot.Character\")" (CustomCharacter|GetAssociatedPlayerPawn) "(TagName=\"Cosmetics.Variant.Channel.Slot.3\")" (Utilities|Array|MakeArray "(TagName=\"Cosmetics.Variant.Property.Mat27\")" "(TagName=\"Cosmetics.Variant.Property.Mat16\")" "(TagName=\"Cosmetics.Variant.Property.Mat17\")" "(TagName=\"Cosmetics.Variant.Property.Mat18\")" "(TagName=\"Cosmetics.Variant.Property.Mat19\")" "(TagName=\"Cosmetics.Variant.Property.Mat20\")" "(TagName=\"Cosmetics.Variant.Property.Mat21\")" "(TagName=\"Cosmetics.Variant.Property.Mat22\")" "(TagName=\"Cosmetics.Variant.Property.Mat23\")") false false))
        (Variables|Default|SetDecal_index _index)
        (Utilities|FlowControl|Sequence
          (:then_0
            (switch int (Utilities|Array|Get(acopy) (Utilities|Array|MakeArray 0 1 1 1 1 1 1 1 1) (Variables|Default|GetDecal_index))
              (:0
                (for _array_element_5 (CustomItemWrap|GetWrappedMeshComponents)
                  (for _array_element_6 (Rendering|Material|GetMaterials _array_element_5)
                    (bind _asmaterial_instance_dynamic_5 (Utilities|Casting|CastToMaterialInstanceDynamic _array_element_6))
                    (bind _returnvalue_9 (Rendering|Material|GetVectorParameterValue _asmaterial_instance_dynamic_5 "AccentV2_BaseColor_" :!class /Script/Engine.MaterialInstanceDynamic))
                    (bind _returnvalue_10 (Rendering|Material|GetScalarParameterValue _asmaterial_instance_dynamic_5 "AccentV2_Metalness" :!class /Script/Engine.MaterialInstanceDynamic))
                    (bind _returnvalue_11 (Rendering|Material|GetScalarParameterValue _asmaterial_instance_dynamic_5 "AccentV2_Roughness" :!class /Script/Engine.MaterialInstanceDynamic))
                    (bind _returnvalue_12 (Rendering|Material|GetScalarParameterValue _asmaterial_instance_dynamic_5 "AccentV2_Specular" :!class /Script/Engine.MaterialInstanceDynamic))
                    (Rendering|Material|SetVectorParameterValue _asmaterial_instance_dynamic_5 "AccentV2_BaseColor_" _returnvalue_9 :!class /Script/Engine.MaterialInstanceDynamic)
                    (Rendering|Material|SetScalarParameterValue _asmaterial_instance_dynamic_5 "AccentV2_Metalness" _returnvalue_10 :!class /Script/Engine.MaterialInstanceDynamic)
                    (Rendering|Material|SetScalarParameterValue _asmaterial_instance_dynamic_5 "AccentV2_Roughness" _returnvalue_11 :!class /Script/Engine.MaterialInstanceDynamic)
                    (Rendering|Material|SetScalarParameterValue _asmaterial_instance_dynamic_5 "AccentV2_Specular" _returnvalue_12 :!class /Script/Engine.MaterialInstanceDynamic))))
              (:1
                (bind _returnvalue_13 (Rendering|Material|GetVectorParameterValue _character_material "Decal_Color" :!class /Script/Engine.MaterialInstanceDynamic))
                (bind _returnvalue_14 (Rendering|Material|GetScalarParameterValue _character_material "Decal_MetallicBias" :!class /Script/Engine.MaterialInstanceDynamic))
                (bind _returnvalue_15 (Rendering|Material|GetScalarParameterValue _character_material "Decal_SpecularBias" :!class /Script/Engine.MaterialInstanceDynamic))
                (bind _returnvalue_16 (Rendering|Material|GetScalarParameterValue _character_material "Decal_RoughnessBias" :!class /Script/Engine.MaterialInstanceDynamic))
                (for _array_element_7 (CustomItemWrap|GetWrappedMeshComponents)
                  (for _array_element_8 (Rendering|Material|GetMaterials _array_element_7)
                    (bind _asmaterial_instance_dynamic_6 (Utilities|Casting|CastToMaterialInstanceDynamic _array_element_8))
                    (bind _returnvalue_17 (Rendering|Material|GetScalarParameterValue _asmaterial_instance_dynamic_6 "AccentV2_Metalness" :!class /Script/Engine.MaterialInstanceDynamic))
                    (bind _returnvalue_18 (Rendering|Material|GetScalarParameterValue _asmaterial_instance_dynamic_6 "AccentV2_Roughness" :!class /Script/Engine.MaterialInstanceDynamic))
                    (bind _returnvalue_19 (Rendering|Material|GetScalarParameterValue _asmaterial_instance_dynamic_6 "AccentV2_Specular" :!class /Script/Engine.MaterialInstanceDynamic))
                    (Rendering|Material|SetVectorParameterValue _asmaterial_instance_dynamic_6 "AccentV2_BaseColor_" (* _returnvalue_13 "(R=0.300000,G=0.300000,B=0.300000,A=1.000000)") :!class /Script/Engine.MaterialInstanceDynamic)
                    (Rendering|Material|SetScalarParameterValue _asmaterial_instance_dynamic_6 "AccentV2_Metalness" (+ _returnvalue_17 _returnvalue_14) :!class /Script/Engine.MaterialInstanceDynamic)
                    (Rendering|Material|SetScalarParameterValue _asmaterial_instance_dynamic_6 "AccentV2_Roughness" (+ _returnvalue_18 _returnvalue_16) :!class /Script/Engine.MaterialInstanceDynamic)
                    (Rendering|Material|SetScalarParameterValue _asmaterial_instance_dynamic_6 "AccentV2_Specular" (+ _returnvalue_19 _returnvalue_15) :!class /Script/Engine.MaterialInstanceDynamic))))))
          (:then_1
            (CallFunction|Decal))))
      (:then_3
        (CallFunction|Icon)))))

(event Custom|OnBannerFullyApplied
  (params (Icon :type Texture%202D%20Object%20Reference) (PrimaryColor :type Linear%20Color%20Structure)))

(event Custom|Decal_
  (bind _character_material (Variables|Default|GetCharacter%20Material))
  (switch int (Utilities|Array|Get(acopy) (Utilities|Array|MakeArray 0 1 2 3 4 5 6 7 8) (Variables|Default|GetDecal_index))
    (:0
      (for _array_element (CustomItemWrap|GetWrappedMeshComponents)
        (for _array_element_2 (Rendering|Material|GetMaterials _array_element)
          (bind _asmaterial_instance_dynamic (Utilities|Casting|CastToMaterialInstanceDynamic _array_element_2))
          (Rendering|Material|SetTextureParameterValue _asmaterial_instance_dynamic "TopS_02__Tex_Mask" "/Game/Global/Textures/Default/Blanks/T_BlackMask.T_BlackMask")
          (Rendering|Material|SetVectorParameterValue _asmaterial_instance_dynamic "TopS_02_Base Color_Simple" "(R=0.010000,G=0.010000,B=0.010000,A=1.000000)" :!class /Script/Engine.MaterialInstanceDynamic)
          (Rendering|Material|SetScalarParameterValue _asmaterial_instance_dynamic "TopS_02_Base_Metalic_" 0.0 :!class /Script/Engine.MaterialInstanceDynamic)
          (Rendering|Material|SetScalarParameterValue _asmaterial_instance_dynamic "TopS_02_Base_Roughness_High" 0.0 :!class /Script/Engine.MaterialInstanceDynamic)
          (Rendering|Material|SetScalarParameterValue _asmaterial_instance_dynamic "TopS_02_Base_Spec_" 0.5 :!class /Script/Engine.MaterialInstanceDynamic))))
    (:1
      (Variables|Default|SetDecal_Texture "/BRCosmetics/Weapons/WeaponSkins/Wraps/Textures/WingBath/T_Wrap_WingBath_A1.T_Wrap_WingBath_A1"))
    (:2
      (Variables|Default|SetDecal_Texture "/BRCosmetics/Weapons/WeaponSkins/Wraps/Textures/WingBath/T_Wrap_WingBath_A2.T_Wrap_WingBath_A2"))
    (:3
      (Variables|Default|SetDecal_Texture "/BRCosmetics/Weapons/WeaponSkins/Wraps/Textures/WingBath/T_Wrap_WingBath_A3.T_Wrap_WingBath_A3"))
    (:4
      (Variables|Default|SetDecal_Texture "/BRCosmetics/Weapons/WeaponSkins/Wraps/Textures/WingBath/T_Wrap_WingBath_A4.T_Wrap_WingBath_A4"))
    (:5
      (Variables|Default|SetDecal_Texture "/BRCosmetics/Weapons/WeaponSkins/Wraps/Textures/WingBath/T_Wrap_WingBath_A5.T_Wrap_WingBath_A5"))
    (:6
      (Variables|Default|SetDecal_Texture "/BRCosmetics/Weapons/WeaponSkins/Wraps/Textures/WingBath/T_Wrap_WingBath_A6.T_Wrap_WingBath_A6"))
    (:7
      (Variables|Default|SetDecal_Texture "/BRCosmetics/Weapons/WeaponSkins/Wraps/Textures/WingBath/T_Wrap_WingBath_A7.T_Wrap_WingBath_A7"))
    (:8
      (Variables|Default|SetDecal_Texture "/BRCosmetics/Weapons/WeaponSkins/Wraps/Textures/WingBath/T_Wrap_WingBath_A8.T_Wrap_WingBath_A8")))
  (bind _returnvalue (Rendering|Material|GetVectorParameterValue _character_material "Decal_Color" :!class /Script/Engine.MaterialInstanceDynamic))
  (bind _returnvalue_2 (Rendering|Material|GetScalarParameterValue _character_material "Decal_MetallicBias" :!class /Script/Engine.MaterialInstanceDynamic))
  (bind _returnvalue_3 (Rendering|Material|GetScalarParameterValue _character_material "Decal_SpecularBias" :!class /Script/Engine.MaterialInstanceDynamic))
  (bind _returnvalue_4 (Rendering|Material|GetScalarParameterValue _character_material "Decal_RoughnessBias" :!class /Script/Engine.MaterialInstanceDynamic))
  (for _array_element_3 (CustomItemWrap|GetWrappedMeshComponents)
    (for _array_element_4 (Rendering|Material|GetMaterials _array_element_3)
      (bind _asmaterial_instance_dynamic_2 (Utilities|Casting|CastToMaterialInstanceDynamic _array_element_4))
      (Rendering|Material|SetTextureParameterValue _asmaterial_instance_dynamic_2 "TopS_02__Tex_Mask" (Variables|Default|GetDecal_Texture))
      (Rendering|Material|SetVectorParameterValue _asmaterial_instance_dynamic_2 "TopS_02_Base Color_Simple" (* _returnvalue "(R=0.400000,G=0.400000,B=0.400000,A=1.000000)") :!class /Script/Engine.MaterialInstanceDynamic)
      (Rendering|Material|SetScalarParameterValue _asmaterial_instance_dynamic_2 "TopS_02_Base_Metalic_" _returnvalue_2 :!class /Script/Engine.MaterialInstanceDynamic)
      (Rendering|Material|SetScalarParameterValue _asmaterial_instance_dynamic_2 "TopS_02_Base_Roughness_High" (+ _returnvalue_4 0.35) :!class /Script/Engine.MaterialInstanceDynamic)
      (Rendering|Material|SetScalarParameterValue _asmaterial_instance_dynamic_2 "TopS_02_Base_Spec_" (+ _returnvalue_3 0.35) :!class /Script/Engine.MaterialInstanceDynamic))))

(event Custom|Icon
  (bind _frame (Variables|Default|GetFrame))
  (bind _index (CosmeticLoadout|GetVariantChannelActiveIndexforSlotActor "(TagName=\"LoadoutSlot.Character\")" (CustomCharacter|GetAssociatedPlayerPawn) "(TagName=\"Cosmetics.Variant.Channel.Slot.6\")" (Utilities|Array|MakeArray "(TagName=\"Cosmetics.Variant.Property.Mat27\")" "(TagName=\"Cosmetics.Variant.Property.Mat16\")" "(TagName=\"Cosmetics.Variant.Property.Mat28\")" "(TagName=\"Cosmetics.Variant.Property.Mat29\")" "(TagName=\"Cosmetics.Variant.Property.Mat36\")" "(TagName=\"Cosmetics.Variant.Property.Mat37\")" "(TagName=\"Cosmetics.Variant.Property.Mat38\")" "(TagName=\"Cosmetics.Variant.Property.Mat39\")" "(TagName=\"Cosmetics.Variant.Property.Mat39\")") false false))
  (switch int (Utilities|Array|Get(acopy) (Utilities|Array|MakeArray 0 1 2 3 4 5 6 7 8) _index)
    (:0
      (for _array_element (CustomItemWrap|GetWrappedMeshComponents)
        (for _array_element_2 (Rendering|Material|GetMaterials _array_element)
          (bind _asmaterial_instance_dynamic (Utilities|Casting|CastToMaterialInstanceDynamic _array_element_2))
          (Rendering|Material|SetScalarParameterValue _asmaterial_instance_dynamic "EDRec_FlipBook_FrameA" 0.0 :!class /Script/Engine.MaterialInstanceDynamic)
          (Rendering|Material|SetScalarParameterValue _asmaterial_instance_dynamic "EDRec_FlipBook_FrameB" 0.0 :!class /Script/Engine.MaterialInstanceDynamic))))
    (:1
      (Variables|Default|SetFrame 1.0))
    (:2
      (Variables|Default|SetFrame 2.0))
    (:3
      (Variables|Default|SetFrame 3.0))
    (:4
      (Variables|Default|SetFrame 4.0))
    (:5
      (Variables|Default|SetFrame 5.0))
    (:6
      (Variables|Default|SetFrame 6.0))
    (:7
      (Variables|Default|SetFrame 7.0)))
  (bind _returnvalue (Rendering|Material|GetVectorParameterValue (Variables|Default|GetSigelMat) "Sigil_Color" :!class /Script/Engine.MaterialInstanceDynamic))
  (for _array_element_3 (CustomItemWrap|GetWrappedMeshComponents)
    (for _array_element_4 (Rendering|Material|GetMaterials _array_element_3)
      (bind _asmaterial_instance_dynamic_2 (Utilities|Casting|CastToMaterialInstanceDynamic _array_element_4))
      (Rendering|Material|SetVectorParameterValue _asmaterial_instance_dynamic_2 "EDRec_Emissive_A_Color-" (* _returnvalue "(R=0.450000,G=0.450000,B=0.450000,A=1.000000)") :!class /Script/Engine.MaterialInstanceDynamic)
      (Rendering|Material|SetScalarParameterValue _asmaterial_instance_dynamic_2 "EDRec_FlipBook_FrameA" _frame :!class /Script/Engine.MaterialInstanceDynamic)
      (Rendering|Material|SetScalarParameterValue _asmaterial_instance_dynamic_2 "EDRec_FlipBook_FrameB" _frame :!class /Script/Engine.MaterialInstanceDynamic))))

(fn ConstructionScript)
