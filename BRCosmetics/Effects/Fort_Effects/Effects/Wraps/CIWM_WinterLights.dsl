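; Winter Lights weapon wrap Blueprint that animates the wrap's materials each time the weapon fires.
; On fire, it writes game time to the WNL_Game_Time scalar on the wrapped mesh components, then waits briefly.
; A switch on a stored integer then turns on one of three material sets (A, B, C) and turns off the others.
; Each branch advances the integer, so successive shots cycle through the three sets.
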(event EventOnWeaponFired
  (bind _returnvalue (CustomItemWrap|GetWrappedMeshComponents))
  (bind _returnvalue_2 (CustomItemWrap|GetWrappedMeshComponents))
  (bind _returnvalue_3 (CustomItemWrap|GetWrappedMeshComponents))
  (Rendering|Material|SetScalarParameterValueonMaterials (CustomItemWrap|GetWrappedMeshComponents) "WNL_Game_Time" (Utilities|Time|GetGameTimeinSeconds))
  (Utilities|FlowControl|Delay 0.15)
  (switch int (Variables|Default|GetSet_Varable)
    (:0
      (Rendering|Material|SetScalarParameterValueonMaterials _returnvalue "Activate SetA" 0.0)
      (Rendering|Material|SetScalarParameterValueonMaterials _returnvalue "Activate SetB" 1.0)
      (Rendering|Material|SetScalarParameterValueonMaterials _returnvalue "Activate SetC" 0.0)
      (Variables|Default|SetSet_Varable 1))
    (:1
      (Rendering|Material|SetScalarParameterValueonMaterials _returnvalue_2 "Activate SetA" 0.0)
      (Rendering|Material|SetScalarParameterValueonMaterials _returnvalue_2 "Activate SetB" 0.0)
      (Rendering|Material|SetScalarParameterValueonMaterials _returnvalue_2 "Activate SetC" 1.0)
      (Variables|Default|SetSet_Varable 2))
    (:2
      (Rendering|Material|SetScalarParameterValueonMaterials _returnvalue_3 "Activate SetA" 1.0)
      (Rendering|Material|SetScalarParameterValueonMaterials _returnvalue_3 "Activate SetB" 0.0)
      (Rendering|Material|SetScalarParameterValueonMaterials _returnvalue_3 "Activate SetC" 0.0)
      (Variables|Default|SetSet_Varable 0))))

(fn ConstructionScript)
