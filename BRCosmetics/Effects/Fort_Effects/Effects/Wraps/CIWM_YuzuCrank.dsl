(event EventBeginPlay
  (Utilities|FlowControl|Delay 0.2)
  (if (CustomCosmeticModifier|IsinFrontEnd)
    (Timeline|Timeline PlayFromStart
      (:NewTrack_0
        (switch int (Variables|Default|GetFire%20seq)
          (:0
            (if (not (Variables|Default|GetPause_A))
              (Variables|Default|SetPause_A true)
              (Variables|Default|SetFire%20seq 1)
              ;[1 1 1] multi A
              (CallFunction|RandomSet "EDRec_Emissive_A_Brightness_Time_BP" "EDRec_DecalA_Offset_Weapon" "EDRec_DecalA_Rotation_Weapon" -0.38 0.2 -0.01 0.01 1.0 1.0 "EDRec_Emissive_A_Gradent_Time_BP")
              (CallFunction|RandomSet "0" "EDRec_DecalA_Offset_Pickaxe" "EDRec_DecalA_Rotation_Pickaxe" -0.3 0.15 0.2 -0.2 1.0 1.0)
              (CallFunction|RandomSet "0" "EDRec_DecalA_Offset_Backpack----A=Scale" "EDRec_DecalA_Rotation_Backpack" -0.15 0.15 -0.25 0.25 1.0 1.0)
              (CallFunction|RandomSet "0" "EDRec_DecalA_Offset_Character" "EDRec_DecalA_Rotation_Character" -0.15 0.15 -0.2 0.28 1.0 1.0)
              (CallFunction|RandomSet "0" "EDRec_DecalA_Offset_Vehicle" "EDRec_DecalA_Rotation_Vehicle" 0.15 0.4 -0.15 0.15 1.0 1.0)
              (CallFunction|RandomSet "0" "EDRec_DecalA_Offset_VehicleSide" "EDRec_DecalA_Rotation_Vehicle_Side" 0.3 0.42 -0.1 -0.13 1.0 1.0)
              (CallFunction|RandomSet "0" "EDRec_DecalA_Offset_Mech" "EDRec_DecalA_Rotation_Mech" -0.1 0.1 0.32 0.2 1.0 1.0)
              (Utilities|FlowControl|Delay 0.5)
              (Variables|Default|SetPause_A false)))
          (:1
            (if (not (Variables|Default|GetPause_B))
              (Variables|Default|SetPause_B true)
              (Variables|Default|SetFire%20seq 2)
              ;[1 1 1] multi B
              (CallFunction|RandomSet "EDRec_Emissive_B_Brightness_Time_BP" "EDRec_DecalB_Offset_Weapon" "EDRec_DecalB_Rotation_Weapon" -0.38 0.2 -0.01 0.01 1.0 1.0 "EDRec_Emissive_B_Gradent_Time_BP")
              (CallFunction|RandomSet "0" "EDRec_DecalB_Offset_Pickaxe" "EDRec_DecalB_Rotation_Pickaxe" -0.3 0.15 0.2 -0.2 1.0 1.0)
              (CallFunction|RandomSet "0" "EDRec_DecalB_Offset_Backpack----A=Scale" "EDRec_DecalB_Rotation_Backpack" -0.15 0.15 -0.25 0.25 1.0 1.0)
              (CallFunction|RandomSet "0" "EDRec_DecalB_Offset_Character" "EDRec_DecalB_Rotation_Character" -0.15 0.15 -0.2 0.28 1.0 1.0)
              (CallFunction|RandomSet "0" "EDRec_DecalB_Offset_Vehicle" "EDRec_DecalB_Rotation_Vehicle" 0.15 0.4 -0.15 0.15 1.0 1.0)
              (CallFunction|RandomSet "0" "EDRec_DecalB_Offset_VehicleSide" "EDRec_DecalB_Rotation_Vehicle_Side" 0.3 0.42 -0.1 -0.13 0.12 1.0)
              (CallFunction|RandomSet "0" "EDRec_DecalB_Offset_Mech" "EDRec_DecalB_Rotation_Mech" -0.1 0.1 0.32 0.2 1.0 1.0)
              (Utilities|FlowControl|Delay 0.5)
              (Variables|Default|SetPause_B false)))
          (:2
            (if (not (Variables|Default|GetPause_C))
              (Variables|Default|SetPause_C true)
              (Variables|Default|SetFire%20seq 3)
              ; C
              (CallFunction|RandomSet "EDRec_Emissive_C_Brightness_Time_BP" "EDRec_DecalC_Offset_Weapon" "EDRec_DecalC_Rotation_Weapon" -0.38 0.2 -0.01 0.01 1.0 1.0 "EDRec_Emissive_C_Gradent_Time_BP")
              (CallFunction|RandomSet "0" "EDRec_DecalC_Offset_Pickaxe" "EDRec_DecalC_Rotation_Pickaxe" -0.3 0.15 0.2 -0.2 1.0 1.0)
              (CallFunction|RandomSet "0" "EDRec_DecalC_Offset_Backpack----A=Scale" "EDRec_DecalC_Rotation_Backpack" -0.15 0.15 -0.25 0.25 1.0 1.0)
              (CallFunction|RandomSet "0" "EDRec_DecalC_Offset_Character" "EDRec_DecalC_Rotation_Character" -0.15 0.15 -0.2 0.28 1.0 1.0)
              (CallFunction|RandomSet "0" "EDRec_DecalC_Offset_Vehicle" "EDRec_DecalC_Rotation_Vehicle" 0.15 0.4 -0.15 0.15 1.0 1.0)
              (CallFunction|RandomSet "0" "EDRec_DecalC_Offset_VehicleSide" "EDRec_DecalC_Rotation_Vehicle_Side" 0.3 0.42 -0.1 -0.13 1.0 1.0)
              (CallFunction|RandomSet "0" "EDRec_DecalC_Offset_Mech" "EDRec_DecalC_Rotation_Mech" -0.1 0.1 0.32 0.2 1.0 1.0)
              (Utilities|FlowControl|Delay 0.5)
              (Variables|Default|SetPause_C false)))
          (:3
            (if (not (Variables|Default|GetPause_D))
              (Variables|Default|SetPause_D true)
              (Variables|Default|SetFire%20seq 0)
              ; C
              (CallFunction|RandomSet "EDRec_Emissive_D_Brightness_Time_BP" "EDRec_DecalD_Offset_Weapon" "EDRec_DecalD_Rotation_Weapon" -0.38 0.2 -0.01 0.01 1.0 1.0 "EDRec_Emissive_D_Gradent_Time_BP")
              (CallFunction|RandomSet "0" "EDRec_DecalD_Offset_Pickaxe" "EDRec_DecalD_Rotation_Pickaxe" -0.3 0.15 0.2 -0.2 1.0 1.0)
              (CallFunction|RandomSet "0" "EDRec_DecalD_Offset_Backpack----A=Scale" "EDRec_DecalD_Rotation_Backpack" -0.15 0.15 -0.25 0.25 1.0 1.0)
              (CallFunction|RandomSet "0" "EDRec_DecalD_Offset_Character" "EDRec_DecalD_Rotation_Character" -0.15 0.15 -0.2 0.28 1.0 1.0)
              (CallFunction|RandomSet "0" "EDRec_DecalD_Offset_Vehicle" "EDRec_DecalD_Rotation_Vehicle" 0.15 0.4 -0.15 0.15 1.0 1.0)
              (CallFunction|RandomSet "0" "EDRec_DecalD_Offset_VehicleSide" "EDRec_DecalD_Rotation_Vehicle_Side" 0.3 0.42 -0.1 -0.13 1.0 1.0)
              (CallFunction|RandomSet "0" "EDRec_DecalD_Offset_Mech" "EDRec_DecalD_Rotation_Mech" -0.1 0.1 0.32 0.2 1.0 1.0)
              (Utilities|FlowControl|Delay 0.5)
              (Variables|Default|SetPause_D false))))))))

(event EventOnWeaponFired
  (bind _returnvalue (CustomItemWrap|GetWrappedMeshComponents))
  (bind _returnvalue_2 (Utilities|Time|GetGameTimeinSeconds))
  (if (CustomCosmeticModifier|IsinFrontEnd)
    (Utilities|FlowControl|Delay 0.01)
    (bind _output_get (Variables|Default|SetShotsFired (+ (Variables|Default|GetShotsFired) 1)))
    (if (>= _output_get 5)
      (if (>= _output_get 6)
        (Variables|Default|SetShotsFired 0))
      (Rendering|Material|SetScalarParameterValueonMaterials _returnvalue "TopS Emissive_02-Brightness-TBP" _returnvalue_2)
      (Rendering|Material|SetScalarParameterValueonMaterials _returnvalue "TopS Emissive_02_Offset----TBP" _returnvalue_2))))

(event EventOnPawnSet
  (bind _returnvalue (CustomItemWrap|GetWrappedMeshComponents))
  (bind _returnvalue_2 (Utilities|Time|GetGameTimeinSeconds))
  (Utilities|FlowControl|Sequence
    (:then_0
      (Cosmetics|WatchCosmeticStat (CustomCharacter|GetAssociatedPlayerPawn) "(TagName=\"StatManager.DamageReceived\")" 0 false
        (:OnStatChanged
          (switch int (Variables|Default|GetFire%20seq)
            (:0
              (if (not (Variables|Default|GetPause_A))
                (Variables|Default|SetPause_A true)
                (Variables|Default|SetFire%20seq 1)
                ;[1 1 1] multi A
                (CallFunction|RandomSet "EDRec_Emissive_A_Brightness_Time_BP" "EDRec_DecalA_Offset_Weapon" "EDRec_DecalA_Rotation_Weapon" -0.38 0.2 -0.01 0.01 1.0 1.0 "EDRec_Emissive_A_Gradent_Time_BP")
                (CallFunction|RandomSet "0" "EDRec_DecalA_Offset_Pickaxe" "EDRec_DecalA_Rotation_Pickaxe" -0.3 0.15 0.2 -0.2 1.0 1.0)
                (CallFunction|RandomSet "0" "EDRec_DecalA_Offset_Backpack----A=Scale" "EDRec_DecalA_Rotation_Backpack" -0.15 0.15 -0.25 0.25 1.0 1.0)
                (CallFunction|RandomSet "0" "EDRec_DecalA_Offset_Character" "EDRec_DecalA_Rotation_Character" -0.15 0.15 -0.2 0.28 1.0 1.0)
                (CallFunction|RandomSet "0" "EDRec_DecalA_Offset_Vehicle" "EDRec_DecalA_Rotation_Vehicle" 0.15 0.4 -0.15 0.15 1.0 1.0)
                (CallFunction|RandomSet "0" "EDRec_DecalA_Offset_VehicleSide" "EDRec_DecalA_Rotation_Vehicle_Side" 0.3 0.42 -0.1 -0.13 1.0 1.0)
                (CallFunction|RandomSet "0" "EDRec_DecalA_Offset_Mech" "EDRec_DecalA_Rotation_Mech" -0.1 0.1 0.32 0.2 1.0 1.0)
                (Utilities|FlowControl|Delay 0.5)
                (Variables|Default|SetPause_A false)))
            (:1
              (if (not (Variables|Default|GetPause_B))
                (Variables|Default|SetPause_B true)
                (Variables|Default|SetFire%20seq 2)
                ;[1 1 1] multi B
                (CallFunction|RandomSet "EDRec_Emissive_B_Brightness_Time_BP" "EDRec_DecalB_Offset_Weapon" "EDRec_DecalB_Rotation_Weapon" -0.38 0.2 -0.01 0.01 1.0 1.0 "EDRec_Emissive_B_Gradent_Time_BP")
                (CallFunction|RandomSet "0" "EDRec_DecalB_Offset_Pickaxe" "EDRec_DecalB_Rotation_Pickaxe" -0.3 0.15 0.2 -0.2 1.0 1.0)
                (CallFunction|RandomSet "0" "EDRec_DecalB_Offset_Backpack----A=Scale" "EDRec_DecalB_Rotation_Backpack" -0.15 0.15 -0.25 0.25 1.0 1.0)
                (CallFunction|RandomSet "0" "EDRec_DecalB_Offset_Character" "EDRec_DecalB_Rotation_Character" -0.15 0.15 -0.2 0.28 1.0 1.0)
                (CallFunction|RandomSet "0" "EDRec_DecalB_Offset_Vehicle" "EDRec_DecalB_Rotation_Vehicle" 0.15 0.4 -0.15 0.15 1.0 1.0)
                (CallFunction|RandomSet "0" "EDRec_DecalB_Offset_VehicleSide" "EDRec_DecalB_Rotation_Vehicle_Side" 0.3 0.42 -0.1 -0.13 0.12 1.0)
                (CallFunction|RandomSet "0" "EDRec_DecalB_Offset_Mech" "EDRec_DecalB_Rotation_Mech" -0.1 0.1 0.32 0.2 1.0 1.0)
                (Utilities|FlowControl|Delay 0.5)
                (Variables|Default|SetPause_B false)))
            (:2
              (if (not (Variables|Default|GetPause_C))
                (Variables|Default|SetPause_C true)
                (Variables|Default|SetFire%20seq 3)
                ; C
                (CallFunction|RandomSet "EDRec_Emissive_C_Brightness_Time_BP" "EDRec_DecalC_Offset_Weapon" "EDRec_DecalC_Rotation_Weapon" -0.38 0.2 -0.01 0.01 1.0 1.0 "EDRec_Emissive_C_Gradent_Time_BP")
                (CallFunction|RandomSet "0" "EDRec_DecalC_Offset_Pickaxe" "EDRec_DecalC_Rotation_Pickaxe" -0.3 0.15 0.2 -0.2 1.0 1.0)
                (CallFunction|RandomSet "0" "EDRec_DecalC_Offset_Backpack----A=Scale" "EDRec_DecalC_Rotation_Backpack" -0.15 0.15 -0.25 0.25 1.0 1.0)
                (CallFunction|RandomSet "0" "EDRec_DecalC_Offset_Character" "EDRec_DecalC_Rotation_Character" -0.15 0.15 -0.2 0.28 1.0 1.0)
                (CallFunction|RandomSet "0" "EDRec_DecalC_Offset_Vehicle" "EDRec_DecalC_Rotation_Vehicle" 0.15 0.4 -0.15 0.15 1.0 1.0)
                (CallFunction|RandomSet "0" "EDRec_DecalC_Offset_VehicleSide" "EDRec_DecalC_Rotation_Vehicle_Side" 0.3 0.42 -0.1 -0.13 1.0 1.0)
                (CallFunction|RandomSet "0" "EDRec_DecalC_Offset_Mech" "EDRec_DecalC_Rotation_Mech" -0.1 0.1 0.32 0.2 1.0 1.0)
                (Utilities|FlowControl|Delay 0.5)
                (Variables|Default|SetPause_C false)))
            (:3
              (if (not (Variables|Default|GetPause_D))
                (Variables|Default|SetPause_D true)
                (Variables|Default|SetFire%20seq 0)
                ; C
                (CallFunction|RandomSet "EDRec_Emissive_D_Brightness_Time_BP" "EDRec_DecalD_Offset_Weapon" "EDRec_DecalD_Rotation_Weapon" -0.38 0.2 -0.01 0.01 1.0 1.0 "EDRec_Emissive_D_Gradent_Time_BP")
                (CallFunction|RandomSet "0" "EDRec_DecalD_Offset_Pickaxe" "EDRec_DecalD_Rotation_Pickaxe" -0.3 0.15 0.2 -0.2 1.0 1.0)
                (CallFunction|RandomSet "0" "EDRec_DecalD_Offset_Backpack----A=Scale" "EDRec_DecalD_Rotation_Backpack" -0.15 0.15 -0.25 0.25 1.0 1.0)
                (CallFunction|RandomSet "0" "EDRec_DecalD_Offset_Character" "EDRec_DecalD_Rotation_Character" -0.15 0.15 -0.2 0.28 1.0 1.0)
                (CallFunction|RandomSet "0" "EDRec_DecalD_Offset_Vehicle" "EDRec_DecalD_Rotation_Vehicle" 0.15 0.4 -0.15 0.15 1.0 1.0)
                (CallFunction|RandomSet "0" "EDRec_DecalD_Offset_VehicleSide" "EDRec_DecalD_Rotation_Vehicle_Side" 0.3 0.42 -0.1 -0.13 1.0 1.0)
                (CallFunction|RandomSet "0" "EDRec_DecalD_Offset_Mech" "EDRec_DecalD_Rotation_Mech" -0.1 0.1 0.32 0.2 1.0 1.0)
                (Utilities|FlowControl|Delay 0.5)
                (Variables|Default|SetPause_D false)))))))
    (:then_1
      (Cosmetics|WatchCosmeticStat (CustomCharacter|GetAssociatedPlayerPawn) "(TagName=\"StatManager.DamageDealt\")" 0 false
        (:OnStatChanged
          (Rendering|Material|SetScalarParameterValueonMaterials _returnvalue "TopS Emissive_02-Brightness-TBP" _returnvalue_2)
          (Rendering|Material|SetScalarParameterValueonMaterials _returnvalue "TopS Emissive_02_Offset----TBP" _returnvalue_2))))))

(fn RandomSet
  (params (TimeBP_Brightness :type Name) (Offset :type Name) (Rotation :type Name) (X=min :type Float%20%28double-precision%29) (X=Max :type Float%20%28double-precision%29) (Y=Min :type Float%20%28double-precision%29) (Y=Max :type Float%20%28double-precision%29) (Scale=Min :type Float%20%28double-precision%29) (Scale=Max :type Float%20%28double-precision%29) (TimeBP_Gradent :type Name))
  (bind _returnvalue (CustomItemWrap|GetWrappedMeshComponents))
  (bind _returnvalue_2 (Utilities|Time|GetGameTimeinSeconds))
  (bind _stream (Variables|Default|GetStream))
  (bind _returnvalue_3 (Math|Random|RandomFloatinRangefromStream _stream X=min X=Max))
  (Rendering|Material|SetScalarParameterValueonMaterials _returnvalue TimeBP_Brightness _returnvalue_2)
  (Rendering|Material|SetScalarParameterValueonMaterials _returnvalue TimeBP_Gradent _returnvalue_2)
  (Rendering|Material|SetColorParameterValueonMaterials _returnvalue Offset (Math|Color|MakeColor _returnvalue_3 (Math|Random|RandomFloatinRangefromStream _stream Y=Min Y=Max) 0.0 (+ 1.8 0)))
  (Rendering|Material|SetScalarParameterValueonMaterials _returnvalue Rotation (Math|Random|RandomFloat))
  (Variables|Default|SetPosition _returnvalue_3))

(fn ConstructionScript)
