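; ZenCrash weapon wrap modifier that drives emissive material flashes on the wrapped meshes from firing, damage and kills.
; Damage stat changes raise a step count that decays over idle time, lerping the C channel over the first half and D over the second.
; Kills jump to full glow with a brief pause, while clip size sets step count and pause time.
; In the front end, repeated shots simulate the damage and kill responses.
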(event EventOnWeaponFired
  (bind _maxsteps (Variables|Default|GetMaxSteps))
  (bind _returnvalue (/ _maxsteps 2))
  (bind _stepcount (Variables|Default|GetStepCount))
  (bind _returnvalue_2 (- _stepcount _returnvalue))
  (bind _returnvalue_3 (Math|Integer|Clamp(Integer) _returnvalue_2 0 _returnvalue))
  (bind _returnvalue_4 (Math|Conversions|ToFloat(Integer) _returnvalue_3))
  (bind _returnvalue_5 (Utilities|Time|GetGameTimeinSeconds))
  (bind _returnvalue_6 (CustomItemWrap|GetWrappedMeshComponents))
  (bind _returnvalue_7 (Utilities|Time|GetGameTimeinSeconds))
  (bind _returnvalue_8 (CustomItemWrap|GetWrappedMeshComponents))
  (bind _maxsteps_2 (Variables|Default|GetMaxSteps))
  (bind _returnvalue_9 (/ _maxsteps_2 2))
  (bind _stepcount_2 (Variables|Default|GetStepCount))
  (bind _returnvalue_10 (Math|Integer|Clamp(Integer) _stepcount_2 0 _returnvalue_9))
  (bind _returnvalue_11 (Math|Conversions|ToFloat(Integer) _returnvalue_10))
  (bind _returnvalue_12 (/ _maxsteps_2 2))
  (bind _returnvalue_13 (/ _returnvalue_11 _returnvalue_12))
  (bind _returnvalue_14 (/ _maxsteps 2))
  (bind _returnvalue_15 (/ _returnvalue_4 _returnvalue_14))
  (if (CustomCosmeticModifier|IsinFrontEnd)
    (bind _output_get (Variables|Default|SetIterationCount (+ (Variables|Default|GetIterationCount) 1)))
    (Rendering|Material|SetScalarParameterValueonMaterials (CustomItemWrap|GetWrappedMeshComponents) "EDRec_Emissive_A_Brightness_Time_BP" (Utilities|Time|GetGameTimeinSeconds))
    (if (>= _output_get 5)
      (if (== (Variables|Default|GetIterationCount) 7)
        (Variables|Default|SetIterationCount 0)
        (Variables|Default|SetPause true)
        ;[1 1 1] on kill
        (Rendering|Material|SetScalarParameterValueonMaterials _returnvalue_8 "EDRec_Emissive_B_Brightness_Time_BP" _returnvalue_7)
        (Rendering|Material|SetScalarParameterValueonMaterials _returnvalue_8 "EDRec_Emissive_C_Brightness_Time__Start" 1.5)
        (Rendering|Material|SetScalarParameterValueonMaterials _returnvalue_8 "EDRec_Emissive_C_Brightness_Time_BP" _returnvalue_7)
        (Rendering|Material|SetScalarParameterValueonMaterials _returnvalue_8 "EDRec_Emissive_C_Brightness_DarkenBase_Time_Start" 1.0)
        (Rendering|Material|SetScalarParameterValueonMaterials _returnvalue_8 "EDRec_Emissive_D_Brightness_Time__Start" 1.5)
        (Rendering|Material|SetScalarParameterValueonMaterials _returnvalue_8 "EDRec_Emissive_D_Brightness_Time_BP" _returnvalue_7)
        (Rendering|Material|SetScalarParameterValueonMaterials _returnvalue_8 "EDRec_Emissive_D_Brightness_DarkenBase_Time_Start" 1.0)
        (Variables|Default|SetStepCount (Variables|Default|GetMaxSteps))
        (Variables|Default|Setwait_timer (+ (Utilities|Time|GetGameTimeinSeconds) (Variables|Default|GetPauseTime)))
        (Utilities|FlowControl|Delay 0.5)
        (Variables|Default|SetPause false)
        (elif (Variables|Default|GetPause)
          _
          (else
            (Variables|Default|SetStepCount (Math|Integer|Min(Integer) (+ (Math|Float|Truncate (Math|Float|Lerp (Math|Conversions|ToFloat(Integer) (Variables|Default|GetStepCount)) 0.0 (Math|Float|Clamp(Float) (/ (- (Utilities|Time|GetGameTimeinSeconds) (Variables|Default|Getwait_timer)) 6.0)))) 1) (Variables|Default|GetMaxSteps)))
            (Utilities|FlowControl|Sequence
              (:then_1
                (Rendering|Material|SetScalarParameterValueonMaterials _returnvalue_6 "EDRec_Emissive_B_Brightness_Time_BP" _returnvalue_5)
                (Rendering|Material|SetScalarParameterValueonMaterials _returnvalue_6 "EDRec_Emissive_C_Brightness_Time__Start" (Math|Float|Lerp 0.0 1.25 _returnvalue_13))
                (Rendering|Material|SetScalarParameterValueonMaterials _returnvalue_6 "EDRec_Emissive_C_Brightness_Time_BP" _returnvalue_5)
                (Rendering|Material|SetScalarParameterValueonMaterials _returnvalue_6 "EDRec_Emissive_C_Brightness_DarkenBase_Time_Start" (Math|Float|Lerp 0.0 1.0 _returnvalue_13))
                (Rendering|Material|SetScalarParameterValueonMaterials _returnvalue_6 "EDRec_Emissive_D_Brightness_Time__Start" (Math|Float|Lerp 0.0 1.25 _returnvalue_15))
                (Rendering|Material|SetScalarParameterValueonMaterials _returnvalue_6 "EDRec_Emissive_D_Brightness_Time_BP" _returnvalue_5)
                (Rendering|Material|SetScalarParameterValueonMaterials _returnvalue_6 "EDRec_Emissive_D_Brightness_DarkenBase_Time_Start" (Math|Float|Lerp 0.0 1.0 _returnvalue_15))
                (Variables|Default|Setwait_timer (+ (Utilities|Time|GetGameTimeinSeconds) (Variables|Default|GetPauseTime)))))))))
    (else
      (Rendering|Material|SetScalarParameterValueonMaterials (CustomItemWrap|GetWrappedMeshComponents) "EDRec_Emissive_A_Brightness_Time_BP" (Utilities|Time|GetGameTimeinSeconds)))))

(event EventBeginPlay
  (Utilities|FlowControl|Delay 0.5)
  (if (CustomCosmeticModifier|IsinFrontEnd)
    (Variables|Default|SetMaxSteps 5)
    (Variables|Default|SetPauseTime 2.0)))

(event EventOnWeaponChanged
  (params (NewWeapon :type Fort%20Weapon%20Object%20Reference) (OldWeapon :type Fort%20Weapon%20Object%20Reference))
  (Utilities|FlowControl|Delay 0.2)
  (if (> (FortWeapon|GetBulletsPerClip (Variables|/Script/FortniteGame.FortPawn|GetCurrentWeapon (CustomCharacter|GetAssociatedPlayerPawn))) 6)
    (Variables|Default|SetMaxSteps 10)
    (Variables|Default|SetPauseTime 1.5)
    (else
      (Variables|Default|SetMaxSteps 5)
      (Variables|Default|SetPauseTime 2.0))))

(event EventOnNewWeapons
  (params (WeaponsAsItems :type Array%20of%20Fort%20Item%20Object%20References))
  (Utilities|FlowControl|Delay 0.2)
  (if (> (FortWeapon|GetBulletsPerClip (Variables|/Script/FortniteGame.FortPawn|GetCurrentWeapon (CustomCharacter|GetAssociatedPlayerPawn))) 6)
    (Variables|Default|SetMaxSteps 10)
    (Variables|Default|SetPauseTime 1.5)
    (else
      (Variables|Default|SetMaxSteps 5)
      (Variables|Default|SetPauseTime 2.0))))

(event EventOnNewPlayerWeaponType
  (params (NewType :type EFortItemType%20Enum :default "WorldItem") (OldType :type EFortItemType%20Enum :default "WorldItem"))
  (Utilities|FlowControl|Delay 0.2)
  (if (> (FortWeapon|GetBulletsPerClip (Variables|/Script/FortniteGame.FortPawn|GetCurrentWeapon (CustomCharacter|GetAssociatedPlayerPawn))) 6)
    (Variables|Default|SetMaxSteps 10)
    (Variables|Default|SetPauseTime 1.5)
    (else
      (Variables|Default|SetMaxSteps 5)
      (Variables|Default|SetPauseTime 2.0))))

(event EventOnPawnSet
  (bind _maxsteps (Variables|Default|GetMaxSteps))
  (bind _returnvalue (/ _maxsteps 2))
  (bind _stepcount (Variables|Default|GetStepCount))
  (bind _returnvalue_2 (- _stepcount _returnvalue))
  (bind _returnvalue_3 (Math|Integer|Clamp(Integer) _returnvalue_2 0 _returnvalue))
  (bind _returnvalue_4 (Math|Conversions|ToFloat(Integer) _returnvalue_3))
  (bind _returnvalue_5 (Utilities|Time|GetGameTimeinSeconds))
  (bind _returnvalue_6 (CustomItemWrap|GetWrappedMeshComponents))
  (bind _returnvalue_7 (Utilities|Time|GetGameTimeinSeconds))
  (bind _returnvalue_8 (CustomItemWrap|GetWrappedMeshComponents))
  (bind _maxsteps_2 (Variables|Default|GetMaxSteps))
  (bind _returnvalue_9 (/ _maxsteps_2 2))
  (bind _stepcount_2 (Variables|Default|GetStepCount))
  (bind _returnvalue_10 (Math|Integer|Clamp(Integer) _stepcount_2 0 _returnvalue_9))
  (bind _returnvalue_11 (Math|Conversions|ToFloat(Integer) _returnvalue_10))
  (bind _returnvalue_12 (/ _maxsteps_2 2))
  (bind _returnvalue_13 (/ _returnvalue_11 _returnvalue_12))
  (bind _returnvalue_14 (/ _maxsteps 2))
  (bind _returnvalue_15 (/ _returnvalue_4 _returnvalue_14))
  (Utilities|FlowControl|Sequence
    (:then_0
      (Cosmetics|WatchCosmeticStat (CustomCharacter|GetAssociatedPlayerPawn) "(TagName=\"StatManager.DamageDealt\")" 0 false
        (:OnStatChanged
          (if (not (Variables|Default|GetPause))
            (Variables|Default|SetStepCount (Math|Integer|Min(Integer) (+ (Math|Float|Truncate (Math|Float|Lerp (Math|Conversions|ToFloat(Integer) (Variables|Default|GetStepCount)) 0.0 (Math|Float|Clamp(Float) (/ (- (Utilities|Time|GetGameTimeinSeconds) (Variables|Default|Getwait_timer)) 6.0)))) 1) (Variables|Default|GetMaxSteps)))
            (Utilities|FlowControl|Sequence
              (:then_1
                (Rendering|Material|SetScalarParameterValueonMaterials _returnvalue_6 "EDRec_Emissive_B_Brightness_Time_BP" _returnvalue_5)
                (Rendering|Material|SetScalarParameterValueonMaterials _returnvalue_6 "EDRec_Emissive_C_Brightness_Time__Start" (Math|Float|Lerp 0.0 1.25 _returnvalue_13))
                (Rendering|Material|SetScalarParameterValueonMaterials _returnvalue_6 "EDRec_Emissive_C_Brightness_Time_BP" _returnvalue_5)
                (Rendering|Material|SetScalarParameterValueonMaterials _returnvalue_6 "EDRec_Emissive_C_Brightness_DarkenBase_Time_Start" (Math|Float|Lerp 0.0 1.0 _returnvalue_13))
                (Rendering|Material|SetScalarParameterValueonMaterials _returnvalue_6 "EDRec_Emissive_D_Brightness_Time__Start" (Math|Float|Lerp 0.0 1.25 _returnvalue_15))
                (Rendering|Material|SetScalarParameterValueonMaterials _returnvalue_6 "EDRec_Emissive_D_Brightness_Time_BP" _returnvalue_5)
                (Rendering|Material|SetScalarParameterValueonMaterials _returnvalue_6 "EDRec_Emissive_D_Brightness_DarkenBase_Time_Start" (Math|Float|Lerp 0.0 1.0 _returnvalue_15))
                (Variables|Default|Setwait_timer (+ (Utilities|Time|GetGameTimeinSeconds) (Variables|Default|GetPauseTime)))))))))
    (:then_1
      (Cosmetics|WatchCosmeticStat (CustomCharacter|GetAssociatedPlayerPawn) "(TagName=\"StatManager.AthenaKills\")" 0 false
        (:OnStatChanged
          (Variables|Default|SetPause true)
          ;[1 1 1] on kill
          (Rendering|Material|SetScalarParameterValueonMaterials _returnvalue_8 "EDRec_Emissive_B_Brightness_Time_BP" _returnvalue_7)
          (Rendering|Material|SetScalarParameterValueonMaterials _returnvalue_8 "EDRec_Emissive_C_Brightness_Time__Start" 1.5)
          (Rendering|Material|SetScalarParameterValueonMaterials _returnvalue_8 "EDRec_Emissive_C_Brightness_Time_BP" _returnvalue_7)
          (Rendering|Material|SetScalarParameterValueonMaterials _returnvalue_8 "EDRec_Emissive_C_Brightness_DarkenBase_Time_Start" 1.0)
          (Rendering|Material|SetScalarParameterValueonMaterials _returnvalue_8 "EDRec_Emissive_D_Brightness_Time__Start" 1.5)
          (Rendering|Material|SetScalarParameterValueonMaterials _returnvalue_8 "EDRec_Emissive_D_Brightness_Time_BP" _returnvalue_7)
          (Rendering|Material|SetScalarParameterValueonMaterials _returnvalue_8 "EDRec_Emissive_D_Brightness_DarkenBase_Time_Start" 1.0)
          (Variables|Default|SetStepCount (Variables|Default|GetMaxSteps))
          (Variables|Default|Setwait_timer (+ (Utilities|Time|GetGameTimeinSeconds) (Variables|Default|GetPauseTime)))
          (Utilities|FlowControl|Delay 0.5)
          (Variables|Default|SetPause false))))
    (:then_2
      (if (> (FortWeapon|GetBulletsPerClip (Variables|/Script/FortniteGame.FortPawn|GetCurrentWeapon (CustomCharacter|GetAssociatedPlayerPawn))) 6)
        (Variables|Default|SetMaxSteps 10)
        (Variables|Default|SetPauseTime 1.5)
        (else
          (Variables|Default|SetMaxSteps 5)
          (Variables|Default|SetPauseTime 2.0))))))

(fn ConstructionScript)
