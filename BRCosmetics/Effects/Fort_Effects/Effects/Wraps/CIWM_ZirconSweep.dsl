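; Cosmetic item wrap modifier (ZirconSweep) that advances an emissive effect on wrapped meshes with the player's kill count.
; A watch on the AthenaKills stat stores count and kill time, then pushes clamped count-based start/end offsets and time-stamped animation parameters to materials.
; Weapon changes re-push after a short delay, as does resurrection.
; In the front end, a timeline steps the count through preview stages.
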(event EventBeginPlay
  (bind _returnvalue (Utilities|Time|GetGameTimeinSeconds))
  (bind _returnvalue_2 (CustomItemWrap|GetWrappedMeshComponents))
  (bind _kill_count (Variables|Default|GetKill_Count))
  (if (CustomCosmeticModifier|IsinFrontEnd)
    (Timeline|Timeline PlayFromStart
      (:Stage1
        (Variables|Default|SetKill_Count 0))
      (:Stage2
        (Variables|Default|SetKill_Count 1))
      (:Stage3
        (Variables|Default|SetKill_Count 2))
      (:Stage4
        (Variables|Default|SetKill_Count 3))
      (:stage5
        (Variables|Default|SetKill_Count 4))
      (:stage6
        (Variables|Default|SetKill_Count 5)))
    (Rendering|Material|SetScalarParameterValueonMaterials _returnvalue_2 "TopS Emissive_01_Offset---End" (/ (Math|Conversions|ToFloat(Integer) (Math|Integer|Clamp(Integer) _kill_count 0 6)) 4.0))
    (Rendering|Material|SetScalarParameterValueonMaterials _returnvalue_2 "TopS Emissive_01_Offset---Start" (/ (Math|Conversions|ToFloat(Integer) (+ (Math|Integer|Clamp(Integer) _kill_count 0 6) -1)) 4.0))
    (Rendering|Material|SetScalarParameterValueonMaterials _returnvalue_2 "TopS Emissive_01_Offset----TBP" _returnvalue)
    (Rendering|Material|SetScalarParameterValueonMaterials _returnvalue_2 "TopS Emissive_01-Brightness-TBP" _returnvalue)))

(event EventOnWeaponFired)

(event EventOnPawnSet
  (bind _returnvalue (Utilities|Time|GetGameTimeinSeconds))
  (bind _returnvalue_2 (CustomItemWrap|GetWrappedMeshComponents))
  (bind _kill_count (Variables|Default|GetKill_Count))
  (Utilities|FlowControl|Sequence
    (:then_0
      (if (CustomCosmeticModifier|IsinFrontEnd)
        (Timeline|Timeline PlayFromStart)))
    (:then_1
      (bind (_statname _statvalue) (Cosmetics|WatchCosmeticStat (CustomCharacter|GetAssociatedPlayerPawn) "(TagName=\"StatManager.AthenaKills\")")
        (:OnStatChanged
          (Variables|Default|SetKill_Count _statvalue)
          (Variables|Default|SetLastKillTime (Utilities|Time|GetGameTimeinSeconds))
          (CallFunction|PushMaterialParameters))))))

(event EventOnPostResIn
  (CallFunction|PushMaterialParameters))

(event EventOnNewWeapons
  (params (WeaponsAsItems :type Array%20of%20Fort%20Item%20Object%20References))
  (Utilities|FlowControl|Delay 0.2)
  ;[1 1 1] Updated kill count on new weapon
  (CallFunction|PushMaterialParameters))

(event EventOnNewPlayerWeaponType
  (params (NewType :type EFortItemType%20Enum :default "WorldItem") (OldType :type EFortItemType%20Enum :default "WorldItem"))
  (Utilities|FlowControl|Delay 0.2)
  ;[1 1 1] Updated kill count on new weapon
  (CallFunction|PushMaterialParameters))

(event Custom|Push%20Material%20Parameters
  (bind _returnvalue (CustomItemWrap|GetWrappedMeshComponents))
  (bind _kill_count (Variables|Default|GetKill_Count))
  (bind _lastkilltime (Variables|Default|GetLastKillTime))
  (Rendering|Material|SetScalarParameterValueonMaterials _returnvalue "TopS Emissive_01_Offset---End" (/ (Math|Conversions|ToFloat(Integer) (Math|Integer|Clamp(Integer) _kill_count 0 6)) 4.0))
  (Rendering|Material|SetScalarParameterValueonMaterials _returnvalue "TopS Emissive_01_Offset---Start" (/ (Math|Conversions|ToFloat(Integer) (+ (Math|Integer|Clamp(Integer) _kill_count 0 6) -1)) 4.0))
  (Rendering|Material|SetScalarParameterValueonMaterials _returnvalue "TopS Emissive_01_Offset----TBP" _lastkilltime)
  (Rendering|Material|SetScalarParameterValueonMaterials _returnvalue "TopS Emissive_01-Brightness-TBP" _lastkilltime))

(fn ConstructionScript)
