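; Reactive weapon wrap (nobleCrusher) whose material offsets and emissive respond to damage the player deals.
; Watched DamageDealt changes feed a capped, time-decayed total that sets offsets, then RestPossition writes start, end, timestamp and duration parameters so materials animate back to rest without Tick.
; In the front end, weapon fire simulates damage and triggers the kill burst (Ouch).
; AthenaKills is watched but has no handler here.
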(event EventOnWeaponFired
  ; Frontnd reactive
  (if (CustomCosmeticModifier|IsinFrontEnd)
    (if (CustomCosmeticModifier|IsinFrontEnd)
      (bind _output_get (Variables|Default|Setweapon%20fire (+ (Variables|Default|Getweapon%20fire) 1)))
      (if (>= _output_get 5)
        (if (== (Variables|Default|Getweapon%20fire) 7)
          (Variables|Default|Setweapon%20fire 0)
          (if (not (Variables|Default|Getkill))
            (CallFunction|Ouch))
          (elif (Variables|Default|Getkill)
            _
            (else
              (Variables|Default|SetDamageAddedUp (+ (Variables|Default|GetDamageAddedUp) 5e+01))
              (Variables|Default|SetLastKillTime (Utilities|Time|GetGameTimeinSeconds))
              (CallFunction|WrapOffset))))))))

(event EventOnPawnSet
  (bind _damageaddedup (Variables|Default|GetDamageAddedUp))
  (Utilities|FlowControl|Delay 0.05)
  (Utilities|FlowControl|Sequence
    (:then_0
      (Utilities|FlowControl|Delay 0.05)
      (Variables|Default|SetLastKillTime -999999999999.0))
    (:then_1
      (bind (_statname _statvalue) (Cosmetics|WatchCosmeticStat (CustomCharacter|GetAssociatedPlayerPawn) "(TagName=\"StatManager.DamageDealt\")" 0 false -1.0 -1.0 false)
        (:OnStatChanged
          (Utilities|FlowControl|Delay 0.05)
          (if (not (Variables|Default|Getkill))
            (Variables|Default|SetDamageAddedUp (Math|Float|Lerp (Variables|Default|GetDamageAddedUp) 0.0 (Math|Float|Clamp(Float) (/ (- (Utilities|Time|GetGameTimeinSeconds) (Variables|Default|Getwait_timer)) (Variables|Default|GetDuration_reset)))))
            (bind _output_get (Variables|Default|SetDamage_Current (- _statvalue (Variables|Default|GetDamage_Last))))
            (Variables|Default|SetDamage_Last _statvalue)
            (Variables|Default|SetDamageAddedUp (Math|Float|Clamp(Float) (+ (Math|Conversions|ToFloat(Integer) _output_get) _damageaddedup) 0.0 4e+02))
            (Variables|Default|SetDamage%20Ref (Math|Float|Clamp(Float) (+ (Math|Conversions|ToFloat(Integer) _output_get) _damageaddedup) 0.0 4e+02))
            (Variables|Default|SetLastKillTime (Utilities|Time|GetGameTimeinSeconds))
            (CallFunction|WrapOffset)))))
    (:then_2
      (Cosmetics|WatchCosmeticStat (CustomCharacter|GetAssociatedPlayerPawn) "(TagName=\"StatManager.AthenaKills\")" 0 false -1.0 -1.0 false))))

(event EventOnSetWrapTarget
  (bind _damageaddedup (Variables|Default|GetDamageAddedUp))
  (Utilities|FlowControl|Delay 0.05)
  (Utilities|FlowControl|Sequence
    (:then_0
      (Utilities|FlowControl|Delay 0.05)
      (Variables|Default|SetLastKillTime -999999999999.0))
    (:then_1
      (bind (_statname _statvalue) (Cosmetics|WatchCosmeticStat (CustomCharacter|GetAssociatedPlayerPawn) "(TagName=\"StatManager.DamageDealt\")" 0 false -1.0 -1.0 false)
        (:OnStatChanged
          (Utilities|FlowControl|Delay 0.05)
          (if (not (Variables|Default|Getkill))
            (Variables|Default|SetDamageAddedUp (Math|Float|Lerp (Variables|Default|GetDamageAddedUp) 0.0 (Math|Float|Clamp(Float) (/ (- (Utilities|Time|GetGameTimeinSeconds) (Variables|Default|Getwait_timer)) (Variables|Default|GetDuration_reset)))))
            (bind _output_get (Variables|Default|SetDamage_Current (- _statvalue (Variables|Default|GetDamage_Last))))
            (Variables|Default|SetDamage_Last _statvalue)
            (Variables|Default|SetDamageAddedUp (Math|Float|Clamp(Float) (+ (Math|Conversions|ToFloat(Integer) _output_get) _damageaddedup) 0.0 4e+02))
            (Variables|Default|SetDamage%20Ref (Math|Float|Clamp(Float) (+ (Math|Conversions|ToFloat(Integer) _output_get) _damageaddedup) 0.0 4e+02))
            (Variables|Default|SetLastKillTime (Utilities|Time|GetGameTimeinSeconds))
            (CallFunction|WrapOffset)))))
    (:then_2
      (Cosmetics|WatchCosmeticStat (CustomCharacter|GetAssociatedPlayerPawn) "(TagName=\"StatManager.AthenaKills\")" 0 false -1.0 -1.0 false))))

(event EventOnCharacterCustomizationFinished
  (params (CharacterPartInfo :type Array%20of%20Applied%20Character%20Part%20Info%20Structures))
  (bind _damageaddedup (Variables|Default|GetDamageAddedUp))
  (Utilities|FlowControl|Delay 0.05)
  (Utilities|FlowControl|Sequence
    (:then_0
      (Utilities|FlowControl|Delay 0.05)
      (Variables|Default|SetLastKillTime -999999999999.0))
    (:then_1
      (bind (_statname _statvalue) (Cosmetics|WatchCosmeticStat (CustomCharacter|GetAssociatedPlayerPawn) "(TagName=\"StatManager.DamageDealt\")" 0 false -1.0 -1.0 false)
        (:OnStatChanged
          (Utilities|FlowControl|Delay 0.05)
          (if (not (Variables|Default|Getkill))
            (Variables|Default|SetDamageAddedUp (Math|Float|Lerp (Variables|Default|GetDamageAddedUp) 0.0 (Math|Float|Clamp(Float) (/ (- (Utilities|Time|GetGameTimeinSeconds) (Variables|Default|Getwait_timer)) (Variables|Default|GetDuration_reset)))))
            (bind _output_get (Variables|Default|SetDamage_Current (- _statvalue (Variables|Default|GetDamage_Last))))
            (Variables|Default|SetDamage_Last _statvalue)
            (Variables|Default|SetDamageAddedUp (Math|Float|Clamp(Float) (+ (Math|Conversions|ToFloat(Integer) _output_get) _damageaddedup) 0.0 4e+02))
            (Variables|Default|SetDamage%20Ref (Math|Float|Clamp(Float) (+ (Math|Conversions|ToFloat(Integer) _output_get) _damageaddedup) 0.0 4e+02))
            (Variables|Default|SetLastKillTime (Utilities|Time|GetGameTimeinSeconds))
            (CallFunction|WrapOffset)))))
    (:then_2
      (Cosmetics|WatchCosmeticStat (CustomCharacter|GetAssociatedPlayerPawn) "(TagName=\"StatManager.AthenaKills\")" 0 false -1.0 -1.0 false))))

(event Custom|Ouch
  (Variables|Default|Setkill true)
  (Variables|Default|SetDamageAddedUp 4e+02)
  (CallFunction|WrapOffset)
  (bind _output_get (Variables|Default|SetDuration_reset 2.0))
  (CallFunction|RestPossition 0.1 -0.6 _output_get 0.0 -1.0 0.2 -0.4)
  (Variables|Default|Setwait_timer (+ (Utilities|Time|GetGameTimeinSeconds) (Variables|Default|GetPauseTime)))
  (Utilities|FlowControl|Delay 2.0)
  (Variables|Default|SetDamageAddedUp 0.0)
  (Variables|Default|Setkill false))

(event Custom|wrap_offset
  (bind _damageaddedup (Variables|Default|GetDamageAddedUp))
  (bind _returnvalue (/ _damageaddedup 4e+02))
  (bind _returnvalue_2 (Math|Float|Clamp(Float) _returnvalue))
  (bind _returnvalue_3 (Math|Float|Lerp 0.1 -0.6 _returnvalue_2))
  (bind _returnvalue_4 (Math|Float|Lerp 0.2 -0.4 _returnvalue_2))
  (bind _returnvalue_5 (Utilities|Time|GetGameTimeinSeconds))
  (bind _damageaddedup_2 (Variables|Default|GetDamageAddedUp))
  (bind _returnvalue_6 (/ _damageaddedup_2 4e+02))
  (bind _returnvalue_7 (Math|Float|Clamp(Float) _returnvalue_6))
  (bind _returnvalue_8 (CustomItemWrap|GetWrappedMeshComponents))
  (bind _returnvalue_9 (CustomItemWrap|GetWrappedMeshComponents))
  (bind _returnvalue_10 (Math|Float|Lerp 0.5 2e+01 _returnvalue_7))
  (Rendering|Material|SetScalarParameterValueonMaterials _returnvalue_9 "TopS_01_Blend_Offset---End" _returnvalue_3)
  (Rendering|Material|SetScalarParameterValueonMaterials _returnvalue_9 "TopS Emissive_01_Offset---End" _returnvalue_4)
  (Rendering|Material|SetScalarParameterValueonMaterials _returnvalue_9 "TopS Emissive_03_Offset---End" _returnvalue_4)
  (Variables|Default|SetCurrent%20Position _returnvalue_3)
  (Variables|Default|SetEmissive%20%20Position _returnvalue_4)
  (Rendering|Material|SetScalarParameterValueonMaterials _returnvalue_8 "TopS Emissive_01-Brightness-TBP" _returnvalue_5)
  (Rendering|Material|SetScalarParameterValueonMaterials _returnvalue_8 "TopS Emissive_02-Brightness-TBP" _returnvalue_5)
  (Rendering|Material|SetScalarParameterValueonMaterials _returnvalue_8 "TopS Emissive_03-Brightness-TBP" _returnvalue_5)
  (if (not (Variables|Default|Getkill))
    (Variables|Default|Setwait_timer (+ (Utilities|Time|GetGameTimeinSeconds) (Variables|Default|GetPauseTime)))
    (Variables|Default|SetDuration_reset _returnvalue_10)
    (CallFunction|RestPossition 0.1 (Variables|Default|GetCurrent%20Position) (Variables|Default|GetDuration_reset) 2.0 (Math|Float|Lerp -0.5 -6.0 _returnvalue_7) (Variables|Default|GetEmissive%20%20Position) 0.2)
    (Development|PrintText (Utilities|Text|ToText(Float) _returnvalue_10) true true "(R=1.000000,G=0.000000,B=0.000000,A=1.000000)")))

(fn RestPossition
  (params (End :type Float%20%28single-precision%29) (Start :type Float%20%28single-precision%29) (Time :type Float%20%28single-precision%29) (delay :type Float%20%28double-precision%29) (Offset%20Speed :type Float%20%28double-precision%29) (Emissive%20Start :type Float%20%28double-precision%29) (Emissive%20End :type Float%20%28double-precision%29))
  (bind _returnvalue (Utilities|Time|GetGameTimeinSeconds))
  (bind _returnvalue_2 (CustomItemWrap|GetWrappedMeshComponents))
  (bind _returnvalue_3 (CustomItemWrap|GetWrappedMeshComponents))
  (bind _returnvalue_4 (CustomItemWrap|GetWrappedMeshComponents))
  (bind _returnvalue_5 (+ _returnvalue delay))
  (Rendering|Material|SetScalarParameterValueonMaterials _returnvalue_4 "TopS_01_Blend_Offset---End" End)
  (Rendering|Material|SetScalarParameterValueonMaterials _returnvalue_4 "TopS_01_Blend_Offset---Start" Start)
  (Rendering|Material|SetScalarParameterValueonMaterials _returnvalue_4 "TopS_01_Blend_Offset-TBP" _returnvalue_5)
  (Rendering|Material|SetScalarParameterValueonMaterials _returnvalue_4 "TopS_01_Blend_Offset---Duration" Time)
  (Rendering|Material|SetScalarParameterValueonMaterials _returnvalue_2 "TopS Emissive_01_Offset---End" Emissive%20End)
  (Rendering|Material|SetScalarParameterValueonMaterials _returnvalue_2 "TopS Emissive_01_Offset---Start" Emissive%20Start)
  (Rendering|Material|SetScalarParameterValueonMaterials _returnvalue_2 "TopS Emissive_01_Offset----TBP" _returnvalue_5)
  (Rendering|Material|SetScalarParameterValueonMaterials _returnvalue_2 "TopS Emissive_01_Offset---Duration" Time)
  (Rendering|Material|SetScalarParameterValueonMaterials _returnvalue_3 "TopS Emissive_03_Offset---End" End)
  (Rendering|Material|SetScalarParameterValueonMaterials _returnvalue_3 "TopS Emissive_03_Offset---Start" Start)
  (Rendering|Material|SetScalarParameterValueonMaterials _returnvalue_3 "TopS Emissive_03_Offset----TBP" _returnvalue_5)
  (Rendering|Material|SetScalarParameterValueonMaterials _returnvalue_3 "TopS Emissive_03_Offset---Duration" Time))

(fn ConstructionScript)
